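# Level Sequence for animating the InchScape_M custom skin's menu animation on a FortPlayerMannequin, over roughly 306 frames.
# The mannequin has spawn and transform tracks, a skeletal animation section, and Control Rig tracks for the Fortnite biped and legacy face rigs.
# Biped controls carry dense per-frame keys on both FK and IK leg controls, which points to baked animation, with FK/IK switches held constant.
# The visible face rig channels hold only defaults.
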LevelSequence /BRCosmetics/Animation/Game/MainPlayer/Menu/BR/CustomSkins/_s39/InchScape_M/IE_anim/IE_LS
bindings [FortPlayerMannequin {
  id=#1
  possessed_class=FortPlayerMannequin
  tracks [
   {type=MovieSceneSpawnTrack, sections [{type=MovieSceneSpawnSection, channels [None {type=MovieSceneScriptingBoolChannel, default=true}]}]}
   {
    type=MovieScene3DTransformTrack
    sections [{
      type=MovieScene3DTransformSection
      channels [
       Location.X {type=MovieSceneScriptingDoubleChannel, default=0}
       Location.Y {type=MovieSceneScriptingDoubleChannel, default=0}
       Location.Z {type=MovieSceneScriptingDoubleChannel, default=0}
       Rotation.X {type=MovieSceneScriptingDoubleChannel, default=0}
       Rotation.Y {type=MovieSceneScriptingDoubleChannel, default=0}
       Rotation.Z {type=MovieSceneScriptingDoubleChannel, default=0}
       Scale.X {type=MovieSceneScriptingDoubleChannel, default=1}
       Scale.Y {type=MovieSceneScriptingDoubleChannel, default=1}
       Scale.Z {type=MovieSceneScriptingDoubleChannel, default=1}
      ]
     }]
   }
   {type=MovieSceneSkeletalAnimationTrack, sections []}
   {type=MovieSceneSkeletalAnimationTrack, sections [{type=MovieSceneSkeletalAnimationSection, end_frame=306, start_frame=0}]}
   {
    type=MovieSceneControlRigParameterTrack
    display_name=Fortnite_Legacy_Face_CtrlRig
    sections [{
      type=MovieSceneControlRigParameterSection
      channels [
       Flipbook {type=MovieSceneScriptingBoolChannel, default=false}
       Face_Ctrl_Master.Location.X {type=MovieSceneScriptingFloatChannel, default=-40}
       Face_Ctrl_Master.Location.Y {type=MovieSceneScriptingFloatChannel, default=0}
       Face_Ctrl_Master.Location.Z {type=MovieSceneScriptingFloatChannel, default=0}
       Face_Ctrl_Master.Rotation.X {type=MovieSceneScriptingFloatChannel, default=0}
       Face_Ctrl_Master.Rotation.Y {type=MovieSceneScriptingFloatChannel, default=0}
       Face_Ctrl_Master.Rotation.Z {type=MovieSceneScriptingFloatChannel, default=0}
       Face_Ctrl_Master.Scale.X {type=MovieSceneScriptingFloatChannel, default=1}
       Face_Ctrl_Master.Scale.Y {type=MovieSceneScriptingFloatChannel, default=1}
       Face_Ctrl_Master.Scale.Z {type=MovieSceneScriptingFloatChannel, default=1}
       gui_follow_head {type=MovieSceneScriptingFloatChannel, default=1}
      ]
     }]
   }
   {
    type=MovieSceneControlRigParameterTrack
    display_name=Fortnite_Biped_CtrlRig
    sections [{
      type=MovieSceneControlRigParameterSection
      channels [
       leg_l_stretch_switch {type=MovieSceneScriptingBoolChannel, default=false, const [false,307]}
       leg_r_stretch_switch {type=MovieSceneScriptingBoolChannel, default=false, const [false,307]}
       arm_l_fk_local {type=MovieSceneScriptingBoolChannel, default=true, const [true,307]}
       Auto Interpolate_3 {type=MovieSceneScriptingBoolChannel, default=false, const [false,307]}
       Auto Interpolate {type=MovieSceneScriptingBoolChannel, default=false, const [false,307]}
       arm_l_stretch_switch {type=MovieSceneScriptingBoolChannel, default=false, const [false,307]}
       Keep Aligned {type=MovieSceneScriptingBoolChannel, default=false, const [false,307]}
       arm_r_fk_local {type=MovieSceneScriptingBoolChannel, default=true, const [true,307]}
       Auto Interpolate_4 {type=MovieSceneScriptingBoolChannel, default=false, const [false,307]}
       Auto Interpolate_2 {type=MovieSceneScriptingBoolChannel, default=false, const [false,307]}
       arm_r_stretch_switch {type=MovieSceneScriptingBoolChannel, default=false, const [false,307]}
       Keep Aligned_2 {type=MovieSceneScriptingBoolChannel, default=false, const [false,307]}
       head_local {type=MovieSceneScriptingBoolChannel, default=true, const [true,307]}
       Stretch {type=MovieSceneScriptingBoolChannel, default=false, const [false,307]}
       head_stretch {type=MovieSceneScriptingBoolChannel, default=false, const [false,307]}
       spine_stretch_switch {type=MovieSceneScriptingBoolChannel, default=false, const [false,307]}
       arm_l_fk_ik_switch {type=MovieSceneScriptingBoolChannel, default=false, const [false,307]}
       leg_l_fk_ik_switch {type=MovieSceneScriptingBoolChannel, default=true, const [true,307]}
       arm_r_fk_ik_switch {type=MovieSceneScriptingBoolChannel, default=false, const [false,307]}
       leg_r_fk_ik_switch {type=MovieSceneScriptingBoolChannel, default=true, const [true,307]}
       ShowBodyControls {type=MovieSceneScriptingBoolChannel, default=true, const [true,307]}
       spine_fk_ik_switch {type=MovieSceneScriptingBoolChannel, default=false, const [false,307]}
       neck_fk_ik_switch {type=MovieSceneScriptingBoolChannel, default=false, const [false,307]}
       space {type=MovieSceneScriptingByteChannel, default=1, const [1,307]}
       global_ctrl.Location.X {type=MovieSceneScriptingFloatChannel, default=0, const [0,307]}
       global_ctrl.Location.Y {type=MovieSceneScriptingFloatChannel, default=0, const [0,307]}
       global_ctrl.Location.Z {type=MovieSceneScriptingFloatChannel, default=0, const [0,307]}
       global_ctrl.Rotation.X {type=MovieSceneScriptingFloatChannel, default=0, const [0,307]}
       global_ctrl.Rotation.Y {type=MovieSceneScriptingFloatChannel, default=0, const [0,307]}
       global_ctrl.Rotation.Z {type=MovieSceneScriptingFloatChannel, default=0, const [0,307]}
       root_ctrl.Location.X {type=MovieSceneScriptingFloatChannel, default=0, const [0,307]}
       root_ctrl.Location.Y {type=MovieSceneScriptingFloatChannel, default=0, const [0,307]}
       root_ctrl.Location.Z {type=MovieSceneScriptingFloatChannel, default=0, const [0,307]}
       root_ctrl.Rotation.X {type=MovieSceneScriptingFloatChannel, default=0, const [0,307]}
       root_ctrl.Rotation.Y {type=MovieSceneScriptingFloatChannel, default=0, const [0,307]}
       root_ctrl.Rotation.Z {type=MovieSceneScriptingFloatChannel, default=0, const [0,307]}
       root_ctrl.Scale.X {type=MovieSceneScriptingFloatChannel, default=1, const [1,307]}
       root_ctrl.Scale.Y {type=MovieSceneScriptingFloatChannel, default=1, const [1,307]}
       root_ctrl.Scale.Z {type=MovieSceneScriptingFloatChannel, default=1, const [1,307]}
       body_offset_ctrl.Location.X {type=MovieSceneScriptingFloatChannel, default=0, const [0,307]}
       body_offset_ctrl.Location.Y {type=MovieSceneScriptingFloatChannel, default=0, const [0,307]}
       body_offset_ctrl.Location.Z {type=MovieSceneScriptingFloatChannel, default=0, const [0,307]}
       body_offset_ctrl.Rotation.X {type=MovieSceneScriptingFloatChannel, default=0, const [0,307]}
       body_offset_ctrl.Rotation.Y {type=MovieSceneScriptingFloatChannel, default=0, const [0,307]}
       body_offset_ctrl.Rotation.Z {type=MovieSceneScriptingFloatChannel, default=0, const [0,307]}
       body_offset_ctrl.Scale.X {type=MovieSceneScriptingFloatChannel, default=1, const [1,307]}
       body_offset_ctrl.Scale.Y {type=MovieSceneScriptingFloatChannel, default=1, const [1,307]}
       body_offset_ctrl.Scale.Z {type=MovieSceneScriptingFloatChannel, default=1, const [1,307]}
       thigh_l_fk_ctrl.Location.X {type=MovieSceneScriptingFloatChannel, default=0, thinned_from=307, keys [[0,0],[306,0]]}
       thigh_l_fk_ctrl.Location.Y {type=MovieSceneScriptingFloatChannel, default=0, thinned_from=307, keys [[0,0],[306,0]]}
       thigh_l_fk_ctrl.Location.Z {type=MovieSceneScriptingFloatChannel, default=0, thinned_from=307, keys [[0,0],[306,0]]}
       thigh_l_fk_ctrl.Rotation.X {
        type=MovieSceneScriptingFloatChannel
        default=1.216
        thinned_from=307
        keys [[0,28.7088],[6,28.657],[8,28.7587],[11,29.1417],[13,29.5519],[14,29.9013],[15,30.4328],[16,31.1284],[17,32.0178],[18,33.1593],[20,35.875],[22,38.3228],[23,39.3396],[24,40.0912],[26,41.2614],[28,42.1839],[31,43.2753],[33,43.8139],[37,44.3433],[49,44.5617],[59,43.8937],[63,43.3335],[65,42.8217],[66,42.4449],[67,41.9643],[68,41.2548],[69,40.2557],[70,39.0763],[71,37.676],[74,33.1777],[75,31.8176],[76,30.8349],[77,30.1088],[78,29.5546],[79,29.1812],[80,29.0111],[108,27.1414],[145,27.4552],[148,27.689],[156,28.8283],[160,29.1523],[165,29.2258],[168,29.0643],[170,28.795],[172,28.3888],[174,27.8167],[176,27.023],[180,25.1986],[182,24.4071],[183,24.0958],[185,23.6415],[187,23.3543],[195,22.848],[197,22.6072],[198,22.4215],[199,22.1754],[200,21.8583],[201,21.4309],[202,20.8747],[203,20.1576],[204,19.2835],[205,18.3037],[208,15.1378],[209,14.1051],[210,13.1261],[211,12.2382],[212,11.4904],[213,10.8825],[214,10.3723],[215,9.9252],[216,9.5358],[217,9.2086],[218,8.9462],[219,8.7494],[220,8.6106],[221,8.5222],[222,8.4704],[223,8.4476],[224,8.4441],[225,8.4635],[226,8.5117],[227,8.6077],[228,8.7556],[229,8.9569],[230,9.2017],[231,9.4896],[232,9.8503],[233,10.3044],[234,10.8613],[235,11.4842],[236,12.1628],[237,12.9187],[238,13.7555],[239,14.6714],[240,15.7031],[241,16.9313],[242,18.3032],[243,19.7467],[244,21.2316],[245,22.857],[246,24.5444],[247,26.1185],[248,27.4173],[249,28.4404],[250,29.2815],[251,29.9616],[252,30.5145],[254,31.4202],[256,32.1504],[258,32.7093],[261,33.2669],[265,33.677],[271,33.8492],[274,33.7398],[277,33.3915],[291,29.9848],[294,29.4223],[298,28.965],[303,28.73],[305,28.73],[306,0]]
       }
       thigh_l_fk_ctrl.Rotation.Y {
        type=MovieSceneScriptingFloatChannel
        default=-1.2128
        thinned_from=307
        keys [[0,5.5223],[4,5.5057],[6,5.571],[8,5.6714],[9,5.747],[10,5.8626],[11,6.0116],[12,6.1982],[16,7.162],[17,7.3826],[18,7.5839],[22,8.26],[23,8.4433],[27,9.3624],[28,9.5742],[29,9.729],[30,9.8146],[31,9.8451],[32,9.8371],[36,9.7011],[38,9.5742],[40,9.3847],[41,9.2642],[43,8.9642],[49,7.976],[50,7.8364],[51,7.7154],[53,7.5237],[54,7.4487],[55,7.3889],[58,7.2619],[61,7.0946],[65,6.991],[67,6.9005],[69,6.9574],[70,6.8957],[71,6.7511],[72,6.5299],[73,6.1882],[74,5.7057],[75,5.1503],[76,4.5606],[77,3.8835],[78,3.2483],[79,2.8869],[80,2.7663],[81,2.7101],[82,2.6945],[83,2.7039],[84,2.7683],[85,2.8625],[86,2.9958],[87,3.1486],[88,3.3143],[89,3.4891],[90,3.6363],[91,3.7523],[92,3.8343],[93,3.8785],[95,3.9225],[96,3.9314],[102,3.9173],[103,3.8985],[105,3.8329],[109,3.6653],[110,3.631],[111,3.6081],[112,3.5951],[113,3.5951],[114,3.5837],[118,3.5844],[128,3.6838],[141,3.7065],[145,3.763],[146,3.7593],[147,3.7241],[149,3.6231],[150,3.5962],[151,3.6018],[152,3.6379],[153,3.6939],[154,3.7685],[155,3.8696],[156,4.0707],[157,4.3391],[158,4.6228],[161,5.3952],[162,5.6393],[163,5.8443],[164,5.9965],[165,6.1051],[166,6.1784],[167,6.2244],[169,6.2634],[179,6.2767],[182,6.2673],[184,6.2968],[186,6.3553],[190,6.3921],[194,6.3792],[195,6.3622],[196,6.3309],[197,6.2781],[198,6.2002],[199,6.0738],[200,5.8823],[201,5.6343],[202,5.3416],[204,4.6768],[205,4.3308],[206,3.9347],[207,3.5063],[208,3.0508],[209,2.5672],[210,2.0766],[211,1.6048],[212,1.1805],[213,0.8175],[214,0.5251],[215,0.3031],[216,0.1371],[217,0.0153],[218,-0.0709],[219,-0.1275],[220,-0.1643],[221,-0.1859],[222,-0.1928],[223,-0.1868],[224,-0.1675],[225,-0.1311],[226,-0.0746],[227,0.0112],[228,0.1251],[229,0.2672],[230,0.4326],[231,0.6226],[232,0.8447],[233,1.0983],[234,1.3903],[235,1.7292],[236,2.0945],[237,2.4431],[239,3.1119],[240,3.4408],[241,3.7553],[242,4.0399],[244,4.559],[245,4.8115],[246,5.048],[249,5.7181],[251,6.1943],[254,6.949],[255,7.1911],[256,7.4131],[257,7.6145],[258,7.7927],[259,7.9359],[260,8.0284],[261,8.0745],[263,8.1209],[265,8.1215],[267,8.0649],[269,7.9518],[271,7.7946],[273,7.5885],[286,6.1464],[288,5.948],[290,5.7909],[292,5.6674],[295,5.5483],[297,5.496],[299,5.47],[302,5.4789],[305,5.4898],[306,0]]
       }
       thigh_l_fk_ctrl.Rotation.Z {
        type=MovieSceneScriptingFloatChannel
        default=-0.1664
        thinned_from=307
        keys [[0,6.2951],[4,6.276],[10,6.1494],[12,6.1403],[13,6.1468],[14,6.2147],[15,6.3559],[16,6.5607],[17,6.8345],[18,7.1894],[20,8.0511],[22,8.8252],[23,9.1425],[24,9.3555],[25,9.5177],[26,9.6565],[33,10.4618],[49,11.4561],[52,11.5627],[56,11.5842],[59,11.5476],[61,11.467],[62,11.3838],[63,11.2513],[64,11.0485],[65,10.76],[66,10.3638],[67,9.7878],[68,8.9356],[69,7.9379],[70,7.0048],[71,6.1159],[72,5.2893],[73,4.5828],[74,3.9857],[75,3.5449],[76,3.2912],[77,3.2119],[78,3.2149],[79,3.1797],[80,3.1063],[82,2.9759],[83,2.9161],[84,2.8409],[85,2.7467],[86,2.6464],[88,2.4207],[89,2.3032],[90,2.1951],[91,2.1108],[92,2.0451],[93,1.9969],[94,1.9548],[96,1.9015],[97,1.88],[102,1.793],[103,1.7833],[105,1.7769],[106,1.7791],[110,1.8094],[113,1.8142],[114,1.8082],[116,1.8082],[117,1.7911],[118,1.7911],[119,1.7735],[124,1.7228],[127,1.7036],[134,1.6822],[138,1.6833],[140,1.7163],[143,1.7725],[144,1.7983],[145,1.8373],[146,1.903],[147,2.0144],[148,2.1613],[149,2.3271],[151,2.6694],[152,2.8576],[153,3.0527],[154,3.2416],[155,3.3936],[156,3.4347],[157,3.3832],[158,3.2809],[160,3.0258],[161,2.9067],[162,2.7965],[163,2.6974],[164,2.6346],[165,2.6178],[166,2.6275],[168,2.6656],[169,2.6641],[170,2.6381],[171,2.5889],[172,2.5192],[173,2.4307],[174,2.3184],[175,2.1826],[176,2.0297],[177,1.8663],[179,1.5307],[180,1.3712],[181,1.2166],[182,1.0686],[184,0.793],[185,0.6518],[186,0.535],[187,0.4547],[188,0.3976],[189,0.3535],[190,0.3193],[191,0.2925],[192,0.269],[193,0.2472],[194,0.227],[195,0.2095],[196,0.1892],[197,0.1666],[198,0.1374],[199,0.1476],[200,0.2481],[201,0.3588],[202,0.3795],[203,0.2274],[204,-0.1677],[205,-0.746],[206,-1.386],[207,-2.0637],[208,-2.7327],[209,-3.3362],[210,-3.8718],[211,-4.3206],[212,-4.6831],[213,-4.9756],[214,-5.2064],[215,-5.3867],[216,-5.5264],[217,-5.6305],[218,-5.7055],[219,-5.7584],[220,-5.7912],[221,-5.8062],[225,-5.7923],[228,-5.7442],[230,-5.6737],[231,-5.6158],[232,-5.5327],[233,-5.4268],[234,-5.2809],[235,-5.065],[236,-4.7248],[237,-4.2084],[238,-3.5662],[239,-2.8633],[240,-2.1042],[241,-1.2174],[242,-0.2447],[243,0.7303],[244,1.6413],[245,2.5389],[246,3.4096],[247,4.1853],[248,4.8014],[249,5.2705],[250,5.6322],[251,5.8998],[252,6.0895],[253,6.2211],[254,6.3151],[255,6.3828],[256,6.4351],[257,6.4704],[260,6.5093],[267,6.7053],[272,6.9474],[275,7.1044],[276,7.1377],[277,7.1535],[279,7.1401],[285,7.0035],[287,6.9255],[289,6.8014],[292,6.5881],[294,6.4792],[298,6.3706],[303,6.2992],[305,6.2992],[306,0]]
       }
       thigh_l_fk_ctrl.Scale.X {type=MovieSceneScriptingFloatChannel, default=1, const [1,307]}
       thigh_l_fk_ctrl.Scale.Y {type=MovieSceneScriptingFloatChannel, default=1, const [1,307]}
       thigh_l_fk_ctrl.Scale.Z {type=MovieSceneScriptingFloatChannel, default=1, const [1,307]}
       calf_l_fk_ctrl.Location.X {
        type=MovieSceneScriptingFloatChannel
        default=0
        thinned_from=307
        keys [[0,0.0299],[11,0.0299],[12,0.0364],[33,0.0365],[34,0.0418],[38,0.0417],[39,0.0383],[43,0.0383],[44,0.0348],[47,0.0348],[48,0.0313],[50,0.0313],[51,0.0308],[54,0.0308],[55,0.0276],[66,0.0276],[67,0.0286],[68,0.0286],[69,0.0384],[76,0.0384],[77,0.0286],[82,0.0286],[85,0.0287],[86,0.031],[88,0.031],[89,0.0333],[91,0.0333],[92,0.0356],[98,0.0356],[114,0.0354],[115,0.0345],[130,0.0345],[131,0.0361],[146,0.0361],[147,0.0346],[154,0.0346],[155,0.0399],[157,0.0399],[158,0.0451],[159,0.0451],[160,0.0503],[162,0.0504],[163,0.0509],[178,0.0509],[179,0.0485],[194,0.0485],[195,0.0515],[203,0.0515],[204,0.0554],[205,0.0554],[206,0.0592],[208,0.0592],[209,0.0631],[226,0.0631],[227,0.0588],[228,0.0588],[229,0.0635],[237,0.0637],[238,0.0588],[239,0.0588],[240,0.054],[241,0.054],[242,0.0491],[243,0.046],[244,0.0424],[245,0.0423],[247,0.0352],[254,0.0352],[255,0.0388],[258,0.0388],[259,0.04],[274,0.04],[275,0.0344],[290,0.0344],[291,0.0299],[303,0.0299],[304,0.0306],[305,0.0306],[306,0]]
       }
       calf_l_fk_ctrl.Location.Y {
        type=MovieSceneScriptingFloatChannel
        default=0
        thinned_from=307
        keys [[0,0.001],[11,0.001],[12,0.0017],[33,0.0017],[34,0.0025],[37,0.0025],[38,0.0017],[39,0.002],[43,0.002],[44,0.0015],[47,0.0015],[48,0.0011],[54,0.0011],[55,0.0007],[68,0.0008],[69,0.0019],[76,0.0019],[77,0.0008],[85,0.0008],[86,0.0011],[88,0.0011],[89,0.0013],[91,0.0013],[92,0.0012],[93,0.0016],[114,0.0016],[130,0.0015],[146,0.0016],[154,0.0015],[155,0.0023],[157,0.0023],[158,0.0031],[159,0.0031],[160,0.0027],[161,0.0039],[178,0.004],[179,0.0036],[194,0.0036],[195,0.0042],[203,0.0042],[204,0.005],[205,0.005],[206,0.0059],[208,0.0059],[209,0.0068],[226,0.0068],[227,0.0058],[228,0.0058],[229,0.0055],[230,0.0069],[237,0.0069],[238,0.0058],[239,0.0058],[240,0.0048],[241,0.0048],[242,0.0037],[244,0.0026],[245,0.0018],[246,0.0021],[247,0.0015],[254,0.0015],[255,0.0021],[274,0.0022],[275,0.0014],[290,0.0014],[291,0.001],[303,0.001],[304,0.0086],[305,0.0086],[306,0]]
       }
       calf_l_fk_ctrl.Location.Z {
        type=MovieSceneScriptingFloatChannel
        default=0
        thinned_from=307
        keys [[0,0],[10,0],[11,-0.0002],[156,0],[157,-0.0001],[303,0],[304,-0.0036],[305,-0.0036],[306,0]]
       }
       calf_l_fk_ctrl.Rotation.X {
        type=MovieSceneScriptingFloatChannel
        default=0.0543
        thinned_from=307
        keys [[0,-0.0058],[11,-0.0065],[30,-0.0085],[62,-0.0053],[66,-0.0056],[67,-0.0059],[70,-0.0071],[72,-0.0076],[74,-0.0074],[76,-0.0068],[78,-0.0059],[79,-0.0056],[85,-0.0058],[106,-0.0069],[152,-0.0069],[155,-0.0076],[159,-0.0094],[163,-0.0104],[166,-0.0104],[175,-0.0097],[199,-0.0107],[202,-0.0106],[203,-0.0107],[208,-0.0128],[210,-0.0133],[228,-0.0128],[235,-0.0134],[236,-0.0134],[238,-0.0126],[240,-0.0115],[243,-0.0093],[246,-0.0077],[248,-0.0071],[253,-0.0071],[260,-0.0081],[289,-0.0058],[303,-0.0059],[304,-0.0156],[305,-0.0156],[306,0]]
       }
       calf_l_fk_ctrl.Rotation.Y {
        type=MovieSceneScriptingFloatChannel
        default=1.2238
        thinned_from=307
        keys [[0,-0.0066],[9,-0.0069],[31,-0.0091],[63,-0.0061],[66,-0.0064],[70,-0.0079],[72,-0.0083],[74,-0.0082],[76,-0.0076],[78,-0.0067],[79,-0.0064],[85,-0.0065],[147,-0.0077],[152,-0.0077],[155,-0.0083],[160,-0.0102],[164,-0.0109],[183,-0.0106],[203,-0.011],[209,-0.0129],[236,-0.0131],[240,-0.0117],[245,-0.0088],[248,-0.0079],[253,-0.0079],[260,-0.0087],[303,-0.0067],[304,0.0009],[305,0.0009],[306,0]]
       }
       calf_l_fk_ctrl.Rotation.Z {
        type=MovieSceneScriptingFloatChannel
        default=0.3732
        thinned_from=307
        keys [[0,7.7217],[1,7.7217],[2,7.702],[4,7.702],[6,7.7809],[7,7.8466],[8,7.9551],[9,8.113],[12,8.715],[13,8.9388],[14,9.1],[15,9.2218],[16,9.3206],[18,9.443],[20,9.5323],[22,9.6862],[23,9.794],[24,9.9572],[27,10.5365],[28,10.7152],[29,10.8386],[30,10.8941],[31,10.8941],[33,10.7862],[36,10.5533],[38,10.3222],[40,10.0374],[42,9.6883],[49,8.2393],[50,8.057],[51,7.905],[52,7.7699],[53,7.6579],[54,7.5624],[55,7.4867],[58,7.3286],[61,7.1245],[62,7.0916],[63,7.1047],[64,7.1673],[65,7.2792],[66,7.4472],[67,7.746],[68,8.2733],[69,8.8649],[70,9.3064],[71,9.6149],[72,9.8],[73,9.8],[74,9.6445],[75,9.3508],[76,8.9488],[78,7.8052],[79,7.4751],[80,7.384],[81,7.352],[82,7.352],[83,7.3922],[84,7.4847],[85,7.6217],[86,7.7982],[88,8.201],[89,8.417],[90,8.6073],[91,8.76],[92,8.8749],[93,8.947],[95,9.0362],[99,9.1061],[103,9.1217],[110,8.8919],[111,8.871],[116,8.876],[134,9.3272],[139,9.3447],[141,9.3063],[144,9.2785],[145,9.2541],[146,9.2053],[148,9.0135],[149,8.9335],[150,8.9095],[151,8.9575],[152,9.0696],[153,9.2418],[154,9.474],[155,9.7945],[156,10.2994],[158,11.4828],[159,11.9486],[160,12.3263],[161,12.6519],[162,12.9213],[163,13.1303],[164,13.2231],[165,13.1787],[166,13.0496],[168,12.6906],[169,12.5374],[170,12.4165],[171,12.3319],[172,12.2795],[174,12.2553],[180,12.5539],[182,12.7009],[183,12.8012],[184,12.9416],[185,13.1176],[186,13.267],[187,13.3607],[189,13.4699],[193,13.5569],[196,13.5629],[198,13.5212],[199,13.4764],[201,13.3094],[202,13.3094],[203,13.4645],[204,13.8493],[207,15.3789],[208,15.8068],[209,16.0972],[210,16.2768],[211,16.3466],[212,16.3338],[214,16.1868],[217,15.8991],[222,15.6627],[225,15.6371],[227,15.6966],[232,16.1316],[235,16.4469],[236,16.4131],[237,16.0941],[238,15.5803],[239,14.9921],[240,14.3556],[241,13.5627],[242,12.6663],[243,11.7958],[244,11.0768],[245,10.4589],[246,9.9204],[247,9.5045],[248,9.2517],[249,9.1186],[250,9.0643],[251,9.0833],[252,9.1593],[253,9.2925],[258,10.157],[259,10.2952],[260,10.3707],[262,10.3567],[264,10.298],[265,10.2505],[266,10.1807],[267,10.0857],[269,9.8371],[271,9.5466],[275,8.8459],[276,8.6836],[277,8.542],[279,8.3052],[284,7.8265],[285,7.7416],[286,7.6747],[287,7.6335],[289,7.6335],[293,7.7511],[295,7.7652],[297,7.7369],[299,7.7369],[302,7.7228],[303,7.7417],[304,7.7306],[305,7.7448],[306,0]]
       }
       calf_l_fk_ctrl.Scale.X {type=MovieSceneScriptingFloatChannel, default=1, const [1,307]}
       calf_l_fk_ctrl.Scale.Y {type=MovieSceneScriptingFloatChannel, default=1, const [1,307]}
       calf_l_fk_ctrl.Scale.Z {type=MovieSceneScriptingFloatChannel, default=1, const [1,307]}
       foot_l_fk_ctrl.Location.X {type=MovieSceneScriptingFloatChannel, default=0, thinned_from=307, keys [[0,0],[306,0]]}
       foot_l_fk_ctrl.Location.Y {type=MovieSceneScriptingFloatChannel, default=0.0398, thinned_from=307, keys [[0,0],[306,0]]}
       foot_l_fk_ctrl.Location.Z {type=MovieSceneScriptingFloatChannel, default=-0.8885, thinned_from=307, keys [[0,0],[306,0]]}
       foot_l_fk_ctrl.Rotation.X {
        type=MovieSceneScriptingFloatChannel
        default=-1.1647
        thinned_from=307
        keys [[0,-0.0913],[1,-0.0913],[2,-0.0917],[4,-0.0917],[5,-0.0924],[6,-0.0927],[7,-0.0924],[8,-0.0913],[10,-0.0877],[12,-0.0836],[13,-0.0819],[14,-0.0801],[15,-0.0796],[16,-0.0801],[17,-0.0814],[18,-0.0851],[19,-0.089],[20,-0.0936],[22,-0.1039],[23,-0.1088],[24,-0.1126],[25,-0.1161],[26,-0.1187],[27,-0.1224],[29,-0.1272],[30,-0.1289],[31,-0.1313],[32,-0.133],[33,-0.1343],[34,-0.136],[35,-0.1368],[37,-0.1388],[38,-0.1396],[40,-0.1404],[41,-0.1412],[42,-0.1415],[47,-0.142],[49,-0.1416],[60,-0.1412],[61,-0.1416],[62,-0.1417],[63,-0.1413],[64,-0.1402],[65,-0.1378],[66,-0.1337],[67,-0.1268],[68,-0.1151],[69,-0.1013],[70,-0.0884],[71,-0.0752],[72,-0.063],[73,-0.0517],[74,-0.042],[75,-0.0331],[76,-0.0265],[77,-0.0207],[78,-0.016],[79,-0.0129],[80,-0.0112],[81,-0.0096],[82,-0.0084],[83,-0.0075],[84,-0.0068],[85,-0.0058],[88,-0.0035],[89,-0.003],[90,-0.002],[93,0],[97,0.0016],[98,0.0016],[99,0.0025],[102,0.0025],[103,0.0036],[104,0.0037],[108,0.0051],[111,0.0053],[112,0.0057],[118,0.0057],[119,0.0061],[121,0.0059],[122,0.0064],[125,0.006],[126,0.0065],[129,0.0061],[130,0.0066],[131,0.0065],[132,0.0071],[134,0.0071],[135,0.0074],[139,0.0074],[140,0.0059],[141,0.0059],[142,0.0044],[143,0.0037],[144,0.0027],[147,-0.0027],[148,-0.0051],[151,-0.013],[152,-0.016],[153,-0.0193],[154,-0.0221],[156,-0.0265],[157,-0.0283],[158,-0.0292],[160,-0.0301],[161,-0.0302],[163,-0.0298],[164,-0.0294],[165,-0.0281],[166,-0.0265],[168,-0.0239],[169,-0.0223],[170,-0.0217],[171,-0.02],[172,-0.0186],[173,-0.017],[174,-0.0158],[175,-0.0142],[176,-0.0122],[177,-0.0107],[178,-0.0087],[179,-0.0073],[180,-0.0057],[182,-0.0031],[184,-0.0014],[187,0],[189,0.0004],[190,0.0004],[191,0.0009],[193,0.0009],[194,0.0019],[195,0.0019],[196,0.0028],[198,0.0053],[199,0.0078],[200,0.0099],[201,0.0139],[202,0.0193],[203,0.0292],[204,0.0415],[205,0.0555],[206,0.0698],[207,0.0838],[208,0.0972],[209,0.11],[210,0.1222],[211,0.1331],[212,0.1422],[213,0.1498],[214,0.1556],[215,0.1595],[217,0.1641],[218,0.1654],[219,0.1663],[222,0.1669],[225,0.1666],[226,0.1654],[228,0.1635],[230,0.1605],[231,0.1581],[233,0.1521],[234,0.1476],[235,0.1404],[236,0.1305],[237,0.1201],[238,0.1087],[239,0.0969],[240,0.0843],[241,0.0712],[242,0.0572],[243,0.0428],[244,0.0281],[245,0.0128],[247,-0.0184],[248,-0.0328],[249,-0.0458],[250,-0.0571],[251,-0.0667],[252,-0.0752],[253,-0.0825],[254,-0.089],[255,-0.0948],[256,-0.1002],[257,-0.1046],[258,-0.1083],[259,-0.1112],[260,-0.1133],[261,-0.1148],[262,-0.1161],[263,-0.1166],[264,-0.1174],[265,-0.1174],[266,-0.1176],[268,-0.117],[269,-0.117],[274,-0.1152],[276,-0.1142],[279,-0.1115],[280,-0.1105],[281,-0.1091],[282,-0.1082],[284,-0.1057],[285,-0.1048],[286,-0.1032],[287,-0.1021],[288,-0.1007],[290,-0.0984],[291,-0.097],[293,-0.095],[295,-0.0933],[297,-0.0919],[298,-0.0919],[299,-0.0908],[303,-0.0908],[304,-0.0901],[305,-0.0901],[306,0]]
       }
       foot_l_fk_ctrl.Rotation.Y {
        type=MovieSceneScriptingFloatChannel
        default=0.0792
        thinned_from=307
        keys [[0,-3.106],[4,-3.1121],[5,-3.123],[7,-3.1237],[8,-3.1102],[13,-2.9848],[14,-2.964],[15,-2.9557],[16,-2.9628],[17,-2.9838],[18,-3.0233],[19,-3.0815],[21,-3.2108],[22,-3.2825],[23,-3.3448],[25,-3.4462],[28,-3.5667],[30,-3.6291],[33,-3.6971],[38,-3.7613],[44,-3.7881],[60,-3.7769],[62,-3.7829],[63,-3.7789],[64,-3.7642],[65,-3.7326],[66,-3.6778],[67,-3.5811],[68,-3.4319],[70,-3.0696],[71,-2.8953],[72,-2.7329],[73,-2.5865],[74,-2.4464],[75,-2.328],[76,-2.2338],[77,-2.1552],[78,-2.0896],[79,-2.0449],[80,-2.0212],[81,-2.0027],[86,-1.941],[94,-1.8724],[97,-1.8562],[98,-1.8562],[99,-1.8477],[102,-1.8477],[103,-1.8332],[110,-1.8064],[116,-1.8038],[117,-1.8],[129,-1.7963],[131,-1.7915],[132,-1.786],[139,-1.7846],[140,-1.7986],[141,-1.7986],[142,-1.8178],[144,-1.8464],[145,-1.8658],[146,-1.8903],[148,-1.9507],[150,-2.0189],[151,-2.0562],[154,-2.1808],[156,-2.2438],[157,-2.2685],[158,-2.2857],[160,-2.2988],[162,-2.3018],[163,-2.2967],[165,-2.2741],[172,-2.1465],[174,-2.1073],[178,-2.0127],[181,-1.952],[182,-1.9361],[184,-1.9154],[188,-1.8936],[190,-1.8906],[191,-1.884],[193,-1.884],[194,-1.872],[195,-1.872],[197,-1.8464],[199,-1.7977],[200,-1.7615],[201,-1.7127],[202,-1.6371],[203,-1.5073],[204,-1.336],[205,-1.1506],[207,-0.7707],[208,-0.5901],[209,-0.4189],[210,-0.2561],[211,-0.1075],[212,0.0184],[213,0.1236],[214,0.2045],[215,0.2601],[216,0.2977],[217,0.3261],[218,0.3459],[219,0.3592],[220,0.3668],[221,0.3695],[222,0.3694],[223,0.3676],[225,0.3676],[226,0.3505],[227,0.34],[228,0.3217],[229,0.304],[230,0.2809],[231,0.2494],[232,0.207],[233,0.1621],[234,0.1003],[235,0.0013],[236,-0.1343],[237,-0.2782],[238,-0.4285],[239,-0.5889],[240,-0.7587],[241,-0.9341],[242,-1.1201],[243,-1.3116],[244,-1.5082],[245,-1.7128],[247,-2.1319],[248,-2.3273],[249,-2.499],[250,-2.6498],[251,-2.7816],[252,-2.8946],[253,-2.9948],[254,-3.0844],[255,-3.1637],[256,-3.2355],[257,-3.2953],[258,-3.3451],[259,-3.3862],[260,-3.4157],[262,-3.4517],[264,-3.4683],[274,-3.4367],[279,-3.3834],[292,-3.1707],[297,-3.1173],[298,-3.1173],[299,-3.1033],[303,-3.1033],[304,-3.0909],[305,-3.0909],[306,0]]
       }
       foot_l_fk_ctrl.Rotation.Z {
        type=MovieSceneScriptingFloatChannel
        default=-0.2074
        thinned_from=307
        keys [[0,-2.2086],[1,-2.2086],[2,-2.1909],[4,-2.1909],[5,-2.2042],[6,-2.2249],[7,-2.2657],[8,-2.3419],[9,-2.456],[10,-2.593],[12,-2.8926],[13,-3.0502],[14,-3.1901],[16,-3.4538],[18,-3.7268],[21,-4.1753],[23,-4.443],[26,-4.8579],[27,-4.981],[28,-5.0912],[29,-5.1756],[30,-5.2217],[31,-5.2357],[32,-5.2305],[33,-5.2123],[35,-5.1464],[36,-5.1012],[38,-4.981],[40,-4.8334],[42,-4.648],[49,-3.9261],[50,-3.8345],[51,-3.7534],[52,-3.6833],[53,-3.6225],[54,-3.5695],[55,-3.5243],[57,-3.4557],[58,-3.4152],[59,-3.3653],[60,-3.3047],[62,-3.1664],[63,-3.1117],[64,-3.069],[66,-2.998],[67,-2.9884],[69,-3.094],[70,-3.0854],[71,-3.0228],[72,-2.9147],[73,-2.7446],[74,-2.5039],[75,-2.2267],[76,-1.9298],[77,-1.5836],[78,-1.2465],[79,-1.044],[80,-0.9761],[81,-0.9418],[82,-0.9328],[83,-0.9386],[84,-0.9745],[85,-1.0367],[86,-1.1156],[87,-1.2077],[88,-1.3032],[89,-1.4054],[90,-1.4976],[91,-1.57],[92,-1.6291],[93,-1.6651],[94,-1.6912],[95,-1.7107],[97,-1.7367],[98,-1.7367],[99,-1.7536],[102,-1.7536],[103,-1.7582],[105,-1.7316],[109,-1.6573],[110,-1.6414],[111,-1.6414],[112,-1.6282],[116,-1.6282],[117,-1.6466],[118,-1.6466],[119,-1.6689],[120,-1.6689],[121,-1.6913],[130,-1.8254],[131,-1.8462],[133,-1.8758],[134,-1.8758],[135,-1.8944],[139,-1.8944],[140,-1.9063],[141,-1.9063],[144,-1.9235],[145,-1.9245],[146,-1.9182],[148,-1.8792],[149,-1.8717],[150,-1.8939],[151,-1.9572],[152,-2.0595],[153,-2.1916],[154,-2.3509],[155,-2.5486],[156,-2.8261],[157,-3.1444],[158,-3.4488],[159,-3.6888],[160,-3.8785],[161,-4.043],[162,-4.1755],[163,-4.2756],[164,-4.3126],[165,-4.2852],[166,-4.2148],[168,-4.0204],[169,-3.9341],[170,-3.8583],[171,-3.804],[172,-3.7603],[173,-3.7272],[175,-3.6877],[180,-3.6285],[181,-3.6248],[182,-3.635],[183,-3.6563],[184,-3.7052],[186,-3.8305],[187,-3.8659],[189,-3.9038],[190,-3.9038],[191,-3.9199],[195,-3.9221],[197,-3.8791],[198,-3.8481],[199,-3.8027],[201,-3.6622],[202,-3.6066],[203,-3.6067],[204,-3.6671],[205,-3.7553],[206,-3.8172],[207,-3.8403],[208,-3.8157],[209,-3.7228],[210,-3.5849],[211,-3.4043],[213,-2.9868],[214,-2.7865],[215,-2.6033],[216,-2.4503],[217,-2.3349],[218,-2.2464],[219,-2.1794],[220,-2.1313],[221,-2.0941],[222,-2.0675],[223,-2.0517],[225,-2.0517],[226,-2.0791],[227,-2.1165],[228,-2.1737],[229,-2.2461],[230,-2.3296],[231,-2.4265],[232,-2.5435],[233,-2.6736],[234,-2.8262],[235,-3.0132],[236,-3.1782],[237,-3.2223],[238,-3.1842],[239,-3.1206],[240,-3.0444],[241,-2.9077],[242,-2.7325],[243,-2.5741],[244,-2.4855],[245,-2.4534],[246,-2.4558],[247,-2.4985],[248,-2.5891],[249,-2.6995],[250,-2.8169],[252,-3.0685],[253,-3.2057],[255,-3.4919],[256,-3.616],[258,-3.8325],[259,-3.9136],[260,-3.9573],[261,-3.955],[263,-3.9237],[265,-3.8592],[266,-3.8104],[267,-3.7518],[269,-3.602],[271,-3.4391],[273,-3.2566],[275,-3.0634],[276,-2.9745],[277,-2.8945],[279,-2.7498],[281,-2.6231],[284,-2.4531],[285,-2.4004],[286,-2.3599],[287,-2.3286],[288,-2.3125],[289,-2.3056],[290,-2.3048],[293,-2.3132],[294,-2.3108],[297,-2.2672],[298,-2.2672],[299,-2.2463],[303,-2.2463],[304,-2.2383],[305,-2.2383],[306,0]]
       }
       foot_l_fk_ctrl.Scale.X {type=MovieSceneScriptingFloatChannel, default=1, const [1,307]}
       foot_l_fk_ctrl.Scale.Y {type=MovieSceneScriptingFloatChannel, default=1, const [1,307]}
       foot_l_fk_ctrl.Scale.Z {type=MovieSceneScriptingFloatChannel, default=1, const [1,307]}
       ball_l_fk_ctrl.Location.X {type=MovieSceneScriptingFloatChannel, default=0, const [0,307]}
       ball_l_fk_ctrl.Location.Y {type=MovieSceneScriptingFloatChannel, default=0, const [0,307]}
       ball_l_fk_ctrl.Location.Z {type=MovieSceneScriptingFloatChannel, default=0, const [0,307]}
       ball_l_fk_ctrl.Rotation.X {type=MovieSceneScriptingFloatChannel, default=0, const [0,307]}
       ball_l_fk_ctrl.Rotation.Y {type=MovieSceneScriptingFloatChannel, default=0, const [0,307]}
       ball_l_fk_ctrl.Rotation.Z {type=MovieSceneScriptingFloatChannel, default=0, const [0,307]}
       ball_l_fk_ctrl.Scale.X {type=MovieSceneScriptingFloatChannel, default=1, const [1,307]}
       ball_l_fk_ctrl.Scale.Y {type=MovieSceneScriptingFloatChannel, default=1, const [1,307]}
       ball_l_fk_ctrl.Scale.Z {type=MovieSceneScriptingFloatChannel, default=1, const [1,307]}
       foot_l_ik_ctrl.Location.X {
        type=MovieSceneScriptingFloatChannel
        default=0
        thinned_from=307
        keys [[0,-1.7362],[1,-1.7406],[2,-1.7364],[3,-1.7384],[4,-1.7473],[5,-1.7397],[10,-1.7397],[12,-1.7366],[13,-1.7401],[14,-1.7367],[15,-1.7404],[18,-1.7363],[20,-1.7403],[21,-1.7371],[22,-1.7399],[25,-1.7408],[26,-1.7368],[28,-1.739],[35,-1.7372],[36,-1.7407],[44,-1.7402],[50,-1.7358],[51,-1.7414],[53,-1.7358],[54,-1.7417],[55,-1.7382],[56,-1.7414],[57,-1.7393],[58,-1.7422],[60,-1.7366],[61,-1.7398],[62,-1.7364],[64,-1.742],[65,-1.7354],[66,-1.7436],[70,-1.736],[72,-1.7383],[73,-1.7328],[74,-1.7391],[76,-1.7406],[77,-1.7376],[81,-1.7389],[82,-1.7345],[83,-1.7416],[84,-1.7358],[85,-1.7433],[86,-1.7372],[88,-1.7391],[89,-1.7336],[91,-1.7414],[95,-1.7357],[99,-1.7383],[100,-1.7342],[101,-1.7375],[102,-1.7357],[103,-1.7403],[106,-1.7381],[107,-1.7403],[109,-1.7363],[110,-1.7394],[111,-1.7384],[112,-1.7432],[113,-1.7344],[114,-1.7403],[116,-1.7346],[117,-1.739],[118,-1.7472],[119,-1.7395],[122,-1.7358],[125,-1.7401],[128,-1.7372],[129,-1.7392],[134,-1.7396],[135,-1.7352],[136,-1.7423],[138,-1.7358],[139,-1.7424],[140,-1.7372],[141,-1.7379],[142,-1.7429],[143,-1.7383],[146,-1.7389],[147,-1.7362],[148,-1.7392],[149,-1.7368],[151,-1.7404],[154,-1.7373],[156,-1.7404],[157,-1.7379],[158,-1.7388],[159,-1.7357],[160,-1.7391],[173,-1.7377],[174,-1.734],[177,-1.7397],[179,-1.7371],[193,-1.7376],[194,-1.7342],[196,-1.7425],[197,-1.7366],[201,-1.7372],[202,-1.7437],[203,-1.7381],[210,-1.739],[211,-1.7364],[214,-1.7411],[215,-1.7377],[223,-1.7381],[224,-1.7429],[227,-1.7378],[229,-1.7406],[252,-1.7374],[256,-1.7401],[259,-1.7374],[260,-1.7402],[262,-1.7413],[263,-1.737],[264,-1.742],[268,-1.7382],[269,-1.7416],[270,-1.7369],[271,-1.7429],[278,-1.7373],[282,-1.7369],[283,-1.7399],[284,-1.7378],[286,-1.7396],[287,-1.7368],[288,-1.7449],[290,-1.7368],[293,-1.7399],[294,-1.7371],[295,-1.7408],[297,-1.738],[299,-1.7441],[300,-1.7379],[302,-1.7325],[303,-1.7378],[304,-1.7337],[305,-1.7413],[306,0]]
       }
       foot_l_ik_ctrl.Location.Y {
        type=MovieSceneScriptingFloatChannel
        default=0
        thinned_from=307
        keys [[0,-9.4592],[117,-9.4588],[118,-9.47],[138,-9.4515],[139,-9.4671],[141,-9.456],[305,-9.4611],[306,0]]
       }
       foot_l_ik_ctrl.Location.Z {
        type=MovieSceneScriptingFloatChannel
        default=0
        thinned_from=307
        keys [[0,0],[1,0.0011],[2,0.0002],[3,0.0006],[4,-0.0008],[5,-0.0005],[6,-0.0006],[7,-0.001],[8,-0.0008],[10,-0.0023],[11,-0.0027],[12,0.0027],[13,0.0017],[15,0.0006],[16,0.0008],[17,0.0006],[19,-0.0004],[21,-0.0005],[23,-0.0011],[25,-0.0026],[26,-0.0036],[28,-0.0046],[29,-0.0053],[32,-0.0051],[33,-0.0052],[34,0.0007],[35,0.0006],[36,0.0013],[37,0.0019],[38,0.0017],[39,-0.0007],[40,-0.0007],[41,-0.0005],[42,0.0012],[43,0.0013],[44,-0.0006],[45,-0.0002],[46,0.001],[47,0.0013],[48,-0.0017],[49,-0.0011],[50,0.0001],[51,0.0006],[52,0.0005],[53,0.0009],[54,0.0016],[55,-0.0014],[56,-0.0009],[58,-0.0011],[59,-0.0009],[60,-0.0005],[61,-0.0005],[63,0.0004],[64,-0.0006],[65,-0.0004],[66,-0.0012],[67,-0.0014],[68,-0.0031],[69,0.0041],[70,0.0027],[71,0.0014],[72,0.0004],[73,0.0001],[76,0.0035],[77,-0.0038],[78,-0.0015],[80,-0.0002],[81,-0.0001],[82,-0.0004],[84,0.0001],[85,-0.0012],[86,0.0006],[87,-0.0001],[88,-0.0004],[89,0.0011],[90,-0.0002],[91,-0.0004],[92,0.001],[93,0.0012],[94,0.0009],[95,0.001],[96,0.001],[97,0],[98,0.0007],[99,0.0001],[100,-0.0008],[101,0],[102,-0.0007],[103,0.0005],[104,0],[105,0.0004],[106,0.0001],[107,0.0006],[108,0.0004],[111,0.001],[113,0.003],[114,0.0016],[115,0.0018],[116,0.0002],[117,-0.0007],[118,0.0007],[119,0.0003],[122,-0.0013],[123,0.0006],[125,-0.0001],[126,-0.0002],[128,-0.0012],[129,-0.0016],[130,-0.0023],[131,0.0002],[132,0.0003],[133,0.0001],[134,0.0002],[135,-0.0006],[137,-0.0002],[138,-0.0003],[139,0.0006],[140,-0.0002],[141,-0.0001],[142,0.0007],[144,0],[146,0.0005],[147,0.0005],[148,-0.0017],[149,-0.0008],[150,0.0006],[151,0.0007],[152,0.0004],[153,-0.0011],[154,-0.0011],[155,0.001],[156,0.0015],[157,-0.0026],[158,-0.0003],[159,-0.0018],[160,0.0015],[161,0.0006],[162,-0.0004],[163,-0.0001],[165,-0.0002],[169,0.0024],[170,0.0026],[172,0.0032],[173,0.0033],[174,0.0025],[175,0.0031],[177,0.003],[178,0.0025],[179,-0.001],[180,-0.0011],[181,0.0004],[182,-0.0013],[183,-0.0005],[184,-0.0006],[185,-0.002],[186,-0.0057],[187,-0.0032],[188,-0.0038],[189,-0.0059],[190,-0.0029],[191,-0.0069],[192,-0.0058],[193,-0.0051],[194,-0.0053],[195,-0.0011],[196,-0.0002],[197,-0.0011],[198,-0.0004],[199,-0.0017],[200,0.0002],[201,-0.0003],[202,0.0013],[203,-0.0007],[204,0.0015],[205,0.0004],[206,0.002],[207,0.0005],[208,-0.002],[209,0.0002],[210,-0.0008],[211,-0.0001],[212,-0.0001],[214,0.0009],[215,0.0008],[217,0.0018],[218,0.002],[219,0.0024],[222,0.0028],[223,0.0027],[224,0.0033],[226,0.003],[227,-0.0017],[228,-0.0021],[229,0.0023],[230,0.0025],[233,0.0011],[234,0.0009],[235,0.0004],[236,0.0002],[237,0.0019],[238,-0.0014],[239,0.0005],[240,-0.0017],[241,0.0015],[242,0],[243,0.0003],[244,-0.0002],[245,0.0015],[246,0.0006],[247,-0.0016],[248,-0.0008],[249,0.0002],[250,-0.0002],[251,-0.0001],[253,-0.0008],[254,-0.0014],[255,0.0018],[258,0],[259,0.0016],[260,-0.0015],[261,-0.0004],[262,-0.0012],[263,-0.0004],[264,0],[265,0.002],[266,-0.0006],[267,0.001],[268,0.0016],[270,0.0018],[271,0.0032],[273,0.0046],[274,0.0055],[275,0.0006],[278,0.0025],[283,0.0031],[284,0.0042],[286,0.0046],[287,0.0044],[288,0.0052],[289,0.0044],[290,0.0046],[291,0],[292,0],[293,0.0007],[294,-0.001],[295,0.0003],[296,0.0001],[297,-0.0007],[298,0.001],[300,0.0003],[301,-0.0007],[304,-0.0004],[305,0.0008],[306,0]]
       }
       foot_l_ik_ctrl.Rotation.X {
        type=MovieSceneScriptingFloatChannel
        default=0
        thinned_from=307
        keys [[0,0],[1,0.001],[3,-0.0013],[4,0.0025],[5,-0.0004],[6,0.0011],[7,0.0004],[8,-0.0001],[9,0.0003],[10,0.0022],[11,0.0006],[12,-0.0017],[13,0.003],[14,-0.0021],[15,0.001],[16,0.0026],[17,-0.0009],[18,0.0027],[19,0.0006],[20,-0.0002],[21,-0.0024],[22,0.0002],[23,0.0046],[24,-0.0009],[25,0.004],[26,-0.0012],[27,-0.0018],[28,0.0011],[29,-0.0025],[30,-0.0014],[31,0.0028],[33,-0.0031],[34,0.0006],[35,-0.0011],[36,0.0029],[37,0.0025],[38,0.0015],[39,0.0013],[40,-0.0001],[41,-0.001],[42,0.0038],[43,0.0002],[44,0.0028],[45,0.0015],[46,0.0036],[47,0.0018],[48,-0.0004],[49,-0.0023],[50,-0.0025],[51,0.0038],[52,0.0001],[53,0.0007],[54,-0.0002],[55,-0.0008],[56,0],[57,0.0005],[58,0.0001],[59,0.0013],[60,-0.0017],[61,-0.0002],[62,0.0018],[64,0.0023],[65,-0.002],[66,0.0013],[67,0.0041],[68,-0.0016],[69,0.001],[71,0.0012],[72,-0.0002],[73,-0.0075],[74,0.0031],[75,0.0022],[76,0.0025],[77,0.0003],[78,0.0005],[79,0.0017],[80,-0.0002],[81,0.0008],[82,-0.0079],[83,0.0008],[84,0.0029],[85,-0.0006],[86,0.003],[88,0.0018],[89,0.0017],[90,-0.0011],[91,0.0021],[93,0.001],[94,-0.0018],[95,0.0009],[96,0.0018],[97,0.0006],[98,0.0106],[99,-0.0006],[100,-0.0035],[101,0.008],[102,0.0042],[103,0.0041],[104,0.0021],[105,0.0007],[106,-0.0001],[107,0.0024],[108,-0.0021],[109,-0.0021],[110,0.0024],[111,-0.0083],[112,0.0086],[113,0.0055],[114,0.0058],[115,0.0095],[116,0.0082],[117,0.0013],[118,0.0181],[119,-0.0004],[120,0.0097],[121,0.0016],[122,-0.0023],[123,0],[124,0.0005],[126,0.0004],[127,-0.0014],[128,-0.0026],[129,0.0006],[130,0.0028],[131,-0.0008],[132,0.0036],[133,0.0011],[134,0.0127],[135,-0.0051],[136,0.0111],[137,0.009],[138,0.008],[139,0.0225],[140,-0.0001],[141,0.0019],[142,0.0106],[143,0.0009],[144,-0.001],[146,0.0028],[147,-0.0024],[148,0.0002],[149,-0.0001],[150,0.0031],[151,0.0038],[152,0.001],[153,0.0011],[155,0.0009],[156,0.0031],[157,0.0014],[158,0.0008],[159,-0.003],[160,0.0004],[161,-0.0019],[162,-0.0015],[163,-0.0008],[164,0.004],[165,0.0018],[167,0],[168,-0.0019],[169,-0.0012],[170,0.0007],[171,-0.0021],[173,-0.0013],[174,-0.0076],[175,0.0002],[176,0.001],[177,0.0017],[178,-0.0011],[179,-0.0008],[180,0.0012],[181,0.0011],[182,-0.0007],[183,0.001],[184,-0.002],[185,0.0001],[186,-0.0006],[187,-0.0011],[188,0.0005],[189,-0.001],[190,0.0096],[191,0.0009],[192,0.0035],[193,0.0037],[194,-0.0071],[195,-0.0078],[196,0.0069],[197,-0.0036],[198,0.001],[199,-0.0014],[200,-0.0008],[201,-0.0011],[202,0.0079],[203,-0.0014],[204,-0.0002],[205,0.0001],[206,-0.0008],[207,0.0021],[208,0.002],[209,0.0011],[210,-0.001],[211,-0.0027],[213,0.0023],[214,0.0026],[215,-0.002],[216,0.0011],[217,0.0001],[218,-0.0022],[219,-0.0003],[220,-0.0025],[221,0],[222,0.0017],[223,0.0001],[224,0.0051],[225,0.0085],[226,0.0006],[227,0.0003],[228,-0.0007],[229,0.0013],[230,0.0004],[231,0.0027],[233,-0.0003],[234,-0.0002],[235,-0.002],[236,-0.0002],[237,0.0006],[238,0.0011],[239,-0.0008],[240,-0.0008],[241,0.0012],[242,0.0025],[243,0.0002],[244,0.0004],[245,-0.0006],[246,0.001],[247,0.0021],[248,0.0007],[249,0.0023],[250,-0.0002],[251,-0.0003],[252,-0.001],[253,0.0003],[254,0.001],[255,0.0008],[256,0.0035],[258,0.001],[259,-0.0011],[260,-0.0011],[261,0.0004],[262,0.0015],[263,0.0002],[264,0.0027],[265,0.0024],[266,0.0035],[267,0.0003],[268,0.0002],[269,0.0015],[270,0.0014],[271,0.0024],[272,0.0009],[273,0.0002],[274,0.0002],[275,0.0008],[276,-0.0006],[277,-0.001],[279,0.0012],[281,-0.0002],[282,-0.0002],[283,-0.0012],[284,-0.0007],[285,0.0014],[286,-0.0009],[287,0.0004],[288,0.0115],[289,0.0019],[290,-0.0004],[291,-0.0007],[292,-0.0006],[293,0.0014],[294,-0.0018],[295,0.0035],[296,0.0018],[297,-0.0011],[298,-0.0049],[299,0.0113],[300,-0.0079],[301,-0.0135],[302,-0.0207],[303,-0.0081],[304,-0.007],[305,0.007],[306,0]]
       }
       foot_l_ik_ctrl.Rotation.Y {
        type=MovieSceneScriptingFloatChannel
        default=0
        thinned_from=307
        keys [[0,0.0009],[1,0.0046],[2,0.0017],[3,0.0046],[4,0.0055],[5,-0.0016],[6,-0.001],[7,-0.0014],[8,0.0008],[9,-0.0008],[10,-0.001],[11,0.0013],[12,0.0007],[13,-0.0015],[14,0.0004],[15,-0.0014],[16,0.0013],[17,0.0005],[18,0.0002],[19,0.0008],[20,0],[21,0.0006],[22,-0.0011],[23,0.0011],[24,0.0012],[25,-0.0013],[26,-0.0011],[27,-0.0001],[28,-0.0005],[29,0.0001],[30,-0.0001],[31,0.0014],[32,0.0015],[33,0.0009],[34,0.0008],[35,-0.0001],[36,-0.0003],[37,0.0004],[38,0.0001],[39,0.0014],[40,0.0002],[41,-0.0014],[42,0.001],[43,-0.0001],[44,0.0004],[45,0.0005],[46,0.0017],[47,0.0002],[48,-0.0008],[49,-0.0012],[50,0.0012],[51,0.0004],[53,0.0001],[54,0.0001],[55,-0.0013],[56,0.0011],[58,-0.0012],[59,-0.0001],[60,-0.0006],[61,-0.0015],[62,0.0008],[63,0.0006],[64,-0.001],[65,0.0008],[66,-0.0015],[67,0.0003],[68,0.0011],[69,0.0001],[70,0.0014],[71,0.0018],[72,-0.0007],[74,0.0017],[75,0],[76,-0.0015],[77,0.0002],[78,-0.0007],[79,0.0006],[81,0.0002],[82,-0.0023],[83,-0.001],[84,0.0014],[85,-0.0021],[86,0.0005],[87,-0.0004],[88,0.0009],[89,0.0014],[90,-0.0024],[91,-0.0003],[92,0.001],[93,0.0017],[94,-0.0012],[95,0.0014],[96,0.0019],[97,-0.0025],[98,-0.0024],[99,0.0005],[100,-0.004],[101,-0.0064],[102,-0.0115],[103,0.0006],[104,-0.0008],[105,0.0003],[106,-0.0006],[107,-0.0008],[108,-0.0004],[109,0.0011],[110,-0.0004],[111,-0.0023],[112,-0.0002],[113,0.0051],[114,-0.002],[115,0.0009],[116,0.0027],[117,-0.0001],[118,-0.0002],[119,0.0006],[120,0.0003],[121,0.0003],[122,-0.0006],[123,0.0003],[124,-0.0002],[125,-0.0013],[126,0.0005],[127,0.0012],[128,0.0004],[129,-0.0009],[130,0.0005],[131,-0.0018],[132,0.0011],[133,0.0002],[134,-0.0022],[135,-0.0008],[136,-0.0018],[137,0.0013],[138,0.0007],[139,0.0091],[140,0.0017],[141,0.0108],[142,0.0027],[143,0.0011],[144,0.0018],[145,0.0022],[146,0.002],[147,0.0009],[148,-0.0007],[149,0.0007],[150,0.0004],[151,-0.0007],[152,0.0005],[153,0.0006],[154,0.001],[155,-0.0005],[156,-0.0006],[157,-0.0004],[158,0.0009],[159,0.0009],[160,-0.0004],[162,-0.0009],[163,0.0007],[164,0.0011],[165,-0.0005],[166,-0.0011],[167,0.0011],[168,0.0007],[169,-0.0006],[170,0.0002],[171,-0.0012],[172,-0.0003],[173,0.0017],[174,-0.0011],[175,0.0005],[176,0.0012],[177,-0.001],[178,0.0008],[179,-0.0004],[180,-0.0002],[181,0.0005],[182,0.001],[183,0],[184,-0.0004],[186,-0.0006],[188,0.0009],[189,-0.0005],[190,-0.0032],[191,-0.0002],[192,-0.0036],[193,-0.0082],[194,0.0005],[195,-0.0051],[196,0.0005],[197,-0.0012],[198,0.0033],[199,-0.0028],[200,-0.0016],[201,-0.0011],[202,-0.004],[203,-0.0032],[204,-0.0015],[205,0.0017],[206,0.0009],[207,0.002],[208,0.0019],[209,-0.001],[210,-0.0018],[211,-0.0001],[212,-0.0002],[213,-0.0007],[215,0.0004],[216,-0.0005],[217,0.0007],[218,0.0001],[219,0.0001],[220,0.0007],[221,-0.0001],[222,-0.0003],[223,-0.0001],[224,0.0028],[225,0.0086],[226,-0.0003],[227,0.0019],[228,-0.0017],[229,0.0001],[230,0.0007],[231,0.0003],[232,-0.0009],[233,0.0004],[234,-0.0007],[235,-0.0004],[236,-0.0003],[237,-0.0008],[238,0.0006],[239,0.0012],[240,-0.0011],[241,0.0015],[243,-0.0009],[244,-0.0008],[245,-0.0002],[246,0.0011],[247,0.0001],[248,-0.0015],[249,0.0014],[250,0.0002],[251,-0.0002],[252,-0.0004],[253,-0.001],[254,0.0004],[255,0.0009],[256,-0.0003],[258,0.0009],[259,0.0003],[260,-0.0008],[261,0.0005],[262,0.0008],[263,0],[264,0.0003],[266,0.0004],[267,-0.0003],[268,0.0005],[269,-0.0001],[270,-0.0005],[271,0.0005],[272,-0.0003],[273,0.0011],[274,-0.0002],[275,-0.001],[276,-0.0001],[277,0.0012],[278,0.0011],[279,0.0002],[280,0.0007],[281,-0.0003],[282,0.0007],[283,-0.001],[284,0],[286,-0.0007],[287,0.0005],[288,0.0001],[290,0.0007],[291,-0.0013],[292,0.0001],[293,-0.0004],[294,0.0008],[295,-0.0006],[296,-0.0001],[297,-0.0009],[298,-0.0083],[299,0],[300,-0.0047],[301,-0.0069],[302,-0.0096],[303,-0.0102],[304,0.0003],[305,0.0003],[306,0]]
       }
       foot_l_ik_ctrl.Rotation.Z {type=MovieSceneScriptingFloatChannel, default=0, thinned_from=307, keys [[0,-45.8885],[305,-45.8912],[306,0]]}
       foot_l_ik_ctrl.Scale.X {type=MovieSceneScriptingFloatChannel, default=1, const [1,307]}
       foot_l_ik_ctrl.Scale.Y {type=MovieSceneScriptingFloatChannel, default=1, const [1,307]}
       foot_l_ik_ctrl.Scale.Z {type=MovieSceneScriptingFloatChannel, default=1, const [1,307]}
       ball_l_ik_ctrl.Location.X {type=MovieSceneScriptingFloatChannel, default=0, const [0,307]}
       ball_l_ik_ctrl.Location.Y {type=MovieSceneScriptingFloatChannel, default=0, const [0,307]}
       ball_l_ik_ctrl.Location.Z {type=MovieSceneScriptingFloatChannel, default=0, const [0,307]}
       ball_l_ik_ctrl.Rotation.X {type=MovieSceneScriptingFloatChannel, default=0, const [0,307]}
       ball_l_ik_ctrl.Rotation.Y {type=MovieSceneScriptingFloatChannel, default=0, const [0,307]}
       ball_l_ik_ctrl.Rotation.Z {type=MovieSceneScriptingFloatChannel, default=0, const [0,307]}
       ball_l_ik_ctrl.Scale.X {type=MovieSceneScriptingFloatChannel, default=1, const [1,307]}
       ball_l_ik_ctrl.Scale.Y {type=MovieSceneScriptingFloatChannel, default=1, const [1,307]}
       ball_l_ik_ctrl.Scale.Z {type=MovieSceneScriptingFloatChannel, default=1, const [1,307]}
       foot_bk1_l_ctrl.Location.X {type=MovieSceneScriptingFloatChannel, default=0, const [0,307]}
       foot_bk1_l_ctrl.Location.Y {type=MovieSceneScriptingFloatChannel, default=0, const [0,307]}
       foot_bk1_l_ctrl.Location.Z {type=MovieSceneScriptingFloatChannel, default=0, const [0,307]}
       foot_bk1_l_ctrl.Rotation.X {type=MovieSceneScriptingFloatChannel, default=0, const [0,307]}
       foot_bk1_l_ctrl.Rotation.Y {type=MovieSceneScriptingFloatChannel, default=0, const [0,307]}
       foot_bk1_l_ctrl.Rotation.Z {type=MovieSceneScriptingFloatChannel, default=0, const [0,307]}
       foot_bk1_l_ctrl.Scale.X {type=MovieSceneScriptingFloatChannel, default=1, const [1,307]}
       foot_bk1_l_ctrl.Scale.Y {type=MovieSceneScriptingFloatChannel, default=1, const [1,307]}
       foot_bk1_l_ctrl.Scale.Z {type=MovieSceneScriptingFloatChannel, default=1, const [1,307]}
       foot_roll_l_ctrl.Location.X {type=MovieSceneScriptingFloatChannel, default=0, const [0,307]}
       foot_roll_l_ctrl.Location.Y {type=MovieSceneScriptingFloatChannel, default=0, const [0,307]}
       foot_roll_l_ctrl.Location.Z {type=MovieSceneScriptingFloatChannel, default=0, const [0,307]}
       foot_roll_l_ctrl.Rotation.X {type=MovieSceneScriptingFloatChannel, default=0, const [0,307]}
       foot_roll_l_ctrl.Rotation.Y {type=MovieSceneScriptingFloatChannel, default=0, const [0,307]}
       foot_roll_l_ctrl.Rotation.Z {type=MovieSceneScriptingFloatChannel, default=0, const [0,307]}
       foot_roll_l_ctrl.Scale.X {type=MovieSceneScriptingFloatChannel, default=1, const [1,307]}
       foot_roll_l_ctrl.Scale.Y {type=MovieSceneScriptingFloatChannel, default=1, const [1,307]}
       foot_roll_l_ctrl.Scale.Z {type=MovieSceneScriptingFloatChannel, default=1, const [1,307]}
       roll_blend_l_ctrl {type=MovieSceneScriptingFloatChannel, default=1, const [1,307]}
       tip_l_ctrl.Location.X {type=MovieSceneScriptingFloatChannel, default=0, const [0,307]}
       tip_l_ctrl.Location.Y {type=MovieSceneScriptingFloatChannel, default=0, const [0,307]}
       tip_l_ctrl.Location.Z {type=MovieSceneScriptingFloatChannel, default=0, const [0,307]}
       tip_l_ctrl.Rotation.X {type=MovieSceneScriptingFloatChannel, default=0, const [0,307]}
       tip_l_ctrl.Rotation.Y {type=MovieSceneScriptingFloatChannel, default=0, const [0,307]}
       tip_l_ctrl.Rotation.Z {type=MovieSceneScriptingFloatChannel, default=0, const [0,307]}
       tip_l_ctrl.Scale.X {type=MovieSceneScriptingFloatChannel, default=1, const [1,307]}
       tip_l_ctrl.Scale.Y {type=MovieSceneScriptingFloatChannel, default=1, const [1,307]}
       tip_l_ctrl.Scale.Z {type=MovieSceneScriptingFloatChannel, default=1, const [1,307]}
       heel_l_ctrl.Location.X {type=MovieSceneScriptingFloatChannel, default=0, const [0,307]}
       heel_l_ctrl.Location.Y {type=MovieSceneScriptingFloatChannel, default=0, const [0,307]}
       heel_l_ctrl.Location.Z {type=MovieSceneScriptingFloatChannel, default=0, const [0,307]}
       heel_l_ctrl.Rotation.X {type=MovieSceneScriptingFloatChannel, default=0, const [0,307]}
       heel_l_ctrl.Rotation.Y {type=MovieSceneScriptingFloatChannel, default=0, const [0,307]}
       heel_l_ctrl.Rotation.Z {type=MovieSceneScriptingFloatChannel, default=0, const [0,307]}
       heel_l_ctrl.Scale.X {type=MovieSceneScriptingFloatChannel, default=1, const [1,307]}
       heel_l_ctrl.Scale.Y {type=MovieSceneScriptingFloatChannel, default=1, const [1,307]}
       heel_l_ctrl.Scale.Z {type=MovieSceneScriptingFloatChannel, default=1, const [1,307]}
       leg_l_softness {type=MovieSceneScriptingFloatChannel, default=0, const [0,307]}
       leg_l_pv_ik_ctrl.Location.X {
        type=MovieSceneScriptingFloatChannel
        default=0
        thinned_from=307
        keys [[0,-0.4491],[4,-0.4493],[9,-0.457],[13,-0.4658],[14,-0.4651],[15,-0.4613],[16,-0.4549],[17,-0.4454],[18,-0.4331],[19,-0.4181],[20,-0.4041],[21,-0.3916],[22,-0.3805],[23,-0.373],[24,-0.3696],[27,-0.372],[28,-0.3746],[29,-0.3749],[30,-0.3739],[31,-0.3709],[33,-0.3621],[37,-0.3482],[39,-0.3397],[42,-0.324],[44,-0.3109],[46,-0.2972],[49,-0.2757],[50,-0.2697],[52,-0.2613],[53,-0.2583],[54,-0.2558],[55,-0.2543],[59,-0.2527],[60,-0.2534],[61,-0.256],[62,-0.2613],[63,-0.2696],[64,-0.2813],[65,-0.2972],[66,-0.319],[67,-0.3505],[68,-0.397],[69,-0.4501],[70,-0.4945],[71,-0.5324],[72,-0.5632],[73,-0.5833],[74,-0.5955],[75,-0.5992],[76,-0.5946],[77,-0.581],[78,-0.5649],[79,-0.5566],[80,-0.5555],[82,-0.5573],[83,-0.56],[84,-0.5641],[85,-0.5697],[90,-0.6081],[91,-0.6139],[92,-0.6178],[94,-0.6222],[98,-0.6258],[100,-0.6252],[101,-0.6266],[107,-0.6215],[109,-0.6182],[111,-0.6158],[112,-0.6173],[117,-0.618],[118,-0.6194],[119,-0.6196],[120,-0.6217],[121,-0.6216],[130,-0.6295],[133,-0.6302],[134,-0.6317],[135,-0.6301],[136,-0.6316],[139,-0.6316],[140,-0.628],[144,-0.6232],[146,-0.6173],[149,-0.5989],[150,-0.5947],[151,-0.5925],[152,-0.5919],[153,-0.5932],[154,-0.5971],[155,-0.6048],[156,-0.6202],[158,-0.6596],[159,-0.6758],[160,-0.6893],[161,-0.7011],[162,-0.7113],[163,-0.7192],[164,-0.7234],[165,-0.7227],[166,-0.7189],[168,-0.7073],[170,-0.6998],[172,-0.6983],[174,-0.702],[183,-0.7553],[186,-0.7787],[188,-0.7865],[190,-0.792],[193,-0.795],[195,-0.7954],[196,-0.7974],[199,-0.7999],[201,-0.7987],[202,-0.8035],[203,-0.8161],[204,-0.8413],[206,-0.91],[207,-0.9458],[208,-0.9793],[209,-1.0067],[210,-1.028],[211,-1.0433],[212,-1.0542],[213,-1.0616],[215,-1.0692],[227,-1.0713],[233,-1.0804],[234,-1.0803],[235,-1.0773],[236,-1.0669],[237,-1.0432],[238,-1.0097],[239,-0.972],[240,-0.9308],[241,-0.8813],[243,-0.7701],[244,-0.7198],[245,-0.6714],[246,-0.6255],[247,-0.5846],[248,-0.5521],[249,-0.5269],[250,-0.5077],[251,-0.4943],[252,-0.4853],[253,-0.4805],[254,-0.4783],[256,-0.4781],[257,-0.4794],[259,-0.4846],[260,-0.486],[265,-0.4856],[267,-0.483],[269,-0.4775],[271,-0.4699],[275,-0.4492],[276,-0.4446],[278,-0.4385],[285,-0.4263],[286,-0.4254],[287,-0.4253],[288,-0.4275],[289,-0.4285],[293,-0.4392],[295,-0.4426],[298,-0.4437],[299,-0.4464],[300,-0.4451],[302,-0.4451],[303,-0.4469],[305,-0.4477],[306,0]]
       }
       leg_l_pv_ik_ctrl.Location.Y {
        type=MovieSceneScriptingFloatChannel
        default=0
        thinned_from=307
        keys [[0,2.8732],[1,2.8732],[2,2.8661],[4,2.8661],[6,2.8944],[7,2.9179],[8,2.9567],[9,3.013],[12,3.2267],[13,3.3059],[14,3.3628],[15,3.4057],[16,3.4404],[18,3.483],[20,3.5139],[22,3.5675],[23,3.6051],[24,3.6621],[27,3.864],[28,3.9261],[29,3.9689],[30,3.9881],[31,3.988],[33,3.9505],[36,3.8693],[38,3.7891],[40,3.6893],[42,3.5669],[49,3.0546],[50,2.9895],[51,2.9352],[52,2.8869],[53,2.8468],[54,2.8126],[55,2.7855],[58,2.7288],[61,2.6555],[62,2.6438],[63,2.6487],[64,2.6714],[65,2.7119],[66,2.7726],[67,2.8802],[68,3.0691],[69,3.2795],[70,3.436],[71,3.545],[72,3.6104],[73,3.6107],[74,3.5563],[75,3.4532],[76,3.3113],[77,3.1114],[78,2.9048],[79,2.7865],[80,2.7538],[81,2.7423],[82,2.7423],[83,2.7568],[84,2.79],[85,2.8392],[86,2.9024],[88,3.0463],[89,3.1232],[90,3.1908],[91,3.245],[92,3.2858],[93,3.3111],[95,3.3427],[99,3.3674],[103,3.3729],[110,3.2916],[111,3.2842],[116,3.2861],[134,3.4454],[139,3.4515],[141,3.438],[144,3.4281],[145,3.4195],[146,3.4023],[148,3.3345],[149,3.3061],[150,3.2976],[151,3.3145],[152,3.3541],[153,3.4148],[154,3.4966],[155,3.6089],[156,3.7856],[157,3.9949],[158,4.1958],[159,4.356],[160,4.4856],[161,4.5961],[162,4.6877],[163,4.7585],[164,4.7899],[165,4.7749],[166,4.7312],[168,4.6093],[169,4.5572],[170,4.5159],[171,4.487],[172,4.4692],[174,4.461],[181,4.5854],[183,4.6476],[184,4.6953],[186,4.8056],[187,4.8373],[190,4.8855],[196,4.9055],[198,4.8914],[199,4.8764],[201,4.82],[202,4.8201],[203,4.8726],[204,5.002],[206,5.3474],[207,5.5119],[208,5.6531],[209,5.7483],[210,5.8072],[211,5.8302],[212,5.8261],[214,5.7782],[217,5.684],[222,5.6064],[225,5.598],[227,5.6177],[232,5.7602],[235,5.8631],[236,5.852],[237,5.7476],[238,5.5789],[239,5.3843],[240,5.1725],[241,4.9062],[243,4.3048],[244,4.0567],[245,3.8421],[246,3.6534],[247,3.5071],[248,3.4177],[249,3.3704],[250,3.351],[251,3.3574],[252,3.3841],[253,3.431],[258,3.7338],[259,3.7821],[260,3.8084],[262,3.8035],[264,3.783],[265,3.7665],[266,3.7421],[267,3.7089],[269,3.6219],[271,3.5199],[275,3.2729],[276,3.2154],[277,3.1651],[279,3.0809],[284,2.9101],[285,2.8797],[286,2.8558],[287,2.841],[289,2.8411],[293,2.8835],[295,2.8886],[297,2.8785],[299,2.8786],[302,2.8735],[303,2.8803],[305,2.8803],[306,0]]
       }
       leg_l_pv_ik_ctrl.Location.Z {
        type=MovieSceneScriptingFloatChannel
        default=0
        thinned_from=307
        keys [[0,2.9169],[1,2.9169],[2,2.909],[4,2.909],[6,2.94],[7,2.966],[8,3.0089],[9,3.0715],[12,3.3106],[13,3.3999],[14,3.4646],[15,3.5138],[16,3.5541],[18,3.6053],[20,3.6439],[22,3.7081],[23,3.7522],[24,3.8184],[27,4.0526],[28,4.1249],[29,4.175],[30,4.1977],[31,4.1979],[32,4.1808],[35,4.0957],[36,4.0611],[38,3.9685],[40,3.8539],[42,3.7141],[49,3.1386],[50,3.0667],[51,3.0067],[52,2.9535],[53,2.9094],[54,2.8718],[55,2.8421],[58,2.7797],[61,2.6989],[62,2.6854],[63,2.6898],[64,2.7134],[65,2.7561],[66,2.8204],[67,2.9358],[68,3.1409],[69,3.3718],[70,3.5446],[71,3.6651],[72,3.7368],[73,3.7349],[74,3.6711],[75,3.5527],[76,3.3918],[78,2.9392],[79,2.8093],[80,2.7733],[81,2.7606],[82,2.7605],[83,2.7761],[84,2.8123],[85,2.8661],[86,2.9352],[88,3.0938],[89,3.179],[90,3.2542],[91,3.3147],[92,3.3604],[93,3.3888],[95,3.4243],[99,3.4521],[103,3.4583],[110,3.367],[111,3.3589],[116,3.3607],[134,3.5403],[139,3.5473],[141,3.5323],[144,3.5215],[145,3.5119],[146,3.4927],[148,3.417],[149,3.3855],[150,3.3763],[151,3.3957],[152,3.4407],[153,3.5096],[154,3.6025],[155,3.7306],[156,3.9332],[158,4.4106],[159,4.5998],[160,4.7541],[161,4.8864],[162,4.9967],[163,5.0824],[164,5.1204],[165,5.1021],[166,5.049],[168,4.9018],[169,4.839],[170,4.7895],[171,4.7548],[172,4.7333],[173,4.7232],[174,4.723],[180,4.8429],[182,4.9023],[183,4.9431],[184,5.0004],[185,5.0724],[186,5.1336],[187,5.172],[189,5.2168],[193,5.2523],[196,5.2544],[198,5.237],[199,5.2183],[201,5.1491],[202,5.1487],[203,5.2118],[204,5.3691],[207,5.9994],[208,6.1767],[209,6.2966],[210,6.3707],[211,6.3989],[212,6.3924],[213,6.3673],[217,6.2068],[221,6.1205],[223,6.0963],[226,6.1044],[228,6.1483],[235,6.4382],[236,6.4248],[237,6.2919],[238,6.0783],[239,5.8343],[240,5.572],[241,5.2463],[242,4.8808],[243,4.5284],[244,4.2395],[245,3.9933],[246,3.7795],[247,3.6159],[248,3.5174],[249,3.4663],[250,3.4464],[251,3.4552],[252,3.4866],[253,3.5405],[258,3.8885],[259,3.9442],[260,3.9746],[262,3.969],[264,3.9452],[265,3.926],[266,3.8978],[267,3.8596],[269,3.7597],[271,3.6434],[275,3.3644],[276,3.2999],[277,3.2437],[279,3.1497],[284,2.9602],[285,2.9267],[286,2.9003],[287,2.884],[289,2.8837],[293,2.9294],[295,2.9347],[297,2.9234],[299,2.9232],[302,2.9177],[303,2.925],[305,2.9249],[306,0]]
       }
       leg_l_pv_ik_ctrl.Rotation.X {
        type=MovieSceneScriptingFloatChannel
        default=0
        thinned_from=307
        keys [[0,-3.9895],[1,-3.9895],[2,-3.9791],[4,-3.9791],[6,-4.0201],[7,-4.0544],[8,-4.1111],[9,-4.1936],[12,-4.5081],[13,-4.6251],[14,-4.7097],[15,-4.774],[16,-4.8265],[18,-4.8932],[20,-4.9433],[22,-5.0267],[23,-5.0841],[24,-5.17],[27,-5.4733],[28,-5.5667],[29,-5.6313],[30,-5.6605],[31,-5.6608],[32,-5.6387],[35,-5.5288],[36,-5.4841],[38,-5.3644],[40,-5.2159],[42,-5.0342],[49,-4.2809],[50,-4.1861],[51,-4.107],[52,-4.0367],[53,-3.9784],[54,-3.9287],[55,-3.8893],[58,-3.8067],[61,-3.6994],[62,-3.6815],[63,-3.6874],[64,-3.7188],[65,-3.7756],[66,-3.861],[67,-4.0139],[68,-4.2847],[69,-4.5881],[70,-4.8144],[71,-4.9717],[72,-5.0652],[73,-5.0628],[74,-4.9799],[75,-4.8255],[76,-4.6153],[78,-4.0198],[79,-3.8477],[80,-3.8001],[81,-3.7832],[82,-3.7831],[83,-3.8038],[84,-3.8518],[85,-3.923],[86,-4.0146],[88,-4.2239],[89,-4.336],[90,-4.435],[91,-4.5143],[92,-4.5743],[93,-4.6115],[95,-4.6579],[99,-4.6944],[103,-4.7025],[110,-4.583],[111,-4.5722],[116,-4.5747],[134,-4.8095],[139,-4.8187],[141,-4.7991],[144,-4.785],[145,-4.7724],[146,-4.7473],[148,-4.6482],[149,-4.607],[150,-4.5949],[151,-4.6203],[152,-4.6792],[153,-4.7692],[154,-4.8906],[155,-5.0573],[156,-5.3203],[158,-5.936],[159,-6.1784],[160,-6.3755],[161,-6.5441],[162,-6.6842],[163,-6.7929],[164,-6.8411],[165,-6.8179],[166,-6.7506],[168,-6.5636],[169,-6.4837],[170,-6.4207],[171,-6.3765],[172,-6.349],[174,-6.3359],[180,-6.4888],[182,-6.5645],[183,-6.6164],[184,-6.6892],[185,-6.7806],[186,-6.8581],[187,-6.9067],[189,-6.9634],[193,-7.0083],[196,-7.011],[198,-6.989],[199,-6.9654],[201,-6.8779],[202,-6.8773],[203,-6.9573],[204,-7.156],[207,-7.9467],[208,-8.1675],[209,-8.3166],[210,-8.4086],[211,-8.4435],[212,-8.4356],[214,-8.357],[217,-8.2054],[222,-8.0814],[225,-8.0681],[227,-8.0995],[232,-8.3266],[235,-8.4924],[236,-8.4758],[237,-8.3109],[238,-8.0453],[239,-7.7407],[240,-7.4118],[241,-7.0013],[242,-6.5377],[243,-6.0876],[244,-5.7164],[245,-5.3985],[246,-5.1211],[247,-4.908],[248,-4.7792],[249,-4.7123],[250,-4.6861],[251,-4.6976],[252,-4.7385],[253,-4.8089],[258,-5.2616],[259,-5.3337],[260,-5.3731],[262,-5.3658],[264,-5.3351],[265,-5.3102],[266,-5.2737],[267,-5.2241],[269,-5.0944],[271,-4.943],[275,-4.5784],[276,-4.4939],[277,-4.4201],[279,-4.2965],[284,-4.0466],[285,-4.0023],[286,-3.9674],[287,-3.9458],[289,-3.9455],[293,-4.006],[295,-4.0131],[297,-3.9981],[299,-3.9978],[302,-3.9905],[303,-4.0002],[305,-4.0001],[306,0]]
       }
       leg_l_pv_ik_ctrl.Rotation.Y {
        type=MovieSceneScriptingFloatChannel
        default=0
        thinned_from=307
        keys [[0,0.0105],[4,0.0104],[6,0.0107],[7,0.0111],[8,0.0117],[9,0.0127],[11,0.015],[13,0.0178],[20,0.0256],[23,0.0293],[27,0.035],[28,0.0362],[29,0.0371],[30,0.0377],[31,0.0378],[33,0.0375],[36,0.0362],[39,0.0341],[42,0.0313],[49,0.0239],[51,0.0223],[54,0.0206],[59,0.0189],[63,0.017],[66,0.0163],[67,0.0164],[68,0.0172],[69,0.0182],[70,0.0188],[71,0.0189],[72,0.0185],[73,0.0174],[74,0.0155],[75,0.0131],[76,0.0105],[78,0.0044],[79,0.0027],[80,0.0022],[82,0.0019],[83,0.002],[84,0.0023],[86,0.0036],[90,0.0073],[92,0.0087],[95,0.0095],[103,0.0099],[110,0.0088],[116,0.0087],[136,0.0112],[145,0.0112],[149,0.0102],[150,0.0102],[151,0.0107],[152,0.0115],[153,0.0127],[154,0.0142],[155,0.0161],[156,0.0191],[158,0.0262],[159,0.0292],[160,0.0316],[161,0.0338],[162,0.0356],[163,0.037],[164,0.0376],[165,0.0372],[166,0.0363],[169,0.0327],[171,0.0312],[173,0.0304],[176,0.0304],[181,0.0312],[183,0.032],[184,0.0328],[186,0.0349],[188,0.0358],[191,0.0365],[195,0.0366],[198,0.036],[199,0.0356],[201,0.0341],[202,0.0338],[203,0.0345],[204,0.0367],[207,0.0455],[208,0.0479],[209,0.0494],[210,0.0501],[211,0.0499],[212,0.0492],[214,0.047],[216,0.0446],[218,0.043],[221,0.0417],[223,0.0412],[225,0.0412],[227,0.0417],[228,0.0422],[231,0.0447],[235,0.0491],[236,0.0493],[237,0.0473],[238,0.044],[240,0.0364],[241,0.0318],[242,0.0269],[243,0.0225],[244,0.0195],[245,0.0174],[246,0.0159],[247,0.0151],[248,0.0151],[249,0.0156],[251,0.0172],[253,0.0195],[256,0.0236],[258,0.0258],[259,0.0267],[260,0.0272],[262,0.0272],[265,0.0263],[267,0.0252],[272,0.0211],[276,0.0173],[281,0.014],[286,0.0116],[289,0.0111],[294,0.0113],[305,0.0107],[306,0]]
       }
       leg_l_pv_ik_ctrl.Rotation.Z {
        type=MovieSceneScriptingFloatChannel
        default=0
        thinned_from=307
        keys [[0,0.6038],[4,0.6041],[9,0.6141],[13,0.6261],[14,0.6253],[15,0.6207],[16,0.6128],[17,0.6009],[18,0.5855],[19,0.5668],[20,0.5492],[21,0.5337],[22,0.52],[23,0.5108],[24,0.5069],[27,0.5112],[28,0.5148],[29,0.5155],[30,0.5143],[31,0.5106],[34,0.4945],[36,0.4862],[39,0.4701],[42,0.4492],[46,0.4134],[49,0.3847],[50,0.3766],[52,0.3651],[53,0.361],[54,0.3576],[55,0.3555],[59,0.3528],[60,0.3535],[61,0.3567],[62,0.3635],[63,0.374],[64,0.3892],[65,0.4097],[66,0.4378],[67,0.4785],[68,0.5384],[69,0.6061],[70,0.6625],[71,0.7104],[72,0.749],[73,0.7741],[74,0.7894],[75,0.7941],[76,0.7884],[77,0.7712],[78,0.751],[79,0.7406],[80,0.7392],[82,0.7415],[83,0.7448],[84,0.7501],[85,0.7571],[90,0.8055],[91,0.8127],[93,0.821],[96,0.8257],[98,0.8276],[100,0.8268],[101,0.8285],[107,0.8222],[109,0.8181],[111,0.8151],[112,0.817],[117,0.8178],[118,0.8196],[119,0.8198],[120,0.8224],[121,0.8223],[130,0.8321],[133,0.8331],[134,0.8349],[135,0.8328],[136,0.8347],[139,0.8347],[140,0.8302],[144,0.8242],[146,0.8168],[149,0.7938],[150,0.7885],[151,0.7857],[152,0.7849],[153,0.7866],[154,0.7914],[155,0.801],[156,0.8203],[158,0.8693],[159,0.8892],[160,0.9059],[161,0.9203],[162,0.9328],[163,0.9425],[164,0.9476],[165,0.9467],[166,0.9422],[168,0.928],[170,0.9188],[172,0.9169],[174,0.9215],[183,0.9868],[186,1.0151],[188,1.0246],[190,1.0312],[193,1.0349],[195,1.0353],[199,1.0409],[201,1.0396],[202,1.0455],[203,1.0607],[204,1.0911],[207,1.2159],[208,1.2556],[209,1.288],[210,1.3133],[211,1.3315],[212,1.3445],[213,1.3536],[215,1.363],[227,1.3664],[233,1.3762],[234,1.3757],[235,1.372],[236,1.3596],[237,1.3318],[238,1.2925],[239,1.2481],[240,1.1993],[241,1.1401],[243,1.0057],[244,0.9439],[245,0.8842],[246,0.8269],[247,0.7756],[248,0.7348],[249,0.7032],[250,0.679],[251,0.662],[252,0.6508],[253,0.6448],[254,0.6423],[256,0.6425],[257,0.6443],[259,0.6511],[260,0.653],[265,0.6524],[267,0.6488],[269,0.6417],[271,0.6318],[276,0.599],[278,0.591],[285,0.5749],[287,0.5735],[288,0.5762],[289,0.5775],[293,0.5912],[295,0.5955],[298,0.597],[299,0.6004],[300,0.5987],[302,0.5987],[304,0.602],[305,0.602],[306,0]]
       }
       leg_l_pv_ik_ctrl.Scale.X {type=MovieSceneScriptingFloatChannel, default=1, const [1,307]}
       leg_l_pv_ik_ctrl.Scale.Y {type=MovieSceneScriptingFloatChannel, default=1, const [1,307]}
       leg_l_pv_ik_ctrl.Scale.Z {type=MovieSceneScriptingFloatChannel, default=1, const [1,307]}
       thigh_r_fk_ctrl.Location.X {type=MovieSceneScriptingFloatChannel, default=0, thinned_from=307, keys [[0,0],[306,0]]}
       thigh_r_fk_ctrl.Location.Y {type=MovieSceneScriptingFloatChannel, default=0, thinned_from=307, keys [[0,0],[306,0]]}
       thigh_r_fk_ctrl.Location.Z {type=MovieSceneScriptingFloatChannel, default=0, thinned_from=307, keys [[0,0],[306,0]]}
       thigh_r_fk_ctrl.Rotation.X {
        type=MovieSceneScriptingFloatChannel
        default=1.2139
        thinned_from=307
        keys [[0,29.5781],[6,29.6326],[8,29.53],[10,29.2923],[12,28.9306],[13,28.6943],[14,28.3318],[15,27.7929],[16,27.091],[17,26.2041],[18,25.0732],[20,22.4052],[21,21.1755],[22,19.9971],[23,18.9954],[24,18.244],[25,17.6185],[26,17.0588],[27,16.5598],[28,16.1129],[29,15.7114],[30,15.3476],[31,15.0226],[32,14.7414],[33,14.5086],[34,14.3289],[36,14.0947],[38,13.946],[40,13.8665],[46,13.8802],[51,14.0443],[56,14.3815],[59,14.6674],[61,14.9276],[63,15.267],[64,15.5045],[65,15.8153],[66,16.216],[67,16.7229],[68,17.4558],[69,18.471],[70,19.6516],[71,21.0365],[73,23.9719],[74,25.4899],[75,26.835],[76,27.8125],[77,28.5289],[78,29.0692],[79,29.4324],[80,29.6096],[108,31.5394],[145,31.2095],[148,30.9568],[152,30.3097],[156,29.4694],[158,29.1806],[161,28.9835],[165,28.9746],[168,29.1937],[170,29.4871],[172,29.9076],[174,30.4774],[176,31.2591],[181,33.4571],[183,34.1208],[185,34.5523],[188,34.918],[195,35.3139],[197,35.5488],[199,35.9612],[200,36.2523],[201,36.6413],[202,37.1628],[203,37.8444],[204,38.692],[206,40.6945],[210,44.9928],[211,45.9502],[212,46.7664],[213,47.437],[215,48.4952],[217,49.2786],[219,49.7747],[222,50.0752],[226,50.018],[228,49.7431],[231,48.9228],[232,48.5253],[233,48.028],[235,46.7482],[237,45.2154],[239,43.4131],[240,42.3614],[241,41.1329],[243,38.3414],[244,36.8728],[246,33.6075],[247,32.0628],[248,30.7883],[249,29.7852],[250,28.9648],[251,28.2999],[252,27.76],[254,26.8747],[256,26.1609],[258,25.6139],[260,25.2264],[263,24.8467],[267,24.604],[271,24.5557],[274,24.677],[276,24.8797],[278,25.2114],[291,28.3404],[294,28.881],[298,29.3263],[304,29.569],[305,29.569],[306,0]]
       }
       thigh_r_fk_ctrl.Rotation.Y {
        type=MovieSceneScriptingFloatChannel
        default=-1.213
        thinned_from=307
        keys [[0,1.0376],[1,1.0433],[2,1.058],[3,1.0694],[4,1.0694],[5,1.0523],[6,1.0203],[7,0.9767],[8,0.9236],[9,0.8398],[10,0.7062],[11,0.5204],[12,0.2691],[13,-0.0739],[14,-0.4674],[15,-0.8665],[16,-1.2598],[17,-1.6407],[18,-1.9792],[19,-2.2668],[21,-2.7984],[23,-3.309],[25,-3.8445],[26,-4.1028],[27,-4.3495],[28,-4.5761],[29,-4.7477],[30,-4.8519],[31,-4.9068],[32,-4.9281],[36,-4.9432],[37,-4.9322],[38,-4.907],[39,-4.8657],[40,-4.8076],[41,-4.7296],[42,-4.6336],[49,-3.855],[50,-3.7583],[52,-3.5993],[53,-3.5284],[54,-3.4663],[55,-3.4121],[57,-3.3273],[58,-3.2727],[59,-3.2059],[61,-3.0498],[62,-2.9621],[63,-2.8608],[64,-2.7299],[65,-2.5534],[66,-2.3103],[67,-1.9938],[68,-1.6257],[69,-1.2128],[70,-0.755],[71,-0.2582],[72,0.2616],[73,0.8363],[74,1.5007],[75,2.1866],[76,2.854],[77,3.5687],[78,4.2126],[79,4.5842],[80,4.7174],[81,4.7808],[82,4.8025],[83,4.7976],[84,4.7448],[85,4.6601],[86,4.5417],[87,4.4004],[88,4.248],[89,4.0843],[90,3.9448],[91,3.8379],[92,3.762],[93,3.722],[94,3.6983],[95,3.6838],[97,3.6751],[102,3.6943],[103,3.7148],[105,3.7811],[109,3.948],[110,3.9809],[112,4.0204],[113,4.0204],[114,4.0401],[115,4.0401],[116,4.0541],[117,4.0541],[118,4.0635],[129,4.0744],[134,4.1015],[139,4.1059],[141,4.0901],[145,4.0098],[146,4.0038],[147,4.0301],[148,4.0763],[150,4.1831],[154,4.339],[155,4.3845],[156,4.3856],[157,4.3424],[158,4.2629],[159,4.1522],[160,3.9914],[161,3.7887],[162,3.5767],[164,3.1992],[166,2.7744],[167,2.5684],[168,2.3894],[169,2.2448],[170,2.1307],[171,2.043],[172,1.9775],[173,1.9301],[174,1.8966],[175,1.8735],[177,1.841],[180,1.8082],[181,1.8027],[182,1.8034],[183,1.8126],[184,1.8369],[185,1.874],[186,1.9058],[187,1.9203],[191,1.9513],[194,1.9669],[195,1.9807],[196,2.0094],[197,2.0592],[198,2.1347],[199,2.2592],[200,2.4546],[201,2.7153],[202,3.0494],[203,3.4715],[204,3.976],[205,4.5444],[206,5.1578],[208,6.4372],[209,7.0336],[210,7.5821],[211,8.0635],[212,8.4737],[213,8.8126],[214,9.0657],[215,9.2315],[216,9.3303],[217,9.3805],[219,9.3989],[222,9.3627],[224,9.3011],[226,9.1887],[227,9.1061],[228,9.0052],[230,8.7514],[231,8.5973],[233,8.2337],[234,8.0222],[235,7.7679],[236,7.4318],[237,6.9782],[238,6.4504],[239,5.8925],[240,5.3187],[241,4.6978],[242,4.056],[243,3.4433],[244,2.8994],[245,2.4107],[246,1.9676],[247,1.5754],[248,1.2355],[249,0.9212],[250,0.6183],[251,0.3285],[252,0.0539],[253,-0.2005],[254,-0.432],[255,-0.6418],[256,-0.832],[257,-0.9958],[258,-1.1251],[259,-1.2228],[260,-1.2929],[261,-1.3435],[262,-1.3831],[263,-1.4144],[264,-1.4369],[265,-1.4525],[266,-1.4613],[267,-1.4649],[268,-1.4621],[269,-1.4504],[270,-1.4282],[271,-1.3937],[272,-1.3482],[273,-1.2928],[274,-1.2285],[275,-1.1569],[276,-1.08],[277,-0.9959],[278,-0.904],[279,-0.8048],[280,-0.6989],[281,-0.5891],[282,-0.4751],[283,-0.357],[284,-0.2331],[285,-0.1031],[286,0.0326],[287,0.1719],[288,0.3135],[289,0.4527],[290,0.5859],[291,0.7086],[292,0.8167],[293,0.9058],[294,0.9728],[295,1.0196],[296,1.0545],[297,1.0801],[298,1.0971],[299,1.1061],[300,1.1075],[301,1.1011],[304,1.0625],[305,1.0625],[306,0]]
       }
       thigh_r_fk_ctrl.Rotation.Z {
        type=MovieSceneScriptingFloatChannel
        default=-0.1592
        thinned_from=307
        keys [[0,4.5299],[3,4.5021],[4,4.5021],[5,4.4906],[6,4.4958],[7,4.5138],[8,4.5515],[9,4.6247],[10,4.7314],[11,4.8821],[12,5.0849],[13,5.352],[14,5.646],[15,5.9157],[16,6.1523],[17,6.3429],[18,6.432],[19,6.4196],[20,6.3877],[22,6.3694],[23,6.411],[24,6.5232],[26,6.8267],[28,7.1534],[29,7.2696],[30,7.3109],[31,7.3],[33,7.2103],[34,7.1789],[36,7.1587],[37,7.1377],[38,7.1022],[39,7.0495],[40,6.9707],[41,6.8612],[44,6.4719],[49,5.9186],[50,5.8305],[53,5.6166],[55,5.5048],[57,5.4236],[58,5.367],[59,5.2838],[60,5.1639],[61,5.0003],[62,4.791],[63,4.5392],[64,4.2243],[65,3.8281],[66,3.3332],[67,2.7007],[68,1.9664],[69,1.2958],[70,0.7825],[71,0.4386],[72,0.2019],[73,0.0449],[74,-0.0467],[75,-0.1227],[76,-0.2521],[77,-0.4392],[78,-0.6127],[79,-0.7135],[80,-0.7715],[81,-0.8116],[82,-0.8231],[83,-0.8131],[84,-0.7757],[85,-0.7178],[86,-0.6412],[87,-0.5522],[88,-0.4586],[89,-0.359],[90,-0.2697],[91,-0.1928],[92,-0.1288],[93,-0.0794],[94,-0.0379],[95,0.0002],[96,0.0349],[97,0.0674],[98,0.0984],[99,0.1281],[100,0.1565],[101,0.1836],[102,0.2087],[103,0.228],[104,0.2406],[105,0.2516],[106,0.2568],[107,0.2575],[108,0.2576],[110,0.2546],[111,0.2537],[112,0.2512],[113,0.2512],[114,0.2454],[115,0.2454],[116,0.2333],[117,0.2333],[118,0.2176],[119,0.2176],[120,0.1984],[121,0.1862],[122,0.1753],[124,0.15],[126,0.12],[127,0.1041],[128,0.0892],[129,0.075],[130,0.0616],[131,0.0517],[132,0.0434],[133,0.0372],[134,0.0304],[135,0.0304],[136,0.0256],[139,0.0256],[140,0.0438],[141,0.0547],[142,0.0698],[143,0.0872],[144,0.1039],[145,0.1121],[146,0.118],[147,0.1196],[148,0.127],[149,0.1402],[150,0.1643],[151,0.2065],[152,0.2739],[154,0.4412],[155,0.5095],[156,0.5516],[157,0.5709],[158,0.5693],[159,0.5552],[160,0.5662],[161,0.6124],[162,0.6719],[163,0.7256],[164,0.8041],[165,0.9277],[166,1.0857],[167,1.2626],[168,1.4365],[169,1.5965],[170,1.7397],[171,1.8694],[172,1.9898],[173,2.106],[174,2.2296],[175,2.3612],[176,2.5027],[179,2.9452],[180,3.0864],[181,3.2103],[182,3.3123],[183,3.3822],[184,3.4122],[186,3.4049],[193,3.4612],[195,3.4986],[196,3.527],[197,3.5642],[198,3.6109],[199,3.7064],[200,3.8909],[201,4.1264],[202,4.3474],[203,4.5358],[204,4.6173],[205,4.5959],[206,4.5491],[208,4.3983],[209,4.3646],[210,4.3595],[211,4.3792],[212,4.3868],[213,4.3772],[214,4.3767],[215,4.4006],[216,4.4444],[218,4.5711],[219,4.6229],[220,4.6622],[222,4.712],[225,4.7512],[226,4.7533],[227,4.7453],[228,4.7192],[229,4.6804],[231,4.575],[232,4.5059],[233,4.4064],[234,4.2869],[235,4.1877],[236,4.153],[237,4.2113],[238,4.3212],[241,4.6992],[242,4.8379],[243,4.9578],[244,5.0203],[245,4.9997],[246,4.9252],[247,4.831],[248,4.753],[249,4.7119],[250,4.6952],[253,4.6965],[254,4.6758],[256,4.6074],[257,4.5625],[260,4.3653],[262,4.2731],[264,4.1982],[265,4.1687],[266,4.148],[267,4.1365],[272,4.1314],[275,4.1552],[278,4.2],[285,4.3459],[287,4.3684],[292,4.3306],[293,4.3319],[294,4.3424],[302,4.5096],[304,4.5385],[305,4.5385],[306,0]]
       }
       thigh_r_fk_ctrl.Scale.X {type=MovieSceneScriptingFloatChannel, default=1, const [1,307]}
       thigh_r_fk_ctrl.Scale.Y {type=MovieSceneScriptingFloatChannel, default=1, const [1,307]}
       thigh_r_fk_ctrl.Scale.Z {type=MovieSceneScriptingFloatChannel, default=1, const [1,307]}
       calf_r_fk_ctrl.Location.X {
        type=MovieSceneScriptingFloatChannel
        default=0
        thinned_from=307
        keys [[0,0.0399],[10,0.0399],[11,0.0342],[13,0.0341],[14,0.0285],[15,0.0285],[16,0.0228],[17,0.0205],[18,0.0205],[19,0.0134],[23,0.0134],[24,0.0063],[27,0.0063],[28,-0.0007],[44,-0.0007],[45,0.0082],[55,0.0082],[56,0.0135],[62,0.0136],[63,0.019],[64,0.019],[65,0.0244],[66,0.0244],[67,0.027],[68,0.027],[69,0.0365],[71,0.0365],[72,0.0458],[76,0.046],[77,0.0555],[82,0.0555],[98,0.0554],[99,0.05],[114,0.05],[115,0.0508],[130,0.0508],[131,0.0517],[146,0.0517],[154,0.0515],[155,0.0606],[162,0.0606],[163,0.0589],[165,0.0588],[166,0.0542],[167,0.0542],[168,0.0495],[172,0.0495],[173,0.0447],[178,0.0447],[179,0.0442],[184,0.0442],[185,0.0467],[194,0.0467],[195,0.0462],[202,0.0462],[203,0.0532],[205,0.0532],[206,0.0602],[208,0.0602],[209,0.0672],[210,0.0672],[211,0.0685],[226,0.0685],[227,0.0701],[236,0.0701],[237,0.0631],[239,0.063],[240,0.0562],[241,0.0562],[242,0.0493],[243,0.0462],[244,0.0415],[245,0.0415],[246,0.0367],[250,0.0367],[251,0.032],[258,0.032],[259,0.0327],[274,0.0327],[275,0.0315],[284,0.0315],[285,0.038],[290,0.038],[291,0.0383],[305,0.0383],[306,0]]
       }
       calf_r_fk_ctrl.Location.Y {
        type=MovieSceneScriptingFloatChannel
        default=0
        thinned_from=307
        keys [[0,0.0021],[9,0.0021],[10,0.001],[11,0.0015],[12,0.0015],[13,0.0004],[14,0.0009],[15,0.0009],[16,0.0003],[17,0.0002],[18,-0.0004],[19,0.0001],[21,0.0001],[22,-0.0005],[23,-0.0005],[24,0],[26,0],[27,-0.0006],[28,0],[44,0],[55,-0.0003],[56,-0.0007],[57,0],[62,0],[63,0.0002],[64,0.0002],[65,-0.0002],[66,0.0004],[67,0.0006],[68,0.0006],[69,0.002],[71,0.002],[72,0.0013],[73,0.0035],[76,0.0035],[77,0.0049],[98,0.0048],[99,0.0037],[114,0.0037],[130,0.0039],[154,0.004],[155,0.0059],[162,0.0059],[163,0.0056],[164,0.0056],[165,0.0043],[166,0.0047],[167,0.0047],[168,0.0038],[172,0.0038],[173,0.0028],[178,0.0028],[184,0.0027],[185,0.0031],[202,0.003],[203,0.0046],[205,0.0046],[206,0.0062],[208,0.0062],[209,0.0077],[210,0.0077],[211,0.0081],[226,0.0081],[227,0.0085],[236,0.0085],[237,0.0069],[238,0.0069],[239,0.0048],[240,0.0053],[241,0.0053],[242,0.0037],[244,0.0024],[245,0.0024],[246,0.0017],[250,0.0017],[251,0.0011],[274,0.0011],[284,0.001],[285,0.0018],[305,0.0018],[306,0]]
       }
       calf_r_fk_ctrl.Location.Z {
        type=MovieSceneScriptingFloatChannel
        default=0
        thinned_from=307
        keys [[0,0],[22,0],[23,0.0001],[55,0],[56,0.0001],[64,0],[65,0.0001],[71,0],[72,0.0001],[73,-0.0001],[164,-0.0001],[165,0],[238,-0.0001],[239,0.0001],[284,0],[285,0.0001],[306,0]]
       }
       calf_r_fk_ctrl.Rotation.X {
        type=MovieSceneScriptingFloatChannel
        default=0.0561
        thinned_from=307
        keys [[0,-0.0123],[7,-0.0124],[10,-0.0117],[12,-0.0107],[17,-0.0061],[19,-0.0048],[28,-0.0004],[30,0.0001],[39,0.0001],[60,-0.004],[63,-0.0053],[65,-0.0066],[67,-0.0084],[70,-0.0115],[78,-0.0173],[79,-0.0177],[84,-0.0177],[93,-0.0162],[147,-0.0166],[153,-0.0177],[158,-0.0195],[160,-0.0196],[164,-0.0187],[169,-0.0161],[172,-0.0151],[181,-0.0141],[200,-0.0151],[202,-0.0157],[203,-0.0162],[205,-0.0177],[208,-0.0206],[210,-0.022],[212,-0.023],[215,-0.0236],[235,-0.0224],[236,-0.0221],[237,-0.0214],[240,-0.0184],[243,-0.0146],[245,-0.0128],[247,-0.0116],[253,-0.0102],[279,-0.0101],[294,-0.0124],[305,-0.0123],[306,0]]
       }
       calf_r_fk_ctrl.Rotation.Y {
        type=MovieSceneScriptingFloatChannel
        default=1.1982
        thinned_from=307
        keys [[0,-0.0085],[8,-0.0084],[11,-0.0079],[13,-0.007],[18,-0.0041],[29,0],[39,0.0001],[60,-0.0031],[64,-0.0044],[66,-0.0055],[70,-0.008],[78,-0.0111],[85,-0.0112],[151,-0.011],[159,-0.0121],[164,-0.0118],[172,-0.01],[184,-0.0095],[201,-0.0101],[204,-0.0109],[209,-0.0129],[215,-0.0138],[236,-0.0132],[240,-0.0116],[244,-0.0092],[247,-0.0081],[257,-0.0071],[287,-0.0078],[305,-0.0085],[306,0]]
       }
       calf_r_fk_ctrl.Rotation.Z {
        type=MovieSceneScriptingFloatChannel
        default=0.2616
        thinned_from=307
        keys [[0,10.0857],[4,10.1759],[5,10.1759],[6,10.1502],[7,10.0943],[8,9.9956],[9,9.8217],[10,9.5685],[11,9.2273],[12,8.7706],[13,8.1576],[14,7.4186],[16,5.9015],[17,5.1686],[18,4.5455],[19,4.0609],[20,3.6397],[21,3.2422],[22,2.8698],[23,2.487],[24,2.0648],[25,1.6256],[26,1.1956],[27,0.7697],[28,0.3676],[29,0.0561],[30,-0.1068],[31,-0.1672],[32,-0.1672],[33,-0.1396],[34,-0.1396],[35,-0.1572],[36,-0.1764],[37,-0.1764],[38,-0.138],[39,-0.0677],[40,0.0517],[41,0.2266],[42,0.4313],[43,0.6574],[44,0.875],[45,1.0712],[46,1.2504],[48,1.6366],[49,1.8222],[50,1.9844],[53,2.4044],[55,2.6498],[57,2.8582],[58,2.9923],[59,3.1697],[60,3.407],[61,3.7062],[62,4.0633],[63,4.4701],[64,4.9493],[65,5.5236],[66,6.2058],[67,7.0089],[68,7.8774],[69,8.7014],[70,9.4341],[71,10.0716],[72,10.6337],[76,12.7986],[77,13.3698],[78,13.8688],[79,14.1447],[80,14.2356],[81,14.2719],[82,14.2719],[83,14.2392],[85,14.0667],[88,13.6131],[90,13.2836],[91,13.1597],[92,13.066],[95,12.921],[101,12.8178],[105,12.847],[115,13.0505],[129,13.2532],[141,13.3041],[145,13.2647],[146,13.281],[147,13.3312],[149,13.5187],[151,13.7916],[153,14.1692],[154,14.4043],[157,15.2412],[158,15.453],[159,15.5577],[160,15.5434],[161,15.4482],[163,15.1498],[164,14.9228],[165,14.577],[166,14.1631],[167,13.7206],[168,13.3071],[169,12.9549],[170,12.6604],[171,12.4198],[172,12.2259],[173,12.068],[175,11.842],[178,11.5981],[180,11.4746],[181,11.4331],[183,11.4487],[184,11.5315],[186,11.7724],[187,11.832],[191,11.9278],[194,11.9278],[196,11.9453],[198,12.026],[199,12.1014],[200,12.2252],[201,12.4029],[202,12.6507],[203,12.9955],[204,13.5077],[205,14.1605],[206,14.8464],[208,16.2799],[209,16.865],[210,17.3422],[211,17.7135],[212,18.0028],[213,18.2259],[214,18.3829],[215,18.4701],[217,18.505],[229,18.0175],[234,17.6642],[235,17.5719],[236,17.3592],[237,16.8727],[238,16.2036],[239,15.4647],[240,14.6866],[241,13.7724],[242,12.7784],[243,11.8301],[244,11.0511],[245,10.4189],[246,9.9076],[247,9.5243],[248,9.2726],[250,8.8786],[252,8.5504],[253,8.4301],[254,8.3572],[256,8.2734],[257,8.2588],[260,8.3722],[263,8.3722],[266,8.317],[270,8.1512],[271,8.1305],[272,8.1305],[273,8.1097],[275,8.1253],[277,8.191],[279,8.311],[282,8.5659],[285,8.8767],[286,9.0017],[288,9.3077],[290,9.6467],[292,9.9355],[293,10.0421],[294,10.1145],[296,10.1688],[300,10.1688],[305,10.0622],[306,0]]
       }
       calf_r_fk_ctrl.Scale.X {type=MovieSceneScriptingFloatChannel, default=1, const [1,307]}
       calf_r_fk_ctrl.Scale.Y {type=MovieSceneScriptingFloatChannel, default=1, const [1,307]}
       calf_r_fk_ctrl.Scale.Z {type=MovieSceneScriptingFloatChannel, default=1, const [1,307]}
       foot_r_fk_ctrl.Location.X {type=MovieSceneScriptingFloatChannel, default=0, thinned_from=307, keys [[0,0],[306,0]]}
       foot_r_fk_ctrl.Location.Y {type=MovieSceneScriptingFloatChannel, default=-0.0329, thinned_from=307, keys [[0,0],[306,0]]}
       foot_r_fk_ctrl.Location.Z {type=MovieSceneScriptingFloatChannel, default=-0.9075, thinned_from=307, keys [[0,0],[306,0]]}
       foot_r_fk_ctrl.Rotation.X {
        type=MovieSceneScriptingFloatChannel
        default=-1.1648
        thinned_from=307
        keys [[0,0.0906],[1,0.0906],[2,0.0912],[6,0.0914],[7,0.0911],[8,0.0899],[10,0.0837],[12,0.0769],[14,0.0716],[15,0.0698],[16,0.0687],[17,0.0687],[18,0.0695],[19,0.0714],[21,0.0748],[22,0.0767],[23,0.0782],[25,0.0808],[26,0.0817],[27,0.0829],[30,0.086],[32,0.0882],[34,0.0901],[37,0.0934],[39,0.0953],[40,0.0965],[43,0.0992],[44,0.0999],[47,0.1013],[48,0.1014],[49,0.1018],[50,0.102],[52,0.1029],[53,0.1029],[54,0.1037],[57,0.1046],[59,0.1063],[60,0.108],[61,0.1102],[63,0.1171],[65,0.1247],[66,0.129],[68,0.1388],[69,0.1426],[70,0.145],[71,0.1456],[72,0.1445],[73,0.1418],[75,0.1335],[76,0.1297],[78,0.1238],[79,0.1224],[80,0.1218],[82,0.1218],[83,0.1207],[85,0.1196],[86,0.1187],[91,0.1148],[92,0.1137],[94,0.112],[95,0.112],[96,0.11],[97,0.11],[98,0.1085],[99,0.1085],[100,0.1071],[101,0.1071],[102,0.1057],[103,0.1057],[104,0.1044],[106,0.1034],[107,0.1031],[109,0.1022],[111,0.1021],[112,0.1016],[115,0.1015],[118,0.1013],[123,0.1014],[124,0.1019],[127,0.1019],[128,0.1025],[131,0.1025],[132,0.1014],[137,0.1014],[138,0.1001],[143,0.1],[144,0.0995],[148,0.1004],[149,0.1002],[150,0.1007],[151,0.1005],[154,0.0992],[155,0.099],[156,0.0986],[157,0.0985],[158,0.0981],[159,0.0981],[160,0.0979],[161,0.0979],[163,0.0975],[164,0.0979],[165,0.0976],[167,0.0974],[168,0.0967],[169,0.0963],[171,0.0945],[173,0.0918],[174,0.0903],[176,0.0868],[181,0.0769],[183,0.074],[185,0.0718],[188,0.0701],[189,0.0697],[190,0.0697],[191,0.0693],[194,0.0693],[195,0.0682],[196,0.0682],[197,0.0668],[198,0.0658],[199,0.0642],[200,0.0618],[201,0.0584],[202,0.0539],[203,0.0474],[204,0.0405],[205,0.0343],[206,0.0283],[207,0.0235],[208,0.0198],[209,0.0164],[210,0.0134],[211,0.0113],[212,0.0104],[213,0.0104],[216,0.0134],[217,0.0133],[218,0.013],[220,0.0129],[221,0.0126],[222,0.0126],[224,0.0133],[225,0.0133],[226,0.0136],[227,0.0142],[228,0.0154],[229,0.0162],[230,0.0174],[232,0.0207],[233,0.0228],[234,0.0258],[235,0.0278],[237,0.031],[238,0.0332],[240,0.0394],[241,0.0429],[242,0.0472],[243,0.052],[244,0.0563],[246,0.0666],[247,0.0711],[248,0.0746],[249,0.0774],[250,0.0797],[253,0.0857],[254,0.0879],[256,0.0918],[257,0.0941],[261,0.1015],[265,0.1079],[268,0.1116],[271,0.1142],[272,0.1142],[273,0.1149],[276,0.1148],[278,0.1136],[281,0.1107],[284,0.1067],[288,0.0998],[291,0.0951],[292,0.0942],[293,0.093],[295,0.0912],[296,0.0904],[297,0.0904],[298,0.0897],[301,0.0897],[302,0.089],[303,0.089],[304,0.0887],[305,0.0887],[306,0]]
       }
       foot_r_fk_ctrl.Rotation.Y {
        type=MovieSceneScriptingFloatChannel
        default=0.1067
        thinned_from=307
        keys [[0,-0.0896],[1,-0.0896],[2,-0.0833],[3,-0.081],[6,-0.081],[7,-0.0798],[8,-0.1003],[9,-0.136],[11,-0.2151],[12,-0.2541],[13,-0.2892],[14,-0.3159],[15,-0.3331],[16,-0.3422],[17,-0.3393],[18,-0.3235],[20,-0.2707],[21,-0.2451],[22,-0.2186],[23,-0.1945],[24,-0.1754],[26,-0.1397],[27,-0.1216],[28,-0.1048],[29,-0.0887],[30,-0.0731],[31,-0.0589],[32,-0.0451],[33,-0.0319],[34,-0.0193],[35,-0.0063],[36,0.0079],[37,0.0217],[38,0.0351],[39,0.0478],[40,0.0597],[41,0.0719],[42,0.0831],[43,0.0928],[44,0.1],[45,0.1055],[46,0.11],[47,0.1132],[49,0.1188],[50,0.122],[51,0.1255],[53,0.1319],[54,0.136],[56,0.1454],[57,0.1497],[58,0.1561],[59,0.1659],[60,0.1843],[61,0.2143],[62,0.2529],[63,0.294],[64,0.3384],[65,0.3869],[66,0.4378],[67,0.4943],[68,0.5536],[69,0.6003],[70,0.6265],[71,0.6284],[72,0.6107],[73,0.5724],[74,0.5155],[75,0.4555],[76,0.4059],[77,0.3621],[78,0.3248],[79,0.3023],[80,0.2946],[82,0.2946],[83,0.2803],[84,0.2743],[85,0.2668],[88,0.2412],[89,0.2322],[91,0.2105],[92,0.1989],[93,0.1868],[94,0.1752],[95,0.1752],[96,0.1524],[97,0.1524],[98,0.1306],[99,0.1306],[100,0.1116],[101,0.1116],[102,0.0946],[103,0.0946],[104,0.0802],[105,0.074],[106,0.0662],[107,0.0601],[108,0.0556],[109,0.0494],[110,0.0466],[111,0.0466],[112,0.0421],[113,0.0421],[114,0.0393],[115,0.0393],[116,0.0383],[118,0.0383],[119,0.039],[123,0.039],[124,0.0444],[127,0.0444],[128,0.0467],[131,0.0467],[132,0.0347],[137,0.0347],[138,0.023],[140,0.023],[141,0.0177],[143,0.0177],[144,0.0161],[145,0.017],[146,0.0194],[147,0.021],[148,0.0224],[149,0.0232],[150,0.0225],[151,0.0202],[152,0.0149],[153,0.0081],[154,0.0016],[155,-0.0044],[156,-0.0095],[158,-0.0192],[159,-0.0217],[160,-0.0238],[161,-0.0241],[162,-0.0236],[163,-0.0237],[164,-0.0222],[165,-0.0203],[166,-0.0191],[167,-0.0195],[168,-0.0225],[169,-0.029],[170,-0.0379],[171,-0.0489],[172,-0.0626],[173,-0.079],[174,-0.0978],[175,-0.1204],[176,-0.1444],[180,-0.2465],[181,-0.2704],[182,-0.2906],[183,-0.3088],[184,-0.3246],[185,-0.3365],[186,-0.3474],[187,-0.3549],[188,-0.3615],[189,-0.3667],[190,-0.3667],[191,-0.375],[194,-0.375],[195,-0.3874],[196,-0.3874],[197,-0.405],[198,-0.4185],[199,-0.4388],[200,-0.4707],[201,-0.516],[202,-0.5767],[203,-0.6636],[204,-0.7543],[205,-0.839],[206,-0.9201],[207,-0.9846],[208,-1.0361],[209,-1.0819],[210,-1.1224],[211,-1.1528],[212,-1.1646],[213,-1.1646],[215,-1.1359],[216,-1.1255],[217,-1.1254],[219,-1.1297],[222,-1.1289],[225,-1.1223],[227,-1.1109],[228,-1.0989],[229,-1.0833],[231,-1.0453],[232,-1.0229],[233,-0.9924],[234,-0.9593],[235,-0.9334],[236,-0.912],[237,-0.8877],[238,-0.8562],[239,-0.8198],[240,-0.7784],[241,-0.7236],[242,-0.6624],[243,-0.5976],[244,-0.5365],[245,-0.4655],[246,-0.3974],[247,-0.3344],[248,-0.2886],[249,-0.2517],[251,-0.1931],[252,-0.1663],[253,-0.1391],[254,-0.1138],[255,-0.0881],[256,-0.0601],[257,-0.0344],[258,-0.0106],[259,0.0155],[260,0.0374],[261,0.0623],[262,0.085],[264,0.1262],[265,0.1446],[266,0.1628],[267,0.1788],[269,0.2069],[270,0.2174],[271,0.2264],[272,0.2264],[273,0.2355],[274,0.2355],[275,0.2365],[276,0.2338],[277,0.2279],[278,0.2189],[279,0.2078],[280,0.194],[281,0.1792],[282,0.1628],[283,0.1452],[284,0.1255],[285,0.1041],[287,0.0569],[288,0.0339],[289,0.0125],[290,-0.0081],[291,-0.0274],[292,-0.0445],[293,-0.059],[294,-0.0707],[295,-0.0817],[296,-0.0897],[297,-0.0897],[298,-0.101],[301,-0.101],[302,-0.1087],[303,-0.1087],[304,-0.1095],[305,-0.1095],[306,0]]
       }
       foot_r_fk_ctrl.Rotation.Z {
        type=MovieSceneScriptingFloatChannel
        default=-0.1036
        thinned_from=307
        keys [[0,-4.0752],[1,-4.0752],[2,-4.1083],[3,-4.1273],[6,-4.1273],[7,-4.0938],[8,-4.0374],[9,-3.934],[10,-3.7878],[11,-3.5986],[12,-3.3472],[13,-3.0192],[14,-2.6144],[16,-1.7559],[17,-1.3229],[18,-0.9216],[19,-0.5694],[20,-0.2491],[21,0.0498],[22,0.3274],[23,0.5978],[24,0.8609],[25,1.1204],[26,1.3727],[27,1.6144],[28,1.8454],[29,2.0268],[30,2.1373],[31,2.1947],[32,2.2201],[33,2.2278],[34,2.2413],[36,2.2809],[37,2.2859],[38,2.278],[39,2.2487],[40,2.1939],[41,2.1135],[42,2.016],[43,1.9099],[44,1.8079],[45,1.7144],[46,1.6251],[49,1.3442],[50,1.2634],[54,0.9742],[56,0.8465],[57,0.7906],[58,0.7108],[59,0.6152],[60,0.4959],[61,0.3452],[62,0.1708],[63,-0.0194],[64,-0.2416],[65,-0.4956],[66,-0.7815],[67,-1.0956],[68,-1.4058],[70,-1.9973],[71,-2.2965],[72,-2.5863],[73,-2.8912],[74,-3.2318],[75,-3.5622],[77,-4.1557],[78,-4.4086],[79,-4.5491],[80,-4.5878],[82,-4.5878],[83,-4.574],[84,-4.5361],[85,-4.4874],[86,-4.4223],[89,-4.1889],[90,-4.1184],[91,-4.0644],[92,-4.0213],[93,-3.9946],[94,-3.9789],[95,-3.9789],[96,-3.9582],[103,-3.9652],[110,-4.085],[111,-4.085],[112,-4.0999],[113,-4.0999],[114,-4.1147],[115,-4.1147],[116,-4.1272],[118,-4.1272],[119,-4.1448],[123,-4.1448],[124,-4.1621],[127,-4.1621],[128,-4.1796],[137,-4.1832],[138,-4.1997],[140,-4.1997],[141,-4.2213],[143,-4.2213],[146,-4.2905],[147,-4.337],[148,-4.4007],[149,-4.475],[151,-4.6609],[152,-4.7727],[153,-4.9005],[154,-5.0389],[157,-5.4806],[158,-5.5818],[159,-5.6298],[160,-5.614],[161,-5.5608],[162,-5.4864],[163,-5.3992],[164,-5.2825],[165,-5.1011],[167,-4.6671],[168,-4.46],[169,-4.292],[170,-4.1565],[171,-4.047],[172,-3.9636],[173,-3.9061],[174,-3.8682],[175,-3.8435],[177,-3.8265],[180,-3.8388],[181,-3.8502],[182,-3.8748],[183,-3.9095],[184,-3.9678],[185,-4.0461],[186,-4.1109],[187,-4.1484],[189,-4.193],[190,-4.193],[191,-4.2136],[194,-4.2136],[195,-4.2291],[196,-4.2291],[198,-4.2777],[199,-4.313],[200,-4.3779],[201,-4.4724],[202,-4.5839],[203,-4.7076],[204,-4.8749],[206,-5.3174],[207,-5.5559],[208,-5.7793],[209,-5.9445],[210,-6.0606],[211,-6.1353],[212,-6.1727],[214,-6.183],[216,-6.156],[219,-6.0791],[221,-6.0459],[222,-6.0459],[228,-5.9422],[231,-5.8734],[233,-5.8377],[234,-5.8294],[236,-5.8601],[237,-5.7398],[238,-5.5399],[239,-5.3066],[240,-5.0467],[241,-4.707],[242,-4.3208],[243,-3.9527],[244,-3.6549],[245,-3.4121],[246,-3.2208],[247,-3.0896],[248,-3.0318],[252,-2.9233],[253,-2.9049],[254,-2.9036],[257,-2.921],[258,-2.9412],[259,-2.9683],[260,-2.9802],[262,-2.9571],[265,-2.8995],[269,-2.7666],[270,-2.7401],[271,-2.724],[272,-2.724],[273,-2.7106],[274,-2.7106],[275,-2.7233],[276,-2.7448],[277,-2.777],[278,-2.8179],[279,-2.8696],[281,-2.9903],[283,-3.1263],[285,-3.2796],[286,-3.3639],[288,-3.558],[290,-3.7626],[291,-3.8561],[292,-3.9371],[293,-3.9991],[294,-4.0469],[295,-4.071],[296,-4.0856],[297,-4.0856],[298,-4.1003],[301,-4.1003],[302,-4.0818],[303,-4.0818],[304,-4.063],[305,-4.063],[306,0]]
       }
       foot_r_fk_ctrl.Scale.X {type=MovieSceneScriptingFloatChannel, default=1, const [1,307]}
       foot_r_fk_ctrl.Scale.Y {type=MovieSceneScriptingFloatChannel, default=1, const [1,307]}
       foot_r_fk_ctrl.Scale.Z {type=MovieSceneScriptingFloatChannel, default=1, const [1,307]}
       ball_r_fk_ctrl.Location.X {type=MovieSceneScriptingFloatChannel, default=0, const [0,307]}
       ball_r_fk_ctrl.Location.Y {type=MovieSceneScriptingFloatChannel, default=0, const [0,307]}
       ball_r_fk_ctrl.Location.Z {type=MovieSceneScriptingFloatChannel, default=0, const [0,307]}
       ball_r_fk_ctrl.Rotation.X {type=MovieSceneScriptingFloatChannel, default=0, const [0,307]}
       ball_r_fk_ctrl.Rotation.Y {type=MovieSceneScriptingFloatChannel, default=0, const [0,307]}
       ball_r_fk_ctrl.Rotation.Z {type=MovieSceneScriptingFloatChannel, default=0, const [0,307]}
       ball_r_fk_ctrl.Scale.X {type=MovieSceneScriptingFloatChannel, default=1, const [1,307]}
       ball_r_fk_ctrl.Scale.Y {type=MovieSceneScriptingFloatChannel, default=1, const [1,307]}
       ball_r_fk_ctrl.Scale.Z {type=MovieSceneScriptingFloatChannel, default=1, const [1,307]}
       foot_r_ik_ctrl.Location.X {
        type=MovieSceneScriptingFloatChannel
        default=0
        thinned_from=307
        keys [[0,0.0001],[1,-0.0004],[2,0.0004],[3,-0.0005],[4,0.005],[5,-0.0011],[6,0],[7,0.0003],[8,-0.0002],[9,-0.0002],[10,-0.0006],[11,0.0004],[12,-0.0001],[13,-0.0008],[14,0.0002],[16,0.0002],[17,-0.0018],[18,-0.0025],[19,0.0011],[20,0.0015],[21,-0.0016],[22,0.0009],[23,-0.0008],[24,0.0016],[25,0.002],[26,-0.0022],[27,-0.0001],[28,0.0006],[29,-0.0014],[30,-0.0003],[31,-0.0014],[32,0.0011],[34,0.0015],[35,-0.0009],[36,0.0006],[37,0.0007],[38,-0.0004],[39,-0.0001],[40,-0.0004],[41,-0.0001],[42,0.0007],[43,0.0006],[44,0.0012],[45,-0.0006],[46,0],[47,-0.0002],[48,-0.0001],[49,-0.0006],[50,0.0003],[51,0.0009],[52,0.0005],[53,-0.0018],[54,0.0046],[55,0.0005],[56,0.0033],[57,-0.0003],[58,0.0033],[59,-0.0005],[60,-0.0011],[61,0.0002],[62,-0.0014],[63,0.001],[64,0.0019],[65,-0.0017],[66,0.0034],[67,-0.0002],[68,0.0004],[69,0.0001],[70,-0.0001],[71,0.0005],[72,-0.0015],[73,-0.0003],[74,0.0004],[76,0.0005],[77,-0.0001],[79,0.0004],[80,0.0001],[81,0.0003],[82,-0.0006],[83,0.001],[84,-0.002],[85,0.0035],[86,-0.0025],[87,-0.0001],[88,0.001],[89,-0.004],[90,-0.0002],[91,0.0028],[92,-0.0013],[93,-0.0003],[94,-0.0005],[95,-0.0016],[96,0.002],[97,-0.0008],[98,0.0012],[100,-0.0001],[101,0.0004],[102,-0.0018],[103,-0.0001],[104,0.0012],[105,0.0011],[106,0.0002],[107,-0.0002],[108,-0.0011],[109,0.0004],[111,-0.0007],[112,0],[113,-0.0128],[114,-0.0004],[115,-0.008],[116,0.0023],[117,-0.0041],[118,0.0003],[119,-0.0004],[120,0.0012],[121,-0.0009],[122,0.0017],[123,-0.0003],[125,0.0003],[126,-0.0007],[127,0.0006],[128,-0.0006],[129,0.0003],[130,0.0015],[131,-0.0002],[132,0.0014],[133,0.0008],[134,0.0018],[135,-0.004],[136,0.0016],[137,0.0005],[138,-0.0011],[139,-0.0005],[140,0.0009],[141,0],[142,0.0004],[143,-0.0004],[144,-0.0018],[145,-0.0004],[146,0.0004],[148,-0.0005],[154,0.0004],[155,-0.0002],[157,-0.0001],[158,0.0002],[159,-0.0005],[161,0],[165,-0.0006],[167,0.0003],[168,0.0004],[170,0.0002],[172,-0.0004],[173,0.0004],[175,0.0003],[176,-0.0001],[177,0.0003],[178,0.0002],[179,-0.0005],[181,-0.0001],[182,-0.0025],[183,0.0011],[185,-0.001],[186,0.0011],[187,0.0002],[189,0.001],[190,-0.0001],[191,-0.0006],[192,0.0023],[193,0.001],[194,0.0016],[195,0],[196,0.003],[197,0.0001],[198,-0.0012],[199,-0.0007],[200,-0.001],[201,-0.0004],[202,-0.0004],[203,0.0003],[204,0.0003],[205,-0.0006],[206,0.0011],[207,0.0003],[208,0.0011],[209,-0.0003],[210,0.0011],[211,-0.0006],[212,0.0017],[213,0.0001],[214,0.0008],[215,0.0012],[216,0.0006],[217,-0.0018],[218,0.0011],[219,0.0006],[220,0.0003],[221,-0.0009],[224,0.0006],[225,-0.0006],[226,-0.0012],[227,-0.0001],[228,-0.0005],[229,0.0003],[231,-0.0003],[233,0],[234,0.0001],[235,-0.0005],[236,-0.0007],[237,0.0005],[238,-0.0002],[239,-0.0007],[240,0.0005],[241,0.0001],[242,0.0002],[243,-0.0005],[245,-0.0009],[246,0],[247,-0.0008],[248,-0.0011],[249,0.0008],[251,0.0001],[252,-0.0009],[253,0.0006],[254,0.0003],[255,0.0003],[257,-0.0001],[258,0],[260,0.0016],[261,0.002],[262,0.003],[263,-0.0024],[264,0.0033],[265,0.001],[266,-0.0007],[267,0.0003],[268,0.0006],[269,0.0026],[270,-0.0012],[271,0.0036],[272,-0.0004],[273,0.0025],[274,0.0023],[275,-0.0001],[276,0.0004],[277,-0.001],[278,-0.0009],[279,0.0006],[281,-0.0006],[282,-0.0008],[283,0.0013],[284,-0.0002],[285,0.0005],[286,0.001],[287,-0.0007],[288,-0.0007],[289,0.0007],[290,-0.0002],[291,0.0005],[292,0],[293,0.0009],[294,0.0005],[295,0.0013],[296,0.001],[297,0.0019],[298,0.0008],[299,-0.0001],[300,-0.0028],[301,-0.0006],[304,-0.0007],[305,0.0016],[306,0]]
       }
       foot_r_ik_ctrl.Location.Y {
        type=MovieSceneScriptingFloatChannel
        default=0
        thinned_from=307
        keys [[0,-0.0017],[1,0.0021],[2,0.0016],[3,-0.002],[4,-0.0136],[5,-0.0024],[6,-0.0005],[7,0.0007],[8,0.0009],[9,-0.0004],[10,-0.0006],[11,-0.0021],[12,-0.0003],[13,-0.0015],[14,-0.0002],[15,0.0009],[16,-0.0007],[17,-0.0013],[18,-0.0001],[19,-0.0005],[20,0.0009],[21,0.0012],[22,-0.0004],[23,-0.0003],[24,0.0005],[25,0.0009],[27,0.0006],[28,-0.0016],[29,0.0009],[30,0.0004],[31,-0.0013],[32,0.002],[33,0.0012],[34,0.004],[35,-0.0006],[36,0.0008],[37,0.004],[38,0.0006],[39,-0.0017],[40,-0.0011],[41,0.0011],[42,0.0014],[43,-0.002],[44,0.0002],[45,0.0003],[46,0.0026],[47,0.0016],[48,0.0001],[49,0.0015],[50,-0.0003],[51,-0.0016],[52,-0.0001],[53,-0.0009],[54,-0.0029],[55,0.0029],[56,0.0017],[57,0.0007],[58,0.0018],[59,-0.0007],[60,0.003],[61,0.0007],[62,-0.0005],[63,0.0028],[64,-0.0012],[65,0.0023],[66,0.0006],[67,0.0017],[68,0],[70,0.0022],[71,-0.0039],[72,-0.0003],[73,0.0016],[74,-0.0001],[75,-0.0025],[76,-0.0018],[77,-0.002],[78,0.0029],[79,-0.0018],[80,-0.0037],[81,-0.0037],[82,-0.0024],[83,-0.0003],[84,0.002],[85,-0.0011],[86,-0.0002],[87,-0.0004],[88,-0.0011],[89,-0.0005],[90,0.0007],[91,0.0022],[92,-0.0005],[93,-0.0009],[94,0.0002],[95,0.0016],[97,0.0024],[98,-0.0007],[99,0.0013],[100,-0.0006],[101,0.0001],[102,-0.0036],[103,-0.0022],[104,0.0022],[105,0.0026],[106,-0.0001],[107,-0.0009],[108,-0.0011],[109,0.0003],[110,0.0023],[111,0.0003],[112,-0.0021],[113,-0.0102],[114,-0.0019],[115,-0.012],[116,0.0098],[117,-0.0093],[118,0.0002],[119,-0.013],[120,-0.0032],[121,-0.0003],[122,0.0076],[123,0.0013],[124,-0.0012],[125,0.0012],[126,-0.0008],[127,-0.0001],[128,-0.0012],[129,-0.0007],[130,0.0097],[131,-0.0004],[132,0.0053],[133,-0.0022],[134,0.005],[135,-0.0081],[136,0.0065],[137,0.0047],[138,0.0129],[139,0.0176],[140,0.0015],[141,0.002],[142,-0.0006],[143,-0.0007],[144,-0.0047],[145,0.003],[146,-0.0011],[147,0.0004],[148,-0.0011],[149,0.0002],[150,-0.0015],[151,0.0031],[152,-0.0006],[153,-0.0013],[154,0.0003],[155,0.0007],[156,-0.0012],[157,0.0005],[158,-0.0007],[159,-0.0003],[160,0.0018],[161,-0.0006],[162,-0.0023],[163,-0.0003],[164,-0.0013],[165,-0.0004],[166,-0.0004],[167,0.0005],[168,0.0024],[169,0.0013],[170,0.0005],[171,-0.0006],[172,-0.0007],[173,0.001],[174,-0.002],[175,0.0009],[176,0],[177,0],[178,-0.001],[179,-0.0023],[180,-0.0021],[181,0.0017],[182,-0.0063],[183,0.0013],[184,0.0023],[186,0.0019],[187,-0.001],[188,-0.0007],[189,-0.0008],[190,-0.0012],[191,0.0019],[192,0.0044],[193,0.0107],[194,0.0072],[195,-0.0042],[196,0.0097],[197,-0.001],[198,0],[199,-0.0001],[200,0.0006],[201,-0.0029],[202,-0.0001],[203,0.0023],[204,0.0008],[205,-0.0037],[206,-0.0005],[207,0.0031],[208,0.0041],[209,0.0027],[210,-0.0007],[211,-0.0029],[212,0.0021],[213,0.0051],[214,-0.0001],[215,-0.0054],[216,0.0015],[217,-0.0056],[218,-0.0013],[219,-0.0014],[220,0.0023],[221,0.0048],[222,-0.0003],[223,0.0008],[224,-0.0023],[225,-0.0049],[226,0.0027],[227,0.0009],[228,0.0009],[229,-0.0016],[230,0.0017],[231,0.0008],[232,-0.0015],[233,0.0003],[234,-0.0005],[235,0.002],[236,0.0025],[237,-0.0027],[238,0.0002],[239,-0.0019],[240,-0.0035],[241,-0.001],[242,-0.0038],[243,0.003],[244,0.0045],[245,0.001],[246,-0.0028],[247,0],[248,0.0018],[249,-0.0029],[250,-0.0032],[251,0.0018],[252,-0.0013],[253,0.0055],[254,-0.0042],[255,0.0043],[256,-0.004],[257,0.0055],[258,-0.0031],[259,0.0034],[260,-0.0012],[261,-0.0005],[262,-0.002],[263,-0.0089],[264,-0.0002],[265,-0.003],[266,0.0018],[267,-0.0023],[268,0.0025],[269,0.0013],[270,0.0039],[271,0.0042],[272,-0.0101],[273,-0.0021],[274,-0.0002],[275,0.0014],[276,-0.0017],[277,0.0022],[279,-0.0007],[280,0.0035],[281,0.0001],[282,-0.0008],[283,0.0008],[284,0.002],[285,0],[286,0.0029],[287,-0.0017],[288,0.0001],[289,-0.0016],[290,-0.0027],[291,-0.0023],[292,0.0002],[293,0.0001],[294,0.0022],[295,-0.0002],[296,-0.0017],[297,0.0067],[298,0.0043],[299,-0.0021],[300,-0.0068],[301,0.0028],[302,-0.0014],[303,0.0009],[304,0.0001],[305,0.0087],[306,0]]
       }
       foot_r_ik_ctrl.Location.Z {
        type=MovieSceneScriptingFloatChannel
        default=0
        thinned_from=307
        keys [[0,0.0012],[1,0.0008],[2,0.0007],[3,0.0011],[4,0.0019],[5,0.001],[6,0.0008],[7,0.0008],[9,0.0021],[10,0.003],[11,-0.0012],[12,0.0002],[13,0.0027],[14,-0.0003],[15,0.0026],[16,-0.0001],[17,0.0007],[18,0.0031],[19,-0.0028],[20,-0.001],[21,0.0011],[22,0.0021],[23,0.0036],[24,-0.0013],[25,-0.0002],[27,0.0035],[28,-0.0016],[30,-0.0002],[31,0.0001],[32,0.0001],[33,-0.0003],[34,-0.0004],[36,0.0001],[37,0.0001],[38,-0.0003],[39,-0.0005],[40,-0.0011],[41,-0.0015],[42,-0.0025],[44,-0.004],[45,0.0039],[46,0.0032],[47,0.0027],[48,0.0017],[52,-0.0003],[53,-0.001],[55,-0.0017],[56,0.0028],[58,0.002],[59,0.0008],[60,0.0002],[61,-0.001],[62,-0.0024],[63,0.0017],[64,-0.0007],[65,0.0029],[66,0.0005],[67,-0.0003],[68,-0.0034],[69,0.0029],[71,-0.0026],[72,0.0046],[73,0.0033],[74,0.0005],[75,-0.0014],[76,-0.0039],[77,0.0035],[79,0.0006],[80,0.0003],[81,0.0004],[82,0.001],[83,0.0001],[84,0.0007],[85,0.0012],[86,0.0014],[87,0.0019],[88,0.0029],[89,0.0032],[90,0.0041],[91,0.0045],[93,0.0049],[94,0.0054],[96,0.0053],[98,0.0057],[99,-0.0001],[100,0.0001],[101,0],[102,0.001],[104,0],[105,-0.0003],[106,0.0001],[107,-0.0001],[108,-0.0001],[109,-0.0006],[110,-0.0007],[111,-0.0006],[112,-0.0001],[113,-0.001],[115,0.0009],[116,-0.0014],[117,-0.0009],[118,0.0006],[119,0.0004],[120,-0.0003],[121,-0.0007],[122,-0.0025],[123,0.0008],[125,0.0004],[126,0],[127,-0.0002],[129,-0.001],[130,-0.0028],[131,0.0002],[132,-0.0007],[133,0.0002],[134,-0.0005],[135,0.0001],[137,-0.0009],[138,-0.001],[139,-0.0014],[141,0.0002],[142,0.0001],[143,0.0003],[144,0.0013],[145,0.0002],[146,0.0002],[147,0.001],[148,-0.0018],[149,-0.002],[150,-0.001],[152,-0.0019],[153,-0.0038],[154,-0.0037],[155,0.0027],[156,0.0042],[157,0.0011],[158,-0.0004],[159,-0.0001],[160,-0.0004],[161,0.0007],[162,0.001],[163,0.0005],[165,0.0023],[166,-0.0005],[167,0.0009],[168,-0.0024],[169,-0.0009],[171,0.0013],[172,0.0022],[173,-0.0023],[174,-0.0016],[175,-0.0014],[176,-0.0008],[178,-0.0003],[179,-0.0014],[180,-0.001],[181,0.0006],[182,0.0002],[184,0.0007],[185,0.0018],[186,-0.0016],[187,0.001],[188,0.0005],[189,-0.0013],[190,0.0017],[191,-0.0023],[192,-0.0018],[194,-0.0015],[195,0],[196,-0.0014],[197,-0.0002],[198,0],[199,-0.0015],[200,-0.0005],[201,-0.0018],[202,-0.0024],[203,0.0027],[205,-0.0006],[206,0.0032],[207,0.0004],[208,-0.003],[209,0.0015],[210,-0.0005],[211,0.0003],[213,-0.002],[215,-0.0028],[216,-0.0027],[217,-0.0014],[218,-0.0027],[220,-0.0025],[222,-0.0019],[224,-0.0022],[225,-0.0014],[226,-0.0017],[227,0.0002],[229,0.0009],[231,0.0013],[232,0.0017],[233,0.0019],[234,0.0023],[235,0.0025],[236,0.0029],[237,-0.0017],[238,0.0008],[239,0.0029],[240,-0.0003],[241,0.0029],[242,-0.0003],[243,0.0004],[244,-0.0009],[245,0.001],[246,-0.0013],[248,0.0007],[250,0.0023],[251,-0.0016],[252,-0.0009],[254,-0.0004],[255,0.0002],[256,0.0003],[257,0.0002],[258,0.0004],[259,0.0017],[260,-0.0008],[261,-0.0003],[262,-0.0011],[263,0.0007],[264,-0.0002],[265,0.0018],[266,-0.001],[267,0.0004],[268,0.0008],[269,0.0005],[270,0.0005],[271,0.0007],[272,0.0022],[273,0.0012],[274,0.0012],[275,0.0005],[277,0],[278,0],[279,-0.0004],[280,-0.0014],[281,-0.0012],[282,-0.0021],[283,-0.0024],[284,-0.0025],[285,0.0036],[286,0.0031],[287,0.0021],[288,0.0013],[289,0.0008],[291,0.0007],[292,0.0003],[293,0.0003],[294,-0.0015],[295,-0.0005],[296,-0.0007],[297,-0.0022],[298,-0.0012],[299,-0.0014],[300,0.0008],[301,-0.0014],[302,-0.0005],[303,-0.0012],[304,0],[305,-0.0014],[306,0]]
       }
       foot_r_ik_ctrl.Rotation.X {
        type=MovieSceneScriptingFloatChannel
        default=0
        thinned_from=307
        keys [[0,0.003],[1,0.0108],[2,-0.002],[3,0.001],[4,0.0187],[5,0.0138],[6,0.0004],[7,0.002],[8,-0.001],[9,0.0008],[11,0.0016],[12,0.0029],[13,0.0014],[14,0.0012],[15,0.0021],[16,0.0023],[17,0.001],[18,0],[20,-0.0002],[21,0.0006],[22,-0.001],[23,0.0017],[24,0.0004],[25,-0.0013],[26,0.0005],[27,-0.002],[28,0.0018],[29,-0.0012],[30,0.0005],[31,0.0007],[32,-0.0008],[33,-0.0006],[34,-0.0054],[35,0.0003],[36,-0.0015],[37,-0.0061],[38,0.0009],[39,0.002],[40,0.0006],[41,0],[42,-0.0016],[43,0.0014],[44,0.0012],[45,0.0019],[46,-0.002],[47,-0.0008],[48,0.0024],[49,0.0002],[50,0.0029],[51,0.0043],[52,0.0028],[53,0.005],[54,0],[55,0.0034],[56,-0.0002],[57,0.0036],[58,-0.0026],[59,-0.0034],[60,-0.0002],[61,0.0007],[62,-0.0003],[63,0.0018],[64,-0.0004],[65,-0.0004],[66,0.0018],[67,-0.0002],[68,-0.0009],[69,0.0006],[70,0.001],[71,0.0002],[72,0.0004],[73,0.0039],[74,-0.0012],[75,-0.0011],[76,0.0008],[77,0],[78,0.0004],[79,-0.0006],[80,0.0003],[81,0.0109],[82,0.007],[83,0.0005],[84,-0.0002],[85,0.0006],[87,0.0016],[88,0.0016],[89,0.002],[90,0.0004],[91,-0.0018],[92,0.0001],[93,0],[94,-0.0015],[95,-0.0131],[96,0.0013],[97,-0.0021],[98,-0.0011],[99,-0.0022],[100,-0.0004],[101,0.0012],[102,0.0074],[103,0.0122],[104,-0.0006],[105,-0.002],[106,0.0002],[107,-0.0012],[108,0.0019],[109,-0.001],[110,-0.0026],[111,0.0082],[112,0.0042],[113,0.0189],[114,0.0012],[115,0.0141],[116,-0.0153],[117,0.011],[118,0.0078],[119,0.005],[120,0.002],[121,0.006],[122,-0.0038],[123,0.0119],[124,-0.0011],[125,0.0037],[126,0.0072],[127,0.0125],[128,-0.0027],[129,0.0006],[130,-0.0107],[131,0.0005],[132,-0.0103],[133,0.0003],[134,-0.0068],[135,0.0079],[136,0.0018],[137,0.0028],[138,-0.0147],[139,-0.0138],[140,0.0067],[141,-0.0013],[142,0.0039],[143,0.0117],[144,0.0121],[145,-0.0033],[146,0.0011],[147,-0.0007],[148,-0.0006],[149,0.0008],[150,-0.0027],[151,-0.003],[152,0.0006],[153,0.0011],[154,0.0021],[155,0.0002],[156,0.0027],[157,-0.0023],[158,0],[159,-0.0022],[160,-0.0003],[161,0.0023],[162,0.0002],[163,0.0038],[164,-0.0005],[165,-0.0015],[166,0.0002],[167,-0.0031],[168,-0.0012],[169,-0.001],[170,-0.0019],[171,0.0013],[172,0.004],[173,-0.0002],[174,0.0022],[175,0.0016],[176,0.0016],[177,0.0008],[178,0.001],[179,-0.0005],[180,0.0012],[181,0.0002],[182,0.01],[183,0.0011],[184,0.0008],[185,0.0002],[186,0.0008],[187,0.0031],[188,0.0021],[189,-0.0018],[190,0.0129],[191,-0.0003],[192,-0.0017],[193,-0.0049],[194,-0.0011],[195,0.0023],[196,-0.0005],[197,0.0003],[198,-0.0016],[199,0.0003],[200,0.0011],[201,0.0026],[202,0.0002],[203,-0.0007],[204,0.0011],[205,0.0022],[206,0.0005],[207,-0.0005],[208,-0.0011],[209,0.0003],[210,0.0006],[211,-0.0009],[212,-0.0013],[213,0.0087],[214,-0.0015],[215,0.0001],[216,-0.0009],[217,0.0116],[218,-0.0013],[219,0.0007],[220,-0.0023],[221,0.0007],[222,-0.012],[223,-0.0013],[224,-0.0021],[225,0.0009],[226,-0.0006],[227,-0.0004],[228,-0.0044],[229,0.0038],[230,0.0009],[231,0.001],[233,0.0014],[234,-0.0017],[235,0.0012],[236,-0.0026],[237,0.0037],[238,-0.0002],[239,-0.0019],[240,-0.0001],[241,0.002],[242,0.0034],[243,-0.0018],[244,-0.0003],[245,0.0009],[246,0.0013],[247,-0.0007],[248,0.0003],[249,-0.0001],[250,-0.0022],[251,0.0003],[252,-0.0001],[253,0.0008],[254,0.0026],[255,-0.0012],[256,0.0002],[257,-0.0015],[258,0.0025],[259,-0.0005],[260,0.0006],[261,0.0001],[262,0.0031],[263,0.009],[264,0],[265,-0.0002],[266,0.0028],[267,0.0004],[269,0.0009],[270,0.0004],[271,-0.0014],[272,0.0019],[273,-0.001],[274,-0.0005],[275,0.0007],[276,0.0024],[277,-0.0007],[278,0.0012],[279,0.0001],[280,-0.0021],[281,0.0009],[282,0.0005],[283,0.0004],[284,-0.0013],[285,0.0002],[286,-0.0004],[287,0.0009],[288,0.0011],[289,0.0019],[290,0.0041],[291,0.0012],[292,-0.0014],[293,0.0014],[294,-0.0008],[295,0.0012],[296,0.0022],[297,0.0043],[298,-0.0053],[299,0.0017],[300,0.0103],[301,-0.0073],[302,0.0027],[303,-0.0076],[304,0.0033],[305,-0.0087],[306,0]]
       }
       foot_r_ik_ctrl.Rotation.Y {
        type=MovieSceneScriptingFloatChannel
        default=0
        thinned_from=307
        keys [[0,0.0002],[1,-0.0042],[2,-0.0002],[3,0.0006],[4,0.0035],[5,-0.0046],[6,-0.0064],[7,-0.0001],[8,0.0001],[10,-0.0002],[11,0.0003],[12,0],[13,0.0002],[16,-0.0001],[17,0.0002],[18,-0.0001],[19,0.0002],[20,-0.0003],[22,-0.0001],[23,-0.0002],[24,0.0001],[25,0.0002],[26,-0.0001],[27,0.0001],[28,0.0002],[29,-0.0003],[30,0.0002],[33,-0.0005],[34,0.0001],[35,-0.0002],[36,0.0003],[38,-0.0001],[39,0.0003],[40,-0.0001],[41,-0.0002],[42,0],[43,0.0004],[44,0.0003],[45,-0.0002],[48,-0.0002],[49,-0.0003],[50,0.0002],[52,0.0004],[53,-0.0002],[54,0.0007],[55,0.0003],[56,-0.0004],[57,-0.0007],[58,0.0006],[59,0.0006],[60,-0.0002],[61,0.0005],[62,0.0005],[63,-0.0004],[65,-0.0001],[67,-0.0003],[68,-0.0001],[71,0.0002],[72,-0.0002],[73,0.0001],[77,0.0002],[78,-0.0001],[80,0.0003],[82,0.0085],[83,-0.0003],[84,0.0003],[85,0.0006],[87,0.0006],[89,-0.0003],[90,-0.0004],[92,0.0001],[93,0],[94,0.0002],[95,0.012],[96,0],[97,0.0113],[98,0],[99,0.0094],[100,-0.0002],[101,0.0083],[102,-0.0012],[103,0.0067],[104,-0.0002],[105,0.0006],[106,-0.0005],[107,-0.0004],[108,0.0008],[109,-0.0005],[110,0.0001],[111,0.0031],[112,0.0006],[113,-0.0047],[114,0.0001],[115,-0.0024],[116,-0.0001],[117,-0.0011],[118,-0.0005],[119,0.0018],[120,-0.0002],[123,-0.004],[124,0.0003],[125,-0.0011],[126,-0.002],[127,-0.0026],[128,-0.0003],[129,0.0015],[130,0.0045],[131,0.0083],[132,0.0006],[133,0.0047],[134,0.0065],[135,0.0069],[136,0.0098],[137,0.0102],[138,-0.0023],[139,-0.0016],[140,0.0035],[141,0],[142,0.001],[143,0.0019],[144,-0.0002],[145,-0.0007],[146,0.0003],[147,-0.0004],[148,-0.0005],[150,0],[152,-0.0004],[153,-0.0004],[154,-0.0001],[157,0.0001],[158,-0.0002],[159,0.0002],[160,-0.0001],[161,-0.0002],[162,0],[163,-0.0005],[164,0.0003],[166,-0.0001],[167,0.0001],[168,-0.0002],[170,-0.0001],[171,0.0003],[172,0.0001],[173,-0.0003],[174,0.0004],[175,-0.0004],[176,-0.0001],[177,-0.0001],[178,0.0004],[179,0.0004],[181,-0.0006],[182,0.0006],[183,0.0007],[184,-0.0005],[185,0.0002],[186,0.0002],[187,0.0011],[188,0.0015],[189,0.0002],[190,0.0043],[191,-0.0008],[192,0.0027],[193,0.0041],[194,0.0078],[195,0.0001],[196,0.0076],[197,-0.0009],[198,-0.0004],[200,-0.0001],[201,0.0005],[202,0],[203,-0.0002],[204,0.0003],[205,0],[206,0.0007],[207,0],[208,0],[209,-0.0005],[210,0],[211,0.0002],[212,0.0007],[213,-0.0028],[214,0.0002],[216,0.0004],[217,-0.0001],[218,0],[219,-0.0004],[220,-0.0002],[221,-0.0003],[222,-0.0015],[223,0.0001],[224,0.0003],[225,-0.0001],[226,-0.0003],[227,-0.0002],[228,0.0001],[229,0.0012],[230,0.0006],[232,0.0003],[233,0.0007],[234,0.0003],[235,-0.0017],[237,-0.0002],[238,0],[239,0.0013],[240,0.0006],[241,0.0005],[242,-0.0001],[243,-0.0001],[244,-0.0018],[245,0.0003],[246,-0.0006],[247,0],[248,-0.001],[249,0.0016],[250,-0.0002],[251,0.0004],[252,0],[253,0.0007],[255,-0.0007],[256,0.0013],[257,0.0005],[258,-0.0013],[259,0.0005],[260,-0.0007],[261,0.0004],[263,-0.0001],[264,-0.0001],[265,-0.0004],[266,0.0001],[267,0.0001],[268,-0.0006],[269,0.0003],[271,0.0004],[272,-0.0056],[273,0.0006],[274,-0.002],[275,-0.0001],[276,-0.0001],[277,-0.0005],[278,-0.0003],[279,0.0004],[280,-0.0009],[281,0.0003],[282,-0.0002],[283,-0.0001],[284,-0.0003],[285,0.0003],[286,-0.0004],[287,-0.0001],[288,-0.0005],[289,0.0002],[290,0.0007],[291,0.0005],[292,-0.0001],[294,0.0008],[295,-0.0005],[296,-0.0009],[297,0.0057],[298,-0.0005],[299,0.0032],[300,0.0053],[301,0.0071],[302,0.0006],[303,0.0009],[304,0.0006],[305,0.0006],[306,0]]
       }
       foot_r_ik_ctrl.Rotation.Z {type=MovieSceneScriptingFloatChannel, default=0, thinned_from=307, keys [[0,-14.549],[305,-14.5476],[306,0]]}
       foot_r_ik_ctrl.Scale.X {type=MovieSceneScriptingFloatChannel, default=1, const [1,307]}
       foot_r_ik_ctrl.Scale.Y {type=MovieSceneScriptingFloatChannel, default=1, const [1,307]}
       foot_r_ik_ctrl.Scale.Z {type=MovieSceneScriptingFloatChannel, default=1, const [1,307]}
       ball_r_ik_ctrl.Location.X {type=MovieSceneScriptingFloatChannel, default=0, const [0,307]}
       ball_r_ik_ctrl.Location.Y {type=MovieSceneScriptingFloatChannel, default=0, const [0,307]}
       ball_r_ik_ctrl.Location.Z {type=MovieSceneScriptingFloatChannel, default=0, const [0,307]}
       ball_r_ik_ctrl.Rotation.X {type=MovieSceneScriptingFloatChannel, default=0, const [0,307]}
       ball_r_ik_ctrl.Rotation.Y {type=MovieSceneScriptingFloatChannel, default=0, const [0,307]}
       ball_r_ik_ctrl.Rotation.Z {type=MovieSceneScriptingFloatChannel, default=0, const [0,307]}
       ball_r_ik_ctrl.Scale.X {type=MovieSceneScriptingFloatChannel, default=1, const [1,307]}
       ball_r_ik_ctrl.Scale.Y {type=MovieSceneScriptingFloatChannel, default=1, const [1,307]}
       ball_r_ik_ctrl.Scale.Z {type=MovieSceneScriptingFloatChannel, default=1, const [1,307]}
       foot_bk1_r_ctrl.Location.X {type=MovieSceneScriptingFloatChannel, default=0, const [0,307]}
       foot_bk1_r_ctrl.Location.Y {type=MovieSceneScriptingFloatChannel, default=0, const [0,307]}
       foot_bk1_r_ctrl.Location.Z {type=MovieSceneScriptingFloatChannel, default=0, const [0,307]}
       foot_bk1_r_ctrl.Rotation.X {type=MovieSceneScriptingFloatChannel, default=0, const [0,307]}
       foot_bk1_r_ctrl.Rotation.Y {type=MovieSceneScriptingFloatChannel, default=0, const [0,307]}
       foot_bk1_r_ctrl.Rotation.Z {type=MovieSceneScriptingFloatChannel, default=0, const [0,307]}
       foot_bk1_r_ctrl.Scale.X {type=MovieSceneScriptingFloatChannel, default=1, const [1,307]}
       foot_bk1_r_ctrl.Scale.Y {type=MovieSceneScriptingFloatChannel, default=1, const [1,307]}
       foot_bk1_r_ctrl.Scale.Z {type=MovieSceneScriptingFloatChannel, default=1, const [1,307]}
       foot_roll_r_ctrl.Location.X {type=MovieSceneScriptingFloatChannel, default=0, const [0,307]}
       foot_roll_r_ctrl.Location.Y {type=MovieSceneScriptingFloatChannel, default=0, const [0,307]}
       foot_roll_r_ctrl.Location.Z {type=MovieSceneScriptingFloatChannel, default=0, const [0,307]}
       foot_roll_r_ctrl.Rotation.X {type=MovieSceneScriptingFloatChannel, default=0, const [0,307]}
       foot_roll_r_ctrl.Rotation.Y {type=MovieSceneScriptingFloatChannel, default=0, const [0,307]}
       foot_roll_r_ctrl.Rotation.Z {type=MovieSceneScriptingFloatChannel, default=0, const [0,307]}
       foot_roll_r_ctrl.Scale.X {type=MovieSceneScriptingFloatChannel, default=1, const [1,307]}
       foot_roll_r_ctrl.Scale.Y {type=MovieSceneScriptingFloatChannel, default=1, const [1,307]}
       foot_roll_r_ctrl.Scale.Z {type=MovieSceneScriptingFloatChannel, default=1, const [1,307]}
       roll_blend_r_ctrl {type=MovieSceneScriptingFloatChannel, default=1, const [1,307]}
       tip_r_ctrl.Location.X {type=MovieSceneScriptingFloatChannel, default=0, const [0,307]}
       tip_r_ctrl.Location.Y {type=MovieSceneScriptingFloatChannel, default=0, const [0,307]}
       tip_r_ctrl.Location.Z {type=MovieSceneScriptingFloatChannel, default=0, const [0,307]}
       tip_r_ctrl.Rotation.X {type=MovieSceneScriptingFloatChannel, default=0, const [0,307]}
       tip_r_ctrl.Rotation.Y {type=MovieSceneScriptingFloatChannel, default=0, const [0,307]}
       tip_r_ctrl.Rotation.Z {type=MovieSceneScriptingFloatChannel, default=0, const [0,307]}
       tip_r_ctrl.Scale.X {type=MovieSceneScriptingFloatChannel, default=1, const [1,307]}
       tip_r_ctrl.Scale.Y {type=MovieSceneScriptingFloatChannel, default=1, const [1,307]}
       tip_r_ctrl.Scale.Z {type=MovieSceneScriptingFloatChannel, default=1, const [1,307]}
       heel_r_ctrl.Location.X {type=MovieSceneScriptingFloatChannel, default=0, const [0,307]}
       heel_r_ctrl.Location.Y {type=MovieSceneScriptingFloatChannel, default=0, const [0,307]}
       heel_r_ctrl.Location.Z {type=MovieSceneScriptingFloatChannel, default=0, const [0,307]}
       heel_r_ctrl.Rotation.X {type=MovieSceneScriptingFloatChannel, default=0, const [0,307]}
       heel_r_ctrl.Rotation.Y {type=MovieSceneScriptingFloatChannel, default=0, const [0,307]}
       heel_r_ctrl.Rotation.Z {type=MovieSceneScriptingFloatChannel, default=0, const [0,307]}
       heel_r_ctrl.Scale.X {type=MovieSceneScriptingFloatChannel, default=1, const [1,307]}
       heel_r_ctrl.Scale.Y {type=MovieSceneScriptingFloatChannel, default=1, const [1,307]}
       heel_r_ctrl.Scale.Z {type=MovieSceneScriptingFloatChannel, default=1, const [1,307]}
       leg_r_softness {type=MovieSceneScriptingFloatChannel, default=0, const [0,307]}
       leg_r_pv_ik_ctrl.Location.X {
        type=MovieSceneScriptingFloatChannel
        default=0
        thinned_from=307
        keys [[0,-0.4787],[1,-0.4798],[5,-0.4816],[7,-0.478],[8,-0.4748],[9,-0.4689],[10,-0.4606],[11,-0.449],[12,-0.4337],[13,-0.4134],[15,-0.3661],[16,-0.344],[17,-0.3239],[18,-0.3096],[19,-0.3017],[22,-0.284],[23,-0.2767],[24,-0.2663],[26,-0.2426],[27,-0.2318],[28,-0.216],[29,-0.2103],[30,-0.2069],[31,-0.2068],[32,-0.2086],[33,-0.2109],[34,-0.2119],[35,-0.2124],[36,-0.2123],[37,-0.2126],[38,-0.2146],[39,-0.2174],[40,-0.2219],[41,-0.2282],[42,-0.2354],[44,-0.2514],[45,-0.2583],[46,-0.2645],[49,-0.2844],[50,-0.2899],[53,-0.3037],[54,-0.3018],[55,-0.3067],[57,-0.3151],[58,-0.3205],[59,-0.3277],[60,-0.3373],[61,-0.3494],[62,-0.3639],[63,-0.3739],[64,-0.392],[65,-0.4144],[66,-0.4424],[67,-0.4769],[68,-0.522],[69,-0.5568],[70,-0.588],[71,-0.6154],[72,-0.63],[73,-0.6475],[75,-0.6803],[76,-0.698],[77,-0.7185],[78,-0.7368],[79,-0.7472],[81,-0.7537],[83,-0.7523],[85,-0.7459],[88,-0.7279],[90,-0.7146],[92,-0.7056],[95,-0.6981],[102,-0.6913],[105,-0.6904],[113,-0.6967],[115,-0.6973],[116,-0.6963],[118,-0.6993],[122,-0.7012],[123,-0.7032],[124,-0.703],[127,-0.7067],[130,-0.7078],[131,-0.7094],[132,-0.709],[137,-0.7112],[138,-0.7097],[140,-0.7097],[144,-0.7043],[145,-0.702],[147,-0.7015],[149,-0.7049],[151,-0.7117],[153,-0.7217],[154,-0.7288],[157,-0.7585],[158,-0.7667],[159,-0.7711],[160,-0.7712],[161,-0.7681],[163,-0.7576],[164,-0.7496],[165,-0.737],[166,-0.7221],[168,-0.69],[169,-0.6764],[170,-0.6645],[172,-0.6463],[178,-0.6087],[180,-0.5981],[181,-0.5938],[183,-0.5899],[184,-0.5915],[186,-0.5981],[190,-0.6022],[196,-0.6014],[198,-0.6033],[199,-0.606],[200,-0.6099],[201,-0.6154],[202,-0.6248],[203,-0.6405],[204,-0.6659],[205,-0.6988],[206,-0.7341],[208,-0.8103],[209,-0.844],[210,-0.8735],[211,-0.8982],[212,-0.9195],[213,-0.9387],[214,-0.9511],[215,-0.9595],[216,-0.9641],[217,-0.9665],[221,-0.9619],[222,-0.9592],[229,-0.9446],[233,-0.9275],[234,-0.9221],[235,-0.9124],[236,-0.8934],[237,-0.8635],[238,-0.826],[240,-0.7463],[241,-0.702],[242,-0.6563],[243,-0.6133],[244,-0.5778],[245,-0.5495],[246,-0.5268],[247,-0.5088],[248,-0.4951],[249,-0.4825],[250,-0.471],[251,-0.4607],[252,-0.4516],[253,-0.4446],[254,-0.44],[255,-0.4364],[256,-0.4339],[257,-0.433],[259,-0.436],[260,-0.4381],[263,-0.4421],[266,-0.4441],[275,-0.4422],[279,-0.4446],[285,-0.4537],[287,-0.4593],[292,-0.4792],[293,-0.4819],[294,-0.4834],[297,-0.4839],[298,-0.4826],[300,-0.4826],[301,-0.4804],[302,-0.48],[303,-0.4784],[305,-0.4781],[306,0]]
       }
       leg_r_pv_ik_ctrl.Location.Y {
        type=MovieSceneScriptingFloatChannel
        default=0
        thinned_from=307
        keys [[0,3.7153],[4,3.7469],[5,3.7469],[6,3.7378],[7,3.7183],[8,3.6837],[9,3.6226],[10,3.534],[11,3.4133],[12,3.2515],[13,3.0336],[14,2.7683],[16,2.2184],[17,1.9499],[18,1.7204],[19,1.5406],[20,1.3838],[21,1.2354],[22,1.0961],[23,0.9521],[24,0.7924],[25,0.6259],[26,0.4622],[27,0.2998],[28,0.1454],[29,0.0256],[30,-0.0372],[31,-0.0605],[32,-0.0604],[33,-0.0498],[34,-0.0497],[35,-0.0565],[36,-0.0639],[37,-0.0639],[38,-0.0491],[39,-0.0219],[40,0.0241],[41,0.0914],[42,0.1701],[43,0.2569],[44,0.3402],[45,0.4151],[46,0.4835],[48,0.6305],[49,0.701],[50,0.7625],[53,0.9213],[55,1.0137],[57,1.092],[58,1.1424],[59,1.2089],[60,1.2978],[61,1.4096],[62,1.5425],[63,1.6933],[64,1.8703],[65,2.0818],[66,2.3308],[67,2.6222],[68,2.935],[69,3.2287],[70,3.4882],[71,3.7124],[72,3.9094],[76,4.654],[77,4.8476],[78,5.016],[79,5.1088],[80,5.1392],[81,5.1514],[82,5.1514],[83,5.1405],[85,5.0826],[88,4.9299],[90,4.8184],[91,4.7764],[92,4.7446],[95,4.6953],[101,4.6605],[105,4.6704],[115,4.7395],[141,4.8255],[145,4.8121],[146,4.8176],[147,4.8346],[149,4.8979],[151,4.9899],[153,5.1168],[154,5.1955],[157,5.4737],[158,5.5438],[159,5.5784],[160,5.5737],[161,5.5423],[163,5.4435],[164,5.3681],[165,5.2535],[168,4.8263],[169,4.7068],[170,4.6065],[171,4.5243],[172,4.4579],[173,4.404],[175,4.3263],[178,4.2421],[180,4.1995],[181,4.1851],[183,4.1904],[184,4.219],[186,4.3019],[187,4.3224],[191,4.3553],[194,4.3553],[196,4.3613],[198,4.389],[199,4.4149],[200,4.4573],[201,4.5182],[202,4.6028],[203,4.7198],[204,4.8934],[206,5.3423],[208,5.816],[209,6.0068],[210,6.1618],[211,6.2817],[212,6.3748],[213,6.4465],[214,6.4967],[215,6.5247],[217,6.5359],[230,6.36],[235,6.2362],[236,6.1673],[237,6.0097],[238,5.7911],[239,5.5486],[240,5.2896],[241,4.9829],[242,4.6466],[243,4.322],[244,4.0529],[245,3.8327],[246,3.6538],[247,3.5189],[248,3.43],[250,3.2903],[252,3.1738],[253,3.1309],[254,3.1049],[256,3.0749],[257,3.0697],[260,3.1102],[263,3.1102],[266,3.0906],[270,3.0314],[271,3.024],[272,3.024],[273,3.0166],[275,3.0223],[277,3.0457],[279,3.0885],[283,3.2137],[285,3.2894],[286,3.3336],[288,3.4418],[290,3.5613],[292,3.6628],[293,3.7001],[294,3.7255],[296,3.7445],[300,3.7445],[304,3.7113],[305,3.7113],[306,0]]
       }
       leg_r_pv_ik_ctrl.Location.Z {
        type=MovieSceneScriptingFloatChannel
        default=0
        thinned_from=307
        keys [[0,3.8598],[4,3.896],[5,3.896],[6,3.8857],[7,3.8633],[8,3.8237],[9,3.7541],[10,3.6534],[11,3.517],[12,3.3354],[13,3.0934],[14,2.8022],[15,2.5033],[16,2.2108],[17,1.9276],[18,1.6883],[19,1.5023],[20,1.3414],[21,1.1902],[22,1.0493],[23,0.9048],[24,0.7461],[25,0.5819],[26,0.4219],[27,0.2642],[28,0.116],[29,0.0011],[30,-0.0588],[31,-0.0811],[32,-0.0813],[33,-0.0713],[34,-0.0714],[35,-0.0779],[36,-0.085],[37,-0.085],[38,-0.071],[39,-0.0453],[40,-0.0016],[41,0.0626],[42,0.1379],[43,0.2211],[44,0.3014],[45,0.374],[46,0.4404],[48,0.5839],[49,0.6531],[50,0.7136],[53,0.871],[55,0.9636],[57,1.0417],[58,1.0922],[59,1.1591],[60,1.2487],[61,1.3619],[62,1.4974],[63,1.6529],[64,1.8362],[65,2.0573],[66,2.3204],[67,2.6325],[68,2.9717],[69,3.2961],[70,3.5868],[71,3.8413],[72,4.0682],[76,4.9471],[77,5.1812],[78,5.3869],[79,5.5009],[80,5.5384],[81,5.5534],[82,5.5534],[83,5.5399],[84,5.5098],[85,5.4685],[88,5.2814],[90,5.1459],[91,5.095],[92,5.0566],[94,5.0121],[101,4.9557],[105,4.9679],[115,5.0513],[116,5.0514],[121,5.0857],[122,5.0857],[129,5.1342],[141,5.1552],[145,5.1395],[146,5.1463],[147,5.1671],[148,5.2004],[150,5.2975],[151,5.3574],[152,5.4301],[153,5.5137],[154,5.6113],[157,5.9588],[158,6.0472],[159,6.0909],[160,6.0849],[161,6.045],[163,5.9206],[164,5.826],[165,5.683],[168,5.1582],[169,5.0137],[170,4.8932],[171,4.795],[172,4.716],[173,4.6522],[175,4.5606],[178,4.4624],[180,4.4129],[181,4.3964],[183,4.4031],[184,4.4368],[186,4.5346],[187,4.5589],[191,4.598],[194,4.598],[196,4.6052],[198,4.6382],[199,4.6688],[200,4.7194],[201,4.7919],[202,4.8931],[203,5.0335],[204,5.2434],[205,5.512],[206,5.7951],[208,6.3924],[209,6.6373],[210,6.838],[211,6.9943],[212,7.1163],[213,7.2104],[214,7.2767],[215,7.3135],[216,7.328],[217,7.3278],[229,7.1202],[234,6.9709],[235,6.9323],[236,6.8431],[237,6.6389],[238,6.3584],[239,6.0511],[240,5.7273],[241,5.3498],[242,4.9427],[243,4.5569],[244,4.2421],[245,3.9879],[246,3.7834],[247,3.6307],[248,3.5308],[250,3.3751],[252,3.246],[253,3.1987],[254,3.17],[256,3.1372],[257,3.1315],[260,3.1762],[263,3.1758],[266,3.1536],[270,3.0877],[271,3.0795],[272,3.0795],[273,3.0713],[274,3.0713],[276,3.088],[278,3.1251],[280,3.1819],[284,3.3316],[286,3.4258],[288,3.5479],[290,3.6834],[291,3.7452],[292,3.7991],[293,3.8419],[294,3.871],[296,3.893],[300,3.8931],[304,3.8553],[305,3.8553],[306,0]]
       }
       leg_r_pv_ik_ctrl.Rotation.X {
        type=MovieSceneScriptingFloatChannel
        default=0
        thinned_from=307
        keys [[0,-5.2246],[4,-5.2716],[5,-5.2716],[6,-5.2583],[7,-5.2292],[8,-5.1778],[9,-5.0873],[10,-4.9563],[11,-4.7783],[12,-4.5407],[13,-4.2225],[14,-3.8376],[16,-3.0484],[17,-2.6669],[18,-2.3426],[19,-2.0893],[20,-1.8693],[21,-1.6618],[22,-1.4679],[23,-1.2682],[24,-1.048],[25,-0.8194],[26,-0.5957],[27,-0.3744],[28,-0.1655],[29,-0.0031],[30,0.0818],[31,0.1134],[32,0.1137],[33,0.0995],[34,0.0996],[35,0.1089],[36,0.1189],[37,0.119],[38,0.0991],[39,0.0626],[40,0.0007],[41,-0.0901],[42,-0.1965],[43,-0.3139],[44,-0.4268],[45,-0.5287],[46,-0.6218],[48,-0.8224],[49,-0.9189],[50,-1.0032],[53,-1.2215],[55,-1.3497],[57,-1.4577],[58,-1.5272],[59,-1.6193],[60,-1.7425],[61,-1.8978],[62,-2.0831],[63,-2.2949],[64,-2.5438],[65,-2.8424],[66,-3.1961],[67,-3.6128],[68,-4.0627],[69,-4.4898],[70,-4.8703],[71,-5.2014],[72,-5.4953],[76,-6.6214],[77,-6.9182],[78,-7.1779],[79,-7.3215],[80,-7.3687],[81,-7.3875],[82,-7.3875],[83,-7.3706],[85,-7.2808],[88,-7.0448],[90,-6.8735],[91,-6.8091],[92,-6.7604],[95,-6.6851],[101,-6.6322],[105,-6.6476],[115,-6.7536],[129,-6.8587],[141,-6.8853],[145,-6.8653],[146,-6.8739],[147,-6.9002],[149,-6.9982],[151,-7.1406],[153,-7.3375],[154,-7.4601],[157,-7.8951],[158,-8.0053],[159,-8.0597],[160,-8.0522],[161,-8.0026],[163,-7.8473],[164,-7.7291],[165,-7.55],[168,-6.889],[169,-6.7058],[170,-6.5526],[171,-6.4276],[172,-6.3269],[173,-6.2453],[175,-6.1282],[178,-6.0023],[180,-5.9388],[181,-5.9176],[183,-5.9262],[184,-5.9694],[186,-6.0948],[187,-6.1259],[191,-6.1759],[194,-6.1759],[196,-6.1851],[198,-6.2272],[199,-6.2664],[200,-6.3309],[201,-6.4235],[202,-6.5523],[203,-6.7307],[204,-6.9965],[205,-7.3352],[206,-7.6905],[208,-8.4342],[209,-8.737],[210,-8.9843],[211,-9.1763],[212,-9.3258],[213,-9.4409],[214,-9.5219],[215,-9.5668],[217,-9.5844],[229,-9.3307],[235,-9.1002],[236,-8.9907],[237,-8.7392],[238,-8.3922],[239,-8.0103],[240,-7.6056],[241,-7.131],[242,-6.6155],[243,-6.1234],[244,-5.7192],[245,-5.3909],[246,-5.1258],[247,-4.927],[248,-4.7966],[250,-4.5929],[252,-4.4234],[253,-4.3612],[254,-4.3235],[256,-4.2803],[257,-4.2728],[260,-4.3316],[263,-4.3311],[266,-4.3019],[270,-4.2152],[271,-4.2044],[272,-4.2044],[273,-4.1936],[275,-4.2018],[277,-4.2361],[279,-4.2988],[282,-4.4319],[285,-4.5939],[286,-4.6591],[288,-4.8187],[290,-4.9954],[292,-5.1459],[293,-5.2015],[294,-5.2392],[296,-5.2676],[300,-5.2678],[304,-5.2188],[305,-5.2188],[306,0]]
       }
       leg_r_pv_ik_ctrl.Rotation.Y {
        type=MovieSceneScriptingFloatChannel
        default=0
        thinned_from=307
        keys [[0,0.0255],[4,0.026],[6,0.0259],[7,0.0256],[8,0.025],[9,0.0241],[10,0.0227],[11,0.0209],[12,0.0185],[13,0.0155],[15,0.0086],[16,0.0055],[17,0.0026],[18,0.0002],[19,-0.0018],[21,-0.005],[22,-0.0063],[25,-0.0097],[27,-0.0115],[28,-0.0119],[29,-0.0126],[32,-0.0132],[39,-0.0134],[53,-0.0095],[54,-0.0089],[58,-0.008],[60,-0.0071],[62,-0.0055],[64,-0.0026],[65,-0.0007],[66,0.0015],[68,0.0073],[70,0.0144],[71,0.0176],[72,0.0212],[73,0.0243],[76,0.0352],[77,0.0392],[78,0.0427],[79,0.0447],[80,0.0453],[82,0.0455],[83,0.0453],[85,0.0441],[88,0.0409],[90,0.0386],[91,0.0378],[93,0.0368],[101,0.0358],[103,0.0358],[121,0.038],[141,0.0392],[143,0.039],[146,0.0394],[147,0.0398],[149,0.0413],[151,0.0434],[152,0.0448],[153,0.0464],[154,0.0482],[157,0.0546],[158,0.0562],[159,0.057],[160,0.0569],[161,0.0561],[163,0.0538],[164,0.052],[165,0.0494],[167,0.0431],[168,0.0402],[169,0.0378],[170,0.0359],[171,0.0344],[172,0.0332],[173,0.0323],[175,0.0311],[178,0.0301],[181,0.0295],[183,0.0299],[184,0.0305],[185,0.0314],[186,0.0321],[189,0.033],[194,0.0332],[196,0.0334],[198,0.0339],[199,0.0344],[200,0.0352],[201,0.0365],[202,0.0381],[203,0.0402],[204,0.0434],[205,0.0476],[206,0.0521],[208,0.0622],[209,0.0664],[210,0.0699],[211,0.0726],[212,0.0746],[213,0.076],[214,0.0771],[215,0.0776],[216,0.0778],[217,0.0776],[221,0.0763],[228,0.0738],[235,0.0702],[236,0.069],[237,0.0654],[238,0.0605],[239,0.0554],[240,0.0499],[241,0.0438],[242,0.0374],[243,0.0317],[244,0.0274],[245,0.024],[246,0.0214],[247,0.0196],[248,0.0186],[252,0.016],[254,0.0153],[257,0.015],[260,0.0155],[266,0.0148],[270,0.0138],[274,0.0135],[277,0.014],[282,0.0162],[284,0.0174],[287,0.0197],[291,0.0235],[293,0.025],[294,0.0254],[296,0.0258],[301,0.0258],[305,0.0254],[306,0]]
       }
       leg_r_pv_ik_ctrl.Rotation.Z {
        type=MovieSceneScriptingFloatChannel
        default=0
        thinned_from=307
        keys [[0,0.6435],[1,0.6449],[5,0.6473],[7,0.6426],[8,0.6386],[9,0.631],[10,0.6202],[11,0.6053],[12,0.5854],[13,0.5589],[15,0.4967],[16,0.4673],[17,0.4404],[18,0.421],[19,0.4101],[22,0.3856],[23,0.3758],[24,0.3615],[26,0.3291],[27,0.3143],[28,0.2927],[29,0.2848],[30,0.28],[31,0.2799],[32,0.2823],[33,0.2855],[34,0.2869],[35,0.2874],[36,0.2873],[37,0.2878],[38,0.2905],[39,0.2943],[40,0.3004],[41,0.309],[42,0.3189],[44,0.3407],[45,0.3501],[46,0.3585],[49,0.3854],[50,0.3927],[53,0.4114],[54,0.4089],[55,0.4155],[57,0.4267],[58,0.4339],[59,0.4435],[60,0.4562],[61,0.4723],[62,0.4916],[63,0.5048],[64,0.5286],[65,0.558],[66,0.5943],[67,0.6389],[68,0.6966],[69,0.7407],[70,0.7801],[71,0.8143],[72,0.8327],[76,0.9165],[77,0.9417],[78,0.9639],[79,0.9765],[81,0.9844],[83,0.9826],[85,0.9749],[88,0.9531],[90,0.9368],[92,0.9258],[95,0.9166],[102,0.9083],[105,0.9072],[113,0.9149],[115,0.9157],[116,0.9144],[118,0.9181],[122,0.9205],[123,0.9228],[130,0.9286],[131,0.9305],[132,0.93],[137,0.9327],[138,0.9308],[140,0.9308],[144,0.9243],[145,0.9214],[147,0.9208],[149,0.925],[151,0.9333],[153,0.9456],[154,0.9542],[157,0.9902],[158,1.0002],[159,1.0055],[160,1.0055],[161,1.0018],[163,0.9891],[164,0.9794],[165,0.9642],[166,0.946],[168,0.9068],[169,0.8901],[170,0.8754],[172,0.853],[179,0.7995],[181,0.7878],[183,0.7829],[184,0.785],[186,0.7933],[190,0.7984],[196,0.7975],[198,0.7999],[200,0.8081],[201,0.815],[202,0.8267],[203,0.8462],[204,0.8774],[205,0.9177],[206,0.9607],[208,1.0525],[209,1.0928],[210,1.1278],[211,1.157],[212,1.1821],[213,1.2047],[214,1.2193],[215,1.2291],[216,1.2346],[217,1.2374],[221,1.232],[222,1.2289],[229,1.2118],[234,1.1853],[235,1.1739],[236,1.1514],[237,1.116],[238,1.0713],[240,0.9754],[241,0.9215],[242,0.8653],[243,0.8121],[244,0.7677],[245,0.7322],[246,0.7035],[247,0.6807],[248,0.6632],[249,0.6472],[250,0.6326],[251,0.6194],[252,0.6077],[253,0.5988],[254,0.5929],[255,0.5883],[256,0.5851],[257,0.5839],[259,0.5877],[260,0.5905],[263,0.5955],[266,0.5981],[275,0.5954],[279,0.5986],[285,0.6108],[287,0.6181],[292,0.644],[293,0.6475],[294,0.6495],[297,0.6501],[298,0.6485],[300,0.6485],[301,0.6457],[302,0.6452],[303,0.6431],[305,0.6428],[306,0]]
       }
       leg_r_pv_ik_ctrl.Scale.X {type=MovieSceneScriptingFloatChannel, default=1, const [1,307]}
       leg_r_pv_ik_ctrl.Scale.Y {type=MovieSceneScriptingFloatChannel, default=1, const [1,307]}
       leg_r_pv_ik_ctrl.Scale.Z {type=MovieSceneScriptingFloatChannel, default=1, const [1,307]}
       upperarm_l_fk_ctrl.Location.X {type=MovieSceneScriptingFloatChannel, default=0, thinned_from=307, keys [[0,0],[306,0]]}
       upperarm_l_fk_ctrl.Location.Y {type=MovieSceneScriptingFloatChannel, default=0, thinned_from=307, keys [[0,0],[306,0]]}
       upperarm_l_fk_ctrl.Location.Z {type=MovieSceneScriptingFloatChannel, default=0, thinned_from=307, keys [[0,0],[306,0]]}
       upperarm_l_fk_ctrl.Rotation.X {
        type=MovieSceneScriptingFloatChannel
        default=0
        thinned_from=307
        keys [[0,7.0555],[4,7.0405],[7,7.0108],[10,7.0467],[11,7.0787],[12,7.1399],[13,7.254],[14,7.3856],[16,7.5885],[17,7.7104],[18,7.8634],[19,8.1207],[20,8.5981],[21,9.3783],[22,10.3696],[23,11.3366],[24,12.0902],[25,12.6936],[26,13.2021],[27,13.6001],[28,13.8887],[29,14.094],[30,14.2269],[32,14.3437],[36,14.4579],[38,14.5372],[40,14.584],[41,14.5711],[42,14.5925],[52,14.3367],[56,14.0694],[58,13.8556],[59,13.7154],[60,13.5454],[61,13.3446],[64,12.6273],[65,12.2843],[66,11.7477],[67,10.8138],[69,8.2815],[70,7.2118],[71,6.3193],[72,5.641],[73,5.3132],[74,5.2304],[75,5.2836],[76,5.3755],[77,5.4783],[78,5.5558],[79,5.5498],[80,5.4143],[81,5.2588],[82,5.1413],[84,4.9343],[85,4.8202],[87,4.6265],[89,4.4648],[91,4.3255],[93,4.2059],[96,4.0554],[97,3.9959],[98,3.9023],[99,3.7588],[100,3.6243],[101,3.5153],[102,3.4155],[103,3.3296],[104,3.2509],[107,3.0508],[108,3.0158],[109,3.0212],[110,3.0577],[113,3.2538],[114,3.3253],[116,3.5183],[117,3.6074],[118,3.6649],[119,3.7115],[120,3.7404],[121,3.7433],[123,3.72],[124,3.697],[125,3.654],[126,3.6018],[128,3.4203],[129,3.3189],[130,3.2451],[131,3.2259],[132,3.2295],[133,3.2212],[135,3.2216],[137,3.228],[138,3.2396],[139,3.2446],[140,3.3],[141,3.3711],[142,3.449],[143,3.5485],[144,3.6669],[145,3.8059],[146,3.9604],[147,4.1325],[148,4.3165],[149,4.5335],[150,4.7869],[151,5.1615],[152,5.7282],[153,6.3818],[154,7.214],[155,8.2828],[156,9.6957],[157,11.2664],[158,12.5979],[159,13.5999],[160,14.1451],[161,14.4913],[162,14.7168],[163,14.8534],[164,14.9306],[165,14.9591],[166,14.9415],[167,14.829],[168,14.3149],[169,13.4437],[170,12.3773],[171,11.1685],[172,9.8808],[173,8.6642],[174,7.5347],[175,6.4755],[176,5.3977],[177,4.2937],[178,3.2277],[179,2.2665],[180,1.4917],[181,0.8634],[182,0.3788],[183,0.0279],[184,-0.1305],[185,-0.0473],[186,0.2252],[187,0.6524],[188,1.1551],[189,1.6933],[190,2.2148],[191,2.7106],[192,3.1747],[193,3.5844],[194,3.9414],[195,4.2489],[196,4.4993],[197,4.7021],[198,4.863],[199,4.9835],[200,5.0403],[201,4.9604],[202,4.7224],[203,4.3032],[204,3.7057],[205,2.9458],[206,2.0435],[207,1.0343],[208,-0.1514],[209,-1.6131],[210,-3.2452],[211,-4.8645],[212,-6.3958],[213,-7.8185],[214,-9.0773],[215,-10.092],[216,-10.8922],[217,-11.4995],[218,-11.9782],[219,-12.3697],[220,-12.7044],[221,-12.988],[222,-13.2227],[223,-13.4191],[224,-13.5778],[226,-13.8133],[227,-13.9042],[231,-14.0576],[233,-14.0567],[235,-13.9593],[236,-13.8681],[237,-13.7316],[238,-13.5258],[239,-13.2364],[240,-12.8412],[241,-12.3132],[242,-11.5994],[243,-10.6666],[244,-9.4721],[245,-8.0327],[246,-6.4376],[247,-4.7981],[248,-3.2029],[249,-1.5767],[250,0.0805],[251,1.6326],[252,2.9524],[253,4.0602],[254,5.0641],[255,5.9665],[256,6.7633],[257,7.4519],[258,8.0322],[259,8.5272],[260,8.9612],[261,9.3403],[262,9.6681],[263,9.9434],[264,10.175],[265,10.3715],[266,10.5349],[267,10.6715],[269,10.8717],[271,10.9886],[274,11.0444],[277,11.0238],[278,10.9979],[280,10.8446],[281,10.7268],[282,10.5766],[283,10.388],[284,10.1695],[285,9.9303],[290,8.603],[293,7.8662],[294,7.6347],[295,7.4453],[296,7.3059],[297,7.2265],[298,7.1742],[300,7.1034],[301,7.0808],[302,7.0817],[303,7.0562],[305,7.0597],[306,0]]
       }
       upperarm_l_fk_ctrl.Rotation.Y {
        type=MovieSceneScriptingFloatChannel
        default=0
        thinned_from=307
        keys [[0,7.2449],[2,7.2594],[4,7.3143],[5,7.3836],[6,7.4325],[7,7.4988],[9,7.7207],[10,7.8442],[11,8.0005],[14,8.5283],[15,8.6308],[16,8.656],[17,8.6177],[19,8.3128],[20,8.1856],[21,8.106],[23,7.9818],[28,7.5752],[32,7.1882],[33,7.1286],[34,7.0941],[35,7.1074],[36,7.1529],[37,7.2229],[38,7.3139],[41,7.6693],[42,7.845],[43,7.9965],[45,8.3272],[49,8.9067],[51,9.1254],[55,9.5233],[58,9.7685],[61,10.0509],[62,10.1252],[63,10.1582],[65,10.1281],[66,10.1433],[68,10.3063],[69,10.6262],[70,11.1692],[71,11.752],[72,12.3765],[73,13.0661],[74,13.8426],[75,14.4297],[76,14.566],[77,14.3026],[78,13.8123],[80,12.6283],[82,11.5432],[83,11.0909],[84,10.7749],[85,10.549],[86,10.366],[87,10.2203],[88,10.1094],[89,10.0265],[91,9.9244],[96,9.8297],[97,9.8448],[98,9.9566],[99,10.2109],[100,10.4205],[101,10.5022],[102,10.5182],[104,10.4759],[105,10.4224],[106,10.3458],[107,10.24],[108,10.0843],[109,9.8824],[110,9.7032],[111,9.5609],[112,9.4614],[113,9.4168],[117,9.4144],[118,9.4426],[119,9.5169],[123,9.9222],[125,10.0645],[128,10.1897],[129,10.1953],[130,10.1785],[132,10.0377],[133,9.9879],[134,9.9879],[135,9.948],[140,9.8875],[145,9.8923],[148,9.8604],[150,9.9127],[151,9.9771],[152,10.0873],[153,10.2484],[154,10.5194],[155,10.9674],[156,11.6419],[157,12.4385],[158,13.1793],[159,13.7809],[160,14.1979],[161,14.5136],[162,14.7417],[163,14.8812],[164,14.9385],[165,14.93],[166,14.8579],[167,14.7284],[168,14.4564],[169,13.9955],[170,13.3719],[171,12.7007],[172,12.0843],[173,11.5785],[174,11.1266],[175,10.7603],[176,10.4562],[177,10.1877],[181,9.2051],[182,8.9826],[183,8.7893],[184,8.6469],[185,8.585],[186,8.5918],[187,8.6647],[194,9.3202],[196,9.4536],[197,9.4935],[198,9.509],[199,9.5457],[201,9.7614],[202,9.8472],[203,9.8923],[204,9.8436],[205,9.6629],[206,9.3274],[207,8.8138],[208,8.2277],[209,7.7604],[210,7.4432],[211,6.9986],[212,6.397],[213,5.7568],[214,5.0747],[215,4.3597],[216,3.6658],[217,3.1695],[218,2.8774],[219,2.7037],[220,2.5787],[221,2.4658],[223,2.2588],[227,1.8746],[228,1.8029],[229,1.724],[230,1.6766],[231,1.5709],[232,1.4864],[233,1.3514],[234,1.2014],[235,1.0031],[236,0.7857],[237,0.533],[238,0.266],[239,0.0089],[240,-0.2167],[241,-0.3734],[242,-0.4223],[243,-0.3408],[244,-0.0987],[245,0.2934],[246,0.7794],[247,1.3053],[248,1.8171],[249,2.3474],[250,2.8987],[251,3.4297],[252,3.8951],[255,5.1895],[257,6.0275],[258,6.4125],[259,6.7671],[260,7.0818],[261,7.3549],[263,7.8227],[265,8.203],[268,8.7101],[271,9.0824],[274,9.3854],[276,9.5588],[279,9.777],[281,9.8611],[282,9.8771],[284,9.8495],[287,9.7163],[289,9.5403],[291,9.2958],[292,9.1344],[298,7.8966],[299,7.7195],[300,7.5801],[301,7.4749],[303,7.3328],[304,7.2915],[305,7.2769],[306,0]]
       }
       upperarm_l_fk_ctrl.Rotation.Z {
        type=MovieSceneScriptingFloatChannel
        default=0
        thinned_from=307
        keys [[0,-31.5393],[10,-31.4721],[13,-31.2308],[17,-30.8095],[18,-30.7681],[20,-31.1543],[27,-31.5332],[35,-31.5616],[41,-31.1035],[58,-29.3019],[63,-29.014],[64,-28.908],[65,-28.6841],[66,-28.3633],[68,-27.9933],[69,-27.6617],[70,-27.2372],[71,-26.6725],[72,-26.1615],[75,-25.1629],[76,-24.9562],[77,-24.8504],[78,-24.8462],[79,-24.9688],[80,-25.234],[82,-25.8445],[84,-26.3251],[87,-26.8214],[91,-27.1645],[98,-27.5042],[100,-27.677],[103,-28.2354],[107,-28.655],[111,-28.7574],[119,-28.2276],[121,-28.2685],[126,-28.8073],[130,-29.3339],[133,-29.4741],[142,-29.3591],[149,-29.0117],[151,-28.9907],[153,-28.7453],[154,-28.5416],[155,-28.2536],[157,-27.4324],[158,-27.1395],[159,-26.9997],[163,-27.2405],[165,-27.1905],[167,-26.9678],[168,-27.0035],[173,-27.5101],[175,-27.5498],[176,-27.674],[177,-27.9821],[178,-28.4985],[181,-30.4442],[182,-30.9677],[183,-31.3577],[184,-31.6321],[185,-31.7602],[186,-31.7479],[187,-31.6123],[193,-30.2066],[195,-29.8538],[198,-29.5678],[201,-29.642],[203,-29.8094],[204,-29.8465],[208,-29.6053],[209,-29.7567],[210,-30.0339],[214,-31.4139],[215,-31.6609],[223,-32.8046],[228,-33.9874],[232,-35.2739],[238,-37.5835],[240,-38.1651],[241,-38.3927],[242,-38.5379],[243,-38.5555],[244,-38.4044],[245,-38.1043],[247,-37.2259],[248,-36.7234],[249,-36.1441],[251,-34.8288],[254,-33.1522],[257,-31.7095],[260,-30.4664],[263,-29.4992],[266,-28.756],[269,-28.1936],[274,-27.6069],[278,-27.3598],[280,-27.374],[282,-27.5128],[284,-27.8024],[286,-28.2205],[294,-30.6197],[296,-31.0597],[299,-31.387],[305,-31.5413],[306,0]]
       }
       upperarm_l_fk_ctrl.Scale.X {type=MovieSceneScriptingFloatChannel, default=1, const [1,307]}
       upperarm_l_fk_ctrl.Scale.Y {type=MovieSceneScriptingFloatChannel, default=1, const [1,307]}
       upperarm_l_fk_ctrl.Scale.Z {type=MovieSceneScriptingFloatChannel, default=1, const [1,307]}
       lowerarm_l_fk_ctrl.Location.X {type=MovieSceneScriptingFloatChannel, default=0, thinned_from=307, keys [[0,0],[133,0],[134,0.0002],[135,0],[306,0]]}
       lowerarm_l_fk_ctrl.Location.Y {type=MovieSceneScriptingFloatChannel, default=0, thinned_from=307, keys [[0,0],[133,0],[134,0.0003],[135,0],[306,0]]}
       lowerarm_l_fk_ctrl.Location.Z {type=MovieSceneScriptingFloatChannel, default=0, thinned_from=307, keys [[0,0],[133,0],[134,-0.0029],[135,0],[306,0]]}
       lowerarm_l_fk_ctrl.Rotation.X {
        type=MovieSceneScriptingFloatChannel
        default=0
        thinned_from=307
        keys [[0,-0.0825],[5,-0.0731],[6,-0.0711],[8,-0.0663],[9,-0.0636],[10,-0.06],[11,-0.0555],[12,-0.0497],[13,-0.0414],[14,-0.0308],[15,-0.0192],[16,-0.0069],[17,0.0057],[18,0.0176],[19,0.0286],[21,0.0487],[22,0.0586],[23,0.0665],[24,0.0713],[25,0.0736],[26,0.0744],[27,0.0744],[28,0.0739],[29,0.0727],[30,0.0707],[31,0.0681],[32,0.065],[33,0.0615],[34,0.0573],[35,0.0524],[36,0.0468],[38,0.035],[41,0.0179],[43,0.0064],[44,0.0009],[46,-0.0093],[47,-0.0141],[48,-0.0183],[49,-0.0221],[50,-0.0254],[51,-0.0284],[52,-0.0311],[53,-0.0336],[56,-0.04],[58,-0.0437],[59,-0.0451],[60,-0.0462],[61,-0.0468],[62,-0.0472],[63,-0.0465],[64,-0.0441],[65,-0.0401],[66,-0.0352],[67,-0.0268],[68,-0.013],[69,0.005],[70,0.018],[71,0.0231],[72,0.0252],[73,0.023],[74,0.0156],[75,0.0072],[76,0.0017],[77,-0.0018],[78,-0.0047],[79,-0.0063],[80,-0.0068],[81,-0.0064],[82,-0.0053],[83,-0.003],[84,0.0003],[85,0.004],[87,0.0125],[88,0.0156],[89,0.0178],[90,0.0192],[92,0.0209],[94,0.0216],[97,0.0216],[98,0.0209],[99,0.0168],[100,0.0107],[101,0.0022],[102,-0.0078],[103,-0.0159],[104,-0.0223],[105,-0.0276],[106,-0.0315],[107,-0.0343],[108,-0.0369],[109,-0.0392],[112,-0.0448],[114,-0.048],[115,-0.0494],[116,-0.0504],[118,-0.0515],[119,-0.0522],[120,-0.0534],[121,-0.0553],[122,-0.0575],[123,-0.0601],[124,-0.0631],[125,-0.0663],[126,-0.0707],[129,-0.0877],[130,-0.0928],[131,-0.0966],[132,-0.0991],[133,-0.0998],[134,-0.1026],[135,-0.0998],[137,-0.0998],[140,-0.0958],[141,-0.0942],[144,-0.0885],[145,-0.0871],[147,-0.0856],[149,-0.0846],[151,-0.084],[153,-0.084],[154,-0.0836],[155,-0.0881],[156,-0.1176],[157,-0.1509],[158,-0.1702],[159,-0.1751],[160,-0.1645],[161,-0.1431],[162,-0.1144],[163,-0.0825],[164,-0.0509],[165,-0.0213],[166,0.0057],[167,0.03],[168,0.0513],[169,0.0703],[170,0.086],[171,0.0967],[172,0.1014],[173,0.1006],[174,0.0964],[175,0.0892],[176,0.0787],[177,0.0652],[178,0.0485],[179,0.0292],[180,0.0104],[181,-0.005],[182,-0.0171],[183,-0.0262],[184,-0.0325],[185,-0.0363],[186,-0.038],[187,-0.038],[188,-0.035],[189,-0.0306],[191,-0.0198],[192,-0.0152],[195,-0.0043],[196,-0.0012],[197,0.001],[198,0.0021],[199,0.0018],[200,-0.0006],[201,-0.0072],[202,-0.0203],[203,-0.0428],[204,-0.0752],[205,-0.1129],[206,-0.154],[207,-0.2011],[208,-0.246],[209,-0.2766],[210,-0.2934],[211,-0.2977],[212,-0.2926],[213,-0.2826],[214,-0.2705],[215,-0.2557],[218,-0.2015],[219,-0.1847],[220,-0.1695],[221,-0.155],[222,-0.1423],[223,-0.132],[224,-0.1238],[225,-0.1166],[227,-0.1031],[231,-0.0737],[233,-0.0585],[234,-0.0506],[235,-0.0431],[236,-0.0358],[238,-0.0219],[239,-0.0153],[240,-0.0092],[242,0.0013],[243,0.0076],[244,0.016],[245,0.027],[246,0.0386],[247,0.049],[248,0.0567],[249,0.0624],[250,0.0668],[251,0.0696],[252,0.0707],[253,0.0697],[254,0.0667],[255,0.0619],[256,0.0554],[257,0.0469],[258,0.0355],[259,0.0221],[260,0.0077],[261,-0.0071],[262,-0.023],[263,-0.04],[265,-0.0749],[266,-0.0922],[268,-0.1276],[269,-0.1443],[270,-0.1587],[271,-0.1701],[272,-0.1789],[273,-0.1857],[274,-0.191],[276,-0.1992],[277,-0.2028],[278,-0.2056],[279,-0.2076],[280,-0.2088],[281,-0.2094],[283,-0.2094],[284,-0.207],[285,-0.2032],[286,-0.1984],[287,-0.1924],[288,-0.1849],[292,-0.1505],[293,-0.1416],[294,-0.1332],[295,-0.1256],[296,-0.1183],[297,-0.1113],[298,-0.1048],[299,-0.0991],[300,-0.0946],[301,-0.0914],[302,-0.089],[303,-0.0872],[305,-0.084],[306,0]]
       }
       lowerarm_l_fk_ctrl.Rotation.Y {
        type=MovieSceneScriptingFloatChannel
        default=0
        thinned_from=307
        keys [[0,-3.3417],[9,-3.2825],[11,-3.2565],[12,-3.2371],[13,-3.2092],[15,-3.131],[22,-2.7969],[23,-2.7549],[24,-2.7278],[25,-2.7147],[27,-2.71],[29,-2.7203],[32,-2.7628],[34,-2.8033],[43,-3.0333],[47,-3.112],[53,-3.1822],[60,-3.2253],[63,-3.2264],[65,-3.2049],[66,-3.1879],[67,-3.1582],[68,-3.1082],[69,-3.0387],[70,-2.9853],[71,-2.9636],[72,-2.955],[73,-2.9643],[75,-3.0297],[76,-3.0517],[78,-3.0764],[80,-3.0845],[82,-3.0788],[85,-3.0427],[88,-2.9956],[93,-2.9714],[98,-2.9732],[99,-2.9903],[100,-3.0157],[102,-3.0884],[103,-3.1189],[105,-3.1613],[115,-3.2362],[121,-3.2555],[125,-3.2912],[130,-3.3726],[132,-3.3913],[134,-3.3988],[154,-3.345],[155,-3.3584],[157,-3.5353],[158,-3.5852],[159,-3.5977],[160,-3.5707],[161,-3.5143],[162,-3.4355],[164,-3.2413],[169,-2.7337],[170,-2.6393],[171,-2.5668],[172,-2.5318],[173,-2.5383],[174,-2.5688],[175,-2.6188],[176,-2.6844],[177,-2.7619],[179,-2.9373],[180,-3.0169],[181,-3.0777],[182,-3.1232],[183,-3.1562],[184,-3.1784],[185,-3.1916],[187,-3.1974],[189,-3.1717],[192,-3.1165],[196,-3.0632],[198,-3.0501],[199,-3.0512],[200,-3.0609],[201,-3.0863],[202,-3.1348],[203,-3.2139],[204,-3.3191],[206,-3.5433],[207,-3.6617],[208,-3.7662],[209,-3.8332],[210,-3.869],[211,-3.878],[212,-3.8674],[214,-3.8201],[215,-3.7876],[221,-3.5459],[223,-3.4846],[233,-3.266],[240,-3.0939],[242,-3.0534],[243,-3.0284],[244,-2.9937],[245,-2.9471],[247,-2.8451],[248,-2.8064],[249,-2.7768],[250,-2.753],[251,-2.7371],[252,-2.7313],[253,-2.7371],[254,-2.7539],[255,-2.7796],[256,-2.8129],[257,-2.8557],[258,-2.9089],[265,-3.3184],[268,-3.4725],[269,-3.5177],[270,-3.5557],[272,-3.6071],[277,-3.6657],[283,-3.6817],[286,-3.655],[298,-3.4078],[300,-3.3781],[305,-3.3461],[306,0]]
       }
       lowerarm_l_fk_ctrl.Rotation.Z {
        type=MovieSceneScriptingFloatChannel
        default=0
        thinned_from=307
        keys [[0,3.6378],[5,3.2471],[6,3.162],[8,2.9591],[9,2.8412],[10,2.6869],[11,2.4911],[12,2.2313],[13,1.8593],[14,1.3702],[15,0.816],[16,0.2075],[17,-0.4462],[18,-1.0875],[19,-1.7064],[20,-2.3114],[21,-2.9372],[22,-3.5838],[23,-4.1305],[24,-4.4832],[25,-4.6552],[26,-4.7154],[27,-4.7154],[28,-4.681],[29,-4.5821],[30,-4.4359],[31,-4.2467],[32,-4.0272],[33,-3.7862],[34,-3.4986],[35,-3.1684],[36,-2.8099],[37,-2.4419],[38,-2.0829],[39,-1.7517],[40,-1.4248],[41,-1.1025],[42,-0.7845],[43,-0.4804],[44,-0.1901],[45,0.0769],[46,0.3301],[47,0.5646],[48,0.7711],[49,0.9542],[50,1.1139],[51,1.2544],[52,1.3822],[53,1.4985],[54,1.6032],[55,1.7021],[56,1.7951],[57,1.8824],[58,1.962],[59,2.026],[60,2.0745],[61,2.1056],[62,2.1192],[63,2.0881],[64,1.9794],[65,1.801],[66,1.5741],[67,1.1783],[68,0.5131],[69,-0.408],[70,-1.1131],[71,-1.3976],[72,-1.5131],[73,-1.3892],[74,-0.9806],[75,-0.5266],[76,-0.2358],[77,-0.0523],[78,0.0918],[79,0.1766],[80,0.1992],[81,0.1794],[82,0.1229],[83,0.0068],[84,-0.1612],[85,-0.3551],[87,-0.8065],[88,-0.977],[89,-1.1013],[90,-1.1793],[91,-1.2284],[92,-1.2689],[93,-1.2949],[94,-1.3093],[97,-1.3093],[98,-1.2718],[99,-1.0455],[100,-0.7112],[101,-0.2585],[102,0.2497],[103,0.6548],[104,0.9661],[105,1.2193],[106,1.4011],[107,1.5327],[108,1.6529],[109,1.7602],[110,1.8496],[112,2.0138],[113,2.0886],[114,2.1586],[115,2.2192],[116,2.2619],[118,2.3124],[119,2.3435],[120,2.3978],[121,2.4794],[122,2.5766],[123,2.6893],[124,2.8176],[125,2.9577],[126,3.1445],[129,3.8497],[130,4.0564],[131,4.2087],[132,4.3093],[133,4.3344],[137,4.3344],[140,4.178],[141,4.1137],[144,3.882],[145,3.8262],[147,3.7631],[151,3.697],[153,3.697],[154,3.6831],[155,3.8639],[156,5.031],[157,6.2719],[158,6.9589],[159,7.1309],[160,6.759],[161,5.9849],[162,4.9089],[163,3.6373],[164,2.288],[166,-0.4467],[167,-1.7932],[169,-4.4074],[170,-5.6344],[171,-6.5735],[172,-7.0263],[173,-6.9425],[174,-6.5477],[175,-5.9007],[176,-5.0484],[177,-4.0405],[178,-2.9227],[179,-1.743],[180,-0.6952],[181,0.109],[182,0.7106],[183,1.1508],[184,1.4469],[185,1.6233],[186,1.7009],[187,1.7009],[188,1.5668],[189,1.3588],[190,1.105],[191,0.8444],[192,0.6226],[194,0.2567],[195,0.0747],[196,-0.084],[197,-0.2003],[198,-0.257],[199,-0.2427],[200,-0.1141],[201,0.2232],[202,0.8664],[203,1.9208],[204,3.3336],[205,4.8493],[206,6.3817],[207,8.0173],[208,9.4762],[209,10.4219],[210,10.9299],[211,11.0586],[212,10.9069],[213,10.6036],[214,10.2375],[215,9.7773],[218,8.031],[219,7.4614],[220,6.9302],[221,6.4185],[222,5.9574],[223,5.5777],[224,5.2668],[225,4.9934],[227,4.4663],[228,4.1799],[230,3.5794],[231,3.27],[232,2.9515],[233,2.619],[234,2.2751],[235,1.9362],[238,0.9469],[239,0.6255],[240,0.3231],[241,0.0584],[242,-0.2132],[243,-0.5428],[244,-1.0015],[245,-1.6164],[246,-2.3046],[247,-2.9528],[248,-3.4599],[249,-3.8456],[250,-4.1571],[251,-4.3634],[252,-4.4373],[253,-4.3634],[254,-4.1455],[255,-3.8106],[256,-3.3741],[257,-2.8162],[258,-2.117],[259,-1.3405],[260,-0.5517],[261,0.2141],[262,0.9991],[263,1.7945],[264,2.5781],[265,3.3229],[267,4.736],[268,5.413],[269,6.0305],[270,6.5518],[271,6.9541],[272,7.2618],[273,7.495],[274,7.6762],[276,7.9512],[277,8.0724],[278,8.167],[279,8.2321],[280,8.2721],[281,8.2942],[283,8.2942],[284,8.2114],[285,8.0872],[286,7.9246],[287,7.7222],[288,7.4668],[291,6.5701],[294,5.621],[295,5.3342],[297,4.7867],[298,4.5335],[299,4.3104],[300,4.1308],[301,3.9992],[302,3.9035],[303,3.8287],[305,3.6972],[306,0]]
       }
       lowerarm_l_fk_ctrl.Scale.X {type=MovieSceneScriptingFloatChannel, default=1, const [1,307]}
       lowerarm_l_fk_ctrl.Scale.Y {type=MovieSceneScriptingFloatChannel, default=1, const [1,307]}
       lowerarm_l_fk_ctrl.Scale.Z {type=MovieSceneScriptingFloatChannel, default=1, const [1,307]}
       hand_l_fk_ctrl.Location.X {type=MovieSceneScriptingFloatChannel, default=0, thinned_from=307, keys [[0,0],[306,0]]}
       hand_l_fk_ctrl.Location.Y {type=MovieSceneScriptingFloatChannel, default=0, thinned_from=307, keys [[0,0],[306,0]]}
       hand_l_fk_ctrl.Location.Z {type=MovieSceneScriptingFloatChannel, default=0, thinned_from=307, keys [[0,0],[306,0]]}
       hand_l_fk_ctrl.Rotation.X {
        type=MovieSceneScriptingFloatChannel
        default=0
        thinned_from=307
        keys [[0,43.0812],[169,42.8777],[179,43.5613],[199,43.31],[208,43.7066],[210,43.6241],[214,42.9991],[217,42.8766],[237,43.4331],[252,44.064],[295,42.8379],[305,43.0795],[306,0]]
       }
       hand_l_fk_ctrl.Rotation.Y {
        type=MovieSceneScriptingFloatChannel
        default=0
        thinned_from=307
        keys [[0,1.1968],[1,1.1718],[2,1.1531],[3,1.1291],[5,1.0927],[6,1.0927],[7,1.0749],[9,1.0749],[10,1.1051],[11,1.1239],[12,1.1531],[14,1.1905],[15,1.1905],[16,1.1416],[17,0.9934],[18,0.8002],[20,0.3184],[21,0.1029],[22,-0.1035],[23,-0.2716],[24,-0.3795],[25,-0.4496],[26,-0.497],[27,-0.5255],[28,-0.5381],[29,-0.5381],[30,-0.516],[31,-0.478],[32,-0.4238],[33,-0.3574],[34,-0.2818],[35,-0.2005],[36,-0.1032],[37,0.0055],[38,0.122],[39,0.2404],[41,0.4738],[42,0.5984],[43,0.7248],[44,0.8431],[45,0.9469],[46,1.0327],[47,1.1105],[48,1.1832],[49,1.2496],[51,1.3682],[53,1.4687],[55,1.5523],[58,1.6528],[59,1.6778],[60,1.7113],[61,1.7353],[62,1.7532],[63,1.7783],[65,1.8118],[66,1.8118],[67,1.8304],[72,1.8304],[73,1.8097],[74,1.7828],[75,1.7452],[76,1.6976],[77,1.6338],[78,1.5639],[79,1.484],[82,1.2273],[83,1.1353],[84,1.0296],[85,0.9189],[86,0.8182],[87,0.7425],[88,0.6918],[89,0.6567],[90,0.6311],[91,0.6117],[92,0.6117],[93,0.5961],[110,0.5961],[111,0.5676],[112,0.5081],[113,0.4401],[114,0.3807],[115,0.3508],[116,0.3861],[117,0.4746],[118,0.5825],[119,0.7063],[121,0.9644],[122,1.0864],[123,1.1627],[124,1.1927],[141,1.1927],[142,1.1668],[144,1.0884],[145,1.0559],[146,1.0271],[147,1.0017],[148,0.9802],[149,0.963],[152,0.963],[153,0.9888],[154,1.0233],[156,1.1087],[157,1.1604],[158,1.2415],[159,1.3658],[160,1.5496],[161,1.7679],[162,1.9903],[164,2.3913],[165,2.605],[166,2.8002],[167,2.9418],[168,2.9958],[169,2.8646],[170,2.5123],[171,2.0128],[172,1.4363],[173,0.8518],[174,0.3186],[175,-0.2664],[176,-0.9552],[177,-1.6418],[178,-2.2281],[179,-2.6132],[180,-2.6925],[181,-2.2757],[182,-1.5973],[183,-1.0282],[184,-0.7074],[185,-0.499],[186,-0.376],[187,-0.3167],[188,-0.2771],[189,-0.2567],[190,-0.2425],[194,-0.2425],[195,-0.2822],[196,-0.3315],[197,-0.4001],[198,-0.4876],[199,-0.6399],[200,-0.9006],[201,-1.2366],[202,-1.6515],[203,-2.1113],[204,-2.6214],[205,-3.1607],[206,-3.5439],[207,-3.702],[208,-3.7414],[209,-3.6297],[210,-3.1054],[211,-2.2128],[212,-0.9018],[213,0.5995],[214,1.8694],[215,2.5943],[216,2.8537],[217,2.8739],[218,2.7379],[219,2.509],[220,2.247],[221,1.9551],[223,1.3354],[224,1.0206],[225,0.6878],[226,0.3573],[227,0.0589],[228,-0.1816],[229,-0.3716],[230,-0.5169],[231,-0.6395],[232,-0.7433],[233,-0.7887],[234,-0.7887],[235,-0.8569],[236,-1.1067],[237,-1.6149],[238,-2.2591],[239,-2.8868],[241,-3.9918],[242,-4.5248],[243,-5.0044],[244,-5.382],[245,-5.6643],[246,-5.8786],[247,-6.0453],[248,-6.1896],[249,-6.3024],[250,-6.3786],[251,-6.421],[252,-6.4351],[253,-6.3927],[254,-6.2739],[255,-6.0764],[256,-5.8109],[257,-5.4585],[258,-5.01],[259,-4.5068],[261,-3.4652],[262,-2.9111],[263,-2.3417],[264,-1.7884],[265,-1.2826],[266,-0.8441],[267,-0.4543],[268,-0.1032],[269,0.209],[270,0.4829],[271,0.7211],[272,0.9263],[273,1.1017],[274,1.2528],[275,1.3826],[276,1.4968],[277,1.5928],[278,1.6607],[279,1.7074],[281,1.7695],[282,1.8092],[283,1.8431],[284,1.859],[285,1.8867],[286,1.9546],[287,2.1031],[288,2.3413],[289,2.6197],[290,2.8698],[291,3.0345],[292,3.1275],[293,3.1759],[294,3.1961],[295,3.1961],[296,3.0542],[297,2.7633],[298,2.4112],[299,2.0747],[300,1.8321],[301,1.6618],[302,1.5036],[303,1.3701],[304,1.2725],[305,1.2124],[306,0]]
       }
       hand_l_fk_ctrl.Rotation.Z {
        type=MovieSceneScriptingFloatChannel
        default=0
        thinned_from=307
        keys [[0,0.6856],[2,0.6792],[4,0.6784],[5,0.6756],[9,0.6756],[11,0.6802],[12,0.6792],[14,0.6847],[15,0.6847],[16,0.6801],[17,0.669],[18,0.6551],[20,0.6196],[22,0.5878],[23,0.5752],[24,0.5667],[25,0.5608],[26,0.5576],[28,0.5547],[29,0.5547],[31,0.5588],[34,0.5737],[36,0.588],[42,0.6399],[44,0.6582],[45,0.666],[47,0.6781],[50,0.6923],[55,0.7095],[60,0.7212],[61,0.7199],[62,0.7242],[65,0.7286],[73,0.7295],[76,0.7211],[77,0.7172],[81,0.6933],[82,0.6851],[86,0.6568],[88,0.6466],[90,0.6416],[92,0.6424],[93,0.6398],[110,0.6398],[112,0.6323],[114,0.6232],[115,0.6219],[116,0.6244],[118,0.6389],[122,0.6754],[123,0.6816],[124,0.6836],[141,0.6836],[149,0.6664],[152,0.6664],[155,0.6737],[157,0.682],[158,0.6874],[159,0.6965],[160,0.7092],[163,0.7559],[164,0.7695],[165,0.7849],[166,0.7979],[167,0.8081],[168,0.8119],[169,0.8025],[170,0.7782],[171,0.742],[173,0.6583],[174,0.6192],[175,0.5757],[177,0.4687],[178,0.4221],[179,0.3903],[180,0.3839],[181,0.4193],[182,0.4732],[183,0.5178],[184,0.5408],[185,0.5587],[186,0.5664],[187,0.5709],[189,0.5773],[194,0.5766],[196,0.5708],[197,0.5653],[198,0.5582],[199,0.5472],[200,0.5266],[201,0.5003],[202,0.468],[203,0.4316],[204,0.3899],[205,0.3459],[206,0.3142],[207,0.3012],[208,0.2984],[209,0.3074],[210,0.3513],[211,0.4242],[212,0.5266],[213,0.6403],[214,0.7325],[215,0.7831],[216,0.8017],[217,0.8032],[218,0.7934],[219,0.7779],[221,0.7387],[224,0.6718],[226,0.622],[227,0.6001],[228,0.581],[229,0.5676],[230,0.556],[231,0.5462],[232,0.5396],[233,0.5359],[234,0.5359],[235,0.5304],[236,0.5102],[237,0.4705],[238,0.4208],[239,0.3683],[241,0.2775],[242,0.2313],[243,0.1903],[244,0.158],[245,0.1326],[246,0.114],[247,0.0984],[248,0.0848],[249,0.0749],[250,0.0682],[251,0.0644],[252,0.0632],[253,0.0669],[254,0.0784],[255,0.0957],[256,0.1199],[257,0.1503],[258,0.1898],[259,0.2333],[260,0.2776],[261,0.3206],[263,0.4129],[264,0.4573],[265,0.4969],[266,0.5311],[267,0.5613],[268,0.5879],[269,0.6109],[270,0.6315],[271,0.6493],[272,0.6641],[273,0.6773],[274,0.6882],[275,0.6976],[277,0.7123],[279,0.7208],[281,0.7259],[285,0.7334],[286,0.7378],[287,0.7484],[288,0.7658],[289,0.7857],[290,0.8025],[291,0.815],[292,0.821],[293,0.8241],[295,0.8255],[296,0.8148],[297,0.796],[299,0.7465],[300,0.7305],[302,0.7057],[304,0.6887],[305,0.6861],[306,0]]
       }
       hand_l_fk_ctrl.Scale.X {type=MovieSceneScriptingFloatChannel, default=1, const [1,307]}
       hand_l_fk_ctrl.Scale.Y {type=MovieSceneScriptingFloatChannel, default=1, const [1,307]}
       hand_l_fk_ctrl.Scale.Z {type=MovieSceneScriptingFloatChannel, default=1, const [1,307]}
       thumb_01_l_ctrl.Location.X {type=MovieSceneScriptingFloatChannel, default=0, thinned_from=307, keys [[0,0],[306,0]]}
       thumb_01_l_ctrl.Location.Y {type=MovieSceneScriptingFloatChannel, default=0, const [0,307]}
       thumb_01_l_ctrl.Location.Z {type=MovieSceneScriptingFloatChannel, default=0, thinned_from=307, keys [[0,0],[306,0]]}
       thumb_01_l_ctrl.Rotation.X {
        type=MovieSceneScriptingFloatChannel
        default=0
        thinned_from=307
        keys [[0,20.2229],[14,20.2361],[26,20.2265],[27,20.2697],[28,20.2246],[31,20.2224],[32,20.2661],[33,20.2216],[67,20.2416],[76,20.2166],[79,20.2412],[86,20.2223],[87,20.2589],[102,20.217],[103,20.249],[104,20.2083],[110,20.2564],[116,20.2223],[117,20.252],[118,20.2261],[168,20.2227],[169,20.2479],[172,20.2148],[173,20.2438],[175,20.2185],[177,20.2486],[178,20.2212],[193,20.2139],[207,20.2083],[208,20.2509],[209,20.2155],[218,20.2406],[219,20.2182],[221,20.2198],[222,20.2576],[224,20.206],[241,20.2344],[242,20.2011],[243,20.2433],[265,20.22],[268,20.2543],[269,20.2167],[305,20.2229],[306,0]]
       }
       thumb_01_l_ctrl.Rotation.Y {
        type=MovieSceneScriptingFloatChannel
        default=0
        thinned_from=307
        keys [[0,14.0036],[3,14.0045],[4,13.9811],[10,14.0082],[11,13.9924],[14,14.0082],[15,13.9923],[81,14.0119],[82,13.9969],[84,14.0131],[85,13.9854],[86,14.011],[88,13.9815],[90,14.0072],[91,13.9803],[119,14.0145],[121,13.9991],[122,14.02],[126,14.0011],[161,13.9957],[163,14.0086],[164,13.9875],[165,14.0088],[166,13.9878],[169,14.0033],[189,13.9931],[190,14.01],[191,13.9952],[204,14.0126],[205,13.9832],[206,13.9835],[207,14.0128],[217,13.9978],[218,14.0143],[219,14.0014],[220,14.0181],[238,14.0001],[239,14.0274],[241,13.9998],[264,14.0118],[265,13.9799],[291,14.0108],[293,13.992],[299,14.0107],[300,13.9915],[305,14.0035],[306,0]]
       }
       thumb_01_l_ctrl.Rotation.Z {
        type=MovieSceneScriptingFloatChannel
        default=0
        thinned_from=307
        keys [[0,-0.0014],[1,-0.0014],[2,-0.0482],[3,-0.1417],[4,-0.269],[5,-0.4223],[9,-1.1708],[10,-1.3112],[11,-1.3835],[12,-1.4048],[13,-1.4048],[14,-1.358],[15,-1.1963],[16,-0.8432],[17,-0.3356],[19,0.7034],[20,1.1577],[21,1.4172],[22,1.547],[23,1.6119],[24,1.6119],[26,1.4821],[27,1.4057],[28,1.2874],[29,1.1577],[30,1.0928],[31,0.963],[32,0.8866],[33,0.7683],[34,0.6525],[35,0.5572],[36,0.466],[37,0.356],[38,0.2647],[39,0.1922],[40,0.1238],[41,0.0719],[42,0.0325],[43,0.0137],[44,0.0097],[66,0.0097],[67,-0.0507],[68,-0.081],[76,-0.081],[77,-0.0874],[78,-0.0874],[79,-0.1242],[80,-0.1914],[81,-0.3385],[82,-0.5487],[83,-0.8037],[84,-1.0656],[85,-1.3654],[86,-1.642],[87,-1.913],[88,-2.0992],[89,-2.1662],[90,-2.2187],[91,-2.2565],[92,-2.2711],[97,-2.2711],[98,-2.376],[99,-2.6935],[100,-2.9164],[101,-2.6935],[102,-2.2479],[103,-1.812],[104,-1.4128],[105,-0.9772],[107,-0.0873],[108,0.1907],[109,0.3574],[110,0.4686],[111,0.5337],[112,0.5893],[113,0.5893],[114,0.4782],[115,0.3577],[116,0.1515],[117,-0.1032],[118,-0.3435],[119,-0.5684],[120,-0.7561],[121,-0.8387],[122,-0.85],[123,-0.8387],[124,-0.7974],[125,-0.6736],[126,-0.5085],[127,-0.3208],[128,-0.1372],[129,0.069],[130,0.234],[131,0.3398],[132,0.367],[133,0.367],[134,0.3173],[135,0.2582],[136,0.2085],[137,0.1765],[138,0.1269],[142,0.0181],[143,0.0133],[144,0.0133],[145,-0.0091],[146,-0.0091],[147,0.0029],[150,0.0029],[151,-0.0132],[152,-0.0425],[153,-0.1011],[154,-0.2183],[155,-0.3941],[156,-0.6577],[157,-0.964],[158,-1.2571],[159,-1.4785],[160,-1.6089],[161,-1.7131],[162,-1.7798],[163,-1.8085],[164,-1.797],[165,-1.767],[166,-1.7139],[167,-1.6008],[168,-1.4346],[169,-1.1926],[170,-0.8831],[171,-0.5282],[172,-0.1889],[173,0.1359],[174,0.3506],[175,0.5165],[176,0.6337],[177,0.7167],[178,0.7654],[179,0.8355],[180,0.8968],[181,0.9329],[182,0.9636],[183,0.9804],[184,0.9022],[185,0.7266],[186,0.4979],[187,0.3137],[188,0.1074],[189,-0.1075],[190,-0.3003],[191,-0.5068],[192,-0.6689],[193,-0.8086],[194,-0.9061],[195,-0.9988],[200,-1.2958],[201,-1.2958],[202,-1.177],[203,-0.9394],[204,-0.6424],[205,-0.2698],[206,0.205],[207,0.7821],[208,1.424],[209,1.9088],[210,2.383],[211,2.7246],[212,2.9118],[213,3.0054],[214,3.0521],[217,3.0521],[218,2.8439],[219,2.4907],[220,2.0484],[221,1.5077],[222,0.9842],[223,0.5242],[224,0.2515],[225,0.1962],[226,0.1493],[227,0.0961],[228,-0.1168],[229,-0.4894],[230,-0.9687],[231,-1.3416],[232,-1.5014],[234,-1.2883],[235,-1.1285],[237,-0.9154],[238,-0.7557],[239,-0.6106],[240,-0.4362],[241,-0.2764],[242,-0.1076],[243,0.1523],[244,0.4612],[245,0.8319],[246,1.2642],[247,1.7273],[248,2.1594],[249,2.6223],[250,3.1158],[251,3.5476],[252,3.7999],[253,3.9232],[254,3.9849],[255,4.0157],[258,4.0157],[259,4.0466],[260,4.0466],[261,3.9823],[262,3.6607],[263,3.2104],[264,2.6957],[265,2.1982],[266,1.783],[268,0.7526],[269,0.3132],[270,0.0554],[271,-0.0091],[282,-0.0091],[283,0.0663],[284,0.3165],[285,0.6604],[286,0.9886],[287,1.2543],[288,1.5043],[289,1.7413],[290,1.9402],[291,2.0866],[292,2.1685],[293,2.2046],[294,2.231],[295,2.231],[296,2.1227],[297,1.906],[299,1.3281],[300,1.085],[301,0.8584],[303,0.3524],[304,0.1717],[305,0.0271],[306,0]]
       }
       thumb_01_l_ctrl.Scale.X {type=MovieSceneScriptingFloatChannel, default=1, const [1,307]}
       thumb_01_l_ctrl.Scale.Y {type=MovieSceneScriptingFloatChannel, default=1, const [1,307]}
       thumb_01_l_ctrl.Scale.Z {type=MovieSceneScriptingFloatChannel, default=1, const [1,307]}
       thumb_02_l_ctrl.Location.X {type=MovieSceneScriptingFloatChannel, default=0, thinned_from=307, keys [[0,0],[44,0],[45,0.0006],[66,0.0006],[67,0],[306,0]]}
       thumb_02_l_ctrl.Location.Y {type=MovieSceneScriptingFloatChannel, default=0, thinned_from=307, keys [[0,0],[44,0],[45,-0.0008],[66,-0.0008],[67,0],[306,0]]}
       thumb_02_l_ctrl.Location.Z {type=MovieSceneScriptingFloatChannel, default=0, thinned_from=307, keys [[0,0],[44,0],[45,0.0009],[66,0.0009],[67,0],[306,0]]}
       thumb_02_l_ctrl.Rotation.X {
        type=MovieSceneScriptingFloatChannel
        default=0
        thinned_from=307
        keys [[0,0.0006],[1,0.0006],[2,-0.0003],[3,0.0006],[4,-0.0003],[5,0.0006],[6,0.0006],[7,-0.0004],[8,-0.0004],[9,0.0006],[10,0.0006],[11,-0.0004],[12,0.0006],[15,0.0005],[16,-0.0004],[17,0.0005],[18,-0.0004],[27,-0.0004],[28,0.0005],[30,0.0005],[31,-0.0005],[32,0.0005],[33,0.0004],[34,0.0001],[35,0.0003],[36,-0.0012],[38,-0.0008],[39,0.0008],[40,-0.0007],[43,-0.0005],[44,0.001],[45,-0.0076],[46,-0.0076],[47,-0.009],[51,-0.009],[52,-0.0082],[63,-0.0085],[64,-0.0077],[66,-0.008],[67,0.0007],[68,-0.0003],[77,-0.0002],[78,0.0008],[79,0.0008],[80,-0.0002],[82,-0.0003],[83,0.0017],[84,-0.0006],[85,0.0014],[86,-0.0009],[87,0.0011],[88,-0.0012],[89,0.0009],[98,0.0008],[99,-0.0005],[100,0.0018],[101,0.0017],[102,-0.0007],[103,-0.0008],[104,0.0015],[105,-0.0009],[106,-0.001],[107,0.0012],[108,-0.0012],[109,0.0011],[110,0.0011],[111,-0.0013],[114,-0.0013],[115,-0.0004],[117,-0.0008],[118,0.0008],[119,-0.0011],[121,-0.0013],[122,0.0004],[127,0.0008],[128,-0.0007],[130,-0.0005],[131,0.0026],[132,-0.0003],[133,-0.0003],[134,-0.0007],[136,-0.0007],[137,0.0022],[140,0.0022],[141,0.0019],[146,0.0019],[147,-0.0001],[150,-0.0001],[151,0.0011],[152,0.0011],[153,-0.0001],[158,-0.0001],[159,0.001],[160,-0.0002],[162,-0.0002],[168,-0.0005],[169,0.0011],[170,-0.0008],[171,0.0009],[172,0.0008],[173,-0.0011],[174,0.0006],[178,0.0005],[179,0],[194,-0.0001],[195,0.0007],[202,0.0008],[203,-0.0005],[204,-0.0005],[205,0.0006],[210,0.0002],[211,0.0021],[212,-0.0009],[216,-0.001],[217,0.002],[218,-0.001],[219,0.0021],[220,0.0019],[221,0.0021],[222,-0.0009],[225,-0.0006],[226,-0.0003],[227,0.001],[228,-0.0006],[229,-0.0006],[230,0.0004],[237,0.0005],[238,-0.0007],[239,0.0006],[240,-0.0005],[241,0.0007],[242,0.0008],[243,-0.0013],[244,0.0013],[245,0.0013],[246,-0.0013],[247,0.0012],[248,-0.0013],[249,0.0012],[250,0.0011],[251,-0.0013],[258,-0.0013],[259,0.0001],[266,0.0001],[267,-0.0014],[268,0.0001],[274,0],[275,-0.0005],[276,0.0017],[277,-0.0004],[278,0.0017],[279,-0.0004],[286,-0.0002],[287,-0.0003],[288,-0.0002],[290,-0.0001],[291,-0.0011],[296,-0.0011],[297,0.0019],[298,-0.0011],[299,-0.0011],[300,0.0019],[301,-0.0012],[302,-0.0013],[303,0.0018],[305,0.0018],[306,0]]
       }
       thumb_02_l_ctrl.Rotation.Y {
        type=MovieSceneScriptingFloatChannel
        default=0
        thinned_from=307
        keys [[0,0.0003],[1,0.0003],[2,-0.0002],[3,0.0003],[4,-0.0002],[5,0.0002],[6,0.0002],[7,-0.0002],[8,-0.0002],[9,0.0001],[10,0.0001],[11,-0.0002],[12,0.0002],[14,0.0001],[15,0.0003],[16,-0.0002],[17,0.0003],[18,-0.0001],[23,-0.0002],[24,0],[25,-0.0002],[27,-0.0002],[28,0.0002],[30,0.0002],[31,-0.0002],[32,0.0001],[33,0.0002],[35,0.0001],[36,-0.0004],[37,-0.0006],[38,-0.0006],[39,0.0004],[40,-0.0001],[41,-0.0003],[43,-0.0003],[44,0.0006],[45,0.005],[46,0.005],[47,0.0042],[51,0.0042],[52,0.0047],[63,0.0046],[64,0.0051],[65,0.0049],[66,0.0051],[67,0.0002],[69,-0.0002],[77,-0.0002],[78,0.0003],[79,0.0003],[80,-0.0002],[82,-0.0001],[83,0.0008],[84,-0.0004],[85,0.0007],[86,-0.0005],[87,0.0005],[88,-0.0006],[89,0.0005],[98,0.0005],[99,-0.0003],[100,0.0009],[101,0.0008],[102,-0.0001],[103,-0.0002],[104,0.0004],[105,-0.0003],[106,-0.0004],[107,0.0004],[108,-0.0003],[109,0.0004],[110,0.0005],[111,-0.0002],[114,-0.0002],[115,0],[117,-0.0002],[118,0.0003],[119,-0.0005],[121,-0.0006],[122,0.0003],[127,0.0003],[128,-0.0004],[129,-0.0001],[130,-0.0001],[131,0.0011],[132,-0.0003],[133,-0.0003],[134,0.0001],[136,0.0001],[137,0.0009],[146,0.0008],[147,-0.0001],[150,-0.0001],[151,0.0005],[152,0.0005],[153,-0.0001],[154,-0.0001],[155,0.0001],[156,-0.0002],[158,-0.0002],[159,0.0004],[160,-0.0002],[161,0.0001],[162,0.0001],[163,-0.0003],[168,-0.0004],[169,0.0004],[170,-0.0003],[171,0.0002],[172,0.0002],[173,-0.0004],[174,0.0001],[181,0.0001],[184,-0.0001],[185,0.0001],[186,-0.0002],[194,0],[195,0.0004],[202,0.0002],[203,-0.0002],[204,-0.0003],[205,0.0002],[210,-0.0001],[211,0.0006],[212,-0.0005],[216,-0.0004],[217,0.0008],[218,-0.0004],[219,0.0004],[220,0.001],[221,0.0007],[222,-0.0003],[223,-0.0001],[224,-0.0002],[225,0],[226,-0.0001],[227,0.0006],[228,-0.0001],[229,-0.0005],[230,0.0003],[237,0.0003],[238,-0.0002],[239,0.0003],[240,-0.0005],[241,0.0003],[242,0.0004],[243,-0.0002],[244,0.0006],[245,0.0005],[246,-0.0005],[247,0.0003],[248,-0.0006],[249,0.0003],[250,0.0004],[251,-0.0009],[258,-0.0008],[259,0.0001],[261,0.0001],[262,-0.0002],[263,0],[266,0],[267,-0.0005],[268,0],[270,0],[271,0.0003],[272,0],[275,0.0001],[276,0.0007],[277,0],[278,0.0006],[279,-0.0001],[285,-0.0001],[286,-0.0004],[287,0],[288,-0.0003],[289,0.0001],[291,-0.0005],[296,-0.0005],[297,0.0008],[298,-0.0006],[299,-0.0006],[300,0.0007],[301,-0.0007],[302,-0.0002],[303,0.0007],[305,0.0008],[306,0]]
       }
       thumb_02_l_ctrl.Rotation.Z {
        type=MovieSceneScriptingFloatChannel
        default=0
        thinned_from=307
        keys [[0,1.5286],[1,1.5286],[2,1.4804],[4,1.2878],[6,0.9023],[7,0.7576],[9,0.3715],[10,0.2266],[12,0.13],[13,0.13],[14,0.1783],[15,0.3232],[16,0.7576],[17,1.2611],[18,1.634],[19,1.7272],[21,1.8204],[22,1.8204],[23,1.8669],[24,1.8204],[25,1.6806],[26,1.401],[27,1.0744],[28,0.7009],[30,0.0464],[31,-0.2813],[33,-1.0314],[34,-1.3537],[39,-2.9537],[40,-3.1944],[41,-3.355],[44,-3.5959],[45,-3.6058],[46,-3.6058],[47,-3.6862],[50,-3.6862],[51,-3.7224],[52,-3.7689],[58,-3.7689],[59,-3.6758],[60,-3.4898],[61,-3.2574],[62,-2.9788],[63,-2.6076],[64,-2.1441],[65,-1.5889],[66,-0.9883],[67,-0.1116],[68,0.8945],[69,1.3702],[70,1.4758],[77,1.4758],[78,1.423],[79,1.423],[80,1.3702],[81,1.2117],[82,1.0003],[83,0.6965],[84,0.4777],[85,0.1494],[86,-0.0698],[87,-0.3986],[89,-0.618],[90,-0.618],[91,-0.7278],[97,-0.7278],[98,-0.8376],[99,-1.1847],[100,-1.4194],[101,-1.1847],[103,-0.2474],[104,0.1035],[106,1.0376],[108,1.7366],[111,2.0856],[114,2.0856],[115,1.875],[116,1.7018],[118,1.1819],[120,0.8347],[122,0.6611],[123,0.6611],[125,0.8348],[127,1.1819],[128,1.4419],[130,1.7884],[131,1.8174],[132,1.9616],[133,1.9616],[134,1.8173],[136,1.8173],[137,1.673],[140,1.673],[141,1.5286],[146,1.5286],[147,1.5643],[150,1.5643],[151,1.5038],[152,1.5038],[153,1.4432],[154,1.322],[155,1.1402],[156,0.8976],[158,0.2905],[159,0.0474],[161,-0.1959],[162,-0.2567],[163,-0.2177],[165,-0.2177],[167,-0.0431],[169,0.3061],[173,1.6993],[175,2.0469],[178,2.3074],[179,2.3321],[182,2.5227],[183,2.5227],[184,2.4592],[186,2.0779],[187,1.8235],[192,0.8681],[194,0.6129],[195,0.5294],[199,0.2872],[200,0.2872],[201,0.2266],[202,0.3478],[203,0.5899],[204,0.8923],[205,1.255],[206,1.7379],[208,2.9427],[209,3.4837],[210,3.9639],[211,4.1971],[212,4.4959],[213,4.4959],[214,4.6454],[216,4.6454],[217,4.496],[218,4.0474],[219,3.149],[220,2.0982],[221,0.8939],[222,-0.3143],[223,-1.5265],[224,-2.5907],[225,-3.8109],[226,-4.8824],[227,-5.8312],[228,-6.5476],[229,-7.0696],[230,-7.5271],[231,-7.7234],[232,-7.7888],[233,-7.7234],[239,-6.939],[240,-6.7432],[241,-6.6127],[242,-6.4172],[244,-5.9026],[245,-5.5172],[246,-5.0042],[247,-4.6199],[248,-4.1083],[249,-3.7251],[251,-2.7055],[252,-2.451],[253,-2.3239],[258,-2.3239],[259,-2.294],[261,-2.294],[267,-1.3725],[268,-1.2958],[271,-0.8361],[272,-0.5299],[273,-0.3004],[274,0],[275,0.2266],[277,0.6502],[281,1.0733],[283,1.496],[286,2.4452],[287,2.6559],[288,2.9716],[289,3.182],[290,3.4973],[291,3.6793],[293,3.6793],[294,3.8325],[295,3.8325],[296,3.6793],[297,3.3728],[298,3.2194],[300,2.6055],[301,2.4517],[303,1.8366],[305,1.5286],[306,0]]
       }
       thumb_02_l_ctrl.Scale.X {type=MovieSceneScriptingFloatChannel, default=1, const [1,307]}
       thumb_02_l_ctrl.Scale.Y {type=MovieSceneScriptingFloatChannel, default=1, const [1,307]}
       thumb_02_l_ctrl.Scale.Z {type=MovieSceneScriptingFloatChannel, default=1, const [1,307]}
       thumb_03_l_ctrl.Location.X {type=MovieSceneScriptingFloatChannel, default=0, thinned_from=307, keys [[0,0],[306,0]]}
       thumb_03_l_ctrl.Location.Y {type=MovieSceneScriptingFloatChannel, default=0, thinned_from=307, keys [[0,0],[306,0]]}
       thumb_03_l_ctrl.Location.Z {type=MovieSceneScriptingFloatChannel, default=0, thinned_from=307, keys [[0,0],[306,0]]}
       thumb_03_l_ctrl.Rotation.X {
        type=MovieSceneScriptingFloatChannel
        default=0
        thinned_from=307
        keys [[0,0.0005],[3,0.0004],[15,0.0002],[16,0.0003],[17,-0.0001],[23,0],[29,-0.0001],[30,0.0025],[31,-0.0001],[57,-0.0004],[58,0.0008],[60,0.0008],[61,-0.0004],[67,0.0001],[82,-0.0001],[83,-0.0004],[85,-0.0003],[86,0.0013],[87,-0.0001],[98,0],[99,0.0005],[102,0.0004],[103,-0.0013],[104,0.0002],[113,0.0001],[114,-0.0016],[115,-0.0016],[116,0],[130,0],[135,0.0002],[136,-0.0007],[146,-0.0005],[147,0.0005],[172,-0.0004],[173,0.0007],[174,0.0007],[175,-0.0005],[176,0.0007],[178,0.0007],[179,0.0002],[183,0.0002],[184,-0.0036],[185,-0.0036],[186,0.0002],[190,0.0002],[191,-0.0038],[192,-0.0038],[193,0.0001],[209,0.0002],[210,-0.0014],[211,-0.0001],[217,0],[221,-0.0011],[223,-0.0011],[224,-0.0009],[225,0.0035],[226,-0.0001],[227,0.0004],[228,-0.0003],[229,-0.0002],[230,0.0007],[231,0],[240,-0.0003],[241,0.0005],[242,0.0005],[243,-0.0002],[244,0.0014],[245,0.0013],[246,-0.0003],[248,-0.0003],[249,0.0014],[250,-0.0002],[284,-0.0001],[285,-0.0002],[290,0],[291,-0.0004],[295,-0.0004],[296,0.0006],[297,0.0007],[298,-0.0002],[303,0.0003],[304,-0.0007],[305,0.0004],[306,0]]
       }
       thumb_03_l_ctrl.Rotation.Y {
        type=MovieSceneScriptingFloatChannel
        default=0
        thinned_from=307
        keys [[0,0.0004],[3,0.0005],[4,-0.002],[5,-0.002],[6,0.0004],[16,0.0004],[17,-0.0001],[21,-0.0002],[22,0.002],[23,0.002],[24,-0.0002],[57,-0.0003],[58,0.0008],[60,0.0008],[61,-0.0002],[66,-0.0001],[67,0.0004],[70,0],[82,0.0001],[83,-0.0003],[85,-0.0002],[86,0.0011],[87,-0.0001],[98,-0.0001],[99,0.0005],[102,0.0006],[103,-0.0007],[104,-0.0008],[105,0.0005],[106,0.0005],[107,-0.0009],[108,0.0004],[113,0.0003],[114,-0.0009],[115,-0.0009],[116,-0.0001],[130,-0.0001],[131,0],[135,0.0001],[136,-0.0006],[146,-0.0003],[147,0.0004],[165,-0.0002],[166,0.0004],[167,-0.0006],[168,-0.0006],[169,0.0003],[173,-0.0001],[174,0.0007],[175,-0.0002],[176,0.0006],[178,0.0006],[179,0.0001],[183,0.0001],[184,0.0003],[190,0.0002],[191,-0.0028],[194,-0.003],[195,-0.0002],[210,-0.0001],[212,0.0001],[217,0],[221,0.0012],[223,0.0011],[224,-0.0027],[225,0.0005],[226,-0.0001],[238,0.0001],[239,-0.0007],[240,0.0001],[242,0.0001],[243,-0.0001],[246,0.0001],[247,-0.0014],[248,0.0001],[270,0.0002],[271,-0.0015],[272,0.0003],[273,0.0003],[274,-0.0015],[275,-0.0002],[284,0.0001],[285,-0.0023],[286,0.0002],[297,0],[301,0.0004],[302,-0.0004],[304,-0.0003],[305,0.0005],[306,0]]
       }
       thumb_03_l_ctrl.Rotation.Z {
        type=MovieSceneScriptingFloatChannel
        default=0
        thinned_from=307
        keys [[0,5.5942],[1,5.5942],[2,5.379],[3,5.379],[4,5.1639],[5,5.1639],[9,4.3034],[10,4.3034],[11,4.0883],[14,4.0883],[15,4.3034],[16,4.7336],[17,5.2929],[19,5.6881],[21,5.6881],[22,5.8857],[23,5.8857],[24,5.6881],[25,5.6881],[26,5.2929],[27,5.0954],[29,4.3052],[30,4.1077],[33,2.9228],[34,2.6801],[39,0.9115],[40,0.9115],[41,0.5578],[44,0.5578],[45,0.2042],[50,0.2042],[51,0.2496],[54,0.2496],[55,0.1556],[57,0.1556],[59,0.3435],[61,0.7194],[62,1.0012],[63,1.3771],[64,1.8469],[66,2.9745],[67,3.8455],[68,4.7891],[69,5.261],[70,5.497],[78,5.497],[79,5.261],[81,5.261],[82,5.025],[83,4.6582],[84,4.438],[85,4.1077],[86,3.8875],[87,3.5573],[89,3.3372],[90,3.3372],[91,3.2271],[97,3.2271],[98,3.1171],[99,2.8148],[100,2.583],[101,2.8148],[102,3.2784],[103,3.6262],[106,5.0174],[108,5.7133],[111,6.0613],[113,6.0613],[114,5.9453],[115,5.9453],[116,5.5595],[121,4.6225],[124,4.6225],[130,5.7469],[132,5.8718],[133,5.8718],[136,5.6844],[137,5.6844],[138,5.622],[139,5.622],[140,5.5595],[141,5.5595],[142,5.497],[146,5.497],[147,5.5942],[152,5.5942],[153,5.3235],[154,5.3235],[159,3.9703],[160,3.9703],[161,3.6998],[165,3.6998],[167,3.869],[168,4.0382],[169,4.292],[173,5.6459],[174,5.8998],[175,5.9845],[176,6.1538],[177,6.2384],[178,6.2384],[179,6.2815],[180,6.2815],[181,6.5661],[183,6.5661],[184,6.2815],[185,6.2815],[187,5.7122],[188,5.7122],[191,4.8586],[192,4.8586],[193,4.574],[194,4.574],[195,4.552],[196,4.4298],[197,4.4298],[198,4.3076],[199,4.3076],[200,4.1854],[201,4.1854],[202,4.3076],[204,4.7965],[205,5.2854],[206,5.6522],[209,7.4868],[210,7.8539],[211,8.3011],[212,8.3011],[213,8.6087],[217,8.6087],[218,7.9935],[220,6.1491],[224,1.2362],[225,0.3153],[226,-0.9124],[227,-1.835],[228,-2.6057],[229,-3.1196],[230,-3.5051],[231,-3.7621],[232,-3.8264],[233,-3.7621],[239,-2.9911],[243,-2.2235],[247,-0.6822],[249,0.3451],[250,0.7303],[251,1.244],[252,1.5008],[253,1.6292],[258,1.6292],[259,1.6605],[261,1.6605],[267,2.6121],[268,2.6121],[271,3.088],[273,3.7224],[274,3.8811],[275,4.1854],[278,4.846],[280,4.846],[285,5.9473],[286,6.3879],[288,6.8286],[289,7.2695],[290,7.4899],[293,7.708],[295,7.708],[296,7.6342],[297,7.413],[299,6.8233],[301,6.3811],[302,6.0864],[304,5.6444],[305,5.5707],[306,0]]
       }
       thumb_03_l_ctrl.Scale.X {type=MovieSceneScriptingFloatChannel, default=1, const [1,307]}
       thumb_03_l_ctrl.Scale.Y {type=MovieSceneScriptingFloatChannel, default=1, const [1,307]}
       thumb_03_l_ctrl.Scale.Z {type=MovieSceneScriptingFloatChannel, default=1, const [1,307]}
       index_metacarpal_l_ctrl.Location.X {type=MovieSceneScriptingFloatChannel, default=0, const [0,307]}
       index_metacarpal_l_ctrl.Location.Y {type=MovieSceneScriptingFloatChannel, default=0, thinned_from=307, keys [[0,0],[306,0]]}
       index_metacarpal_l_ctrl.Location.Z {type=MovieSceneScriptingFloatChannel, default=0, thinned_from=307, keys [[0,0],[306,0]]}
       index_metacarpal_l_ctrl.Rotation.X {
        type=MovieSceneScriptingFloatChannel
        default=0
        thinned_from=307
        keys [[0,-0.0001],[3,-0.0001],[4,0.0001],[9,-0.0002],[10,0.0001],[15,0.0001],[18,0],[21,0.0002],[22,-0.0002],[24,-0.0001],[25,0.0003],[26,-0.0002],[27,0.0002],[28,0.0002],[29,-0.0003],[30,-0.0003],[31,0.0001],[33,0],[38,0.0002],[39,-0.0004],[40,0.0002],[42,0.0002],[43,-0.0004],[50,-0.0004],[51,0.0001],[60,0.0001],[61,-0.0004],[66,-0.0004],[67,0.0002],[68,-0.0003],[78,-0.0003],[79,0.0002],[88,0.0001],[89,-0.0003],[90,-0.0002],[91,0.0001],[97,0.0001],[98,-0.0002],[100,0],[103,-0.0002],[104,0.0001],[105,-0.0003],[106,0.0001],[107,0.0001],[108,-0.0002],[111,-0.0002],[112,0.0002],[117,0.0002],[118,-0.0001],[119,-0.0001],[120,0.0002],[121,0],[125,0],[126,0.0002],[127,-0.0001],[128,0.0002],[129,0.0001],[130,-0.0001],[131,0.0003],[133,0.0003],[134,-0.0001],[141,0.0002],[142,-0.0002],[151,-0.0001],[157,0],[161,0.0001],[163,-0.0003],[164,-0.0003],[165,0.0002],[166,-0.0003],[167,0.0002],[172,-0.0001],[173,0.0004],[174,-0.0001],[175,-0.0001],[176,0.0004],[177,0.0004],[178,-0.0001],[186,-0.0003],[187,0.0001],[188,-0.0003],[189,0],[207,-0.0001],[208,0.0003],[209,0],[210,0.0001],[217,0],[218,0.0006],[219,-0.0001],[221,0],[222,-0.0006],[223,0.0001],[224,-0.0004],[226,-0.0004],[227,-0.0001],[228,-0.0002],[230,0.0002],[231,-0.0002],[232,0.0001],[233,-0.0002],[234,-0.0002],[235,0.0002],[236,-0.0001],[240,0],[241,-0.0003],[242,0.0001],[249,0.0001],[250,-0.0002],[251,0.0003],[252,-0.0001],[263,-0.0001],[264,0.0003],[265,-0.0002],[266,-0.0002],[267,0.0003],[268,0],[269,0.0001],[270,-0.0002],[284,-0.0003],[285,0.0001],[294,0.0001],[295,-0.0004],[296,0.0001],[297,-0.0004],[298,0],[302,0.0001],[303,-0.0003],[304,0.0002],[305,-0.0002],[306,0]]
       }
       index_metacarpal_l_ctrl.Rotation.Y {
        type=MovieSceneScriptingFloatChannel
        default=0
        thinned_from=307
        keys [[0,0.0495],[1,0.0491],[3,0.0491],[4,0.0495],[15,0.0494],[16,0.0491],[19,0.049],[20,0.0495],[21,0.0495],[22,0.049],[24,0.049],[25,0.0495],[28,0.0495],[29,0.049],[30,0.049],[31,0.0495],[32,0.0495],[33,0.049],[34,0.0495],[42,0.0495],[43,0.0488],[50,0.0488],[51,0.0494],[60,0.0494],[66,0.0492],[67,0.0495],[68,0.0489],[78,0.0489],[79,0.0494],[81,0.0492],[82,0.0496],[83,0.0494],[85,0.0496],[86,0.0493],[87,0.0494],[88,0.049],[90,0.0491],[91,0.0495],[97,0.0495],[98,0.0491],[99,0.0494],[102,0.0494],[103,0.049],[104,0.0494],[105,0.0489],[106,0.0498],[107,0.0492],[111,0.0491],[112,0.0495],[114,0.0496],[115,0.0492],[116,0.0495],[118,0.0495],[119,0.0491],[120,0.0494],[123,0.0494],[124,0.0491],[125,0.0494],[126,0.0495],[127,0.0492],[128,0.0495],[129,0.0495],[130,0.0492],[131,0.0496],[133,0.0496],[134,0.0492],[135,0.0497],[136,0.0493],[141,0.0495],[142,0.049],[151,0.0491],[153,0.0495],[154,0.0493],[155,0.0492],[156,0.0495],[157,0.0492],[158,0.0494],[160,0.0493],[161,0.0495],[163,0.049],[164,0.049],[165,0.0497],[166,0.049],[167,0.0497],[169,0.0497],[170,0.049],[171,0.0497],[172,0.049],[173,0.0496],[174,0.0489],[175,0.0496],[178,0.0495],[179,0.049],[180,0.049],[181,0.0495],[183,0.0495],[184,0.049],[186,0.0489],[187,0.0494],[190,0.0492],[191,0.0496],[192,0.0491],[193,0.049],[194,0.0495],[203,0.0494],[204,0.049],[205,0.0495],[217,0.0492],[218,0.05],[220,0.0487],[221,0.0495],[222,0.0488],[223,0.0495],[224,0.0488],[226,0.0488],[227,0.0495],[228,0.0491],[229,0.0495],[235,0.0494],[236,0.049],[237,0.0494],[238,0.0494],[239,0.049],[240,0.0495],[241,0.0491],[242,0.0495],[244,0.0496],[245,0.0492],[261,0.0494],[262,0.0489],[263,0.049],[264,0.0496],[265,0.0491],[266,0.0491],[267,0.0496],[268,0.0491],[269,0.0496],[270,0.049],[284,0.049],[285,0.0495],[288,0.0495],[289,0.049],[290,0.0494],[294,0.0494],[295,0.0488],[296,0.0494],[297,0.0489],[298,0.0495],[302,0.0495],[303,0.049],[304,0.0495],[305,0.0495],[306,0]]
       }
       index_metacarpal_l_ctrl.Rotation.Z {
        type=MovieSceneScriptingFloatChannel
        default=0
        thinned_from=307
        keys [[0,0.0006],[1,-0.0227],[2,-0.0693],[3,-0.1392],[4,-0.2557],[5,-0.4189],[9,-1.1647],[10,-1.3045],[11,-1.3744],[12,-1.3977],[13,-1.3977],[14,-1.3511],[15,-1.188],[16,-0.8384],[17,-0.3334],[18,0.1776],[19,0.7207],[20,1.1679],[21,1.4235],[22,1.5513],[23,1.6152],[24,1.6152],[25,1.5832],[28,1.2957],[29,1.1679],[32,0.8804],[33,0.7526],[34,0.6528],[39,0.1814],[43,-0.0071],[50,-0.0071],[51,0.011],[60,0.011],[61,-0.0252],[66,-0.0252],[67,-0.0452],[68,-0.0833],[78,-0.0833],[79,-0.1213],[80,-0.1975],[81,-0.3497],[82,-0.5401],[83,-0.794],[85,-1.3541],[86,-1.6597],[87,-1.9143],[88,-2.0925],[89,-2.1689],[90,-2.2199],[92,-2.2708],[97,-2.2708],[98,-2.3726],[99,-2.7231],[100,-2.9164],[101,-2.6955],[102,-2.2536],[103,-1.8393],[104,-1.3975],[105,-0.9832],[106,-0.5137],[107,-0.0994],[108,0.1768],[109,0.3425],[110,0.453],[111,0.5359],[112,0.5911],[113,0.5911],[114,0.5083],[115,0.3602],[116,0.1573],[117,-0.1064],[118,-0.3498],[119,-0.573],[120,-0.7353],[121,-0.8367],[122,-0.857],[123,-0.857],[124,-0.7961],[125,-0.6744],[126,-0.5121],[127,-0.3295],[129,0.0762],[130,0.2385],[131,0.3464],[132,0.3736],[133,0.3736],[136,0.2104],[137,0.1832],[138,0.1288],[142,0.0201],[143,0.0201],[144,-0.0071],[151,-0.0071],[152,-0.0401],[153,-0.0983],[154,-0.2148],[155,-0.3895],[156,-0.6516],[157,-0.9719],[158,-1.2485],[159,-1.4668],[160,-1.6124],[161,-1.7143],[162,-1.7871],[163,-1.8017],[164,-1.8017],[166,-1.7188],[167,-1.5945],[168,-1.4288],[169,-1.1802],[170,-0.8902],[171,-0.5173],[173,0.1457],[174,0.3529],[175,0.5186],[176,0.6429],[177,0.7258],[179,0.8087],[181,0.933],[182,0.9641],[183,0.9641],[184,0.902],[185,0.7155],[186,0.4979],[187,0.3114],[188,0.0939],[189,-0.0926],[190,-0.3101],[192,-0.683],[193,-0.8073],[194,-0.9006],[195,-0.9918],[200,-1.2851],[201,-1.2851],[202,-1.1678],[203,-0.9625],[204,-0.6691],[205,-0.2585],[206,0.2108],[207,0.7976],[208,1.4137],[209,1.9419],[210,2.4115],[211,2.7207],[212,2.9073],[213,3.0006],[214,3.0473],[217,3.0473],[218,2.8607],[219,2.4875],[220,2.0211],[221,1.5081],[222,0.9486],[223,0.529],[224,0.2493],[225,0.156],[226,0.156],[227,0.1034],[228,-0.1073],[229,-0.5022],[230,-0.9761],[231,-1.371],[232,-1.4763],[233,-1.4236],[234,-1.292],[235,-1.134],[236,-1.0287],[239,-0.6338],[240,-0.4495],[241,-0.2916],[242,-0.0809],[243,0.1597],[244,0.468],[245,0.8381],[246,1.2698],[247,1.7324],[248,2.1643],[249,2.627],[250,3.1207],[251,3.5528],[252,3.7997],[253,3.9232],[254,3.9849],[255,4.0158],[260,4.0158],[261,3.9509],[262,3.6634],[263,3.2163],[265,2.1947],[266,1.7479],[267,1.2692],[268,0.7267],[269,0.3119],[270,0.0248],[271,-0.0071],[282,-0.0071],[283,0.0559],[284,0.308],[286,1.0012],[288,1.5055],[290,1.9468],[291,2.0962],[292,2.1688],[294,2.2413],[295,2.2051],[296,2.1325],[298,1.6247],[299,1.3346],[300,1.0807],[301,0.8631],[303,0.3555],[304,0.1742],[305,0.0291],[306,0]]
       }
       index_metacarpal_l_ctrl.Scale.X {type=MovieSceneScriptingFloatChannel, default=1, const [1,307]}
       index_metacarpal_l_ctrl.Scale.Y {type=MovieSceneScriptingFloatChannel, default=1, const [1,307]}
       index_metacarpal_l_ctrl.Scale.Z {type=MovieSceneScriptingFloatChannel, default=1, const [1,307]}
       index_01_l_ctrl.Location.X {type=MovieSceneScriptingFloatChannel, default=0, thinned_from=307, keys [[0,0],[306,0]]}
       index_01_l_ctrl.Location.Y {type=MovieSceneScriptingFloatChannel, default=0, thinned_from=307, keys [[0,0],[306,0]]}
       index_01_l_ctrl.Location.Z {type=MovieSceneScriptingFloatChannel, default=0, thinned_from=307, keys [[0,0],[306,0]]}
       index_01_l_ctrl.Rotation.X {
        type=MovieSceneScriptingFloatChannel
        default=0
        thinned_from=307
        keys [[0,-10.0783],[80,-10.0754],[81,-10.086],[287,-10.0754],[295,-10.0872],[305,-10.0749],[306,0]]
       }
       index_01_l_ctrl.Rotation.Y {
        type=MovieSceneScriptingFloatChannel
        default=0
        thinned_from=307
        keys [[0,-9.471],[214,-9.474],[215,-9.059],[216,-8.0099],[217,-7.2096],[218,-6.8542],[219,-6.6312],[220,-6.5081],[221,-6.4728],[222,-6.467],[223,-6.4094],[224,-6.0763],[225,-5.2204],[226,-4.1749],[227,-3.3655],[228,-3.1878],[230,-3.1851],[231,-3.1772],[232,-3.1826],[233,-3.1772],[239,-3.1841],[241,-3.1811],[242,-3.1885],[243,-3.1815],[244,-3.1809],[245,-3.1873],[246,-3.1819],[248,-3.1839],[249,-3.18],[250,-3.1869],[251,-3.1818],[252,-3.1857],[260,-3.1805],[261,-3.2628],[262,-3.7689],[263,-4.5248],[264,-5.3658],[265,-6.1172],[266,-6.5303],[267,-6.861],[268,-7.0177],[269,-7.2593],[270,-7.4121],[274,-8.3458],[276,-8.6367],[277,-8.7238],[278,-8.7775],[281,-8.8086],[282,-8.858],[283,-8.966],[285,-9.337],[286,-9.4569],[287,-9.4758],[305,-9.4716],[306,0]]
       }
       index_01_l_ctrl.Rotation.Z {
        type=MovieSceneScriptingFloatChannel
        default=0
        thinned_from=307
        keys [[0,5.8532],[1,5.8532],[3,5.7283],[4,5.6036],[7,5.0424],[8,4.8241],[9,4.6368],[10,4.4811],[11,4.4185],[12,4.4185],[14,4.356],[15,4.2625],[16,3.9816],[17,3.6504],[18,3.551],[21,3.7001],[24,3.7001],[25,3.551],[26,3.2536],[27,2.906],[28,2.509],[31,1.466],[32,1.0683],[33,0.7202],[34,0.3374],[35,-0.0106],[36,-0.3638],[37,-0.712],[38,-1.0503],[39,-1.3482],[40,-1.6058],[41,-1.7928],[42,-1.9039],[43,-1.9645],[44,-1.9999],[45,-2.0303],[46,-2.0707],[47,-2.1061],[49,-2.1465],[50,-2.1465],[51,-2.1589],[52,-2.1934],[54,-2.1934],[55,-2.2283],[57,-2.2283],[58,-2.1936],[59,-2.02],[60,-1.7773],[61,-1.4305],[62,-0.9798],[63,-0.391],[64,0.3014],[65,1.1666],[66,2.0656],[67,3.3957],[68,4.9092],[69,5.6261],[70,5.7856],[78,5.7856],[79,5.7058],[80,5.7058],[81,5.5467],[82,5.3079],[83,5.0525],[85,4.5165],[86,4.1945],[87,3.9263],[88,3.7654],[91,3.6044],[97,3.6044],[98,3.4968],[99,3.1695],[100,2.9424],[101,3.1695],[102,3.6242],[103,4.0219],[104,4.4759],[105,4.8739],[107,5.7818],[108,6.0659],[109,6.2359],[110,6.293],[111,6.4062],[112,6.4633],[113,6.4633],[115,6.2308],[116,6.0115],[117,5.7483],[119,5.3095],[120,5.134],[121,5.0462],[122,5.0022],[123,5.0022],[125,5.1778],[127,5.529],[128,5.7479],[130,6.099],[132,6.246],[133,6.246],[135,6.1021],[136,6.1021],[137,6.0307],[138,6.0307],[139,5.9582],[140,5.9582],[141,5.8868],[146,5.8868],[147,5.8552],[152,5.8552],[153,5.7354],[154,5.6749],[155,5.4947],[158,4.5926],[160,4.2316],[163,4.0513],[164,4.0513],[166,4.1363],[167,4.2633],[168,4.4328],[169,4.6451],[173,6.0009],[174,6.2129],[175,6.3822],[176,6.5093],[177,6.5942],[179,6.6692],[182,6.8594],[183,6.8594],[184,6.7956],[185,6.6057],[186,6.3523],[189,5.782],[190,5.5291],[191,5.3386],[194,4.9586],[195,4.8556],[199,4.6153],[200,4.6153],[201,4.5557],[202,4.6754],[203,4.9151],[204,5.2149],[205,5.5752],[206,6.0546],[208,7.2531],[209,7.7928],[210,8.2718],[211,8.5936],[212,8.7883],[213,8.8845],[214,8.9241],[216,8.9322],[217,8.8818],[218,8.3478],[219,7.4192],[220,6.3252],[221,5.0641],[222,3.7468],[223,2.4631],[224,1.2533],[225,0.0015],[226,-1.2071],[227,-2.2279],[228,-3.0217],[229,-3.6068],[230,-3.9974],[231,-4.1932],[232,-4.2579],[233,-4.1932],[239,-3.4121],[240,-3.2174],[241,-3.0873],[242,-2.8916],[243,-2.6453],[244,-2.3298],[245,-1.9506],[248,-0.6265],[249,-0.1229],[250,0.3187],[251,0.7589],[252,1.0112],[253,1.1366],[255,1.2626],[258,1.2626],[259,1.2249],[260,1.2249],[261,1.2731],[262,1.3976],[263,1.6252],[264,1.8597],[265,2.1282],[266,2.3278],[267,2.4791],[268,2.6579],[269,2.8437],[270,3.0637],[271,3.2908],[272,3.5591],[274,4.1785],[275,4.4769],[277,4.9185],[280,5.2412],[282,5.5624],[283,5.7823],[284,6.0616],[286,6.7201],[287,6.9853],[288,7.3015],[289,7.5649],[290,7.7759],[291,7.952],[293,8.1049],[295,8.1049],[296,7.9524],[298,7.494],[299,7.1885],[302,6.502],[303,6.1964],[305,5.8911],[306,0]]
       }
       index_01_l_ctrl.Scale.X {type=MovieSceneScriptingFloatChannel, default=1, const [1,307]}
       index_01_l_ctrl.Scale.Y {type=MovieSceneScriptingFloatChannel, default=1, const [1,307]}
       index_01_l_ctrl.Scale.Z {type=MovieSceneScriptingFloatChannel, default=1, const [1,307]}
       index_02_l_ctrl.Location.X {type=MovieSceneScriptingFloatChannel, default=0, thinned_from=307, keys [[0,0],[306,0]]}
       index_02_l_ctrl.Location.Y {type=MovieSceneScriptingFloatChannel, default=0, thinned_from=307, keys [[0,0],[306,0]]}
       index_02_l_ctrl.Location.Z {type=MovieSceneScriptingFloatChannel, default=0, thinned_from=307, keys [[0,0],[306,0]]}
       index_02_l_ctrl.Rotation.X {
        type=MovieSceneScriptingFloatChannel
        default=0
        thinned_from=307
        keys [[0,0],[4,0],[9,-0.0001],[15,0.0001],[16,-0.0001],[17,0.0002],[18,0.0002],[19,-0.0001],[24,-0.0001],[25,0.0002],[26,-0.0001],[27,-0.0002],[28,0.0001],[29,-0.0002],[30,0.0001],[31,-0.0002],[32,0.0001],[36,-0.0001],[37,0.0001],[38,0.0001],[39,-0.0002],[59,0.0001],[66,-0.0001],[67,0.0001],[79,0.0001],[80,-0.0002],[81,0.0001],[83,0.0002],[84,-0.0001],[85,0.0002],[86,-0.0001],[87,0.0002],[88,-0.0001],[90,-0.0001],[91,0.0002],[98,0.0002],[99,-0.0001],[100,0.0002],[101,0.0002],[102,-0.0001],[103,-0.0001],[104,0.0002],[105,-0.0001],[106,-0.0001],[107,0.0002],[108,-0.0002],[109,0.0001],[114,0.0002],[115,0],[119,0.0001],[128,0],[146,0],[153,0.0002],[154,-0.0001],[155,0.0002],[156,0.0002],[157,-0.0001],[158,0.0002],[159,-0.0001],[161,-0.0001],[162,0.0002],[165,0.0002],[166,-0.0001],[168,-0.0001],[169,0.0004],[170,-0.0002],[173,-0.0001],[174,0.0003],[175,-0.0002],[178,-0.0002],[179,0.0002],[180,-0.0002],[186,-0.0002],[187,0.0002],[188,-0.0001],[189,0.0002],[190,-0.0001],[191,0.0002],[192,-0.0002],[193,-0.0001],[194,0.0002],[195,0.0002],[196,-0.0001],[199,-0.0001],[200,0.0002],[201,0.0002],[202,-0.0001],[203,-0.0001],[204,0.0002],[211,0.0001],[212,-0.0001],[217,-0.0002],[218,0.0001],[219,-0.0001],[223,0],[224,-0.0002],[225,0],[228,-0.0001],[237,0.0001],[238,-0.0001],[239,0.0001],[240,-0.0001],[242,0.0001],[248,0],[250,-0.0001],[251,0.0001],[252,-0.0001],[254,0.0001],[261,-0.0001],[262,0],[266,-0.0001],[277,0],[280,-0.0001],[281,0.0001],[282,-0.0001],[287,0.0001],[288,-0.0001],[289,0.0001],[290,-0.0001],[306,0]]
       }
       index_02_l_ctrl.Rotation.Y {
        type=MovieSceneScriptingFloatChannel
        default=0
        thinned_from=307
        keys [[0,0],[15,0],[16,-0.0001],[18,0.0001],[24,-0.0001],[25,0.0001],[26,-0.0001],[28,0],[39,0],[40,-0.0001],[83,0.0001],[84,-0.0001],[85,0.0001],[86,-0.0001],[87,0.0001],[88,-0.0001],[102,0],[103,-0.0001],[104,0.0001],[106,-0.0001],[107,0.0001],[108,-0.0001],[109,0.0002],[118,-0.0001],[146,0],[153,0.0001],[154,-0.0001],[155,0.0001],[156,0.0001],[157,-0.0001],[158,0.0002],[159,-0.0001],[161,-0.0001],[162,0.0002],[165,0.0002],[166,-0.0001],[168,-0.0001],[169,0.0003],[172,0.0001],[173,-0.0001],[174,0.0002],[175,-0.0001],[178,-0.0001],[179,0.0002],[180,-0.0001],[186,-0.0001],[187,0.0002],[188,-0.0001],[189,0.0002],[190,-0.0001],[191,0.0002],[193,-0.0001],[194,0.0002],[211,0],[212,-0.0001],[223,0],[224,-0.0001],[226,0.0001],[228,-0.0001],[236,0.0001],[238,-0.0001],[239,0.0001],[240,-0.0001],[241,0.0001],[246,0],[266,0],[287,0.0001],[288,0],[289,0.0001],[306,0]]
       }
       index_02_l_ctrl.Rotation.Z {
        type=MovieSceneScriptingFloatChannel
        default=0
        thinned_from=307
        keys [[0,18.3674],[1,18.3028],[2,18.3028],[5,17.9149],[10,16.9442],[11,16.8794],[13,16.8794],[15,16.7499],[16,16.4907],[17,16.0717],[18,15.972],[20,16.1715],[24,16.1715],[26,15.7723],[28,14.9732],[29,14.6732],[30,14.273],[31,13.9726],[33,13.1708],[36,12.1586],[37,11.7397],[38,11.4043],[41,10.6489],[43,10.4809],[45,10.4809],[46,10.3968],[47,10.3968],[48,10.3128],[50,10.3128],[51,10.2788],[58,10.2788],[59,10.4195],[60,10.7006],[61,11.0518],[62,11.543],[63,12.1038],[64,12.8039],[65,13.6428],[66,14.5501],[67,15.8588],[68,17.3741],[69,18.0906],[70,18.2498],[79,18.2498],[80,18.1702],[81,18.0111],[83,17.5],[84,17.2805],[85,16.9512],[86,16.7315],[87,16.4018],[89,16.1819],[90,16.1819],[91,16.072],[97,16.072],[98,15.962],[99,15.6276],[100,15.3945],[101,15.6276],[103,16.5588],[104,16.9076],[106,17.837],[108,18.5331],[111,18.8809],[114,18.8809],[115,18.7054],[116,18.49],[117,18.2315],[118,18.016],[119,17.7574],[120,17.5849],[121,17.4986],[123,17.4986],[124,17.5417],[125,17.6711],[127,18.016],[128,18.2315],[130,18.5762],[132,18.7255],[133,18.7255],[136,18.5274],[137,18.5274],[138,18.4614],[139,18.4614],[140,18.3954],[142,18.3954],[143,18.3293],[153,18.3053],[154,18.1786],[157,17.4177],[158,17.037],[162,16.5289],[165,16.5289],[166,16.7074],[167,16.7074],[169,17.0644],[173,18.4902],[176,19.0242],[177,19.0242],[178,19.202],[179,19.1405],[180,19.2739],[184,19.2739],[185,19.1405],[187,18.6067],[188,18.4731],[189,18.206],[190,18.0724],[191,17.8051],[194,17.404],[195,17.404],[196,17.2895],[197,17.2895],[198,17.1667],[199,17.1667],[200,17.0438],[201,17.0438],[202,17.1667],[204,17.6579],[205,18.026],[206,18.5165],[208,19.7414],[210,20.7197],[211,21.0572],[212,21.2965],[213,21.3763],[217,21.3763],[218,20.8178],[219,19.9393],[220,18.8198],[223,14.9669],[224,13.7577],[225,12.4649],[226,11.25],[227,10.2607],[228,9.4731],[229,8.8815],[230,8.4867],[231,8.2892],[232,8.2233],[233,8.2892],[239,9.0788],[240,9.276],[241,9.4074],[242,9.6044],[243,9.8435],[244,10.158],[245,10.5352],[248,11.8531],[250,12.8549],[251,13.2301],[252,13.5425],[255,13.7299],[261,13.7299],[262,13.8827],[265,14.6092],[266,14.8027],[267,14.9479],[270,15.528],[271,15.7696],[275,16.9441],[277,17.382],[278,17.5461],[281,17.8742],[282,18.0381],[283,18.2567],[284,18.5299],[286,19.1849],[287,19.4577],[288,19.7849],[289,20.0029],[290,20.2754],[291,20.4194],[292,20.5101],[295,20.5646],[296,20.4557],[297,20.2379],[299,19.6749],[303,18.7114],[304,18.5113],[305,18.3839],[306,0]]
       }
       index_02_l_ctrl.Scale.X {type=MovieSceneScriptingFloatChannel, default=1, const [1,307]}
       index_02_l_ctrl.Scale.Y {type=MovieSceneScriptingFloatChannel, default=1, const [1,307]}
       index_02_l_ctrl.Scale.Z {type=MovieSceneScriptingFloatChannel, default=1, const [1,307]}
       index_03_l_ctrl.Location.X {type=MovieSceneScriptingFloatChannel, default=0, thinned_from=307, keys [[0,0],[306,0]]}
       index_03_l_ctrl.Location.Y {type=MovieSceneScriptingFloatChannel, default=0, thinned_from=307, keys [[0,0],[306,0]]}
       index_03_l_ctrl.Location.Z {type=MovieSceneScriptingFloatChannel, default=0, thinned_from=307, keys [[0,0],[306,0]]}
       index_03_l_ctrl.Rotation.X {
        type=MovieSceneScriptingFloatChannel
        default=0
        thinned_from=307
        keys [[0,-0.1116],[5,-0.1117],[6,-0.1098],[7,-0.1098],[8,-0.1117],[9,-0.1098],[15,-0.1098],[16,-0.1118],[17,-0.1111],[18,-0.1111],[19,-0.1118],[24,-0.1118],[25,-0.1111],[26,-0.1112],[27,-0.1119],[28,-0.1112],[29,-0.1119],[30,-0.1112],[31,-0.1119],[32,-0.1112],[33,-0.1112],[34,-0.1115],[50,-0.1113],[51,-0.1107],[60,-0.1107],[61,-0.1117],[65,-0.1119],[66,-0.1109],[67,-0.1107],[68,-0.1117],[83,-0.1117],[84,-0.1101],[85,-0.1101],[86,-0.1118],[87,-0.1118],[88,-0.1101],[97,-0.1101],[98,-0.1118],[99,-0.1119],[100,-0.1111],[101,-0.1119],[103,-0.1118],[104,-0.111],[105,-0.1117],[107,-0.1116],[108,-0.1108],[111,-0.1108],[112,-0.1115],[115,-0.1116],[116,-0.1104],[117,-0.1118],[120,-0.112],[121,-0.1108],[124,-0.1108],[125,-0.112],[130,-0.1116],[131,-0.1105],[132,-0.1116],[133,-0.1116],[134,-0.1105],[146,-0.1106],[147,-0.1116],[152,-0.1116],[153,-0.1098],[154,-0.1098],[155,-0.1117],[164,-0.1118],[165,-0.1115],[180,-0.1113],[182,-0.1111],[186,-0.1113],[187,-0.1112],[188,-0.1114],[191,-0.1113],[193,-0.1115],[194,-0.1113],[195,-0.1117],[196,-0.1113],[201,-0.1113],[202,-0.1118],[203,-0.1113],[204,-0.1113],[205,-0.1117],[206,-0.1112],[209,-0.1109],[210,-0.1112],[216,-0.1111],[217,-0.11],[218,-0.1112],[219,-0.1103],[220,-0.1116],[224,-0.1119],[225,-0.1107],[226,-0.1117],[227,-0.1113],[230,-0.1113],[231,-0.1104],[232,-0.1104],[233,-0.1113],[236,-0.1113],[237,-0.1104],[239,-0.1104],[240,-0.1113],[242,-0.1113],[243,-0.1107],[244,-0.1116],[246,-0.1117],[247,-0.1109],[248,-0.1118],[249,-0.1109],[250,-0.111],[251,-0.1119],[264,-0.1119],[265,-0.1105],[267,-0.1105],[268,-0.1119],[270,-0.1119],[271,-0.1105],[272,-0.1118],[274,-0.1118],[275,-0.111],[277,-0.111],[278,-0.1117],[281,-0.1117],[282,-0.1109],[283,-0.1109],[284,-0.1116],[286,-0.1115],[287,-0.1107],[288,-0.1114],[289,-0.1106],[290,-0.1106],[291,-0.111],[295,-0.111],[296,-0.1108],[297,-0.1108],[298,-0.1111],[299,-0.1112],[300,-0.111],[302,-0.111],[303,-0.1113],[304,-0.1111],[305,-0.1114],[306,0]]
       }
       index_03_l_ctrl.Rotation.Y {
        type=MovieSceneScriptingFloatChannel
        default=0
        thinned_from=307
        keys [[0,0.008],[2,0.008],[3,0.0074],[4,0.0074],[6,0.0065],[7,0.0065],[8,0.0057],[9,0.0054],[12,0.0054],[13,0.0048],[15,0.0048],[17,0.0035],[18,0.0033],[20,0.0036],[24,0.0036],[25,0.0033],[27,0.0021],[38,-0.0056],[40,-0.0067],[42,-0.0073],[58,-0.0078],[59,-0.0075],[61,-0.0063],[62,-0.0055],[63,-0.0043],[64,-0.0029],[65,-0.0014],[66,0.0004],[67,0.003],[68,0.0059],[69,0.0075],[80,0.0075],[83,0.0065],[84,0.0056],[87,0.004],[88,0.0038],[97,0.0038],[99,0.0026],[100,0.0022],[101,0.0026],[103,0.0044],[107,0.0077],[108,0.0082],[112,0.0091],[113,0.0091],[115,0.0087],[116,0.008],[120,0.0065],[121,0.0062],[124,0.0062],[130,0.0083],[132,0.0087],[133,0.0087],[134,0.0084],[136,0.0084],[137,0.0081],[141,0.0081],[142,0.0078],[146,0.0078],[152,0.008],[153,0.0075],[154,0.0075],[158,0.0053],[159,0.0049],[160,0.0049],[161,0.0044],[166,0.0045],[168,0.005],[173,0.0081],[175,0.0089],[183,0.0098],[184,0.0096],[192,0.0066],[199,0.0054],[201,0.0054],[203,0.006],[204,0.0066],[206,0.0082],[209,0.0116],[210,0.0126],[211,0.0132],[212,0.0136],[213,0.0136],[214,0.0139],[216,0.0139],[217,0.0136],[218,0.0127],[220,0.0089],[225,-0.0036],[226,-0.0059],[227,-0.0079],[228,-0.0095],[229,-0.0106],[230,-0.0114],[231,-0.0117],[234,-0.0116],[235,-0.0111],[236,-0.0111],[237,-0.0106],[242,-0.0091],[244,-0.0081],[249,-0.0039],[251,-0.002],[253,-0.0013],[261,-0.0012],[271,0.0028],[272,0.0034],[273,0.0038],[274,0.0047],[278,0.0063],[282,0.0072],[283,0.0077],[286,0.0095],[290,0.0115],[292,0.0121],[295,0.0121],[296,0.0119],[304,0.0081],[305,0.0079],[306,0]]
       }
       index_03_l_ctrl.Rotation.Z {
        type=MovieSceneScriptingFloatChannel
        default=0
        thinned_from=307
        keys [[0,27.5182],[2,27.5182],[3,27.238],[4,27.238],[6,26.6777],[7,26.6777],[9,26.1172],[12,26.1172],[13,25.837],[15,25.837],[16,25.5567],[17,25.1986],[18,25.097],[20,25.3002],[24,25.3002],[25,25.097],[27,24.4872],[28,24.0806],[29,23.7756],[30,23.3689],[31,23.0638],[33,22.2501],[38,20.5386],[40,19.9813],[41,19.7989],[42,19.6824],[50,19.4289],[51,19.3518],[58,19.3518],[59,19.4938],[60,19.7779],[62,20.6297],[63,21.1972],[64,21.9064],[66,23.6073],[67,24.937],[68,26.455],[69,27.298],[80,27.298],[82,26.9608],[83,26.744],[84,26.2573],[88,25.2838],[97,25.2838],[100,24.5239],[101,24.7538],[103,25.6733],[104,26.018],[107,27.3967],[108,27.6264],[112,28.0859],[113,28.0859],[115,27.8795],[116,27.5108],[121,26.5892],[124,26.5892],[130,27.6952],[131,27.7246],[132,27.8795],[133,27.8795],[134,27.7246],[136,27.7246],[137,27.5698],[141,27.5698],[142,27.415],[146,27.415],[147,27.5182],[152,27.5182],[153,27.2528],[154,27.2528],[159,25.9255],[160,25.9255],[161,25.6599],[165,25.6936],[166,25.7362],[168,25.9918],[169,26.2475],[172,27.27],[174,27.8238],[175,27.9941],[177,28.1645],[183,28.4365],[184,28.3742],[191,26.9423],[194,26.5376],[195,26.4884],[196,26.3671],[199,26.1852],[201,26.1852],[202,26.3065],[203,26.4884],[204,26.7916],[206,27.6403],[208,28.8525],[209,29.398],[210,29.8829],[211,30.24],[212,30.4014],[213,30.4014],[214,30.5627],[216,30.5627],[217,30.4014],[218,29.9173],[220,27.981],[225,21.5203],[226,20.3876],[227,19.3518],[228,18.5433],[229,18.004],[230,17.5994],[231,17.3295],[232,17.3295],[233,17.4645],[234,17.4645],[235,17.7343],[236,17.7343],[239,18.1389],[240,18.4085],[242,18.6781],[243,18.9373],[247,20.481],[248,20.9953],[249,21.3808],[251,22.4085],[252,22.6653],[253,22.7937],[261,22.8187],[271,24.8166],[272,25.2159],[273,25.4156],[274,25.8148],[278,26.6808],[282,27.1203],[283,27.3401],[286,28.3289],[287,28.5486],[288,28.8782],[291,29.516],[292,29.6265],[295,29.6634],[296,29.5528],[298,29.0737],[300,28.5208],[304,27.5993],[305,27.4887],[306,0]]
       }
       index_03_l_ctrl.Scale.X {type=MovieSceneScriptingFloatChannel, default=1, const [1,307]}
       index_03_l_ctrl.Scale.Y {type=MovieSceneScriptingFloatChannel, default=1, const [1,307]}
       index_03_l_ctrl.Scale.Z {type=MovieSceneScriptingFloatChannel, default=1, const [1,307]}
       middle_metacarpal_l_ctrl.Location.X {type=MovieSceneScriptingFloatChannel, default=0, thinned_from=307, keys [[0,0],[306,0]]}
       middle_metacarpal_l_ctrl.Location.Y {type=MovieSceneScriptingFloatChannel, default=0, thinned_from=307, keys [[0,0],[306,0]]}
       middle_metacarpal_l_ctrl.Location.Z {type=MovieSceneScriptingFloatChannel, default=0, thinned_from=307, keys [[0,0],[306,0]]}
       middle_metacarpal_l_ctrl.Rotation.X {
        type=MovieSceneScriptingFloatChannel
        default=0
        thinned_from=307
        keys [[0,0],[78,0.0001],[139,0.0001],[140,-0.0001],[196,0],[197,0.0001],[198,0],[199,0.0001],[201,0],[203,0.0001],[306,0]]
       }
       middle_metacarpal_l_ctrl.Rotation.Y {type=MovieSceneScriptingFloatChannel, default=0, thinned_from=307, keys [[0,0],[306,0]]}
       middle_metacarpal_l_ctrl.Rotation.Z {
        type=MovieSceneScriptingFloatChannel
        default=0
        thinned_from=307
        keys [[0,0],[1,-0.0139],[2,-0.0599],[3,-0.1374],[4,-0.2581],[5,-0.4276],[6,-0.6144],[7,-0.7925],[8,-0.9792],[9,-1.1602],[10,-1.3009],[11,-1.3756],[12,-1.3957],[13,-1.3957],[14,-1.3584],[15,-1.1947],[16,-0.8298],[17,-0.3263],[18,0.189],[19,0.73],[20,1.1707],[21,1.4206],[22,1.5642],[23,1.6272],[24,1.6272],[25,1.578],[26,1.4934],[27,1.3891],[28,1.2809],[30,1.0821],[31,0.9759],[32,0.8657],[33,0.7595],[34,0.6547],[37,0.3636],[38,0.2721],[39,0.1917],[40,0.1224],[41,0.0697],[42,0.0337],[43,0.0143],[44,0],[61,0],[62,-0.0122],[63,-0.0122],[64,-0.026],[65,-0.026],[66,-0.0433],[67,-0.0433],[68,-0.0714],[77,-0.0714],[79,-0.1257],[80,-0.1892],[81,-0.3342],[82,-0.5517],[83,-0.7974],[84,-1.0768],[85,-1.3592],[86,-1.6574],[87,-1.9241],[88,-2.0873],[89,-2.1626],[90,-2.2097],[91,-2.2505],[92,-2.2725],[97,-2.2725],[98,-2.3603],[99,-2.7142],[100,-2.9164],[101,-2.6902],[102,-2.2516],[103,-1.83],[104,-1.4016],[105,-0.9697],[106,-0.5275],[107,-0.1058],[108,0.1857],[109,0.3469],[110,0.4532],[111,0.5389],[112,0.58],[113,0.58],[114,0.4977],[115,0.3599],[116,0.1499],[117,-0.1102],[118,-0.3503],[119,-0.5704],[120,-0.7454],[121,-0.8329],[122,-0.8579],[123,-0.8579],[124,-0.7879],[125,-0.6754],[126,-0.5179],[127,-0.3253],[128,-0.1302],[129,0.0723],[130,0.2424],[131,0.3352],[132,0.3611],[133,0.3611],[134,0.3158],[135,0.2575],[136,0.2122],[139,0.0957],[141,0.0439],[142,0.0245],[143,0.0115],[144,0],[151,0],[152,-0.0368],[153,-0.1019],[154,-0.214],[155,-0.3913],[156,-0.6554],[157,-0.9665],[158,-1.2559],[159,-1.4694],[160,-1.614],[161,-1.7189],[162,-1.7804],[163,-1.7996],[164,-1.7996],[165,-1.769],[166,-1.7103],[167,-1.6057],[168,-1.4425],[169,-1.2001],[170,-0.8838],[171,-0.5318],[172,-0.1822],[173,0.1265],[174,0.3613],[175,0.5221],[176,0.6318],[177,0.7109],[178,0.7696],[179,0.8227],[180,0.8849],[181,0.9365],[182,0.9662],[183,0.9662],[184,0.9116],[185,0.7289],[186,0.5089],[187,0.31],[188,0.1082],[189,-0.1023],[190,-0.3098],[191,-0.5029],[192,-0.6693],[193,-0.8022],[194,-0.9055],[195,-0.9869],[197,-1.1181],[198,-1.1764],[199,-1.2419],[200,-1.2784],[201,-1.2784],[202,-1.1764],[203,-0.9577],[204,-0.6516],[205,-0.2653],[206,0.2086],[207,0.7991],[208,1.4044],[209,1.9369],[210,2.4038],[211,2.7315],[212,2.9154],[213,3.0188],[214,3.0562],[216,3.0562],[217,3.0447],[218,2.8407],[219,2.4872],[220,2.0419],[222,0.9706],[223,0.5256],[224,0.2672],[225,0.1782],[226,0.1581],[227,0.1017],[228,-0.1026],[229,-0.5079],[230,-0.9812],[231,-1.3605],[232,-1.4821],[233,-1.414],[234,-1.2811],[235,-1.1401],[236,-1.0169],[237,-0.8969],[238,-0.7689],[239,-0.6213],[240,-0.456],[241,-0.2841],[242,-0.0896],[243,0.1491],[244,0.4571],[245,0.8416],[246,1.274],[247,1.72],[248,2.1545],[249,2.6273],[250,3.1233],[251,3.5466],[252,3.8072],[253,3.926],[254,3.9892],[255,4.0141],[260,4.0141],[261,3.9634],[262,3.6778],[263,3.2256],[264,2.7023],[265,2.2068],[266,1.7589],[267,1.2517],[268,0.7328],[269,0.2971],[270,0.0396],[271,0],[282,0],[283,0.0698],[284,0.3184],[285,0.6603],[286,0.9866],[287,1.2509],[288,1.5035],[289,1.7406],[290,1.9388],[291,2.0826],[292,2.1675],[293,2.21],[294,2.2234],[295,2.2234],[296,2.1161],[297,1.8927],[298,1.6112],[299,1.3253],[300,1.0863],[301,0.8585],[302,0.6061],[303,0.365],[304,0.164],[305,0.0345],[306,0]]
       }
       middle_metacarpal_l_ctrl.Scale.X {type=MovieSceneScriptingFloatChannel, default=1, const [1,307]}
       middle_metacarpal_l_ctrl.Scale.Y {type=MovieSceneScriptingFloatChannel, default=1, const [1,307]}
       middle_metacarpal_l_ctrl.Scale.Z {type=MovieSceneScriptingFloatChannel, default=1, const [1,307]}
       middle_01_l_ctrl.Location.X {type=MovieSceneScriptingFloatChannel, default=0, thinned_from=307, keys [[0,0],[306,0]]}
       middle_01_l_ctrl.Location.Y {type=MovieSceneScriptingFloatChannel, default=0, thinned_from=307, keys [[0,0],[306,0]]}
       middle_01_l_ctrl.Location.Z {type=MovieSceneScriptingFloatChannel, default=0, thinned_from=307, keys [[0,0],[306,0]]}
       middle_01_l_ctrl.Rotation.X {
        type=MovieSceneScriptingFloatChannel
        default=0
        thinned_from=307
        keys [[0,-19.1213],[214,-19.1216],[215,-19.5618],[216,-20.707],[217,-21.5746],[218,-21.9733],[219,-22.2155],[220,-22.3438],[222,-22.3931],[223,-22.4558],[224,-22.815],[225,-23.7389],[226,-24.8826],[227,-25.7568],[228,-25.9615],[260,-25.9616],[261,-25.8524],[262,-25.3048],[263,-24.473],[264,-23.5666],[265,-22.7977],[266,-22.3037],[267,-21.9819],[268,-21.7442],[270,-21.3625],[274,-20.387],[275,-20.1817],[276,-20.021],[277,-19.9206],[282,-19.7833],[283,-19.6625],[285,-19.2687],[286,-19.14],[305,-19.1214],[306,0]]
       }
       middle_01_l_ctrl.Rotation.Y {type=MovieSceneScriptingFloatChannel, default=0, thinned_from=307, keys [[0,-7.9613],[305,-7.9614],[306,0]]}
       middle_01_l_ctrl.Rotation.Z {
        type=MovieSceneScriptingFloatChannel
        default=0
        thinned_from=307
        keys [[0,9.9549],[1,9.9428],[2,9.8943],[3,9.8134],[4,9.684],[9,8.7338],[10,8.5882],[11,8.5194],[14,8.4506],[15,8.3617],[16,8.0746],[17,7.7196],[18,7.6256],[20,7.7573],[22,7.8074],[23,7.8074],[24,7.7917],[25,7.6224],[26,7.334],[28,6.6127],[30,5.9196],[33,4.7932],[34,4.4288],[37,3.3796],[38,3.0436],[39,2.7436],[40,2.4851],[41,2.2988],[42,2.185],[43,2.1281],[44,2.0919],[45,2.0608],[46,2.0194],[47,1.9833],[48,1.9625],[53,1.8911],[54,1.8911],[55,1.8738],[57,1.8738],[58,1.9106],[59,2.0631],[60,2.3206],[61,2.6754],[62,3.1208],[63,3.6873],[64,4.4059],[65,5.2495],[66,6.1576],[67,7.4876],[68,8.9967],[69,9.7232],[70,9.8767],[77,9.8767],[79,9.828],[80,9.7647],[81,9.6184],[82,9.4015],[83,9.157],[85,8.5936],[86,8.2985],[87,8.0303],[88,7.8692],[89,7.7954],[91,7.7015],[97,7.688],[98,7.5942],[99,7.2414],[100,7.0389],[101,7.2627],[102,7.7033],[105,8.9858],[106,9.4298],[107,9.849],[108,10.1403],[109,10.3036],[110,10.4102],[111,10.4919],[112,10.538],[113,10.538],[114,10.4563],[115,10.3159],[116,10.1036],[117,9.8434],[118,9.6046],[119,9.3869],[120,9.2117],[121,9.1215],[122,9.0949],[123,9.0949],[124,9.1692],[125,9.2807],[126,9.4347],[129,10.0292],[130,10.1992],[131,10.2876],[132,10.319],[133,10.319],[139,10.0521],[143,9.9695],[151,9.9578],[152,9.9192],[153,9.8552],[154,9.7423],[155,9.5654],[156,9.3018],[157,8.9894],[158,8.6995],[159,8.4886],[160,8.3418],[161,8.2363],[162,8.1761],[163,8.1556],[164,8.1556],[165,8.1874],[166,8.2457],[167,8.3517],[168,8.5107],[169,8.7544],[170,9.0723],[172,9.7718],[173,10.0843],[174,10.3174],[175,10.4763],[176,10.5876],[178,10.7253],[181,10.8933],[182,10.9247],[183,10.9247],[184,10.8699],[185,10.6821],[186,10.4631],[191,9.4539],[192,9.2895],[193,9.1567],[194,9.0472],[199,8.7191],[200,8.6749],[201,8.6749],[202,8.7856],[203,9],[204,9.3029],[205,9.6945],[206,10.1675],[207,10.7513],[208,11.3572],[209,11.8892],[210,12.3622],[211,12.6856],[212,12.8728],[213,12.9752],[214,13.0142],[216,13.0157],[217,12.9687],[218,12.2477],[219,11.0048],[220,9.5407],[221,7.8928],[224,2.643],[225,0.7646],[226,-1.0849],[227,-2.6688],[228,-3.7821],[229,-4.4413],[230,-4.787],[231,-4.9143],[232,-4.9303],[233,-4.862],[234,-4.7301],[235,-4.5868],[237,-4.3458],[238,-4.2162],[239,-4.0685],[241,-3.7321],[242,-3.5368],[243,-3.2966],[244,-2.9921],[245,-2.6059],[246,-2.1732],[247,-1.7251],[248,-1.2925],[249,-0.8214],[250,-0.3233],[251,0.097],[252,0.3614],[253,0.4781],[254,0.5404],[255,0.5675],[260,0.5675],[261,0.646],[262,1.0126],[263,1.5984],[264,2.2777],[265,2.9189],[266,3.4594],[268,4.5184],[269,5.0379],[270,5.5358],[273,6.9467],[274,7.3836],[275,7.7928],[276,8.1597],[277,8.4673],[278,8.7069],[279,8.9024],[281,9.2606],[282,9.4794],[283,9.7478],[286,10.7258],[289,11.6063],[290,11.8444],[291,12.0114],[292,12.1125],[293,12.1631],[295,12.1792],[296,12.0712],[297,11.8505],[299,11.2806],[301,10.814],[303,10.3222],[304,10.12],[305,9.989],[306,0]]
       }
       middle_01_l_ctrl.Scale.X {type=MovieSceneScriptingFloatChannel, default=1, const [1,307]}
       middle_01_l_ctrl.Scale.Y {type=MovieSceneScriptingFloatChannel, default=1, const [1,307]}
       middle_01_l_ctrl.Scale.Z {type=MovieSceneScriptingFloatChannel, default=1, const [1,307]}
       middle_02_l_ctrl.Location.X {type=MovieSceneScriptingFloatChannel, default=0, thinned_from=307, keys [[0,-0.0452],[305,-0.0452],[306,0]]}
       middle_02_l_ctrl.Location.Y {type=MovieSceneScriptingFloatChannel, default=0, thinned_from=307, keys [[0,0.0507],[305,0.0507],[306,0]]}
       middle_02_l_ctrl.Location.Z {type=MovieSceneScriptingFloatChannel, default=0, thinned_from=307, keys [[0,0.1146],[305,0.1146],[306,0]]}
       middle_02_l_ctrl.Rotation.X {
        type=MovieSceneScriptingFloatChannel
        default=0
        thinned_from=307
        keys [[0,-3.0347],[2,-3.0347],[10,-3.0902],[14,-3.0957],[16,-3.112],[17,-3.1267],[18,-3.1306],[24,-3.1247],[27,-3.1536],[39,-3.3119],[42,-3.3316],[58,-3.3412],[60,-3.3277],[62,-3.2996],[63,-3.2795],[65,-3.2212],[66,-3.1871],[67,-3.1354],[68,-3.0751],[69,-3.0447],[70,-3.0381],[80,-3.0431],[83,-3.067],[87,-3.1136],[98,-3.132],[99,-3.1459],[100,-3.1525],[101,-3.1433],[107,-3.0387],[109,-3.0194],[113,-3.0097],[115,-3.0191],[119,-3.0605],[123,-3.0692],[125,-3.0637],[128,-3.042],[129,-3.0303],[132,-3.0173],[136,-3.0207],[137,-3.0274],[141,-3.0274],[142,-3.0342],[152,-3.0342],[154,-3.0399],[159,-3.0951],[162,-3.1081],[167,-3.1047],[174,-3.0192],[176,-3.0078],[177,-3.0078],[178,-3.0002],[184,-2.9973],[191,-3.0566],[195,-3.0775],[201,-3.0877],[203,-3.075],[205,-3.0461],[207,-3.0026],[210,-2.9303],[212,-2.9079],[217,-2.9031],[218,-2.9369],[220,-3.0512],[223,-3.2533],[226,-3.4382],[227,-3.4874],[228,-3.5183],[229,-3.5364],[231,-3.55],[241,-3.5179],[245,-3.4846],[251,-3.4007],[253,-3.3883],[261,-3.3832],[262,-3.3719],[265,-3.3055],[275,-3.1234],[277,-3.0959],[284,-3.0324],[285,-3.0146],[289,-2.963],[292,-2.9413],[296,-2.9434],[304,-3.0279],[305,-3.0327],[306,0]]
       }
       middle_02_l_ctrl.Rotation.Y {
        type=MovieSceneScriptingFloatChannel
        default=0
        thinned_from=307
        keys [[0,2.4048],[2,2.4048],[4,2.3903],[6,2.3685],[10,2.332],[11,2.332],[12,2.3247],[14,2.3247],[15,2.3173],[17,2.2843],[18,2.2787],[20,2.2871],[24,2.2871],[25,2.2787],[29,2.2051],[37,2.0396],[40,1.9899],[42,1.9725],[43,1.9683],[46,1.965],[47,1.96],[50,1.96],[57,1.9545],[58,1.9565],[59,1.9653],[60,1.979],[61,1.9993],[62,2.027],[63,2.0593],[64,2.0999],[66,2.1988],[67,2.2719],[68,2.3537],[69,2.3915],[70,2.4002],[79,2.3981],[82,2.3756],[83,2.3631],[84,2.3447],[85,2.3323],[86,2.3138],[88,2.292],[90,2.2889],[91,2.2826],[97,2.2826],[98,2.2764],[99,2.2577],[100,2.247],[101,2.2597],[107,2.3988],[108,2.4142],[112,2.4356],[113,2.4356],[115,2.4234],[119,2.3732],[121,2.3596],[124,2.362],[125,2.3667],[130,2.418],[132,2.4257],[136,2.4214],[137,2.4131],[141,2.4131],[142,2.4048],[152,2.4048],[153,2.398],[154,2.398],[159,2.3256],[162,2.3082],[165,2.3082],[167,2.3126],[170,2.3566],[172,2.3954],[174,2.4242],[176,2.4385],[177,2.4385],[178,2.4481],[179,2.4462],[180,2.4522],[184,2.4522],[192,2.3679],[195,2.3508],[196,2.3508],[197,2.3442],[198,2.3442],[199,2.3376],[201,2.3376],[202,2.3419],[204,2.3683],[205,2.388],[206,2.4141],[208,2.4779],[210,2.5299],[212,2.5558],[213,2.5616],[217,2.5616],[218,2.5234],[219,2.4583],[220,2.3834],[221,2.2928],[223,2.0996],[224,1.9986],[226,1.7794],[227,1.6824],[228,1.6142],[229,1.5749],[230,1.554],[231,1.5447],[232,1.5447],[233,1.5493],[239,1.598],[242,1.6303],[243,1.6445],[244,1.6642],[247,1.7385],[249,1.7928],[250,1.8237],[251,1.8506],[252,1.8659],[253,1.8735],[258,1.8735],[260,1.8761],[261,1.8826],[262,1.9021],[265,2.0166],[269,2.1362],[274,2.2664],[277,2.3247],[278,2.338],[282,2.3785],[289,2.4911],[291,2.5117],[293,2.5195],[295,2.5195],[296,2.5156],[303,2.4228],[305,2.4068],[306,0]]
       }
       middle_02_l_ctrl.Rotation.Z {
        type=MovieSceneScriptingFloatChannel
        default=0
        thinned_from=307
        keys [[0,9.0421],[2,9.0421],[5,8.641],[6,8.3732],[8,8.1054],[9,7.8373],[10,7.7032],[11,7.7032],[12,7.5691],[14,7.5691],[15,7.4349],[16,7.1664],[17,6.8256],[18,6.7247],[20,6.9266],[24,6.9266],[25,6.7247],[27,6.1186],[28,5.7139],[29,5.41],[30,5.0046],[31,4.7003],[33,3.8873],[34,3.591],[35,3.2469],[36,2.8162],[37,2.4713],[38,2.2124],[39,1.8668],[40,1.6074],[41,1.4344],[43,1.2613],[44,1.2613],[45,1.1747],[46,1.1747],[47,1.088],[50,1.088],[51,1.0722],[52,1.0014],[57,1.0014],[58,1.0721],[59,1.2135],[60,1.4253],[61,1.7781],[62,2.2714],[63,2.8343],[64,3.5367],[65,4.3775],[66,5.2862],[67,6.6103],[68,8.1422],[69,8.8658],[70,9.0264],[77,9.0264],[78,8.9461],[79,8.9461],[80,8.8658],[82,8.5443],[83,8.2994],[84,7.9529],[85,7.7217],[86,7.3747],[87,7.1432],[89,6.9116],[90,6.9116],[91,6.7957],[97,6.7957],[98,6.6798],[99,6.3573],[100,6.1817],[101,6.4159],[103,7.2348],[104,7.7019],[105,8.1102],[106,8.5763],[107,8.9838],[108,9.2745],[109,9.4489],[110,9.507],[111,9.6232],[112,9.6812],[113,9.6812],[115,9.4338],[116,9.2587],[119,8.4701],[120,8.2947],[121,8.207],[123,8.207],[125,8.3825],[128,8.9084],[129,9.1712],[130,9.3462],[131,9.36],[132,9.5213],[133,9.5213],[134,9.36],[136,9.36],[137,9.1988],[141,9.1988],[142,9.0375],[152,9.0375],[153,8.9176],[154,8.9176],[159,7.6246],[162,7.2359],[165,7.2359],[166,7.4165],[167,7.4165],[168,7.5968],[169,7.9575],[170,8.1377],[173,9.2173],[176,9.756],[177,9.756],[178,9.9354],[179,9.8756],[180,10.0089],[184,10.0089],[185,9.7421],[186,9.6087],[187,9.3416],[188,9.208],[189,8.9407],[190,8.807],[191,8.5394],[194,8.1377],[195,8.0371],[196,8.0371],[197,7.9141],[198,7.9141],[199,7.791],[201,7.791],[202,7.9142],[204,8.4059],[205,8.7743],[206,9.265],[208,10.4891],[210,11.4661],[211,11.7672],[212,11.9867],[213,12.0965],[217,12.0965],[218,11.3276],[219,10.1168],[220,8.6813],[221,7.0188],[222,5.2373],[223,3.5593],[224,1.7604],[225,-0.0484],[226,-1.982],[227,-3.559],[228,-4.6725],[229,-5.2681],[230,-5.6412],[231,-5.7905],[232,-5.7905],[233,-5.7158],[235,-5.4173],[236,-5.3428],[241,-4.5981],[242,-4.3751],[243,-4.2034],[247,-2.6482],[248,-2.1317],[249,-1.7449],[251,-0.7159],[252,-0.4592],[253,-0.3309],[258,-0.3309],[259,-0.3447],[260,-0.3447],[261,-0.2331],[262,0.1013],[265,2.1004],[266,2.5429],[269,4.1975],[270,4.6374],[271,5.1865],[274,6.5009],[275,6.8884],[276,7.2919],[278,7.8291],[279,7.9633],[280,8.2316],[281,8.3656],[284,9.169],[285,9.5703],[287,10.1045],[288,10.5048],[289,10.7714],[290,10.9046],[291,11.1555],[293,11.3064],[295,11.3064],[296,11.2309],[297,11.0046],[299,10.4003],[302,9.7195],[303,9.4165],[305,9.1134],[306,0]]
       }
       middle_02_l_ctrl.Scale.X {type=MovieSceneScriptingFloatChannel, default=1, const [1,307]}
       middle_02_l_ctrl.Scale.Y {type=MovieSceneScriptingFloatChannel, default=1, const [1,307]}
       middle_02_l_ctrl.Scale.Z {type=MovieSceneScriptingFloatChannel, default=1, const [1,307]}
       middle_03_l_ctrl.Location.X {type=MovieSceneScriptingFloatChannel, default=0, thinned_from=307, keys [[0,0.0206],[305,0.0206],[306,0]]}
       middle_03_l_ctrl.Location.Y {type=MovieSceneScriptingFloatChannel, default=0, thinned_from=307, keys [[0,-0.0583],[305,-0.0583],[306,0]]}
       middle_03_l_ctrl.Location.Z {type=MovieSceneScriptingFloatChannel, default=0, thinned_from=307, keys [[0,0.1324],[305,0.1324],[306,0]]}
       middle_03_l_ctrl.Rotation.X {
        type=MovieSceneScriptingFloatChannel
        default=0
        thinned_from=307
        keys [[0,0.9169],[2,0.9169],[5,0.9016],[6,0.8914],[7,0.8863],[8,0.8761],[11,0.8606],[14,0.8606],[15,0.8555],[16,0.8452],[17,0.8305],[22,0.8305],[23,0.8386],[24,0.8344],[25,0.8344],[26,0.8142],[27,0.8058],[29,0.7733],[30,0.7648],[32,0.7322],[33,0.7195],[34,0.7037],[37,0.6618],[38,0.6512],[39,0.6374],[40,0.6266],[41,0.6196],[42,0.6196],[43,0.6126],[47,0.6126],[48,0.6055],[58,0.6049],[59,0.6111],[61,0.6345],[62,0.6516],[63,0.6748],[64,0.7042],[66,0.7733],[67,0.8227],[68,0.8812],[69,0.9071],[70,0.9135],[80,0.9135],[82,0.8943],[83,0.8865],[87,0.8441],[88,0.8377],[89,0.8336],[90,0.8336],[91,0.8313],[97,0.8313],[98,0.8271],[99,0.8125],[100,0.8033],[101,0.8125],[103,0.8495],[104,0.863],[105,0.8814],[106,0.8954],[107,0.9137],[109,0.9319],[111,0.9319],[112,0.941],[113,0.941],[114,0.9362],[117,0.9141],[118,0.9031],[120,0.8881],[121,0.8842],[123,0.8842],[124,0.888],[125,0.888],[130,0.9255],[131,0.9302],[134,0.9302],[135,0.9263],[137,0.9249],[138,0.9222],[139,0.9222],[140,0.9195],[142,0.9195],[143,0.9169],[152,0.9169],[155,0.9025],[156,0.8928],[158,0.8686],[159,0.8614],[161,0.8516],[162,0.849],[164,0.849],[166,0.8504],[167,0.8571],[168,0.8621],[169,0.8706],[172,0.9107],[174,0.9307],[175,0.9373],[178,0.9473],[180,0.9508],[181,0.9508],[182,0.9533],[183,0.9533],[184,0.9508],[185,0.9447],[191,0.8978],[195,0.8754],[198,0.8754],[199,0.8657],[201,0.8657],[205,0.9046],[207,0.9432],[208,0.972],[210,1.0101],[211,1.0199],[212,1.0242],[213,1.0326],[216,1.0326],[217,1.0283],[218,1.0031],[219,0.9567],[220,0.9012],[221,0.8409],[224,0.6313],[225,0.5601],[226,0.4847],[227,0.4228],[228,0.3819],[229,0.3483],[230,0.3347],[233,0.3347],[234,0.3419],[236,0.3419],[237,0.3554],[238,0.3554],[239,0.3689],[240,0.3689],[241,0.3824],[242,0.3824],[243,0.3971],[244,0.4078],[245,0.4235],[247,0.4604],[248,0.4761],[249,0.4945],[250,0.5155],[251,0.5338],[252,0.5442],[253,0.5442],[254,0.5496],[260,0.5496],[261,0.5536],[262,0.5689],[263,0.593],[265,0.6454],[266,0.665],[268,0.7082],[271,0.7666],[275,0.8343],[276,0.8498],[278,0.8704],[279,0.8756],[280,0.8857],[281,0.891],[285,0.9319],[286,0.9471],[290,0.9877],[291,0.9946],[292,0.9976],[295,0.9976],[296,0.9947],[297,0.9886],[299,0.965],[300,0.9589],[301,0.9471],[302,0.941],[303,0.9292],[305,0.9171],[306,0]]
       }
       middle_03_l_ctrl.Rotation.Y {
        type=MovieSceneScriptingFloatChannel
        default=0
        thinned_from=307
        keys [[0,2.1659],[5,2.1722],[11,2.1887],[15,2.1907],[17,2.1996],[22,2.1996],[23,2.1966],[25,2.2],[40,2.2672],[58,2.2741],[62,2.2608],[63,2.2539],[65,2.234],[67,2.2037],[68,2.1808],[69,2.1702],[80,2.1675],[88,2.1976],[98,2.2017],[100,2.2097],[109,2.1583],[111,2.1583],[113,2.1544],[121,2.1796],[125,2.178],[130,2.1607],[152,2.1652],[162,2.1939],[168,2.1886],[175,2.157],[184,2.1506],[195,2.1821],[198,2.1821],[201,2.186],[203,2.1806],[207,2.1534],[210,2.1232],[213,2.1121],[216,2.1121],[217,2.1162],[218,2.1274],[219,2.1473],[222,2.2222],[225,2.2855],[227,2.3133],[229,2.3275],[233,2.3294],[249,2.2989],[252,2.2879],[261,2.2848],[278,2.1844],[285,2.159],[292,2.1287],[296,2.1301],[305,2.1659],[306,0]]
       }
       middle_03_l_ctrl.Rotation.Z {
        type=MovieSceneScriptingFloatChannel
        default=0
        thinned_from=307
        keys [[0,9.1751],[2,9.1751],[5,8.7695],[6,8.4991],[7,8.3638],[8,8.0932],[11,7.6873],[14,7.6873],[15,7.5519],[16,7.2811],[17,6.9309],[22,6.9309],[23,7.141],[24,6.931],[25,6.931],[26,6.5104],[27,6.3002],[29,5.4585],[30,5.248],[33,3.9839],[34,3.5948],[37,2.5252],[38,2.3467],[40,1.6327],[41,1.4541],[42,1.4541],[43,1.2754],[47,1.2754],[48,1.0967],[50,1.0967],[51,1.0262],[58,1.0262],[62,2.2107],[63,2.8022],[66,5.4585],[67,6.6507],[68,8.2052],[69,8.8954],[70,9.0679],[80,9.0679],[81,8.8954],[82,8.5504],[83,8.3591],[84,8.1352],[85,7.7993],[86,7.5752],[87,7.2391],[89,7.015],[90,7.015],[91,6.9029],[97,6.9029],[98,6.7908],[99,6.4731],[100,6.2301],[101,6.4731],[103,7.4445],[104,7.6873],[107,9.1424],[109,9.6271],[111,9.6271],[112,9.8694],[113,9.8694],[114,9.6272],[115,9.4601],[121,8.2612],[123,8.2612],[124,8.461],[125,8.461],[130,9.46],[131,9.52],[134,9.52],[136,9.3802],[137,9.3802],[138,9.3103],[139,9.3103],[140,9.2404],[142,9.2404],[143,9.1704],[152,9.1704],[155,8.7882],[157,8.2658],[158,7.8739],[162,7.3512],[164,7.3512],[165,7.4393],[166,7.4393],[167,7.6154],[168,7.7035],[169,7.9677],[172,9.0238],[174,9.5515],[175,9.7274],[178,9.9911],[179,10.0272],[180,10.0925],[181,10.0925],[182,10.1579],[183,10.1579],[184,10.0925],[189,9.1118],[190,8.8501],[191,8.6539],[194,8.2612],[195,8.0979],[198,8.0979],[199,7.8413],[201,7.8413],[205,8.8674],[207,9.8927],[208,10.6611],[210,11.685],[211,11.8656],[213,12.313],[216,12.313],[217,12.0894],[218,11.4181],[219,10.2985],[221,7.1589],[223,3.5587],[224,1.7527],[225,-0.0578],[226,-1.8734],[227,-3.4349],[229,-5.3879],[230,-5.7141],[233,-5.7141],[234,-5.388],[236,-5.388],[237,-5.062],[238,-5.062],[239,-4.7362],[240,-4.7362],[241,-4.4106],[242,-4.4106],[243,-4.0154],[244,-3.7538],[246,-2.9697],[247,-2.4477],[249,-1.6654],[251,-0.624],[252,-0.3639],[253,-0.3639],[254,-0.2339],[260,-0.2339],[261,-0.1342],[262,0.199],[265,2.1945],[266,2.6372],[269,4.295],[270,4.7366],[271,5.2881],[274,6.6106],[275,7.001],[276,7.4076],[278,7.9496],[279,8.0851],[280,8.356],[281,8.4914],[285,9.5741],[286,9.9799],[290,11.0613],[292,11.3872],[295,11.3872],[297,11.0708],[299,10.4376],[300,10.2793],[301,9.9626],[302,9.8042],[303,9.4874],[305,9.1705],[306,0]]
       }
       middle_03_l_ctrl.Scale.X {type=MovieSceneScriptingFloatChannel, default=1, const [1,307]}
       middle_03_l_ctrl.Scale.Y {type=MovieSceneScriptingFloatChannel, default=1, const [1,307]}
       middle_03_l_ctrl.Scale.Z {type=MovieSceneScriptingFloatChannel, default=1, const [1,307]}
       ring_metacarpal_l_ctrl.Location.X {type=MovieSceneScriptingFloatChannel, default=0, thinned_from=307, keys [[0,0],[306,0]]}
       ring_metacarpal_l_ctrl.Location.Y {type=MovieSceneScriptingFloatChannel, default=0, thinned_from=307, keys [[0,0],[306,0]]}
       ring_metacarpal_l_ctrl.Location.Z {type=MovieSceneScriptingFloatChannel, default=0, thinned_from=307, keys [[0,0],[306,0]]}
       ring_metacarpal_l_ctrl.Rotation.X {
        type=MovieSceneScriptingFloatChannel
        default=0
        thinned_from=307
        keys [[0,0],[1,0.0011],[2,-0.0006],[10,0.0007],[13,0.0008],[14,-0.0009],[15,-0.001],[16,0.0003],[17,-0.0001],[33,0],[34,0.0012],[39,0.0011],[40,-0.0023],[41,-0.0023],[42,0.0011],[43,0],[66,-0.0001],[67,0.0004],[78,0.0005],[79,0.0007],[80,-0.0004],[81,0],[82,0.0007],[83,-0.0006],[85,-0.0008],[86,0.0009],[87,-0.0011],[88,0.0006],[98,0.0004],[100,0],[103,0.0009],[104,-0.0008],[108,-0.0003],[114,-0.0003],[115,0.0007],[116,-0.0005],[119,0.0005],[120,-0.0008],[121,0.0008],[122,-0.0006],[123,-0.0006],[124,0.0008],[125,0.0006],[126,-0.001],[127,0.0002],[129,-0.0004],[130,0.0009],[131,-0.0003],[143,-0.0003],[144,0],[146,0],[147,0.002],[152,0.002],[153,-0.0022],[155,-0.002],[156,0.0026],[157,-0.0015],[160,-0.0011],[161,0.0032],[164,0.0032],[165,0],[168,0.0002],[169,-0.0027],[170,0.0004],[171,-0.0024],[172,0.0006],[173,-0.0023],[174,0.0007],[176,0.0007],[177,-0.0023],[178,0.0007],[179,-0.0021],[183,-0.0023],[184,0.0024],[185,0.0026],[186,-0.0018],[187,-0.0016],[188,0.0032],[189,-0.0012],[193,-0.0007],[194,0.0039],[195,-0.0003],[199,-0.0006],[200,0.0015],[201,0.0015],[202,-0.0005],[207,0.0009],[208,-0.001],[210,-0.0008],[211,0.0005],[214,0.0005],[215,-0.0012],[216,-0.0012],[217,0.0005],[219,0.0004],[220,-0.0013],[221,-0.0015],[222,0],[224,-0.0005],[226,-0.0006],[229,0.0002],[230,-0.0013],[231,0.0009],[232,-0.0009],[233,0.0009],[234,0.0009],[235,-0.0011],[236,0.0007],[238,0.0005],[239,-0.0016],[240,0.0002],[243,-0.0003],[248,-0.0007],[249,-0.0009],[250,0.001],[251,-0.0016],[252,0.0005],[260,0.0003],[263,0.001],[264,-0.0009],[265,0.0016],[266,0.0018],[267,-0.0004],[269,-0.0003],[270,0.002],[271,0],[288,-0.0001],[289,0.0021],[290,-0.0003],[291,0],[296,0],[305,0.0003],[306,0]]
       }
       ring_metacarpal_l_ctrl.Rotation.Y {
        type=MovieSceneScriptingFloatChannel
        default=0
        thinned_from=307
        keys [[0,0],[1,0],[2,-0.0006],[3,0.0001],[10,0.0003],[11,-0.0003],[15,-0.0003],[16,0.0001],[17,-0.0003],[18,0],[20,0],[21,-0.0003],[25,-0.0002],[26,0.0002],[27,-0.0003],[28,0.0001],[33,-0.0001],[34,0.0006],[35,0.0006],[36,-0.0007],[38,-0.0007],[39,0.0006],[40,-0.0005],[42,-0.0006],[43,0],[61,0],[62,0.0003],[79,0.0003],[80,0],[81,0.0002],[83,-0.0001],[84,0.0004],[86,-0.0001],[87,-0.0002],[88,0.0005],[90,0.0005],[91,-0.0003],[97,-0.0003],[98,0.0004],[99,0.0007],[100,0],[101,-0.0001],[102,0.0004],[104,0.0003],[105,-0.0004],[108,0.0004],[109,-0.0002],[113,0.0001],[117,-0.0002],[118,0.0003],[119,0.0003],[120,-0.0002],[121,0.0003],[122,-0.0002],[123,-0.0002],[124,0.0003],[125,-0.0003],[126,-0.0002],[127,0.0003],[128,-0.0003],[129,-0.0002],[130,0.0004],[131,-0.0006],[132,0.0001],[136,0.0001],[137,-0.0006],[138,0.0002],[143,0.0002],[144,0],[146,0],[147,0.001],[152,0.001],[153,-0.0005],[155,-0.0007],[156,0.0008],[158,0.001],[159,-0.0006],[160,0.0012],[164,0.0011],[165,-0.0001],[170,-0.0003],[172,-0.0002],[173,0.0001],[175,0.0002],[176,-0.0009],[177,-0.0006],[178,0.0005],[179,-0.0005],[183,-0.0005],[184,0.001],[185,0.001],[186,-0.0006],[187,-0.0006],[188,0.001],[189,-0.0005],[193,-0.0001],[194,-0.0001],[195,0.0003],[196,0.0003],[197,-0.0005],[199,-0.0004],[200,0.0005],[201,0.0005],[202,-0.0004],[203,0.0002],[204,-0.0008],[205,-0.0001],[207,-0.0001],[208,-0.0005],[209,-0.0001],[210,-0.0003],[212,0.0002],[213,-0.0003],[214,0.0004],[215,-0.0001],[217,-0.0003],[218,0.0001],[219,-0.0003],[220,0.0002],[221,-0.0001],[222,0.0003],[223,-0.0002],[226,-0.0001],[227,0.0001],[229,0.0001],[230,-0.0004],[231,0.0005],[232,-0.0001],[233,0.0005],[234,0.0004],[235,-0.0004],[236,0.0003],[238,0.0002],[239,-0.0006],[240,0.0001],[243,0.0001],[244,-0.0003],[245,0.0001],[247,-0.0005],[248,0.0002],[250,0.0002],[251,-0.0002],[261,0],[262,0.0006],[263,0.0003],[265,0],[266,0.0002],[267,-0.0003],[268,0.0002],[269,-0.0002],[270,0.0001],[282,0],[283,0.0003],[287,-0.0001],[288,-0.0001],[289,0.0007],[290,0.0001],[291,-0.0003],[296,-0.0003],[297,0.0006],[298,-0.0004],[305,0.0004],[306,0]]
       }
       ring_metacarpal_l_ctrl.Rotation.Z {
        type=MovieSceneScriptingFloatChannel
        default=0
        thinned_from=307
        keys [[0,0],[1,-0.024],[2,-0.0475],[3,-0.1407],[4,-0.2572],[5,-0.4203],[9,-1.1658],[10,-1.3055],[11,-1.3754],[12,-1.3987],[13,-1.3987],[14,-1.3522],[15,-1.1892],[16,-0.8396],[17,-0.319],[18,0.188],[19,0.7267],[20,1.1705],[21,1.4241],[22,1.5668],[23,1.6302],[24,1.6302],[25,1.5826],[27,1.3924],[28,1.2814],[29,1.1863],[30,1.0754],[31,0.9803],[33,0.7584],[34,0.6514],[39,0.18],[40,0.1326],[41,0.0855],[42,0.0386],[43,0],[61,0],[62,-0.0163],[63,-0.0163],[64,-0.0319],[65,-0.0319],[66,-0.0474],[67,-0.0638],[68,-0.0638],[69,-0.0822],[77,-0.0822],[78,-0.1006],[79,-0.1375],[80,-0.1928],[81,-0.3401],[82,-0.561],[83,-0.7953],[85,-1.3552],[86,-1.6605],[87,-1.9151],[88,-2.0931],[89,-2.1695],[90,-2.2204],[92,-2.2713],[97,-2.2713],[98,-2.373],[99,-2.7232],[100,-2.9164],[101,-2.6957],[102,-2.2541],[103,-1.8401],[106,-0.5151],[107,-0.1009],[108,0.1753],[109,0.341],[110,0.4515],[111,0.5344],[112,0.5896],[113,0.5896],[114,0.5068],[115,0.3588],[116,0.1558],[117,-0.1079],[118,-0.3512],[119,-0.5743],[120,-0.7366],[121,-0.8379],[122,-0.8583],[123,-0.8583],[124,-0.7973],[125,-0.6757],[126,-0.5136],[127,-0.3309],[129,0.0747],[130,0.2371],[131,0.3449],[132,0.3721],[133,0.3721],[136,0.2089],[137,0.1817],[138,0.1273],[142,0.0186],[143,0.0186],[144,0],[146,0],[147,-0.0137],[152,-0.0137],[153,-0.0737],[154,-0.1929],[155,-0.3718],[158,-1.2661],[160,-1.6237],[161,-1.7426],[162,-1.8022],[164,-1.8022],[166,-1.7197],[167,-1.5955],[168,-1.4298],[169,-1.1815],[170,-0.8914],[171,-0.5188],[172,-0.1872],[173,0.144],[174,0.3514],[175,0.5172],[176,0.6415],[177,0.7241],[178,0.7658],[179,0.8351],[181,0.9615],[183,0.9615],[184,0.8987],[187,0.3299],[188,0.0776],[189,-0.1122],[192,-0.6804],[194,-0.9326],[200,-1.286],[201,-1.286],[202,-1.1689],[203,-0.9637],[204,-0.6412],[205,-0.2599],[206,0.2094],[207,0.7962],[208,1.4122],[209,1.9406],[210,2.4104],[211,2.7228],[212,2.908],[213,3.0238],[215,3.07],[216,3.07],[217,3.047],[218,2.8386],[219,2.4913],[220,2.0514],[221,1.5191],[222,0.964],[223,0.5245],[224,0.2701],[225,0.1776],[226,0.1545],[227,0.1019],[228,-0.1087],[229,-0.5035],[230,-0.9774],[231,-1.3719],[232,-1.4773],[233,-1.4246],[234,-1.293],[235,-1.1353],[236,-1.0299],[238,-0.7667],[241,-0.293],[242,-0.0824],[243,0.1582],[244,0.4665],[245,0.8365],[246,1.2683],[247,1.7311],[248,2.1631],[249,2.626],[250,3.1201],[251,3.5523],[252,3.7996],[253,3.9232],[254,3.9849],[255,4.0158],[260,4.0158],[261,3.9508],[262,3.6632],[263,3.2158],[265,2.1937],[266,1.7467],[267,1.2677],[268,0.7252],[269,0.3104],[270,0.0235],[271,0],[282,0],[283,0.0544],[284,0.3065],[286,0.9998],[288,1.5041],[290,1.9455],[291,2.095],[292,2.1676],[294,2.2402],[295,2.2402],[296,2.1313],[298,1.6233],[299,1.3331],[300,1.0792],[301,0.8616],[303,0.354],[304,0.1727],[305,0.0277],[306,0]]
       }
       ring_metacarpal_l_ctrl.Scale.X {type=MovieSceneScriptingFloatChannel, default=1, const [1,307]}
       ring_metacarpal_l_ctrl.Scale.Y {type=MovieSceneScriptingFloatChannel, default=1, const [1,307]}
       ring_metacarpal_l_ctrl.Scale.Z {type=MovieSceneScriptingFloatChannel, default=1, const [1,307]}
       ring_01_l_ctrl.Location.X {type=MovieSceneScriptingFloatChannel, default=0, thinned_from=307, keys [[0,0],[306,0]]}
       ring_01_l_ctrl.Location.Y {type=MovieSceneScriptingFloatChannel, default=0, thinned_from=307, keys [[0,0],[306,0]]}
       ring_01_l_ctrl.Location.Z {type=MovieSceneScriptingFloatChannel, default=0, thinned_from=307, keys [[0,0],[306,0]]}
       ring_01_l_ctrl.Rotation.X {
        type=MovieSceneScriptingFloatChannel
        default=0
        thinned_from=307
        keys [[0,0.0001],[6,0.0002],[7,-0.0007],[8,0.0002],[10,0.0001],[11,-0.0009],[13,-0.0009],[14,0.0001],[16,0.0002],[17,0],[28,0.0001],[29,-0.0011],[30,-0.0011],[31,0.0001],[35,0.0002],[36,-0.0005],[37,-0.0002],[38,-0.0001],[41,-0.0002],[42,0.0007],[43,0.0001],[66,0],[67,0.0001],[68,-0.0006],[79,-0.0006],[80,0.0002],[81,0.0002],[82,-0.0006],[83,-0.0006],[84,0],[85,-0.0003],[86,0.0003],[87,0.0003],[88,-0.0015],[89,0.0004],[90,0.0004],[91,-0.0002],[98,-0.0002],[99,-0.0006],[100,0],[101,0.0001],[102,-0.0003],[104,0.0002],[107,-0.0003],[108,0.0005],[111,0.0002],[112,0],[113,0],[114,-0.0007],[115,0.0001],[116,0.0002],[117,-0.0008],[118,0.0007],[119,-0.0007],[120,0.0003],[121,0.0003],[122,-0.0007],[123,-0.0007],[124,0.0003],[125,0.0003],[126,-0.0007],[127,0.0003],[130,0.0001],[131,0.0004],[133,0.0004],[134,-0.0017],[136,-0.0017],[137,-0.0003],[138,-0.0003],[139,0.0002],[140,-0.0019],[143,-0.0019],[144,-0.0004],[146,-0.0004],[147,0],[151,0],[152,0.0002],[153,-0.0008],[154,0.0002],[155,-0.0007],[156,-0.0006],[157,0.0004],[158,0.0004],[159,-0.0006],[160,-0.0006],[161,0.0004],[162,0.0004],[163,-0.0003],[164,-0.0003],[165,0.0018],[167,0.002],[168,-0.0009],[169,-0.0007],[170,0.0003],[171,0.0004],[172,-0.0003],[173,0.0005],[174,0.0005],[175,-0.0024],[176,0.0005],[178,0.0004],[179,0.0002],[186,0.0002],[187,-0.0001],[188,0.0002],[190,0],[191,0.0007],[192,-0.0004],[193,0.0006],[194,0.0005],[195,-0.0002],[196,0.0003],[199,0.0002],[200,-0.0005],[201,-0.0005],[202,0.0002],[204,0.0006],[205,-0.0012],[207,-0.001],[209,-0.0013],[210,-0.0002],[212,0.0002],[213,-0.0002],[217,-0.0003],[218,-0.0013],[219,0.0004],[220,-0.001],[221,-0.0005],[222,0.0005],[225,0.0002],[226,-0.0003],[227,0.0002],[228,0.0002],[229,0],[230,0.0001],[231,-0.0002],[234,-0.0002],[235,0.0001],[236,-0.0001],[237,-0.0001],[238,0.0002],[241,0.0002],[242,0],[244,0.0004],[245,-0.0008],[246,-0.0006],[247,-0.0011],[248,-0.001],[249,-0.0006],[250,0.0008],[251,0.0005],[252,-0.0001],[253,0.0003],[258,0.0002],[259,-0.0001],[263,-0.0002],[264,0.0001],[266,0.0003],[267,0.0001],[268,-0.0003],[270,0.0001],[282,0],[284,0.0003],[285,-0.0006],[286,0.0006],[287,0.0004],[288,0.0004],[289,-0.0004],[290,0.0004],[291,0],[292,-0.0009],[293,0.0002],[298,0.0004],[299,-0.0004],[300,0.0005],[301,-0.0008],[302,-0.0008],[303,0.0003],[306,0]]
       }
       ring_01_l_ctrl.Rotation.Y {
        type=MovieSceneScriptingFloatChannel
        default=0
        thinned_from=307
        keys [[0,-0.0005],[6,-0.0006],[7,0.004],[8,-0.0006],[10,-0.0008],[11,0.0038],[13,0.0038],[14,-0.0008],[28,-0.0005],[29,0.0059],[30,0.0059],[31,-0.0005],[35,-0.0005],[36,0.0005],[37,-0.0005],[38,-0.0005],[39,0.0006],[41,0.0006],[42,-0.0004],[43,-0.0005],[44,0],[66,0],[67,-0.0005],[68,0.0032],[79,0.0032],[80,-0.0009],[81,-0.0013],[82,0.002],[83,0.0018],[84,-0.0007],[85,0.0017],[86,-0.0008],[87,-0.0009],[88,0.0013],[89,-0.001],[90,-0.001],[91,0.0015],[98,0.0014],[99,0],[103,0.0004],[104,-0.0007],[105,0.0007],[107,0.0008],[108,-0.0004],[111,-0.0005],[112,0.0009],[114,0.0007],[115,0.0002],[116,0.0002],[117,0],[118,-0.0018],[119,0],[121,0.0001],[122,-0.0001],[123,-0.0001],[125,0.0001],[126,0],[127,0.0002],[130,0.0002],[131,-0.0016],[133,-0.0016],[134,0.0011],[136,0.0013],[137,0.0017],[138,0.0017],[139,-0.0009],[140,0.0018],[143,0.0018],[144,0.0023],[146,0.0023],[147,0],[151,0],[152,-0.0005],[153,0.0007],[154,-0.0005],[155,0.0007],[156,0.0007],[157,-0.0006],[158,-0.0006],[159,0.0005],[160,0.0005],[161,-0.0008],[164,-0.001],[165,-0.0005],[167,-0.0006],[168,0.003],[169,0.003],[170,-0.0011],[171,-0.0012],[172,0.0028],[173,-0.0014],[174,-0.0016],[175,0.002],[176,-0.0017],[178,-0.0017],[179,-0.0005],[186,-0.0005],[187,0.001],[188,-0.0005],[191,-0.0005],[192,0.0008],[193,-0.0005],[194,-0.0006],[195,0.0021],[196,-0.0007],[199,-0.0007],[200,0.0021],[201,0.0021],[202,-0.0007],[204,-0.0007],[205,0.0018],[209,0.0009],[210,0.0008],[211,0.0015],[212,-0.0009],[213,0.0014],[218,0.0013],[219,-0.0007],[220,0.0015],[221,-0.0007],[222,-0.0004],[225,-0.0005],[226,0.0018],[227,-0.0005],[228,-0.0005],[229,0.0007],[230,-0.0006],[231,0.0005],[234,0.0005],[235,-0.0007],[236,0.0006],[237,0.0006],[238,-0.0006],[241,-0.0005],[242,0.0007],[243,-0.0005],[244,-0.0004],[245,-0.0007],[246,-0.0007],[247,0.0023],[248,0.0022],[249,-0.0009],[250,-0.0008],[251,-0.001],[252,0.0018],[253,-0.0013],[258,-0.0013],[259,0.0002],[261,0.0003],[262,-0.0003],[263,0],[264,-0.0004],[265,-0.0001],[266,-0.0006],[267,-0.0004],[268,0.0005],[269,-0.0001],[282,0],[283,-0.0006],[284,-0.0007],[285,0.0005],[286,-0.0011],[287,0.0002],[289,-0.0003],[290,-0.0003],[291,0.0011],[292,0.001],[293,-0.0007],[298,-0.0005],[299,-0.0006],[300,-0.0004],[301,0.0011],[302,0.0011],[303,-0.0005],[305,-0.0005],[306,0]]
       }
       ring_01_l_ctrl.Rotation.Z {
        type=MovieSceneScriptingFloatChannel
        default=0
        thinned_from=307
        keys [[0,0.0308],[1,0.0308],[4,-0.2578],[7,-0.8351],[8,-0.9317],[10,-1.3172],[11,-1.4133],[13,-1.4133],[15,-1.2208],[16,-0.8354],[17,-0.3503],[18,0.1869],[19,0.7236],[20,1.1258],[21,1.3939],[23,1.6618],[24,1.6618],[25,1.5278],[26,1.5278],[29,1.1262],[30,1.1262],[33,0.7236],[34,0.6609],[35,0.5697],[36,0.4557],[39,0.1819],[40,0.1134],[41,0.0678],[43,0.0221],[44,0],[66,0],[67,-0.0236],[68,-0.1012],[79,-0.1012],[80,-0.1792],[81,-0.3348],[82,-0.5681],[83,-0.8108],[84,-1.0704],[85,-1.3817],[87,-1.9014],[88,-2.1092],[91,-2.2653],[97,-2.2653],[98,-2.3693],[99,-2.7218],[100,-2.9164],[101,-2.6941],[102,-2.2497],[103,-1.8335],[104,-1.3899],[105,-0.9743],[106,-0.5313],[107,-0.1163],[108,0.1877],[109,0.3535],[110,0.464],[111,0.5469],[112,0.5746],[113,0.5746],[114,0.4917],[115,0.3706],[116,0.1643],[117,-0.1245],[118,-0.3311],[119,-0.5788],[120,-0.7441],[121,-0.8267],[122,-0.868],[123,-0.868],[124,-0.7854],[126,-0.5375],[129,0.0818],[130,0.2468],[131,0.3495],[133,0.3495],[135,0.2254],[136,0.2254],[137,0.1631],[138,0.1631],[139,0.1008],[140,0.0388],[143,0.0388],[144,-0.0235],[146,-0.0235],[147,0],[151,0],[152,-0.0271],[154,-0.2008],[155,-0.4034],[156,-0.6642],[158,-1.2442],[159,-1.4763],[160,-1.6214],[161,-1.7086],[163,-1.823],[164,-1.823],[165,-1.7376],[166,-1.7376],[167,-1.5668],[168,-1.4811],[169,-1.225],[172,-0.2017],[173,0.1388],[174,0.3942],[175,0.4795],[176,0.6494],[177,0.7345],[178,0.7345],[179,0.8261],[181,0.9487],[182,0.9793],[183,0.9793],[184,0.918],[185,0.7342],[187,0.3052],[188,0.1211],[190,-0.3084],[192,-0.6768],[193,-0.7997],[194,-0.8919],[195,-1.0055],[200,-1.3043],[201,-1.3043],[202,-1.1849],[203,-0.946],[204,-0.6475],[205,-0.2892],[206,0.1878],[208,1.3787],[209,1.9139],[210,2.3892],[211,2.724],[212,2.9115],[213,3.0053],[214,3.0522],[217,3.0522],[218,2.8178],[219,2.4894],[220,2.0203],[222,0.9866],[223,0.5163],[224,0.2809],[225,0.1868],[226,0.1398],[227,0.1138],[228,-0.0936],[229,-0.5085],[230,-0.9758],[231,-1.3654],[232,-1.4954],[233,-1.4174],[234,-1.2875],[235,-1.1317],[236,-1.0277],[238,-0.7681],[240,-0.4567],[242,-0.0935],[243,0.1397],[244,0.4506],[245,0.8236],[248,2.1271],[250,3.1183],[251,3.5516],[252,3.7992],[253,3.9227],[254,3.9846],[255,3.9846],[256,4.0464],[258,4.0464],[259,4.0147],[260,4.0147],[261,3.9669],[262,3.6804],[263,3.2185],[264,2.7085],[265,2.2141],[266,1.7671],[267,1.2561],[268,0.7286],[269,0.2966],[270,0.0405],[271,0],[282,0],[283,0.0723],[284,0.328],[285,0.6475],[286,0.9987],[288,1.5092],[290,1.9556],[291,2.078],[293,2.2226],[295,2.2226],[296,2.114],[297,1.897],[299,1.3178],[300,1.1005],[302,0.5932],[304,0.1578],[305,0.049],[306,0]]
       }
       ring_01_l_ctrl.Scale.X {type=MovieSceneScriptingFloatChannel, default=1, const [1,307]}
       ring_01_l_ctrl.Scale.Y {type=MovieSceneScriptingFloatChannel, default=1, const [1,307]}
       ring_01_l_ctrl.Scale.Z {type=MovieSceneScriptingFloatChannel, default=1, const [1,307]}
       ring_02_l_ctrl.Location.X {type=MovieSceneScriptingFloatChannel, default=0, thinned_from=307, keys [[0,0],[306,0]]}
       ring_02_l_ctrl.Location.Y {type=MovieSceneScriptingFloatChannel, default=0, thinned_from=307, keys [[0,0],[306,0]]}
       ring_02_l_ctrl.Location.Z {type=MovieSceneScriptingFloatChannel, default=0, thinned_from=307, keys [[0,0],[306,0]]}
       ring_02_l_ctrl.Rotation.X {type=MovieSceneScriptingFloatChannel, default=0, thinned_from=307, keys [[0,0],[306,0]]}
       ring_02_l_ctrl.Rotation.Y {type=MovieSceneScriptingFloatChannel, default=0, thinned_from=307, keys [[0,0],[306,0]]}
       ring_02_l_ctrl.Rotation.Z {
        type=MovieSceneScriptingFloatChannel
        default=0
        thinned_from=307
        keys [[0,0],[1,0],[2,-0.0572],[4,-0.2435],[9,-1.1764],[10,-1.3166],[12,-1.41],[13,-1.41],[14,-1.3633],[15,-1.1764],[16,-0.8496],[17,-0.3455],[18,0.1747],[19,0.7591],[20,1.1482],[21,1.4074],[22,1.5369],[23,1.6017],[25,1.6017],[26,1.4722],[27,1.4074],[28,1.2778],[29,1.213],[31,0.9537],[32,0.8888],[33,0.7591],[34,0.6889],[40,0.0826],[42,0.0826],[43,-0.0186],[50,-0.0186],[51,0.0359],[61,0.0359],[62,-0.0458],[66,-0.0458],[67,-0.0186],[78,-0.0186],[79,-0.2002],[80,-0.2002],[82,-0.5636],[83,-0.7546],[85,-1.3991],[86,-1.6142],[87,-1.937],[90,-2.2601],[97,-2.2601],[98,-2.3679],[99,-2.6881],[100,-2.9164],[101,-2.6881],[103,-1.7761],[104,-1.4347],[106,-0.5258],[107,-0.0721],[109,0.381],[110,0.4942],[111,0.4942],[112,0.6074],[113,0.6074],[114,0.4942],[115,0.3898],[117,-0.1112],[118,-0.3619],[119,-0.5291],[120,-0.78],[121,-0.8637],[123,-0.8637],[125,-0.6964],[127,-0.3619],[128,-0.1112],[130,0.2229],[131,0.2718],[132,0.417],[133,0.417],[134,0.2718],[136,0.2718],[137,0.1267],[140,0.1267],[141,-0.0185],[146,-0.0185],[147,0.0359],[151,0.0359],[152,-0.0857],[153,-0.0857],[154,-0.2074],[157,-0.9383],[158,-1.3043],[159,-1.4264],[160,-1.6707],[161,-1.6707],[162,-1.7929],[163,-1.8202],[164,-1.7785],[165,-1.7785],[167,-1.6118],[168,-1.4452],[169,-1.1953],[172,-0.1976],[173,0.1345],[174,0.3418],[175,0.5076],[176,0.632],[177,0.7148],[178,0.7562],[179,0.8313],[180,0.8926],[183,0.9844],[184,0.9232],[185,0.7394],[186,0.4943],[187,0.3103],[188,0.0956],[189,-0.0886],[190,-0.3035],[192,-0.6723],[194,-0.9183],[195,-1.0015],[200,-1.3007],[201,-1.3007],[202,-1.181],[203,-0.9417],[204,-0.6429],[205,-0.2845],[206,0.1928],[208,1.3835],[209,1.9182],[210,2.3929],[211,2.7271],[212,2.9143],[213,3.0079],[214,3.0546],[217,3.0546],[218,2.8207],[219,2.493],[220,2.0245],[222,0.9917],[223,0.5215],[224,0.2861],[225,0.1919],[226,0.1448],[227,0.1448],[228,-0.073],[231,-1.3827],[232,-1.492],[237,-0.9456],[238,-0.7273],[240,-0.5091],[242,-0.073],[243,0.1448],[244,0.3976],[245,0.9028],[247,1.6594],[250,3.1685],[251,3.545],[252,3.7958],[254,4.0464],[260,4.0464],[261,3.9822],[262,3.661],[263,3.2108],[265,2.1801],[267,1.2762],[268,0.7588],[269,0.3056],[270,0.0463],[271,-0.0186],[282,-0.0186],[283,0.0465],[284,0.3064],[285,0.631],[286,1.0203],[287,1.2796],[288,1.4739],[289,1.7329],[290,1.9271],[291,2.1159],[292,2.1893],[295,2.1893],[296,2.1159],[297,1.8956],[299,1.3076],[301,0.866],[302,0.5714],[304,0.129],[305,0.0552],[306,0]]
       }
       ring_02_l_ctrl.Scale.X {type=MovieSceneScriptingFloatChannel, default=1, const [1,307]}
       ring_02_l_ctrl.Scale.Y {type=MovieSceneScriptingFloatChannel, default=1, const [1,307]}
       ring_02_l_ctrl.Scale.Z {type=MovieSceneScriptingFloatChannel, default=1, const [1,307]}
       ring_03_l_ctrl.Location.X {type=MovieSceneScriptingFloatChannel, default=0, thinned_from=307, keys [[0,0],[306,0]]}
       ring_03_l_ctrl.Location.Y {type=MovieSceneScriptingFloatChannel, default=0, thinned_from=307, keys [[0,0],[306,0]]}
       ring_03_l_ctrl.Location.Z {type=MovieSceneScriptingFloatChannel, default=0, thinned_from=307, keys [[0,0],[306,0]]}
       ring_03_l_ctrl.Rotation.X {type=MovieSceneScriptingFloatChannel, default=0, thinned_from=307, keys [[0,0],[306,0]]}
       ring_03_l_ctrl.Rotation.Y {type=MovieSceneScriptingFloatChannel, default=0, thinned_from=307, keys [[0,0],[306,0]]}
       ring_03_l_ctrl.Rotation.Z {
        type=MovieSceneScriptingFloatChannel
        default=0
        thinned_from=307
        keys [[0,0.0543],[1,0.0543],[2,-0.1557],[4,-0.1557],[9,-1.206],[10,-1.206],[11,-1.4162],[14,-1.4162],[15,-1.206],[16,-0.7857],[17,-0.3267],[19,0.6983],[20,1.1464],[21,1.4024],[22,1.5943],[23,1.5943],[24,1.6583],[25,1.5943],[26,1.4664],[27,1.4024],[28,1.2744],[29,1.2104],[31,0.9544],[32,0.8904],[33,0.7623],[34,0.7071],[35,0.4714],[37,0.4714],[38,0.2356],[40,0.2356],[41,0],[66,0],[67,-0.0274],[68,-0.1051],[79,-0.1051],[80,-0.1829],[81,-0.3384],[82,-0.5717],[83,-0.8142],[84,-1.0734],[85,-1.3845],[87,-1.9034],[88,-2.111],[91,-2.2667],[97,-2.2667],[98,-2.3706],[99,-2.6817],[100,-2.9164],[101,-2.6817],[103,-1.7433],[104,-1.5089],[107,-0.1036],[109,0.3644],[110,0.3644],[111,0.5983],[114,0.5983],[115,0.4079],[116,0.2253],[117,-0.1401],[121,-0.8713],[124,-0.8713],[130,0.2253],[131,0.2719],[132,0.4079],[133,0.4079],[134,0.2719],[136,0.2719],[137,0.1359],[140,0.1359],[141,0],[151,0],[152,-0.0651],[154,-0.1845],[155,-0.3636],[158,-1.2597],[160,-1.6184],[161,-1.738],[162,-1.7978],[163,-1.8251],[164,-1.8251],[165,-1.7398],[166,-1.7398],[167,-1.5692],[168,-1.4839],[169,-1.2281],[173,0.135],[174,0.3904],[175,0.5606],[177,0.7308],[178,0.7308],[179,0.8774],[181,0.8774],[182,1.0061],[183,1.0061],[185,0.7488],[186,0.4913],[187,0.3626],[189,-0.1525],[190,-0.2814],[191,-0.5391],[194,-0.9258],[195,-0.9433],[196,-1.0646],[197,-1.0646],[198,-1.1859],[199,-1.1859],[200,-1.3072],[201,-1.3072],[202,-1.1859],[204,-0.7009],[207,0.7527],[209,1.9624],[210,2.4459],[211,2.7028],[213,3.0972],[217,3.0972],[218,2.9001],[220,2.111],[222,0.9264],[223,0.5312],[225,0.1359],[226,0.1359],[227,0.0832],[228,-0.1274],[229,-0.4962],[230,-0.9706],[231,-1.3397],[232,-1.498],[234,-1.287],[235,-1.1288],[237,-0.9179],[241,-0.2855],[243,0.1359],[244,0.397],[245,0.9191],[246,1.1801],[249,2.7444],[250,3.005],[251,3.5259],[253,4.0465],[260,4.0465],[261,3.9824],[262,3.6617],[263,3.2127],[265,2.1856],[266,1.7359],[268,0.7075],[269,0.3216],[270,0.0642],[271,0],[283,0],[290,1.9843],[291,2.1056],[292,2.1056],[293,2.2558],[295,2.2558],[297,1.9553],[300,1.0532],[301,0.9028],[303,0.3009],[305,0],[306,0]]
       }
       ring_03_l_ctrl.Scale.X {type=MovieSceneScriptingFloatChannel, default=1, const [1,307]}
       ring_03_l_ctrl.Scale.Y {type=MovieSceneScriptingFloatChannel, default=1, const [1,307]}
       ring_03_l_ctrl.Scale.Z {type=MovieSceneScriptingFloatChannel, default=1, const [1,307]}
       pinky_metacarpal_l_ctrl.Location.X {type=MovieSceneScriptingFloatChannel, default=0, thinned_from=307, keys [[0,0],[306,0]]}
       pinky_metacarpal_l_ctrl.Location.Y {type=MovieSceneScriptingFloatChannel, default=0, thinned_from=307, keys [[0,0],[306,0]]}
       pinky_metacarpal_l_ctrl.Location.Z {type=MovieSceneScriptingFloatChannel, default=0, thinned_from=307, keys [[0,0],[306,0]]}
       pinky_metacarpal_l_ctrl.Rotation.X {
        type=MovieSceneScriptingFloatChannel
        default=0
        thinned_from=307
        keys [[0,0],[1,0],[2,-0.0016],[3,0.0021],[4,-0.0021],[5,0.0015],[10,-0.0008],[11,-0.001],[14,-0.001],[15,-0.0006],[16,-0.0035],[17,-0.0008],[19,0.0008],[20,0.0012],[21,-0.0012],[22,0.0015],[24,0.0015],[25,-0.0012],[26,0.0014],[27,-0.0013],[29,-0.0015],[30,0.0011],[31,0.0009],[32,-0.0018],[33,0.0007],[34,0.0003],[36,0.0002],[37,-0.0001],[40,-0.0003],[42,-0.0004],[43,0],[61,0],[62,0.0011],[66,0.0011],[67,-0.0009],[68,0.0022],[78,0.0022],[79,-0.0013],[80,0.0014],[81,0.0005],[82,0.0024],[83,0.001],[84,-0.0012],[85,0.001],[86,-0.0012],[87,0.0008],[98,0.0006],[99,0.0013],[100,0.0008],[101,0.0013],[102,-0.0001],[103,0.0008],[104,-0.0008],[106,0.0005],[107,-0.0013],[109,-0.0008],[111,-0.0006],[112,0.0017],[114,0.0017],[115,-0.0012],[116,-0.0019],[117,0.0008],[118,0.0001],[119,0.0027],[120,-0.0014],[121,-0.0017],[122,0.0017],[123,0.0017],[124,-0.0015],[125,-0.0012],[130,0.002],[131,-0.0014],[133,-0.0014],[134,0.0036],[135,0.0028],[136,0.0028],[137,0.0025],[138,0.0017],[139,0.0017],[140,0.0014],[142,0.0014],[143,0],[146,0],[147,0.0033],[152,0.0033],[153,-0.0018],[155,-0.0023],[156,0.0018],[157,0.0013],[158,0.0004],[160,-0.0004],[162,-0.0009],[163,0.0015],[164,-0.0001],[168,0.0003],[169,0.0005],[170,-0.0009],[172,-0.0003],[174,0],[175,-0.0016],[176,0.0002],[177,0.0002],[178,-0.0015],[179,-0.0025],[180,-0.0024],[183,-0.0026],[184,0.0027],[185,0.0026],[186,-0.0029],[188,-0.0032],[189,0.0017],[190,0.0015],[192,0.0009],[194,0.0007],[195,0.0002],[201,-0.0003],[203,0.0002],[204,0.0007],[205,-0.0012],[207,0],[209,0.0007],[211,0.0008],[212,-0.001],[213,-0.0001],[214,-0.0006],[215,0.0008],[216,0.0013],[217,-0.0021],[218,-0.0011],[219,-0.0009],[220,-0.0003],[221,0],[222,0.0008],[223,0.0005],[224,0.0018],[225,0.0005],[226,-0.0013],[227,0.0005],[228,-0.0002],[229,-0.0016],[230,0.0005],[231,-0.001],[232,-0.0006],[234,-0.0013],[235,0.0011],[238,-0.0005],[239,0.0019],[240,-0.0017],[241,0.0004],[242,-0.0004],[243,0.0006],[246,0.0002],[249,-0.0011],[250,0.0017],[251,-0.0024],[252,0.0007],[255,0.0003],[260,0.0003],[261,-0.0016],[263,-0.0004],[265,-0.0005],[266,-0.0015],[267,-0.0016],[268,0.0017],[269,-0.0001],[270,-0.0006],[271,0.0001],[272,0],[273,-0.0016],[274,-0.0006],[275,0.0009],[277,-0.0014],[278,0.0007],[280,0.0007],[282,-0.0011],[283,0],[284,0],[286,-0.0006],[288,-0.0005],[289,-0.0001],[290,-0.002],[291,-0.0007],[292,0.0024],[293,-0.0006],[296,-0.0007],[298,-0.0014],[299,0.0011],[302,-0.0001],[303,0.0024],[304,-0.001],[305,0.0017],[306,0]]
       }
       pinky_metacarpal_l_ctrl.Rotation.Y {
        type=MovieSceneScriptingFloatChannel
        default=0
        thinned_from=307
        keys [[0,0],[1,0],[3,0.0011],[4,-0.0007],[5,0.0008],[6,-0.0002],[7,0.0016],[8,0.0005],[10,0.0004],[11,0.0014],[14,0.0014],[15,-0.0005],[16,-0.0011],[17,-0.0013],[19,0.0013],[20,0.0004],[21,0.0007],[22,0.0005],[23,-0.0003],[25,-0.0005],[26,-0.0004],[27,0.0002],[28,-0.001],[29,-0.0006],[30,0.0008],[31,-0.0004],[32,0.0002],[33,0.0001],[34,-0.0012],[36,0.0004],[37,-0.0019],[39,-0.0003],[40,-0.0014],[41,0.0006],[42,-0.0005],[43,0],[61,0],[62,0.004],[66,0.004],[67,-0.0001],[68,0.0011],[78,0.0011],[79,-0.0002],[80,-0.0004],[81,-0.0005],[82,-0.0008],[83,0.0002],[84,-0.0011],[85,0.0012],[86,-0.0006],[87,0.0007],[89,0.0002],[91,0.001],[97,0.001],[98,0.0002],[99,0.0007],[100,-0.0003],[102,-0.0005],[103,0.0006],[104,-0.0005],[105,0.0009],[106,-0.0001],[107,-0.0008],[108,-0.0001],[109,-0.0001],[110,-0.0006],[111,0],[113,0.0001],[114,0.0007],[115,-0.0009],[116,-0.0014],[117,0.0011],[118,0.0006],[119,0.0016],[120,-0.0003],[121,0.0001],[122,0.0016],[124,0.0014],[125,-0.0007],[126,-0.0015],[127,0.0006],[128,0.0011],[129,-0.0013],[130,0.0008],[131,0.0015],[133,0.0015],[134,0.0035],[135,-0.0006],[136,-0.0006],[137,0.0018],[138,-0.0022],[139,-0.0022],[140,0.0002],[142,0.0002],[146,0],[147,-0.0003],[152,-0.0003],[153,-0.001],[154,0.0003],[155,0.0006],[156,-0.0014],[157,0.0003],[158,-0.0013],[159,0.0016],[160,-0.0004],[161,0.0011],[162,0.0001],[163,-0.0001],[165,0.0004],[166,0],[167,0],[168,-0.0004],[169,0],[170,-0.0006],[171,0.0001],[172,-0.0001],[173,-0.0008],[174,-0.0002],[177,-0.0002],[178,-0.0004],[180,0.0014],[181,-0.0011],[183,-0.0011],[184,0.0009],[185,0.0015],[186,-0.0009],[188,0.0004],[189,-0.0002],[190,0.0005],[191,-0.0021],[192,-0.0014],[194,0.0019],[195,0.0008],[199,-0.001],[200,-0.0001],[201,-0.0001],[202,-0.0006],[203,-0.0004],[204,0.0005],[205,-0.0003],[206,-0.0006],[207,0.0003],[208,0.0001],[209,0.0009],[210,-0.0001],[211,0.0119],[212,0.0052],[213,0.0017],[214,0],[215,-0.2284],[216,-0.8377],[217,-1.3198],[218,-1.5399],[219,-1.6532],[220,-1.7365],[222,-1.7444],[223,-1.7775],[224,-1.9687],[225,-2.4712],[226,-3.0783],[227,-3.5565],[228,-3.6658],[260,-3.668],[261,-3.6149],[262,-3.3121],[263,-2.8678],[264,-2.3763],[265,-1.9758],[266,-1.7114],[267,-1.5346],[268,-1.4015],[269,-1.3054],[270,-1.2054],[271,-1.0812],[274,-0.6782],[275,-0.5697],[276,-0.4834],[277,-0.4295],[278,-0.3973],[280,-0.3973],[282,-0.3541],[283,-0.2921],[284,-0.1836],[285,-0.0776],[286,-0.005],[288,-0.0032],[289,-0.0011],[290,0.0013],[291,0.0003],[292,0.0006],[293,-0.0002],[295,-0.0002],[296,0.0006],[297,0.0006],[299,-0.0007],[300,0.0008],[301,-0.0007],[302,0.0008],[303,0.0006],[304,-0.0004],[305,0.0001],[306,0]]
       }
       pinky_metacarpal_l_ctrl.Rotation.Z {
        type=MovieSceneScriptingFloatChannel
        default=0
        thinned_from=307
        keys [[0,0],[1,0],[2,-0.0502],[3,-0.1445],[4,-0.2396],[8,-0.9967],[10,-1.2807],[11,-1.3753],[14,-1.3753],[15,-1.1861],[16,-0.8078],[17,-0.3111],[19,0.7162],[20,1.1657],[21,1.4225],[22,1.5512],[23,1.6154],[24,1.6154],[25,1.5831],[29,1.1976],[30,1.0694],[32,0.8765],[33,0.7482],[34,0.6614],[39,0.1813],[42,0.0372],[43,0],[61,0],[62,-0.0289],[66,-0.0289],[67,-0.0477],[68,-0.0844],[78,-0.0844],[79,-0.1216],[80,-0.1953],[81,-0.343],[82,-0.5644],[83,-0.8092],[84,-1.0643],[85,-1.3697],[87,-1.9302],[88,-2.0831],[89,-2.1595],[91,-2.2613],[97,-2.2613],[98,-2.3632],[99,-2.7227],[100,-2.9164],[101,-2.6951],[102,-2.2526],[103,-1.8376],[104,-1.395],[105,-0.9797],[107,-0.094],[108,0.1831],[109,0.3493],[110,0.4601],[111,0.5432],[112,0.5711],[113,0.5711],[114,0.488],[115,0.3635],[116,0.1579],[117,-0.1294],[118,-0.335],[119,-0.5812],[120,-0.746],[121,-0.8281],[122,-0.8688],[123,-0.8688],[124,-0.787],[125,-0.6638],[127,-0.335],[129,0.076],[130,0.2405],[131,0.3453],[133,0.3453],[134,0.2864],[135,0.2269],[136,0.2269],[138,0.108],[139,0.108],[140,0.0487],[142,0.0487],[143,0],[146,0],[147,-0.014],[152,-0.014],[153,-0.0735],[154,-0.1913],[155,-0.3682],[158,-1.2521],[159,-1.4877],[161,-1.7234],[162,-1.7824],[163,-1.8099],[165,-1.7686],[166,-1.7065],[167,-1.603],[168,-1.4374],[169,-1.2097],[170,-0.8785],[172,-0.1743],[173,0.1364],[174,0.3644],[175,0.53],[176,0.6338],[177,0.7167],[178,0.7787],[179,0.8264],[181,0.9515],[183,0.9515],[184,0.8894],[188,0.1382],[189,-0.1116],[192,-0.6744],[194,-0.9243],[195,-0.99],[200,-1.2842],[201,-1.2842],[202,-1.1666],[203,-0.9606],[204,-0.6663],[205,-0.2544],[206,0.2166],[208,1.3949],[209,1.9255],[210,2.3972],[211,2.7413],[212,2.9227],[213,3.0135],[214,3.0589],[215,3.0504],[217,3.0551],[218,2.8423],[219,2.4968],[220,2.0389],[221,1.5158],[222,0.9702],[223,0.5372],[224,0.2573],[225,0.1697],[226,0.1692],[227,0.1042],[228,-0.1081],[229,-0.4985],[230,-0.9924],[231,-1.3565],[232,-1.4865],[233,-1.4085],[239,-0.6283],[240,-0.4464],[241,-0.2903],[242,-0.0822],[243,0.1551],[244,0.4636],[245,0.8339],[246,1.266],[247,1.7291],[248,2.1614],[249,2.6249],[250,3.1196],[251,3.5523],[252,3.7999],[253,3.9236],[254,3.9855],[255,4.0164],[260,4.0164],[261,3.9671],[262,3.6825],[263,3.2232],[264,2.7009],[265,2.2087],[266,1.7611],[267,1.2616],[268,0.7283],[269,0.2915],[270,0.0344],[271,0.0067],[272,-0.0043],[274,0.0063],[275,-0.0017],[276,0.0016],[277,0.0037],[278,0.0051],[280,0.0051],[282,0.0067],[283,0.0726],[284,0.3302],[285,0.651],[286,1.0022],[288,1.5091],[290,1.9528],[291,2.0814],[293,2.2258],[295,2.2258],[296,2.1175],[297,1.9009],[299,1.3236],[300,1.0711],[301,0.8546],[303,0.35],[304,0.1694],[305,0.0255],[306,0]]
       }
       pinky_metacarpal_l_ctrl.Scale.X {type=MovieSceneScriptingFloatChannel, default=1, const [1,307]}
       pinky_metacarpal_l_ctrl.Scale.Y {type=MovieSceneScriptingFloatChannel, default=1, const [1,307]}
       pinky_metacarpal_l_ctrl.Scale.Z {type=MovieSceneScriptingFloatChannel, default=1, const [1,307]}
       pinky_01_l_ctrl.Location.X {type=MovieSceneScriptingFloatChannel, default=0, thinned_from=307, keys [[0,0],[306,0]]}
       pinky_01_l_ctrl.Location.Y {type=MovieSceneScriptingFloatChannel, default=0, thinned_from=307, keys [[0,0],[306,0]]}
       pinky_01_l_ctrl.Location.Z {type=MovieSceneScriptingFloatChannel, default=0, thinned_from=307, keys [[0,0],[306,0]]}
       pinky_01_l_ctrl.Rotation.X {
        type=MovieSceneScriptingFloatChannel
        default=0
        thinned_from=307
        keys [[0,-22.2779],[214,-22.2785],[215,-23.0026],[216,-24.8584],[217,-26.2763],[218,-26.9131],[219,-27.3063],[220,-27.5254],[222,-27.5916],[223,-27.7036],[224,-28.2822],[225,-29.788],[226,-31.6487],[227,-33.0682],[228,-33.4058],[260,-33.3815],[261,-33.2307],[262,-32.3402],[263,-31.0185],[264,-29.4758],[265,-28.2403],[266,-27.4521],[267,-26.9575],[268,-26.5356],[269,-26.2666],[271,-25.5757],[274,-24.3244],[275,-24.0009],[276,-23.7566],[277,-23.5735],[278,-23.4865],[280,-23.4714],[282,-23.3653],[283,-23.1486],[285,-22.5367],[286,-22.2909],[305,-22.2834],[306,0]]
       }
       pinky_01_l_ctrl.Rotation.Y {
        type=MovieSceneScriptingFloatChannel
        default=0
        thinned_from=307
        keys [[0,-10.4787],[19,-10.4829],[20,-10.4686],[25,-10.4819],[26,-10.4686],[29,-10.4831],[30,-10.4693],[32,-10.4836],[35,-10.4717],[38,-10.4828],[41,-10.473],[42,-10.4845],[43,-10.4727],[46,-10.4842],[49,-10.4725],[115,-10.4833],[117,-10.4703],[118,-10.4824],[123,-10.4709],[124,-10.4819],[125,-10.4712],[126,-10.4819],[214,-10.4792],[215,-10.8223],[216,-11.7174],[217,-12.3967],[218,-12.7078],[219,-12.8995],[220,-12.996],[222,-13.0359],[223,-13.0858],[224,-13.3683],[225,-14.0893],[226,-14.985],[227,-15.6713],[228,-15.8283],[260,-15.8212],[261,-15.7553],[262,-15.3277],[263,-14.6648],[264,-13.9534],[265,-13.3518],[266,-12.964],[267,-12.7168],[268,-12.5276],[270,-12.2268],[274,-11.4621],[275,-11.3011],[276,-11.1809],[277,-11.1017],[278,-11.0632],[281,-11.0283],[282,-10.9946],[283,-10.9029],[285,-10.5921],[286,-10.4925],[305,-10.477],[306,0]]
       }
       pinky_01_l_ctrl.Rotation.Z {
        type=MovieSceneScriptingFloatChannel
        default=0
        thinned_from=307
        keys [[0,12.7285],[1,12.7285],[2,12.6956],[3,12.5985],[4,12.4683],[5,12.3051],[9,11.5237],[10,11.3609],[11,11.2957],[12,11.2957],[14,11.2305],[15,11.1651],[16,10.8718],[17,10.4762],[18,10.3742],[20,10.576],[24,10.5775],[26,10.17],[28,9.3583],[30,8.7473],[32,7.9342],[33,7.6272],[34,7.2467],[35,6.9053],[36,6.4791],[38,5.7954],[40,5.2812],[42,4.94],[43,4.9388],[44,4.854],[45,4.854],[46,4.7685],[49,4.7673],[50,4.683],[58,4.683],[59,4.8308],[60,5.114],[61,5.4683],[62,5.893],[63,6.459],[64,7.167],[65,8.0142],[66,8.9309],[67,10.2688],[68,11.778],[69,12.4925],[70,12.6514],[78,12.6514],[80,12.5717],[81,12.4135],[83,11.9468],[85,11.3824],[86,11.0803],[87,10.8174],[88,10.6596],[89,10.5809],[91,10.4888],[97,10.4757],[98,10.3836],[99,10.0123],[100,9.8369],[101,10.0698],[102,10.477],[103,10.9411],[105,11.7542],[106,12.2173],[107,12.6242],[108,12.9139],[109,13.0869],[110,13.2026],[112,13.3187],[113,13.3187],[114,13.2608],[116,12.9095],[118,12.372],[120,12.0121],[121,11.9226],[124,11.9232],[125,12.1023],[126,12.1924],[127,12.4609],[130,12.9989],[131,13.0889],[132,13.1186],[133,13.1186],[135,12.9995],[136,12.9693],[137,12.91],[142,12.7602],[144,12.7602],[145,12.7306],[152,12.7306],[153,12.6108],[154,12.5481],[155,12.3623],[156,12.114],[158,11.4939],[159,11.2464],[161,10.9979],[162,10.9355],[164,10.9355],[166,11.0211],[168,11.2777],[169,11.5342],[172,12.5593],[173,12.8587],[174,13.1142],[176,13.3703],[177,13.4558],[181,13.6923],[182,13.6923],[183,13.7242],[184,13.6614],[185,13.4721],[186,13.2519],[187,13.0628],[188,12.8425],[189,12.6533],[190,12.4321],[191,12.2434],[193,11.928],[195,11.7488],[200,11.4485],[201,11.4485],[202,11.5689],[203,11.8082],[204,12.1084],[205,12.4694],[206,12.9486],[208,14.1479],[209,14.6859],[210,15.1648],[211,15.4844],[212,15.6625],[213,15.7624],[216,15.8136],[217,15.7756],[218,15.1653],[219,14.1072],[220,12.889],[221,11.4889],[222,10.005],[223,8.5564],[224,7.1251],[225,5.6527],[226,4.2137],[227,2.9414],[228,2.0753],[229,1.4341],[230,1.0761],[231,0.9326],[232,0.8611],[233,0.9353],[236,1.3599],[237,1.4341],[240,1.8611],[242,2.2873],[243,2.4691],[244,2.7875],[245,3.169],[249,4.9471],[250,5.4538],[251,5.8971],[252,6.1494],[253,6.2763],[254,6.3394],[258,6.3394],[259,6.3764],[260,6.3764],[261,6.397],[262,6.6126],[263,6.9874],[264,7.3919],[265,7.7555],[266,8.1003],[268,8.7102],[269,9.0401],[270,9.3173],[271,9.6844],[272,9.9974],[275,11.0856],[276,11.3646],[277,11.5908],[278,11.7666],[280,12.0129],[282,12.3714],[283,12.5935],[285,13.2584],[288,14.1556],[290,14.6513],[291,14.8114],[293,14.9584],[295,14.9584],[296,14.8843],[297,14.6613],[299,14.0682],[301,13.6219],[302,13.3258],[304,12.8792],[305,12.805],[306,0]]
       }
       pinky_01_l_ctrl.Scale.X {type=MovieSceneScriptingFloatChannel, default=1, const [1,307]}
       pinky_01_l_ctrl.Scale.Y {type=MovieSceneScriptingFloatChannel, default=1, const [1,307]}
       pinky_01_l_ctrl.Scale.Z {type=MovieSceneScriptingFloatChannel, default=1, const [1,307]}
       pinky_02_l_ctrl.Location.X {type=MovieSceneScriptingFloatChannel, default=0, thinned_from=307, keys [[0,0],[52,0],[53,-0.0002],[58,-0.0002],[59,0],[306,0]]}
       pinky_02_l_ctrl.Location.Y {
        type=MovieSceneScriptingFloatChannel
        default=0
        thinned_from=307
        keys [[0,0],[44,0],[45,0.0002],[52,0],[53,0.0004],[58,0.0004],[59,0],[306,0]]
       }
       pinky_02_l_ctrl.Location.Z {
        type=MovieSceneScriptingFloatChannel
        default=0
        thinned_from=307
        keys [[0,0],[44,0],[45,0.0005],[46,0],[52,0],[53,0.0005],[58,0.0005],[59,0],[306,0]]
       }
       pinky_02_l_ctrl.Rotation.X {
        type=MovieSceneScriptingFloatChannel
        default=0
        thinned_from=307
        keys [[0,0.0001],[5,-0.0001],[7,0],[8,-0.0001],[44,0],[45,-0.0124],[46,0],[52,0.0001],[53,-0.0064],[58,-0.0064],[59,0],[78,0],[80,0.0001],[84,-0.0001],[103,0.0001],[165,0],[167,-0.0001],[172,0.0001],[173,-0.0001],[174,0.0001],[211,0.0001],[216,0],[217,0.0001],[218,0],[219,0.0001],[223,0.0001],[224,-0.0001],[229,0.0001],[230,-0.0001],[236,0.0001],[239,-0.0001],[241,0.0001],[242,-0.0001],[263,-0.0001],[267,0],[269,-0.0001],[272,0.0001],[273,-0.0001],[274,0.0001],[277,-0.0001],[306,0]]
       }
       pinky_02_l_ctrl.Rotation.Y {
        type=MovieSceneScriptingFloatChannel
        default=0
        thinned_from=307
        keys [[0,0],[44,0],[45,0.0041],[46,0],[52,0],[53,0.0075],[58,0.0075],[59,0],[78,0],[79,-0.0003],[80,-0.0003],[81,0],[193,0],[194,-0.0002],[195,0],[306,0]]
       }
       pinky_02_l_ctrl.Rotation.Z {
        type=MovieSceneScriptingFloatChannel
        default=0
        thinned_from=307
        keys [[0,8.9767],[2,8.9767],[5,8.5704],[6,8.2993],[8,8.0282],[9,7.7568],[10,7.6211],[11,7.6211],[12,7.4854],[14,7.4854],[15,7.3496],[16,7.0779],[17,6.7584],[18,6.6569],[19,6.7584],[20,6.7584],[21,6.8599],[24,6.8599],[25,6.6569],[27,6.0474],[28,5.6406],[29,5.3352],[30,4.9278],[31,4.622],[33,3.8055],[34,3.4857],[39,1.7783],[40,1.5216],[41,1.3504],[43,1.1791],[44,1.1791],[45,1.0967],[46,1.0934],[47,1.0077],[50,1.0077],[51,0.9333],[52,0.9333],[53,0.9424],[58,0.9424],[59,1.1441],[60,1.3549],[61,1.7058],[62,2.1967],[63,2.757],[64,3.4563],[65,4.294],[66,5.1997],[67,6.4952],[68,8.0294],[69,8.7545],[70,8.9155],[78,8.9155],[79,8.835],[80,8.835],[81,8.674],[83,8.187],[84,7.9623],[85,7.625],[86,7.4],[87,7.0624],[89,6.8371],[90,6.8371],[91,6.7245],[97,6.7245],[98,6.6118],[99,6.3203],[100,6.0869],[101,6.3203],[102,6.7866],[103,7.1361],[106,8.5317],[108,9.228],[111,9.5759],[113,9.5759],[114,9.4599],[115,9.3739],[116,9.1588],[118,8.6422],[119,8.4267],[120,8.2544],[121,8.1681],[122,8.125],[124,8.2112],[126,8.4698],[127,8.6852],[128,8.8575],[129,9.0727],[130,9.2448],[132,9.3937],[133,9.3937],[136,9.1949],[137,9.1949],[138,9.1286],[139,9.1286],[140,9.0623],[142,9.0623],[143,8.996],[152,8.996],[153,8.9267],[154,8.8032],[155,8.6178],[156,8.3705],[158,7.7518],[160,7.3802],[161,7.2563],[162,7.1944],[165,7.1944],[166,7.3729],[167,7.3729],[169,7.7297],[173,9.1546],[176,9.6879],[178,9.6879],[179,9.8487],[180,9.9123],[181,9.9123],[182,9.9758],[183,9.9758],[184,9.9123],[187,9.3402],[188,9.0858],[192,8.3218],[194,8.0669],[195,7.9619],[196,7.9619],[197,7.84],[198,7.84],[199,7.7181],[201,7.7181],[202,7.84],[204,8.3274],[205,8.6927],[206,9.1794],[208,10.3941],[210,11.364],[211,11.7537],[212,11.937],[213,11.937],[214,12.1202],[216,12.1202],[217,11.937],[218,11.387],[220,9.182],[223,4.7438],[224,3.4414],[225,1.9481],[226,0.4493],[227,-0.735],[228,-1.7327],[229,-2.304],[231,-2.8763],[233,-2.8763],[241,-1.7327],[242,-1.4474],[243,-1.2691],[244,-0.9535],[245,-0.575],[246,-0.1339],[248,0.7467],[250,1.7507],[251,2.1266],[252,2.4396],[255,2.6272],[258,2.6272],[259,2.5722],[260,2.5722],[262,2.8642],[263,3.302],[264,3.5936],[265,4.0306],[267,4.6126],[268,5.0486],[272,6.2092],[273,6.6438],[274,6.9332],[275,7.3108],[276,7.6655],[277,7.9019],[280,8.2562],[281,8.4923],[282,8.6103],[283,8.8463],[285,9.5534],[286,9.7889],[287,10.142],[288,10.3773],[289,10.73],[290,10.8475],[291,11.055],[292,11.1651],[294,11.2384],[295,11.2384],[296,11.1284],[297,10.9082],[299,10.3205],[301,9.8793],[303,9.3642],[304,9.1802],[305,9.0328],[306,0]]
       }
       pinky_02_l_ctrl.Scale.X {type=MovieSceneScriptingFloatChannel, default=1, const [1,307]}
       pinky_02_l_ctrl.Scale.Y {type=MovieSceneScriptingFloatChannel, default=1, const [1,307]}
       pinky_02_l_ctrl.Scale.Z {type=MovieSceneScriptingFloatChannel, default=1, const [1,307]}
       pinky_03_l_ctrl.Location.X {type=MovieSceneScriptingFloatChannel, default=0, thinned_from=307, keys [[0,0],[306,0]]}
       pinky_03_l_ctrl.Location.Y {type=MovieSceneScriptingFloatChannel, default=0, thinned_from=307, keys [[0,0],[306,0]]}
       pinky_03_l_ctrl.Location.Z {type=MovieSceneScriptingFloatChannel, default=0, thinned_from=307, keys [[0,0],[306,0]]}
       pinky_03_l_ctrl.Rotation.X {
        type=MovieSceneScriptingFloatChannel
        default=0
        thinned_from=307
        keys [[0,-0.0004],[4,-0.0003],[5,0.0003],[6,-0.0002],[7,0.0005],[8,0.0005],[9,0],[15,0.0001],[16,-0.0005],[24,-0.0005],[25,0.0016],[26,-0.0006],[27,0.0014],[28,-0.0008],[29,0.0012],[30,0.001],[31,-0.0012],[32,-0.0013],[33,0.0007],[34,0.0011],[38,0.0006],[39,-0.0012],[40,0.0005],[50,0.0003],[51,-0.0006],[52,0.0008],[58,0.0008],[59,-0.0006],[60,0.0008],[61,-0.0006],[65,-0.0006],[66,0.0009],[67,0.0012],[68,0.0008],[82,0.0007],[83,-0.0004],[84,0.0021],[85,-0.0004],[87,-0.0004],[88,0.002],[97,0.002],[98,-0.0005],[99,-0.0001],[114,0.0001],[115,0.0009],[116,0.001],[117,-0.0008],[120,-0.0005],[121,0.0015],[123,0.0015],[124,-0.0005],[130,-0.001],[131,0.0005],[146,0.0005],[147,-0.0005],[151,-0.0005],[152,0.0002],[154,0.0002],[155,-0.0003],[156,0.0004],[157,-0.0001],[159,0],[160,-0.0005],[161,0.0001],[162,0.0002],[163,-0.0004],[170,-0.0006],[171,0.003],[172,-0.0008],[178,-0.0011],[179,0.0003],[180,0.0004],[181,-0.0003],[185,-0.0004],[186,0.0002],[191,-0.0003],[192,0.0003],[194,0.0001],[195,-0.0005],[196,0.0008],[202,0.0008],[203,-0.0005],[204,0.0007],[205,-0.0006],[206,0.0006],[207,0.0005],[208,-0.0008],[209,0.0004],[210,-0.0009],[211,0.0005],[217,0.0005],[218,-0.0008],[221,-0.0007],[222,0.0007],[223,-0.0005],[226,0.0004],[227,-0.0002],[228,-0.0005],[229,0],[236,0],[237,-0.0006],[238,0.0001],[240,0.0002],[241,-0.0005],[242,0.0003],[243,-0.001],[244,0.0016],[245,-0.0009],[246,0.0017],[247,-0.0008],[248,-0.0007],[249,0.0019],[250,-0.0006],[260,-0.0005],[261,0.0003],[269,0.0001],[270,-0.0006],[271,0.0002],[273,0.0002],[274,-0.0005],[275,0.001],[276,0.0009],[277,-0.0006],[280,-0.0007],[281,0.0008],[284,0.0007],[285,-0.0008],[289,-0.0009],[290,0.0006],[291,-0.0002],[296,-0.0002],[297,-0.0003],[298,0.0012],[299,-0.0005],[301,-0.0007],[302,0.0007],[305,0.0005],[306,0]]
       }
       pinky_03_l_ctrl.Rotation.Y {
        type=MovieSceneScriptingFloatChannel
        default=0
        thinned_from=307
        keys [[0,0],[24,-0.0001],[25,0.0002],[26,-0.0001],[27,0.0001],[28,-0.0001],[29,0.0001],[30,0.0001],[31,-0.0002],[32,-0.0002],[34,0.0002],[38,0.0001],[39,-0.0002],[40,0.0001],[50,0.0001],[51,-0.0001],[52,0.0002],[58,0.0002],[59,-0.0001],[60,0.0001],[61,-0.0001],[65,-0.0001],[83,0],[84,0.0002],[87,-0.0001],[88,0.0002],[97,0.0002],[98,-0.0001],[116,0.0001],[118,-0.0001],[120,-0.0001],[121,0.0001],[123,0.0001],[124,-0.0001],[170,0],[171,0.0002],[172,-0.0001],[202,0.0001],[203,0],[207,0.0001],[208,-0.0001],[221,-0.0001],[222,0.0001],[225,0],[226,0.0001],[228,-0.0001],[229,0],[236,0],[237,-0.0002],[238,0],[240,0],[241,-0.0001],[242,0.0001],[243,-0.0003],[244,0.0006],[245,-0.0003],[246,0.0005],[247,-0.0001],[248,-0.0004],[249,0.0004],[250,-0.0001],[251,-0.0003],[260,-0.0001],[269,0],[270,-0.0001],[274,-0.0001],[276,0.0001],[277,-0.0001],[280,-0.0001],[284,0.0001],[297,-0.0001],[301,0],[306,0]]
       }
       pinky_03_l_ctrl.Rotation.Z {
        type=MovieSceneScriptingFloatChannel
        default=0
        thinned_from=307
        keys [[0,6.138],[1,6.138],[3,6.0039],[4,5.8697],[8,5.0644],[9,4.9301],[10,4.7287],[11,4.6615],[13,4.6615],[15,4.5272],[16,4.2585],[17,3.8969],[24,3.8969],[26,3.4746],[27,3.0521],[28,2.8407],[30,1.995],[31,1.7834],[34,0.5129],[38,-0.8713],[39,-1.0445],[40,-1.391],[41,-1.5643],[42,-1.5643],[43,-1.7377],[46,-1.7377],[47,-1.9111],[50,-1.9111],[51,-1.8515],[52,-1.995],[58,-1.995],[59,-1.7081],[60,-1.5647],[62,-0.7048],[63,-0.1321],[64,0.5833],[66,2.2975],[67,3.608],[68,5.1573],[69,5.8452],[70,6.0171],[79,6.0171],[80,5.8452],[81,5.8452],[82,5.5013],[83,5.4403],[84,4.9409],[88,3.9414],[97,3.9414],[98,3.6913],[99,3.4278],[100,3.1911],[101,3.4278],[102,3.901],[103,4.2558],[105,5.2013],[106,5.5557],[107,6.028],[109,6.5002],[110,6.6182],[111,6.6182],[112,6.7362],[113,6.7362],[114,6.6182],[115,6.4029],[121,5.1906],[123,5.1906],[124,5.3927],[125,5.3927],[130,6.4029],[131,6.4385],[135,6.4385],[136,6.2722],[140,6.2722],[141,6.1059],[146,6.1059],[147,6.1434],[151,6.1434],[153,6.0146],[155,5.757],[156,5.4349],[157,5.1772],[158,4.8549],[159,4.6615],[161,4.4036],[163,4.2746],[166,4.2746],[167,4.6673],[168,4.6673],[169,5.0598],[170,5.0598],[174,6.6285],[176,6.6285],[177,7.0204],[178,7.0204],[179,6.9293],[181,7.058],[182,7.058],[183,7.1223],[184,7.058],[185,6.8649],[186,6.6075],[192,5.4483],[194,5.1906],[195,5.153],[196,5.0268],[197,5.0268],[198,4.9006],[202,4.9006],[204,5.4054],[205,5.91],[206,6.2883],[208,7.5485],[210,8.5559],[211,8.8589],[213,9.1282],[217,9.1282],[218,8.3202],[219,6.8378],[220,5.0838],[221,3.1917],[222,1.0249],[223,-1.0117],[224,-3.1902],[225,-5.5128],[226,-7.9822],[227,-9.8505],[228,-11.0817],[229,-11.7757],[230,-12.0844],[231,-12.2389],[232,-12.2389],[233,-12.1617],[234,-12.0072],[235,-11.93],[237,-11.6214],[238,-11.5442],[240,-11.2358],[241,-11.0047],[242,-10.8506],[243,-10.6169],[244,-10.3499],[245,-9.8165],[246,-9.55],[248,-8.4854],[249,-8.2196],[251,-7.1578],[252,-6.8927],[254,-6.8927],[255,-6.6276],[258,-6.6276],[259,-6.7189],[260,-6.7189],[261,-6.6409],[262,-6.0959],[263,-5.2405],[264,-4.2311],[265,-3.3008],[268,-0.9807],[269,-0.209],[270,0.562],[272,1.794],[273,2.3324],[274,2.9474],[275,3.3578],[277,4.2773],[279,4.8898],[281,5.196],[284,6.1139],[285,6.5726],[289,7.7951],[290,7.9478],[291,8.1227],[292,8.2777],[296,8.2777],[298,7.6575],[299,7.5024],[300,7.1922],[301,7.037],[303,6.4163],[305,6.1059],[306,0]]
       }
       pinky_03_l_ctrl.Scale.X {type=MovieSceneScriptingFloatChannel, default=1, const [1,307]}
       pinky_03_l_ctrl.Scale.Y {type=MovieSceneScriptingFloatChannel, default=1, const [1,307]}
       pinky_03_l_ctrl.Scale.Z {type=MovieSceneScriptingFloatChannel, default=1, const [1,307]}
       hand_l_ik_ctrl.Location.X {
        type=MovieSceneScriptingFloatChannel
        default=0
        thinned_from=307
        keys [[0,-10.1528],[5,-10.2699],[7,-10.2614],[8,-10.227],[9,-10.1651],[10,-10.0689],[11,-9.9465],[12,-9.7894],[13,-9.583],[14,-9.3272],[15,-9.0188],[16,-8.6434],[17,-8.1974],[18,-7.6769],[20,-6.6679],[22,-5.5815],[23,-5.0998],[24,-4.7187],[25,-4.3772],[26,-4.0536],[27,-3.7564],[28,-3.4862],[29,-3.2449],[30,-3.0443],[31,-2.888],[32,-2.7551],[33,-2.6501],[34,-2.5526],[35,-2.4651],[37,-2.3004],[38,-2.2264],[39,-2.1823],[40,-2.1295],[42,-2.0736],[45,-2.0575],[47,-2.0574],[50,-2.0824],[51,-2.103],[52,-2.1119],[53,-2.146],[56,-2.2416],[57,-2.2857],[58,-2.3453],[59,-2.4364],[60,-2.5505],[61,-2.6815],[62,-2.8219],[63,-2.9765],[64,-3.1868],[65,-3.4201],[66,-3.7098],[67,-4.0951],[68,-4.5809],[69,-5.1971],[70,-5.8345],[71,-6.4408],[72,-7.0706],[74,-8.4475],[75,-9.0519],[76,-9.4694],[77,-9.7575],[78,-9.9392],[79,-9.9842],[81,-9.9704],[83,-9.9331],[87,-10.0232],[90,-10.0137],[92,-10.029],[96,-10.1708],[97,-10.232],[98,-10.3278],[99,-10.4705],[100,-10.5848],[102,-10.7372],[104,-10.8469],[107,-10.9548],[108,-10.9445],[111,-10.7885],[113,-10.6615],[117,-10.4531],[118,-10.4191],[119,-10.4137],[121,-10.4666],[128,-10.7998],[130,-10.8349],[133,-10.7497],[139,-10.6876],[143,-10.5581],[145,-10.4407],[150,-10.0537],[151,-9.9354],[152,-9.7457],[153,-9.5357],[154,-9.2908],[155,-8.9952],[157,-8.0931],[158,-7.8029],[159,-7.6673],[160,-7.7048],[161,-7.8127],[163,-8.1278],[167,-8.8299],[168,-9.1143],[170,-9.7833],[172,-10.3916],[173,-10.6188],[175,-10.9359],[176,-11.1408],[177,-11.4302],[178,-11.8019],[181,-12.9837],[182,-13.3033],[183,-13.5481],[184,-13.6989],[185,-13.7575],[186,-13.7384],[187,-13.6511],[192,-13.0189],[193,-12.9199],[194,-12.8484],[196,-12.7938],[197,-12.8247],[198,-12.8886],[199,-13.0102],[200,-13.2034],[201,-13.448],[202,-13.7588],[203,-14.1373],[204,-14.5646],[205,-15.0432],[207,-16.0771],[208,-16.6887],[209,-17.5834],[210,-18.6628],[211,-19.6939],[212,-20.5788],[213,-21.3436],[214,-21.9699],[215,-22.4122],[216,-22.7046],[219,-23.351],[221,-23.6733],[225,-24.0618],[228,-24.2145],[231,-24.2501],[233,-24.1706],[234,-24.0796],[236,-23.7141],[238,-23.2076],[239,-22.8732],[240,-22.464],[241,-21.9692],[242,-21.378],[243,-20.6753],[244,-19.8443],[245,-18.878],[247,-16.7385],[248,-15.7184],[250,-13.7678],[251,-12.8669],[252,-12.0785],[253,-11.3594],[254,-10.6848],[255,-10.045],[256,-9.4644],[257,-8.9234],[258,-8.4301],[259,-7.9827],[260,-7.5983],[261,-7.2787],[262,-6.9924],[263,-6.751],[264,-6.5401],[265,-6.3683],[266,-6.2216],[267,-6.1124],[268,-6.0223],[269,-5.9544],[270,-5.9109],[271,-5.8942],[272,-5.8995],[273,-5.9217],[275,-6.0152],[277,-6.1564],[278,-6.2558],[280,-6.5338],[281,-6.6828],[283,-7.0369],[284,-7.2319],[285,-7.4451],[287,-7.9154],[291,-8.915],[293,-9.3791],[294,-9.5811],[295,-9.7491],[296,-9.876],[297,-9.96],[299,-10.0702],[301,-10.1265],[303,-10.1429],[305,-10.1436],[306,0]]
       }
       hand_l_ik_ctrl.Location.Y {
        type=MovieSceneScriptingFloatChannel
        default=0
        thinned_from=307
        keys [[0,3.8823],[2,3.9031],[4,3.885],[5,3.8397],[6,3.8171],[7,3.7589],[8,3.6545],[9,3.5066],[10,3.3455],[11,3.1568],[13,2.7574],[14,2.533],[15,2.3172],[17,1.8994],[18,1.7157],[19,1.5476],[20,1.3912],[21,1.246],[22,1.1312],[23,1.0819],[24,1.1108],[26,1.2046],[28,1.3148],[29,1.3665],[30,1.4091],[32,1.4725],[33,1.4826],[34,1.4881],[35,1.4757],[36,1.4467],[37,1.3988],[38,1.3388],[40,1.1574],[41,1.0695],[42,0.9281],[43,0.7942],[44,0.6409],[45,0.5007],[46,0.3657],[47,0.2247],[48,0.0879],[49,-0.0464],[50,-0.1564],[51,-0.2543],[52,-0.3385],[53,-0.4394],[54,-0.5135],[55,-0.6045],[56,-0.6651],[57,-0.7354],[58,-0.8017],[59,-0.8803],[60,-0.9484],[61,-0.9931],[62,-1.0148],[63,-0.9832],[65,-0.8008],[66,-0.7653],[67,-0.7786],[68,-0.7295],[69,-0.804],[70,-1.0001],[72,-1.434],[73,-1.6756],[74,-1.9895],[75,-2.2161],[76,-2.1873],[77,-1.9322],[78,-1.541],[79,-1.0477],[80,-0.5414],[81,-0.0496],[82,0.4232],[83,0.8393],[84,1.1546],[85,1.4035],[86,1.6188],[87,1.8129],[88,1.9771],[89,2.1077],[90,2.2222],[91,2.3002],[92,2.357],[93,2.4056],[94,2.4488],[95,2.4793],[96,2.5176],[97,2.5317],[98,2.477],[99,2.3104],[100,2.1719],[101,2.1401],[102,2.1741],[104,2.269],[105,2.3332],[106,2.4045],[107,2.4997],[108,2.6197],[109,2.7746],[110,2.9082],[111,3.0106],[112,3.0701],[113,3.0883],[114,3.0801],[115,3.0513],[117,3.0073],[118,2.9663],[120,2.8279],[121,2.7512],[122,2.7032],[123,2.6471],[124,2.6089],[125,2.5764],[126,2.5543],[127,2.5407],[128,2.517],[129,2.513],[130,2.5207],[131,2.5544],[132,2.6075],[133,2.6331],[134,2.6328],[135,2.6512],[136,2.6629],[137,2.6661],[138,2.6816],[140,2.6952],[143,2.6836],[144,2.6902],[145,2.6854],[146,2.6879],[147,2.6756],[149,2.6234],[150,2.588],[151,2.5416],[152,2.4551],[153,2.3298],[154,2.1296],[155,1.8031],[156,1.3467],[157,0.878],[158,0.5069],[159,0.2312],[160,0.0747],[161,-0.0418],[162,-0.1261],[163,-0.1923],[164,-0.2381],[165,-0.2652],[166,-0.2694],[167,-0.2486],[168,-0.1225],[169,0.144],[170,0.5323],[171,0.982],[172,1.4279],[173,1.8034],[174,2.1405],[175,2.4288],[176,2.7063],[177,3.0134],[178,3.3669],[179,3.7607],[180,4.1788],[181,4.6114],[182,5.005],[183,5.3333],[184,5.5899],[185,5.7403],[186,5.7949],[187,5.7609],[188,5.6957],[191,5.4497],[193,5.2576],[194,5.1757],[195,5.1162],[196,5.0816],[197,5.0736],[199,5.1122],[201,5.0755],[202,5.0676],[203,5.0765],[204,5.1149],[205,5.2133],[206,5.4271],[207,5.7649],[208,6.2163],[209,6.7998],[210,7.448],[211,8.3091],[212,9.315],[215,12.3961],[216,13.3615],[217,14.135],[218,14.7492],[219,15.247],[220,15.6766],[221,16.0628],[222,16.414],[224,17.0026],[227,17.761],[229,18.124],[231,18.435],[233,18.6286],[235,18.6898],[236,18.6569],[237,18.5836],[238,18.4461],[239,18.233],[240,17.8947],[241,17.3648],[242,16.6412],[243,15.7501],[244,14.6969],[245,13.4599],[246,12.1322],[247,10.8619],[248,9.7439],[249,8.754],[250,7.8423],[251,7.0379],[252,6.3705],[253,5.7928],[254,5.2461],[255,4.7315],[256,4.2563],[257,3.8147],[258,3.4108],[259,3.0477],[260,2.7315],[261,2.457],[263,2.0035],[264,1.8229],[265,1.6535],[266,1.4952],[267,1.3483],[268,1.2066],[269,1.1057],[270,0.9934],[271,0.8946],[272,0.8011],[273,0.7196],[274,0.6279],[275,0.5613],[276,0.4984],[278,0.4166],[279,0.3869],[280,0.3759],[281,0.3869],[282,0.4095],[283,0.4531],[284,0.5159],[285,0.5966],[286,0.6875],[287,0.8001],[288,0.9426],[289,1.1182],[290,1.3065],[291,1.5097],[292,1.7395],[293,1.9949],[294,2.2596],[295,2.5078],[296,2.747],[297,2.9721],[298,3.1842],[299,3.3646],[300,3.511],[301,3.6213],[302,3.6997],[303,3.7665],[304,3.8106],[305,3.8288],[306,0]]
       }
       hand_l_ik_ctrl.Location.Z {
        type=MovieSceneScriptingFloatChannel
        default=0
        thinned_from=307
        keys [[0,-23.0481],[13,-23.3857],[14,-23.3853],[15,-23.333],[16,-23.2171],[17,-23.0129],[19,-22.3527],[20,-21.9563],[21,-21.4505],[22,-20.8853],[23,-20.3825],[26,-19.3682],[30,-18.3142],[36,-16.993],[40,-16.0558],[41,-15.8551],[42,-15.62],[47,-14.6514],[50,-14.1801],[52,-13.9483],[55,-13.7012],[58,-13.5557],[60,-13.5774],[61,-13.6342],[63,-13.8268],[65,-14.1369],[66,-14.3829],[67,-15.0354],[68,-16.1787],[69,-17.4704],[70,-18.6336],[71,-19.5943],[72,-20.4768],[73,-21.2489],[74,-21.8414],[75,-22.2885],[76,-22.6267],[79,-23.4259],[82,-24.4198],[87,-25.827],[89,-26.2348],[93,-26.706],[98,-27.0727],[110,-27.1257],[119,-26.5044],[140,-26.5702],[144,-26.3901],[146,-26.1672],[150,-25.4383],[151,-25.1818],[153,-24.4099],[154,-23.9105],[155,-23.2531],[156,-22.1175],[157,-20.912],[158,-20.0739],[159,-19.6192],[160,-19.5886],[161,-19.7799],[162,-20.1129],[167,-22.3344],[168,-22.9736],[170,-24.5397],[171,-25.2682],[172,-25.8803],[173,-26.2848],[174,-26.5444],[176,-26.9353],[177,-27.2141],[181,-28.6788],[182,-29.0037],[183,-29.2505],[184,-29.4174],[185,-29.4759],[186,-29.4372],[193,-28.4432],[195,-28.2793],[198,-28.268],[201,-28.5712],[205,-29.3038],[207,-29.8284],[208,-30.1931],[209,-30.7896],[210,-31.5475],[212,-33.1542],[213,-33.8492],[214,-34.4321],[215,-34.8854],[220,-36.4256],[224,-37.2576],[233,-38.4606],[234,-38.5219],[235,-38.5034],[237,-38.2429],[239,-37.7748],[240,-37.4572],[241,-37.058],[242,-36.5694],[243,-35.9859],[244,-35.3056],[245,-34.5074],[246,-33.594],[251,-28.608],[255,-25.2216],[259,-22.1179],[260,-21.4121],[262,-20.1305],[264,-18.9857],[266,-18.0125],[268,-17.1634],[269,-16.8005],[270,-16.4904],[271,-16.2423],[272,-16.0561],[274,-15.7935],[276,-15.6731],[277,-15.6617],[278,-15.6943],[279,-15.7871],[280,-15.9146],[281,-16.0891],[282,-16.2989],[283,-16.5435],[284,-16.8511],[285,-17.209],[286,-17.602],[288,-18.5069],[293,-20.8843],[294,-21.3162],[295,-21.6834],[296,-21.989],[298,-22.4413],[300,-22.7505],[303,-22.9659],[305,-23.0221],[306,0]]
       }
       hand_l_ik_ctrl.Rotation.X {
        type=MovieSceneScriptingFloatChannel
        default=0
        thinned_from=307
        keys [[0,32.7366],[4,32.7204],[7,32.8284],[10,33.1826],[12,33.5618],[13,33.8065],[14,34.1469],[15,34.5747],[16,35.0943],[17,35.77],[18,36.6302],[19,37.7563],[20,39.0755],[21,40.5896],[22,42.2599],[23,43.7612],[24,44.8577],[25,45.7315],[26,46.4833],[27,47.0991],[29,47.9802],[32,48.7973],[49,51.179],[55,51.2806],[60,50.7982],[63,50.1549],[64,49.8206],[65,49.3667],[66,48.7544],[67,47.8255],[68,46.4977],[69,45.0672],[71,42.5788],[72,41.4975],[73,40.7578],[74,40.2805],[77,39.3318],[78,38.9505],[82,36.962],[84,36.1511],[88,34.9126],[90,34.4278],[94,33.8175],[97,33.5146],[98,33.4637],[100,33.514],[106,32.9593],[110,32.4262],[112,32.3545],[126,33.1758],[139,33.0305],[144,33.3062],[147,33.7123],[150,34.4055],[151,34.755],[152,35.2579],[153,35.8461],[154,36.6354],[155,37.7224],[156,39.2075],[157,40.8405],[158,42.2147],[159,43.2314],[160,43.7537],[161,44.0325],[162,44.1631],[164,44.1428],[166,43.8202],[167,43.514],[168,42.8125],[169,41.7225],[170,40.3759],[172,37.3171],[173,35.8806],[174,34.5167],[178,29.4529],[179,28.2763],[180,27.2225],[181,26.2604],[182,25.4472],[183,24.8001],[184,24.3677],[185,24.202],[186,24.2598],[187,24.4974],[192,26.2211],[194,26.787],[195,26.9925],[196,27.115],[197,27.1698],[199,27.0755],[200,26.9754],[201,26.7334],[202,26.3199],[203,25.6924],[204,24.8352],[205,23.7565],[206,22.4652],[207,21.0285],[208,19.4131],[209,17.5945],[210,15.7356],[212,11.8936],[213,10.1198],[214,8.4687],[215,6.8971],[216,5.425],[217,4.2335],[218,3.294],[219,2.5378],[220,1.888],[221,1.2983],[222,0.7689],[223,0.2947],[224,-0.128],[225,-0.5297],[226,-0.9091],[227,-1.2581],[228,-1.5101],[229,-1.7307],[230,-1.8826],[231,-2.0616],[232,-2.1374],[233,-2.1633],[234,-2.0818],[235,-1.9761],[236,-1.8494],[237,-1.7325],[238,-1.54],[239,-1.1787],[240,-0.5418],[241,0.4822],[242,1.9239],[243,3.7591],[244,6.0154],[245,8.7643],[247,14.9031],[248,17.7545],[249,20.4286],[250,23.0058],[251,25.3532],[252,27.3494],[253,29.0777],[254,30.6848],[255,32.1596],[256,33.4874],[257,34.6658],[258,35.6853],[259,36.5608],[260,37.3103],[261,37.9493],[263,38.9492],[265,39.6585],[268,40.4262],[271,40.8884],[274,41.0903],[277,41.0145],[279,40.7686],[282,40.0659],[285,39.0035],[294,34.777],[296,34.0072],[298,33.4819],[300,33.1139],[303,32.825],[305,32.7719],[306,0]]
       }
       hand_l_ik_ctrl.Rotation.Y {
        type=MovieSceneScriptingFloatChannel
        default=0
        thinned_from=307
        keys [[0,-16.5117],[5,-17.1046],[7,-17.2972],[9,-17.4233],[12,-17.5525],[13,-17.669],[15,-18.015],[16,-18.2287],[18,-18.7656],[19,-19.1992],[20,-19.6726],[22,-20.5643],[23,-20.9694],[24,-21.228],[25,-21.3069],[26,-21.2597],[27,-21.1367],[28,-20.9613],[29,-20.7066],[31,-20.0692],[33,-19.3253],[34,-18.9026],[35,-18.4349],[38,-16.8358],[42,-14.8252],[44,-13.8804],[45,-13.4525],[47,-12.6756],[49,-12.0122],[50,-11.7213],[52,-11.2143],[53,-10.9902],[55,-10.5907],[57,-10.2278],[58,-10.0685],[59,-9.9627],[60,-9.8832],[61,-9.8397],[62,-9.8176],[63,-9.8251],[66,-10.1044],[67,-10.5167],[68,-11.3913],[69,-12.5405],[70,-13.5331],[73,-15.3376],[74,-15.7137],[76,-16.2963],[79,-17.2962],[82,-18.5202],[83,-18.9558],[86,-20.3837],[87,-20.8075],[88,-21.1293],[89,-21.3731],[91,-21.6821],[98,-22.247],[99,-22.1771],[100,-21.9962],[102,-21.4873],[103,-21.2808],[105,-20.9946],[114,-20.3407],[115,-20.2273],[116,-20.0641],[119,-19.4277],[121,-19.0807],[122,-18.9469],[123,-18.8538],[126,-18.704],[131,-18.154],[132,-18.0746],[137,-18.0297],[142,-18.2032],[144,-18.3055],[145,-18.3198],[147,-18.1797],[149,-17.9414],[150,-17.7777],[151,-17.5431],[152,-17.1649],[153,-16.712],[154,-16.2114],[155,-15.4858],[157,-12.0643],[158,-11.0725],[159,-10.7347],[160,-11.0277],[161,-11.6883],[162,-12.6211],[163,-13.7322],[164,-14.8755],[165,-15.9757],[167,-18.1177],[168,-19.3374],[169,-20.7684],[170,-22.2802],[171,-23.6661],[172,-24.7218],[173,-25.2748],[174,-25.4457],[176,-25.5458],[178,-25.8054],[179,-25.808],[180,-25.6525],[181,-25.2047],[182,-24.5791],[183,-24.07],[184,-23.8001],[185,-23.6032],[187,-23.3137],[189,-23.2218],[194,-23.3317],[198,-23.8325],[200,-24.2147],[201,-24.3411],[202,-24.3491],[203,-24.1387],[205,-23.2336],[206,-22.6776],[207,-21.8547],[208,-21.2043],[209,-21.1511],[210,-21.3135],[211,-21.5198],[212,-21.5145],[213,-21.3775],[214,-21.4067],[215,-21.8053],[216,-22.5283],[217,-23.3708],[220,-26.1762],[221,-27.0538],[222,-27.8638],[223,-28.5869],[226,-30.4773],[227,-31.033],[228,-31.5262],[231,-32.7249],[232,-33.0411],[233,-33.2797],[235,-33.6095],[236,-33.8673],[238,-34.7088],[239,-35.0772],[240,-35.308],[241,-35.4369],[242,-35.4861],[243,-35.4619],[245,-35.2177],[247,-34.8176],[248,-34.5584],[249,-34.2113],[250,-33.7765],[252,-32.6932],[253,-31.9945],[254,-31.1169],[255,-30.0804],[256,-28.9202],[257,-27.5865],[258,-26.041],[260,-22.7439],[263,-17.9236],[264,-16.359],[265,-14.9076],[266,-13.5815],[267,-12.3316],[268,-11.1655],[269,-10.1143],[270,-9.2202],[271,-8.4982],[272,-7.9223],[273,-7.4649],[274,-7.0978],[275,-6.8115],[276,-6.5751],[277,-6.3877],[278,-6.2878],[279,-6.2819],[280,-6.3426],[281,-6.4411],[282,-6.5911],[283,-6.7788],[284,-7.0889],[285,-7.4707],[287,-8.267],[289,-9.0328],[290,-9.4547],[291,-9.9438],[292,-10.5169],[294,-11.7378],[295,-12.2697],[296,-12.8699],[298,-14.2503],[299,-14.8741],[300,-15.344],[301,-15.6765],[302,-15.9399],[303,-16.1603],[305,-16.4414],[306,0]]
       }
       hand_l_ik_ctrl.Rotation.Z {
        type=MovieSceneScriptingFloatChannel
        default=0
        thinned_from=307
        keys [[0,-4.6521],[1,-4.6501],[3,-4.6903],[4,-4.7028],[5,-4.7428],[7,-4.7463],[8,-4.7379],[9,-4.718],[10,-4.6612],[11,-4.5852],[12,-4.4751],[13,-4.2791],[14,-4.0302],[15,-3.7223],[16,-3.3454],[17,-2.8965],[18,-2.3377],[19,-1.6787],[20,-0.9757],[21,-0.2489],[22,0.4843],[23,1.1634],[24,1.7542],[25,2.2714],[26,2.7339],[27,3.1414],[28,3.4963],[29,3.7917],[30,4.0283],[31,4.2094],[32,4.3626],[33,4.467],[34,4.5603],[35,4.6171],[36,4.6473],[37,4.657],[38,4.6394],[40,4.4826],[41,4.3845],[42,4.2278],[43,4.0531],[44,3.8648],[46,3.5068],[49,2.9284],[50,2.7616],[51,2.6038],[52,2.475],[53,2.3077],[54,2.1685],[55,2.0151],[56,1.8867],[57,1.7426],[58,1.59],[59,1.3974],[60,1.1818],[61,0.9438],[62,0.695],[63,0.4818],[64,0.2865],[65,0.0136],[66,-0.4638],[67,-1.2717],[68,-2.3434],[69,-3.6855],[70,-5.1566],[71,-6.6561],[72,-8.0976],[73,-9.4326],[74,-10.8167],[75,-11.9983],[76,-12.7457],[77,-13.1755],[78,-13.3637],[79,-13.2551],[80,-12.9815],[81,-12.6571],[83,-11.9531],[84,-11.6741],[87,-10.9756],[89,-10.5844],[90,-10.4266],[91,-10.3297],[92,-10.2576],[94,-10.1958],[96,-10.1923],[97,-10.2172],[98,-10.3292],[99,-10.5819],[100,-10.8089],[101,-10.9241],[104,-11.0596],[106,-11.0897],[107,-11.0748],[108,-11.0169],[111,-10.6974],[113,-10.5744],[117,-10.5286],[118,-10.5561],[120,-10.6774],[123,-10.8812],[129,-11.1863],[130,-11.2122],[133,-11.1129],[139,-11.0235],[142,-10.9039],[143,-10.84],[145,-10.6683],[147,-10.4735],[149,-10.3072],[150,-10.195],[151,-10.0289],[153,-9.5714],[155,-9.0753],[156,-8.7372],[158,-7.8122],[159,-7.3861],[160,-7.0578],[161,-6.7849],[162,-6.5766],[163,-6.4419],[165,-6.2692],[166,-6.2095],[167,-6.1916],[168,-6.2476],[169,-6.3187],[170,-6.3349],[172,-6.295],[173,-6.3261],[174,-6.4004],[175,-6.5436],[177,-6.9346],[178,-7.1087],[179,-7.256],[180,-7.3605],[181,-7.4886],[182,-7.6562],[183,-7.7853],[184,-7.7993],[185,-7.7524],[186,-7.663],[187,-7.5491],[188,-7.397],[191,-6.8905],[192,-6.7641],[193,-6.668],[194,-6.5944],[196,-6.4816],[197,-6.4625],[198,-6.4691],[199,-6.5541],[200,-6.8161],[201,-7.2386],[202,-7.8123],[203,-8.5513],[204,-9.4348],[205,-10.4067],[206,-11.4641],[207,-12.6741],[208,-13.9324],[209,-15.1531],[210,-16.4631],[211,-17.6486],[212,-18.753],[213,-19.8036],[214,-20.5887],[215,-20.8628],[216,-20.6804],[217,-20.3426],[218,-19.9298],[220,-18.9595],[224,-16.6914],[225,-16.0797],[230,-12.7968],[232,-11.4179],[234,-9.9919],[235,-9.1836],[236,-8.2286],[237,-7.0403],[238,-5.7379],[239,-4.473],[240,-3.32],[241,-2.2731],[242,-1.3504],[243,-0.5686],[244,0.0574],[245,0.5299],[246,0.8636],[247,1.07],[248,1.1583],[249,1.1801],[250,1.131],[251,1.0244],[252,0.9077],[253,0.7551],[254,0.5495],[255,0.3168],[256,0.0727],[257,-0.1848],[258,-0.4445],[259,-0.6878],[260,-0.8961],[261,-1.0678],[262,-1.196],[263,-1.328],[264,-1.4261],[267,-1.6974],[268,-1.7958],[269,-1.8719],[270,-1.9671],[271,-2.0704],[272,-2.186],[273,-2.3136],[275,-2.6195],[276,-2.7925],[277,-2.9516],[278,-3.1462],[280,-3.6041],[281,-3.8234],[285,-4.7444],[286,-4.9843],[287,-5.1976],[288,-5.3972],[290,-5.6885],[291,-5.7838],[292,-5.8335],[293,-5.8344],[294,-5.7975],[295,-5.7364],[296,-5.6352],[297,-5.4834],[298,-5.3142],[299,-5.1562],[300,-5.0267],[301,-4.9255],[302,-4.8371],[303,-4.7708],[304,-4.7154],[305,-4.682],[306,0]]
       }
       hand_l_ik_ctrl.Scale.X {type=MovieSceneScriptingFloatChannel, default=1, const [1,307]}
       hand_l_ik_ctrl.Scale.Y {type=MovieSceneScriptingFloatChannel, default=1, const [1,307]}
       hand_l_ik_ctrl.Scale.Z {type=MovieSceneScriptingFloatChannel, default=1, const [1,307]}
       Softness {type=MovieSceneScriptingFloatChannel, default=0, const [0,307]}
       hand_l_ikorient_ctrl.Rotation.Roll {type=MovieSceneScriptingFloatChannel, default=0, const [0,307]}
       hand_l_ikorient_ctrl.Rotation.Pitch {type=MovieSceneScriptingFloatChannel, default=0, const [0,307]}
       hand_l_ikorient_ctrl.Rotation.Yaw {type=MovieSceneScriptingFloatChannel, default=0, const [0,307]}
       arm_l_pv_ik_ctrl.Location.X {
        type=MovieSceneScriptingFloatChannel
        default=0
        thinned_from=307
        keys [[0,-12.9858],[3,-13.0483],[4,-13.0418],[6,-12.9708],[7,-12.8945],[8,-12.7383],[9,-12.5124],[10,-12.2349],[11,-11.8962],[12,-11.4962],[13,-11.0238],[14,-10.4415],[15,-9.7908],[16,-9.0639],[17,-8.2294],[18,-7.291],[19,-6.2551],[20,-5.1494],[21,-3.9634],[22,-2.7472],[23,-1.7003],[24,-0.9456],[25,-0.3373],[26,0.1952],[27,0.6386],[28,1.0039],[29,1.3028],[30,1.5431],[31,1.7284],[32,1.878],[33,2.0027],[34,2.1131],[36,2.3083],[37,2.4214],[38,2.5249],[39,2.6174],[40,2.7219],[41,2.7876],[42,2.8981],[43,2.968],[44,3.0564],[45,3.1334],[47,3.2636],[49,3.3771],[50,3.4134],[52,3.4497],[53,3.4521],[54,3.4355],[55,3.4261],[56,3.3907],[57,3.3475],[58,3.2876],[59,3.2051],[60,3.0847],[61,2.9201],[62,2.7145],[63,2.4588],[64,2.1166],[65,1.704],[66,1.1956],[67,0.4737],[68,-0.5541],[69,-1.6372],[71,-3.6034],[72,-4.5437],[73,-5.3684],[74,-6.0956],[75,-6.7515],[76,-7.3847],[78,-8.7402],[79,-9.3702],[80,-9.9287],[81,-10.4413],[82,-10.9011],[83,-11.299],[84,-11.611],[85,-11.8657],[86,-12.0725],[87,-12.2531],[88,-12.4038],[91,-12.722],[97,-13.2067],[98,-13.2571],[100,-13.3033],[101,-13.4168],[109,-14.6358],[110,-14.7424],[111,-14.8034],[112,-14.8203],[114,-14.7563],[120,-14.3248],[121,-14.2723],[124,-14.2367],[126,-14.2716],[130,-14.4788],[133,-14.5611],[139,-14.557],[142,-14.421],[144,-14.2727],[146,-14.0614],[148,-13.7821],[149,-13.616],[150,-13.4159],[151,-13.1338],[152,-12.7043],[153,-12.1919],[154,-11.523],[155,-10.635],[156,-9.4754],[157,-8.2108],[158,-7.1547],[159,-6.3587],[160,-5.9124],[161,-5.6016],[162,-5.3866],[163,-5.2383],[164,-5.1394],[165,-5.078],[166,-5.0615],[167,-5.1193],[168,-5.4649],[169,-6.0857],[170,-6.8774],[171,-7.7888],[172,-8.7655],[173,-9.6955],[175,-11.4827],[176,-12.4315],[177,-13.4779],[179,-15.7374],[180,-16.803],[181,-17.7827],[182,-18.6128],[183,-19.2704],[184,-19.717],[185,-19.9177],[186,-19.909],[187,-19.7201],[188,-19.4363],[192,-18.1208],[193,-17.8379],[194,-17.6014],[195,-17.4274],[196,-17.3144],[197,-17.2848],[198,-17.3325],[199,-17.4643],[200,-17.6779],[201,-18.0433],[202,-18.595],[203,-19.3629],[204,-20.3432],[205,-21.5159],[206,-22.8966],[207,-24.4658],[208,-26.2123],[209,-28.1663],[211,-32.234],[212,-34.2013],[213,-35.9973],[214,-37.6264],[215,-39.0421],[216,-40.2323],[217,-41.1333],[218,-41.8093],[219,-42.3269],[221,-43.0955],[224,-43.8615],[227,-44.3498],[229,-44.3977],[231,-44.3162],[233,-43.9921],[235,-43.3737],[237,-42.4729],[238,-41.8772],[239,-41.1534],[240,-40.2341],[241,-39.0323],[242,-37.5218],[243,-35.7129],[244,-33.5752],[245,-31.0523],[247,-25.6203],[248,-23.2049],[249,-20.9959],[250,-18.9148],[251,-17.0534],[252,-15.4872],[253,-14.1421],[254,-12.9067],[255,-11.7742],[256,-10.7565],[257,-9.8474],[258,-9.0501],[259,-8.3566],[260,-7.7576],[261,-7.2444],[262,-6.8088],[263,-6.4104],[264,-6.0854],[265,-5.7993],[266,-5.5487],[267,-5.3347],[268,-5.1493],[269,-5.0137],[270,-4.8906],[271,-4.7987],[272,-4.7363],[273,-4.7067],[274,-4.6869],[275,-4.7102],[276,-4.7604],[277,-4.8413],[278,-4.9588],[279,-5.1351],[280,-5.3452],[281,-5.5908],[282,-5.8712],[283,-6.1905],[284,-6.545],[285,-6.9311],[286,-7.3456],[287,-7.7775],[293,-10.5906],[294,-11.0396],[295,-11.4319],[296,-11.771],[297,-12.0481],[298,-12.2878],[299,-12.4827],[300,-12.6413],[301,-12.7559],[303,-12.8955],[305,-12.9368],[306,0]]
       }
       arm_l_pv_ik_ctrl.Location.Y {
        type=MovieSceneScriptingFloatChannel
        default=0
        thinned_from=307
        keys [[0,-20.5491],[7,-20.5774],[10,-20.4799],[12,-20.3378],[13,-20.222],[14,-20.0567],[15,-19.8234],[16,-19.5109],[17,-19.1009],[18,-18.5779],[19,-18.0077],[20,-17.3589],[21,-16.5819],[22,-15.7093],[23,-14.9066],[24,-14.3005],[25,-13.8015],[26,-13.3567],[27,-12.9728],[28,-12.6483],[29,-12.3662],[30,-12.1263],[31,-11.9328],[32,-11.7658],[34,-11.5164],[38,-11.1192],[40,-10.9434],[41,-10.8851],[42,-10.7963],[49,-10.416],[52,-10.3255],[55,-10.337],[57,-10.3961],[58,-10.4435],[59,-10.524],[60,-10.63],[61,-10.7657],[63,-11.1023],[64,-11.3109],[65,-11.5474],[66,-11.8875],[67,-12.5365],[68,-13.519],[69,-14.5635],[70,-15.5314],[71,-16.4227],[72,-17.2469],[73,-18.0026],[74,-18.684],[75,-19.2328],[76,-19.602],[77,-19.8546],[81,-20.5894],[86,-21.2303],[91,-21.6353],[98,-22.0102],[104,-22.6278],[107,-22.8016],[111,-22.8029],[119,-22.5253],[126,-22.8194],[130,-23.0712],[139,-23.0625],[143,-22.877],[146,-22.6102],[150,-22.117],[151,-21.957],[153,-21.4813],[154,-21.1529],[155,-20.7502],[156,-20.2317],[157,-19.6383],[158,-19.1397],[159,-18.7762],[160,-18.6011],[162,-18.429],[166,-18.2983],[167,-18.3088],[168,-18.5071],[170,-19.2108],[173,-20.4465],[175,-21.1508],[176,-21.5492],[177,-22.019],[179,-23.0741],[180,-23.5467],[181,-23.9514],[182,-24.2696],[183,-24.4988],[184,-24.6469],[185,-24.7099],[186,-24.6971],[188,-24.5095],[193,-23.8334],[196,-23.6123],[198,-23.6296],[199,-23.7168],[200,-23.8838],[201,-24.116],[202,-24.4378],[203,-24.876],[206,-26.3324],[208,-27.1211],[210,-27.7579],[211,-27.9248],[212,-27.9555],[213,-27.8876],[214,-27.735],[215,-27.5015],[216,-27.2017],[217,-26.9619],[218,-26.8048],[222,-26.5305],[226,-26.4861],[229,-26.6247],[234,-27.1039],[235,-27.1529],[242,-27.0297],[243,-26.9731],[244,-26.8376],[245,-26.5911],[246,-26.1864],[247,-25.6303],[248,-24.975],[249,-24.2528],[251,-22.68],[252,-21.9615],[254,-20.6622],[255,-20.0496],[257,-18.9774],[259,-18.1013],[260,-17.7408],[261,-17.4233],[262,-17.1425],[264,-16.6901],[266,-16.3573],[268,-16.1235],[271,-15.9106],[274,-15.8474],[276,-15.8912],[278,-16.0298],[280,-16.3047],[282,-16.6771],[284,-17.1394],[292,-19.3907],[294,-19.8734],[295,-20.0622],[296,-20.206],[299,-20.4389],[303,-20.5422],[305,-20.5438],[306,0]]
       }
       arm_l_pv_ik_ctrl.Location.Z {
        type=MovieSceneScriptingFloatChannel
        default=0
        thinned_from=307
        keys [[0,4.9822],[2,4.9996],[3,5.0208],[4,5.0294],[5,5.0562],[6,5.05],[8,5.0688],[9,5.0683],[10,5.0418],[11,5.0014],[12,4.9358],[13,4.8319],[14,4.705],[15,4.5544],[16,4.3687],[18,3.9577],[19,3.882],[20,3.871],[21,3.8776],[22,3.9073],[23,3.9261],[24,3.9047],[25,3.8556],[26,3.7883],[27,3.6993],[29,3.4855],[30,3.4043],[31,3.3534],[32,3.3164],[34,3.2925],[35,3.2949],[37,3.3253],[38,3.3473],[41,3.4962],[45,3.8259],[48,4.0415],[49,4.1074],[51,4.2113],[52,4.2417],[53,4.2908],[55,4.3571],[57,4.394],[58,4.4246],[61,4.5972],[62,4.6301],[64,4.6732],[65,4.6518],[66,4.5941],[67,4.4585],[68,4.1724],[69,3.918],[70,3.7507],[71,3.5917],[72,3.5398],[73,3.7212],[74,4.0246],[75,4.3097],[76,4.4535],[77,4.4822],[78,4.4195],[79,4.2326],[80,4.0122],[82,3.6193],[83,3.4513],[84,3.3306],[87,3.0396],[89,2.8632],[90,2.7847],[92,2.6881],[93,2.6674],[94,2.6537],[96,2.6597],[97,2.6841],[98,2.7492],[99,2.8995],[100,3.0309],[101,3.1323],[102,3.2259],[103,3.2898],[104,3.3253],[105,3.3433],[106,3.3498],[107,3.3362],[108,3.2863],[110,3.1349],[112,3.0127],[113,2.9747],[116,2.9236],[117,2.8897],[118,2.8757],[119,2.9016],[120,2.9537],[121,3.0231],[122,3.1053],[123,3.1983],[124,3.2766],[127,3.4771],[128,3.532],[129,3.5629],[130,3.5786],[131,3.5635],[132,3.5301],[133,3.505],[137,3.4593],[139,3.4411],[141,3.4359],[142,3.4229],[143,3.4232],[144,3.4078],[145,3.4188],[146,3.4046],[147,3.4155],[148,3.4336],[149,3.4646],[151,3.5384],[152,3.5594],[153,3.5929],[154,3.6762],[155,3.8521],[156,4.1546],[157,4.5444],[159,5.3826],[160,5.7358],[161,6.0224],[162,6.2436],[163,6.3904],[164,6.4745],[165,6.5097],[166,6.4884],[167,6.4136],[168,6.2566],[169,5.9718],[170,5.5563],[171,5.1008],[172,4.6856],[174,3.9758],[175,3.6684],[176,3.4294],[177,3.2814],[178,3.2458],[179,3.2909],[180,3.3858],[181,3.4975],[182,3.5863],[183,3.649],[185,3.74],[186,3.7918],[187,3.8346],[188,3.8644],[190,3.9004],[191,3.9101],[194,3.971],[195,3.997],[196,4.0107],[198,4.0598],[200,4.1235],[201,4.1254],[202,4.1128],[203,4.0737],[204,3.9756],[205,3.8075],[206,3.5603],[207,3.1838],[208,2.7986],[209,2.5937],[210,2.533],[211,2.4384],[212,2.2745],[213,2.0988],[214,1.9181],[216,1.4929],[217,1.3345],[218,1.2179],[219,1.1537],[220,1.1135],[221,1.0855],[222,1.0685],[223,1.0673],[224,1.0788],[225,1.114],[226,1.1622],[227,1.2212],[229,1.3875],[231,1.6067],[232,1.7314],[233,1.8636],[235,2.158],[236,2.3193],[237,2.5052],[240,3.0921],[241,3.3197],[242,3.5646],[243,3.7944],[244,3.9874],[245,4.1588],[246,4.3093],[247,4.4233],[248,4.5032],[249,4.5443],[250,4.5544],[251,4.536],[252,4.5048],[255,4.3687],[259,4.2399],[260,4.2171],[261,4.2145],[262,4.203],[263,4.2102],[264,4.2089],[266,4.2288],[268,4.2821],[269,4.302],[273,4.4555],[274,4.5112],[276,4.6086],[277,4.6457],[280,4.8117],[281,4.8496],[283,4.9493],[286,5.0796],[290,5.2756],[292,5.3479],[293,5.3705],[294,5.3741],[295,5.363],[296,5.3305],[299,5.1737],[301,5.0917],[304,4.9999],[305,5.0002],[306,0]]
       }
       arm_l_pv_ik_ctrl.Rotation.X {
        type=MovieSceneScriptingFloatChannel
        default=0
        thinned_from=307
        keys [[0,14.1351],[9,14.1791],[13,14.0656],[17,13.8209],[18,13.7954],[19,14.0172],[21,14.6994],[23,15.4591],[24,15.7549],[25,15.9814],[26,16.1516],[27,16.2627],[28,16.316],[36,16.2409],[46,15.8195],[55,15.166],[59,14.7984],[62,14.5543],[64,14.3129],[65,14.056],[67,13.2595],[68,12.8053],[70,11.6564],[71,11.0359],[72,10.5581],[73,10.4026],[74,10.3325],[75,10.312],[76,10.3327],[77,10.3793],[78,10.4504],[79,10.5559],[80,10.7021],[82,11.0411],[84,11.3074],[86,11.4864],[89,11.6515],[91,11.7018],[99,11.8086],[100,11.8371],[101,11.9368],[102,12.0863],[103,12.1973],[105,12.317],[107,12.3707],[111,12.412],[114,12.3585],[116,12.2951],[118,12.1855],[119,12.1711],[120,12.1871],[121,12.2295],[128,12.8431],[131,13.0043],[133,13.0498],[144,13.0578],[145,13.0837],[152,13.0871],[153,13.0726],[154,13.096],[155,13.2128],[156,13.5074],[157,13.9377],[159,14.9447],[160,15.3756],[161,15.7083],[162,15.9511],[163,16.1015],[164,16.1686],[165,16.1567],[166,16.0582],[167,15.8772],[168,15.5873],[169,15.1405],[170,14.5271],[171,13.8498],[172,13.2008],[174,11.981],[175,11.4205],[176,10.9662],[177,10.6806],[178,10.6173],[179,10.7306],[180,10.9625],[181,11.239],[182,11.4847],[184,11.8489],[186,12.1536],[187,12.2521],[189,12.377],[192,12.4567],[196,12.4552],[198,12.4016],[199,12.3384],[200,12.1934],[201,11.9452],[202,11.6355],[203,11.2762],[204,10.8149],[205,10.2266],[206,9.5825],[207,8.8691],[208,8.1335],[209,7.4579],[210,6.8291],[211,6.2838],[212,5.8093],[213,5.3912],[214,5.0706],[215,4.8331],[216,4.6635],[217,4.5415],[218,4.448],[219,4.4031],[220,4.3889],[221,4.4034],[222,4.4478],[223,4.5273],[224,4.6279],[225,4.7855],[226,4.9777],[227,5.1993],[228,5.4401],[229,5.7164],[230,6.0155],[232,6.6955],[233,7.0879],[234,7.5106],[236,8.4286],[239,9.9188],[240,10.4337],[242,11.5048],[243,12.0064],[244,12.4534],[245,12.8579],[246,13.2282],[247,13.5555],[248,13.829],[249,14.0568],[250,14.228],[251,14.3348],[252,14.4107],[254,14.4465],[256,14.3993],[257,14.3459],[268,13.2592],[275,12.773],[278,12.6469],[281,12.6006],[284,12.6635],[286,12.7793],[288,12.9743],[294,13.7816],[296,13.9585],[299,14.0918],[303,14.1391],[305,14.1398],[306,0]]
       }
       arm_l_pv_ik_ctrl.Rotation.Y {
        type=MovieSceneScriptingFloatChannel
        default=0
        thinned_from=307
        keys [[0,6.7894],[1,6.7952],[4,6.8844],[5,6.9422],[6,6.9538],[8,7.0056],[9,7.0203],[11,6.9707],[12,6.9111],[13,6.7913],[14,6.6431],[15,6.4531],[16,6.2093],[17,5.9207],[18,5.5547],[19,5.1672],[20,4.7531],[21,4.3006],[23,3.324],[24,2.8619],[25,2.4504],[26,2.0851],[27,1.764],[28,1.4848],[29,1.2635],[30,1.1023],[31,0.9979],[32,0.9262],[33,0.9022],[34,0.8979],[35,0.9357],[36,1.0071],[37,1.1062],[38,1.2297],[40,1.5882],[41,1.7808],[42,2.0405],[43,2.3081],[44,2.5886],[45,2.8524],[46,3.1073],[48,3.6275],[49,3.8768],[50,4.0917],[51,4.2872],[52,4.45],[53,4.6402],[54,4.7932],[55,4.9569],[58,5.3609],[62,6.0528],[63,6.1748],[64,6.2437],[65,6.3372],[66,6.5584],[67,6.968],[68,7.4764],[69,8.1873],[70,9.0997],[72,11.0655],[73,12.0812],[74,13.2121],[75,14.1971],[76,14.8048],[77,15.1258],[78,15.2223],[79,15.0437],[81,14.3732],[82,13.9954],[83,13.6528],[84,13.3793],[87,12.7178],[89,12.3456],[90,12.1887],[92,12.0102],[94,11.9338],[96,11.9172],[97,11.9374],[98,12.0344],[99,12.2562],[100,12.443],[101,12.529],[104,12.5935],[106,12.5707],[107,12.532],[108,12.4473],[110,12.1935],[111,12.086],[112,12.0061],[113,11.9534],[117,11.8772],[118,11.8802],[120,11.9864],[125,12.3401],[129,12.5484],[130,12.5648],[133,12.4592],[142,12.3144],[143,12.2846],[145,12.188],[146,12.1149],[150,11.9443],[151,11.8612],[153,11.6439],[154,11.582],[155,11.6064],[157,11.7916],[158,11.8638],[159,11.8659],[160,11.7698],[161,11.6087],[165,10.8703],[168,10.3036],[169,10.0815],[171,9.5228],[172,9.3023],[173,9.1535],[174,9.0278],[175,8.9541],[176,8.9291],[179,8.938],[180,8.8943],[183,8.6618],[185,8.4803],[191,8.1038],[193,8.0436],[196,8.0416],[198,8.0801],[199,8.161],[200,8.3734],[201,8.6566],[202,8.9922],[203,9.3514],[204,9.6793],[205,9.9493],[206,10.1456],[207,10.2532],[208,10.2985],[209,10.3734],[210,10.5221],[211,10.5331],[212,10.376],[213,10.1335],[214,9.808],[215,9.3635],[216,8.8757],[217,8.5258],[218,8.3283],[219,8.2417],[226,7.9611],[227,7.9053],[228,7.8699],[229,7.8094],[230,7.7755],[232,7.6406],[234,7.4212],[235,7.2634],[236,7.088],[237,6.8739],[238,6.6303],[240,6.096],[241,5.8394],[242,5.6041],[243,5.3864],[244,5.187],[245,5.0156],[247,4.7042],[248,4.5643],[250,4.2652],[251,4.141],[252,4.0345],[253,3.954],[254,3.9185],[255,3.921],[256,3.9648],[257,4.047],[258,4.1592],[261,4.5811],[262,4.7094],[263,4.8629],[266,5.2809],[268,5.5897],[269,5.7204],[273,6.3005],[274,6.4638],[276,6.7623],[277,6.8896],[280,7.3368],[283,7.6786],[286,7.9563],[288,8.096],[290,8.1416],[291,8.1325],[292,8.0874],[293,8.0073],[295,7.7849],[296,7.6514],[298,7.3474],[299,7.2104],[300,7.0996],[301,7.0107],[302,6.9384],[304,6.8322],[305,6.8168],[306,0]]
       }
       arm_l_pv_ik_ctrl.Rotation.Z {
        type=MovieSceneScriptingFloatChannel
        default=0
        thinned_from=307
        keys [[0,-19.5405],[4,-19.608],[6,-19.5773],[7,-19.5311],[8,-19.4264],[9,-19.2772],[10,-19.0884],[11,-18.8434],[12,-18.5341],[13,-18.1295],[14,-17.586],[15,-16.9142],[16,-16.1069],[17,-15.1081],[18,-13.874],[19,-12.3928],[20,-10.7234],[21,-8.8429],[22,-6.8122],[23,-4.9907],[24,-3.6284],[25,-2.5276],[26,-1.5731],[27,-0.7745],[28,-0.1235],[29,0.4155],[30,0.8586],[31,1.2106],[32,1.5067],[33,1.7435],[34,1.9455],[35,2.11],[36,2.2624],[37,2.4284],[38,2.5675],[39,2.6781],[40,2.8099],[41,2.8773],[42,2.9895],[43,3.0397],[44,3.101],[45,3.154],[47,3.2149],[48,3.2402],[49,3.2572],[50,3.2626],[52,3.2243],[53,3.1682],[54,3.0956],[55,3.0146],[56,2.9128],[57,2.7837],[58,2.6386],[59,2.4533],[60,2.2271],[61,1.961],[62,1.6523],[63,1.3232],[64,0.9363],[65,0.4292],[66,-0.3005],[67,-1.4945],[68,-3.2096],[69,-5.1424],[70,-7.0228],[71,-8.9437],[72,-10.7709],[73,-12.3259],[74,-13.7043],[75,-14.8426],[76,-15.6914],[77,-16.3748],[78,-16.9739],[79,-17.4987],[80,-17.9681],[81,-18.384],[82,-18.7476],[84,-19.3514],[86,-19.8527],[89,-20.4386],[100,-21.9673],[107,-23.1711],[109,-23.3881],[111,-23.4345],[114,-23.2823],[119,-22.8133],[121,-22.7203],[124,-22.7051],[126,-22.7675],[130,-23.0749],[139,-23.0852],[142,-22.8585],[144,-22.6113],[146,-22.2305],[148,-21.7176],[149,-21.4009],[150,-21.0404],[151,-20.548],[152,-19.8432],[153,-19.0353],[154,-18.0151],[155,-16.7169],[156,-15.0494],[157,-13.2258],[158,-11.6822],[159,-10.5135],[160,-9.8524],[161,-9.4147],[162,-9.1198],[163,-8.9357],[164,-8.8309],[165,-8.79],[166,-8.8231],[167,-8.9823],[168,-9.5985],[169,-10.6311],[170,-11.9053],[171,-13.3553],[172,-14.9087],[173,-16.4034],[175,-19.2704],[176,-20.7662],[178,-23.9488],[179,-25.4807],[180,-26.8306],[181,-28.0132],[182,-28.9883],[183,-29.7457],[184,-30.2271],[185,-30.3899],[186,-30.2894],[187,-29.9705],[188,-29.5354],[192,-27.6012],[193,-27.2047],[194,-26.8712],[195,-26.6146],[196,-26.4484],[197,-26.3813],[198,-26.4145],[199,-26.559],[200,-26.8472],[201,-27.3803],[202,-28.2038],[203,-29.3738],[204,-30.8728],[205,-32.6375],[206,-34.6429],[207,-36.826],[208,-39.2035],[209,-41.8216],[211,-47.2051],[212,-49.6779],[213,-51.9045],[214,-53.87],[215,-55.5188],[216,-56.87],[217,-57.9208],[218,-58.7375],[219,-59.3808],[221,-60.3269],[224,-61.1611],[227,-61.5317],[231,-61.0752],[233,-60.4211],[235,-59.3481],[237,-57.9116],[238,-57.0034],[239,-55.9336],[240,-54.6319],[241,-52.9978],[242,-51.0039],[243,-48.6708],[244,-45.9647],[245,-42.8108],[247,-36.0465],[248,-33.0043],[249,-30.1858],[250,-27.5],[251,-25.0783],[252,-23.0328],[253,-21.2804],[254,-19.6756],[255,-18.2135],[256,-16.9043],[257,-15.7465],[258,-14.7422],[259,-13.8716],[260,-13.1151],[261,-12.4598],[262,-11.8919],[263,-11.3991],[264,-10.9845],[265,-10.6308],[266,-10.3302],[267,-10.0768],[268,-9.867],[269,-9.7003],[270,-9.5686],[271,-9.4676],[272,-9.4143],[273,-9.3989],[274,-9.4029],[275,-9.4583],[276,-9.5489],[277,-9.6818],[278,-9.8694],[279,-10.1377],[280,-10.445],[281,-10.8057],[282,-11.2045],[283,-11.652],[284,-12.1456],[285,-12.6742],[287,-13.8158],[290,-15.573],[293,-17.2035],[294,-17.7009],[295,-18.1239],[296,-18.4724],[297,-18.7384],[298,-18.9537],[300,-19.27],[301,-19.3729],[303,-19.4958],[305,-19.5217],[306,0]]
       }
       arm_l_pv_ik_ctrl.Scale.X {type=MovieSceneScriptingFloatChannel, default=1, const [1,307]}
       arm_l_pv_ik_ctrl.Scale.Y {type=MovieSceneScriptingFloatChannel, default=1, const [1,307]}
       arm_l_pv_ik_ctrl.Scale.Z {type=MovieSceneScriptingFloatChannel, default=1, const [1,307]}
       upperarm_r_fk_ctrl.Location.X {type=MovieSceneScriptingFloatChannel, default=0, thinned_from=307, keys [[0,0],[306,0]]}
       upperarm_r_fk_ctrl.Location.Y {type=MovieSceneScriptingFloatChannel, default=0, thinned_from=307, keys [[0,0],[306,0]]}
       upperarm_r_fk_ctrl.Location.Z {type=MovieSceneScriptingFloatChannel, default=0, thinned_from=307, keys [[0,0],[306,0]]}
       upperarm_r_fk_ctrl.Rotation.X {
        type=MovieSceneScriptingFloatChannel
        default=0
        thinned_from=307
        keys [[0,6.2968],[4,6.3102],[7,6.347],[8,6.3405],[10,6.2907],[11,6.2459],[12,6.1706],[14,5.9068],[15,5.7954],[16,5.6959],[18,5.4455],[19,5.1952],[20,4.7206],[21,3.9513],[22,2.9743],[23,2.0259],[24,1.2831],[25,0.6852],[26,0.1922],[27,-0.1971],[28,-0.4718],[29,-0.6614],[30,-0.7857],[31,-0.8493],[32,-0.8918],[33,-0.9147],[34,-0.9425],[35,-0.9883],[36,-1.0314],[37,-1.0956],[38,-1.1434],[39,-1.174],[40,-1.2336],[41,-1.2453],[42,-1.3023],[43,-1.3101],[44,-1.3299],[45,-1.346],[48,-1.3453],[50,-1.3372],[51,-1.3126],[52,-1.297],[53,-1.2455],[54,-1.1853],[55,-1.1177],[56,-1.0349],[57,-0.9235],[58,-0.7997],[59,-0.6553],[60,-0.4811],[61,-0.2829],[62,-0.0539],[63,-0.3704],[64,-1.9642],[65,-4.3407],[66,-6.9271],[67,-9.4985],[68,-12.5329],[69,-16.1457],[70,-20.5422],[71,-26.6496],[72,-34.5379],[73,-44.0028],[74,-54.9343],[75,-66.0846],[76,-76.4398],[77,-86.3456],[78,-95.8581],[79,-104.7169],[80,-112.7047],[81,-127.2906],[82,-261.9346],[83,-273.0137],[85,-273.5105],[86,-274.424],[87,-276.0234],[89,-280.3985],[90,-281.6039],[92,-281.3704],[94,-280.6596],[95,-280.9964],[96,-282.1172],[97,-284.3824],[98,-288.5742],[100,-298.9963],[101,-301.9598],[102,-303.0836],[103,-301.6234],[105,-294.8743],[106,-292.7539],[107,-292.386],[111,-296.8041],[112,-296.9872],[114,-295.6733],[117,-289.3092],[118,-288.2559],[120,-288.0948],[121,-288.6304],[122,-289.9263],[124,-295.6115],[125,-299.1847],[126,-301.9489],[127,-303.7696],[129,-304.6898],[130,-302.9291],[132,-296.2027],[133,-293.6264],[134,-292.1562],[135,-291.5455],[136,-291.7466],[138,-293.6393],[139,-293.8361],[141,-290.321],[142,-287.7969],[143,-284.6084],[144,-280.4782],[145,-275.744],[146,-268.043],[147,-254.8392],[148,-227.6797],[149,-165.8577],[150,-119.6417],[151,-99.5631],[152,-85.8824],[153,-73.7748],[154,-62.5087],[155,-51.4219],[156,-39.3849],[157,-27.7154],[158,-18.1477],[159,-11.9633],[160,-7.7726],[161,-4.6187],[162,-2.6571],[163,-1.9127],[164,-1.7754],[165,-1.6765],[166,-1.5864],[167,-1.4406],[168,-0.9382],[169,-0.1008],[170,0.9386],[171,2.1326],[172,3.3693],[173,4.5534],[174,5.6817],[176,7.8224],[177,8.9274],[178,9.9642],[179,10.8489],[180,11.5132],[181,12.0173],[182,12.3658],[183,12.5797],[184,12.5848],[185,12.3443],[186,11.9001],[187,11.2811],[189,9.8992],[190,9.2778],[191,8.7262],[192,8.2471],[193,7.8242],[194,7.4876],[195,7.2258],[196,7.0542],[197,6.9512],[198,6.921],[199,6.9625],[200,7.0861],[201,7.3237],[202,7.7042],[203,8.2449],[204,8.9434],[205,9.8069],[206,10.8237],[207,11.9311],[208,13.2293],[209,14.8439],[211,18.4236],[212,20.102],[213,21.6699],[214,23.0613],[215,24.1893],[216,25.0715],[217,25.7309],[218,26.2312],[220,26.9763],[222,27.5136],[224,27.8829],[227,28.217],[235,28.7989],[237,28.7913],[238,28.6887],[239,28.4689],[240,28.114],[241,27.5929],[242,26.8496],[243,25.8398],[244,24.5188],[245,22.9041],[246,21.1028],[247,19.2517],[250,13.9117],[251,12.2889],[252,10.915],[253,9.7381],[254,8.6667],[255,7.7053],[256,6.8564],[257,6.123],[258,5.5],[259,4.9634],[260,4.4919],[261,4.0781],[262,3.7196],[263,3.4065],[264,3.1414],[265,2.9187],[266,2.7269],[267,2.5652],[268,2.429],[269,2.3184],[270,2.2361],[271,2.1752],[272,2.139],[274,2.0918],[275,2.0883],[277,2.1029],[278,2.1359],[279,2.2206],[280,2.3147],[281,2.4437],[282,2.6027],[283,2.798],[284,3.0226],[285,3.2664],[287,3.7973],[289,4.3463],[290,4.613],[293,5.3825],[294,5.6273],[295,5.834],[296,5.9886],[297,6.0851],[298,6.1531],[300,6.242],[301,6.2695],[302,6.2726],[303,6.2973],[305,6.2974],[306,0]]
       }
       upperarm_r_fk_ctrl.Rotation.Y {
        type=MovieSceneScriptingFloatChannel
        default=0
        thinned_from=307
        keys [[0,6.3062],[2,6.3048],[4,6.2927],[6,6.2441],[8,6.1507],[10,6.1297],[12,6.0839],[13,6.0823],[14,6.1174],[15,6.207],[16,6.3551],[18,6.7348],[19,6.8826],[20,6.9477],[21,6.9205],[23,6.7757],[24,6.715],[26,6.5457],[27,6.4748],[28,6.4342],[32,6.3954],[34,6.3227],[35,6.2476],[37,6.0372],[38,5.9193],[40,5.6496],[42,5.4452],[43,5.3311],[45,5.1521],[47,5.0224],[52,4.795],[55,4.6071],[57,4.4225],[58,4.2999],[60,4.1286],[61,4.0838],[62,4.1035],[63,3.9144],[64,3.301],[65,2.6296],[66,2.3159],[67,2.5583],[68,3.2719],[69,4.3802],[70,6.4704],[71,11.0294],[72,17.7368],[73,26.0541],[74,36.0929],[75,46.4059],[76,55.8054],[77,64.4196],[78,72.027],[79,78.3798],[80,83.6709],[81,87.9104],[82,88.206],[83,85.8296],[84,84.5069],[85,84.103],[86,84.084],[87,84.4414],[88,85.0181],[89,85.357],[90,85.4181],[92,84.9035],[94,84.2366],[95,84.1344],[96,84.2812],[97,84.6927],[99,85.9206],[100,86.2549],[102,86.2994],[103,86.0283],[106,84.3808],[107,84.1191],[109,84.2841],[112,84.7924],[114,84.7381],[118,83.5826],[121,83.6333],[122,83.8883],[125,85.2531],[126,85.4923],[128,85.503],[130,85.0873],[133,83.6032],[135,83.1189],[137,83.3385],[138,83.5817],[139,83.6204],[144,82.8047],[145,83.1437],[147,84.5335],[148,85.9221],[149,86.2639],[150,82.7147],[151,76.3589],[152,68.8534],[154,53.0416],[155,44.8738],[156,34.9732],[157,24.5737],[158,16.3033],[159,12.4782],[160,10.8794],[161,9.1545],[162,7.5967],[163,6.2987],[164,5.3494],[165,4.7264],[166,4.3718],[167,4.2489],[168,4.4263],[169,4.8746],[170,5.515],[171,6.2432],[172,6.9499],[173,7.567],[175,8.7014],[176,9.2875],[177,9.8218],[178,10.2395],[179,10.5027],[180,10.6098],[181,10.6189],[184,10.5408],[186,10.341],[188,10.2484],[189,10.1703],[190,10.0552],[192,9.7592],[194,9.5124],[195,9.4223],[196,9.3675],[197,9.3487],[198,9.3727],[199,9.4649],[200,9.6445],[201,9.9043],[202,10.2239],[205,11.3659],[206,11.7981],[207,12.3551],[208,12.7897],[209,12.748],[210,12.3533],[211,12.0698],[213,11.8429],[214,11.7838],[215,11.8013],[217,12.045],[219,12.3982],[220,12.5128],[224,12.7777],[226,12.8601],[227,12.8523],[228,12.7889],[229,12.6726],[233,11.9896],[234,11.7771],[235,11.6487],[237,11.5305],[238,11.5791],[239,11.7517],[240,12.0752],[243,13.1627],[245,13.8459],[246,14.1643],[247,14.4077],[248,14.5113],[249,14.0928],[250,13.1997],[251,12.1403],[252,11.2376],[254,9.8234],[255,9.1407],[256,8.4853],[257,7.8747],[258,7.3268],[259,6.854],[260,6.4603],[261,6.1513],[262,5.9048],[263,5.709],[265,5.4217],[266,5.3106],[267,5.2309],[268,5.1784],[271,5.0918],[272,5.0822],[273,5.0857],[274,5.0687],[275,5.0711],[276,5.0598],[277,5.0676],[280,5.0536],[281,5.0784],[282,5.1203],[283,5.1813],[285,5.3405],[288,5.6087],[290,5.7427],[293,5.9],[296,6.0888],[298,6.1898],[300,6.2579],[303,6.2987],[305,6.3034],[306,0]]
       }
       upperarm_r_fk_ctrl.Rotation.Z {
        type=MovieSceneScriptingFloatChannel
        default=0
        thinned_from=307
        keys [[0,-31.6252],[10,-31.3857],[12,-31.2098],[13,-31.0439],[15,-30.4771],[17,-29.7862],[19,-29.0044],[20,-28.7801],[21,-28.7889],[22,-28.9365],[23,-29.1866],[26,-30.0829],[28,-30.5218],[41,-32.4427],[46,-33.26],[51,-33.756],[58,-33.8727],[64,-33.4175],[65,-33.2541],[66,-32.9749],[68,-32.233],[69,-32.0057],[70,-32.4612],[71,-33.4322],[73,-34.8897],[74,-35.7209],[75,-36.9409],[76,-38.6451],[77,-40.91],[78,-43.8824],[79,-47.2487],[80,-50.9614],[81,-62.5258],[82,-195.2373],[83,-205.3736],[84,-205.0987],[85,-205.4404],[86,-206.2892],[87,-207.8773],[89,-212.2791],[90,-213.4635],[91,-213.5509],[93,-212.4431],[94,-212.292],[95,-212.611],[96,-213.7494],[97,-216.0683],[98,-220.3418],[100,-230.885],[101,-233.8682],[102,-234.9898],[103,-233.4922],[105,-226.571],[106,-224.3644],[107,-223.9471],[109,-226.0638],[110,-227.5816],[111,-228.4525],[112,-228.6707],[114,-227.4254],[116,-223.0073],[117,-221.133],[118,-220.1037],[119,-219.805],[121,-220.676],[122,-222.0551],[123,-224.7343],[125,-231.6153],[126,-234.4243],[127,-236.249],[128,-236.9961],[129,-237.1148],[130,-235.2944],[132,-228.4127],[133,-225.7589],[134,-224.2425],[135,-223.6081],[136,-223.8194],[138,-225.7796],[139,-225.9859],[141,-222.5419],[142,-220.168],[143,-217.2597],[145,-209.5136],[146,-202.9215],[147,-191.3505],[148,-166.7283],[149,-108.42],[150,-66.9178],[151,-53.7902],[152,-48.5573],[153,-45.7956],[154,-44.3426],[156,-42.5182],[157,-41.5389],[160,-38.1364],[161,-37.0578],[162,-36.0941],[163,-35.2279],[164,-34.5132],[165,-33.9842],[167,-33.191],[168,-32.6918],[169,-32.0753],[170,-31.3639],[171,-30.5557],[173,-28.7236],[174,-27.7452],[175,-26.6953],[176,-25.4967],[177,-24.2123],[178,-23.0554],[179,-22.2663],[180,-21.9111],[181,-21.8637],[182,-22.0843],[183,-22.5377],[184,-23.2477],[185,-24.2308],[186,-25.4164],[188,-28.1364],[189,-29.4014],[190,-30.3766],[191,-31.0763],[192,-31.5877],[193,-31.9551],[194,-32.2036],[195,-32.3478],[196,-32.3897],[200,-32.246],[201,-32.1393],[202,-31.9066],[203,-31.5639],[204,-31.0879],[205,-30.4736],[206,-29.7374],[207,-28.8768],[211,-25.0043],[212,-24.075],[213,-23.2183],[214,-22.4481],[215,-21.789],[216,-21.2112],[217,-20.7415],[218,-20.3335],[219,-19.9845],[220,-19.7119],[221,-19.4964],[223,-19.1715],[226,-18.8332],[230,-18.527],[232,-18.2936],[234,-17.9198],[235,-17.7668],[236,-17.697],[239,-17.7664],[240,-17.7379],[241,-17.7707],[242,-17.9231],[243,-18.2266],[244,-18.7133],[245,-19.3683],[246,-20.1553],[247,-21.0352],[248,-21.9737],[249,-23.1138],[250,-24.3909],[251,-25.6175],[252,-26.6636],[254,-28.4624],[255,-29.284],[256,-30.0082],[257,-30.6287],[258,-31.1643],[259,-31.6259],[261,-32.3584],[263,-32.877],[267,-33.5739],[274,-34.2225],[280,-34.4412],[283,-34.327],[287,-33.8414],[296,-32.2276],[300,-31.7972],[305,-31.6409],[306,0]]
       }
       upperarm_r_fk_ctrl.Scale.X {type=MovieSceneScriptingFloatChannel, default=1, const [1,307]}
       upperarm_r_fk_ctrl.Scale.Y {type=MovieSceneScriptingFloatChannel, default=1, const [1,307]}
       upperarm_r_fk_ctrl.Scale.Z {type=MovieSceneScriptingFloatChannel, default=1, const [1,307]}
       lowerarm_r_fk_ctrl.Location.X {type=MovieSceneScriptingFloatChannel, default=0, thinned_from=307, keys [[0,0],[306,0]]}
       lowerarm_r_fk_ctrl.Location.Y {type=MovieSceneScriptingFloatChannel, default=0, thinned_from=307, keys [[0,0],[306,0]]}
       lowerarm_r_fk_ctrl.Location.Z {type=MovieSceneScriptingFloatChannel, default=0, thinned_from=307, keys [[0,0],[306,0]]}
       lowerarm_r_fk_ctrl.Rotation.X {
        type=MovieSceneScriptingFloatChannel
        default=0
        thinned_from=307
        keys [[0,-0.1193],[1,-0.118],[2,-0.1161],[3,-0.1148],[4,-0.1148],[5,-0.1158],[6,-0.117],[7,-0.1187],[8,-0.1212],[9,-0.1235],[10,-0.1249],[11,-0.1242],[12,-0.1225],[13,-0.1196],[14,-0.1149],[15,-0.1105],[16,-0.1078],[17,-0.1065],[19,-0.1065],[20,-0.1086],[21,-0.1167],[22,-0.1326],[23,-0.1505],[24,-0.1652],[25,-0.1784],[26,-0.19],[27,-0.1991],[28,-0.2044],[29,-0.205],[30,-0.203],[31,-0.1998],[32,-0.1954],[33,-0.1906],[36,-0.1777],[37,-0.1728],[39,-0.1641],[40,-0.1601],[41,-0.1568],[42,-0.1531],[44,-0.1472],[47,-0.1391],[48,-0.1371],[50,-0.1323],[53,-0.1275],[55,-0.1248],[58,-0.1214],[60,-0.1198],[61,-0.1194],[63,-0.1194],[64,-0.1245],[65,-0.139],[66,-0.1692],[67,-0.2261],[68,-0.3238],[69,-0.466],[70,-0.6687],[71,-1.0144],[72,-1.4777],[73,-2.0371],[75,-3.3521],[76,-3.9242],[77,-4.4429],[78,-4.9037],[79,-5.3131],[80,-5.622],[81,-5.7783],[82,-5.9155],[83,-6.031],[84,-6.0799],[85,-6.084],[87,-5.9155],[88,-5.8535],[89,-5.8276],[90,-5.8256],[91,-5.8554],[93,-5.9784],[94,-6.0175],[95,-6.028],[96,-6.015],[97,-5.9648],[98,-5.8806],[99,-5.8115],[100,-5.7643],[101,-5.745],[102,-5.7514],[103,-5.8165],[105,-5.9791],[106,-6.0352],[107,-6.073],[108,-6.0895],[109,-6.0772],[110,-6.0288],[112,-5.9035],[114,-5.825],[115,-5.8072],[116,-5.8265],[117,-5.8738],[119,-5.9984],[120,-6.0514],[121,-6.0813],[122,-6.0876],[123,-6.0502],[124,-5.9872],[126,-5.8454],[127,-5.7929],[128,-5.7623],[129,-5.7595],[130,-5.7842],[131,-5.8445],[133,-5.999],[134,-6.0547],[136,-6.122],[137,-6.1403],[140,-6.0169],[141,-6.0259],[144,-6.1327],[146,-6.1448],[147,-6.139],[148,-6.091],[149,-5.9814],[150,-5.7924],[151,-5.4862],[152,-5.1594],[153,-4.7963],[154,-4.3457],[155,-3.7563],[156,-3.0265],[157,-2.2505],[158,-1.5359],[159,-0.9632],[160,-0.5536],[161,-0.2421],[162,-0.0242],[163,0.0826],[164,0.1359],[165,0.1577],[166,0.1627],[167,0.1576],[168,0.1472],[169,0.1336],[170,0.1099],[171,0.075],[172,0.0326],[173,-0.0105],[174,-0.0572],[175,-0.1054],[176,-0.1582],[177,-0.2153],[178,-0.2694],[179,-0.3198],[180,-0.3672],[181,-0.4099],[182,-0.4473],[183,-0.4766],[184,-0.492],[185,-0.4862],[186,-0.4614],[187,-0.417],[188,-0.3709],[190,-0.2776],[191,-0.2357],[192,-0.1993],[193,-0.1652],[194,-0.1368],[195,-0.1132],[196,-0.0966],[197,-0.0888],[198,-0.0882],[199,-0.0921],[200,-0.0997],[201,-0.1108],[202,-0.1246],[203,-0.1412],[204,-0.16],[205,-0.1844],[207,-0.2426],[208,-0.2734],[209,-0.3032],[210,-0.3312],[212,-0.3831],[213,-0.4082],[214,-0.4313],[215,-0.4514],[216,-0.4698],[218,-0.502],[220,-0.531],[221,-0.5422],[223,-0.559],[224,-0.5654],[227,-0.5791],[229,-0.5844],[233,-0.5885],[235,-0.5893],[236,-0.5918],[237,-0.5983],[238,-0.6079],[241,-0.6417],[242,-0.6508],[243,-0.6563],[244,-0.6556],[245,-0.6378],[246,-0.6068],[248,-0.5307],[249,-0.4885],[251,-0.3965],[252,-0.3549],[253,-0.3166],[254,-0.2801],[255,-0.2468],[256,-0.2179],[257,-0.194],[258,-0.1739],[259,-0.1568],[260,-0.1423],[261,-0.1299],[262,-0.119],[263,-0.11],[264,-0.1022],[265,-0.0948],[266,-0.0881],[267,-0.0821],[268,-0.0768],[269,-0.0721],[270,-0.0678],[271,-0.064],[272,-0.0605],[273,-0.0574],[276,-0.0493],[279,-0.0418],[281,-0.0375],[283,-0.034],[284,-0.0325],[285,-0.0315],[286,-0.0308],[287,-0.0312],[288,-0.0337],[289,-0.0376],[290,-0.0426],[291,-0.0481],[292,-0.0549],[294,-0.0703],[295,-0.0774],[297,-0.0907],[298,-0.0973],[299,-0.1031],[300,-0.1076],[301,-0.1112],[302,-0.114],[303,-0.1163],[304,-0.1181],[305,-0.1189],[306,0]]
       }
       lowerarm_r_fk_ctrl.Rotation.Y {
        type=MovieSceneScriptingFloatChannel
        default=0
        thinned_from=307
        keys [[0,-0.3892],[1,-0.3867],[2,-0.3813],[3,-0.377],[4,-0.377],[5,-0.3796],[7,-0.3882],[9,-0.402],[10,-0.4053],[11,-0.4039],[12,-0.3996],[13,-0.3909],[14,-0.3773],[15,-0.365],[16,-0.3574],[17,-0.3537],[19,-0.3537],[20,-0.36],[21,-0.3829],[22,-0.4268],[23,-0.4745],[24,-0.5127],[25,-0.5455],[26,-0.5744],[27,-0.5964],[28,-0.6094],[29,-0.6107],[30,-0.6062],[31,-0.598],[34,-0.5647],[39,-0.5101],[41,-0.4903],[43,-0.4735],[47,-0.4446],[49,-0.432],[51,-0.4215],[54,-0.4089],[58,-0.396],[60,-0.3915],[61,-0.3902],[63,-0.3902],[64,-0.4043],[65,-0.4439],[66,-0.5225],[67,-0.6605],[68,-0.8704],[69,-1.1344],[70,-1.4506],[72,-2.3168],[73,-2.5483],[74,-2.1787],[75,-1.0449],[76,0.7787],[77,3.4934],[78,6.9309],[79,10.9139],[80,14.5893],[81,17.5474],[82,19.9285],[83,20.3237],[85,20.3431],[90,20.2552],[96,20.3188],[102,20.2332],[109,20.341],[116,20.2561],[122,20.345],[130,20.2429],[137,20.3645],[147,20.3452],[148,20.2932],[149,19.9083],[150,18.2741],[151,16.2261],[152,13.9524],[153,11.8531],[154,9.5032],[155,6.8156],[156,3.7707],[157,0.6009],[158,-2.4457],[159,-5.186],[160,-7.5179],[161,-9.9167],[162,-11.6182],[163,-11.5309],[164,-11.2815],[165,-10.8372],[166,-10.1569],[167,-9.293],[168,-8.3254],[169,-7.2451],[170,-6.124],[171,-4.8253],[172,-3.7469],[173,-3.8474],[174,-4.0098],[177,-4.4417],[178,-4.5636],[179,-4.6661],[180,-4.5339],[181,-4.2786],[182,-3.9732],[183,-3.2911],[184,-2.4579],[185,-1.7116],[186,-1.516],[187,-2.0284],[188,-2.6057],[189,-3.2067],[190,-3.8302],[191,-4.3964],[192,-4.8497],[193,-5.1978],[194,-5.4839],[195,-5.7931],[196,-6.0753],[197,-6.3065],[198,-6.514],[199,-6.6978],[200,-6.8487],[201,-7.0189],[202,-7.2479],[203,-7.4148],[204,-7.5007],[205,-7.0418],[206,-6.0642],[207,-5.1069],[208,-3.9344],[209,-2.6222],[210,-1.2527],[211,-0.1818],[212,0.2127],[213,0.1582],[214,-0.1542],[215,-0.4352],[216,-0.687],[217,-0.8921],[218,-1.0733],[219,-1.2107],[220,-1.2425],[221,-1.2604],[223,-1.2864],[227,-1.3181],[231,-1.3304],[235,-1.3337],[236,-1.3373],[237,-1.3473],[241,-1.4115],[242,-1.4248],[243,-1.4329],[244,-1.4318],[245,-1.4059],[246,-1.3601],[247,-1.3027],[248,-1.2417],[249,-1.1725],[250,-1.0932],[251,-1.0107],[252,-0.9318],[254,-0.7799],[255,-0.7071],[256,-0.6411],[257,-0.5842],[258,-0.5345],[259,-0.4911],[260,-0.4528],[261,-0.419],[262,-0.3895],[263,-0.3641],[264,-0.3411],[265,-0.32],[266,-0.2997],[267,-0.2814],[268,-0.2654],[269,-0.251],[270,-0.2378],[271,-0.2256],[272,-0.2146],[273,-0.2043],[277,-0.1696],[278,-0.1612],[280,-0.1454],[281,-0.1384],[282,-0.1317],[283,-0.126],[284,-0.1211],[285,-0.1175],[286,-0.1156],[287,-0.1167],[288,-0.1251],[289,-0.139],[290,-0.1557],[291,-0.1743],[292,-0.1962],[294,-0.2452],[295,-0.2672],[297,-0.308],[298,-0.327],[299,-0.3436],[300,-0.3567],[301,-0.367],[302,-0.3752],[303,-0.3819],[304,-0.3864],[305,-0.3889],[306,0]]
       }
       lowerarm_r_fk_ctrl.Rotation.Z {
        type=MovieSceneScriptingFloatChannel
        default=0
        thinned_from=307
        keys [[0,5.2493],[1,5.213],[2,5.1353],[3,5.0783],[4,5.0783],[5,5.1146],[7,5.2311],[9,5.4177],[10,5.4644],[11,5.4462],[12,5.3866],[13,5.27],[14,5.0861],[15,4.9152],[16,4.8142],[17,4.7626],[19,4.7626],[20,4.8466],[21,5.1577],[22,5.7553],[23,6.413],[24,6.9367],[25,7.3891],[26,7.7871],[27,8.0923],[28,8.2704],[29,8.2916],[30,8.228],[31,8.1135],[34,7.652],[39,6.8963],[41,6.6258],[43,6.397],[47,6.0021],[49,5.8256],[52,5.6223],[56,5.4196],[59,5.3041],[61,5.2577],[63,5.2577],[64,5.4513],[65,5.9908],[66,7.0699],[67,8.9834],[68,11.9498],[69,15.794],[70,20.6123],[71,27.7178],[72,35.951],[73,44.7467],[74,54.1501],[75,63.0563],[76,70.5333],[77,77.2233],[78,83.1004],[79,88.2134],[80,91.9195],[81,93.6164],[83,96.4367],[84,97.0516],[85,97.097],[87,94.9995],[88,94.2329],[89,93.9141],[90,93.8825],[91,94.2564],[93,95.7805],[94,96.2607],[95,96.4044],[96,96.2405],[97,95.6023],[98,94.5612],[99,93.7176],[100,93.1369],[101,92.8934],[102,92.9716],[103,93.7732],[105,95.7827],[106,96.4955],[107,96.959],[108,97.1705],[109,97.0118],[112,94.8515],[114,93.885],[115,93.6566],[116,93.8964],[120,96.6873],[121,97.0657],[122,97.1456],[123,96.6777],[126,94.1324],[127,93.4838],[128,93.1124],[129,93.0763],[130,93.3793],[131,94.1153],[133,96.035],[134,96.7428],[136,97.5667],[137,97.8017],[140,96.2684],[141,96.3782],[144,97.7036],[146,97.8655],[147,97.7897],[148,97.1925],[149,95.8687],[150,93.7108],[151,90.0889],[152,86.1562],[153,81.7038],[154,76.0945],[155,68.6055],[156,58.9802],[157,48.0523],[158,36.9051],[159,26.6658],[160,18.062],[161,10.2662],[162,3.4776],[163,-1.4072],[164,-4.8175],[165,-6.9608],[166,-8.2365],[167,-8.747],[168,-8.6484],[169,-8.1491],[170,-6.5684],[171,-4.3131],[172,-1.8136],[173,0.4902],[174,2.6581],[175,4.6453],[176,6.613],[177,8.562],[178,10.2736],[179,11.7721],[180,13.1052],[181,14.2643],[182,15.224],[183,15.9888],[184,16.3661],[185,16.2727],[186,15.6526],[187,14.4707],[188,13.1824],[189,11.8611],[190,10.5015],[191,9.2084],[192,8.0466],[193,6.9499],[194,5.9596],[195,5.128],[196,4.5442],[197,4.2762],[198,4.2786],[199,4.4741],[200,4.8044],[201,5.259],[202,5.8214],[203,6.4451],[204,7.1121],[205,7.845],[206,8.6487],[207,9.4862],[210,12.1343],[211,12.9744],[212,13.7284],[213,14.3978],[214,14.9765],[215,15.4855],[216,15.9243],[219,17.0649],[220,17.4036],[221,17.6763],[223,18.0829],[226,18.4597],[228,18.6293],[233,18.7785],[235,18.7998],[236,18.8566],[237,19.0083],[241,20.0033],[242,20.208],[243,20.3337],[244,20.3167],[245,19.914],[246,19.2054],[247,18.325],[248,17.4005],[249,16.3619],[250,15.1855],[251,13.9759],[252,12.8325],[254,10.662],[255,9.6355],[256,8.7135],[257,7.9211],[258,7.2362],[259,6.6393],[260,6.1122],[261,5.6494],[262,5.2506],[263,4.9019],[264,4.5898],[265,4.305],[266,4.0294],[267,3.7857],[268,3.5692],[269,3.3753],[270,3.1995],[271,3.0328],[272,2.8842],[273,2.7492],[274,2.6277],[278,2.1699],[279,2.0624],[280,1.9597],[281,1.8651],[282,1.7785],[283,1.7015],[284,1.6358],[285,1.5877],[286,1.562],[287,1.5781],[288,1.6887],[289,1.8731],[290,2.0977],[291,2.3453],[292,2.6386],[294,3.2967],[295,3.5943],[297,4.144],[298,4.4031],[299,4.6269],[300,4.8046],[301,4.9433],[302,5.0571],[303,5.1461],[304,5.2065],[305,5.2421],[306,0]]
       }
       lowerarm_r_fk_ctrl.Scale.X {type=MovieSceneScriptingFloatChannel, default=1, const [1,307]}
       lowerarm_r_fk_ctrl.Scale.Y {type=MovieSceneScriptingFloatChannel, default=1, const [1,307]}
       lowerarm_r_fk_ctrl.Scale.Z {type=MovieSceneScriptingFloatChannel, default=1, const [1,307]}
       hand_r_fk_ctrl.Location.X {type=MovieSceneScriptingFloatChannel, default=0, thinned_from=307, keys [[0,0],[306,0]]}
       hand_r_fk_ctrl.Location.Y {type=MovieSceneScriptingFloatChannel, default=0, thinned_from=307, keys [[0,0],[306,0]]}
       hand_r_fk_ctrl.Location.Z {type=MovieSceneScriptingFloatChannel, default=0, thinned_from=307, keys [[0,0],[306,0]]}
       hand_r_fk_ctrl.Rotation.X {
        type=MovieSceneScriptingFloatChannel
        default=0
        thinned_from=307
        keys [[0,22.9734],[62,23.0136],[63,23.3107],[64,24.4945],[65,26.6842],[66,29.7265],[67,37.0248],[68,46.9595],[69,52.123],[70,52.7437],[71,48.0979],[73,34.76],[74,27.9839],[75,21.5653],[76,16.2109],[77,10.3235],[78,3.0611],[79,-6.5358],[80,-16.0109],[81,-23.1571],[82,-28.608],[83,-32.5758],[84,-35.8246],[85,-38.183],[86,-39.4622],[87,-40.5467],[88,-41.4255],[89,-40.5938],[90,-38.542],[91,-36.2134],[92,-34.18],[93,-33.021],[94,-33.1452],[95,-34.5818],[96,-38.2492],[97,-41.2177],[98,-42.3648],[99,-42.7877],[100,-42.8554],[101,-42.5608],[102,-41.0777],[103,-38.4675],[104,-36.6787],[105,-35.6895],[106,-35.1973],[107,-35.157],[108,-35.5841],[109,-36.5097],[110,-37.7472],[111,-39.1056],[112,-40.1254],[113,-40.4513],[114,-40.3029],[115,-39.4864],[116,-37.9782],[117,-36.2545],[118,-34.7177],[119,-34.3559],[120,-34.9521],[121,-36.9256],[122,-39.3061],[123,-41.3057],[124,-42.6643],[125,-43.2636],[126,-43.4273],[127,-42.8881],[128,-40.8583],[129,-38.5739],[130,-36.6386],[131,-35.2667],[132,-34.6773],[133,-35.2347],[134,-37.2948],[135,-40.3437],[136,-42.3113],[137,-43.4077],[138,-43.8649],[139,-43.867],[140,-43.2615],[141,-41.6808],[142,-39.3726],[144,-33.8596],[145,-31.3489],[146,-29.3658],[147,-27.6296],[148,-26.0378],[150,-23.0215],[151,-21.9187],[152,-20.1028],[153,-16.0906],[154,-10.3054],[155,-3.3103],[156,4.3625],[157,11.7825],[158,17.8973],[159,21.7773],[160,23.0838],[161,23.821],[162,24.4929],[163,25.0181],[164,25.3895],[165,25.6588],[166,25.8357],[167,25.9315],[168,25.9592],[169,25.9066],[170,25.7674],[175,24.6139],[178,23.7915],[180,23.3625],[183,22.9381],[186,22.6727],[194,22.5451],[199,22.6581],[201,22.7715],[202,22.7818],[203,22.1884],[204,20.9654],[205,19.1898],[206,16.9738],[207,14.4148],[208,11.5053],[209,9.6503],[210,10.6427],[211,13.0373],[212,16.358],[213,20.5511],[214,24.8715],[215,29.4428],[216,34.2273],[217,37.8767],[218,40.1353],[219,41.6956],[220,42.7141],[221,43.3393],[222,43.6616],[223,43.7543],[234,43.2235],[237,42.7941],[239,42.3645],[240,42.0728],[241,41.4988],[242,40.3447],[243,38.3124],[244,35.1521],[245,31.1367],[246,26.874],[247,22.9584],[248,19.8851],[249,17.1686],[250,14.5965],[251,12.6688],[252,11.8446],[253,12.2316],[254,13.4281],[255,15.0761],[256,16.8476],[257,18.5691],[258,20.1851],[259,21.499],[260,22.282],[261,22.5034],[265,23.1456],[269,23.5242],[279,23.9189],[284,23.9118],[299,23.0683],[305,22.9737],[306,0]]
       }
       hand_r_fk_ctrl.Rotation.Y {
        type=MovieSceneScriptingFloatChannel
        default=0
        thinned_from=307
        keys [[0,-1.6285],[8,-1.6285],[10,-1.6368],[11,-1.6496],[13,-1.7018],[14,-1.7427],[15,-1.7648],[16,-1.759],[22,-1.6432],[23,-1.6293],[35,-1.6265],[36,-1.6201],[39,-1.5864],[41,-1.5708],[44,-1.562],[46,-1.562],[47,-1.5732],[52,-1.6517],[54,-1.6795],[58,-1.7507],[60,-1.7919],[61,-1.8203],[62,-1.8536],[63,-1.9464],[64,-2.1187],[65,-2.2732],[66,-2.3405],[67,-1.9645],[68,-1.8192],[69,-2.4213],[70,-3.0527],[71,-3.5681],[72,-3.8863],[73,-3.8889],[74,-3.5942],[75,-2.9513],[76,-1.8892],[77,-0.7203],[78,0.2089],[79,0.4779],[80,0.4809],[81,0.6258],[82,1.1792],[83,2.1436],[84,3.3019],[85,4.542],[86,5.6478],[87,6.3312],[88,6.5083],[89,6.5813],[90,6.8036],[91,6.6016],[92,5.8942],[93,5.0399],[94,5.1613],[95,5.7387],[96,6.963],[97,7.9764],[98,8.4545],[99,8.8313],[100,8.8899],[101,8.3202],[102,7.4763],[103,6.6555],[104,5.9136],[105,5.2607],[106,4.5623],[107,4.3513],[108,4.5869],[109,5.2092],[111,6.8995],[112,7.4904],[113,7.6133],[114,7.3771],[115,6.8183],[116,6.3156],[117,5.913],[118,5.638],[119,4.9361],[120,4.1437],[121,3.8665],[122,3.9123],[123,4.3687],[124,5.2169],[125,6.2906],[126,7.2984],[127,7.8506],[128,7.8908],[129,7.7262],[130,7.3377],[131,6.508],[132,5.5225],[133,4.6211],[134,3.9866],[135,3.5575],[136,3.7594],[137,4.2432],[138,4.6247],[139,4.6647],[140,4.341],[141,3.6044],[142,2.7926],[143,2.2842],[144,2.0576],[145,2.174],[146,3.1309],[147,5.0396],[148,7.3772],[149,9.2267],[150,9.9461],[151,10.0822],[152,10.3221],[153,10.9982],[154,11.7976],[155,12.0409],[156,10.6632],[157,7.8985],[158,4.4938],[159,1.1216],[160,-2.6129],[161,-6.4549],[162,-9.8118],[163,-12.3042],[164,-13.9708],[165,-15.1212],[166,-15.8345],[167,-16.1847],[168,-16.2431],[169,-15.9548],[170,-15.2855],[171,-14.3501],[172,-13.255],[173,-12.1057],[174,-10.999],[175,-9.8208],[176,-8.4939],[177,-7.1202],[178,-5.7907],[179,-4.5993],[180,-3.6263],[181,-2.8105],[182,-2.0831],[183,-1.4459],[184,-0.8987],[185,-0.4393],[186,-0.0642],[187,0.2247],[188,0.4224],[189,0.5439],[190,0.6066],[191,0.6286],[192,0.6289],[193,0.6229],[194,0.6039],[195,0.5654],[196,0.4915],[197,0.3764],[198,0.2106],[199,-0.0046],[200,-0.2593],[201,-0.5777],[202,-1.0049],[203,-1.7012],[204,-2.6988],[205,-3.9718],[206,-5.4852],[207,-7.2111],[208,-9.1422],[209,-10.9927],[210,-12.3923],[211,-13.5864],[212,-14.6061],[213,-15.4062],[214,-16.0591],[216,-17.142],[217,-17.7412],[218,-18.2519],[219,-18.4312],[220,-18.2942],[221,-18.0805],[224,-17.308],[227,-16.3303],[228,-16.0353],[231,-15.3213],[233,-14.7029],[234,-14.3207],[235,-13.673],[236,-12.7264],[237,-11.6079],[238,-10.294],[239,-8.766],[240,-6.8672],[241,-4.5125],[242,-2.0767],[243,-0.0306],[244,1.2222],[245,1.8329],[246,2.0529],[247,2.036],[248,1.9644],[249,1.8388],[250,1.6623],[251,1.5607],[252,1.6576],[253,1.9504],[254,2.293],[255,2.6128],[256,2.7981],[257,2.7978],[258,2.5883],[259,2.1782],[260,1.584],[261,0.8232],[262,-0.0275],[263,-0.9103],[264,-1.7609],[265,-2.5161],[266,-3.135],[267,-3.6452],[268,-4.0737],[269,-4.4408],[270,-4.7677],[271,-5.0685],[272,-5.3362],[273,-5.5687],[274,-5.7707],[275,-5.9476],[276,-6.0988],[277,-6.2274],[278,-6.3318],[279,-6.4066],[280,-6.4589],[281,-6.49],[282,-6.5063],[283,-6.4841],[284,-6.3721],[285,-6.2027],[286,-5.9946],[287,-5.7684],[288,-5.508],[289,-5.2131],[290,-4.897],[291,-4.5668],[292,-4.1988],[293,-3.8101],[294,-3.4406],[295,-3.1198],[296,-2.8348],[297,-2.565],[298,-2.3221],[299,-2.1152],[300,-1.9562],[301,-1.8369],[302,-1.7485],[303,-1.6867],[304,-1.6499],[305,-1.6311],[306,0]]
       }
       hand_r_fk_ctrl.Rotation.Z {
        type=MovieSceneScriptingFloatChannel
        default=0
        thinned_from=307
        keys [[0,-12.3993],[8,-12.3993],[9,-12.3492],[10,-12.2521],[11,-12.055],[12,-11.7635],[13,-11.384],[14,-10.8779],[15,-10.5573],[16,-10.6397],[17,-10.8737],[22,-12.1929],[23,-12.3684],[24,-12.4004],[34,-12.4004],[35,-12.4148],[36,-12.4964],[39,-12.9458],[40,-13.0686],[42,-13.2057],[44,-13.2675],[46,-13.2675],[47,-13.1304],[48,-12.9386],[50,-12.4837],[54,-11.6786],[56,-11.2296],[59,-10.4703],[60,-10.1836],[61,-9.8252],[62,-9.357],[63,-8.7341],[64,-7.873],[65,-6.695],[66,-5.1907],[67,-2.0965],[68,2.4631],[69,6.0911],[70,8.1125],[71,7.8303],[72,6.8476],[73,5.9568],[74,4.9074],[75,3.0705],[76,0.3561],[77,-3.7441],[78,-8.7626],[79,-13.6382],[80,-17.3833],[81,-19.2965],[82,-20.5151],[83,-21.2224],[84,-21.2732],[85,-21.0142],[86,-19.2946],[87,-16.4584],[88,-13.3874],[89,-11.7907],[90,-11.5416],[91,-13.1402],[92,-15.9584],[93,-19.1271],[94,-20.8987],[95,-21.2947],[96,-20.5246],[97,-18.7091],[98,-15.9823],[99,-12.1644],[100,-8.5378],[101,-6.9404],[102,-7.1473],[103,-9.9644],[104,-13.6645],[105,-17.166],[106,-18.8026],[107,-19.5291],[108,-19.7512],[109,-19.5739],[110,-18.5787],[111,-16.222],[112,-13.4017],[113,-10.7995],[114,-9.1883],[115,-8.7162],[116,-10.0483],[117,-12.8024],[118,-16.0219],[119,-17.9118],[120,-18.2788],[121,-17.3939],[122,-15.3802],[123,-12.8486],[124,-10.886],[125,-9.7312],[126,-8.8693],[127,-8.6579],[128,-10.5381],[129,-13.1994],[130,-16.0072],[131,-19.0089],[132,-21.5306],[133,-22.836],[134,-22.8445],[135,-21.4279],[136,-18.1243],[137,-14.4897],[138,-12.532],[139,-11.3924],[140,-10.9418],[141,-11.5787],[142,-13.9653],[143,-18.4524],[144,-22.2948],[145,-24.1544],[146,-25.2707],[147,-25.6389],[148,-25.2633],[149,-24.2141],[150,-22.5252],[152,-18.4334],[154,-14.4047],[155,-12.207],[156,-9.7191],[157,-7.1078],[158,-4.6813],[159,-2.7122],[160,-1.1766],[161,0.0733],[162,0.8582],[163,0.9785],[164,0.5534],[165,-0.1471],[166,-1.0143],[167,-1.9508],[168,-2.8608],[169,-3.7847],[170,-4.7899],[171,-5.8199],[172,-6.8206],[173,-7.7361],[174,-8.5101],[175,-9.152],[176,-9.7084],[177,-10.1897],[178,-10.6056],[179,-10.9628],[180,-11.2683],[181,-11.5212],[182,-11.7219],[183,-11.8734],[184,-11.9776],[185,-12.0357],[186,-12.0522],[187,-12.0271],[188,-11.9558],[191,-11.6276],[192,-11.5377],[195,-11.3772],[197,-11.3265],[198,-11.3277],[199,-11.3518],[200,-11.4118],[201,-11.4969],[202,-11.6135],[203,-11.8225],[204,-12.1166],[205,-12.5017],[206,-12.9466],[208,-13.9258],[209,-14.2736],[210,-14.2801],[211,-14.1089],[212,-13.7791],[213,-13.2458],[214,-12.5982],[215,-11.8191],[216,-10.9333],[217,-10.2341],[218,-9.8097],[219,-9.4844],[220,-9.1946],[221,-8.9531],[222,-8.7556],[223,-8.5964],[226,-8.188],[227,-8.07],[228,-7.9786],[230,-7.8446],[234,-7.6244],[235,-7.5378],[237,-7.2939],[238,-7.1518],[239,-6.9927],[240,-6.8104],[241,-6.6476],[242,-6.6063],[243,-6.7919],[244,-7.2706],[245,-7.9827],[246,-8.7733],[247,-9.518],[248,-10.1371],[250,-11.2674],[251,-11.7367],[252,-12.0628],[253,-12.242],[254,-12.3274],[258,-12.3532],[259,-12.3783],[260,-12.4397],[264,-12.9065],[265,-12.9929],[267,-13.0821],[273,-13.1044],[277,-13.0097],[282,-12.758],[287,-12.3538],[290,-12.1955],[292,-12.1482],[295,-12.1659],[304,-12.3883],[305,-12.3977],[306,0]]
       }
       hand_r_fk_ctrl.Scale.X {type=MovieSceneScriptingFloatChannel, default=1, const [1,307]}
       hand_r_fk_ctrl.Scale.Y {type=MovieSceneScriptingFloatChannel, default=1, const [1,307]}
       hand_r_fk_ctrl.Scale.Z {type=MovieSceneScriptingFloatChannel, default=1, const [1,307]}
       thumb_01_r_ctrl.Location.X {type=MovieSceneScriptingFloatChannel, default=0, thinned_from=307, keys [[0,0],[306,0]]}
       thumb_01_r_ctrl.Location.Y {type=MovieSceneScriptingFloatChannel, default=0, thinned_from=307, keys [[0,0],[306,0]]}
       thumb_01_r_ctrl.Location.Z {type=MovieSceneScriptingFloatChannel, default=0, thinned_from=307, keys [[0,0],[306,0]]}
       thumb_01_r_ctrl.Rotation.X {
        type=MovieSceneScriptingFloatChannel
        default=0
        thinned_from=307
        keys [[0,20.2177],[2,20.1991],[4,20.2508],[5,20.2154],[6,20.2492],[7,20.233],[8,20.2595],[10,20.1982],[14,20.233],[15,20.1989],[17,20.2477],[19,20.1799],[21,20.2625],[23,20.2114],[31,20.2832],[32,20.2173],[36,20.2683],[37,20.217],[38,20.246],[39,20.2239],[40,20.2852],[43,20.1988],[45,20.2186],[46,20.2806],[47,20.2267],[48,20.2599],[49,20.2061],[59,20.2598],[60,20.2161],[62,20.2347],[63,19.9559],[64,19.0673],[65,17.7595],[66,16.2664],[67,14.4391],[68,11.308],[69,4.9935],[70,2.0052],[71,2.2612],[72,2.6317],[73,3.0968],[74,3.6176],[75,4.0262],[76,4.361],[78,4.6424],[79,4.7188],[80,4.3411],[81,-4.1124],[82,-14.9765],[83,-22.9603],[84,-23.2898],[130,-23.294],[135,-23.2693],[142,-23.3153],[143,-23.2571],[145,-23.2632],[146,-18.9508],[147,-7.4977],[148,3.1011],[149,9.1121],[150,9.8645],[151,11.8431],[153,17.8558],[154,19.8473],[155,20.2337],[157,20.21],[158,20.2506],[161,20.2024],[163,20.2416],[165,20.2061],[167,20.2455],[168,20.2252],[171,20.2418],[180,20.2177],[181,20.2486],[184,20.207],[186,20.2521],[188,20.2178],[190,20.2566],[191,20.2212],[193,20.2468],[194,20.2212],[195,20.2432],[200,20.2172],[201,20.2465],[203,20.2007],[205,20.2076],[206,20.2544],[207,20.2127],[208,20.2594],[209,20.2594],[210,20.2143],[211,20.2494],[214,20.2456],[215,20.2168],[220,20.2438],[224,20.22],[227,20.2482],[228,20.2047],[237,20.217],[238,20.2495],[249,20.2384],[250,20.2094],[256,20.2184],[257,20.1909],[274,20.2399],[275,20.2011],[277,20.246],[279,20.2056],[281,20.2331],[282,20.2068],[288,20.2498],[289,20.2187],[290,20.2448],[291,20.2223],[292,20.2492],[295,20.2072],[298,20.2424],[300,20.2057],[305,20.2457],[306,0]]
       }
       thumb_01_r_ctrl.Rotation.Y {
        type=MovieSceneScriptingFloatChannel
        default=0
        thinned_from=307
        keys [[0,6.8763],[1,6.8763],[2,6.9091],[4,6.885],[5,6.9095],[6,6.9179],[8,6.8972],[9,6.9007],[10,6.9334],[12,6.9334],[13,6.9007],[14,6.9094],[15,6.9011],[18,6.9204],[19,6.886],[20,6.9292],[21,6.8667],[22,6.9204],[23,6.8682],[24,6.8593],[25,6.9038],[26,6.8947],[27,6.939],[31,6.902],[32,6.9113],[33,6.9019],[34,6.872],[36,6.872],[37,6.9309],[38,6.8807],[39,6.8894],[40,6.8794],[43,6.9135],[44,6.8715],[45,6.9198],[46,6.9091],[47,6.8776],[48,6.9173],[49,6.8859],[50,6.8859],[51,6.9042],[57,6.9063],[59,6.8955],[60,6.9072],[62,6.9043],[63,6.9262],[64,7.003],[65,7.1274],[67,7.4265],[68,7.7287],[69,8.1648],[70,8.7994],[71,9.6672],[72,10.8074],[73,12.1377],[74,13.7646],[75,15.2277],[76,16.1662],[77,16.7573],[79,17.3929],[80,17.2841],[81,12.7409],[82,6.8564],[83,2.5369],[84,2.8392],[85,3.6989],[86,4.2623],[87,4.6391],[89,4.7995],[90,4.7995],[91,4.0734],[92,3.3239],[93,2.9483],[94,2.7583],[95,2.6774],[96,2.7583],[97,3.3239],[98,4.3146],[99,4.6943],[100,4.7479],[101,4.7883],[102,4.6943],[103,3.9656],[104,3.2481],[105,2.6779],[106,2.4743],[107,2.3928],[108,2.352],[109,2.6779],[110,3.5322],[111,4.1679],[112,4.5731],[113,4.7346],[114,4.775],[115,4.7974],[116,4.3147],[117,3.5464],[118,3.0612],[119,2.7533],[120,2.6237],[121,3.316],[122,4.3147],[123,4.5683],[124,4.7211],[125,4.7733],[126,4.7974],[127,4.7974],[128,4.1859],[129,3.5464],[130,3.0086],[131,2.7399],[133,2.6046],[134,2.7399],[135,3.4585],[136,4.0979],[137,4.5561],[138,4.6902],[139,4.7573],[140,4.8124],[141,4.5009],[142,3.8408],[143,3.2442],[144,2.7955],[145,2.6046],[146,4.4374],[147,9.2573],[148,14.0466],[149,15.8276],[150,14.9242],[151,13.2113],[152,10.9312],[153,8.6953],[154,7.1859],[155,6.8953],[157,6.9047],[158,6.8951],[161,6.9131],[162,6.8992],[163,6.9164],[164,6.8958],[165,6.9091],[166,6.8953],[167,6.9044],[168,6.8874],[172,6.901],[173,6.915],[174,6.8959],[176,6.8902],[177,6.9055],[178,6.894],[179,6.9228],[180,6.8952],[181,6.8881],[182,6.9291],[183,6.926],[184,6.8868],[185,6.8987],[186,6.8787],[187,6.9046],[190,6.8822],[191,6.8866],[192,6.9131],[193,6.9131],[194,6.8866],[196,6.9105],[197,6.8844],[199,6.9061],[200,6.9012],[201,6.9142],[202,6.9021],[203,6.9203],[204,6.8943],[205,6.8999],[206,6.8909],[207,6.9031],[209,6.8915],[210,6.9032],[214,6.8925],[215,6.9168],[216,6.9097],[217,6.8876],[218,6.893],[219,6.9133],[220,6.8856],[221,6.9066],[222,6.8862],[223,6.9129],[224,6.8866],[225,6.9033],[226,6.8871],[228,6.9179],[229,6.8933],[237,6.8933],[238,6.8853],[239,6.9158],[240,6.8903],[242,6.8967],[245,6.885],[247,6.8882],[249,6.9145],[250,6.9071],[251,6.9241],[252,6.8885],[255,6.9172],[256,6.8932],[257,6.9053],[267,6.8928],[268,6.8994],[269,6.8856],[272,6.9038],[273,6.8945],[285,6.8836],[286,6.9102],[287,6.8817],[288,6.8836],[289,6.9157],[291,6.8939],[293,6.9049],[294,6.8855],[296,6.9012],[297,6.8815],[302,6.911],[303,6.8857],[304,6.8907],[305,6.9151],[306,0]]
       }
       thumb_01_r_ctrl.Rotation.Z {
        type=MovieSceneScriptingFloatChannel
        default=0
        thinned_from=307
        keys [[0,-4.0207],[1,-4.0207],[2,-4.0368],[4,-4.1927],[5,-4.3438],[6,-4.5158],[7,-4.6509],[8,-4.7289],[9,-4.786],[10,-4.8022],[12,-4.8022],[13,-4.786],[14,-4.6509],[15,-4.4788],[17,-4.2115],[18,-4.0987],[19,-4.0472],[20,-3.9859],[21,-4.0723],[22,-4.0987],[23,-4.2729],[24,-4.3857],[25,-4.525],[26,-4.6379],[27,-4.7773],[31,-5.2292],[32,-5.3171],[33,-5.4302],[34,-5.4603],[36,-5.4603],[37,-5.3789],[38,-5.3538],[39,-5.2474],[40,-5.1661],[42,-4.9532],[43,-4.7405],[44,-4.6092],[45,-4.4466],[46,-4.3655],[49,-4.2092],[50,-4.2092],[51,-4.1724],[52,-4.1724],[53,-4.1505],[57,-4.1505],[59,-4.0993],[60,-4.056],[61,-4.0316],[62,-4.0316],[63,-4.0072],[65,-4.0077],[66,-3.9974],[67,-4.0157],[68,-3.8871],[69,-5.0163],[70,-6.7712],[71,-7.7428],[72,-7.4101],[73,-6.7474],[74,-4.9146],[75,-2.403],[76,-0.4809],[77,1.6946],[78,3.8838],[79,5.4597],[80,5.8435],[81,2.797],[82,-1.1651],[83,-4.0512],[84,-4.183],[85,-4.1874],[86,-4.2117],[87,-4.1907],[90,-4.1977],[91,-4.1708],[92,-4.2039],[93,-4.2203],[94,-4.1796],[96,-4.1796],[97,-4.2039],[98,-4.1812],[99,-4.1886],[100,-4.2052],[101,-4.2055],[102,-4.1886],[103,-4.1839],[104,-4.1953],[109,-4.1919],[110,-4.1977],[111,-4.1857],[114,-4.1893],[115,-4.2006],[116,-4.1785],[117,-4.1936],[118,-4.1717],[119,-4.1764],[120,-4.2079],[122,-4.1786],[123,-4.2042],[124,-4.2018],[125,-4.1712],[126,-4.2006],[127,-4.2006],[128,-4.2103],[129,-4.1936],[131,-4.2087],[133,-4.2017],[135,-4.2188],[136,-4.177],[139,-4.1982],[140,-4.1873],[141,-4.1984],[142,-4.1771],[143,-4.1939],[145,-4.2017],[146,-3.9934],[147,-3.4743],[148,-3.0254],[149,-2.6774],[150,-2.4162],[151,-1.9879],[152,-1.6043],[153,-1.4313],[154,-1.3974],[155,-1.3715],[156,-1.1197],[157,-0.7451],[158,-0.5323],[159,-0.6387],[160,-1.0582],[161,-1.8212],[162,-2.9763],[163,-4.4474],[164,-6.6495],[166,-12.3663],[167,-14.5553],[168,-15.4495],[169,-14.9496],[170,-13.7423],[171,-11.9747],[172,-10.08],[173,-8.2087],[174,-6.7063],[175,-5.2801],[176,-3.7348],[177,-2.3214],[178,-1.1108],[179,-0.3368],[180,-0.1647],[181,-0.7402],[182,-1.7424],[183,-2.8692],[184,-4.1781],[185,-5.682],[186,-6.945],[187,-7.6602],[188,-8.1444],[189,-8.4293],[190,-8.6431],[191,-8.7717],[192,-8.8042],[193,-8.8042],[194,-8.7717],[195,-8.6557],[196,-8.4453],[197,-8.2171],[198,-7.8807],[199,-7.1949],[200,-6.0036],[201,-4.8466],[202,-3.7857],[203,-2.7444],[204,-1.8952],[205,-1.203],[206,-0.6636],[207,-0.3734],[208,-0.1799],[209,-0.1799],[210,-0.2352],[211,-0.322],[212,-0.4613],[213,-0.6114],[214,-0.8115],[215,-1.1132],[216,-1.3136],[217,-1.4409],[218,-1.051],[219,0.4669],[220,1.9773],[221,3.4714],[222,4.3287],[223,4.3665],[224,4.4281],[226,4.5772],[227,4.6171],[229,4.6421],[237,4.6421],[238,4.6802],[239,5.2247],[240,5.7389],[241,5.933],[244,5.9306],[245,5.7778],[246,4.1718],[247,1.9656],[248,-0.6344],[249,-1.8705],[250,-2.764],[251,-3.3245],[252,-3.5209],[253,-3.0734],[254,-2.3944],[255,-1.7933],[256,-1.3115],[257,-0.926],[258,-0.5574],[259,-0.3429],[260,-0.2374],[261,-0.2374],[262,-0.3429],[263,-0.5188],[264,-0.7458],[265,-1.02],[266,-1.3647],[267,-1.7608],[268,-2.2203],[269,-2.7069],[270,-3.1319],[271,-3.4864],[272,-3.7779],[273,-4.0888],[275,-4.594],[276,-4.8182],[277,-4.9865],[279,-5.2988],[281,-5.5234],[282,-5.6235],[285,-5.9606],[286,-6.031],[287,-6.0168],[288,-5.9606],[290,-5.7623],[291,-5.6349],[292,-5.5214],[293,-5.3937],[294,-5.2511],[295,-5.1235],[296,-4.9463],[297,-4.8039],[298,-4.6268],[300,-4.3719],[301,-4.2587],[302,-4.195],[303,-4.1164],[304,-4.0527],[305,-4.0181],[306,0]]
       }
       thumb_01_r_ctrl.Scale.X {type=MovieSceneScriptingFloatChannel, default=1, const [1,307]}
       thumb_01_r_ctrl.Scale.Y {type=MovieSceneScriptingFloatChannel, default=1, const [1,307]}
       thumb_01_r_ctrl.Scale.Z {type=MovieSceneScriptingFloatChannel, default=1, const [1,307]}
       thumb_02_r_ctrl.Location.X {type=MovieSceneScriptingFloatChannel, default=0, thinned_from=307, keys [[0,0],[306,0]]}
       thumb_02_r_ctrl.Location.Y {type=MovieSceneScriptingFloatChannel, default=0, thinned_from=307, keys [[0,0],[306,0]]}
       thumb_02_r_ctrl.Location.Z {type=MovieSceneScriptingFloatChannel, default=0, thinned_from=307, keys [[0,0],[306,0]]}
       thumb_02_r_ctrl.Rotation.X {
        type=MovieSceneScriptingFloatChannel
        default=0
        thinned_from=307
        keys [[0,-0.0001],[4,-0.0001],[5,-0.0002],[6,-0.0001],[13,-0.0002],[16,-0.0001],[17,-0.0006],[21,-0.0006],[22,0.0017],[23,-0.0007],[24,-0.0006],[27,-0.0007],[28,0.0016],[29,-0.0007],[30,-0.0008],[31,0.0016],[32,0.0015],[34,-0.0029],[35,0.0008],[36,-0.0029],[39,-0.0029],[40,-0.0031],[43,-0.0031],[44,0.0009],[45,-0.0031],[50,-0.0031],[51,0],[64,0],[65,0.0002],[66,0],[67,-0.0004],[68,0],[69,-0.0009],[70,0],[71,-0.0007],[72,0.0009],[74,-0.0005],[75,0.0018],[76,-0.0004],[77,0.0002],[78,-0.0013],[79,0.0022],[80,0],[81,-0.0003],[82,0.0021],[83,-0.0007],[84,0.0016],[85,-0.0015],[87,-0.0007],[90,-0.0005],[91,0],[92,0.0014],[94,0.0017],[95,-0.001],[96,-0.0012],[97,-0.0015],[99,-0.0003],[101,-0.0001],[103,-0.0003],[105,0.0013],[106,-0.0011],[107,0.002],[108,-0.0008],[109,0.0013],[110,0],[111,-0.0003],[112,-0.0003],[113,0.0024],[114,-0.0002],[115,0.0019],[116,-0.0011],[117,-0.0017],[118,0.0009],[120,0.0012],[121,0.0007],[122,0.0012],[123,0.0015],[124,-0.0007],[127,-0.0005],[128,0.0011],[129,-0.0017],[130,0.0009],[132,0.001],[133,-0.0014],[134,-0.0016],[135,-0.0002],[136,-0.0006],[137,0.0018],[138,0.0018],[139,-0.0005],[141,-0.0006],[142,-0.0004],[144,0.0008],[145,0.0011],[146,-0.0015],[147,0.0021],[148,0.0018],[150,0.0018],[152,0.0021],[153,0.0021],[154,-0.0019],[155,0.0018],[157,0.0015],[158,-0.0023],[159,0.0014],[162,0.0011],[163,-0.0029],[164,0.0014],[165,-0.0029],[166,0.0014],[167,0.0019],[168,0.002],[169,0.0018],[170,-0.0024],[171,0.0018],[172,0.0014],[173,0.0015],[174,-0.003],[175,0.0016],[176,-0.0028],[177,-0.0027],[178,0.0021],[179,0.0005],[182,0.0003],[183,0],[184,-0.0001],[185,-0.0003],[190,-0.0004],[191,0.0024],[193,0.0024],[194,-0.0002],[195,0.0024],[196,-0.0002],[200,-0.0004],[204,0.0001],[205,-0.0025],[206,0.0003],[210,0.0005],[211,-0.0025],[212,0.0023],[213,0.0022],[214,-0.0027],[215,0.002],[218,0.0019],[219,0.0021],[220,-0.0026],[221,0.0023],[222,0.0024],[223,-0.0025],[224,0.0022],[226,0.002],[227,0.0003],[235,0.0001],[236,-0.0001],[237,0.0027],[238,0.0025],[239,-0.0009],[240,0.002],[242,0.0018],[243,0.0013],[245,0.0013],[246,0.002],[247,0.0001],[248,0.0013],[249,-0.001],[250,-0.0006],[252,-0.0006],[253,-0.0011],[254,-0.0014],[255,0.0009],[256,-0.0022],[257,0.0005],[259,0.0001],[269,0],[270,0.0001],[274,-0.0001],[275,0.0011],[278,0.001],[279,-0.001],[280,-0.0011],[281,0.001],[290,0.001],[291,0.0019],[292,0.0019],[293,-0.0005],[296,-0.0004],[297,0.002],[298,-0.0004],[305,-0.0004],[306,0]]
       }
       thumb_02_r_ctrl.Rotation.Y {
        type=MovieSceneScriptingFloatChannel
        default=0
        thinned_from=307
        keys [[0,0],[4,0],[5,0.0004],[6,-0.0001],[7,0.0003],[13,0.0003],[14,-0.0001],[16,0],[17,-0.0003],[21,-0.0002],[22,0.0008],[23,0],[24,-0.0005],[25,-0.0001],[27,-0.0002],[28,0.0005],[29,-0.0002],[30,-0.0002],[31,0.0006],[32,0.0007],[34,-0.0017],[35,0.0003],[36,-0.0017],[39,-0.0017],[40,-0.001],[43,-0.001],[44,0.0002],[45,-0.0011],[50,-0.0011],[51,0.0004],[64,0.0002],[65,-0.0002],[66,0.0003],[67,-0.0001],[69,-0.0004],[70,0],[71,-0.0002],[72,0.0007],[73,0.0001],[74,0.0001],[75,0.001],[76,0.0001],[77,0.0001],[78,-0.0002],[79,0.0005],[80,0],[82,0.0003],[83,0],[84,0.0002],[85,-0.0004],[88,-0.0005],[89,-0.0002],[91,0],[92,0.0005],[93,0.0006],[94,0.0005],[95,-0.0003],[96,-0.0002],[97,-0.0002],[98,-0.0007],[99,0],[101,-0.0005],[102,-0.0004],[103,0.0001],[104,-0.0002],[105,0.0003],[106,-0.0002],[107,0.0005],[108,-0.0001],[109,0.0004],[111,-0.0003],[112,-0.0003],[113,0.001],[114,-0.0001],[115,0.0008],[116,-0.0006],[117,-0.0005],[118,0.0001],[122,0.0005],[123,0.0009],[124,-0.0004],[126,-0.0003],[127,-0.0004],[128,0.0004],[129,-0.0004],[130,0.0002],[131,0],[132,0.0003],[133,-0.0003],[134,-0.0005],[135,-0.0001],[136,-0.0003],[137,0.0007],[138,0.0009],[139,-0.0002],[142,-0.0004],[143,0.0001],[145,0.0005],[146,-0.0005],[147,0.0006],[149,0.0004],[150,0.0005],[151,0.0004],[152,0.0007],[153,0.0007],[154,-0.0005],[155,0.0006],[156,0.0004],[157,0.0007],[158,-0.0008],[159,0.0006],[160,0.0006],[161,0.0003],[162,0.0002],[163,-0.001],[164,0.001],[165,-0.0017],[166,0.0015],[167,0.001],[169,0.0016],[170,-0.0018],[171,0.0003],[172,0.001],[173,0.0007],[174,-0.0014],[175,0.0007],[176,-0.0013],[177,-0.0008],[178,0.0011],[179,0],[181,0.0001],[182,-0.0001],[183,0.0002],[184,-0.0002],[185,0.0003],[188,-0.0002],[190,0.0001],[191,0.0012],[193,0.0012],[194,-0.0003],[195,0.0011],[196,-0.0005],[198,0],[200,0.0001],[201,0],[202,0],[203,-0.0001],[204,0.0002],[205,-0.0009],[206,0.0001],[207,0],[209,0.0001],[210,0],[211,-0.0006],[212,0.0006],[213,0.0005],[214,-0.0008],[215,0.0006],[217,0.0008],[219,0.0006],[220,-0.0007],[221,0.0006],[222,0.0007],[223,-0.0006],[224,0.0005],[226,0.0006],[227,-0.0002],[228,0.0003],[234,0.0003],[236,0],[237,0.001],[238,0.0006],[239,-0.0005],[240,0.0004],[242,0.0006],[243,0.0003],[245,0.0002],[246,0.0006],[247,-0.0001],[248,0.0004],[250,-0.0006],[253,-0.0003],[254,-0.0009],[255,0.0005],[256,-0.0006],[257,0.0001],[259,-0.0002],[261,-0.0001],[262,0.0003],[269,0.0001],[270,-0.0004],[271,0],[274,0],[276,0.0005],[278,0.0004],[279,-0.0003],[280,-0.0003],[281,0.0003],[290,0.0003],[291,0.0009],[292,0.0009],[293,-0.0003],[296,-0.0003],[297,0.0009],[298,-0.0003],[306,0]]
       }
       thumb_02_r_ctrl.Rotation.Z {
        type=MovieSceneScriptingFloatChannel
        default=0
        thinned_from=307
        keys [[0,2.137],[2,2.137],[3,1.9957],[4,1.9957],[5,1.713],[7,1.4302],[8,1.4302],[9,1.2886],[13,1.2886],[14,1.4302],[15,1.713],[16,1.8544],[17,1.9313],[18,2.047],[19,2.047],[20,2.1626],[21,2.047],[24,1.3525],[27,0.3085],[28,-0.0401],[30,-0.5055],[31,-0.8549],[32,-1.088],[33,-1.2047],[34,-1.2502],[35,-1.4461],[36,-1.2502],[39,-1.2502],[40,-1.0542],[41,-1.0542],[42,-0.8584],[43,-0.8584],[45,-0.4671],[47,-0.4671],[48,-0.2716],[50,-0.2716],[51,-0.3891],[55,-0.3891],[56,-0.5456],[57,-0.5456],[58,-0.7022],[59,-0.7022],[61,-1.0157],[63,-1.9577],[64,-2.7446],[65,-3.6914],[67,-6.2216],[68,-8.7645],[69,-15.3563],[70,-17.9889],[71,-16.1431],[72,-13.7902],[73,-10.301],[74,-5.842],[75,-1.3213],[76,2.4017],[77,6.3349],[78,10.3555],[79,13.0189],[80,13.9844],[82,12.7773],[83,12.2785],[84,8.1781],[85,2.6554],[86,-1.396],[87,-3.7925],[88,-4.6425],[89,-5.0683],[90,-5.0683],[91,0],[92,5.0111],[93,7.6289],[94,9.2747],[95,9.8221],[96,9.2747],[97,5.1492],[98,-1.6771],[99,-4.3767],[100,-4.7933],[101,-5.0712],[102,-4.3767],[103,0.7187],[104,5.7458],[105,9.5121],[106,11.1175],[107,11.6517],[108,11.9185],[109,9.6461],[110,3.5795],[111,-0.6512],[112,-3.5452],[113,-4.6543],[114,-5.0712],[115,-5.2103],[116,-1.7233],[117,3.5659],[118,6.9732],[119,9.2709],[120,10.1147],[121,5.2729],[122,-1.5993],[123,-3.5867],[124,-4.46],[125,-4.8349],[126,-5.0852],[127,-4.96],[128,-0.9807],[129,3.8102],[130,7.4578],[131,9.1579],[133,10.1147],[134,9.1578],[135,4.1028],[136,-0.1565],[137,-3.3524],[138,-4.4659],[139,-4.8379],[140,-5.0862],[141,-3.1057],[142,1.6745],[143,5.7941],[144,8.7986],[145,9.9952],[146,9.7561],[147,8.5282],[148,7.419],[149,7.0487],[150,7.419],[151,8.1587],[152,8.7128],[153,8.7128],[154,8.3434],[155,7.6041],[157,6.4928],[158,6.1216],[160,5.0067],[161,4.2619],[162,3.142],[163,1.8013],[164,-0.5026],[165,-3.2871],[166,-6.33],[167,-8.4545],[168,-9.4036],[169,-8.9287],[170,-7.5086],[171,-5.86],[172,-3.9867],[173,-2.1241],[174,-0.5027],[175,0.8815],[176,2.4898],[177,3.8633],[178,5.0045],[179,5.8141],[180,5.9521],[181,5.3997],[182,4.4314],[184,1.931],[185,0.3951],[186,-0.8663],[187,-1.5689],[188,-1.9912],[190,-2.555],[191,-2.6961],[193,-2.6961],[194,-2.555],[195,-2.5709],[198,-1.7374],[199,-1.0443],[200,0.0619],[201,1.3024],[202,2.4017],[203,3.3611],[205,5.0005],[207,5.818],[208,5.954],[209,5.954],[210,5.818],[211,5.6491],[214,4.2745],[215,3.5856],[216,3.1256],[217,2.8954],[218,3.1256],[220,4.9623],[221,5.6492],[222,6.1065],[223,5.6491],[226,3.5856],[227,3.2971],[228,3.142],[234,3.142],[235,3.452],[236,4.0714],[237,4.8444],[239,6.6939],[240,7.3088],[241,7.6159],[242,7.6159],[243,7.644],[244,7.644],[245,7.5045],[246,5.6877],[247,3.4413],[248,0.8993],[249,-0.236],[250,-0.6627],[251,-0.805],[252,-0.805],[253,0.0481],[254,0.8991],[255,1.4655],[256,2.1722],[258,3.0187],[259,3.3556],[260,3.4952],[261,3.4952],[262,3.3556],[263,3.3556],[264,3.2159],[265,3.2159],[270,2.5169],[271,2.237],[272,2.097],[274,1.5363],[277,0.9417],[285,0.1475],[288,0.1475],[290,0.3462],[291,0.4207],[292,0.5449],[293,0.7929],[297,1.2887],[298,1.5362],[301,1.9074],[302,1.9074],[303,2.031],[304,2.031],[305,2.1546],[306,0]]
       }
       thumb_02_r_ctrl.Scale.X {type=MovieSceneScriptingFloatChannel, default=1, const [1,307]}
       thumb_02_r_ctrl.Scale.Y {type=MovieSceneScriptingFloatChannel, default=1, const [1,307]}
       thumb_02_r_ctrl.Scale.Z {type=MovieSceneScriptingFloatChannel, default=1, const [1,307]}
       thumb_03_r_ctrl.Location.X {type=MovieSceneScriptingFloatChannel, default=0, thinned_from=307, keys [[0,0],[306,0]]}
       thumb_03_r_ctrl.Location.Y {type=MovieSceneScriptingFloatChannel, default=0, thinned_from=307, keys [[0,0],[306,0]]}
       thumb_03_r_ctrl.Location.Z {type=MovieSceneScriptingFloatChannel, default=0, thinned_from=307, keys [[0,0],[306,0]]}
       thumb_03_r_ctrl.Rotation.X {
        type=MovieSceneScriptingFloatChannel
        default=0
        thinned_from=307
        keys [[0,0.0018],[2,0.0018],[3,-0.0004],[4,-0.0004],[5,0.0018],[6,0.0019],[7,-0.0003],[16,-0.0004],[17,0.0011],[21,0.0011],[22,-0.0004],[33,0.0004],[34,-0.0009],[36,-0.0009],[37,0.0002],[64,-0.0007],[65,0.0011],[66,0.0011],[67,-0.0001],[68,0.0002],[69,0.001],[70,0.0001],[71,-0.0005],[72,-0.0002],[74,0.0006],[75,0.0001],[76,0.0005],[77,0.0019],[78,-0.0002],[79,-0.0005],[80,0],[81,-0.0015],[82,-0.0009],[84,-0.0011],[85,0.0005],[86,-0.0006],[87,0.0002],[88,0.0015],[89,0],[90,0],[91,0.0008],[92,0.0005],[93,0.0008],[94,-0.0006],[95,0.0012],[96,0.0011],[97,0.0005],[98,-0.0005],[101,0],[102,-0.0003],[103,0.0002],[104,0.0002],[105,-0.0007],[108,-0.0004],[109,0.0005],[110,0.0001],[111,-0.0006],[112,-0.0005],[114,0],[115,0],[116,0.0018],[117,0.0012],[118,0.0014],[119,-0.001],[120,-0.0008],[121,-0.0016],[123,-0.0005],[124,-0.0002],[127,0],[128,-0.0002],[129,-0.0017],[130,0.0015],[131,0.0003],[132,-0.0001],[133,0],[134,0.0003],[137,0],[138,-0.0003],[140,0],[141,-0.0009],[142,-0.0002],[144,0.0003],[146,-0.0006],[147,0.001],[148,0.0003],[149,0],[150,0.0002],[151,0.0007],[152,-0.0021],[153,-0.0007],[154,0],[155,0.0005],[156,-0.0002],[157,-0.0007],[158,0.0027],[160,0.0018],[161,-0.0025],[162,0.0009],[163,-0.0004],[164,0.0003],[165,0],[166,-0.0008],[167,0.0001],[168,0.0006],[169,0.0004],[171,-0.001],[172,0.0001],[173,0.0001],[174,-0.0012],[175,0.0007],[176,0],[177,0.0009],[178,0.0003],[179,0.0001],[180,0.0001],[181,-0.0001],[182,0.0012],[183,0.0008],[184,-0.0013],[185,0.0004],[188,0.0007],[189,-0.001],[190,0.0008],[191,-0.0009],[198,-0.001],[199,0.0025],[201,0.0024],[202,0.0025],[203,-0.0009],[206,-0.0001],[210,0.0001],[211,0.0013],[215,-0.0001],[216,-0.0003],[217,0.002],[218,-0.0007],[219,0.001],[220,0.0007],[223,0.0005],[226,0.0011],[227,0.0002],[239,0],[242,0.0002],[243,-0.0003],[247,-0.0001],[248,0.0001],[249,0.0001],[252,-0.0003],[257,0.0005],[258,-0.0008],[259,-0.0003],[274,0.0001],[290,0.0001],[291,0.0006],[301,-0.0002],[302,0.0034],[305,0.0034],[306,0]]
       }
       thumb_03_r_ctrl.Rotation.Y {
        type=MovieSceneScriptingFloatChannel
        default=0
        thinned_from=307
        keys [[0,0.0003],[2,0.0003],[3,0],[4,0],[5,0.0013],[7,0.0007],[8,-0.0012],[13,-0.0012],[14,0.0007],[15,0.0011],[16,-0.0003],[17,0.0006],[23,0.0006],[24,-0.0008],[25,0.0005],[32,-0.0002],[33,0.0011],[34,-0.0003],[39,-0.0005],[40,0.0006],[41,0.0005],[42,-0.0008],[43,0.0002],[44,0.0001],[45,0.001],[46,0.001],[47,-0.0002],[48,0.0008],[50,0.0008],[51,0.0003],[53,0.0003],[54,-0.0016],[56,-0.0016],[57,0.0004],[60,0.0005],[61,-0.0013],[62,0.0007],[63,-0.0011],[64,-0.001],[65,0.0011],[66,0.0011],[68,0.0005],[69,0.0016],[71,-0.0015],[72,0.0005],[74,0.0001],[75,-0.0016],[76,0.0002],[77,0.0008],[78,0.0005],[79,-0.0018],[80,0],[81,-0.0002],[83,-0.0012],[84,-0.0002],[85,-0.0001],[86,-0.0013],[87,-0.002],[88,0.0007],[89,0],[90,0],[91,-0.0005],[92,0.0001],[93,0.0002],[94,-0.0003],[95,0],[97,0.0001],[98,-0.0014],[99,0.0001],[101,0],[102,0.0001],[103,0.0008],[104,0.0004],[105,-0.0007],[106,0.001],[108,0.0007],[109,0.0017],[110,0.0007],[111,-0.0018],[112,0.0002],[115,0],[116,0.0011],[117,0.001],[118,0.0004],[119,-0.0006],[120,-0.0009],[121,0.0002],[122,0.0004],[124,0.0001],[127,0],[128,-0.0036],[129,0.0003],[130,0.0003],[131,-0.0017],[132,0.0009],[133,0.0008],[134,-0.0017],[135,-0.0012],[136,0.0014],[137,-0.0023],[138,0.0002],[140,0],[141,0.0004],[142,0.0015],[143,-0.0013],[144,-0.0016],[145,0.001],[146,0.0014],[147,-0.0007],[148,0.0005],[149,0.0008],[150,0.0006],[151,-0.0001],[152,-0.0012],[153,0.0009],[154,0],[155,0.0032],[156,0.0003],[157,0.0009],[159,0.0017],[160,-0.0013],[161,-0.0006],[162,-0.0002],[163,0.0007],[164,0.0002],[165,0.0005],[166,-0.0001],[167,0.0002],[168,0.0011],[169,0],[170,-0.0007],[172,0.0005],[173,-0.0009],[174,-0.0012],[175,-0.0003],[176,0.0003],[177,0.0007],[178,-0.0003],[179,-0.0001],[180,-0.0002],[183,0.0006],[184,-0.001],[185,0.0005],[187,-0.0001],[188,0.0013],[189,-0.0006],[190,0.001],[191,-0.0007],[196,-0.0008],[198,-0.0005],[200,0.0003],[203,0.0005],[207,-0.0001],[210,-0.0001],[211,0.0004],[213,0.0009],[214,-0.001],[215,0.0014],[217,0.0015],[218,-0.0005],[219,-0.0003],[220,-0.0003],[221,-0.0006],[222,0.0012],[225,0.0002],[226,-0.0002],[227,0],[239,0.0001],[242,-0.0001],[243,-0.0009],[245,-0.0009],[246,0.0004],[247,0.0003],[248,-0.0012],[249,-0.0012],[250,0.0003],[251,-0.0009],[252,0.0004],[253,-0.001],[254,0.0002],[256,-0.0002],[257,0.0008],[258,-0.0006],[259,-0.0012],[264,-0.0012],[265,-0.001],[267,-0.001],[268,-0.0008],[269,-0.0008],[270,-0.0006],[274,0],[275,-0.0002],[290,0],[291,-0.0019],[292,0.0013],[301,0.0009],[305,0.0007],[306,0]]
       }
       thumb_03_r_ctrl.Rotation.Z {
        type=MovieSceneScriptingFloatChannel
        default=0
        thinned_from=307
        keys [[0,2.1673],[2,2.1673],[3,2.0004],[4,2.0004],[8,1.3325],[13,1.3325],[16,1.8334],[17,1.9808],[18,2.0987],[21,2.0987],[23,1.6274],[24,1.274],[25,1.0383],[28,-0.022],[30,-0.4933],[31,-0.8468],[32,-1.0825],[33,-1.2003],[34,-1.3766],[36,-1.3766],[37,-1.2889],[38,-1.2889],[42,-0.9382],[43,-0.7629],[44,-0.6753],[45,-0.4999],[46,-0.4999],[48,-0.3246],[50,-0.3246],[51,-0.3048],[53,-0.3048],[54,-0.4603],[56,-0.4603],[57,-0.6159],[58,-0.6159],[59,-0.7714],[60,-0.7714],[62,-1.3936],[63,-2.0157],[64,-2.7936],[65,-3.7274],[66,-4.817],[67,-6.2343],[68,-8.8017],[69,-14.8748],[70,-18.205],[71,-17.4636],[72,-15.7984],[73,-13.5835],[74,-10.0871],[75,-6.4175],[76,-3.4874],[77,-0.3771],[78,2.5495],[79,4.5621],[80,4.9281],[81,-2.7553],[82,-11.5578],[83,-18.9695],[84,-22.9341],[85,-25.1874],[86,-27.6255],[87,-29.0246],[89,-29.7258],[90,-29.7258],[92,-23.7996],[93,-22.2427],[94,-21.3796],[95,-21.0347],[96,-21.3796],[97,-23.7996],[98,-27.8001],[99,-29.2502],[100,-29.5672],[101,-29.7258],[102,-29.2502],[103,-26.2499],[104,-23.4247],[105,-21.2378],[106,-20.3031],[107,-19.9918],[108,-19.8363],[109,-21.0819],[110,-24.6784],[111,-27.1952],[112,-28.7751],[113,-29.4087],[114,-29.7258],[115,-29.7258],[116,-27.7133],[117,-24.5685],[118,-22.5774],[119,-21.4429],[120,-20.8765],[121,-23.7142],[122,-27.7132],[123,-28.8621],[124,-29.4377],[125,-29.4377],[126,-29.7258],[127,-29.7258],[128,-27.4265],[129,-24.5684],[130,-22.2936],[131,-21.5213],[132,-21.1339],[133,-20.9402],[134,-21.5213],[135,-24.436],[136,-26.7796],[137,-28.7415],[138,-29.3318],[140,-29.7258],[141,-28.545],[142,-25.8018],[143,-23.4627],[144,-21.9091],[145,-21.3276],[146,-17.4648],[147,-10.687],[148,-3.8423],[149,0.3145],[150,1.205],[151,2.986],[152,5.0647],[153,6.8475],[154,7.7394],[155,8.0368],[160,5.0647],[162,3.2829],[163,1.6901],[164,-0.4956],[166,-6.2515],[167,-8.4427],[168,-9.2508],[169,-8.789],[170,-7.6351],[171,-5.9059],[172,-3.9479],[173,-2.1063],[177,3.8761],[178,5.0271],[179,5.8515],[180,5.9931],[181,5.4267],[182,4.4359],[183,3.3038],[185,0.4745],[186,-0.7986],[187,-1.506],[189,-2.3549],[191,-2.638],[195,-2.638],[198,-1.7848],[199,-0.9364],[202,2.4568],[205,5.0026],[206,5.5686],[207,5.8516],[210,5.8516],[211,5.6588],[212,5.2921],[215,3.6425],[216,3.276],[217,3.0927],[218,2.5431],[221,-1.8539],[222,-2.7702],[223,-3.3201],[226,-5.5207],[227,-5.978],[228,-6.1338],[234,-6.1338],[235,-5.8222],[236,-5.355],[238,-3.4868],[240,-1.9309],[241,-1.6198],[242,-1.6198],[243,-1.6869],[245,-1.6869],[246,-2.355],[247,-3.1345],[248,-3.8028],[249,-3.8028],[250,-2.9118],[252,-0.9077],[254,0.8733],[255,1.5412],[257,2.6544],[258,2.9884],[259,3.3363],[264,3.3363],[265,3.0232],[267,3.0232],[268,2.7101],[269,2.7101],[270,2.397],[271,2.397],[274,1.4577],[275,1.3743],[277,0.9151],[278,0.9151],[279,0.6855],[280,0.6855],[281,0.456],[283,0.456],[284,0.2264],[285,0.2264],[286,0],[287,0],[288,0.2264],[289,0.2264],[290,0.456],[291,0.4351],[292,0.7064],[293,0.7064],[294,0.9777],[295,0.9777],[296,1.249],[297,1.249],[298,1.5203],[299,1.5203],[300,1.7917],[301,1.7917],[302,2.0629],[305,2.0629],[306,0]]
       }
       thumb_03_r_ctrl.Scale.X {type=MovieSceneScriptingFloatChannel, default=1, const [1,307]}
       thumb_03_r_ctrl.Scale.Y {type=MovieSceneScriptingFloatChannel, default=1, const [1,307]}
       thumb_03_r_ctrl.Scale.Z {type=MovieSceneScriptingFloatChannel, default=1, const [1,307]}
       index_metacarpal_r_ctrl.Location.X {type=MovieSceneScriptingFloatChannel, default=0, thinned_from=307, keys [[0,0],[306,0]]}
       index_metacarpal_r_ctrl.Location.Y {type=MovieSceneScriptingFloatChannel, default=0, thinned_from=307, keys [[0,0],[306,0]]}
       index_metacarpal_r_ctrl.Location.Z {type=MovieSceneScriptingFloatChannel, default=0, thinned_from=307, keys [[0,0],[306,0]]}
       index_metacarpal_r_ctrl.Rotation.X {
        type=MovieSceneScriptingFloatChannel
        default=0
        thinned_from=307
        keys [[0,0],[2,0],[3,-0.0004],[4,0.0001],[5,0.0003],[6,0],[7,-0.0005],[8,-0.0006],[9,-0.0004],[13,-0.0004],[14,0.0005],[15,-0.0002],[18,-0.0004],[19,0.0003],[23,0.0003],[24,-0.0004],[25,-0.0004],[26,0.0003],[27,-0.0004],[28,-0.0003],[29,0.0003],[36,0.0003],[38,-0.0004],[42,0.0002],[43,-0.0001],[44,0.0001],[45,-0.0002],[46,-0.0001],[47,-0.0005],[48,0.0001],[49,-0.0004],[50,0.0002],[51,0.007],[57,0.007],[58,0.0076],[59,0.0076],[60,0.0081],[61,0],[62,0],[63,-0.021],[64,-0.0506],[65,-0.1094],[66,-0.1987],[67,-0.2752],[68,-0.4249],[69,-0.6858],[70,-0.9067],[71,-0.9852],[72,-1.072],[73,-1.1796],[74,-1.3231],[75,-1.4567],[76,-1.539],[77,-1.5905],[78,-1.6204],[81,-1.6849],[83,-1.7316],[145,-1.7322],[146,-1.6809],[147,-1.5488],[149,-1.3276],[150,-1.2014],[151,-0.9541],[152,-0.6095],[153,-0.2706],[154,-0.0403],[155,-0.0025],[156,0.0024],[157,-0.0011],[158,-0.0008],[159,0.0022],[160,-0.0022],[161,-0.0006],[162,0.0024],[163,0],[166,-0.0004],[167,0.0005],[169,-0.0004],[171,0],[172,-0.0001],[173,0.0003],[174,0.0005],[175,0.0002],[176,0.0003],[177,0.0001],[178,0.0005],[180,0.0001],[181,0.0002],[183,-0.0004],[184,0.0001],[185,0.0003],[186,-0.0002],[188,-0.0001],[189,0.0001],[190,0.0001],[193,0.0004],[194,-0.0002],[195,0.0002],[196,0.0003],[197,0.0001],[198,0.0002],[199,0],[200,0.0001],[202,-0.0002],[203,0.0003],[205,0.0001],[207,0.0003],[208,0.0001],[210,0.0001],[211,-0.0001],[212,0.0001],[214,-0.0001],[216,0],[217,0.0003],[218,0],[219,0.0001],[220,-0.0003],[221,0],[222,-0.0002],[224,0],[226,-0.0002],[227,0],[228,0.0006],[230,-0.0009],[231,0.0004],[236,0.0004],[238,-0.0007],[239,0.0008],[241,-0.0004],[243,0],[245,0.0001],[246,-0.0001],[247,0.0001],[249,0],[251,0.0003],[252,-0.0002],[253,0.0002],[255,0.0001],[256,0.0003],[257,0],[258,0.0002],[259,-0.0001],[261,0.0001],[262,-0.0001],[263,-0.0001],[264,0.0002],[265,0],[266,0],[267,0.0003],[268,-0.0002],[269,-0.0002],[270,0.0002],[271,-0.0001],[272,0],[273,0.0005],[274,-0.0002],[277,0],[278,0.0006],[279,0],[281,0.0001],[282,-0.0005],[283,-0.0005],[284,0.0008],[285,0.0008],[286,-0.0004],[287,-0.0004],[288,0.0008],[289,0.0001],[290,0.0007],[291,-0.0001],[292,0],[294,-0.0002],[295,0.0003],[296,0.0002],[297,-0.0001],[299,-0.0003],[301,0.0001],[302,-0.0001],[303,0.0004],[304,-0.0001],[306,0]]
       }
       index_metacarpal_r_ctrl.Rotation.Y {
        type=MovieSceneScriptingFloatChannel
        default=0
        thinned_from=307
        keys [[0,0],[3,0.0001],[4,-0.0004],[5,0],[6,0],[7,0.0002],[8,-0.0005],[9,-0.0001],[13,-0.0001],[14,0.0002],[15,0.0002],[16,-0.0002],[17,-0.0001],[18,0.0005],[21,0.0003],[22,-0.0003],[23,0.0001],[24,-0.0003],[25,0.0001],[26,-0.0004],[27,0.0002],[28,0.0006],[30,-0.0004],[32,0.0003],[33,-0.0003],[38,-0.0005],[39,0.0001],[41,-0.0004],[42,-0.0001],[43,-0.0004],[45,0.0003],[49,-0.0003],[50,0.0004],[51,0.0029],[57,0.0029],[58,0.0036],[59,0.0036],[60,0.0043],[61,0],[62,0],[63,0.0055],[64,0.0185],[65,0.057],[66,0.0828],[67,0.1191],[68,0.1828],[69,0.2965],[70,0.3915],[71,0.4256],[72,0.4631],[73,0.5099],[74,0.5718],[75,0.6288],[76,0.6656],[77,0.6882],[78,0.6996],[81,0.7287],[82,0.7395],[83,0.7482],[87,0.749],[90,0.7479],[100,0.7491],[102,0.7483],[108,0.7491],[109,0.748],[130,0.7487],[131,0.7479],[133,0.749],[134,0.7479],[135,0.7496],[136,0.7485],[137,0.7495],[141,0.7481],[142,0.7492],[143,0.7479],[145,0.7479],[146,0.7268],[147,0.6691],[148,0.6231],[149,0.575],[150,0.5193],[151,0.4121],[152,0.2629],[153,0.1169],[154,0.0201],[155,0.0003],[156,0.0002],[157,0.0006],[158,0.0015],[159,-0.0016],[160,0],[161,0.0012],[162,0.0002],[164,-0.0001],[166,-0.0007],[167,0.0003],[168,0],[169,-0.0008],[170,-0.0007],[171,0.0002],[172,-0.0002],[174,0],[175,-0.0001],[177,0.0007],[178,0.0002],[179,-0.0001],[180,0.0003],[183,-0.0002],[185,0.0002],[186,0.0001],[187,-0.0003],[190,-0.0002],[191,0.0004],[193,0.0004],[194,-0.0002],[195,0.0003],[199,0.0001],[200,0.0004],[201,-0.0001],[202,0],[203,0.0004],[204,-0.0002],[205,0.0004],[206,0],[207,0.0002],[211,0],[212,0.0002],[213,-0.0001],[214,0.0002],[216,0],[217,0.0002],[219,0],[220,-0.0003],[221,-0.0002],[222,0.0002],[225,-0.0003],[226,-0.0001],[227,-0.0005],[228,0.0004],[229,0.0009],[230,-0.0002],[231,0.0003],[236,0.0003],[237,0.0009],[238,0],[239,0.0005],[240,-0.0004],[241,0.0002],[244,0.0003],[245,-0.0001],[248,0],[249,-0.0002],[250,0.0001],[251,0.0003],[252,-0.0004],[253,-0.0001],[254,-0.0003],[255,0.0001],[258,0.0001],[259,0.0002],[260,-0.0004],[264,0],[265,0.0003],[266,0],[267,0.0003],[269,-0.0001],[270,0.0001],[271,-0.0002],[275,0.0004],[277,0],[278,0.0005],[279,0.0001],[280,0.0005],[282,-0.0006],[285,0.0005],[286,-0.0001],[287,-0.0001],[288,0.0005],[289,-0.0006],[290,0.0008],[292,-0.0002],[299,0],[300,-0.0004],[301,0.0003],[302,0],[303,0.0002],[304,0],[306,0]]
       }
       index_metacarpal_r_ctrl.Rotation.Z {
        type=MovieSceneScriptingFloatChannel
        default=0
        thinned_from=307
        keys [[0,0],[2,0],[3,-0.0687],[4,-0.1898],[5,-0.3411],[6,-0.5227],[7,-0.6437],[8,-0.7345],[9,-0.795],[13,-0.795],[14,-0.6437],[15,-0.4622],[16,-0.3108],[17,-0.1696],[18,-0.0687],[19,-0.0183],[21,-0.0183],[23,-0.2201],[24,-0.3714],[25,-0.4722],[27,-0.7749],[28,-0.8758],[29,-1.0271],[33,-1.4305],[36,-1.4305],[37,-1.3902],[38,-1.3448],[41,-1.0724],[42,-0.8909],[43,-0.7547],[44,-0.5731],[45,-0.4369],[46,-0.3461],[50,-0.1646],[51,-0.1409],[57,-0.1409],[58,-0.0963],[59,-0.0963],[60,-0.0516],[61,0],[62,0],[63,-0.0066],[65,-0.0064],[66,-0.0504],[67,-0.0485],[68,-0.1028],[69,-1.7184],[70,-3.9372],[71,-5.1976],[72,-5.3897],[73,-5.1436],[74,-3.3909],[75,-0.8439],[76,1.3749],[78,6.4223],[79,8.2087],[80,8.5388],[81,2.827],[82,-4.4621],[83,-9.8854],[84,-10.452],[85,-10.5228],[86,-10.6999],[87,-10.7708],[88,-10.8062],[90,-10.8062],[92,-10.452],[93,-10.3457],[95,-10.2749],[96,-10.3103],[97,-10.452],[98,-10.6999],[99,-10.7758],[101,-10.8264],[102,-10.8011],[103,-10.5987],[104,-10.4216],[105,-10.2951],[106,-10.2445],[108,-10.194],[109,-10.2951],[110,-10.4975],[111,-10.6493],[112,-10.7505],[113,-10.8011],[115,-10.821],[116,-10.6977],[117,-10.5073],[118,-10.384],[119,-10.2944],[120,-10.272],[121,-10.4401],[122,-10.6865],[123,-10.7538],[125,-10.8098],[127,-10.8098],[128,-10.6641],[129,-10.4961],[130,-10.3616],[131,-10.3056],[132,-10.272],[133,-10.272],[134,-10.3056],[136,-10.6417],[137,-10.7426],[139,-10.8098],[140,-10.8098],[141,-10.7425],[142,-10.5745],[143,-10.44],[144,-10.3392],[145,-10.3056],[146,-8.8607],[148,-1.7066],[149,-0.2634],[150,-0.2629],[151,-0.2279],[152,-0.1935],[153,-0.1599],[154,-0.0923],[155,0.1823],[156,1.0408],[157,2.1401],[158,3.0679],[159,3.3085],[160,2.9648],[161,2.1744],[162,1.0408],[163,-0.4095],[164,-2.6094],[166,-8.3662],[167,-10.5143],[168,-11.4196],[169,-10.9669],[170,-9.7226],[171,-7.9709],[172,-6.0511],[173,-4.189],[174,-2.6658],[175,-1.2556],[176,0.2675],[177,1.7346],[178,2.92],[179,3.6817],[180,3.8551],[181,3.3002],[182,2.2601],[183,1.1163],[184,-0.1657],[185,-1.6554],[186,-2.9371],[187,-3.6646],[188,-4.1151],[189,-4.4269],[190,-4.6348],[191,-4.7388],[192,-4.7734],[193,-4.7734],[194,-4.7387],[195,-4.6189],[196,-4.4502],[197,-4.1804],[198,-3.8431],[199,-3.1686],[200,-1.9884],[201,-0.8418],[202,0.2374],[203,1.2831],[204,2.1266],[205,2.8353],[206,3.3416],[207,3.6792],[208,3.8143],[209,3.8143],[210,3.7805],[211,3.693],[212,3.5776],[213,3.4045],[214,3.2026],[215,2.9142],[216,2.5104],[217,1.9337],[219,0.2617],[220,-0.4588],[221,-1.1217],[222,-1.7268],[223,-2.2743],[224,-2.7354],[225,-3.1389],[226,-3.4559],[227,-3.6855],[228,-3.8106],[231,-3.9983],[236,-3.9983],[241,-3.3728],[242,-3.1852],[243,-3.1144],[244,-3.1144],[245,-3.1683],[247,-3.3302],[248,-3.3302],[249,-2.8176],[250,-1.8195],[251,-0.6864],[252,0.231],[253,0.9596],[254,1.6073],[255,2.2011],[256,2.7141],[257,3.1191],[258,3.4432],[259,3.6783],[260,3.78],[261,3.78],[262,3.6783],[263,3.5088],[264,3.2716],[265,3.0005],[266,2.6616],[267,2.2551],[269,1.3066],[270,0.9002],[271,0.5277],[272,0.223],[274,-0.3188],[275,-0.5656],[277,-0.969],[279,-1.2717],[281,-1.4734],[282,-1.6247],[286,-2.0282],[287,-2.0282],[288,-1.9273],[289,-1.8769],[290,-1.7256],[291,-1.6299],[292,-1.5157],[296,-0.9449],[297,-0.7737],[299,-0.4883],[301,-0.2599],[303,-0.0887],[304,-0.0316],[305,0],[306,0]]
       }
       index_metacarpal_r_ctrl.Scale.X {type=MovieSceneScriptingFloatChannel, default=1, const [1,307]}
       index_metacarpal_r_ctrl.Scale.Y {type=MovieSceneScriptingFloatChannel, default=1, const [1,307]}
       index_metacarpal_r_ctrl.Scale.Z {type=MovieSceneScriptingFloatChannel, default=1, const [1,307]}
       index_01_r_ctrl.Location.X {type=MovieSceneScriptingFloatChannel, default=0, thinned_from=307, keys [[0,0],[306,0]]}
       index_01_r_ctrl.Location.Y {type=MovieSceneScriptingFloatChannel, default=0, thinned_from=307, keys [[0,0],[306,0]]}
       index_01_r_ctrl.Location.Z {type=MovieSceneScriptingFloatChannel, default=0, thinned_from=307, keys [[0,0],[306,0]]}
       index_01_r_ctrl.Rotation.X {
        type=MovieSceneScriptingFloatChannel
        default=0
        thinned_from=307
        keys [[0,-10.0755],[2,-10.0712],[3,-10.0872],[8,-10.0735],[9,-10.0896],[15,-10.0742],[46,-10.0844],[57,-10.071],[58,-10.0976],[60,-10.0754],[61,-10.0967],[62,-10.0914],[63,-10.0338],[64,-9.8926],[66,-9.3774],[67,-9.0469],[68,-8.4899],[69,-7.6601],[70,-6.5449],[71,-5.1754],[72,-3.4242],[73,-1.3594],[74,1.203],[75,3.4935],[76,4.9961],[77,5.9212],[78,6.4096],[79,6.8177],[81,7.551],[82,7.9855],[83,8.3648],[137,8.3843],[145,8.3752],[146,7.8282],[147,6.4127],[148,5.2798],[149,4.0669],[150,2.7441],[151,0.0854],[152,-3.5819],[153,-7.1855],[154,-9.646],[155,-10.0952],[159,-10.0676],[160,-10.1076],[161,-10.0579],[162,-10.0513],[163,-10.0819],[275,-10.0834],[276,-10.0726],[289,-10.0841],[305,-10.0791],[306,0]]
       }
       index_01_r_ctrl.Rotation.Y {
        type=MovieSceneScriptingFloatChannel
        default=0
        thinned_from=307
        keys [[0,-2.3233],[1,-2.3233],[2,-2.3285],[3,-2.3195],[4,-2.3197],[5,-2.3262],[8,-2.3278],[9,-2.3189],[10,-2.324],[13,-2.324],[14,-2.3175],[15,-2.3263],[16,-2.321],[17,-2.3257],[18,-2.3218],[21,-2.3218],[24,-2.3266],[29,-2.3231],[37,-2.3265],[38,-2.3203],[39,-2.3241],[40,-2.3163],[42,-2.324],[44,-2.3161],[47,-2.3261],[48,-2.3261],[49,-2.3198],[51,-2.3198],[52,-2.3311],[55,-2.3372],[56,-2.3433],[57,-2.3065],[59,-2.3342],[60,-2.3065],[61,-2.3252],[62,-2.3313],[63,-2.2353],[64,-1.9542],[65,-1.5245],[66,-1.0095],[67,-0.3828],[68,0.4512],[69,1.6127],[70,2.1111],[71,1.4761],[72,0.8055],[73,0.3679],[74,0.4294],[75,0.7788],[76,1.246],[77,1.6162],[78,1.8082],[79,1.9746],[81,2.2674],[82,2.4432],[83,2.5896],[84,2.5879],[85,2.5952],[90,2.5898],[91,2.5952],[92,2.5946],[93,2.5888],[94,2.5933],[95,2.5879],[96,2.5933],[97,2.5882],[98,2.598],[101,2.5897],[112,2.5916],[116,2.5947],[117,2.591],[132,2.5952],[133,2.5907],[135,2.5965],[136,2.5913],[137,2.5927],[138,2.5879],[139,2.5929],[141,2.5936],[142,2.5866],[145,2.5907],[146,2.2301],[147,1.3035],[148,0.5475],[149,-0.2105],[150,-1.0971],[151,-2.816],[152,-5.0434],[153,-6.756],[154,-7.2852],[155,-6.441],[156,-5.122],[157,-3.9615],[158,-2.9954],[159,-2.4404],[160,-2.3268],[162,-2.3311],[163,-2.3215],[166,-2.321],[168,-2.3264],[169,-2.3204],[172,-2.3267],[173,-2.3196],[174,-2.3186],[175,-2.3262],[177,-2.3224],[178,-2.328],[179,-2.3223],[180,-2.3261],[182,-2.3237],[191,-2.325],[193,-2.3217],[195,-2.3276],[196,-2.3193],[197,-2.316],[198,-2.3275],[199,-2.3133],[200,-2.3229],[201,-2.3114],[202,-2.3283],[203,-2.4025],[204,-2.5493],[205,-2.7526],[206,-3.0153],[207,-3.3077],[208,-3.6516],[209,-3.9171],[210,-3.953],[211,-3.867],[212,-3.6997],[213,-3.4339],[214,-3.0914],[215,-2.5831],[216,-1.935],[217,-1.4746],[218,-1.2671],[219,-1.1284],[220,-1.0358],[221,-1.0057],[222,-0.985],[223,-0.9376],[224,-0.8163],[225,-0.5789],[226,-0.3433],[227,-0.1372],[228,-0.1167],[229,-0.1296],[230,-0.098],[231,-0.1059],[232,-0.1059],[233,-0.1108],[235,-0.1108],[236,-0.1472],[237,-0.335],[238,-0.5996],[239,-0.8197],[240,-1.0066],[241,-1.0714],[242,-1.1766],[243,-1.4288],[244,-1.9483],[245,-2.6931],[246,-3.5562],[247,-4.252],[248,-4.605],[249,-4.6606],[250,-4.4805],[251,-4.2465],[252,-3.9876],[253,-3.6925],[254,-3.4619],[255,-3.1361],[257,-2.7025],[258,-2.5422],[259,-2.394],[260,-2.3333],[262,-2.3333],[263,-2.3212],[265,-2.327],[268,-2.3258],[269,-2.3183],[270,-2.324],[271,-2.3212],[275,-2.3308],[277,-2.3234],[278,-2.329],[282,-2.3186],[283,-2.3258],[285,-2.3213],[286,-2.3285],[287,-2.3285],[288,-2.3213],[290,-2.3258],[292,-2.3206],[293,-2.3252],[296,-2.3222],[298,-2.3261],[299,-2.3233],[300,-2.3279],[302,-2.3224],[303,-2.325],[304,-2.3196],[305,-2.3196],[306,0]]
       }
       index_01_r_ctrl.Rotation.Z {
        type=MovieSceneScriptingFloatChannel
        default=0
        thinned_from=307
        keys [[0,1.9951],[1,1.9951],[2,1.9313],[3,1.93],[4,1.8014],[5,1.6087],[8,1.2226],[9,1.2212],[10,1.1574],[13,1.1574],[14,1.3502],[15,1.4801],[16,1.6725],[17,1.8019],[18,1.909],[19,1.9626],[20,1.9626],[21,1.909],[22,1.7481],[25,0.9426],[28,-0.0243],[29,-0.2931],[30,-0.5085],[31,-0.7772],[33,-1.2072],[34,-1.2606],[37,-1.2606],[42,-0.8578],[44,-0.5354],[47,-0.2926],[48,-0.2926],[49,-0.2127],[51,-0.2127],[52,-0.2483],[53,-0.2483],[55,-0.3256],[56,-0.403],[57,-0.484],[58,-0.601],[59,-0.7944],[60,-0.9915],[61,-1.186],[62,-1.2634],[63,-1.2299],[64,-1.1689],[65,-1.0358],[66,-0.9079],[67,-0.6803],[68,-0.202],[69,-0.4288],[70,-1.1889],[71,-1.696],[72,-1.0578],[73,0.3289],[74,2.4827],[75,4.9062],[76,6.5455],[77,8.3573],[78,10.187],[79,11.508],[80,11.8944],[81,10.1897],[82,8.2903],[83,6.6375],[84,5.1277],[85,3.7006],[86,2.5237],[87,1.851],[88,1.5985],[89,1.5142],[90,1.5142],[92,4.3725],[93,5.2117],[94,5.6317],[95,5.7992],[96,5.6317],[97,4.456],[98,2.4401],[99,1.7091],[100,1.5895],[101,1.5098],[102,1.7091],[103,3.1821],[104,4.6142],[105,5.7272],[106,6.1645],[107,6.3233],[108,6.3632],[109,5.7274],[110,3.9781],[111,2.7842],[112,1.9481],[113,1.6297],[114,1.5101],[115,1.5101],[116,2.4395],[117,3.9873],[118,4.9589],[119,5.642],[120,5.8581],[121,4.4912],[122,2.5113],[123,1.9351],[124,1.6828],[125,1.5745],[126,1.5024],[127,1.5383],[128,2.6914],[129,4.0594],[130,5.1027],[131,5.5671],[133,5.8564],[134,5.5671],[135,4.1167],[136,2.8826],[137,2.0119],[138,1.7203],[139,1.5756],[140,1.503],[141,2.0844],[142,3.463],[143,4.6244],[144,5.494],[145,5.8563],[146,5.6491],[147,4.9463],[148,4.2901],[149,4.0738],[150,4.7776],[151,5.975],[152,7.1301],[153,7.7686],[154,7.8691],[155,7.5139],[157,6.5274],[159,5.4275],[160,4.9315],[161,4.1606],[162,3.0105],[163,1.5562],[164,-0.59],[165,-3.5326],[166,-6.3712],[167,-8.5349],[168,-9.4477],[169,-8.9916],[170,-7.7372],[171,-6.0299],[173,-2.1736],[174,-0.7031],[175,0.6536],[176,2.2335],[177,3.6992],[178,4.8266],[179,5.6625],[180,5.8003],[181,5.2504],[182,4.2197],[183,3.0855],[184,1.8126],[185,0.2977],[186,-0.9434],[187,-1.6677],[188,-2.1507],[189,-2.4616],[190,-2.6687],[191,-2.7724],[192,-2.8073],[193,-2.8073],[194,-2.7724],[195,-2.6169],[196,-2.5202],[197,-2.2284],[198,-1.8386],[199,-1.159],[200,0.0079],[201,1.1715],[202,2.2402],[203,3.4095],[204,4.4877],[205,5.667],[206,6.6577],[207,7.5543],[208,8.3586],[209,9.0595],[210,9.6437],[211,10.0828],[212,10.3517],[213,10.5172],[214,10.5342],[215,10.3947],[216,10.007],[217,9.1167],[218,7.588],[219,5.8762],[220,4.451],[221,3.1249],[222,1.8934],[223,0.7535],[224,-0.1566],[225,-0.982],[226,-1.6179],[227,-2.0833],[228,-2.3447],[229,-2.4998],[231,-2.7101],[232,-2.7101],[233,-2.762],[235,-2.762],[236,-2.7068],[237,-2.4832],[238,-2.2009],[239,-1.8708],[240,-1.6471],[241,-1.4338],[242,-1.217],[243,-0.946],[244,-0.7075],[245,-0.5815],[246,-0.5769],[247,-0.5208],[248,-0.2311],[249,0.3618],[250,1.1964],[251,2.026],[252,2.5265],[253,2.8281],[254,3.0044],[255,3.0423],[256,3.0246],[257,3.0714],[258,3.058],[259,3.0615],[262,3.0565],[264,2.9417],[265,2.8276],[266,2.77],[269,2.4269],[271,2.0847],[274,1.3995],[275,1.1365],[276,0.9539],[277,0.8621],[278,0.6798],[282,0.3139],[283,0.2233],[284,0.1315],[285,0.0401],[286,-0.0505],[287,-0.0505],[288,0.0401],[289,0.1315],[290,0.2233],[291,0.3312],[293,0.5705],[294,0.7497],[295,0.869],[296,1.0482],[297,1.1683],[298,1.3474],[302,1.8251],[304,1.9443],[305,1.9443],[306,0]]
       }
       index_01_r_ctrl.Scale.X {type=MovieSceneScriptingFloatChannel, default=1, const [1,307]}
       index_01_r_ctrl.Scale.Y {type=MovieSceneScriptingFloatChannel, default=1, const [1,307]}
       index_01_r_ctrl.Scale.Z {type=MovieSceneScriptingFloatChannel, default=1, const [1,307]}
       index_02_r_ctrl.Location.X {type=MovieSceneScriptingFloatChannel, default=0, thinned_from=307, keys [[0,0],[306,0]]}
       index_02_r_ctrl.Location.Y {type=MovieSceneScriptingFloatChannel, default=0, thinned_from=307, keys [[0,0],[306,0]]}
       index_02_r_ctrl.Location.Z {type=MovieSceneScriptingFloatChannel, default=0, thinned_from=307, keys [[0,0],[306,0]]}
       index_02_r_ctrl.Rotation.X {
        type=MovieSceneScriptingFloatChannel
        default=0
        thinned_from=307
        keys [[0,-0.0001],[1,-0.0001],[2,0.0001],[3,0.0001],[4,0.0004],[5,-0.0007],[7,0.0002],[8,0.0004],[13,0.0004],[14,0.0002],[16,-0.0007],[17,-0.0004],[18,0.0004],[19,0.0004],[20,0.0001],[21,0.0004],[22,-0.0004],[23,0.0002],[24,0.0003],[25,-0.0002],[26,-0.0004],[27,-0.0002],[28,0.0004],[30,-0.0006],[31,0.0006],[32,0.0001],[33,0.0004],[34,-0.0008],[36,-0.0008],[37,-0.0011],[38,-0.0011],[39,0.0002],[40,-0.0001],[41,-0.0001],[42,-0.0006],[43,-0.0009],[44,0.0004],[47,-0.0004],[50,-0.0004],[51,-0.001],[56,-0.0006],[60,0.0003],[62,0.0014],[63,-0.025],[64,-0.1183],[65,-0.2478],[66,-0.3988],[67,-0.5812],[68,-0.9023],[69,-1.455],[70,-1.9213],[71,-2.0918],[72,-2.2744],[73,-2.4986],[74,-2.812],[75,-3.0879],[76,-3.2766],[77,-3.3797],[78,-3.4496],[79,-3.4811],[81,-3.5687],[82,-3.6352],[83,-3.6743],[118,-3.6746],[119,-3.6793],[120,-3.6752],[121,-3.6797],[122,-3.6728],[128,-3.6782],[129,-3.674],[145,-3.6774],[146,-3.568],[147,-3.2816],[148,-3.0577],[149,-2.8228],[150,-2.5549],[151,-2.0274],[152,-1.2966],[153,-0.572],[154,-0.0871],[155,-0.0002],[156,-0.0052],[157,0.0005],[158,-0.0018],[159,0.0019],[160,0.0036],[161,-0.0041],[162,-0.001],[163,0],[164,-0.0002],[165,0.0002],[167,0],[168,0.0002],[172,-0.0001],[174,0.0002],[177,0],[178,-0.0002],[179,0.0005],[180,-0.0003],[181,-0.0004],[182,-0.0001],[183,-0.0002],[184,0],[185,-0.0003],[187,0.0003],[188,-0.0003],[189,-0.0003],[190,0.0004],[191,0.0004],[192,-0.0001],[193,-0.0001],[194,-0.0005],[196,0.0001],[198,0],[199,0.0001],[200,-0.0002],[201,0.0002],[202,0.0001],[203,-0.0002],[204,0.0001],[206,-0.0001],[208,0.0001],[209,-0.0002],[210,-0.0002],[211,0],[212,-0.0002],[214,-0.0001],[215,0.0002],[216,0],[217,0.0001],[220,0],[226,0],[228,-0.0004],[231,0.0006],[236,0.0006],[237,-0.0001],[238,0.0002],[239,-0.0005],[240,0.0004],[241,0],[242,0.0008],[243,0],[246,0],[247,-0.0002],[248,0.0003],[250,0],[252,0.0003],[253,-0.0001],[254,0.0001],[258,-0.0001],[259,-0.0004],[262,-0.0004],[263,0],[264,0],[265,0.0005],[266,0.0005],[267,-0.0006],[268,-0.0002],[269,-0.0002],[270,0.0002],[271,-0.0005],[272,0.0003],[273,-0.0008],[274,0],[275,-0.0002],[277,0.0011],[278,-0.0009],[285,0.0009],[288,0.0009],[290,0.0004],[292,0.0002],[293,-0.0002],[294,0.0006],[295,-0.0001],[296,-0.0005],[297,0.0003],[298,-0.0003],[299,0.0003],[302,-0.0004],[303,0.0005],[306,0]]
       }
       index_02_r_ctrl.Rotation.Y {
        type=MovieSceneScriptingFloatChannel
        default=0
        thinned_from=307
        keys [[0,-0.0008],[5,-0.0003],[7,0.0005],[13,0.0007],[16,-0.0003],[17,-0.0001],[19,-0.0001],[20,-0.0004],[21,-0.0001],[22,-0.0001],[23,0.0005],[24,-0.0006],[25,-0.0002],[26,0.0003],[27,-0.0007],[28,-0.0001],[30,0.0005],[31,-0.0002],[33,0.0004],[34,0.0014],[36,0.0014],[37,0.0012],[39,0.0013],[40,0.001],[41,0.001],[42,0.0005],[43,0.0003],[44,0.0004],[47,-0.0004],[50,-0.0004],[51,-0.0022],[56,-0.0018],[60,-0.0009],[62,0.0001],[63,0.0208],[64,0.0844],[65,0.178],[66,0.2849],[67,0.4167],[68,0.6368],[69,1.037],[70,1.3534],[71,1.4869],[72,1.6124],[73,1.7736],[74,1.9884],[75,2.1843],[76,2.2961],[77,2.3811],[78,2.4318],[79,2.4623],[80,2.5011],[81,2.514],[82,2.5479],[83,2.5912],[104,2.5932],[113,2.5945],[116,2.5884],[120,2.5937],[121,2.5888],[122,2.5948],[125,2.5947],[126,2.5894],[131,2.594],[145,2.5928],[146,2.5147],[147,2.3177],[148,2.1477],[149,1.9887],[150,1.8063],[151,1.4282],[152,0.9093],[153,0.3995],[154,0.0667],[155,0.0005],[156,0.0022],[157,0.0059],[158,-0.004],[159,-0.0013],[160,0],[161,-0.0001],[162,0.0024],[163,0],[165,0.0003],[166,-0.0002],[167,0.0002],[168,-0.0001],[169,0.0001],[170,0.0001],[171,-0.0001],[172,-0.0001],[173,0.0003],[174,-0.0001],[175,0.0002],[176,0],[178,-0.0001],[179,0.0005],[180,-0.0001],[182,0],[183,-0.0005],[184,0.0001],[186,0.0001],[189,-0.0005],[190,0.0002],[193,0.0003],[194,0],[195,0.0001],[196,-0.0002],[197,0.0002],[200,-0.0002],[202,0],[203,-0.0001],[204,0.0001],[206,-0.0001],[207,0.0001],[208,0.0001],[209,-0.0002],[210,0.0001],[211,-0.0002],[213,0.0001],[214,0.0001],[215,-0.0001],[216,-0.0001],[217,0.0002],[218,-0.0002],[219,-0.0003],[221,0.0002],[223,-0.0002],[224,-0.0001],[225,0.0002],[226,0.0002],[227,0.0006],[228,-0.0013],[231,-0.0002],[236,-0.0002],[237,-0.0009],[238,0.0009],[239,0.0002],[242,-0.0007],[243,0.0001],[244,-0.0003],[245,0.0002],[247,0.0001],[248,-0.0002],[249,0.0003],[250,-0.0003],[251,-0.0001],[252,-0.0002],[253,0.0001],[255,-0.0003],[258,-0.0003],[262,-0.0004],[263,-0.0001],[264,-0.0001],[265,0.0002],[267,0.0003],[268,0.0006],[269,0.0006],[270,0.001],[271,-0.0012],[272,-0.0005],[273,-0.0004],[274,0.0003],[276,0.0011],[277,-0.0008],[278,-0.0011],[285,0.0006],[288,0.0006],[290,0.0001],[291,0.0004],[293,-0.0001],[294,-0.0002],[295,0.0004],[296,0.0001],[297,0],[298,-0.0004],[299,0.0005],[302,-0.0001],[303,0],[305,-0.0003],[306,0]]
       }
       index_02_r_ctrl.Rotation.Z {
        type=MovieSceneScriptingFloatChannel
        default=0
        thinned_from=307
        keys [[0,14.6938],[1,14.6938],[2,14.6063],[3,14.6063],[4,14.5189],[7,13.994],[8,13.9064],[13,13.9064],[14,13.994],[16,14.344],[18,14.6007],[19,14.6007],[20,14.7191],[22,14.4822],[23,14.2452],[24,13.8895],[25,13.6522],[27,12.9398],[30,12.2264],[31,11.8693],[32,11.6311],[34,11.3792],[36,11.3792],[37,11.4847],[38,11.4847],[40,11.6958],[41,11.6958],[42,11.9067],[47,12.4337],[50,12.4337],[51,12.4747],[52,12.4747],[53,12.4203],[54,12.4203],[56,12.3116],[57,12.2029],[58,12.1485],[60,11.931],[62,11.4956],[63,11.1143],[64,10.6779],[65,10.1863],[66,9.5296],[67,8.8506],[68,7.6455],[69,4.9625],[70,2.6943],[71,2.2529],[72,3.3768],[73,5.8777],[74,8.1873],[75,9.716],[76,10.4762],[77,11.153],[78,12.0012],[79,12.5947],[80,12.677],[81,9.6906],[82,6.8554],[83,3.928],[84,-2.7142],[85,-12.4275],[86,-25.0338],[87,-36.1814],[88,-44.1018],[89,-47.6173],[90,-47.6767],[92,-17.5005],[93,-9.5007],[94,-4.6772],[95,-3.2271],[96,-3.4139],[97,-8.2632],[98,-26.9375],[99,-39.0835],[100,-45.4799],[101,-47.9165],[102,-45.8252],[103,-30.4478],[104,-15.4919],[105,-3.9325],[106,0.8169],[107,2.4455],[108,3.1668],[109,1.9034],[110,-5.964],[111,-24.4272],[112,-40.5165],[113,-45.2503],[114,-47.5656],[115,-48.0338],[116,-37.9838],[117,-21.9303],[118,-11.5895],[119,-4.7674],[120,-2.1486],[121,-7.2436],[122,-18.9505],[123,-33.9652],[124,-43.1164],[125,-45.7568],[126,-47.6167],[127,-47.6167],[128,-35.4839],[129,-21.2252],[130,-10.0771],[131,-5.1205],[132,-3.5684],[133,-2.3667],[134,-2.5381],[135,-7.7257],[136,-23.1219],[137,-39.4359],[138,-44.4636],[139,-46.6154],[140,-47.9241],[141,-42.0318],[142,-27.7497],[143,-15.3401],[144,-6.507],[145,-3.0528],[146,0.03],[147,3.1022],[148,6.3691],[149,9.0725],[150,11.1371],[151,15.0603],[152,19.7468],[153,23.4345],[154,25.3576],[155,25.2707],[156,23.6127],[158,19.5859],[159,18.2682],[160,17.6524],[161,16.8597],[162,15.7131],[163,14.2186],[164,12.0939],[165,9.1279],[166,6.3819],[167,4.0995],[168,3.2552],[169,3.7379],[170,4.9419],[171,6.741],[172,8.6515],[173,10.4352],[174,11.9756],[175,13.3933],[176,14.925],[177,16.3351],[178,17.6249],[179,18.2961],[180,18.5857],[181,18.0064],[182,16.9912],[183,15.8291],[184,14.5189],[185,13.0596],[186,11.7427],[187,11.0096],[188,10.5691],[190,9.9813],[191,9.9813],[192,9.8343],[193,9.8343],[194,9.9813],[195,10.054],[197,10.4623],[198,10.7682],[199,11.4814],[200,12.7015],[202,14.9312],[203,16.1436],[205,18.3602],[206,19.3653],[207,20.2687],[208,21.0708],[209,21.7719],[210,22.3724],[211,22.7664],[212,23.0603],[213,23.2561],[214,23.2561],[215,23.0603],[216,22.6684],[217,21.7862],[218,20.3137],[219,18.6414],[220,17.1624],[221,15.7789],[222,14.5905],[223,13.4989],[224,12.5046],[225,11.7078],[226,11.1094],[227,10.6749],[228,10.3929],[231,9.9696],[236,9.9696],[240,11.0976],[241,11.2384],[242,11.5199],[244,11.9896],[245,12.1402],[246,12.1402],[247,12.2154],[248,12.4412],[249,13.0428],[251,14.6937],[252,15.218],[253,15.5174],[254,15.667],[255,15.7418],[258,15.7418],[259,15.7942],[262,15.7942],[263,15.6371],[264,15.6371],[265,15.48],[266,15.48],[268,15.1656],[269,15.1656],[271,14.8511],[272,14.5364],[273,14.379],[274,14.064],[276,13.6438],[285,12.697],[288,12.697],[290,12.9075],[291,13.0586],[292,13.1265],[294,13.3979],[295,13.6014],[297,13.8726],[298,14.0758],[299,14.2113],[300,14.2791],[301,14.4145],[305,14.6853],[306,0]]
       }
       index_02_r_ctrl.Scale.X {type=MovieSceneScriptingFloatChannel, default=1, const [1,307]}
       index_02_r_ctrl.Scale.Y {type=MovieSceneScriptingFloatChannel, default=1, const [1,307]}
       index_02_r_ctrl.Scale.Z {type=MovieSceneScriptingFloatChannel, default=1, const [1,307]}
       index_03_r_ctrl.Location.X {type=MovieSceneScriptingFloatChannel, default=0, thinned_from=307, keys [[0,0],[306,0]]}
       index_03_r_ctrl.Location.Y {type=MovieSceneScriptingFloatChannel, default=0, thinned_from=307, keys [[0,0],[306,0]]}
       index_03_r_ctrl.Location.Z {type=MovieSceneScriptingFloatChannel, default=0, thinned_from=307, keys [[0,0],[306,0]]}
       index_03_r_ctrl.Rotation.X {
        type=MovieSceneScriptingFloatChannel
        default=0
        thinned_from=307
        keys [[0,0.0001],[2,0.0001],[3,0.0003],[4,0.0003],[6,0.0008],[7,0.0008],[8,0.0011],[13,0.0011],[14,-0.0004],[15,-0.0006],[16,-0.0006],[17,0.0006],[18,-0.0011],[21,-0.0011],[22,0.0006],[25,0.0006],[26,-0.0011],[27,0.0006],[38,0.0007],[39,0],[41,0],[42,-0.0003],[43,-0.0003],[44,-0.0005],[45,-0.0005],[46,-0.0008],[56,-0.0008],[57,0.0007],[62,0.0007],[63,-0.0009],[64,-0.001],[65,0.0007],[66,0.0008],[67,0.0002],[69,0.0003],[70,-0.0007],[71,0.0003],[79,0.0001],[80,-0.0009],[81,0.0001],[82,0.0001],[84,0.0005],[85,-0.0006],[86,-0.0007],[88,0.0009],[89,0.0005],[91,0.0004],[93,0.0006],[95,0.0005],[96,-0.0007],[98,-0.0007],[99,-0.0001],[101,-0.0001],[102,-0.0003],[103,0.0003],[104,-0.0006],[105,-0.0007],[106,0.0005],[109,0.0004],[110,0.0006],[111,-0.0005],[112,-0.0002],[113,-0.0002],[114,0.0006],[115,-0.0004],[116,0.0001],[117,0.0002],[118,-0.0004],[120,-0.0002],[121,0.0008],[122,-0.0006],[123,0.0005],[124,-0.0001],[125,-0.0001],[126,0.0004],[129,0.0004],[130,-0.0003],[131,0.0005],[132,-0.0001],[133,-0.0002],[134,0.0004],[135,-0.0001],[137,-0.0001],[138,0.0001],[141,-0.0002],[143,-0.0001],[144,0.0005],[145,0.0004],[147,-0.0008],[148,-0.0008],[149,0.0003],[151,0],[152,0.0009],[153,-0.0005],[154,0.0005],[157,0.0007],[158,-0.0003],[159,-0.0002],[160,0.0009],[161,-0.0001],[162,0.001],[163,0.0001],[165,-0.0003],[166,0.001],[168,0.0008],[169,0.0008],[170,-0.0006],[176,0.0001],[177,-0.0014],[178,0.0003],[179,0.0007],[180,-0.0013],[181,0.0008],[182,0.0008],[183,-0.0012],[184,0.0009],[185,0.001],[186,-0.001],[189,-0.0009],[190,0.0011],[194,0.0011],[196,-0.0003],[197,0.0003],[198,-0.0004],[199,-0.0005],[201,0.0004],[205,-0.0004],[206,0.0001],[208,-0.0002],[209,0.0004],[210,0.0003],[211,0.0013],[212,-0.0016],[219,-0.0013],[220,0.0016],[221,-0.0012],[222,-0.0011],[223,0.0017],[224,-0.001],[225,-0.001],[226,0.0018],[227,0.0002],[228,-0.0006],[230,-0.0004],[231,0.0006],[236,0.0006],[237,-0.0005],[238,-0.0007],[240,0.0007],[241,-0.0004],[242,0.0004],[243,-0.0006],[244,0.0005],[246,0.0006],[247,-0.0004],[248,0.0007],[249,-0.0001],[251,0.0004],[252,-0.0004],[263,-0.0002],[264,0.0009],[265,0.0009],[266,-0.0001],[271,-0.0001],[272,0.0009],[273,-0.0001],[274,-0.0001],[275,-0.0003],[277,-0.0003],[278,0.0005],[280,0.0005],[281,-0.0003],[282,-0.0003],[283,0.0005],[285,0.0005],[286,-0.0003],[289,-0.0002],[290,0.0005],[291,-0.0003],[292,0.0001],[293,0],[294,0.0003],[298,-0.0002],[299,0.0001],[305,-0.0002],[306,0]]
       }
       index_03_r_ctrl.Rotation.Y {
        type=MovieSceneScriptingFloatChannel
        default=0
        thinned_from=307
        keys [[0,0.0001],[5,0],[6,0.0002],[13,0.0002],[14,-0.0002],[16,-0.0001],[17,0.0001],[18,-0.0003],[21,-0.0003],[22,0.0001],[25,0],[26,-0.0002],[27,0.0002],[38,0.0001],[45,0],[46,-0.0003],[56,-0.0003],[57,0.0001],[60,0],[61,0.0003],[62,0.0002],[63,-0.0004],[65,0.0002],[67,0],[69,0],[70,-0.0002],[71,0.0003],[74,0],[75,0.0001],[76,-0.0001],[78,0.0001],[79,0],[80,-0.0002],[81,0.0001],[83,0],[84,0.0002],[85,-0.0002],[86,-0.0003],[88,0.0007],[90,0.0008],[91,0.0001],[92,0.0005],[94,0.0001],[95,0.0003],[96,-0.0002],[98,-0.0003],[99,0.0003],[101,-0.0006],[103,0.0004],[104,-0.0003],[110,0.0001],[111,-0.0007],[112,0.0004],[113,-0.0001],[114,0.0007],[115,-0.0004],[116,0.0005],[118,-0.0002],[119,0],[120,-0.0004],[121,0.0001],[122,-0.0007],[123,0.0005],[125,-0.0005],[126,0.0006],[127,0.0006],[130,-0.0004],[131,0],[132,-0.0002],[134,0],[135,-0.0001],[138,0.0003],[142,-0.0002],[143,0.0001],[148,-0.0003],[157,0.0001],[158,-0.0001],[160,0.0001],[161,-0.0001],[163,0.0001],[165,-0.0002],[166,0.0003],[167,0.0001],[169,0.0003],[170,-0.0002],[173,0],[175,-0.0001],[176,0],[177,-0.0003],[178,0.0001],[181,-0.0001],[182,0.0002],[183,-0.0001],[184,0.0002],[185,0.0002],[186,-0.0002],[187,-0.0004],[189,-0.0002],[190,0.0001],[195,0.0001],[196,-0.0001],[197,0.0001],[199,-0.0001],[201,0.0001],[211,-0.0001],[215,0],[217,-0.0002],[219,0],[220,0.0003],[221,-0.0002],[222,-0.0001],[223,0.0003],[224,-0.0001],[225,-0.0002],[226,0.0004],[227,0],[230,-0.0002],[231,0.0002],[236,0.0002],[237,-0.0002],[239,-0.0002],[240,0.0001],[241,-0.0001],[242,0],[243,-0.0001],[246,0.0001],[247,-0.0002],[248,0.0001],[251,0.0001],[252,-0.0002],[253,0],[261,-0.0001],[263,0],[264,0.0002],[269,-0.0001],[272,0.0001],[273,0],[274,0.0001],[275,-0.0002],[280,0.0001],[281,-0.0002],[285,0.0001],[289,-0.0001],[290,0],[291,-0.0001],[306,0]]
       }
       index_03_r_ctrl.Rotation.Z {
        type=MovieSceneScriptingFloatChannel
        default=0
        thinned_from=307
        keys [[0,18.8639],[2,18.8639],[3,18.6512],[4,18.6512],[6,18.2257],[7,18.2257],[8,18.0129],[13,18.0129],[15,18.4384],[16,18.4384],[17,18.5448],[18,18.793],[21,18.793],[26,17.5516],[27,17.0547],[33,15.5624],[38,15.5624],[39,15.8113],[41,15.8113],[42,16.0601],[43,16.0601],[44,16.3088],[45,16.3088],[46,16.5575],[56,16.5575],[57,16.3088],[59,16.3088],[63,15.3135],[64,14.8154],[66,13.3194],[67,12.5069],[68,11.078],[69,7.8919],[70,5.3308],[71,4.6886],[72,5.6516],[73,8.2112],[74,10.4421],[75,11.7135],[77,12.6655],[79,13.9332],[80,14.2498],[81,14.0915],[82,14.0915],[83,13.9396],[84,8.6362],[85,-2.7054],[86,-15.1916],[87,-26.1434],[88,-33.6843],[89,-37.0095],[90,-37.0095],[91,-21.9521],[92,-6.8157],[93,1.1699],[94,5.9681],[95,7.4942],[96,7.3037],[97,1.7482],[98,-17.0179],[99,-28.9803],[100,-34.8569],[101,-37.0814],[102,-35.0781],[103,-19.7446],[104,-4.7199],[105,6.8611],[106,11.4214],[107,13.1237],[108,13.879],[109,12.1785],[110,3.4172],[111,-14.4485],[112,-30.0553],[113,-34.636],[114,-36.8577],[115,-37.3054],[116,-27.2566],[117,-11.2306],[118,-0.7864],[119,6.0294],[120,8.5623],[121,3.4848],[122,-8.2393],[123,-23.3354],[124,-32.408],[125,-34.9382],[126,-36.9086],[127,-36.9086],[128,-24.8211],[129,-10.5242],[130,0.5848],[131,5.5438],[132,7.1756],[133,8.3749],[134,8.1181],[135,2.957],[136,-12.3738],[137,-28.7356],[138,-33.8105],[139,-35.8993],[140,-37.2046],[141,-31.3544],[142,-17.0066],[143,-4.6292],[144,4.3383],[145,7.9469],[146,9.1447],[148,9.8016],[149,10.7724],[150,13.1936],[151,17.6943],[152,22.8159],[153,27.1262],[154,29.3594],[155,29.3594],[156,27.7644],[157,25.6902],[158,23.7744],[159,22.3366],[161,21.0575],[162,19.7774],[163,18.3005],[164,16.2062],[165,13.4066],[166,10.3626],[167,8.2472],[168,7.3047],[169,7.7761],[170,9.1882],[171,10.8316],[173,14.5741],[174,16.2062],[176,18.9978],[177,20.6232],[178,21.7831],[179,22.4142],[180,22.7096],[181,22.1188],[182,20.9364],[183,20.0491],[185,17.0868],[186,15.8996],[188,14.7109],[190,14.116],[194,14.116],[195,14.1698],[197,14.578],[198,14.986],[199,15.6996],[202,19.0567],[203,20.2751],[205,22.5061],[206,23.5193],[207,24.4307],[208,25.2406],[210,26.455],[212,27.2956],[215,27.2956],[216,26.8892],[217,26.0761],[220,21.1922],[223,17.5196],[226,15.0645],[227,14.7157],[231,14.1176],[236,14.1176],[239,15.0146],[240,15.1641],[241,15.4628],[242,15.6122],[243,15.9558],[245,16.2446],[246,16.2446],[248,16.5334],[250,17.9762],[251,18.841],[252,19.4172],[253,19.7052],[254,19.8491],[258,19.8491],[259,19.9835],[261,19.9835],[262,19.8347],[263,19.8347],[264,19.686],[265,19.686],[266,19.5372],[267,19.5372],[271,18.9419],[272,18.6441],[273,18.4952],[274,18.1973],[275,18.0649],[276,17.8331],[277,17.7172],[278,17.4854],[279,17.3694],[280,17.3694],[284,16.9056],[285,16.9056],[286,16.7896],[287,16.7896],[288,16.9056],[289,16.9056],[290,17.0215],[291,17.199],[292,17.2711],[296,17.8481],[297,18.0643],[299,18.3526],[300,18.4247],[301,18.5688],[304,18.785],[305,18.785],[306,0]]
       }
       index_03_r_ctrl.Scale.X {type=MovieSceneScriptingFloatChannel, default=1, const [1,307]}
       index_03_r_ctrl.Scale.Y {type=MovieSceneScriptingFloatChannel, default=1, const [1,307]}
       index_03_r_ctrl.Scale.Z {type=MovieSceneScriptingFloatChannel, default=1, const [1,307]}
       middle_metacarpal_r_ctrl.Location.X {type=MovieSceneScriptingFloatChannel, default=0, thinned_from=307, keys [[0,0],[306,0]]}
       middle_metacarpal_r_ctrl.Location.Y {type=MovieSceneScriptingFloatChannel, default=0, thinned_from=307, keys [[0,0],[306,0]]}
       middle_metacarpal_r_ctrl.Location.Z {type=MovieSceneScriptingFloatChannel, default=0, thinned_from=307, keys [[0,0],[306,0]]}
       middle_metacarpal_r_ctrl.Rotation.X {type=MovieSceneScriptingFloatChannel, default=0, thinned_from=307, keys [[0,0],[56,-0.0001],[146,-0.0001],[242,-0.0001],[306,0]]}
       middle_metacarpal_r_ctrl.Rotation.Y {
        type=MovieSceneScriptingFloatChannel
        default=0
        thinned_from=307
        keys [[0,0],[56,0.0001],[57,-0.0002],[59,0.0001],[60,-0.0002],[61,0],[68,0],[69,0.0001],[70,-0.0001],[71,0.0003],[72,0],[74,0.0001],[75,0.0003],[76,-0.0001],[79,0.0002],[80,-0.0576],[81,-1.138],[82,-2.5305],[83,-3.5582],[84,-3.5998],[145,-3.5998],[146,-3.098],[148,-0.4994],[149,0],[154,0],[155,0.0003],[156,-0.0001],[157,0],[158,0.0008],[159,0.0008],[160,0.0003],[161,0.0007],[162,0.0007],[163,0],[229,0.0001],[236,-0.0001],[238,0.0001],[240,-0.0001],[241,0.0001],[242,-0.0001],[306,0]]
       }
       middle_metacarpal_r_ctrl.Rotation.Z {
        type=MovieSceneScriptingFloatChannel
        default=0
        thinned_from=307
        keys [[0,0],[1,0],[2,-0.0207],[3,-0.0714],[4,-0.1768],[5,-0.3385],[6,-0.5078],[7,-0.6338],[8,-0.7203],[9,-0.7805],[10,-0.8049],[12,-0.8049],[13,-0.788],[14,-0.6451],[15,-0.4721],[16,-0.3197],[17,-0.1734],[18,-0.0716],[19,-0.0183],[20,0],[21,-0.0293],[22,-0.1186],[23,-0.2376],[24,-0.3581],[25,-0.4896],[27,-0.7683],[28,-0.8967],[29,-1.0079],[30,-1.1222],[31,-1.2333],[32,-1.3319],[33,-1.4086],[34,-1.4503],[35,-1.4503],[36,-1.4277],[37,-1.3854],[38,-1.3261],[39,-1.25],[40,-1.1654],[41,-1.0554],[42,-0.9115],[43,-0.7451],[44,-0.5843],[45,-0.4517],[46,-0.3558],[47,-0.2881],[48,-0.2373],[49,-0.2034],[50,-0.178],[51,-0.1598],[52,-0.1598],[53,-0.1446],[56,-0.1446],[57,-0.1293],[58,-0.109],[59,-0.0785],[60,-0.0429],[61,-0.0124],[62,0],[68,0],[69,-1.547],[70,-3.7394],[71,-4.9678],[72,-5.1453],[73,-4.8982],[74,-3.127],[75,-0.5445],[76,1.6772],[77,4.2394],[78,6.7775],[79,8.5856],[80,9.0272],[81,5.4619],[82,0.8648],[83,-2.5288],[84,-2.666],[145,-2.666],[146,-2.2963],[147,-1.3351],[148,-0.3706],[149,0],[154,0],[155,0.1103],[156,0.9185],[157,2.07],[158,3.0388],[159,3.3163],[160,2.9635],[161,2.1876],[162,1.0454],[163,-0.4284],[164,-2.6167],[166,-8.3407],[167,-10.5408],[168,-11.4196],[169,-10.9469],[170,-9.6999],[171,-7.9749],[172,-6.0493],[173,-4.1993],[174,-2.6825],[175,-1.2623],[176,0.266],[177,1.7114],[178,2.9059],[179,3.6737],[180,3.8462],[181,3.2837],[182,2.2606],[183,1.1282],[184,-0.1587],[185,-1.6644],[186,-2.9269],[187,-3.6516],[188,-4.1075],[189,-4.4301],[190,-4.6343],[191,-4.7462],[192,-4.787],[193,-4.787],[194,-4.7354],[195,-4.6302],[196,-4.457],[197,-4.1951],[198,-3.8572],[199,-3.1603],[201,-0.8369],[202,0.2491],[203,1.272],[204,2.122],[205,2.8199],[206,3.3404],[207,3.6662],[208,3.8144],[209,3.8144],[210,3.7848],[211,3.6951],[212,3.5667],[213,3.4098],[214,3.1923],[215,2.9035],[216,2.5043],[217,1.9234],[218,1.0968],[219,0.2528],[220,-0.4487],[221,-1.1109],[222,-1.7302],[223,-2.2748],[224,-2.7375],[225,-3.1326],[226,-3.4529],[227,-3.6782],[228,-3.8068],[229,-3.8824],[230,-3.9429],[231,-3.9808],[233,-4.0186],[235,-4.0186],[236,-3.9959],[237,-3.89],[239,-3.6177],[240,-3.4891],[241,-3.3454],[242,-3.2092],[243,-3.1226],[244,-3.109],[245,-3.1702],[246,-3.2553],[247,-3.3267],[248,-3.3267],[249,-2.8096],[250,-1.8194],[251,-0.6929],[252,0.2432],[253,0.9549],[254,1.6156],[255,2.2049],[256,2.7024],[257,3.108],[258,3.4387],[259,3.6703],[260,3.7909],[261,3.7909],[262,3.6828],[263,3.5041],[264,3.2672],[265,2.9847],[266,2.6606],[267,2.2494],[268,1.7843],[269,1.3151],[270,0.8917],[271,0.5389],[272,0.2276],[273,-0.0629],[274,-0.3284],[275,-0.5681],[276,-0.7875],[277,-0.9787],[278,-1.1386],[279,-1.2702],[280,-1.3862],[281,-1.4927],[282,-1.6024],[284,-1.8437],[285,-1.9471],[286,-2.0067],[287,-2.0067],[288,-1.9471],[289,-1.8562],[290,-1.7497],[291,-1.6385],[292,-1.5162],[293,-1.3802],[295,-1.096],[296,-0.9428],[297,-0.7792],[298,-0.6207],[299,-0.476],[300,-0.3606],[301,-0.2625],[302,-0.1746],[303,-0.0988],[304,-0.042],[305,0],[306,0]]
       }
       middle_metacarpal_r_ctrl.Scale.X {type=MovieSceneScriptingFloatChannel, default=1, const [1,307]}
       middle_metacarpal_r_ctrl.Scale.Y {type=MovieSceneScriptingFloatChannel, default=1, const [1,307]}
       middle_metacarpal_r_ctrl.Scale.Z {type=MovieSceneScriptingFloatChannel, default=1, const [1,307]}
       middle_01_r_ctrl.Location.X {type=MovieSceneScriptingFloatChannel, default=0, thinned_from=307, keys [[0,0],[306,0]]}
       middle_01_r_ctrl.Location.Y {type=MovieSceneScriptingFloatChannel, default=0, thinned_from=307, keys [[0,0],[306,0]]}
       middle_01_r_ctrl.Location.Z {type=MovieSceneScriptingFloatChannel, default=0, thinned_from=307, keys [[0,0],[306,0]]}
       middle_01_r_ctrl.Rotation.X {
        type=MovieSceneScriptingFloatChannel
        default=0
        thinned_from=307
        keys [[0,-19.1212],[62,-19.1204],[63,-18.9831],[64,-18.5158],[65,-17.8248],[66,-17.0417],[67,-16.0715],[68,-14.4309],[69,-11.5536],[70,-9.1255],[71,-8.2412],[72,-7.2912],[73,-6.091],[74,-4.5098],[75,-3.0524],[76,-2.121],[77,-1.5514],[78,-1.2455],[79,-0.9308],[80,-0.8402],[81,-1.6537],[82,-2.7037],[83,-3.4801],[84,-3.5113],[145,-3.5114],[146,-3.5915],[147,-3.8013],[148,-3.899],[149,-4.3253],[150,-5.7563],[151,-8.561],[152,-12.3807],[153,-16.1348],[154,-18.6486],[155,-19.1217],[208,-19.1213],[209,-19.2083],[210,-19.5647],[211,-20.0607],[212,-20.6276],[213,-21.2708],[214,-21.9622],[215,-22.8008],[216,-23.7268],[217,-24.3924],[218,-24.7068],[219,-24.8646],[220,-24.9613],[223,-25.0604],[224,-25.2047],[226,-25.721],[227,-25.9171],[235,-25.9624],[236,-25.9046],[237,-25.7154],[239,-25.1958],[240,-25.015],[242,-24.7933],[243,-24.5298],[244,-23.9796],[245,-23.0749],[246,-21.9978],[247,-20.9951],[248,-20.2959],[249,-19.9283],[250,-19.7457],[254,-19.3495],[258,-19.1563],[305,-19.1211],[306,0]]
       }
       middle_01_r_ctrl.Rotation.Y {
        type=MovieSceneScriptingFloatChannel
        default=0
        thinned_from=307
        keys [[0,-0.0005],[1,-0.0005],[2,0],[3,0.0003],[4,-0.0004],[5,0.0004],[6,0.0003],[7,-0.0002],[8,0.0001],[9,0.0006],[10,-0.0001],[12,-0.0001],[13,-0.0006],[14,0.0006],[15,0.0005],[16,-0.0001],[17,0.0001],[18,-0.0001],[19,-0.0007],[21,0.0007],[22,-0.0001],[23,-0.0001],[24,0.0003],[25,0],[27,0],[28,-0.0003],[29,0.0009],[30,0.0001],[31,0.0003],[32,0.0001],[34,0.0005],[36,0.0005],[37,-0.0006],[38,-0.0006],[39,0.001],[41,-0.0006],[42,0.0004],[43,-0.0004],[44,0.0007],[45,-0.001],[46,-0.0004],[47,-0.0006],[49,0.0007],[50,0.0007],[51,-0.0003],[52,-0.0003],[53,0.0002],[54,-0.0003],[56,0.0001],[58,-0.0001],[59,0.0002],[61,-0.0001],[62,0.0003],[63,-0.0003],[64,0.0001],[65,-0.0001],[66,0.0003],[67,-0.0005],[68,-0.0003],[69,-0.0005],[70,-0.0002],[72,-0.0004],[73,0.0002],[74,0.0003],[76,-0.0001],[78,0.0004],[79,-0.0005],[80,0.1181],[81,2.3342],[82,5.189],[83,7.298],[84,6.3675],[85,4.5387],[86,3.3217],[87,2.5804],[88,2.3226],[89,2.1984],[90,2.2181],[91,3.7473],[92,5.2853],[93,6.0944],[94,6.5856],[95,6.7329],[96,6.5628],[97,5.3132],[98,3.2227],[99,2.4217],[100,2.2791],[101,2.1883],[102,2.4087],[103,3.9712],[104,5.4863],[105,6.6649],[106,7.1409],[107,7.3114],[108,7.3813],[109,6.692],[110,4.8478],[111,3.5408],[112,2.6656],[113,2.353],[114,2.2148],[115,2.183],[116,3.2095],[117,4.8305],[118,5.8868],[119,6.5784],[120,6.8362],[121,5.3737],[122,3.2719],[123,2.6761],[124,2.3943],[125,2.2789],[126,2.2036],[127,2.2277],[128,3.4542],[129,4.902],[130,6.0295],[131,6.5383],[132,6.6984],[133,6.823],[134,6.5291],[135,5.0008],[136,3.6939],[137,2.7442],[138,2.4119],[139,2.2778],[140,2.1883],[141,2.7887],[142,4.2435],[143,5.5034],[144,6.4319],[145,6.8105],[146,5.8933],[148,0.9467],[149,-0.0004],[150,0.0004],[152,0.0002],[153,-0.0003],[154,-0.0001],[155,-0.0005],[156,0.0003],[157,0.0004],[158,0.0001],[159,-0.0006],[160,-0.0005],[161,-0.0001],[163,0],[164,0.0006],[165,-0.0005],[166,0.0004],[167,0.0002],[169,0.0003],[170,0.0006],[171,-0.0005],[172,-0.0005],[173,0.0004],[175,-0.0001],[179,0.0001],[180,-0.0001],[181,0.0007],[182,0],[183,0.0002],[187,-0.0003],[189,0.0001],[191,0.0001],[192,-0.0004],[193,-0.0004],[194,-0.0002],[195,-0.0003],[196,-0.0007],[197,-0.0001],[198,-0.0007],[200,-0.0007],[201,-0.0002],[202,-0.0155],[203,-0.1753],[204,-0.4861],[205,-0.9311],[206,-1.4886],[207,-2.1332],[208,-2.8465],[209,-3.6022],[211,-5.1577],[212,-5.9137],[213,-6.6253],[214,-7.2706],[215,-7.8284],[216,-8.2736],[217,-8.5861],[218,-8.742],[243,-8.7598],[244,-8.7439],[245,-8.5846],[246,-8.2729],[247,-7.8285],[248,-7.271],[249,-6.6251],[250,-5.913],[251,-5.1582],[253,-3.6019],[254,-2.8448],[255,-2.1332],[256,-1.4878],[257,-0.9312],[258,-0.4855],[259,-0.1734],[260,-0.0177],[261,0.0013],[262,0.0006],[263,0.0006],[264,0.0002],[265,0.0008],[266,0.0005],[267,-0.0009],[268,0.0003],[269,-0.0012],[270,-0.0014],[271,-0.001],[272,0],[273,0.0001],[274,-0.0009],[275,0.0002],[276,0.0002],[277,0.0008],[278,-0.0002],[279,-0.0005],[280,0.0005],[281,0],[283,-0.0002],[284,-0.0001],[285,0.0004],[286,-0.0007],[287,-0.0007],[288,0.0006],[289,-0.0004],[290,-0.0003],[291,0],[292,-0.0001],[293,0.0003],[294,-0.0005],[295,-0.0003],[296,-0.0003],[297,-0.0006],[298,0.0004],[299,-0.0003],[300,-0.0005],[301,0],[302,-0.0004],[303,-0.0003],[304,0.0004],[305,-0.0006],[306,0]]
       }
       middle_01_r_ctrl.Rotation.Z {
        type=MovieSceneScriptingFloatChannel
        default=0
        thinned_from=307
        keys [[0,-8.5755],[2,-8.5924],[3,-8.6427],[4,-8.7516],[6,-9.0788],[7,-9.2047],[8,-9.297],[9,-9.3558],[10,-9.3809],[12,-9.3809],[13,-9.364],[14,-9.2215],[15,-9.0453],[17,-8.7457],[18,-8.648],[19,-8.5958],[20,-8.5764],[21,-8.6286],[22,-8.8108],[23,-9.0518],[24,-9.319],[25,-9.6189],[27,-10.2648],[28,-10.5585],[31,-11.3297],[32,-11.5585],[33,-11.7352],[34,-11.828],[36,-11.828],[39,-11.67],[41,-11.5119],[44,-11.1122],[46,-10.9228],[48,-10.8176],[49,-10.7862],[52,-10.7729],[55,-10.8825],[57,-11.0215],[58,-11.1273],[59,-11.2704],[63,-11.9458],[65,-12.2641],[68,-12.8488],[69,-14.0809],[70,-14.7746],[71,-13.7871],[72,-11.4353],[73,-8.8401],[75,-1.8643],[76,0.3774],[77,2.6874],[78,5.0036],[79,6.6604],[80,7.0231],[81,2.9445],[82,-1.5783],[83,-5.5505],[84,-9.8423],[85,-13.9068],[86,-17.2839],[87,-19.2992],[88,-20.0111],[90,-20.4888],[91,-22.4397],[92,-21.3798],[93,-13.4126],[94,-9.1228],[95,-7.9834],[96,-8.4451],[97,-11.8518],[98,-17.5478],[99,-19.7312],[100,-20.1193],[101,-20.3458],[103,-20.5062],[104,-17.1049],[105,-9.7873],[106,-7.4162],[107,-6.5658],[108,-6.2174],[109,-8.0924],[110,-13.1192],[111,-16.6818],[112,-19.0672],[113,-19.918],[114,-20.2956],[115,-20.3818],[116,-17.5834],[117,-13.1656],[118,-10.2868],[119,-8.4032],[120,-7.7002],[121,-11.686],[122,-17.4146],[123,-19.0377],[124,-19.8068],[125,-20.1207],[126,-20.3257],[127,-20.26],[128,-16.9169],[129,-12.9716],[130,-9.899],[131,-8.5137],[132,-8.0753],[133,-7.7371],[134,-8.5379],[135,-12.7013],[136,-16.2648],[137,-18.851],[138,-19.7601],[139,-20.1236],[140,-20.3673],[141,-18.7314],[142,-14.766],[143,-11.3233],[144,-8.9639],[145,-8.0565],[146,-6.3202],[147,-4.0234],[148,-1.6775],[149,-0.0588],[150,0.484],[151,1.5344],[152,2.5033],[153,2.5682],[154,1.7575],[155,0.372],[156,-1.1784],[157,-2.6752],[158,-4.0028],[159,-5.0217],[160,-5.6301],[161,-6.3856],[162,-7.5314],[163,-9.0013],[164,-11.1895],[166,-16.9144],[167,-19.1165],[168,-19.9919],[169,-19.4846],[170,-18.1587],[171,-16.3624],[172,-14.4056],[173,-12.6107],[174,-11.2904],[175,-10.305],[176,-9.41],[177,-8.6581],[178,-8.0903],[179,-7.7537],[180,-7.708],[181,-8.4571],[182,-9.8312],[183,-11.3519],[184,-13.0838],[185,-15.1043],[186,-16.8018],[187,-17.7719],[188,-18.383],[189,-18.8141],[190,-19.0939],[191,-19.2398],[192,-19.298],[193,-19.298],[194,-19.193],[195,-18.9931],[196,-18.6524],[197,-18.1489],[198,-17.4897],[199,-16.1432],[201,-11.6238],[202,-9.4313],[203,-7.4766],[204,-6.2803],[205,-5.6677],[206,-5.2433],[207,-4.9087],[208,-4.7617],[209,-4.7685],[210,-4.9111],[211,-5.2682],[212,-5.8248],[213,-6.4984],[214,-7.2697],[215,-8.1011],[216,-9.1036],[217,-10.5875],[218,-12.7303],[219,-14.9139],[220,-16.7388],[221,-18.4577],[222,-20.0585],[223,-21.4683],[224,-22.6646],[225,-23.6916],[226,-24.519],[227,-25.1023],[228,-25.4322],[229,-25.6391],[231,-25.8871],[234,-26.0029],[235,-26.0029],[236,-25.9205],[237,-25.6321],[239,-24.8426],[240,-24.532],[242,-24.008],[243,-23.666],[245,-22.7425],[246,-22.1853],[247,-21.4438],[248,-20.4063],[249,-18.9295],[251,-15.4086],[252,-13.9416],[253,-12.826],[254,-11.9007],[255,-11.1305],[256,-10.4773],[257,-9.9183],[258,-9.4839],[259,-9.1755],[260,-8.9966],[261,-8.9301],[262,-8.9048],[265,-8.8734],[266,-8.8419],[269,-8.6844],[270,-8.653],[271,-8.6656],[272,-8.7852],[273,-8.9865],[275,-9.4676],[276,-9.6688],[277,-9.8127],[279,-10.0311],[285,-10.5377],[286,-10.5836],[287,-10.5836],[288,-10.5204],[289,-10.4339],[292,-10.0934],[296,-9.5151],[298,-9.1933],[299,-9.0543],[301,-8.835],[303,-8.6742],[304,-8.6158],[305,-8.5865],[306,0]]
       }
       middle_01_r_ctrl.Scale.X {type=MovieSceneScriptingFloatChannel, default=1, const [1,307]}
       middle_01_r_ctrl.Scale.Y {type=MovieSceneScriptingFloatChannel, default=1, const [1,307]}
       middle_01_r_ctrl.Scale.Z {type=MovieSceneScriptingFloatChannel, default=1, const [1,307]}
       middle_02_r_ctrl.Location.X {type=MovieSceneScriptingFloatChannel, default=0, thinned_from=307, keys [[0,0.0008],[305,0.0008],[306,0]]}
       middle_02_r_ctrl.Location.Y {type=MovieSceneScriptingFloatChannel, default=0, thinned_from=307, keys [[0,0.0009],[305,0.0009],[306,0]]}
       middle_02_r_ctrl.Location.Z {type=MovieSceneScriptingFloatChannel, default=0, thinned_from=307, keys [[0,0.0476],[305,0.0476],[306,0]]}
       middle_02_r_ctrl.Rotation.X {
        type=MovieSceneScriptingFloatChannel
        default=0
        thinned_from=307
        keys [[0,-0.0079],[2,-0.0079],[4,-0.0075],[5,-0.0096],[6,-0.0092],[7,-0.0067],[12,-0.0064],[13,-0.0088],[14,-0.0067],[15,-0.007],[16,-0.0096],[17,-0.0086],[18,-0.0086],[19,-0.0089],[21,-0.0089],[24,-0.0074],[25,-0.0075],[26,-0.0068],[27,-0.0084],[28,-0.0086],[33,-0.0066],[34,-0.0093],[35,-0.0076],[36,-0.0067],[37,-0.0067],[38,-0.0092],[39,-0.0092],[40,-0.0065],[42,-0.0062],[43,-0.0078],[44,-0.0077],[45,-0.0067],[46,-0.0067],[47,-0.0073],[48,-0.0091],[50,-0.0091],[51,-0.0073],[54,-0.0073],[55,-0.0083],[57,-0.0083],[58,-0.0071],[62,-0.0067],[63,-0.0075],[65,-0.007],[66,-0.0069],[67,-0.0044],[68,-0.0071],[69,-0.0058],[70,-0.0035],[71,-0.0041],[72,-0.0058],[73,-0.0046],[74,-0.0047],[75,-0.0089],[76,-0.0064],[77,-0.0068],[78,-0.0084],[79,-0.0088],[80,-0.0095],[81,-0.0098],[82,-0.0064],[83,-0.0079],[84,-0.0081],[85,-0.0012],[86,0.0003],[87,-0.0008],[88,0.004],[89,0.0031],[90,0.0031],[91,0.0002],[92,-0.0001],[93,-0.0046],[94,-0.0066],[95,-0.007],[96,-0.0072],[97,-0.0022],[98,-0.0017],[99,0.0018],[100,0.0011],[101,0.0009],[102,0.0017],[103,-0.0018],[104,-0.0039],[105,-0.0075],[106,-0.009],[107,-0.0073],[108,-0.0093],[109,-0.0082],[110,-0.0052],[111,-0.0014],[112,0.0007],[114,0.0015],[115,0.0044],[116,0.0009],[117,-0.0014],[118,-0.0028],[119,-0.0083],[120,-0.0073],[121,-0.0067],[122,-0.0009],[123,0.0012],[124,0.0027],[126,0.0011],[127,0.0017],[128,0.0008],[129,-0.0036],[130,-0.0033],[131,-0.0042],[132,-0.0087],[133,-0.0085],[134,-0.0065],[135,-0.0063],[136,-0.0035],[137,-0.0009],[138,0],[139,0.0028],[140,0.0017],[141,0.0036],[142,0.0004],[143,-0.0052],[144,-0.0062],[145,-0.0085],[146,-0.0071],[147,-0.0064],[148,-0.0063],[149,-0.0065],[150,-0.0099],[151,-0.0071],[152,-0.0094],[153,-0.0097],[154,-0.0074],[155,-0.0137],[157,-0.0083],[158,-0.0073],[159,-0.01],[160,-0.0063],[161,-0.0094],[162,-0.0097],[163,-0.0092],[164,-0.0094],[166,-0.0062],[167,-0.0027],[168,-0.008],[170,-0.0036],[171,-0.0029],[172,-0.0062],[173,-0.007],[175,-0.0089],[176,-0.005],[178,-0.0067],[179,-0.0116],[180,-0.0116],[181,-0.0056],[182,-0.0082],[183,-0.0076],[184,-0.0059],[185,-0.008],[186,-0.0076],[187,-0.0077],[188,-0.0063],[189,-0.0064],[190,-0.0081],[191,-0.0037],[193,-0.0037],[194,-0.0066],[195,-0.0041],[196,-0.0034],[197,-0.0068],[198,-0.0072],[199,-0.0094],[200,-0.0093],[201,-0.0088],[202,-0.0081],[203,-0.0084],[204,-0.0082],[205,-0.0093],[206,-0.006],[207,-0.007],[209,-0.0101],[210,-0.0071],[211,-0.0147],[212,-0.015],[213,-0.0093],[214,-0.01],[215,-0.0103],[216,-0.0103],[218,-0.0076],[219,-0.0122],[220,-0.012],[221,-0.0105],[222,-0.0104],[223,-0.0108],[224,-0.0113],[225,-0.0108],[226,-0.0052],[227,-0.0086],[228,-0.0087],[229,-0.0055],[235,-0.0056],[236,-0.0089],[237,-0.0055],[238,-0.0087],[239,-0.0095],[240,-0.0085],[243,-0.0085],[244,-0.0093],[246,-0.0093],[247,-0.0086],[248,-0.0092],[251,-0.0096],[252,-0.0063],[253,-0.0094],[254,-0.0059],[255,-0.009],[256,-0.0082],[258,-0.0082],[259,-0.0087],[260,-0.0066],[261,-0.0081],[268,-0.0083],[269,-0.0071],[271,-0.0071],[272,-0.0083],[273,-0.0084],[274,-0.0076],[275,-0.0081],[276,-0.0077],[277,-0.0094],[279,-0.009],[280,-0.0095],[281,-0.0068],[283,-0.0064],[284,-0.0064],[285,-0.0088],[286,-0.0061],[287,-0.0061],[288,-0.0088],[290,-0.0091],[291,-0.0069],[295,-0.0079],[296,-0.0079],[297,-0.0082],[298,-0.0073],[299,-0.0089],[300,-0.0074],[304,-0.0082],[305,-0.0079],[306,0]]
       }
       middle_02_r_ctrl.Rotation.Y {
        type=MovieSceneScriptingFloatChannel
        default=0
        thinned_from=307
        keys [[0,-0.0109],[4,-0.0109],[5,-0.0116],[6,-0.0116],[7,-0.0109],[12,-0.0109],[13,-0.0116],[14,-0.0109],[15,-0.0109],[16,-0.0116],[17,-0.011],[21,-0.0109],[24,-0.0116],[25,-0.0094],[26,-0.0097],[27,-0.0132],[28,-0.0109],[33,-0.0113],[34,-0.0126],[35,-0.01],[36,-0.0117],[37,-0.0117],[38,-0.0122],[39,-0.0122],[40,-0.0109],[42,-0.0101],[43,-0.0122],[44,-0.0118],[45,-0.0131],[46,-0.0131],[47,-0.0105],[48,-0.0131],[50,-0.0131],[51,-0.0118],[54,-0.0118],[55,-0.0089],[57,-0.0089],[58,-0.0122],[61,-0.0126],[62,-0.013],[63,-0.0105],[65,-0.0119],[66,-0.0124],[67,-0.0107],[68,-0.0145],[69,-0.0121],[70,-0.0123],[71,-0.0102],[72,-0.0105],[74,-0.0104],[75,-0.0126],[76,-0.01],[77,-0.0113],[78,-0.0092],[79,-0.0104],[80,-0.0095],[81,-0.0091],[82,-0.0077],[83,-0.0089],[84,-0.0125],[85,-0.0097],[86,-0.0134],[87,-0.0118],[88,-0.0128],[89,-0.0159],[90,-0.0159],[92,-0.0128],[94,-0.0103],[95,-0.0079],[96,-0.0095],[97,-0.0121],[98,-0.0109],[99,-0.0139],[100,-0.0148],[101,-0.0137],[102,-0.0137],[104,-0.0133],[105,-0.0136],[106,-0.0109],[107,-0.0107],[108,-0.0118],[109,-0.0111],[110,-0.0114],[111,-0.0143],[112,-0.0142],[113,-0.0121],[114,-0.0132],[116,-0.0136],[117,-0.0135],[118,-0.0107],[119,-0.0132],[120,-0.0137],[121,-0.01],[122,-0.0138],[123,-0.0125],[124,-0.0118],[125,-0.0166],[126,-0.0137],[127,-0.0101],[128,-0.013],[129,-0.0121],[130,-0.012],[131,-0.0109],[133,-0.0138],[134,-0.0093],[135,-0.0166],[136,-0.0137],[137,-0.0161],[138,-0.0106],[139,-0.0131],[140,-0.01],[141,-0.0146],[142,-0.0129],[143,-0.0143],[144,-0.0118],[145,-0.0115],[146,-0.0098],[147,-0.0115],[148,-0.0118],[149,-0.0104],[150,-0.0104],[151,-0.0076],[152,-0.0106],[153,-0.0071],[154,-0.0063],[155,-0.0095],[156,-0.0091],[157,-0.0096],[158,-0.0077],[159,-0.0107],[160,-0.0091],[161,-0.0121],[162,-0.0109],[163,-0.0146],[164,-0.0094],[165,-0.011],[166,-0.0145],[167,-0.0107],[168,-0.0118],[169,-0.0163],[170,-0.0078],[171,-0.0089],[172,-0.0149],[173,-0.0144],[175,-0.0124],[176,-0.0099],[178,-0.0079],[179,-0.0114],[180,-0.0114],[181,-0.0108],[182,-0.0134],[183,-0.0102],[184,-0.0103],[185,-0.0151],[186,-0.0136],[187,-0.0124],[188,-0.0154],[189,-0.0148],[190,-0.0104],[191,-0.0082],[193,-0.0082],[194,-0.0142],[195,-0.0099],[196,-0.0108],[197,-0.0139],[198,-0.0116],[199,-0.0161],[200,-0.0124],[201,-0.0109],[202,-0.0118],[203,-0.0114],[204,-0.0137],[205,-0.0121],[206,-0.0087],[207,-0.0092],[209,-0.0088],[210,-0.0069],[211,-0.009],[212,-0.0118],[213,-0.0095],[214,-0.0086],[215,-0.0081],[216,-0.0081],[217,-0.0099],[218,-0.0084],[219,-0.0118],[220,-0.0084],[221,-0.0126],[222,-0.0118],[224,-0.0093],[225,-0.0121],[226,-0.0086],[227,-0.0124],[228,-0.0125],[229,-0.0081],[235,-0.0082],[236,-0.0128],[237,-0.0081],[238,-0.0125],[239,-0.009],[240,-0.0121],[242,-0.0117],[243,-0.0122],[244,-0.0101],[246,-0.0102],[247,-0.0124],[248,-0.01],[251,-0.0104],[252,-0.0095],[253,-0.0102],[254,-0.009],[255,-0.0095],[256,-0.0112],[258,-0.0102],[259,-0.0108],[260,-0.0101],[261,-0.0122],[268,-0.0123],[269,-0.0103],[271,-0.0103],[272,-0.0123],[273,-0.0107],[274,-0.0119],[275,-0.0104],[276,-0.0105],[277,-0.0135],[279,-0.0135],[280,-0.0114],[281,-0.0106],[284,-0.0106],[285,-0.0115],[286,-0.0107],[287,-0.0107],[288,-0.0115],[295,-0.0109],[296,-0.0121],[297,-0.0119],[298,-0.0099],[299,-0.0115],[300,-0.011],[304,-0.0105],[305,-0.0118],[306,0]]
       }
       middle_02_r_ctrl.Rotation.Z {
        type=MovieSceneScriptingFloatChannel
        default=0
        thinned_from=307
        keys [[0,6.5511],[2,6.5511],[5,6.2788],[7,5.9156],[8,5.8247],[9,5.8247],[10,5.7338],[12,5.7338],[14,5.9156],[16,6.2788],[17,6.4457],[18,6.4457],[19,6.5628],[21,6.5628],[24,5.8599],[26,5.1556],[27,4.9204],[28,4.5675],[33,3.3886],[34,3.3619],[35,3.2704],[36,3.362],[37,3.362],[38,3.4534],[39,3.4534],[42,3.728],[43,3.9107],[44,4.0021],[45,4.1848],[46,4.1848],[48,4.3673],[50,4.3673],[51,4.3468],[54,4.3468],[55,4.1996],[57,4.1996],[58,4.0522],[59,4.0522],[61,3.7574],[62,3.4623],[63,2.8714],[64,2.1313],[65,1.2408],[66,0.1987],[67,-1.1306],[68,-3.038],[69,-7.0416],[70,-9.4436],[71,-8.9914],[72,-5.9978],[73,-2.4497],[74,0.9097],[75,3.5136],[78,6.9548],[79,7.81],[80,8.2369],[82,8.5213],[83,8.722],[84,-0.615],[85,-17.4588],[86,-28.4853],[87,-35.4224],[88,-37.7375],[89,-38.8028],[90,-38.8028],[91,-31.5076],[92,-19.347],[93,-6.6416],[94,0.6105],[95,2.7],[96,0.7851],[97,-8.9825],[98,-25.3214],[99,-34.0811],[100,-37.4228],[101,-38.9099],[102,-37.846],[103,-28.2067],[104,-13.6879],[105,0.5393],[106,5.9868],[107,7.8693],[108,8.7219],[109,6.5869],[110,-2.6162],[111,-18.8162],[112,-32.4383],[113,-36.685],[114,-38.5899],[115,-39.0167],[116,-33.2533],[117,-21.3081],[118,-8.3122],[119,0.6475],[120,3.7279],[121,-5.451],[122,-20.6299],[123,-30.3638],[124,-35.8274],[125,-37.507],[126,-38.6377],[127,-39.0166],[128,-35.0874],[129,-23.3569],[130,-7.3549],[131,-0.3378],[132,1.8534],[133,3.5646],[134,2.1653],[135,-6.8518],[136,-20.0847],[137,-32.2682],[138,-36.3545],[139,-37.8695],[140,-39.0166],[141,-36.1664],[142,-26.8644],[143,-12.5449],[144,-1.1243],[145,3.2542],[146,4.0299],[147,4.0985],[148,4.2406],[149,4.8081],[150,7.2097],[151,11.5511],[152,16.539],[153,19.9747],[154,21.2068],[155,20.3856],[156,17.7784],[157,14.744],[158,11.9689],[159,10.2954],[160,9.4564],[161,8.7556],[162,7.6319],[163,6.2225],[164,3.9238],[165,1.0281],[166,-1.6989],[167,-4.0565],[168,-4.8471],[169,-4.2542],[170,-3.0717],[171,-1.3077],[172,0.834],[173,2.5756],[174,3.9237],[175,4.8834],[177,6.4137],[178,6.9858],[179,7.4383],[180,7.4383],[181,6.6764],[182,5.3388],[183,3.8042],[184,2.0692],[185,0.1303],[186,-1.6251],[187,-2.6049],[188,-3.1944],[190,-3.9823],[191,-4.1794],[193,-4.1794],[194,-3.9823],[195,-3.9077],[196,-3.5314],[198,-2.4057],[199,-0.9117],[200,1.316],[201,3.528],[202,5.7255],[203,8.0911],[204,9.9008],[206,12.7804],[208,15.9979],[210,18.8412],[211,20.1433],[212,21.0793],[213,21.7807],[214,22.2478],[215,22.4812],[216,22.4812],[217,21.5471],[218,19.675],[219,17.5616],[220,15.6765],[222,12.3601],[223,10.9313],[224,9.7371],[226,7.819],[227,7.3967],[229,6.7326],[230,6.5663],[231,6.5663],[232,6.4],[235,6.4],[237,6.7326],[238,7.2307],[241,8.2255],[242,8.391],[243,8.8164],[244,9.1871],[245,9.3106],[246,9.3106],[248,9.0636],[249,9.3106],[250,9.6809],[251,9.9275],[252,9.8043],[253,9.4341],[256,7.5782],[258,6.5848],[259,6.3091],[260,6.1579],[261,6.2334],[264,6.2334],[265,6.3091],[266,6.3091],[267,6.3847],[268,6.3847],[269,6.4604],[271,6.4604],[272,6.3847],[274,5.9307],[275,5.6883],[276,5.4761],[283,4.7323],[284,4.7323],[286,4.5195],[287,4.5195],[290,4.8386],[291,4.9352],[295,5.4639],[296,5.6619],[298,5.9259],[299,6.1236],[300,6.1895],[301,6.3213],[305,6.5847],[306,0]]
       }
       middle_02_r_ctrl.Scale.X {type=MovieSceneScriptingFloatChannel, default=1, const [1,307]}
       middle_02_r_ctrl.Scale.Y {type=MovieSceneScriptingFloatChannel, default=1, const [1,307]}
       middle_02_r_ctrl.Scale.Z {type=MovieSceneScriptingFloatChannel, default=1, const [1,307]}
       middle_03_r_ctrl.Location.X {type=MovieSceneScriptingFloatChannel, default=0, thinned_from=307, keys [[0,0],[306,0]]}
       middle_03_r_ctrl.Location.Y {type=MovieSceneScriptingFloatChannel, default=0, thinned_from=307, keys [[0,0],[306,0]]}
       middle_03_r_ctrl.Location.Z {type=MovieSceneScriptingFloatChannel, default=0, thinned_from=307, keys [[0,0],[306,0]]}
       middle_03_r_ctrl.Rotation.X {
        type=MovieSceneScriptingFloatChannel
        default=0
        thinned_from=307
        keys [[0,-0.0134],[1,-0.0134],[2,-0.0106],[3,-0.0106],[4,-0.0079],[6,-0.0094],[7,-0.01],[8,-0.01],[9,-0.0108],[13,-0.0108],[15,-0.0092],[17,-0.0082],[18,-0.0127],[21,-0.0127],[22,-0.0082],[23,-0.0082],[24,-0.0127],[25,-0.0127],[26,-0.0081],[27,-0.0126],[29,-0.0125],[30,-0.0079],[33,-0.0078],[36,-0.008],[37,-0.012],[40,-0.012],[44,-0.0115],[50,-0.0113],[51,-0.0105],[60,-0.0109],[61,-0.0088],[62,-0.0114],[63,-0.0118],[64,-0.0098],[66,-0.0109],[67,-0.0107],[68,-0.0133],[69,-0.0109],[70,-0.0097],[71,-0.0107],[72,-0.0099],[73,-0.0115],[74,-0.0115],[77,-0.0086],[78,-0.0103],[79,-0.0092],[80,-0.0089],[81,-0.011],[82,-0.0093],[83,-0.0158],[84,-0.0082],[85,-0.0152],[86,-0.0098],[87,-0.0134],[88,-0.0107],[89,-0.0085],[90,-0.0085],[91,-0.011],[92,-0.0143],[93,-0.0128],[94,-0.0116],[95,-0.0149],[96,-0.0113],[97,-0.0134],[98,-0.0122],[99,-0.01],[100,-0.0109],[101,-0.0086],[102,-0.0163],[103,-0.0154],[104,-0.0094],[105,-0.0057],[106,-0.0099],[107,-0.0125],[108,-0.0109],[109,-0.0143],[110,-0.0075],[111,-0.0153],[112,-0.0121],[113,-0.0121],[114,-0.0088],[115,-0.0117],[116,-0.0129],[117,-0.01],[118,-0.01],[119,-0.0085],[120,-0.009],[121,-0.0119],[122,-0.0107],[123,-0.0118],[124,-0.0111],[125,-0.0112],[126,-0.0097],[127,-0.0092],[128,-0.01],[129,-0.012],[130,-0.0122],[131,-0.0059],[132,-0.0087],[133,-0.0118],[134,-0.0077],[135,-0.0108],[136,-0.0096],[137,-0.0125],[138,-0.0132],[139,-0.0102],[140,-0.009],[141,-0.0122],[142,-0.0105],[143,-0.0094],[144,-0.0128],[145,-0.0071],[146,-0.0088],[147,-0.0119],[148,-0.0083],[149,-0.0084],[150,-0.0126],[151,-0.0073],[152,-0.0092],[153,-0.0072],[154,-0.0109],[155,-0.0082],[156,-0.0103],[157,-0.012],[158,-0.0085],[159,-0.0128],[160,-0.0128],[161,-0.0083],[162,-0.0124],[163,-0.0092],[164,-0.01],[165,-0.0105],[166,-0.0099],[167,-0.0125],[168,-0.0082],[169,-0.0123],[170,-0.0132],[171,-0.01],[172,-0.0105],[175,-0.0098],[178,-0.0088],[179,-0.012],[180,-0.012],[181,-0.0088],[182,-0.0095],[184,-0.0101],[185,-0.01],[186,-0.0133],[187,-0.0131],[188,-0.009],[189,-0.0088],[190,-0.0123],[194,-0.0122],[195,-0.0091],[196,-0.0094],[197,-0.0134],[198,-0.0105],[199,-0.012],[200,-0.0101],[201,-0.0076],[202,-0.0117],[204,-0.0111],[205,-0.0072],[206,-0.0097],[207,-0.0118],[208,-0.0073],[209,-0.0123],[210,-0.0075],[211,-0.0077],[212,-0.0073],[213,-0.0089],[215,-0.0085],[216,-0.0085],[217,-0.009],[218,-0.0099],[219,-0.0104],[220,-0.0085],[221,-0.0083],[223,-0.0114],[224,-0.0085],[226,-0.0114],[227,-0.01],[231,-0.011],[232,-0.011],[233,-0.0081],[235,-0.0081],[236,-0.011],[238,-0.0102],[240,-0.0092],[241,-0.0118],[242,-0.0086],[243,-0.0097],[244,-0.0095],[246,-0.0094],[248,-0.0096],[249,-0.0117],[250,-0.0093],[251,-0.0114],[252,-0.0092],[253,-0.0116],[254,-0.0096],[255,-0.0121],[256,-0.0123],[257,-0.0101],[264,-0.0104],[265,-0.0131],[268,-0.0131],[269,-0.0101],[273,-0.0104],[274,-0.0106],[275,-0.0097],[278,-0.0104],[280,-0.0104],[281,-0.0106],[282,-0.0106],[283,-0.0109],[289,-0.0109],[290,-0.0106],[291,-0.0106],[292,-0.0144],[293,-0.0144],[294,-0.0141],[295,-0.0141],[296,-0.0138],[297,-0.0138],[298,-0.0135],[299,-0.0135],[300,-0.0131],[305,-0.0131],[306,0]]
       }
       middle_03_r_ctrl.Rotation.Y {
        type=MovieSceneScriptingFloatChannel
        default=0
        thinned_from=307
        keys [[0,0.0047],[1,0.0047],[2,0.0036],[3,0.0036],[4,0.0041],[6,0.0043],[7,0.0028],[15,0.0027],[16,0.0042],[17,0.0033],[18,0.005],[21,0.005],[22,0.0033],[23,0.0032],[24,0.0047],[25,0.0046],[26,0.0028],[27,0.0043],[29,0.0042],[30,0.0026],[33,0.0026],[34,0.0023],[36,0.0023],[37,0.0036],[50,0.004],[51,0.0032],[60,0.003],[61,0.0032],[62,0.0028],[63,0.0038],[64,0.003],[66,0.003],[67,0.0025],[68,0.0029],[69,0.0016],[70,0.0023],[71,0.0021],[72,0.0021],[73,0.0031],[74,0.0035],[75,0.0055],[76,0.004],[77,0.0043],[78,0.0065],[79,0.0052],[80,0.0057],[81,0.0053],[82,0.0046],[83,0.0045],[84,0.0012],[85,-0.0006],[86,-0.0007],[87,-0.005],[88,-0.0023],[91,-0.0024],[92,-0.0012],[93,0.0024],[94,0.0035],[95,0.0033],[96,0.0015],[97,0.0009],[98,-0.0017],[100,-0.0034],[101,-0.0025],[102,-0.005],[103,-0.0023],[104,0.0013],[105,0.0029],[106,0.0047],[107,0.006],[108,0.0042],[109,0.0049],[110,0.0014],[111,-0.0005],[113,-0.0037],[114,-0.0036],[115,-0.004],[116,-0.0034],[118,0.0011],[119,0.003],[121,0.0025],[122,-0.0004],[123,-0.0024],[124,-0.0032],[125,-0.0038],[126,-0.0029],[128,-0.0024],[129,-0.0013],[130,0.0018],[131,0.0028],[132,0.0036],[133,0.0017],[134,0.0035],[135,0.0007],[136,-0.0001],[137,-0.0033],[138,-0.0038],[139,-0.004],[140,-0.0027],[141,-0.0036],[142,-0.0018],[144,0.0039],[145,0.0026],[146,0.0035],[147,0.0026],[148,0.0039],[149,0.0029],[150,0.0045],[151,0.0049],[152,0.0067],[153,0.0062],[154,0.0068],[155,0.0053],[156,0.0057],[157,0.007],[158,0.0048],[160,0.0055],[161,0.0041],[162,0.0044],[163,0.0033],[164,0.0032],[165,0.0038],[167,0.0011],[168,0.001],[169,0.0014],[170,0.0022],[171,0.0022],[172,0.0037],[173,0.0026],[175,0.0039],[176,0.0029],[178,0.0043],[179,0.0048],[180,0.0048],[181,0.0043],[182,0.0047],[183,0.0036],[184,0.0029],[186,0.0022],[187,0.0025],[188,0.0021],[189,0.0023],[190,0.0017],[194,0.0019],[195,0.0027],[196,0.0021],[197,0.0023],[198,0.0021],[199,0.0035],[200,0.0036],[201,0.0026],[202,0.0049],[203,0.0035],[204,0.0042],[205,0.0038],[206,0.005],[207,0.005],[208,0.0054],[209,0.0061],[210,0.005],[211,0.0056],[212,0.0053],[213,0.0075],[214,0.0068],[215,0.0059],[216,0.0057],[217,0.0072],[218,0.0065],[219,0.0062],[220,0.0054],[221,0.0051],[222,0.0052],[223,0.0055],[224,0.0039],[226,0.0049],[227,0.004],[232,0.004],[233,0.0029],[235,0.0029],[236,0.004],[240,0.0041],[241,0.0038],[242,0.0042],[243,0.0043],[244,0.0034],[245,0.0046],[246,0.0034],[247,0.0045],[249,0.0044],[250,0.0037],[251,0.0049],[252,0.0038],[253,0.0045],[254,0.0043],[255,0.0049],[256,0.0046],[257,0.0035],[258,0.0034],[259,0.0041],[260,0.0038],[268,0.0037],[269,0.0044],[271,0.0044],[273,0.0038],[274,0.0032],[275,0.0035],[285,0.0039],[286,0.0019],[287,0.0019],[288,0.0039],[291,0.0038],[293,0.004],[301,0.0033],[302,0.0059],[305,0.0059],[306,0]]
       }
       middle_03_r_ctrl.Rotation.Z {
        type=MovieSceneScriptingFloatChannel
        default=0
        thinned_from=307
        keys [[0,6.6406],[1,6.6406],[2,6.4924],[3,6.4924],[7,5.8988],[8,5.8988],[9,5.7504],[13,5.7504],[16,6.1957],[17,6.3183],[18,6.5536],[21,6.5536],[25,5.6118],[26,5.1406],[29,4.4331],[30,3.9613],[32,3.4891],[33,3.4891],[34,3.2529],[36,3.2529],[37,3.4613],[40,3.4613],[41,3.6696],[42,3.6696],[44,4.0862],[45,4.0862],[46,4.2944],[50,4.2944],[51,4.3481],[54,4.3481],[55,4.2272],[56,4.2272],[57,4.1063],[58,4.1063],[60,3.8645],[62,3.3806],[63,3.0175],[65,1.8059],[66,0.9566],[67,0.0851],[68,-1.2482],[69,-4.2589],[70,-5.7704],[71,-4.5944],[72,-1.415],[73,2.2461],[74,5.8895],[75,9.0246],[77,11.9874],[78,13.631],[79,14.6166],[80,14.945],[81,12.8094],[82,11.0005],[83,9.0702],[84,-1.2151],[85,-13.6024],[86,-23.0715],[87,-28.7636],[88,-30.8024],[89,-31.8281],[90,-31.8281],[91,-26.7409],[92,-16.8377],[93,-5.3094],[94,1.2907],[95,3.1642],[96,1.2907],[97,-6.5755],[98,-20.1034],[99,-27.5286],[100,-30.4973],[101,-31.8282],[102,-30.8294],[103,-23.6216],[104,-11.5646],[105,0.7558],[106,5.6146],[107,7.4296],[108,8.0339],[109,6.5224],[110,-0.7692],[111,-14.3831],[112,-26.2201],[113,-29.8346],[114,-31.4947],[115,-31.8282],[116,-27.5674],[117,-18.2524],[118,-6.7168],[119,1.0427],[120,3.7432],[121,-3.0282],[122,-15.1609],[123,-23.8055],[124,-28.8819],[125,-30.498],[126,-31.5319],[127,-31.8282],[128,-29.7621],[129,-20.2345],[130,-5.8949],[131,0.1446],[132,2.1377],[133,3.5581],[134,2.7062],[135,-4.1476],[136,-15.1908],[137,-25.9695],[138,-29.3449],[139,-30.8939],[140,-31.8281],[141,-29.9634],[142,-22.9374],[143,-10.5067],[144,-0.4261],[145,2.9903],[146,5.2594],[147,6.9966],[148,8.8928],[149,10.5494],[150,12.4404],[151,15.5094],[152,19.0468],[153,21.1689],[154,21.6404],[155,20.2258],[156,17.6321],[158,11.9679],[159,10.3128],[160,9.6029],[162,7.708],[163,6.0506],[164,3.9077],[165,1.1723],[166,-1.7703],[167,-3.9374],[168,-4.9251],[169,-4.3323],[170,-2.9514],[171,-1.1808],[172,0.7807],[173,2.5412],[174,3.9077],[175,4.8824],[177,6.4397],[178,7.023],[179,7.4445],[180,7.4445],[181,6.6803],[182,5.3413],[183,3.8086],[184,2.0808],[185,-0.0364],[186,-1.5805],[187,-2.5475],[188,-3.3221],[189,-3.7099],[191,-4.098],[194,-4.098],[195,-3.8705],[196,-3.5022],[198,-2.3989],[199,-0.9311],[200,1.2644],[201,3.4531],[202,5.8174],[203,7.9947],[204,9.8059],[205,11.253],[207,14.5044],[208,15.948],[209,17.5712],[210,18.8336],[211,20.0113],[212,20.947],[213,21.7658],[214,22.2337],[215,22.4677],[216,22.3507],[217,21.5318],[218,19.6604],[219,17.555],[220,15.6831],[221,13.9271],[222,12.4042],[223,10.9972],[224,9.7065],[225,8.7669],[226,7.9441],[227,7.2872],[228,6.9828],[231,6.526],[232,6.526],[233,6.3737],[235,6.3737],[236,6.526],[238,7.135],[239,7.5915],[241,8.1998],[242,8.3519],[243,8.7495],[244,9.1104],[245,9.351],[247,9.1104],[248,9.1104],[250,9.5915],[251,9.9522],[252,9.832],[253,9.4712],[257,7.0636],[258,6.5814],[259,6.344],[260,6.1956],[264,6.1956],[265,6.344],[268,6.344],[269,6.4924],[271,6.4924],[273,6.1956],[274,5.8988],[275,5.7504],[278,5.1267],[280,5.1267],[281,4.9187],[282,4.9187],[283,4.7106],[285,4.7106],[286,4.5026],[287,4.5026],[288,4.7106],[289,4.7106],[290,4.9187],[291,4.9187],[292,5.186],[293,5.186],[294,5.4534],[295,5.4534],[296,5.7206],[297,5.7206],[298,5.9878],[299,5.9878],[300,6.2549],[301,6.2549],[302,6.522],[305,6.522],[306,0]]
       }
       middle_03_r_ctrl.Scale.X {type=MovieSceneScriptingFloatChannel, default=1, const [1,307]}
       middle_03_r_ctrl.Scale.Y {type=MovieSceneScriptingFloatChannel, default=1, const [1,307]}
       middle_03_r_ctrl.Scale.Z {type=MovieSceneScriptingFloatChannel, default=1, const [1,307]}
       ring_metacarpal_r_ctrl.Location.X {type=MovieSceneScriptingFloatChannel, default=0, thinned_from=307, keys [[0,0],[306,0]]}
       ring_metacarpal_r_ctrl.Location.Y {type=MovieSceneScriptingFloatChannel, default=0, thinned_from=307, keys [[0,0],[306,0]]}
       ring_metacarpal_r_ctrl.Location.Z {type=MovieSceneScriptingFloatChannel, default=0, thinned_from=307, keys [[0,0],[306,0]]}
       ring_metacarpal_r_ctrl.Rotation.X {
        type=MovieSceneScriptingFloatChannel
        default=0
        thinned_from=307
        keys [[0,0.0021],[2,0.0021],[3,-0.0006],[4,-0.001],[5,-0.0017],[6,-0.0002],[7,-0.0007],[8,0.0012],[9,0.001],[13,0.001],[14,-0.0007],[16,0.0007],[17,-0.0009],[21,-0.0008],[23,-0.001],[24,0.0025],[25,-0.0014],[26,0.0021],[28,0.0018],[29,-0.0023],[30,-0.0024],[31,0.0011],[32,-0.0028],[33,0.0008],[34,-0.0002],[39,-0.0003],[40,0.003],[41,-0.0003],[42,-0.0004],[43,0.0029],[44,-0.0006],[47,-0.0007],[48,0.0027],[49,-0.0007],[55,-0.0008],[56,0.0003],[57,0.0003],[58,0.0001],[59,0.0006],[60,-0.0002],[61,0.0004],[62,0],[63,-0.0001],[64,-0.0006],[65,-0.0001],[66,0.0008],[67,0.0003],[68,0.0012],[69,-0.0024],[70,-0.0007],[71,0.0015],[72,-0.0005],[73,-0.0007],[74,-0.0001],[75,-0.0002],[77,-0.0001],[78,-0.0018],[79,0.0007],[80,0.0004],[81,-0.0006],[82,-0.0013],[83,0.0012],[84,-0.0017],[145,-0.0017],[146,-0.0008],[147,-0.0018],[148,0.0014],[149,-0.0011],[150,-0.0008],[151,0.0023],[152,-0.0004],[153,-0.0007],[154,0.0006],[156,-0.0014],[157,0],[158,-0.0002],[159,-0.0015],[160,0.0015],[161,-0.0002],[162,0.0007],[163,0.0021],[164,-0.0011],[165,0.0019],[166,-0.0013],[167,-0.0013],[168,0.0007],[169,0.0017],[170,0.0003],[171,-0.0008],[172,0.0014],[173,-0.0014],[174,-0.0011],[175,0.0026],[176,0.0017],[177,0.0001],[178,0.0025],[179,0.0011],[180,0.0009],[181,-0.0039],[182,0.0015],[183,-0.0038],[184,0.0008],[185,-0.0004],[186,0.0031],[187,0.0021],[188,0.0013],[190,0.0005],[191,0.0002],[194,0.0002],[195,0.0007],[196,-0.0015],[197,0.0012],[198,0.0016],[199,-0.0001],[200,0.0007],[201,-0.0013],[202,0.0013],[203,0.0011],[204,-0.0018],[205,0.0002],[206,-0.0003],[207,0.0019],[208,-0.0007],[210,-0.0007],[211,0.0014],[212,0.0014],[213,-0.0029],[215,-0.0026],[217,-0.0026],[218,0.0015],[219,-0.003],[220,-0.0035],[221,0.0002],[225,-0.0021],[226,0.0015],[227,-0.0043],[228,-0.0041],[229,0.0006],[235,0.0009],[236,-0.0037],[237,0.0006],[239,-0.0001],[240,0.0042],[241,-0.0007],[242,-0.001],[243,0.0009],[248,0.0006],[249,0.0013],[251,-0.0005],[252,0],[253,0.0003],[254,0.0004],[255,-0.0016],[256,0.0004],[258,0.0002],[259,-0.0007],[263,-0.0007],[264,0.0018],[265,-0.0006],[266,0.002],[267,-0.0005],[268,0.0019],[269,0.0018],[270,-0.0009],[271,0.0014],[272,-0.0012],[273,0.001],[274,0.0008],[275,-0.0007],[276,-0.0011],[277,0.0024],[278,0.0022],[279,-0.0019],[280,0.0017],[283,0.0011],[284,-0.0029],[285,-0.003],[286,0.0006],[287,0.0006],[288,-0.003],[289,0.0009],[290,0.0011],[291,-0.0001],[292,0.002],[293,-0.0001],[304,-0.0001],[306,0]]
       }
       ring_metacarpal_r_ctrl.Rotation.Y {
        type=MovieSceneScriptingFloatChannel
        default=0
        thinned_from=307
        keys [[0,0.0007],[2,0.0007],[3,-0.0002],[4,-0.0003],[5,-0.0001],[6,-0.0004],[7,0.0004],[8,0.0002],[9,0.0006],[13,0.0006],[16,-0.0002],[17,0.0011],[18,0.0004],[19,-0.001],[21,-0.001],[23,0.0002],[24,-0.0002],[25,0.0006],[27,0.0001],[28,0.0008],[29,-0.0007],[30,0],[31,-0.0003],[32,-0.001],[33,-0.0004],[34,0.0011],[35,0.0011],[36,-0.0002],[37,0.0007],[38,-0.0006],[39,-0.001],[41,0.0006],[42,-0.0012],[43,-0.0007],[44,-0.0012],[45,-0.0006],[46,-0.0009],[48,0.0009],[50,-0.0016],[55,-0.0016],[56,-0.006],[57,-0.006],[58,0.0063],[61,0.0036],[62,0],[63,-0.0229],[64,-0.115],[65,-0.2331],[66,-0.3774],[67,-0.5548],[68,-0.8575],[69,-1.3837],[70,-1.8249],[71,-1.9859],[72,-2.163],[73,-2.3818],[74,-2.6713],[75,-2.9376],[76,-3.1056],[77,-3.208],[78,-3.2665],[79,-3.3153],[83,-3.491],[145,-3.4908],[146,-3.3867],[147,-3.1257],[149,-2.6865],[150,-2.4284],[151,-1.9369],[153,-0.5421],[154,-0.0759],[155,-0.0018],[156,0.0013],[157,-0.0042],[158,-0.0045],[159,-0.0119],[160,-0.0024],[161,-0.0074],[162,-0.0022],[163,-0.0008],[165,0.001],[166,-0.0006],[167,-0.0003],[168,-0.0003],[169,0.0007],[170,0.0005],[171,-0.0006],[172,0.0003],[173,-0.0004],[174,0],[175,0.0012],[176,0.0006],[177,-0.0008],[178,0.0008],[179,0.0008],[180,-0.0005],[182,0.0008],[183,-0.0012],[184,0.0008],[185,0.0001],[186,0.0005],[187,-0.0009],[188,0.0018],[189,0.0002],[191,0.0008],[194,0.0008],[195,-0.0001],[196,-0.0006],[198,0],[199,-0.0004],[200,0.0009],[201,0.0002],[202,0.0001],[203,0.0008],[204,0.0002],[205,0.0003],[206,-0.0005],[207,0.0001],[208,-0.0002],[210,-0.0003],[211,0.0011],[212,0.0007],[213,-0.0011],[215,0.0006],[216,-0.0006],[217,-0.0005],[218,0.0008],[219,-0.0004],[220,-0.0012],[221,-0.0002],[222,0.0012],[223,-0.0005],[226,0],[227,0.0011],[228,-0.0027],[229,0.0012],[230,-0.0007],[231,-0.0007],[232,0.0013],[235,0.0013],[236,-0.0007],[237,0.0012],[238,0.0011],[239,-0.001],[240,0.0008],[241,-0.0011],[242,-0.0012],[243,-0.0005],[244,-0.0005],[245,0.0003],[248,0.0002],[249,0.0007],[250,-0.0002],[252,-0.0001],[253,0.0006],[254,-0.0003],[255,-0.0002],[256,0.0002],[257,-0.0004],[258,0.0004],[259,-0.0004],[261,-0.0002],[263,-0.0005],[264,-0.0001],[265,-0.0001],[266,0.0005],[267,0.0005],[268,0.0004],[269,0.0005],[270,-0.0001],[271,0.0001],[272,-0.0004],[275,0.0006],[276,-0.0005],[277,0.0007],[278,0.0006],[279,-0.0009],[280,0.0012],[281,-0.0004],[282,0.0003],[283,-0.0013],[285,0.0003],[286,0.0009],[287,0.0009],[288,0.0003],[289,0.0009],[290,0.0002],[292,0.0002],[293,0.0006],[294,-0.0002],[296,0.0004],[297,-0.0001],[298,0.0003],[299,-0.0005],[300,-0.0004],[301,0.0005],[303,0.0001],[304,-0.0004],[305,0],[306,0]]
       }
       ring_metacarpal_r_ctrl.Rotation.Z {
        type=MovieSceneScriptingFloatChannel
        default=0
        thinned_from=307
        keys [[0,-0.0136],[2,-0.0136],[3,-0.076],[4,-0.1694],[5,-0.3249],[6,-0.5113],[7,-0.6357],[8,-0.7289],[9,-0.7911],[13,-0.7911],[17,-0.1693],[18,-0.0657],[19,-0.0139],[21,-0.0139],[23,-0.2212],[24,-0.3764],[25,-0.4803],[27,-0.791],[30,-1.1022],[31,-1.2574],[32,-1.3095],[33,-1.4128],[34,-1.4646],[35,-1.4646],[38,-1.3248],[39,-1.2316],[40,-1.1847],[43,-0.765],[44,-0.5788],[45,-0.4388],[46,-0.3456],[48,-0.252],[50,-0.159],[55,-0.159],[56,-0.1383],[57,-0.1383],[58,-0.1069],[61,-0.0137],[62,0],[63,-0.0041],[64,0.0036],[65,0.0003],[66,-0.0038],[67,0.0008],[68,-0.0077],[69,-1.5349],[70,-3.7314],[71,-4.968],[72,-5.1417],[73,-4.8966],[74,-3.113],[75,-0.5424],[76,1.6695],[77,4.2517],[78,6.7838],[79,8.5881],[80,9.0109],[81,5.2037],[82,0.3132],[83,-3.3136],[84,-3.4517],[145,-3.4517],[146,-2.9795],[147,-1.7158],[148,-0.4926],[149,0.0001],[150,0.0072],[151,-0.0196],[152,-0.0003],[153,0.0193],[154,-0.0078],[155,0.1157],[156,0.9256],[157,2.0599],[158,3.0327],[159,3.3163],[160,2.9518],[161,2.1813],[162,1.0471],[163,-0.43],[164,-2.6284],[165,-5.5014],[166,-8.3187],[167,-10.5172],[168,-11.4195],[169,-10.9682],[170,-9.7277],[171,-7.9806],[172,-6.0648],[173,-4.206],[174,-2.6847],[175,-1.2755],[176,0.2467],[177,1.7135],[178,2.8993],[179,3.6533],[180,3.8625],[181,3.3044],[182,2.2596],[183,1.1449],[184,-0.177],[185,-1.6379],[186,-2.9588],[187,-3.6541],[188,-4.1407],[189,-4.4188],[190,-4.6274],[191,-4.7664],[194,-4.7664],[195,-4.6417],[196,-4.4398],[197,-4.2039],[198,-3.8671],[199,-3.16],[200,-2.0147],[201,-0.8356],[202,0.2432],[203,1.2549],[204,2.1318],[205,2.807],[206,3.3472],[207,3.6513],[208,3.82],[209,3.82],[210,3.7862],[213,3.4369],[214,3.2063],[215,2.918],[216,2.5145],[217,1.9382],[218,1.0745],[219,0.268],[221,-1.1131],[222,-1.746],[223,-2.2638],[224,-2.724],[225,-3.1267],[226,-3.4715],[227,-3.6484],[229,-3.9049],[230,-3.9692],[231,-3.9692],[232,-4.0333],[235,-4.0333],[237,-3.9049],[238,-3.7765],[239,-3.584],[240,-3.5194],[241,-3.3272],[242,-3.1988],[243,-3.1206],[244,-3.1206],[245,-3.1747],[247,-3.3371],[248,-3.3371],[249,-2.8228],[250,-1.8211],[251,-0.6837],[252,0.2374],[253,0.9421],[254,1.6198],[255,2.2163],[256,2.7048],[257,3.1118],[258,3.4375],[259,3.6861],[260,3.7869],[261,3.7869],[262,3.6862],[263,3.5182],[265,2.9809],[266,2.6453],[267,2.2423],[269,1.303],[270,0.9003],[271,0.5315],[273,-0.0721],[274,-0.3402],[275,-0.5761],[277,-0.9904],[278,-1.1459],[279,-1.2498],[280,-1.4049],[286,-2.0266],[287,-2.0266],[288,-1.9233],[289,-1.8712],[290,-1.7676],[291,-1.6469],[292,-1.5295],[296,-0.9433],[297,-0.7674],[299,-0.4742],[300,-0.3569],[303,-0.0929],[304,-0.0343],[305,0],[306,0]]
       }
       ring_metacarpal_r_ctrl.Scale.X {type=MovieSceneScriptingFloatChannel, default=1, const [1,307]}
       ring_metacarpal_r_ctrl.Scale.Y {type=MovieSceneScriptingFloatChannel, default=1, const [1,307]}
       ring_metacarpal_r_ctrl.Scale.Z {type=MovieSceneScriptingFloatChannel, default=1, const [1,307]}
       ring_01_r_ctrl.Location.X {type=MovieSceneScriptingFloatChannel, default=0, thinned_from=307, keys [[0,0],[306,0]]}
       ring_01_r_ctrl.Location.Y {type=MovieSceneScriptingFloatChannel, default=0, thinned_from=307, keys [[0,0],[306,0]]}
       ring_01_r_ctrl.Location.Z {type=MovieSceneScriptingFloatChannel, default=0, thinned_from=307, keys [[0,0],[306,0]]}
       ring_01_r_ctrl.Rotation.X {
        type=MovieSceneScriptingFloatChannel
        default=0
        thinned_from=307
        keys [[0,0],[2,0],[3,-0.0005],[4,0.0017],[5,-0.0009],[6,0.0005],[7,0.0002],[8,0.0008],[9,0],[13,0],[14,0.0002],[15,-0.0011],[16,0.0015],[17,0.0008],[23,0.0008],[24,-0.001],[25,0.0008],[26,-0.001],[27,0.0008],[28,-0.001],[29,0.0002],[30,0.0007],[31,0.0007],[32,0.0002],[33,-0.0012],[34,0.001],[36,0.001],[37,0.002],[38,0.002],[39,0.001],[41,0.001],[42,-0.0015],[43,0.0011],[44,0.002],[45,0.001],[46,0.002],[47,0.002],[48,-0.0015],[49,-0.0015],[50,0.0019],[51,-0.0005],[54,-0.0005],[55,0.0012],[60,0.0012],[61,-0.0003],[62,-0.0001],[64,0.0007],[65,-0.0003],[66,-0.0001],[67,-0.0005],[68,0.0003],[69,-0.001],[70,-0.0012],[71,0.0003],[72,0],[73,-0.0017],[74,-0.0008],[76,-0.0001],[77,-0.0009],[78,0.0003],[79,0.0005],[80,-0.0003],[81,0.0007],[82,0.0003],[83,-0.0003],[84,0],[145,0.0001],[146,0.0005],[147,0.0001],[148,-0.0008],[149,-0.0001],[151,-0.0012],[153,-0.001],[155,-0.0015],[156,-0.0001],[157,0.0002],[158,0.0009],[159,0.0004],[160,0.0008],[161,0.0003],[163,0.0008],[164,0.0004],[165,0.0005],[166,0.0001],[167,0.0013],[168,0.0001],[169,-0.0015],[170,-0.0012],[171,0.0007],[172,0.0001],[173,0.0008],[174,0.0004],[175,-0.0009],[176,-0.0004],[177,0.0023],[178,-0.0018],[180,0.0008],[181,-0.0003],[182,-0.0001],[183,-0.0006],[184,0],[185,-0.0002],[186,0.0006],[187,0.0022],[188,-0.0009],[189,-0.0015],[190,0],[192,-0.0003],[194,-0.0002],[195,0.0002],[196,-0.0012],[198,-0.0001],[199,-0.0007],[200,-0.001],[201,-0.0003],[202,-0.0001],[203,-0.0004],[204,-0.001],[205,-0.0006],[206,-0.0008],[207,0.0004],[208,0.0002],[210,0.0002],[211,-0.0013],[212,-0.0012],[213,0.0008],[214,0.0006],[215,-0.0001],[216,-0.0003],[217,0.0004],[218,0.0006],[219,-0.0016],[220,0],[221,-0.0006],[222,0.0008],[223,-0.0006],[224,-0.0014],[225,0.0009],[227,-0.0005],[235,-0.0007],[236,0.0019],[237,-0.0006],[239,-0.0005],[240,0.0003],[242,0.0004],[243,0.0024],[244,0.0024],[245,-0.0005],[246,-0.0008],[248,-0.0008],[249,0.0001],[250,-0.0012],[251,-0.0001],[252,-0.0023],[253,0.0006],[255,0.0003],[256,-0.0009],[257,-0.0004],[258,-0.0007],[259,0.0001],[265,0.0003],[266,0.0009],[267,0.001],[268,0.0003],[269,-0.0016],[270,-0.0002],[271,0.0002],[272,-0.0001],[273,0.0013],[274,0.001],[275,-0.0007],[276,-0.0007],[277,0.0011],[278,0.0011],[279,-0.0008],[280,0.0005],[282,0.0004],[283,0.0009],[284,0.0002],[290,0.0003],[291,-0.0008],[292,-0.0008],[293,-0.0001],[294,0],[295,-0.0005],[296,0.0001],[300,0.0003],[301,-0.0012],[302,0.0003],[305,0.0003],[306,0]]
       }
       ring_01_r_ctrl.Rotation.Y {
        type=MovieSceneScriptingFloatChannel
        default=0
        thinned_from=307
        keys [[0,0],[2,0],[3,0.0014],[4,-0.0017],[5,0.0012],[6,0.0014],[7,0.0013],[8,-0.0022],[9,0.0011],[14,0.0013],[15,0.0011],[16,-0.0018],[17,-0.001],[23,-0.001],[24,0.0012],[25,-0.0011],[26,0.001],[27,-0.0013],[28,0.0009],[29,0.0011],[30,-0.0015],[31,-0.0016],[32,0.0009],[33,0.0005],[34,0.0013],[36,0.0013],[37,-0.0035],[38,-0.0035],[39,0.0014],[41,0.0016],[42,0.0011],[43,0.0018],[44,-0.003],[45,0.0019],[46,-0.0029],[47,-0.0029],[48,0.0015],[49,0.0015],[50,-0.0028],[51,0.0008],[54,0.0008],[55,-0.0013],[60,-0.0013],[61,0.0005],[62,0.0004],[63,0],[64,-0.0008],[66,0.0014],[67,0],[68,0.0005],[69,0.0001],[70,0.0032],[71,-0.0005],[72,-0.0006],[73,0.003],[74,0.0008],[75,0.0011],[76,0.0003],[77,0.0016],[78,-0.0017],[79,-0.0013],[80,0.0012],[81,-0.0015],[82,-0.0011],[83,0.0013],[98,0.0014],[99,0],[130,0],[131,-0.0002],[145,-0.0002],[148,0.0002],[152,0.0014],[153,0.0012],[154,0.0007],[155,0.0006],[156,0.0012],[157,-0.0017],[158,-0.0015],[159,0.0017],[160,-0.0015],[162,-0.0019],[163,-0.0004],[164,-0.0002],[165,-0.0008],[167,-0.0045],[168,-0.0003],[169,-0.0001],[170,0.0014],[171,-0.0022],[172,-0.0011],[173,-0.0003],[174,-0.0002],[177,-0.0011],[178,0.003],[179,-0.0012],[180,-0.0011],[181,0.0026],[182,-0.0002],[184,0.0006],[186,0.0007],[187,-0.0026],[188,0],[189,-0.0003],[190,0],[194,-0.0001],[195,0],[196,-0.0002],[198,0.0003],[201,0.0006],[202,0.0004],[204,-0.0004],[205,0.0027],[206,-0.0011],[210,-0.0012],[211,0.0005],[212,0.0006],[213,-0.0019],[214,0.0011],[215,-0.0016],[216,0.0015],[217,-0.0009],[218,-0.0004],[219,0.0024],[220,0],[221,0.0001],[222,-0.0025],[223,0.0001],[224,-0.0001],[225,0.0004],[226,0.0002],[227,0.0011],[235,0.0007],[236,-0.0022],[237,0.0009],[239,0.0012],[240,-0.002],[242,-0.0018],[243,0.0006],[244,0.0006],[245,-0.0001],[248,-0.0001],[249,0.0001],[250,-0.0002],[251,-0.0002],[252,-0.0009],[253,-0.0007],[255,-0.0015],[256,0.0035],[257,-0.0022],[258,-0.0026],[259,-0.0009],[260,-0.0011],[261,-0.0011],[263,-0.0007],[265,0.0001],[266,-0.0029],[267,-0.0023],[268,0.0015],[269,0.0017],[270,0.0024],[271,-0.0007],[272,-0.0004],[273,0.0001],[274,0.0002],[275,0.0011],[276,0.001],[277,-0.0014],[278,-0.0015],[279,0.0007],[280,0.0009],[282,0.0007],[283,-0.0019],[284,0.0005],[290,0.0006],[291,0.0019],[292,0.002],[293,-0.0007],[294,-0.0006],[295,0.0024],[296,-0.0004],[300,0],[301,-0.0003],[302,0],[305,0.0001],[306,0]]
       }
       ring_01_r_ctrl.Rotation.Z {
        type=MovieSceneScriptingFloatChannel
        default=0
        thinned_from=307
        keys [[0,0],[2,0],[3,-0.0806],[4,-0.1504],[5,-0.3592],[7,-0.638],[9,-0.7774],[13,-0.7774],[15,-0.4986],[16,-0.2897],[17,-0.1691],[18,-0.0642],[19,-0.0118],[21,-0.0118],[23,-0.2215],[24,-0.3787],[25,-0.4837],[26,-0.641],[27,-0.7461],[28,-0.9033],[33,-1.4285],[34,-1.444],[36,-1.444],[37,-1.3435],[38,-1.3435],[42,-0.9401],[44,-0.5376],[46,-0.3363],[47,-0.3363],[48,-0.2353],[49,-0.2353],[50,-0.135],[51,-0.169],[54,-0.169],[55,-0.1204],[60,-0.1204],[62,-0.023],[63,0.123],[64,0.3662],[65,0.658],[66,0.9982],[67,1.6651],[68,2.3637],[69,1.9446],[70,1.2458],[71,0.6858],[72,0.6157],[73,1.1059],[74,3.4799],[75,6.6103],[76,9.3139],[78,15.391],[79,17.5266],[80,18.0776],[81,15.2531],[82,11.596],[83,8.8445],[84,8.7298],[98,8.7298],[130,8.7436],[131,8.7625],[145,8.7625],[146,8.5658],[147,8.0784],[148,7.6933],[149,7.2439],[150,6.5372],[151,5.1864],[153,1.4426],[154,0.2116],[155,0.0819],[156,0.9246],[157,2.0893],[158,3.0586],[159,3.317],[160,2.994],[161,2.2187],[162,1.0539],[163,-0.3889],[164,-2.6428],[165,-5.4738],[166,-8.322],[167,-10.4995],[168,-11.4196],[169,-10.9592],[170,-9.6957],[171,-7.9794],[172,-6.0421],[173,-4.2262],[174,-2.6428],[175,-1.2893],[176,0.2855],[177,1.7441],[178,2.8642],[179,3.6694],[180,3.8766],[181,3.2552],[182,2.287],[183,1.1097],[184,-0.1393],[185,-1.6695],[186,-2.9247],[187,-3.6235],[188,-4.1129],[189,-4.4629],[191,-4.7431],[192,-4.8131],[193,-4.8131],[194,-4.7431],[195,-4.6269],[196,-4.4887],[197,-4.2123],[198,-3.8671],[199,-3.1773],[201,-0.8389],[202,0.2581],[203,1.2848],[204,2.1049],[205,2.7877],[206,3.3332],[207,3.674],[208,3.8103],[210,3.8103],[211,3.6668],[213,3.434],[214,3.2015],[215,2.9103],[216,2.5029],[217,1.92],[218,1.1033],[219,0.2272],[220,-0.4749],[221,-1.119],[223,-2.2921],[224,-2.7621],[226,-3.4679],[227,-3.6902],[228,-3.8304],[230,-3.9707],[232,-3.9707],[233,-4.0408],[235,-4.0408],[237,-3.9005],[242,-3.1999],[243,-3.0709],[244,-3.0709],[246,-3.2969],[248,-3.2969],[249,-2.8449],[250,-1.8294],[251,-0.7033],[252,0.196],[253,0.9817],[254,1.6544],[255,2.2145],[257,3.1096],[258,3.445],[259,3.6667],[260,3.8059],[261,3.8059],[263,3.5275],[266,2.6915],[267,2.2732],[269,1.2963],[270,0.8771],[271,0.5273],[273,-0.0324],[275,-0.5922],[276,-0.8021],[279,-1.2747],[283,-1.6952],[284,-1.8528],[285,-1.9579],[286,-2.0105],[287,-2.0105],[288,-1.9579],[290,-1.7476],[291,-1.6612],[292,-1.541],[293,-1.3611],[295,-1.1208],[297,-0.761],[298,-0.641],[299,-0.4612],[300,-0.3413],[301,-0.2814],[302,-0.1615],[305,0.0182],[306,0]]
       }
       ring_01_r_ctrl.Scale.X {type=MovieSceneScriptingFloatChannel, default=1, const [1,307]}
       ring_01_r_ctrl.Scale.Y {type=MovieSceneScriptingFloatChannel, default=1, const [1,307]}
       ring_01_r_ctrl.Scale.Z {type=MovieSceneScriptingFloatChannel, default=1, const [1,307]}
       ring_02_r_ctrl.Location.X {type=MovieSceneScriptingFloatChannel, default=0, thinned_from=307, keys [[0,0],[306,0]]}
       ring_02_r_ctrl.Location.Y {type=MovieSceneScriptingFloatChannel, default=0, thinned_from=307, keys [[0,0],[306,0]]}
       ring_02_r_ctrl.Location.Z {type=MovieSceneScriptingFloatChannel, default=0, thinned_from=307, keys [[0,0],[306,0]]}
       ring_02_r_ctrl.Rotation.X {type=MovieSceneScriptingFloatChannel, default=0, thinned_from=307, keys [[0,0],[306,0]]}
       ring_02_r_ctrl.Rotation.Y {type=MovieSceneScriptingFloatChannel, default=0, thinned_from=307, keys [[0,0],[306,0]]}
       ring_02_r_ctrl.Rotation.Z {
        type=MovieSceneScriptingFloatChannel
        default=0
        thinned_from=307
        keys [[0,0],[3,0],[4,-0.1558],[8,-0.7688],[13,-0.7688],[16,-0.3089],[17,-0.1711],[18,-0.0639],[19,-0.0639],[20,0.0433],[26,-0.6001],[27,-0.8148],[33,-1.4596],[37,-1.4596],[38,-1.2484],[40,-1.2484],[42,-0.8263],[43,-0.8263],[45,-0.4047],[46,-0.4047],[47,-0.194],[50,-0.194],[51,-0.1175],[61,-0.1175],[63,0.1504],[65,0.9529],[66,1.487],[67,2.4201],[68,3.3878],[69,3.7098],[70,3.7098],[71,3.3878],[72,3.3878],[73,4.0317],[74,6.5984],[75,9.9473],[76,12.9604],[77,16.2749],[78,19.4187],[79,21.77],[80,22.3962],[81,19.5756],[82,15.9598],[83,13.3476],[84,13.2172],[98,13.2172],[99,13.2506],[145,13.2506],[146,12.8258],[147,12.1123],[148,11.4423],[149,10.6757],[150,9.716],[151,7.6953],[152,4.8916],[153,2.1685],[154,0.3106],[155,0.1146],[156,0.8982],[157,2.0708],[158,3.0456],[159,3.3375],[160,2.9482],[161,2.1684],[162,0.9961],[163,-0.4174],[164,-2.6802],[165,-5.4145],[166,-8.4023],[167,-10.4878],[168,-11.4195],[169,-10.9533],[170,-9.7911],[172,-6.1016],[173,-4.2726],[174,-2.6802],[175,-1.3209],[176,0.2588],[177,1.6078],[178,2.9524],[179,3.6701],[180,3.8771],[181,3.256],[182,2.288],[183,1.1097],[184,-0.1417],[185,-1.6767],[186,-2.9373],[187,-3.6395],[188,-4.1319],[189,-4.4136],[190,-4.625],[191,-4.7661],[194,-4.7661],[195,-4.7009],[198,-3.862],[199,-3.1645],[200,-2.0512],[201,-0.8029],[202,0.3034],[203,1.2688],[204,2.0945],[205,2.7814],[206,3.3301],[208,3.8781],[209,3.8781],[210,3.7411],[211,3.6505],[213,3.4186],[215,2.9545],[216,2.4899],[217,1.9084],[219,0.2758],[220,-0.426],[221,-1.129],[223,-2.3035],[224,-2.7744],[226,-3.4819],[227,-3.6532],[228,-3.7975],[231,-4.0141],[236,-4.0141],[239,-3.5811],[240,-3.509],[242,-3.2206],[243,-3.1237],[245,-3.1237],[247,-3.3468],[248,-3.3468],[249,-2.7894],[250,-1.7884],[251,-0.6793],[252,0.2056],[253,0.9784],[254,1.6395],[256,2.7391],[258,3.3975],[259,3.6674],[260,3.8064],[261,3.8064],[263,3.5283],[266,2.6927],[267,2.2743],[269,1.2964],[270,0.8766],[271,0.5264],[273,-0.0345],[274,-0.3153],[275,-0.5542],[277,-1.0045],[278,-1.1171],[279,-1.2862],[285,-1.9631],[286,-2.0195],[287,-2.0195],[288,-1.9631],[290,-1.7373],[291,-1.6031],[296,-0.9759],[297,-0.7253],[301,-0.2247],[302,-0.2247],[303,-0.0996],[304,-0.0996],[305,0.0254],[306,0]]
       }
       ring_02_r_ctrl.Scale.X {type=MovieSceneScriptingFloatChannel, default=1, const [1,307]}
       ring_02_r_ctrl.Scale.Y {type=MovieSceneScriptingFloatChannel, default=1, const [1,307]}
       ring_02_r_ctrl.Scale.Z {type=MovieSceneScriptingFloatChannel, default=1, const [1,307]}
       ring_03_r_ctrl.Location.X {type=MovieSceneScriptingFloatChannel, default=0, thinned_from=307, keys [[0,0],[306,0]]}
       ring_03_r_ctrl.Location.Y {type=MovieSceneScriptingFloatChannel, default=0, thinned_from=307, keys [[0,0],[306,0]]}
       ring_03_r_ctrl.Location.Z {type=MovieSceneScriptingFloatChannel, default=0, thinned_from=307, keys [[0,0],[306,0]]}
       ring_03_r_ctrl.Rotation.X {type=MovieSceneScriptingFloatChannel, default=0, thinned_from=307, keys [[0,0],[306,0]]}
       ring_03_r_ctrl.Rotation.Y {type=MovieSceneScriptingFloatChannel, default=0, thinned_from=307, keys [[0,0],[306,0]]}
       ring_03_r_ctrl.Rotation.Z {
        type=MovieSceneScriptingFloatChannel
        default=0
        thinned_from=307
        keys [[0,-0.016],[3,-0.016],[5,-0.2906],[6,-0.5654],[7,-0.5654],[9,-0.8402],[13,-0.8402],[14,-0.7028],[15,-0.428],[16,-0.2906],[17,-0.1991],[19,0.0146],[21,0.0146],[24,-0.3059],[25,-0.5196],[27,-0.7333],[28,-0.9471],[32,-1.3748],[33,-1.3748],[34,-1.4818],[35,-1.4818],[36,-1.3855],[37,-1.3855],[38,-1.2893],[39,-1.2893],[41,-1.0968],[43,-0.7119],[44,-0.6157],[45,-0.4234],[46,-0.3272],[47,-0.3272],[48,-0.2311],[49,-0.2311],[50,-0.135],[51,-0.1991],[53,-0.1991],[54,-0.0922],[60,-0.0922],[61,0.0146],[66,0.0146],[67,0.028],[68,0.028],[69,-1.5538],[70,-3.7053],[71,-4.9536],[72,-5.1809],[73,-4.9536],[74,-3.1385],[75,-0.5366],[76,1.7199],[77,4.1968],[78,6.7816],[79,8.5775],[80,9.0263],[81,6.1077],[82,2.2833],[83,-0.5716],[98,-0.5716],[99,-0.6855],[146,-0.6855],[147,-0.3593],[148,-0.1164],[149,0],[154,0],[155,0.1265],[156,0.9761],[157,2.0674],[158,3.0366],[159,3.2788],[160,2.9155],[161,2.1886],[162,1.0974],[163,-0.4432],[164,-2.5798],[165,-5.5129],[166,-8.3452],[167,-10.5068],[168,-11.4195],[169,-10.963],[170,-9.7095],[171,-8.0047],[173,-4.1577],[175,-1.2298],[176,0.2304],[177,1.6885],[178,2.9206],[179,3.6799],[180,3.8193],[181,3.2617],[182,2.2852],[183,1.1683],[184,-0.0897],[185,-1.6295],[186,-2.8914],[187,-3.5933],[188,-4.1553],[189,-4.4365],[191,-4.7178],[194,-4.7178],[195,-4.6973],[198,-3.8692],[199,-3.1798],[200,-1.9404],[202,0.2586],[204,2.1787],[205,2.8636],[207,3.685],[208,3.8218],[210,3.8218],[211,3.6837],[213,3.4531],[214,3.2224],[216,2.5301],[217,1.9528],[218,1.1442],[219,0.2192],[221,-1.1698],[223,-2.329],[224,-2.7932],[226,-3.4899],[227,-3.6393],[229,-3.9152],[231,-3.9152],[232,-4.0531],[236,-4.0531],[242,-3.2257],[243,-3.1838],[246,-3.1838],[247,-3.4095],[248,-3.4095],[249,-2.7327],[250,-1.8313],[251,-0.7057],[252,0.1936],[255,2.2144],[257,3.1113],[258,3.3354],[259,3.6774],[260,3.8117],[261,3.8117],[262,3.6774],[264,3.2743],[265,3.0056],[266,2.6696],[267,2.2662],[269,1.3245],[271,0.5166],[272,0.2472],[273,-0.0897],[274,-0.3593],[275,-0.6684],[276,-0.8974],[277,-0.8974],[279,-1.3557],[280,-1.3557],[281,-1.5849],[282,-1.5849],[283,-1.8142],[284,-1.8142],[285,-2.0436],[288,-2.0436],[289,-1.8142],[290,-1.8142],[291,-1.6423],[294,-1.2893],[295,-1.054],[297,-0.8188],[298,-0.5837],[301,-0.2311],[302,-0.2311],[303,-0.1136],[304,0],[306,0]]
       }
       ring_03_r_ctrl.Scale.X {type=MovieSceneScriptingFloatChannel, default=1, const [1,307]}
       ring_03_r_ctrl.Scale.Y {type=MovieSceneScriptingFloatChannel, default=1, const [1,307]}
       ring_03_r_ctrl.Scale.Z {type=MovieSceneScriptingFloatChannel, default=1, const [1,307]}
       pinky_metacarpal_r_ctrl.Location.X {type=MovieSceneScriptingFloatChannel, default=0, thinned_from=307, keys [[0,0],[306,0]]}
       pinky_metacarpal_r_ctrl.Location.Y {type=MovieSceneScriptingFloatChannel, default=0, thinned_from=307, keys [[0,0],[306,0]]}
       pinky_metacarpal_r_ctrl.Location.Z {type=MovieSceneScriptingFloatChannel, default=0, thinned_from=307, keys [[0,0],[306,0]]}
       pinky_metacarpal_r_ctrl.Rotation.X {
        type=MovieSceneScriptingFloatChannel
        default=0
        thinned_from=307
        keys [[0,0],[2,0],[3,0.0007],[5,-0.0005],[6,0.0042],[7,0.0036],[8,-0.0024],[9,0.0029],[13,0.0029],[14,-0.0019],[16,-0.0005],[17,0.0015],[18,0.0021],[19,0.0026],[20,-0.0016],[21,-0.0018],[22,0.0019],[24,0.0002],[25,-0.0005],[26,-0.0016],[27,0.0018],[28,0.0007],[32,-0.002],[33,0.0016],[34,-0.0015],[36,-0.0015],[37,0.0029],[38,-0.0007],[39,-0.0005],[40,-0.0001],[42,0.0011],[43,-0.002],[44,0.0028],[45,-0.0007],[47,0.0002],[48,0.0003],[49,-0.0034],[50,-0.0034],[51,0.0014],[52,0.0014],[53,-0.0004],[57,-0.0004],[58,0.004],[59,0.0004],[60,-0.0014],[61,0],[62,0],[63,0.0039],[64,-0.0023],[65,-0.0013],[66,-0.0022],[67,0.0019],[68,-0.0006],[69,-0.0005],[70,0.0008],[71,0.0016],[72,0.0014],[73,-0.0012],[74,0.001],[75,-0.0003],[76,0.0004],[77,0.0005],[78,-0.0024],[79,-0.0002],[80,-0.0003],[81,0.0031],[82,-0.0013],[83,0.0001],[84,0.0021],[130,0.0021],[131,-0.0007],[145,-0.0007],[146,0.0019],[147,0.0018],[148,0.0008],[149,-0.0007],[150,0.0009],[151,0.0018],[152,0.0012],[153,0.0002],[154,-0.0005],[155,0],[156,0.0002],[157,0.0019],[158,-0.0035],[159,0.0028],[160,-0.003],[161,0.0013],[162,0.0028],[163,-0.0012],[164,-0.0002],[165,0.0028],[166,-0.0017],[167,-0.003],[169,0.0001],[171,-0.001],[172,0.0022],[173,-0.0006],[174,-0.0002],[175,-0.0006],[176,0.0025],[177,0.0003],[178,0.0026],[179,-0.0022],[180,-0.0021],[181,0.0004],[182,0.0022],[183,0.0007],[184,-0.0017],[185,0.0004],[186,-0.0005],[187,-0.0001],[188,0.0011],[189,-0.0001],[190,-0.0009],[191,-0.0013],[192,0.0013],[193,-0.0015],[194,-0.0013],[195,0.0007],[196,-0.0011],[197,0.0002],[198,0.0019],[199,-0.0002],[200,-0.0005],[201,0.0015],[202,-0.0002],[203,0.0004],[204,0.0001],[205,-0.0009],[206,-0.0026],[207,0.001],[208,-0.0015],[210,0.0007],[211,0.0009],[212,0.0002],[213,0.0001],[214,0.0006],[215,0.0004],[218,-0.0012],[219,0.0006],[220,0.0018],[221,-0.0005],[222,0.0009],[223,-0.0006],[224,-0.0013],[225,-0.0005],[226,-0.0019],[227,0.0021],[228,-0.0014],[231,0.0039],[235,0.0039],[236,-0.0001],[237,-0.0006],[238,0.0008],[239,-0.0031],[240,-0.0028],[241,-0.0041],[242,-0.0035],[243,-0.0008],[245,0.0006],[246,0.0004],[247,0.0022],[248,-0.0012],[250,0.0006],[251,-0.0002],[253,-0.0001],[254,-0.0004],[255,0.0003],[256,-0.0006],[257,0.0004],[258,-0.0002],[259,0.0011],[260,-0.0013],[262,-0.0011],[263,0.0017],[265,0.0016],[266,-0.0012],[267,0.0016],[268,-0.0014],[269,-0.0018],[270,0.0007],[274,-0.0005],[275,0.001],[276,-0.0017],[277,-0.0002],[279,-0.0009],[280,0.0008],[283,-0.0001],[284,0.0017],[286,0.0012],[288,0.0014],[289,-0.0005],[290,-0.0002],[291,-0.0018],[292,0.0007],[293,-0.0015],[295,-0.0012],[296,0.0012],[297,-0.0009],[300,-0.0006],[301,0.0017],[302,-0.0005],[304,-0.0004],[305,0.002],[306,0]]
       }
       pinky_metacarpal_r_ctrl.Rotation.Y {
        type=MovieSceneScriptingFloatChannel
        default=0
        thinned_from=307
        keys [[0,0],[2,0],[3,0.0008],[5,-0.0003],[6,0.0014],[7,0.0008],[8,-0.0019],[9,0.0003],[13,0.0003],[14,0.0012],[15,-0.0009],[16,-0.0003],[18,-0.0008],[19,0.0009],[20,-0.0008],[21,0.0012],[22,0.0012],[23,-0.0023],[24,0],[25,0.0003],[26,-0.0011],[27,0.001],[28,-0.0004],[32,0.001],[33,0.0011],[34,-0.0011],[36,-0.0011],[37,0.0005],[38,0.0006],[39,-0.0013],[40,-0.0014],[41,0.0022],[42,0.0001],[43,0.0002],[44,0.0016],[45,-0.0001],[47,-0.0002],[48,-0.0021],[49,0],[50,0],[51,0.0099],[52,0.0099],[53,0.0092],[57,0.0092],[58,0.008],[59,0.0066],[60,0.0058],[61,0],[62,0],[63,-0.0502],[64,-0.2122],[66,-0.7518],[67,-1.1236],[68,-1.7348],[69,-2.7996],[70,-3.6954],[71,-4.0204],[72,-4.3686],[73,-4.8053],[74,-5.3888],[75,-5.927],[76,-6.2769],[77,-6.4904],[79,-6.6991],[81,-6.8675],[83,-7.0554],[145,-7.0496],[146,-6.8391],[147,-6.328],[148,-5.8683],[149,-5.4254],[150,-4.9092],[151,-3.878],[153,-1.0924],[154,-0.1623],[155,-0.0106],[156,0.0087],[157,0.0149],[158,-0.0216],[159,0.0192],[160,-0.0187],[161,0.0105],[162,0.004],[163,0.0013],[164,0.0002],[165,0.0005],[166,-0.001],[167,-0.0006],[168,0.0006],[169,0],[170,-0.0011],[171,-0.0008],[172,0.0019],[173,0.0018],[174,-0.0001],[175,0.0016],[176,0],[177,-0.0001],[178,0.0012],[179,0.0002],[180,-0.0003],[181,0.0006],[182,-0.0002],[183,0.001],[184,-0.0002],[185,0.0003],[186,-0.0001],[187,-0.0007],[188,0.0005],[190,0.0005],[191,-0.0009],[192,0.0005],[193,-0.0009],[194,0.0005],[195,-0.0002],[196,-0.0006],[197,0.0001],[198,0.0007],[199,0.0008],[200,-0.0007],[201,0.0001],[202,-0.0008],[203,0.0002],[204,0.0003],[205,0.0007],[206,-0.0004],[207,0.0001],[208,-0.0005],[209,-0.0469],[210,-0.2379],[211,-0.5003],[212,-0.8013],[213,-1.1543],[214,-1.5182],[215,-1.9712],[216,-2.4735],[217,-2.8215],[218,-2.9997],[219,-3.0838],[220,-3.1337],[222,-3.1593],[223,-3.1891],[224,-3.2622],[225,-3.3904],[226,-3.5352],[227,-3.6415],[228,-3.6836],[235,-3.6754],[236,-3.6273],[237,-3.5371],[239,-3.2606],[240,-3.1704],[241,-3.128],[242,-3.0377],[243,-2.9046],[244,-2.6041],[245,-2.1246],[246,-1.5436],[247,-1.0001],[248,-0.6238],[249,-0.4322],[250,-0.3346],[251,-0.2804],[252,-0.2303],[253,-0.1701],[254,-0.121],[255,-0.0926],[256,-0.0618],[257,-0.0395],[258,-0.0138],[259,-0.0054],[260,-0.0006],[261,-0.0006],[262,0.0004],[263,0.0009],[265,-0.0005],[266,-0.001],[267,0.0005],[269,-0.0008],[270,-0.0002],[271,-0.0006],[272,0.0006],[274,0.0003],[275,0.0009],[276,0.0002],[277,-0.0007],[278,0.0001],[280,-0.0006],[282,0.0005],[284,0],[285,0.0006],[286,0.0009],[287,0.0009],[289,0.0001],[290,-0.0005],[291,0],[292,0.0008],[295,0.0001],[296,-0.0003],[297,0.0003],[298,-0.001],[299,0.0006],[300,-0.0002],[301,-0.0001],[302,-0.0008],[303,0.0006],[304,0.0001],[305,0.0007],[306,0]]
       }
       pinky_metacarpal_r_ctrl.Rotation.Z {
        type=MovieSceneScriptingFloatChannel
        default=0
        thinned_from=307
        keys [[0,0],[2,0],[3,-0.0722],[5,-0.331],[6,-0.5246],[7,-0.654],[9,-0.7833],[13,-0.7833],[14,-0.6544],[15,-0.4604],[16,-0.331],[17,-0.1754],[18,-0.0763],[19,-0.0266],[20,0.0225],[21,-0.027],[22,-0.1258],[23,-0.2251],[24,-0.3737],[25,-0.4728],[28,-0.9186],[33,-1.4137],[34,-1.4479],[36,-1.4479],[37,-1.4006],[38,-1.307],[39,-1.2602],[41,-1.0723],[42,-0.9316],[43,-0.7441],[45,-0.4624],[47,-0.2746],[49,-0.181],[50,-0.181],[51,-0.158],[52,-0.158],[53,-0.1366],[57,-0.1366],[58,-0.1147],[59,-0.072],[60,-0.0506],[61,0],[62,0],[63,-0.0093],[64,0.0053],[65,-0.0053],[66,0.0055],[67,-0.0038],[68,0],[69,-1.5476],[70,-3.7311],[71,-4.9717],[72,-5.1539],[73,-4.8936],[74,-3.1304],[75,-0.5504],[76,1.6787],[77,4.2271],[78,6.7859],[79,8.5898],[80,9.0117],[81,5.2005],[82,0.3084],[83,-3.3072],[84,-3.459],[130,-3.459],[131,-3.4371],[145,-3.4371],[146,-2.9952],[147,-1.7454],[148,-0.479],[149,0.0057],[150,-0.0141],[151,-0.0145],[152,-0.0004],[153,0.0149],[154,0.0128],[155,0.0965],[156,0.9131],[157,2.0797],[158,3.0505],[159,3.325],[160,2.9728],[161,2.1961],[162,1.0297],[163,-0.4347],[164,-2.6339],[165,-5.5068],[166,-8.3225],[167,-10.5186],[168,-11.4196],[169,-10.969],[170,-9.6738],[171,-7.9847],[172,-6.07],[173,-4.2116],[174,-2.6903],[175,-1.2808],[176,0.2428],[177,1.711],[178,2.8983],[179,3.6867],[180,3.8584],[181,3.275],[182,2.2458],[183,1.1144],[184,-0.1532],[185,-1.659],[186,-2.9246],[187,-3.6426],[188,-4.121],[189,-4.4287],[190,-4.6338],[191,-4.7365],[192,-4.8045],[194,-4.7364],[195,-4.6432],[196,-4.4421],[197,-4.2072],[198,-3.8715],[199,-3.1671],[200,-1.9924],[201,-0.8504],[202,0.2584],[203,1.2673],[204,2.1084],[205,2.8154],[206,3.3542],[207,3.6578],[208,3.826],[209,3.8244],[210,3.7837],[211,3.6862],[212,3.5618],[213,3.4073],[214,3.1958],[215,2.8961],[216,2.5096],[217,1.9304],[218,1.1026],[219,0.2505],[220,-0.4587],[221,-1.1097],[222,-1.7321],[223,-2.2704],[224,-2.7255],[225,-3.1263],[226,-3.4431],[227,-3.682],[228,-3.8134],[231,-4.0071],[236,-4.0053],[242,-3.2041],[243,-3.1229],[244,-3.1105],[245,-3.1715],[247,-3.3401],[248,-3.3245],[249,-2.8049],[250,-1.8315],[251,-0.698],[252,0.2474],[253,0.9509],[254,1.6271],[255,2.1951],[256,2.7092],[257,3.1152],[258,3.4403],[259,3.6565],[260,3.7918],[261,3.7918],[262,3.6904],[263,3.4876],[264,3.2508],[265,2.9802],[266,2.6757],[268,1.7971],[269,1.3241],[270,0.8854],[272,0.2103],[274,-0.3296],[275,-0.5723],[277,-0.983],[278,-1.1369],[280,-1.3933],[282,-1.5985],[284,-1.8548],[285,-1.9574],[286,-2.0087],[287,-2.0087],[288,-1.9574],[290,-1.7524],[291,-1.6281],[292,-1.5151],[296,-0.9508],[298,-0.6123],[299,-0.4711],[300,-0.3582],[301,-0.2733],[302,-0.1606],[304,-0.0477],[305,-0.0192],[306,0]]
       }
       pinky_metacarpal_r_ctrl.Scale.X {type=MovieSceneScriptingFloatChannel, default=1, const [1,307]}
       pinky_metacarpal_r_ctrl.Scale.Y {type=MovieSceneScriptingFloatChannel, default=1, const [1,307]}
       pinky_metacarpal_r_ctrl.Scale.Z {type=MovieSceneScriptingFloatChannel, default=1, const [1,307]}
       pinky_01_r_ctrl.Location.X {type=MovieSceneScriptingFloatChannel, default=0, thinned_from=307, keys [[0,0],[306,0]]}
       pinky_01_r_ctrl.Location.Y {type=MovieSceneScriptingFloatChannel, default=0, thinned_from=307, keys [[0,0],[306,0]]}
       pinky_01_r_ctrl.Location.Z {type=MovieSceneScriptingFloatChannel, default=0, thinned_from=307, keys [[0,0],[306,0]]}
       pinky_01_r_ctrl.Rotation.X {
        type=MovieSceneScriptingFloatChannel
        default=0
        thinned_from=307
        keys [[0,-14.1908],[61,-14.1765],[62,-14.2086],[63,-14.3701],[64,-14.957],[65,-15.8134],[66,-16.7583],[67,-17.9789],[68,-19.8847],[69,-23.643],[70,-25.0627],[71,-23.3338],[72,-21.1641],[73,-18.744],[74,-16.2125],[75,-14.096],[76,-12.8404],[77,-12.1235],[79,-11.3314],[80,-11.219],[81,-12.0681],[82,-13.1387],[83,-13.9409],[145,-13.9721],[146,-13.4518],[148,-10.7584],[149,-10.2184],[150,-10.2251],[151,-10.2597],[152,-10.4211],[153,-10.8554],[154,-11.4203],[155,-12.0281],[157,-13.3448],[158,-13.8554],[159,-14.1325],[161,-14.2074],[162,-14.2164],[208,-14.1957],[209,-14.3318],[210,-14.9159],[211,-15.724],[212,-16.6611],[213,-17.6703],[214,-18.7979],[215,-20.1576],[216,-21.6715],[217,-22.745],[218,-23.2643],[219,-23.5108],[220,-23.7007],[222,-23.7499],[223,-23.8467],[224,-24.1011],[225,-24.4762],[226,-24.9337],[227,-25.2549],[228,-25.3196],[235,-25.3124],[236,-25.2502],[237,-24.9229],[238,-24.4625],[239,-24.0779],[240,-23.7506],[241,-23.6249],[242,-23.4306],[243,-22.9881],[244,-22.0833],[245,-20.6039],[246,-18.8805],[247,-17.2436],[248,-16.1158],[249,-15.492],[250,-15.1988],[252,-14.8999],[254,-14.5514],[256,-14.3729],[259,-14.2115],[305,-14.1933],[306,0]]
       }
       pinky_01_r_ctrl.Rotation.Y {
        type=MovieSceneScriptingFloatChannel
        default=0
        thinned_from=307
        keys [[0,-8.3771],[22,-8.3715],[23,-8.382],[37,-8.369],[41,-8.3811],[42,-8.3699],[43,-8.3801],[44,-8.3662],[48,-8.3813],[54,-8.3797],[55,-8.3665],[58,-8.381],[59,-8.363],[61,-8.3644],[63,-8.378],[65,-8.468],[67,-8.6062],[68,-8.4729],[69,-7.7883],[70,-6.5741],[71,-4.9868],[72,-3.3255],[73,-1.7857],[74,-0.5931],[75,0.0738],[76,0.1646],[77,0.1159],[78,0.1062],[79,0.0801],[80,0.1706],[81,2.2271],[82,4.8866],[83,6.8329],[84,5.5685],[85,3.1409],[86,1.5186],[87,0.5311],[88,0.1943],[89,0.0319],[90,0.0494],[91,2.0699],[92,4.1326],[93,5.2152],[94,5.8595],[95,6.0576],[96,5.8194],[97,4.1659],[98,1.3799],[99,0.3236],[100,0.1298],[101,-0.0066],[102,0.3102],[103,2.3781],[104,4.406],[105,5.9655],[106,6.6004],[107,6.8244],[108,6.9174],[109,5.9959],[110,3.5453],[111,1.813],[112,0.6411],[113,0.2198],[114,0.0423],[115,0.0087],[116,1.3564],[117,3.5347],[118,4.9283],[119,5.8523],[120,6.1895],[121,4.246],[122,1.4524],[123,0.6607],[124,0.2749],[125,0.1382],[126,0.0348],[127,0.0609],[128,1.6857],[129,3.6093],[130,5.123],[131,5.783],[132,5.9976],[133,6.1646],[134,5.783],[135,3.7595],[136,2.0134],[137,0.7498],[138,0.2976],[139,0.1357],[140,0.0089],[141,0.8069],[142,2.7359],[143,4.4139],[144,5.6638],[145,6.1646],[146,5.3903],[147,3.2905],[148,1.1371],[149,0.2969],[150,0.3958],[151,0.6732],[152,0.658],[153,0.0465],[154,-1.1327],[155,-2.779],[156,-4.542],[157,-6.135],[158,-7.4432],[159,-8.2602],[160,-8.3969],[161,-8.3717],[202,-8.3676],[203,-8.3804],[208,-8.3762],[209,-8.4422],[210,-8.7141],[211,-9.1052],[212,-9.5591],[213,-10.0549],[214,-10.5942],[215,-11.2493],[216,-11.9736],[217,-12.4938],[218,-12.7362],[219,-12.8644],[220,-12.9492],[222,-12.9769],[223,-13.0168],[224,-13.1423],[225,-13.3266],[226,-13.5481],[227,-13.7003],[229,-13.7342],[235,-13.7341],[236,-13.6947],[237,-13.5382],[239,-13.1262],[240,-12.9764],[242,-12.8163],[243,-12.6078],[244,-12.171],[245,-11.4576],[246,-10.6313],[247,-9.8404],[248,-9.2963],[249,-9.0022],[250,-8.858],[252,-8.713],[253,-8.6183],[254,-8.5483],[256,-8.4621],[259,-8.3821],[293,-8.373],[294,-8.3815],[295,-8.3664],[297,-8.3807],[298,-8.3631],[300,-8.3775],[305,-8.3686],[306,0]]
       }
       pinky_01_r_ctrl.Rotation.Z {
        type=MovieSceneScriptingFloatChannel
        default=0
        thinned_from=307
        keys [[0,-5.1427],[2,-5.1427],[3,-5.2056],[5,-5.4588],[6,-5.6483],[7,-5.7751],[10,-5.9648],[12,-5.9648],[13,-5.9015],[14,-5.7751],[15,-5.585],[17,-5.3373],[19,-5.1233],[20,-5.1233],[22,-5.3373],[23,-5.6585],[24,-5.873],[28,-7.1618],[30,-7.5922],[31,-7.9155],[33,-8.3469],[34,-8.3788],[37,-8.3788],[39,-8.2033],[40,-8.2033],[41,-8.1159],[42,-7.9409],[43,-7.8531],[44,-7.6787],[47,-7.4157],[48,-7.4157],[49,-7.3289],[50,-7.3289],[51,-7.3053],[52,-7.381],[54,-7.381],[58,-7.6841],[59,-7.6845],[60,-7.7602],[61,-7.9878],[62,-8.2892],[63,-8.8894],[64,-9.6235],[65,-10.6559],[66,-11.8438],[67,-13.2155],[68,-15.7974],[69,-22.4227],[70,-24.9987],[71,-23.1075],[72,-20.9627],[73,-17.9364],[75,-9.0501],[76,-5.4708],[77,-1.7599],[78,1.9684],[79,4.519],[80,5.3822],[81,4.1297],[82,3.1931],[83,1.9354],[84,-3.2363],[85,-9.5483],[86,-14.323],[87,-17.27],[88,-18.2232],[89,-18.5393],[90,-19.0823],[91,-24.5659],[92,-23.705],[93,-10.2105],[94,-2.9453],[95,-0.9958],[96,-1.7019],[97,-6.5749],[98,-14.742],[99,-17.9028],[100,-18.392],[101,-18.6851],[102,-19.6048],[103,-22.9587],[104,-17.454],[105,-4.7571],[106,-0.5916],[107,0.9146],[108,1.4778],[109,-1.2178],[110,-8.36],[111,-13.5271],[112,-16.9231],[113,-18.0967],[114,-18.6876],[115,-18.7876],[116,-14.8071],[117,-8.4404],[118,-4.3809],[119,-1.6249],[120,-0.6403],[121,-6.3696],[122,-14.589],[123,-16.8634],[124,-17.9709],[125,-18.4166],[126,-18.7133],[127,-18.6391],[128,-13.8582],[129,-8.2255],[130,-3.8144],[131,-1.8365],[132,-1.2029],[133,-0.7106],[134,-1.8365],[135,-7.7967],[136,-12.9125],[137,-16.6432],[138,-17.8967],[139,-18.4165],[140,-18.7875],[141,-16.4212],[142,-10.7396],[143,-5.8712],[144,-2.1889],[145,-0.7106],[146,-0.6938],[147,-0.9262],[148,-1.1416],[149,-1.2063],[150,-0.6135],[151,0.4191],[152,1.3548],[153,1.6583],[154,1.4382],[155,0.8459],[156,0.1692],[157,-0.4537],[158,-1.0838],[159,-1.6894],[160,-2.1819],[161,-2.9327],[162,-4.0865],[163,-5.6023],[164,-7.7134],[165,-10.6702],[166,-13.4659],[167,-15.7202],[168,-16.5698],[169,-16.0031],[170,-14.685],[171,-12.9048],[172,-10.9489],[173,-9.1888],[174,-7.8975],[175,-6.9777],[176,-6.1518],[177,-5.5108],[178,-5.0536],[179,-4.7328],[180,-4.7328],[181,-5.494],[182,-6.9253],[183,-8.4578],[184,-10.2861],[185,-12.4152],[186,-14.1676],[187,-15.1454],[188,-15.7337],[189,-16.2247],[190,-16.5198],[192,-16.7166],[193,-16.7166],[194,-16.6181],[195,-16.4475],[196,-16.0772],[198,-14.8468],[199,-13.3771],[201,-8.5246],[202,-6.1232],[203,-4.0928],[204,-2.7847],[205,-2.1914],[207,-1.4805],[208,-1.3618],[209,-1.3527],[210,-1.4313],[211,-1.7388],[212,-2.2023],[213,-2.7905],[214,-3.4331],[215,-4.117],[216,-4.9839],[217,-6.3039],[218,-8.1038],[219,-10.012],[220,-11.5916],[221,-13.0547],[222,-14.457],[223,-15.6468],[224,-16.6844],[225,-17.5729],[226,-18.3148],[227,-18.7869],[228,-19.0579],[229,-19.2652],[230,-19.4043],[232,-19.5427],[235,-19.5427],[236,-19.4801],[237,-19.2333],[238,-18.929],[239,-18.5516],[241,-18.0421],[242,-17.8534],[243,-17.5616],[246,-16.4989],[247,-15.9219],[248,-15.0429],[249,-13.7663],[251,-10.5586],[252,-9.2728],[253,-8.2005],[254,-7.3866],[255,-6.6215],[256,-6.0119],[257,-5.5036],[258,-5.0471],[259,-4.744],[260,-4.6075],[262,-4.6075],[263,-4.6533],[264,-4.6533],[269,-4.8825],[270,-4.9743],[271,-5.1118],[272,-5.2951],[274,-5.7543],[275,-6.0202],[276,-6.1947],[277,-6.2815],[278,-6.4561],[286,-7.1541],[287,-7.1541],[291,-6.797],[294,-6.4401],[295,-6.2024],[297,-5.9647],[298,-5.7279],[301,-5.3718],[302,-5.3718],[304,-5.135],[305,-5.135],[306,0]]
       }
       pinky_01_r_ctrl.Scale.X {type=MovieSceneScriptingFloatChannel, default=1, const [1,307]}
       pinky_01_r_ctrl.Scale.Y {type=MovieSceneScriptingFloatChannel, default=1, const [1,307]}
       pinky_01_r_ctrl.Scale.Z {type=MovieSceneScriptingFloatChannel, default=1, const [1,307]}
       pinky_02_r_ctrl.Location.X {type=MovieSceneScriptingFloatChannel, default=0, thinned_from=307, keys [[0,0],[306,0]]}
       pinky_02_r_ctrl.Location.Y {type=MovieSceneScriptingFloatChannel, default=0, thinned_from=307, keys [[0,0],[306,0]]}
       pinky_02_r_ctrl.Location.Z {type=MovieSceneScriptingFloatChannel, default=0, thinned_from=307, keys [[0,0],[306,0]]}
       pinky_02_r_ctrl.Rotation.X {
        type=MovieSceneScriptingFloatChannel
        default=0
        thinned_from=307
        keys [[0,0.0001],[14,-0.0001],[23,0],[24,-0.0002],[25,0],[28,0.0001],[40,0],[74,-0.0001],[100,-0.0001],[102,0.0001],[104,-0.0001],[113,0.0001],[122,0.0001],[123,-0.0001],[129,0.0001],[165,0],[167,0],[168,0.0001],[171,-0.0001],[172,0],[174,-0.0002],[176,0.0001],[177,0],[178,0.0001],[182,-0.0001],[183,0.0001],[187,0],[188,0.0001],[195,0],[196,-0.0001],[197,0],[226,0.0001],[228,-0.0001],[230,0.0001],[231,-0.0001],[235,0.0001],[236,-0.0001],[237,0.0001],[238,-0.0001],[240,0.0001],[241,-0.0001],[251,0.0001],[253,-0.0002],[254,0],[258,-0.0001],[277,0],[278,-0.0001],[291,0.0001],[304,0],[305,-0.0001],[306,0]]
       }
       pinky_02_r_ctrl.Rotation.Y {
        type=MovieSceneScriptingFloatChannel
        default=0
        thinned_from=307
        keys [[0,0],[23,0],[24,0.0004],[25,0],[54,0],[55,-0.0002],[56,-0.0002],[57,0],[100,-0.0001],[306,0]]
       }
       pinky_02_r_ctrl.Rotation.Z {
        type=MovieSceneScriptingFloatChannel
        default=0
        thinned_from=307
        keys [[0,3.7618],[2,3.7618],[3,3.5987],[4,3.5987],[7,3.109],[8,3.109],[9,2.9457],[12,2.9457],[13,3.109],[15,4.4136],[16,5.2271],[17,5.833],[18,6.2817],[19,6.5058],[20,6.6179],[21,6.6179],[24,5.9452],[28,4.5971],[33,3.4705],[34,3.3889],[37,3.3889],[39,3.5754],[40,3.5754],[41,3.6686],[42,3.855],[43,3.9481],[44,4.1344],[46,4.3206],[47,4.3206],[48,4.4136],[50,4.4136],[51,4.4269],[53,4.4269],[55,4.294],[56,4.294],[57,4.161],[60,3.9614],[61,3.7618],[62,3.4289],[63,2.9623],[64,2.295],[65,1.4927],[66,0.5547],[67,-0.6052],[68,-2.5454],[69,-6.7271],[70,-9.3348],[71,-8.7934],[72,-6.0115],[73,-3.0764],[74,0.3612],[75,3.5075],[76,5.5056],[77,7.4089],[78,9.4769],[79,10.6796],[80,11.5371],[81,11.9654],[82,12.5646],[83,13.0432],[84,5.4509],[85,-8.2237],[86,-17.3095],[87,-22.8752],[88,-24.8444],[89,-25.7532],[90,-25.8361],[91,-25.505],[92,-19.7933],[93,-3.405],[94,5.7479],[95,8.1907],[96,6.9338],[97,-2.4192],[98,-18.028],[99,-24.0217],[100,-25.0919],[101,-25.8361],[102,-25.8361],[103,-24.1038],[104,-12.1028],[105,4.2612],[106,10.1072],[107,12.1639],[108,13.0432],[109,7.8953],[110,-5.9201],[111,-15.7205],[112,-22.2228],[113,-24.5972],[114,-25.5877],[115,-25.8391],[116,-25.5277],[117,-21.7508],[118,-6.0608],[119,5.3439],[120,8.9625],[121,-1.9833],[122,-17.7367],[123,-22.1331],[124,-24.287],[125,-25.1392],[126,-25.45],[127,-26.1509],[128,-28.3477],[129,-23.7465],[130,-3.6227],[131,4.5077],[132,6.9389],[133,8.7707],[134,6.6767],[135,-4.7553],[136,-14.5628],[137,-21.6437],[138,-24.1022],[139,-25.1229],[140,-25.7817],[141,-25.8551],[142,-25.7817],[143,-11.3713],[144,3.8484],[145,8.5747],[146,8.64],[147,7.6952],[148,6.7037],[149,6.5233],[150,8.2352],[151,11.2847],[152,14.5858],[153,16.8073],[154,17.6054],[155,17.1621],[156,15.4752],[157,13.4279],[158,11.5531],[159,10.2103],[160,9.5827],[161,8.8644],[162,7.6952],[163,6.2151],[164,3.9919],[165,1.197],[166,-1.6175],[167,-3.8848],[168,-4.8339],[169,-4.2642],[170,-2.9384],[171,-1.053],[172,0.8229],[173,2.69],[174,3.9919],[175,4.9196],[176,5.6603],[178,6.7693],[179,7.0139],[180,7.1096],[181,6.2483],[182,4.9055],[183,3.2695],[184,1.5304],[185,-0.6053],[186,-2.3617],[187,-3.3412],[188,-4.0285],[189,-4.4218],[190,-4.7172],[191,-4.9142],[193,-4.9142],[194,-4.8157],[195,-4.6553],[197,-3.7366],[198,-2.9729],[199,-1.6024],[200,0.8216],[201,3.2309],[202,5.6266],[203,7.8616],[204,9.3461],[205,10.3828],[207,12.1556],[208,12.8928],[210,14.0705],[211,14.5035],[212,14.722],[213,14.722],[214,14.6128],[215,14.3943],[216,13.9572],[217,12.8631],[218,11.2184],[219,9.3488],[221,6.2551],[222,4.9233],[223,3.699],[224,2.6947],[225,1.8],[226,1.0156],[227,0.5933],[229,0.0294],[230,0.0294],[232,-0.2528],[235,-0.2528],[237,0.0294],[238,0.4524],[241,1.2969],[242,1.4375],[243,1.7185],[244,2.0921],[247,2.6519],[248,3.0247],[249,3.7693],[251,5.6254],[252,6.3657],[253,6.7355],[254,6.9202],[255,6.9202],[256,7.1049],[258,7.1049],[259,7.162],[263,7.162],[264,7.0579],[266,7.0579],[267,6.9538],[268,6.9538],[269,6.8497],[270,6.8497],[274,6.0159],[275,5.7724],[276,5.5637],[284,4.7276],[285,4.7276],[286,4.6229],[287,4.6229],[288,4.7276],[289,4.7276],[290,4.9367],[291,5.0064],[301,6.3982],[302,6.3982],[303,6.5372],[304,6.5372],[305,6.6761],[306,0]]
       }
       pinky_02_r_ctrl.Scale.X {type=MovieSceneScriptingFloatChannel, default=1, const [1,307]}
       pinky_02_r_ctrl.Scale.Y {type=MovieSceneScriptingFloatChannel, default=1, const [1,307]}
       pinky_02_r_ctrl.Scale.Z {type=MovieSceneScriptingFloatChannel, default=1, const [1,307]}
       pinky_03_r_ctrl.Location.X {type=MovieSceneScriptingFloatChannel, default=0, thinned_from=307, keys [[0,0],[306,0]]}
       pinky_03_r_ctrl.Location.Y {type=MovieSceneScriptingFloatChannel, default=0, thinned_from=307, keys [[0,0],[306,0]]}
       pinky_03_r_ctrl.Location.Z {type=MovieSceneScriptingFloatChannel, default=0, thinned_from=307, keys [[0,0],[306,0]]}
       pinky_03_r_ctrl.Rotation.X {
        type=MovieSceneScriptingFloatChannel
        default=0
        thinned_from=307
        keys [[0,0.0008],[5,0.0008],[6,-0.001],[7,0.0007],[12,0.0007],[13,-0.001],[14,0.0008],[15,-0.0008],[16,-0.0007],[17,-0.0002],[23,-0.0001],[24,0.0022],[25,-0.0002],[32,-0.0005],[33,0.0018],[34,-0.0012],[38,-0.0012],[39,-0.0015],[41,-0.0015],[42,0.0002],[43,0.0002],[44,0],[45,0],[46,-0.0003],[50,-0.0003],[51,0.0003],[61,0.0004],[62,-0.0024],[63,0.0007],[64,-0.002],[66,-0.0014],[67,-0.0003],[72,-0.0006],[73,-0.0005],[74,0.0029],[75,0.0001],[79,0.0012],[80,0.0012],[82,-0.0001],[83,-0.0016],[84,0],[85,-0.0001],[88,0.0005],[89,-0.0013],[90,-0.0012],[91,-0.0006],[92,-0.0008],[93,-0.0001],[96,0.0001],[97,-0.0001],[99,0.0003],[100,-0.0005],[101,0.0004],[102,0.0005],[103,-0.0001],[104,0.0002],[105,-0.0002],[107,0.0002],[108,0.0003],[110,-0.0003],[113,0.0003],[114,0.0003],[115,-0.0002],[116,-0.0002],[118,0.0002],[119,-0.0004],[120,0.0002],[121,-0.001],[122,0],[126,-0.0003],[129,0],[130,-0.0015],[132,0.0009],[133,0.0001],[134,-0.0001],[135,-0.0007],[136,0.0001],[139,0.0003],[140,-0.0006],[142,-0.0001],[143,-0.0009],[145,0],[146,-0.0006],[147,0.0005],[148,0.0012],[149,-0.0008],[150,-0.001],[151,0],[153,-0.0004],[155,-0.0005],[156,0.001],[157,-0.0003],[162,0.0004],[163,-0.0005],[165,0.0006],[166,-0.0005],[167,0.0002],[170,0.0003],[171,-0.0005],[175,0.0002],[176,-0.0006],[178,-0.0004],[179,0.0015],[181,0.0015],[182,-0.0004],[184,-0.0003],[185,0.0017],[187,-0.0018],[188,0.0001],[194,0.0002],[195,0.0013],[199,0.0008],[202,0.0002],[210,0.0001],[211,0.0007],[213,0.0007],[214,-0.0009],[215,0.0005],[221,-0.0003],[225,-0.0002],[226,0.0013],[227,0.0001],[242,0.0002],[243,-0.0002],[244,-0.0001],[245,0.0013],[246,-0.0001],[250,0.0003],[251,-0.001],[252,0.0005],[253,-0.0008],[254,0.0006],[255,-0.0008],[258,-0.0007],[259,-0.0003],[261,0.0001],[262,-0.0009],[266,-0.0006],[268,0],[269,-0.0009],[271,-0.0008],[272,0.0003],[274,-0.0003],[275,0.0002],[299,0.0003],[300,-0.0025],[303,-0.0025],[304,0.0004],[305,0.0004],[306,0]]
       }
       pinky_03_r_ctrl.Rotation.Y {
        type=MovieSceneScriptingFloatChannel
        default=0
        thinned_from=307
        keys [[0,0.0001],[5,0.0001],[6,-0.0001],[12,0.0001],[13,-0.0001],[14,0.0001],[23,0],[24,0.0002],[25,-0.0001],[32,0],[33,0.0002],[34,-0.0001],[41,-0.0002],[61,0],[62,-0.0003],[63,0.0001],[64,-0.0003],[66,-0.0002],[67,0],[71,-0.0002],[73,0],[74,0.0005],[75,0],[84,-0.0002],[88,0.0005],[89,-0.0008],[90,-0.0008],[91,-0.0003],[93,-0.0001],[95,0],[96,-0.0002],[97,-0.0001],[98,0.0003],[99,0.0001],[100,-0.0003],[101,0.0002],[102,0.0002],[105,-0.0001],[108,0.0001],[110,-0.0002],[111,-0.0002],[112,0.0002],[114,0.0002],[115,0],[116,0.0002],[117,0],[118,0.0002],[119,0.0001],[120,0.0002],[121,-0.0003],[122,0.0001],[123,0],[124,0],[125,0.0003],[126,0.0003],[127,-0.0002],[128,0.0002],[130,-0.0006],[132,0.0003],[135,-0.0002],[136,0.0002],[137,0.0001],[139,0.0002],[140,-0.0004],[142,0.0001],[143,-0.0003],[148,0.0001],[149,-0.0001],[155,0.0001],[156,0],[157,0.0001],[165,0.0001],[166,-0.0001],[170,0.0001],[178,0],[181,0.0001],[184,-0.0001],[185,0.0003],[187,-0.0004],[188,0.0001],[194,0],[196,0.0003],[204,-0.0001],[205,0.0001],[209,-0.0001],[214,0],[225,-0.0001],[226,0.0003],[229,0],[231,0.0001],[236,0],[237,0.0001],[244,0],[245,0.0003],[246,0],[250,0],[251,-0.0001],[252,0.0001],[253,-0.0001],[254,0.0001],[261,0],[271,-0.0001],[299,0],[303,-0.0001],[304,0.0001],[306,0]]
       }
       pinky_03_r_ctrl.Rotation.Z {
        type=MovieSceneScriptingFloatChannel
        default=0
        thinned_from=307
        keys [[0,3.6539],[3,3.6539],[4,3.4806],[5,3.4806],[7,3.1339],[8,3.1339],[9,2.9606],[12,2.9606],[13,3.1339],[14,3.6539],[15,4.5198],[16,5.2121],[17,5.7843],[18,6.2631],[20,6.7417],[21,6.5024],[22,6.5024],[23,6.2631],[24,5.7843],[27,5.0657],[28,4.5864],[33,3.3873],[34,3.4158],[38,3.4158],[39,3.6443],[41,3.6443],[42,3.8728],[43,3.8728],[44,4.1012],[45,4.1012],[46,4.3296],[50,4.3296],[51,4.4799],[55,4.4799],[56,4.1735],[59,4.1735],[60,3.8671],[61,3.8671],[62,3.5605],[64,2.3334],[66,0.4897],[67,-1.0168],[68,-2.9339],[69,-7.4296],[70,-10.3392],[71,-9.6912],[72,-7.1073],[73,-4.215],[74,-1.0168],[75,2.1676],[78,7.873],[79,9.1374],[80,9.4534],[81,5.0239],[82,0.2585],[83,-4.0626],[84,-10.4615],[85,-17.1734],[86,-22.629],[87,-25.8538],[88,-27.014],[89,-27.2466],[90,-27.7124],[91,-32.4145],[92,-31.2311],[93,-17.399],[94,-9.7959],[95,-7.8044],[96,-8.4674],[97,-14.0278],[98,-23.0876],[99,-26.5479],[100,-27.0762],[101,-27.3936],[102,-28.2418],[103,-30.8023],[104,-24.8639],[105,-11.4846],[106,-7.1532],[107,-5.6506],[108,-5.0507],[109,-7.9562],[110,-15.9596],[111,-21.6271],[112,-25.4943],[113,-26.8648],[114,-27.2877],[115,-27.4507],[116,-30.7865],[117,-31.9078],[118,-19.3802],[119,-10.2251],[120,-7.494],[121,-13.608],[122,-22.8448],[123,-25.4681],[124,-26.5679],[125,-27.2297],[126,-27.2297],[127,-27.8931],[128,-34.847],[129,-33.0341],[130,-17.2322],[131,-10.8538],[132,-9.048],[133,-7.5856],[134,-8.7102],[135,-15.2858],[136,-21.0414],[137,-25.1288],[138,-26.5429],[139,-27.0157],[140,-27.2525],[141,-30.4681],[142,-33.8451],[143,-22.7867],[144,-11.646],[145,-8.2601],[146,-4.7843],[147,-1.2242],[148,2.6769],[149,5.6664],[150,7.3074],[151,10.2863],[152,13.8555],[153,16.5305],[154,17.5706],[155,17.2735],[156,15.4904],[157,13.5583],[158,11.4766],[159,10.1375],[160,9.5421],[161,8.7975],[162,7.6055],[163,6.2398],[164,4.0138],[165,1.1313],[166,-1.6676],[167,-3.9147],[168,-4.7593],[169,-4.29],[170,-2.9775],[171,-1.1071],[172,0.8518],[173,2.6201],[174,3.9209],[175,4.8491],[176,5.6837],[177,6.3325],[178,6.7957],[179,6.9873],[180,6.9873],[181,6.2023],[182,4.8274],[183,3.254],[184,1.4808],[185,-0.6914],[186,-2.2751],[187,-3.2667],[188,-4.0611],[189,-4.4586],[191,-4.8565],[194,-4.8565],[195,-4.6511],[196,-4.3464],[198,-3.1291],[199,-1.6105],[200,0.8125],[202,5.6385],[203,7.7433],[204,9.2448],[205,10.4451],[207,12.2444],[210,14.0425],[211,14.2341],[212,14.2341],[213,14.0808],[214,13.7742],[215,13.1608],[216,12.3939],[217,11.0131],[218,8.7097],[219,6.0953],[221,2.0857],[222,0.2294],[223,-1.476],[224,-2.8743],[226,-5.0554],[227,-5.7435],[228,-6.1216],[230,-6.5],[231,-6.5],[232,-6.6893],[236,-6.6893],[237,-6.3108],[238,-5.7435],[240,-4.9881],[241,-4.7993],[243,-4.045],[244,-3.4556],[245,-3.0138],[247,-1.8373],[248,-1.103],[249,-0.0761],[250,1.3885],[251,2.7045],[252,3.5808],[253,4.1646],[255,4.7481],[256,4.8939],[258,4.8939],[259,4.9858],[260,5.0923],[261,5.3053],[262,5.4117],[264,5.4117],[266,5.6246],[269,6.2631],[270,6.3694],[271,6.3694],[272,6.2631],[274,5.8375],[275,5.6113],[276,5.4251],[282,4.866],[283,4.866],[286,4.5864],[287,4.5864],[290,4.866],[291,4.9859],[292,4.9859],[293,5.2711],[294,5.2711],[295,5.5562],[296,5.5562],[298,6.1263],[299,6.1263],[300,6.4112],[303,6.4112],[304,6.6961],[305,6.6961],[306,0]]
       }
       pinky_03_r_ctrl.Scale.X {type=MovieSceneScriptingFloatChannel, default=1, const [1,307]}
       pinky_03_r_ctrl.Scale.Y {type=MovieSceneScriptingFloatChannel, default=1, const [1,307]}
       pinky_03_r_ctrl.Scale.Z {type=MovieSceneScriptingFloatChannel, default=1, const [1,307]}
       hand_r_ik_ctrl.Location.X {
        type=MovieSceneScriptingFloatChannel
        default=0
        thinned_from=307
        keys [[0,1.2596],[1,1.2684],[2,1.2911],[3,1.3109],[4,1.3225],[5,1.3305],[7,1.3263],[8,1.299],[9,1.2533],[10,1.1861],[11,1.1065],[12,1.0116],[13,0.8899],[14,0.7387],[15,0.5644],[16,0.3534],[17,0.0948],[18,-0.2192],[19,-0.5962],[20,-1.0441],[21,-1.5645],[22,-2.1346],[23,-2.7002],[24,-3.2197],[25,-3.6999],[26,-4.1529],[27,-4.5668],[28,-4.9324],[29,-5.242],[30,-5.5047],[31,-5.7219],[32,-5.9118],[33,-6.0697],[34,-6.2098],[35,-6.3253],[37,-6.5172],[38,-6.5986],[40,-6.7029],[42,-6.7807],[45,-6.8049],[48,-6.7705],[50,-6.7186],[52,-6.6455],[55,-6.4539],[56,-6.3741],[57,-6.2808],[59,-6.0576],[61,-5.7761],[62,-5.6216],[63,-5.443],[66,-4.7533],[67,-4.5312],[68,-4.46],[69,-4.8039],[70,-5.7898],[71,-7.9321],[72,-11.745],[73,-17.7288],[74,-26.1914],[75,-35.6374],[76,-44.0844],[77,-51.1542],[78,-56.5266],[79,-60.2988],[80,-62.6686],[81,-64.1566],[82,-65.1747],[83,-65.6515],[84,-65.9453],[87,-66.4566],[92,-66.7411],[98,-66.4581],[109,-65.7857],[117,-66.4267],[124,-65.7465],[135,-66.1291],[138,-65.972],[141,-66.2512],[144,-66.1334],[145,-65.8956],[146,-65.4368],[147,-64.778],[148,-63.8134],[149,-62.3158],[150,-59.9156],[151,-56.0952],[152,-50.5677],[153,-43.6369],[154,-35.6256],[155,-26.7025],[156,-16.8109],[157,-8.3312],[158,-3.259],[159,-1.2461],[160,-0.5507],[161,-0.4614],[162,-0.7188],[163,-0.9],[164,-0.9252],[165,-0.753],[166,-0.4448],[167,-0.0499],[168,0.438],[169,0.9825],[170,1.6653],[171,2.4051],[172,3.1068],[173,3.6757],[174,4.172],[175,4.6117],[176,5.0128],[177,5.3473],[178,5.5835],[179,5.7063],[180,5.7411],[181,5.7245],[185,5.5221],[186,5.4522],[187,5.3274],[188,5.1383],[190,4.6892],[191,4.4998],[192,4.3356],[193,4.1971],[194,4.0905],[195,4.0198],[196,3.9848],[197,3.9946],[198,4.0405],[199,4.1162],[200,4.2245],[201,4.3835],[202,4.5673],[203,4.7719],[204,4.9964],[205,5.2616],[206,5.5389],[207,5.744],[208,5.9325],[210,6.4012],[211,6.5316],[212,6.5267],[213,6.4276],[214,6.2684],[215,6.093],[217,5.8095],[219,5.566],[220,5.461],[221,5.3687],[223,5.2197],[228,4.9332],[230,4.7882],[232,4.6327],[234,4.4345],[235,4.3528],[236,4.2808],[237,4.1683],[238,4.0302],[239,3.9008],[240,3.7815],[244,3.2439],[245,3.1191],[246,2.9563],[247,2.7332],[248,2.4722],[249,2.1534],[250,1.7857],[252,1.0129],[253,0.6384],[254,0.2532],[255,-0.1281],[256,-0.4965],[257,-0.8314],[258,-1.1399],[259,-1.4089],[260,-1.6499],[261,-1.8627],[262,-2.0465],[263,-2.2117],[264,-2.3518],[265,-2.4752],[266,-2.5809],[267,-2.6752],[268,-2.7465],[269,-2.8033],[270,-2.8444],[271,-2.8718],[272,-2.8817],[273,-2.8721],[274,-2.8459],[275,-2.802],[277,-2.6774],[278,-2.5878],[279,-2.4669],[280,-2.3264],[281,-2.1798],[282,-2.0102],[283,-1.819],[284,-1.6025],[285,-1.3712],[286,-1.1197],[287,-0.8727],[288,-0.6078],[289,-0.3537],[290,-0.0987],[291,0.1304],[292,0.3454],[293,0.5416],[294,0.7149],[295,0.8525],[296,0.963],[297,1.047],[298,1.1115],[299,1.161],[300,1.2014],[301,1.2293],[302,1.236],[303,1.2513],[305,1.2515],[306,0]]
       }
       hand_r_ik_ctrl.Location.Y {
        type=MovieSceneScriptingFloatChannel
        default=0
        thinned_from=307
        keys [[0,-3.0042],[2,-3.023],[4,-3.0206],[6,-2.9755],[8,-2.8644],[9,-2.8154],[11,-2.7384],[12,-2.7204],[13,-2.7493],[14,-2.8078],[15,-2.8975],[16,-3.0099],[18,-3.2715],[19,-3.3816],[20,-3.4715],[21,-3.5337],[22,-3.5759],[24,-3.6337],[25,-3.6353],[29,-3.6156],[30,-3.5997],[32,-3.5322],[34,-3.4231],[35,-3.3547],[36,-3.2694],[38,-3.0749],[39,-2.9509],[40,-2.8414],[41,-2.7433],[42,-2.6552],[43,-2.5357],[45,-2.3424],[46,-2.2599],[47,-2.1837],[49,-2.0582],[51,-1.9578],[52,-1.9157],[53,-1.8589],[54,-1.8255],[55,-1.7807],[56,-1.7234],[57,-1.6591],[58,-1.5769],[59,-1.5348],[60,-1.5124],[61,-1.5429],[62,-1.6382],[63,-1.7131],[64,-1.7741],[65,-2.0316],[66,-2.7132],[67,-3.9006],[68,-5.73],[69,-8.2219],[70,-11.6102],[71,-16.9682],[72,-23.443],[73,-29.6279],[74,-34.2321],[75,-35.7752],[76,-34.7611],[77,-32.2239],[78,-29.1124],[79,-26.1414],[80,-23.9195],[81,-22.7256],[83,-20.768],[84,-20.1776],[85,-19.9885],[86,-20.1745],[88,-20.8124],[89,-20.8986],[90,-20.8254],[91,-20.5866],[93,-19.8093],[94,-19.5668],[95,-19.4859],[96,-19.574],[97,-19.886],[98,-20.4076],[99,-20.8533],[100,-21.1599],[101,-21.2877],[102,-21.2705],[103,-20.9586],[105,-20.0418],[106,-19.7099],[107,-19.5412],[108,-19.5316],[109,-19.6663],[111,-20.2456],[112,-20.5016],[113,-20.6413],[114,-20.7359],[115,-20.7124],[116,-20.5131],[118,-19.9172],[119,-19.657],[120,-19.4824],[121,-19.3956],[122,-19.4526],[123,-19.7389],[125,-20.527],[126,-20.8608],[127,-21.0864],[128,-21.1885],[129,-21.1734],[130,-20.9806],[131,-20.5947],[133,-19.6784],[134,-19.3666],[135,-19.2088],[136,-19.1079],[137,-19.0989],[139,-19.489],[140,-19.5208],[141,-19.3862],[144,-18.653],[145,-18.5788],[147,-18.6238],[148,-19.132],[149,-20.2787],[150,-21.9523],[151,-24.4668],[152,-27.0476],[153,-29.1669],[154,-30.4211],[155,-30.3857],[156,-27.6321],[157,-21.3577],[158,-13.825],[159,-8.8216],[160,-5.851],[161,-3.1247],[162,-1.1376],[163,-0.2272],[164,0.2746],[165,0.4536],[166,0.3516],[167,0.0456],[168,-0.432],[169,-1.0639],[170,-1.7106],[171,-2.34],[172,-2.7467],[173,-2.5877],[174,-2.3165],[175,-2.022],[176,-1.6601],[177,-1.1838],[178,-0.6454],[179,-0.0887],[180,0.3689],[181,0.7354],[182,0.9836],[183,0.9632],[184,0.7133],[185,0.3194],[186,-0.1037],[187,-0.5187],[188,-0.9679],[189,-1.376],[190,-1.6768],[191,-1.8878],[192,-2.051],[193,-2.2052],[194,-2.3261],[195,-2.3832],[196,-2.3927],[197,-2.3576],[199,-2.2069],[200,-2.154],[201,-2.0624],[202,-1.8869],[203,-1.6216],[204,-1.2408],[205,-0.8781],[206,-0.5354],[207,-0.1241],[208,0.4776],[209,1.5069],[210,2.8482],[213,6.9502],[214,8.1869],[215,9.177],[216,9.9399],[217,10.5351],[218,10.9757],[219,11.3291],[220,11.6218],[221,11.8556],[222,12.0288],[223,12.1693],[224,12.2693],[225,12.3417],[227,12.405],[231,12.4243],[233,12.3395],[234,12.2544],[235,12.0685],[236,11.7794],[237,11.4416],[238,10.9853],[239,10.384],[240,9.5959],[241,8.6228],[242,7.4693],[243,6.1567],[244,4.6776],[245,2.9532],[246,1.1334],[247,-0.5909],[248,-2.0342],[249,-3.0024],[250,-3.5432],[251,-3.8081],[252,-3.9941],[253,-4.1903],[254,-4.3104],[255,-4.3605],[256,-4.3458],[257,-4.2829],[259,-4.1124],[260,-4.043],[261,-4.0009],[262,-3.9792],[263,-3.97],[264,-3.9844],[265,-3.9807],[266,-3.9867],[267,-4.0102],[269,-4.0791],[272,-4.1548],[273,-4.1857],[274,-4.1983],[275,-4.2215],[277,-4.257],[279,-4.2702],[280,-4.2582],[282,-4.257],[283,-4.2491],[285,-4.2022],[287,-4.1168],[288,-4.0414],[289,-3.956],[291,-3.7406],[293,-3.5005],[294,-3.3905],[295,-3.3033],[296,-3.2331],[297,-3.1718],[299,-3.0804],[301,-3.0223],[302,-3.0128],[303,-2.9904],[305,-2.9825],[306,0]]
       }
       hand_r_ik_ctrl.Location.Z {
        type=MovieSceneScriptingFloatChannel
        default=0
        thinned_from=307
        keys [[0,-5.6675],[3,-5.6703],[5,-5.6229],[6,-5.5902],[7,-5.5723],[9,-5.564],[10,-5.5727],[11,-5.6128],[13,-5.7132],[14,-5.7943],[15,-5.9098],[16,-6.0636],[17,-6.2959],[18,-6.6349],[19,-7.1501],[20,-7.8167],[21,-8.5959],[22,-9.3956],[23,-10.1213],[24,-10.718],[25,-11.2032],[26,-11.6115],[27,-11.9631],[30,-12.8924],[32,-13.4538],[40,-15.3449],[44,-16.1754],[47,-16.6583],[49,-16.917],[51,-17.0817],[53,-17.1696],[56,-17.162],[58,-17.0868],[60,-16.9255],[62,-16.668],[63,-16.5138],[64,-16.2942],[65,-15.8668],[66,-15.1332],[67,-13.9522],[68,-12.1238],[69,-9.7284],[70,-7.2617],[71,-3.8511],[72,0.8889],[73,6.5351],[74,12.7801],[75,17.9599],[76,21.0542],[77,22.3189],[78,22.056],[79,20.6675],[80,19.0526],[81,17.8476],[82,16.9143],[83,16.6813],[84,16.6957],[85,16.7632],[87,17.2519],[88,17.4543],[89,17.5886],[90,17.6839],[91,17.6989],[93,17.5458],[94,17.5126],[95,17.5401],[96,17.6339],[97,17.8223],[99,18.2756],[100,18.4514],[101,18.5774],[102,18.6563],[103,18.6061],[105,18.3922],[106,18.327],[107,18.3089],[108,18.329],[109,18.4047],[112,18.8545],[114,19.0406],[115,19.0674],[116,19.0361],[120,18.5168],[121,18.4479],[122,18.4239],[123,18.5153],[127,19.1588],[128,19.2575],[129,19.2971],[130,19.2782],[131,19.1624],[133,18.829],[134,18.6912],[136,18.5423],[137,18.5035],[140,18.8123],[141,18.802],[144,18.611],[146,18.5921],[147,18.6691],[148,19.0138],[149,19.6916],[150,20.9298],[151,22.3245],[152,23.1455],[153,22.9206],[154,21.3767],[155,17.9749],[156,11.9058],[157,4.1893],[158,-2.9319],[159,-8.2958],[160,-12.2263],[161,-15.6021],[162,-18.2467],[163,-19.7792],[164,-20.5078],[165,-20.6101],[166,-20.2962],[167,-19.6575],[168,-18.6652],[169,-17.3909],[170,-15.5604],[171,-13.332],[172,-10.9389],[173,-8.7392],[174,-6.5978],[175,-4.4842],[176,-2.2678],[177,-0.0185],[178,1.9927],[179,3.5742],[180,4.7217],[181,5.5242],[182,5.9939],[183,6.237],[184,6.1199],[185,5.546],[186,4.4932],[187,2.9864],[188,1.3755],[189,-0.1895],[190,-1.5648],[191,-2.7152],[192,-3.66],[193,-4.442],[194,-5.0713],[195,-5.5466],[196,-5.8033],[197,-5.8623],[198,-5.7423],[199,-5.5025],[200,-5.1347],[201,-4.5379],[202,-3.7078],[203,-2.67],[204,-1.449],[205,0.0431],[206,1.8062],[207,3.7333],[208,5.7389],[209,7.6057],[210,9.2944],[211,10.8588],[212,12.1576],[213,13.1561],[214,13.9384],[215,14.6242],[216,15.2552],[217,15.7915],[218,16.2801],[219,16.7051],[220,17.0608],[221,17.3375],[223,17.7314],[226,18.1077],[227,18.1656],[229,18.1408],[234,17.8564],[236,17.7955],[238,17.645],[240,17.5869],[241,17.434],[242,17.0943],[243,16.5518],[244,15.7511],[245,14.5637],[246,13.0582],[248,9.6926],[249,7.808],[250,5.7348],[251,3.7257],[252,1.9557],[253,0.3895],[254,-1.145],[255,-2.5769],[256,-3.8803],[257,-5.0107],[258,-6.0116],[259,-6.8721],[260,-7.6377],[261,-8.2959],[262,-8.8639],[263,-9.3572],[264,-9.7792],[265,-10.1621],[266,-10.5053],[267,-10.8102],[268,-11.0657],[269,-11.2953],[271,-11.6683],[272,-11.8071],[274,-12.0066],[276,-12.0948],[278,-12.0797],[280,-11.9319],[282,-11.6526],[283,-11.4666],[284,-11.2489],[285,-11.0071],[286,-10.7316],[287,-10.4213],[288,-10.0675],[291,-8.9034],[294,-7.6687],[295,-7.3075],[296,-6.9857],[297,-6.6978],[298,-6.4421],[299,-6.2308],[300,-6.0528],[301,-5.9179],[303,-5.7415],[304,-5.6919],[305,-5.6788],[306,0]]
       }
       hand_r_ik_ctrl.Rotation.X {
        type=MovieSceneScriptingFloatChannel
        default=0
        thinned_from=307
        keys [[0,38.8489],[6,38.895],[10,38.5097],[12,38.1782],[13,37.9475],[14,37.6349],[15,37.2134],[16,36.6864],[17,36.0044],[18,35.1012],[19,33.9044],[20,32.5152],[21,30.9279],[22,29.1679],[23,27.5902],[24,26.434],[25,25.4899],[26,24.6867],[27,24.029],[28,23.532],[29,23.1633],[30,22.8725],[31,22.6408],[33,22.3077],[37,21.8819],[40,21.6608],[45,21.4595],[52,21.3194],[58,21.4431],[60,21.6055],[61,21.737],[62,21.93],[63,21.8712],[64,21.4492],[65,21.3776],[66,22.234],[67,27.6705],[68,36.1123],[69,40.4581],[70,40.9452],[71,36.8144],[72,30.0623],[73,21.2072],[74,6.6401],[75,-19.883],[76,-54.9215],[77,-83.3634],[78,-102.0031],[79,-117.7927],[80,-131.7062],[81,-142.0749],[82,-150.3752],[83,-156.1237],[84,-159.9081],[85,-162.0899],[88,-162.9137],[89,-161.9194],[91,-157.9378],[92,-156.7005],[93,-156.6619],[94,-157.3755],[95,-158.8575],[96,-161.976],[97,-164.1234],[98,-164.4302],[101,-164.1263],[102,-162.9072],[103,-160.8941],[104,-160.0335],[112,-163.6287],[114,-163.3586],[115,-162.5933],[117,-160.1776],[118,-159.3673],[119,-159.8322],[120,-161.0378],[122,-165.2344],[123,-166.5547],[124,-167.1562],[126,-166.6072],[127,-165.7168],[129,-161.3233],[130,-159.8202],[131,-159.3955],[132,-160.0574],[133,-161.7947],[134,-164.5379],[135,-167.7021],[136,-169.2128],[137,-169.7369],[139,-169.259],[140,-168.5349],[141,-167.2239],[144,-161.8091],[146,-156.6901],[147,-153.5925],[150,-142.2016],[151,-137.9772],[152,-132.3839],[153,-122.8685],[154,-106.6421],[155,-74.3636],[156,-23.3739],[157,6.764],[158,20.4639],[159,27.7645],[160,30.9614],[161,32.1021],[162,32.1477],[164,31.5308],[165,31.7342],[166,32.3793],[167,33.3875],[168,34.913],[170,38.5648],[171,40.1094],[172,41.2221],[176,44.0618],[179,46.6038],[180,47.2845],[181,47.8224],[182,48.2128],[183,48.3768],[184,48.2945],[186,47.8018],[194,46.533],[196,46.4412],[198,46.6155],[199,46.8419],[200,47.1938],[201,47.7238],[202,48.4423],[203,48.8417],[204,48.8896],[205,48.5461],[206,47.8384],[207,46.8432],[208,45.4461],[209,45.0728],[210,47.5831],[211,51.6023],[212,56.7193],[214,68.991],[216,81.5428],[217,86.3822],[218,89.5434],[219,91.8382],[220,93.452],[221,94.5591],[222,95.2676],[224,95.884],[227,96.1556],[231,95.4546],[233,94.7201],[236,92.9897],[239,90.5217],[240,89.5013],[241,88.0207],[242,85.6507],[243,81.948],[244,76.4662],[245,69.4041],[246,61.7062],[247,54.411],[248,48.3469],[249,42.8144],[250,37.5406],[251,33.2571],[252,30.6162],[253,29.5981],[254,29.5786],[255,30.1629],[258,32.7858],[259,33.5004],[260,33.7842],[262,33.4371],[264,33.3237],[270,33.5196],[275,34.0592],[278,34.5959],[282,35.747],[287,37.4112],[290,38.1342],[295,38.8123],[305,38.8455],[306,0]]
       }
       hand_r_ik_ctrl.Rotation.Y {
        type=MovieSceneScriptingFloatChannel
        default=0
        thinned_from=307
        keys [[0,-2.5266],[1,-2.5563],[2,-2.6007],[3,-2.6366],[4,-2.6138],[5,-2.5678],[6,-2.4915],[7,-2.4246],[9,-2.2568],[10,-2.2063],[11,-2.2307],[12,-2.2929],[13,-2.3689],[14,-2.51],[15,-2.5896],[16,-2.532],[17,-2.4639],[18,-2.4509],[19,-2.5877],[20,-2.8594],[21,-3.1531],[22,-3.3382],[23,-3.5093],[24,-3.7491],[25,-3.9736],[26,-4.1809],[27,-4.4012],[28,-4.6603],[29,-5.0136],[30,-5.4134],[33,-6.6724],[34,-7.0605],[40,-9.2798],[41,-9.6026],[42,-9.9539],[44,-10.5716],[45,-10.8562],[47,-11.3513],[49,-11.789],[50,-11.9595],[52,-12.2111],[54,-12.346],[56,-12.4126],[58,-12.4023],[59,-12.3663],[61,-12.2081],[62,-12.0784],[63,-12.0826],[64,-12.1212],[65,-11.7746],[66,-10.6079],[67,-7.7452],[68,-3.5967],[69,0.6346],[70,5.0699],[71,11.5803],[72,20.7768],[73,32.3784],[74,45.9086],[75,56.2947],[76,56.7201],[77,46.2575],[78,30.6353],[79,14.487],[80,1.0258],[81,-8.2473],[82,-15.2318],[83,-18.8929],[84,-20.5735],[85,-21.3859],[86,-20.3272],[87,-17.8564],[88,-14.9193],[89,-13.4568],[90,-13.454],[91,-15.0646],[92,-17.6206],[93,-20.4076],[94,-22.3288],[95,-23.1489],[96,-23.1264],[97,-21.777],[98,-19.0665],[99,-15.4906],[100,-12.1035],[101,-10.3562],[102,-10.2099],[103,-12.7156],[104,-16.1926],[105,-19.5315],[106,-21.0891],[107,-21.921],[108,-22.3477],[109,-22.4354],[110,-21.7212],[111,-19.809],[112,-17.3951],[113,-15.0275],[114,-13.351],[115,-12.6239],[116,-13.6812],[117,-16.1296],[118,-19.0441],[119,-20.5143],[120,-20.5622],[121,-19.7643],[122,-18.0416],[123,-15.7558],[124,-13.983],[125,-12.9588],[126,-12.2534],[127,-12.0341],[128,-13.4715],[129,-15.6641],[130,-18.0906],[131,-20.6681],[132,-22.8347],[133,-24.005],[134,-24.0758],[135,-22.9257],[136,-20.5715],[137,-17.9635],[138,-16.2539],[139,-15.0365],[140,-14.3413],[141,-14.4955],[142,-16.0612],[143,-19.3068],[144,-21.9509],[145,-22.6708],[146,-22.6833],[147,-21.9708],[148,-19.2898],[149,-13.7541],[150,-4.1845],[151,9.2737],[152,24.0722],[153,38.5867],[154,52.4918],[155,63.2771],[156,61.033],[157,45.2868],[158,27.8949],[159,13.2854],[160,0.7332],[161,-11.013],[162,-20.8996],[163,-27.5055],[164,-31.7838],[165,-34.4465],[166,-35.8801],[167,-36.1826],[168,-35.4206],[169,-33.6942],[170,-30.2969],[171,-25.7411],[172,-20.7026],[173,-16.0955],[174,-11.6793],[175,-7.328],[176,-2.7522],[177,1.8598],[178,6.008],[179,9.4018],[180,12.0564],[181,14.1198],[182,15.6596],[183,16.8691],[184,17.4833],[185,17.3039],[186,16.1453],[187,13.9877],[188,11.5669],[189,9.1225],[190,6.8231],[191,4.811],[192,3.1146],[193,1.6611],[194,0.4248],[195,-0.5925],[196,-1.284],[197,-1.6206],[198,-1.6857],[199,-1.5529],[200,-1.234],[201,-0.7118],[202,-0.0702],[203,0.5601],[204,1.1218],[205,1.8622],[207,3.8247],[208,4.8457],[209,5.8264],[210,6.8611],[211,7.7562],[212,8.2042],[213,8.2255],[214,7.9409],[215,7.5649],[216,7.0673],[217,6.6049],[218,6.4036],[219,6.4666],[221,6.9614],[223,7.4295],[226,8.1251],[227,8.3383],[228,8.5119],[230,8.7689],[231,8.9175],[232,9.1112],[233,9.3462],[234,9.6211],[235,10.0486],[236,10.6236],[237,11.2859],[238,12.1731],[239,13.3243],[240,14.8616],[241,16.7729],[242,18.8359],[243,20.7569],[244,22.1737],[245,22.8504],[246,22.7677],[247,22.0061],[248,20.8142],[249,19.0583],[250,16.8057],[251,14.4741],[252,12.4364],[254,8.9852],[255,7.3237],[256,5.7093],[257,4.1399],[258,2.559],[259,0.9578],[260,-0.6692],[261,-2.309],[262,-3.9033],[263,-5.425],[264,-6.8302],[265,-8.0892],[266,-9.1834],[267,-10.1047],[268,-10.8955],[269,-11.5804],[270,-12.1893],[271,-12.7527],[272,-13.2344],[273,-13.6481],[274,-14.0104],[275,-14.309],[276,-14.5663],[277,-14.7757],[278,-14.9401],[279,-15.0313],[280,-15.0805],[281,-15.0551],[282,-14.9908],[283,-14.8478],[284,-14.5719],[285,-14.2104],[286,-13.7603],[287,-13.2383],[288,-12.5811],[289,-11.8243],[291,-10.1834],[292,-9.2707],[293,-8.3078],[294,-7.3753],[295,-6.5486],[296,-5.8036],[297,-5.1078],[298,-4.4709],[299,-3.9272],[300,-3.4973],[301,-3.169],[302,-2.9238],[303,-2.7273],[304,-2.6049],[305,-2.5517],[306,0]]
       }
       hand_r_ik_ctrl.Rotation.Z {
        type=MovieSceneScriptingFloatChannel
        default=0
        thinned_from=307
        keys [[0,-8.1487],[4,-8.1347],[7,-8.0713],[9,-8.1487],[10,-8.2178],[12,-8.2799],[14,-8.3858],[15,-8.6747],[16,-9.4093],[17,-10.3663],[18,-11.4058],[19,-12.5438],[22,-16.0788],[23,-17.107],[24,-17.8382],[25,-18.403],[26,-18.8903],[27,-19.3101],[28,-19.6693],[29,-19.9637],[30,-20.1808],[31,-20.3129],[32,-20.4036],[36,-20.4747],[38,-20.6294],[40,-20.6554],[42,-20.5297],[46,-19.9395],[47,-19.655],[51,-18.2563],[52,-17.9478],[55,-16.8931],[56,-16.5004],[57,-16.07],[59,-15.1419],[60,-14.6298],[61,-14.06],[62,-13.4101],[63,-12.725],[64,-12.1076],[65,-11.5212],[66,-11.048],[67,-9.5976],[68,-7.5071],[69,-7.0629],[70,-9.5472],[71,-17.4034],[72,-28.8682],[73,-43.3824],[74,-64.1636],[75,-96.11],[76,-135.3778],[77,-165.4586],[78,-181.8094],[79,-189.9346],[80,-193.6686],[82,-196.4446],[84,-201.0536],[87,-204.9516],[89,-206.1324],[90,-206.3792],[91,-205.7415],[93,-202.5439],[94,-202.0399],[95,-202.391],[97,-204.7484],[98,-205.3495],[100,-207.6653],[101,-207.5923],[102,-206.7253],[105,-202.0917],[106,-201.2634],[108,-201.0109],[110,-201.9782],[113,-206.2133],[114,-206.6908],[115,-206.579],[116,-205.9464],[119,-202.308],[120,-201.8169],[121,-202.3166],[124,-205.4807],[126,-206.4405],[127,-206.416],[132,-199.0608],[133,-198.4554],[134,-198.6133],[135,-199.5404],[137,-204.8023],[138,-205.7798],[140,-205.9585],[141,-205.3576],[142,-203.9122],[143,-201.4493],[144,-199.5695],[145,-198.8964],[146,-199.3409],[148,-202.8719],[149,-202.8544],[150,-200.6679],[151,-195.4643],[152,-187.8206],[153,-177.0732],[154,-160.3712],[155,-127.4796],[156,-73.5585],[157,-37.9919],[158,-18.2248],[159,-7.3652],[160,-1.1856],[161,3.7971],[162,7.1141],[163,7.7286],[164,7.4234],[165,6.3925],[166,4.6706],[167,2.4869],[168,0.0747],[169,-2.4841],[170,-4.8003],[171,-6.813],[172,-8.0789],[173,-7.8282],[174,-7.2446],[175,-6.4673],[176,-5.4146],[177,-4.0658],[178,-2.5887],[179,-1.1005],[180,0.0761],[181,1.0237],[182,1.7423],[183,1.8643],[184,1.452],[185,0.7067],[186,0.0265],[187,-0.4485],[188,-0.9782],[189,-1.4646],[190,-1.8247],[191,-2.1031],[192,-2.3575],[193,-2.6477],[194,-2.8859],[195,-2.98],[196,-2.9709],[197,-2.8464],[198,-2.6193],[199,-2.3319],[200,-2.0027],[201,-1.5412],[202,-0.9072],[203,-0.3424],[204,0.1321],[205,0.1482],[206,-0.2618],[207,-0.6895],[208,-1.1797],[209,-1.1683],[210,-0.3933],[211,0.7385],[212,2.315],[213,4.2604],[214,6.2444],[215,7.9911],[216,9.4359],[217,10.4127],[218,11.0502],[219,11.7421],[220,12.5331],[221,13.249],[222,13.8765],[223,14.4291],[225,15.3669],[226,15.7991],[227,16.1979],[228,16.5219],[232,17.4331],[234,17.8381],[235,18.1516],[236,18.5986],[237,19.1646],[240,20.9864],[241,21.6228],[242,21.8582],[243,21.2167],[244,19.2532],[245,15.8837],[246,11.7182],[247,7.4431],[248,3.6665],[249,0.4894],[250,-2.1767],[251,-4.2832],[252,-5.8963],[253,-7.1754],[254,-8.2007],[255,-9.0251],[256,-9.7027],[257,-10.2754],[258,-10.7768],[260,-11.7266],[262,-12.745],[263,-13.2273],[264,-13.6679],[265,-14.0202],[266,-14.3051],[267,-14.538],[268,-14.7316],[269,-14.8866],[270,-15.0076],[272,-15.1592],[273,-15.1995],[275,-15.1767],[277,-15.065],[278,-14.9569],[279,-14.8103],[282,-14.2554],[283,-14.0297],[284,-13.7531],[285,-13.4448],[287,-12.764],[289,-11.9383],[293,-10.1393],[294,-9.7301],[295,-9.3862],[296,-9.1003],[297,-8.8617],[298,-8.6625],[299,-8.4981],[300,-8.3695],[301,-8.2774],[303,-8.1751],[305,-8.1542],[306,0]]
       }
       hand_r_ik_ctrl.Scale.X {type=MovieSceneScriptingFloatChannel, default=1, const [1,307]}
       hand_r_ik_ctrl.Scale.Y {type=MovieSceneScriptingFloatChannel, default=1, const [1,307]}
       hand_r_ik_ctrl.Scale.Z {type=MovieSceneScriptingFloatChannel, default=1, const [1,307]}
       Softness_2 {type=MovieSceneScriptingFloatChannel, default=0, const [0,307]}
       hand_r_ikorient_ctrl.Rotation.Roll {type=MovieSceneScriptingFloatChannel, default=0, const [0,307]}
       hand_r_ikorient_ctrl.Rotation.Pitch {type=MovieSceneScriptingFloatChannel, default=0, const [0,307]}
       hand_r_ikorient_ctrl.Rotation.Yaw {type=MovieSceneScriptingFloatChannel, default=0, const [0,307]}
       arm_r_pv_ik_ctrl.Location.X {
        type=MovieSceneScriptingFloatChannel
        default=0
        thinned_from=307
        keys [[0,10.3328],[4,10.3972],[6,10.3689],[7,10.3138],[8,10.2202],[9,10.095],[10,9.938],[11,9.7479],[12,9.5305],[13,9.2892],[14,9.0001],[15,8.6729],[16,8.2956],[17,7.8393],[18,7.2659],[19,6.5212],[20,5.6392],[21,4.6198],[22,3.4962],[23,2.4355],[24,1.5804],[25,0.8318],[26,0.1606],[27,-0.419],[28,-0.9018],[29,-1.3021],[30,-1.6415],[31,-1.9279],[32,-2.1708],[33,-2.3857],[34,-2.5735],[35,-2.7488],[36,-2.9171],[37,-3.0933],[38,-3.2518],[40,-3.5416],[41,-3.6389],[42,-3.7455],[43,-3.8323],[44,-3.9098],[45,-3.978],[46,-4.0212],[47,-4.055],[48,-4.0731],[49,-4.0823],[50,-4.079],[52,-4.0398],[53,-3.9988],[56,-3.8189],[58,-3.666],[59,-3.5353],[60,-3.3589],[61,-3.1143],[62,-2.7903],[63,-2.7388],[64,-3.2985],[65,-4.0689],[66,-4.5725],[67,-4.635],[68,-4.3767],[69,-3.8555],[70,-2.9839],[71,-0.9398],[72,2.3136],[73,6.5947],[74,12.1326],[75,17.883],[76,22.8455],[77,26.9311],[78,29.9669],[79,31.8768],[80,32.909],[81,33.3828],[82,33.5829],[83,33.6512],[84,33.6432],[86,33.4504],[90,32.8239],[102,32.2382],[107,32.3231],[114,32.2239],[122,32.4014],[130,32.1868],[140,32.2454],[143,32.4327],[145,32.7119],[147,33.211],[148,33.3525],[149,33.1831],[150,32.4761],[151,30.9063],[152,28.3894],[153,25.206],[155,18.3796],[156,14.477],[157,10.7482],[158,8.2491],[159,7.6822],[160,7.9446],[161,7.8626],[162,7.5588],[163,7.063],[164,6.6242],[165,6.3962],[166,6.3343],[167,6.4229],[168,6.8082],[169,7.4598],[170,8.2785],[172,10.073],[173,10.886],[176,13.1465],[177,13.8486],[178,14.4371],[179,14.8758],[180,15.1626],[181,15.3579],[183,15.6032],[184,15.6523],[185,15.6208],[187,15.4577],[189,15.1867],[192,14.7029],[193,14.5715],[194,14.474],[195,14.4184],[196,14.4074],[197,14.4535],[198,14.5533],[199,14.7353],[200,15.0214],[201,15.412],[202,15.8949],[207,18.6263],[208,19.0361],[209,19.2197],[210,19.1916],[212,19.0015],[215,18.5021],[217,18.2705],[221,18.0151],[229,17.7719],[234,17.4511],[235,17.4149],[237,17.4337],[238,17.4902],[239,17.5946],[242,18.0693],[243,18.1817],[244,18.2239],[245,18.1601],[246,17.9384],[247,17.5649],[248,17.0621],[249,16.2746],[250,15.2099],[251,14.0501],[252,13.0014],[254,11.189],[255,10.3215],[256,9.5071],[257,8.762],[258,8.1004],[259,7.5299],[260,7.045],[261,6.6438],[262,6.314],[263,6.0366],[264,5.8231],[265,5.6335],[266,5.4787],[267,5.3686],[268,5.2907],[269,5.2423],[270,5.2085],[271,5.1961],[272,5.2063],[274,5.2917],[275,5.3689],[276,5.4635],[277,5.5763],[278,5.7136],[280,6.0918],[281,6.31],[282,6.5493],[284,7.0791],[288,8.2329],[290,8.7604],[292,9.2141],[294,9.6078],[295,9.7736],[296,9.9117],[298,10.1001],[300,10.2198],[303,10.2964],[305,10.3036],[306,0]]
       }
       arm_r_pv_ik_ctrl.Location.Y {
        type=MovieSceneScriptingFloatChannel
        default=0
        thinned_from=307
        keys [[0,-3.1313],[1,-3.1186],[2,-3.0803],[3,-3.0503],[4,-3.0302],[5,-3.0188],[6,-3.018],[7,-3.0346],[8,-3.0904],[9,-3.1679],[10,-3.2619],[11,-3.3817],[12,-3.5233],[13,-3.683],[14,-3.8896],[15,-4.1535],[16,-4.4938],[17,-4.9448],[18,-5.5348],[19,-6.3081],[20,-7.2391],[21,-8.3273],[22,-9.5044],[23,-10.5798],[24,-11.4361],[25,-12.1417],[26,-12.7453],[27,-13.252],[28,-13.6636],[29,-14.0067],[30,-14.2951],[31,-14.5367],[33,-14.9259],[38,-15.6516],[40,-15.8953],[45,-16.3825],[49,-16.5834],[52,-16.6138],[56,-16.4604],[59,-16.2153],[61,-15.9526],[63,-15.6061],[64,-15.3725],[65,-15.0292],[66,-14.5787],[67,-14.0068],[68,-13.2231],[69,-12.2789],[70,-11.5573],[71,-10.8007],[72,-9.3395],[73,-7.0589],[74,-3.6135],[75,0.5951],[76,4.8886],[77,9.2585],[78,13.5081],[79,17.4341],[80,20.8633],[81,23.6167],[82,25.5649],[83,26.5871],[84,26.9841],[86,27.3524],[90,27.792],[97,28.7898],[101,29.3275],[107,30.5146],[109,30.7006],[122,30.3691],[126,30.1644],[135,30.5954],[140,30.5162],[141,30.4473],[142,30.2891],[143,30.0263],[144,29.6263],[145,29.0105],[146,28.0084],[147,26.6532],[148,24.7234],[149,22.2733],[150,19.2267],[151,14.926],[152,10.1609],[153,5.4869],[154,1.06],[155,-3.1522],[156,-7.2914],[157,-10.6483],[158,-12.7261],[159,-13.6908],[160,-13.976],[161,-14.0466],[162,-13.979],[163,-13.7987],[164,-13.4053],[165,-12.7826],[166,-12.0257],[167,-11.2132],[168,-10.2627],[169,-9.1797],[170,-7.9412],[171,-6.5782],[172,-5.1458],[173,-3.7278],[174,-2.2964],[175,-0.8147],[176,0.7947],[177,2.4822],[178,4.065],[179,5.3477],[180,6.2446],[181,6.8448],[182,7.1629],[183,7.2317],[184,6.9985],[185,6.4442],[186,5.6603],[187,4.7064],[188,3.6887],[189,2.7186],[190,1.9283],[191,1.3046],[192,0.8096],[193,0.4253],[194,0.1515],[195,-0.0126],[196,-0.051],[197,0.0044],[198,0.1606],[199,0.3895],[200,0.7192],[201,1.2664],[202,2.0634],[203,3.1067],[204,4.3973],[205,5.9306],[206,7.6834],[207,9.6285],[208,11.6925],[209,13.7844],[210,15.8294],[211,17.8379],[212,19.7169],[213,21.3883],[214,22.862],[215,24.1216],[216,25.2214],[217,26.1093],[218,26.8448],[219,27.4412],[220,27.9112],[221,28.2743],[222,28.5509],[224,28.9286],[226,29.1296],[227,29.1346],[228,29.0532],[231,28.528],[233,27.984],[235,27.245],[236,26.8305],[237,26.2855],[238,25.6708],[239,24.9867],[240,24.1938],[241,23.139],[242,21.7898],[243,20.1693],[244,18.2562],[245,15.9852],[247,11.0575],[248,8.8268],[250,4.466],[251,2.4989],[252,0.8567],[253,-0.5273],[254,-1.8186],[255,-2.9796],[256,-4.0169],[257,-4.9035],[258,-5.6788],[259,-6.3296],[260,-6.9018],[261,-7.387],[262,-7.8025],[263,-8.1622],[264,-8.4606],[265,-8.7261],[266,-8.9508],[267,-9.1503],[268,-9.3061],[269,-9.4421],[271,-9.6348],[272,-9.6894],[274,-9.7226],[276,-9.6446],[277,-9.5665],[278,-9.4576],[279,-9.2981],[280,-9.1125],[281,-8.8944],[282,-8.6474],[283,-8.3662],[284,-8.0545],[285,-7.7187],[287,-6.9782],[290,-5.8247],[292,-5.1096],[293,-4.7635],[294,-4.4378],[295,-4.1576],[296,-3.9195],[297,-3.7257],[298,-3.5661],[299,-3.4397],[300,-3.3306],[301,-3.2505],[302,-3.2123],[303,-3.1601],[304,-3.1408],[305,-3.1408],[306,0]]
       }
       arm_r_pv_ik_ctrl.Location.Z {
        type=MovieSceneScriptingFloatChannel
        default=0
        thinned_from=307
        keys [[0,4.4087],[1,4.4068],[6,4.3376],[7,4.3069],[10,4.2517],[11,4.2192],[13,4.1268],[17,3.8423],[18,3.7475],[20,3.4495],[22,3.1928],[23,3.1327],[24,3.1893],[25,3.288],[26,3.4129],[28,3.6953],[29,3.8228],[30,3.9386],[32,4.1387],[34,4.3057],[35,4.3682],[38,4.5033],[39,4.5164],[42,4.5986],[43,4.6021],[46,4.6368],[49,4.6344],[52,4.6023],[53,4.5699],[54,4.55],[55,4.5195],[57,4.4277],[58,4.3645],[60,4.2721],[61,4.2415],[62,4.2361],[63,4.0917],[64,3.6378],[65,3.1069],[66,2.6757],[67,2.4015],[68,2.3417],[69,2.6326],[70,3.2964],[71,4.1739],[73,5.5982],[74,6.7623],[75,8.8193],[76,11.6519],[77,15.0314],[78,18.6256],[79,22.0466],[80,25.1291],[81,27.7222],[82,30.0642],[83,31.8032],[84,32.8485],[85,33.2092],[86,33.2839],[87,33.1099],[88,32.791],[89,32.6724],[90,32.7733],[91,33.0062],[93,33.6482],[94,33.8476],[95,33.9005],[96,33.7716],[97,33.4271],[99,32.3342],[100,31.9843],[101,31.8458],[102,31.8495],[103,32.05],[104,32.4465],[105,32.9298],[106,33.3126],[107,33.4705],[108,33.3944],[111,32.8316],[112,32.7644],[114,32.8597],[115,33.0822],[117,33.7113],[118,33.9153],[119,33.9842],[121,33.8247],[122,33.5844],[124,32.7222],[125,32.3458],[126,32.1063],[127,32.0214],[129,32.1123],[130,32.3635],[133,33.7032],[134,33.9823],[135,34.1095],[136,34.0717],[138,33.7259],[139,33.6924],[143,34.4951],[144,34.6308],[145,34.3434],[146,33.62],[147,32.6718],[148,30.8445],[149,27.9694],[150,24.2976],[151,20.0284],[152,15.8695],[153,12.4034],[154,9.9411],[155,8.2865],[156,6.87],[157,5.5853],[158,4.4611],[159,3.8014],[160,3.3258],[161,2.6682],[162,1.9533],[163,1.2287],[164,0.6228],[165,0.205],[166,-0.0654],[167,-0.2142],[168,-0.2511],[169,-0.182],[170,-0.0474],[171,0.1066],[172,0.2224],[173,0.2818],[174,0.2925],[175,0.2688],[176,0.2026],[177,0.1038],[178,0.0191],[179,0.0049],[180,0.0629],[181,0.171],[182,0.3306],[183,0.5625],[184,0.8447],[185,1.1579],[186,1.5409],[187,2.0006],[188,2.4861],[189,2.9197],[190,3.2343],[191,3.4201],[192,3.5317],[193,3.5959],[194,3.6216],[195,3.6102],[196,3.5819],[198,3.4858],[199,3.4458],[201,3.3783],[202,3.3652],[203,3.3716],[204,3.3872],[205,3.4188],[206,3.4814],[207,3.6244],[208,3.7008],[209,3.5101],[210,3.1389],[211,2.8307],[212,2.6287],[213,2.467],[214,2.3689],[215,2.3354],[216,2.3513],[217,2.3889],[219,2.5133],[222,2.6577],[223,2.6994],[224,2.7315],[225,2.771],[226,2.823],[227,2.8574],[228,2.8628],[229,2.8612],[231,2.8158],[233,2.7578],[234,2.7165],[235,2.7458],[236,2.8557],[237,3.0126],[238,3.2433],[239,3.5162],[240,3.8319],[242,4.5065],[243,4.8301],[244,5.1273],[245,5.4044],[246,5.6662],[247,5.9075],[248,6.0936],[249,6.1463],[250,6.0698],[251,5.9504],[252,5.897],[255,5.9555],[256,5.9667],[260,5.9622],[263,5.9855],[270,6.0862],[273,6.0966],[275,6.0699],[278,6.0096],[279,5.9787],[282,5.8421],[283,5.7782],[285,5.5966],[287,5.3793],[289,5.1362],[290,5.0063],[292,4.7885],[293,4.6962],[294,4.6191],[295,4.5635],[296,4.5196],[298,4.4602],[300,4.4213],[301,4.4099],[305,4.4097],[306,0]]
       }
       arm_r_pv_ik_ctrl.Rotation.X {
        type=MovieSceneScriptingFloatChannel
        default=0
        thinned_from=307
        keys [[0,14.0082],[5,13.9582],[8,13.8593],[9,13.8065],[10,13.7213],[11,13.5972],[12,13.4206],[13,13.1596],[14,12.8113],[16,12.0273],[17,11.6095],[18,11.1518],[19,10.6324],[20,10.1499],[21,9.7324],[22,9.3362],[23,9.0432],[24,8.9098],[25,8.8312],[26,8.7774],[27,8.7431],[28,8.7284],[29,8.7363],[30,8.7692],[32,8.9005],[35,9.2017],[37,9.4495],[40,9.8901],[45,10.7697],[48,11.2176],[50,11.461],[53,11.716],[62,12.2061],[63,12.0388],[64,11.3585],[65,10.3484],[66,9.1881],[67,7.9596],[68,6.4644],[69,4.6689],[70,2.6525],[71,-1.0631],[72,-8.0738],[73,-18.8318],[74,-33.7828],[75,-49.9586],[76,-64.1374],[77,-76.0505],[78,-85.668],[79,-93.1843],[80,-98.9429],[81,-103.2221],[82,-106.6772],[83,-109.0465],[84,-110.3706],[85,-110.8006],[86,-110.8699],[89,-109.7915],[90,-109.8155],[94,-111.1804],[95,-111.2682],[96,-111.0925],[97,-110.62],[99,-109.2232],[100,-108.7846],[102,-108.5774],[103,-108.8315],[106,-110.5274],[107,-110.7238],[112,-109.6232],[114,-109.7071],[118,-111.0563],[121,-110.9994],[122,-110.7119],[125,-109.065],[127,-108.5882],[129,-108.6521],[131,-109.4766],[133,-110.6538],[135,-111.1921],[139,-110.6068],[144,-111.8712],[145,-111.5818],[146,-110.8749],[147,-109.9198],[148,-107.4821],[149,-103.1593],[150,-96.9117],[151,-88.0162],[152,-76.6473],[153,-62.8941],[154,-47.217],[155,-30.1586],[156,-12.2279],[157,1.6994],[158,9.5677],[159,12.7528],[160,14.0496],[161,14.6566],[162,14.6677],[163,14.166],[164,13.5422],[165,13.0894],[166,12.747],[167,12.4738],[168,12.2388],[171,11.6999],[172,11.4797],[173,11.2141],[174,10.8867],[175,10.4771],[176,9.9515],[177,9.3671],[178,8.8654],[179,8.6159],[180,8.6393],[181,8.8505],[182,9.2247],[183,9.7528],[184,10.4085],[185,11.1865],[186,12.0804],[188,14.0095],[189,14.8637],[190,15.4758],[191,15.8571],[192,16.0956],[193,16.2257],[194,16.2846],[195,16.2796],[196,16.2082],[200,15.7408],[201,15.5456],[204,14.8688],[207,14.2582],[211,13.3838],[213,13.0054],[215,12.7151],[218,12.4387],[220,12.3389],[227,12.2134],[229,12.0995],[231,11.8996],[232,11.7694],[233,11.5958],[234,11.392],[235,11.2533],[236,11.207],[237,11.2222],[239,11.3021],[247,11.4769],[248,11.5202],[249,11.6078],[251,11.8508],[253,12.1988],[256,12.8725],[258,13.2221],[260,13.5128],[262,13.7222],[280,14.8801],[283,14.997],[286,15.0052],[289,14.8944],[300,14.1341],[304,14.0189],[305,14.0189],[306,0]]
       }
       arm_r_pv_ik_ctrl.Rotation.Y {
        type=MovieSceneScriptingFloatChannel
        default=0
        thinned_from=307
        keys [[0,2.2954],[1,2.29],[2,2.2778],[3,2.2589],[4,2.2588],[5,2.2499],[6,2.2572],[7,2.2418],[8,2.28],[9,2.3701],[10,2.5069],[11,2.6704],[12,2.8862],[13,3.1887],[14,3.5456],[15,3.9341],[16,4.3576],[17,4.8002],[18,5.218],[19,5.5779],[20,5.8897],[21,6.1662],[22,6.4293],[23,6.6732],[24,6.9],[25,7.073],[26,7.221],[28,7.4724],[29,7.565],[30,7.6112],[31,7.6028],[32,7.5758],[34,7.4483],[35,7.3624],[37,7.1326],[38,7.0058],[39,6.8278],[42,6.3736],[43,6.1804],[45,5.8644],[47,5.5872],[49,5.3449],[51,5.1228],[52,5.0331],[53,4.9096],[55,4.7115],[56,4.5929],[57,4.4642],[58,4.3021],[60,4.0109],[61,3.8916],[62,3.824],[63,3.7029],[64,3.4884],[65,3.3642],[66,3.5506],[67,4.1708],[68,5.2515],[69,6.7779],[70,9.3655],[71,14.2438],[72,20.2021],[73,25.6892],[74,29.2968],[75,29.7329],[76,27.7952],[77,24.7138],[78,21.4934],[79,18.6217],[80,16.5887],[81,15.4662],[82,15.1195],[83,15.2125],[84,15.4225],[85,15.4844],[86,15.4755],[87,15.3951],[88,15.2727],[89,15.206],[90,15.2082],[92,15.3374],[94,15.5022],[95,15.5326],[96,15.5153],[97,15.4389],[98,15.2863],[99,15.1672],[100,15.0934],[101,15.0729],[102,15.0952],[103,15.1764],[106,15.5752],[107,15.627],[108,15.6042],[111,15.4279],[112,15.4131],[114,15.487],[115,15.5767],[117,15.8414],[119,16.017],[121,16.0817],[122,16.0612],[125,15.8027],[126,15.7454],[129,15.7427],[130,15.8145],[133,16.194],[134,16.2753],[135,16.3134],[137,16.2623],[138,16.2059],[139,16.2006],[140,16.2834],[141,16.4028],[142,16.5637],[143,16.8],[144,17.1121],[145,17.4808],[146,17.9941],[147,18.8367],[148,20.0474],[149,21.7533],[150,24.2507],[151,28.283],[152,33.0847],[153,37.5447],[154,40.4544],[155,40.5766],[156,36.0926],[157,27.4563],[158,18.6154],[159,13.7236],[160,11.2453],[161,8.9391],[162,7.1239],[163,5.8563],[164,5.0528],[165,4.5647],[166,4.3191],[167,4.253],[168,4.3287],[169,4.5298],[170,4.8376],[172,5.5302],[174,6.1276],[175,6.4803],[177,7.2878],[178,7.6096],[179,7.7883],[180,7.8061],[181,7.7139],[184,7.2809],[185,7.0739],[186,6.9052],[187,6.8361],[191,6.7729],[192,6.747],[194,6.726],[195,6.7019],[202,6.6311],[203,6.6115],[204,6.5714],[205,6.4656],[206,6.3432],[207,6.2768],[208,6.0933],[209,5.535],[210,4.7196],[211,4.0569],[212,3.6031],[213,3.2228],[214,2.949],[215,2.8372],[216,2.8786],[217,2.9843],[219,3.3765],[220,3.504],[221,3.5956],[224,3.8203],[226,3.9887],[227,4.0419],[228,4.0488],[229,4.0332],[231,3.9489],[232,3.9142],[233,3.8945],[234,3.8551],[235,3.8755],[236,3.9246],[237,3.9338],[238,4.0496],[239,4.3106],[240,4.7539],[241,5.2514],[243,6.2858],[244,6.8194],[246,7.9654],[247,8.4972],[248,8.8796],[249,8.7649],[250,8.2059],[251,7.5041],[252,6.9615],[254,6.2725],[255,5.9476],[256,5.6364],[257,5.3464],[258,5.0817],[259,4.8539],[260,4.6651],[261,4.5222],[262,4.4183],[263,4.3476],[264,4.3032],[265,4.2474],[266,4.2047],[267,4.1784],[270,4.1322],[273,4.0382],[274,3.9819],[275,3.9356],[276,3.8697],[277,3.821],[278,3.7444],[279,3.6584],[280,3.5523],[281,3.4778],[283,3.3122],[285,3.1577],[286,3.0651],[287,2.9901],[288,2.886],[289,2.7965],[290,2.6838],[291,2.5823],[292,2.4867],[293,2.4039],[294,2.3413],[295,2.3006],[296,2.2757],[297,2.2629],[299,2.2589],[301,2.2655],[302,2.2803],[303,2.2804],[304,2.2938],[305,2.2938],[306,0]]
       }
       arm_r_pv_ik_ctrl.Rotation.Z {
        type=MovieSceneScriptingFloatChannel
        default=0
        thinned_from=307
        keys [[0,12.1486],[1,12.164],[3,12.2349],[6,12.2483],[7,12.2186],[8,12.1369],[9,12.0185],[10,11.8699],[11,11.6769],[12,11.4368],[13,11.1337],[14,10.7282],[15,10.2004],[16,9.5383],[17,8.6837],[18,7.5831],[19,6.145],[20,4.4494],[21,2.5027],[22,0.3718],[23,-1.5798],[24,-3.1001],[25,-4.3588],[26,-5.4458],[27,-6.3635],[28,-7.105],[29,-7.7116],[30,-8.2185],[31,-8.6329],[32,-8.9843],[33,-9.2766],[34,-9.5221],[35,-9.7452],[37,-10.1566],[38,-10.3375],[40,-10.6565],[41,-10.7598],[42,-10.9012],[45,-11.1475],[48,-11.2483],[50,-11.2534],[52,-11.1832],[54,-11.0081],[56,-10.773],[58,-10.4419],[59,-10.2255],[60,-9.9637],[61,-9.6522],[62,-9.2766],[63,-9.4271],[64,-10.7936],[66,-14.8067],[67,-16.5455],[68,-18.3845],[69,-20.446],[70,-23.0259],[71,-26.6316],[72,-31.7403],[73,-38.8167],[74,-47.687],[75,-56.1657],[76,-62.1571],[77,-65.4618],[78,-66.4377],[79,-65.7784],[80,-64.1923],[81,-62.4242],[82,-61.152],[83,-60.6241],[86,-60.2826],[89,-59.6138],[95,-59.1152],[97,-58.728],[100,-57.9041],[103,-57.4504],[107,-57.2131],[110,-56.8339],[114,-56.8813],[119,-57.4047],[122,-57.3788],[129,-56.841],[135,-57.2733],[139,-57.1766],[141,-57.4939],[142,-57.7937],[143,-58.2349],[144,-58.8586],[145,-59.6229],[146,-60.7324],[147,-62.2767],[148,-64.0085],[149,-65.4444],[150,-66.2575],[151,-66.2685],[152,-63.9866],[153,-58.4269],[154,-49.6881],[155,-38.4896],[156,-25.9103],[157,-15.7092],[158,-8.8962],[159,-4.4472],[160,-1.1261],[161,1.2435],[162,2.6393],[163,3.0684],[164,3.0704],[165,3.1171],[166,3.2386],[167,3.4984],[168,4.2306],[169,5.4024],[170,6.8613],[171,8.5327],[172,10.2837],[175,15.4138],[176,17.2063],[177,19.0502],[178,20.7867],[179,22.2738],[180,23.4248],[181,24.3207],[182,24.9605],[183,25.3644],[184,25.4499],[185,25.1894],[186,24.6568],[187,23.9001],[189,22.1829],[190,21.4308],[191,20.786],[192,20.2451],[193,19.7955],[194,19.4611],[195,19.2444],[196,19.1642],[197,19.2114],[198,19.3912],[199,19.6991],[200,20.1617],[201,20.8725],[202,21.8899],[203,23.2524],[204,24.96],[205,26.9637],[206,29.2218],[207,31.6611],[208,34.2681],[211,42.596],[212,45.13],[213,47.4077],[214,49.4064],[215,51.0721],[216,52.4377],[217,53.5076],[218,54.3446],[219,55.0035],[220,55.5333],[222,56.2598],[224,56.6809],[226,56.8611],[228,56.7565],[231,56.159],[233,55.4537],[235,54.3589],[236,53.6437],[237,52.7914],[238,51.8017],[239,50.6438],[240,49.2653],[241,47.5456],[242,45.4663],[243,43.0529],[244,40.2866],[245,37.0894],[247,30.2609],[248,27.18],[249,24.2826],[250,21.4763],[251,18.9129],[252,16.7429],[253,14.8668],[254,13.1277],[255,11.5342],[256,10.1034],[257,8.8425],[258,7.7513],[259,6.8103],[260,5.9984],[261,5.3026],[262,4.708],[263,4.1986],[264,3.7756],[265,3.4167],[266,3.1112],[267,2.864],[268,2.663],[269,2.5056],[270,2.3871],[271,2.3005],[272,2.2628],[273,2.2631],[274,2.2857],[275,2.3585],[276,2.4658],[277,2.6141],[278,2.8178],[279,3.1046],[280,3.4242],[281,3.7958],[282,4.2052],[283,4.6607],[284,5.1573],[285,5.6882],[287,6.8177],[289,7.9596],[290,8.5098],[292,9.5356],[293,10.022],[294,10.4748],[295,10.8658],[296,11.187],[297,11.4314],[298,11.6292],[299,11.7898],[300,11.9159],[301,12.0089],[303,12.1169],[305,12.1358],[306,0]]
       }
       arm_r_pv_ik_ctrl.Scale.X {type=MovieSceneScriptingFloatChannel, default=1, const [1,307]}
       arm_r_pv_ik_ctrl.Scale.Y {type=MovieSceneScriptingFloatChannel, default=1, const [1,307]}
       arm_r_pv_ik_ctrl.Scale.Z {type=MovieSceneScriptingFloatChannel, default=1, const [1,307]}
       body_ctrl.Location.X {
        type=MovieSceneScriptingFloatChannel
        default=0
        thinned_from=307
        keys [[0,-1.7773],[3,-1.8035],[4,-1.8085],[5,-1.8087],[6,-1.8044],[7,-1.7851],[8,-1.7342],[9,-1.653],[10,-1.5539],[11,-1.4443],[12,-1.3232],[13,-1.1903],[14,-1.0506],[15,-0.9069],[16,-0.7613],[18,-0.4666],[19,-0.3238],[20,-0.1868],[21,-0.0584],[22,0.0523],[23,0.1505],[24,0.2288],[25,0.2962],[26,0.3594],[27,0.4085],[28,0.4526],[29,0.488],[30,0.5094],[31,0.5233],[32,0.5264],[33,0.525],[34,0.5214],[35,0.5123],[36,0.4977],[37,0.4806],[38,0.4608],[39,0.4374],[40,0.4126],[41,0.3841],[42,0.3529],[43,0.3208],[44,0.2912],[45,0.2644],[47,0.2129],[48,0.1867],[49,0.1619],[50,0.1367],[51,0.1119],[52,0.0916],[53,0.071],[54,0.0444],[55,0.0265],[56,0.0017],[57,-0.0145],[58,-0.0405],[59,-0.0687],[60,-0.115],[61,-0.1789],[62,-0.2523],[63,-0.3345],[64,-0.4192],[65,-0.5097],[66,-0.618],[67,-0.7281],[69,-0.9688],[70,-1.0923],[71,-1.222],[73,-1.4667],[74,-1.5942],[75,-1.7176],[77,-1.9418],[78,-2.0426],[79,-2.1049],[80,-2.1308],[81,-2.1445],[82,-2.1502],[83,-2.1502],[85,-2.1201],[88,-2.0279],[89,-1.9888],[90,-1.9585],[92,-1.9042],[94,-1.8713],[95,-1.8578],[96,-1.8515],[97,-1.839],[100,-1.814],[102,-1.8059],[105,-1.81],[112,-1.8394],[120,-1.8367],[127,-1.8205],[129,-1.811],[134,-1.7793],[139,-1.7727],[143,-1.7755],[145,-1.7819],[146,-1.7894],[149,-1.8301],[150,-1.8406],[152,-1.8487],[154,-1.843],[155,-1.8325],[156,-1.8078],[157,-1.7709],[158,-1.7261],[159,-1.6764],[162,-1.5157],[163,-1.4691],[165,-1.3899],[166,-1.3559],[167,-1.3271],[168,-1.3055],[169,-1.2911],[170,-1.2824],[171,-1.2781],[173,-1.2775],[174,-1.2812],[175,-1.2881],[176,-1.2981],[181,-1.3688],[183,-1.3911],[186,-1.405],[193,-1.4025],[196,-1.4136],[198,-1.4318],[199,-1.4511],[200,-1.4807],[201,-1.517],[202,-1.5456],[203,-1.5474],[204,-1.5452],[205,-1.557],[206,-1.5838],[207,-1.6407],[209,-1.7878],[211,-1.9242],[214,-2.153],[215,-2.2234],[216,-2.2803],[217,-2.3168],[218,-2.3408],[220,-2.3738],[222,-2.389],[224,-2.393],[226,-2.3818],[228,-2.3592],[230,-2.3218],[232,-2.2732],[233,-2.2428],[235,-2.1691],[237,-2.0812],[239,-1.9839],[240,-1.9303],[241,-1.869],[243,-1.7306],[244,-1.6574],[246,-1.5002],[247,-1.4253],[248,-1.3646],[249,-1.3154],[250,-1.2725],[251,-1.2387],[252,-1.2099],[254,-1.1649],[255,-1.1464],[256,-1.1317],[257,-1.1216],[258,-1.1167],[259,-1.117],[260,-1.1219],[261,-1.1324],[262,-1.1456],[263,-1.1529],[264,-1.1704],[266,-1.1957],[268,-1.2282],[269,-1.2472],[270,-1.2598],[271,-1.2822],[273,-1.3176],[275,-1.357],[276,-1.3799],[278,-1.4217],[280,-1.469],[283,-1.534],[286,-1.5898],[291,-1.6709],[293,-1.6965],[296,-1.7243],[300,-1.7449],[305,-1.7468],[306,0]]
       }
       body_ctrl.Location.Y {
        type=MovieSceneScriptingFloatChannel
        default=0
        thinned_from=307
        keys [[0,-3.496],[6,-3.4775],[9,-3.492],[11,-3.5189],[12,-3.5393],[13,-3.5673],[15,-3.6353],[18,-3.6969],[20,-3.7194],[45,-3.7694],[58,-3.7347],[60,-3.742],[61,-3.755],[63,-3.8044],[64,-3.85],[65,-3.9081],[66,-4.0006],[67,-4.1389],[68,-4.3426],[69,-4.5616],[70,-4.7463],[71,-4.8928],[72,-4.9939],[73,-5.0502],[74,-5.0797],[75,-5.0911],[77,-5.086],[78,-5.072],[92,-5.1646],[114,-5.0807],[135,-5.1232],[139,-5.1027],[143,-5.0359],[147,-4.9182],[151,-4.787],[153,-4.7446],[154,-4.7325],[155,-4.738],[164,-4.8897],[167,-4.9577],[170,-4.9892],[188,-4.9531],[196,-4.9708],[198,-5.0027],[199,-5.0322],[201,-5.1216],[202,-5.1978],[203,-5.3223],[204,-5.4969],[210,-6.7397],[211,-6.929],[213,-7.2595],[214,-7.395],[215,-7.4992],[216,-7.5621],[218,-7.642],[220,-7.6745],[225,-7.683],[230,-7.6373],[233,-7.5467],[234,-7.4916],[235,-7.369],[236,-7.1782],[238,-6.7571],[244,-5.3799],[248,-4.4974],[249,-4.3042],[250,-4.137],[251,-3.987],[252,-3.8724],[253,-3.7668],[255,-3.609],[257,-3.5043],[259,-3.4507],[261,-3.4447],[264,-3.4783],[265,-3.4983],[266,-3.5113],[267,-3.5381],[272,-3.6072],[274,-3.6166],[279,-3.6064],[283,-3.5759],[290,-3.4963],[294,-3.4805],[301,-3.4891],[305,-3.4946],[306,0]]
       }
       body_ctrl.Location.Z {
        type=MovieSceneScriptingFloatChannel
        default=0
        thinned_from=307
        keys [[0,-1.0642],[8,-1.07],[10,-1.066],[11,-1.0606],[12,-1.0519],[13,-1.0382],[15,-0.9994],[17,-0.9679],[18,-0.959],[19,-0.9569],[20,-0.9583],[21,-0.9625],[24,-0.9856],[28,-1.0253],[29,-1.0335],[30,-1.0391],[31,-1.0424],[33,-1.0434],[35,-1.0385],[38,-1.0236],[43,-0.989],[49,-0.9437],[51,-0.9328],[54,-0.9235],[60,-0.9194],[62,-0.9225],[63,-0.926],[65,-0.9382],[66,-0.9463],[67,-0.9577],[69,-0.9904],[70,-1.0043],[71,-1.014],[72,-1.0199],[79,-1.027],[91,-1.0058],[115,-1.0089],[117,-1.0137],[118,-1.0139],[122,-1.0233],[123,-1.0234],[130,-1.0391],[147,-1.051],[149,-1.0673],[150,-1.0798],[151,-1.0993],[152,-1.1261],[153,-1.1599],[154,-1.1971],[155,-1.2429],[156,-1.2931],[157,-1.35],[158,-1.3989],[159,-1.4311],[160,-1.4505],[161,-1.4612],[162,-1.466],[163,-1.4662],[164,-1.4551],[165,-1.4291],[166,-1.3941],[168,-1.3197],[169,-1.2897],[170,-1.2654],[171,-1.2461],[172,-1.2313],[173,-1.2205],[174,-1.2133],[176,-1.2054],[178,-1.2023],[181,-1.2022],[183,-1.2137],[184,-1.2254],[186,-1.2602],[189,-1.2773],[190,-1.2772],[191,-1.2829],[198,-1.2854],[200,-1.2981],[201,-1.3126],[202,-1.3323],[203,-1.3654],[204,-1.4152],[205,-1.4781],[208,-1.6815],[209,-1.7342],[210,-1.7744],[211,-1.801],[212,-1.8181],[213,-1.8279],[214,-1.8315],[216,-1.8266],[223,-1.7915],[235,-1.7785],[236,-1.7639],[237,-1.7123],[238,-1.6387],[240,-1.4777],[241,-1.383],[242,-1.2837],[243,-1.195],[244,-1.129],[245,-1.0807],[246,-1.045],[247,-1.0243],[248,-1.0187],[249,-1.0199],[250,-1.0243],[252,-1.0393],[253,-1.0505],[258,-1.118],[260,-1.1374],[262,-1.1368],[264,-1.1301],[265,-1.1236],[266,-1.1203],[275,-1.0356],[277,-1.0233],[279,-1.0163],[284,-1.0134],[286,-1.0174],[288,-1.0293],[292,-1.0601],[294,-1.0707],[297,-1.0711],[305,-1.0644],[306,0]]
       }
       body_ctrl.Rotation.X {
        type=MovieSceneScriptingFloatChannel
        default=0
        thinned_from=307
        keys [[0,-18.6203],[10,-18.6301],[21,-19.0514],[30,-19.5119],[60,-19.3436],[63,-19.2078],[65,-19.0087],[66,-18.855],[67,-18.6467],[72,-17.2365],[73,-16.9873],[74,-16.7724],[76,-16.4398],[78,-16.1953],[80,-16.0386],[84,-15.9017],[114,-15.9286],[145,-16.0913],[147,-16.1599],[149,-16.3197],[150,-16.4521],[151,-16.632],[152,-16.8668],[155,-17.6955],[156,-17.9367],[157,-18.1245],[158,-18.2415],[159,-18.2814],[183,-17.816],[195,-17.7576],[198,-17.8609],[199,-17.971],[200,-18.1803],[203,-18.9011],[204,-19.0604],[205,-19.1478],[207,-19.2085],[223,-19.0246],[238,-19.063],[250,-18.7303],[283,-18.5422],[305,-18.6243],[306,0]]
       }
       body_ctrl.Rotation.Y {
        type=MovieSceneScriptingFloatChannel
        default=0
        thinned_from=307
        keys [[0,1.8976],[1,1.9033],[3,1.9223],[4,1.9212],[5,1.9089],[7,1.8358],[8,1.758],[9,1.6608],[10,1.5248],[11,1.3514],[12,1.1339],[13,0.8649],[14,0.5819],[15,0.3152],[16,0.0677],[17,-0.1576],[18,-0.3447],[19,-0.4955],[20,-0.6416],[21,-0.7957],[23,-1.0914],[24,-1.2433],[25,-1.3915],[26,-1.5309],[27,-1.6586],[28,-1.7697],[29,-1.8401],[30,-1.8654],[31,-1.8569],[32,-1.8265],[34,-1.7491],[36,-1.6908],[38,-1.6217],[39,-1.5612],[40,-1.5067],[41,-1.4393],[42,-1.3586],[43,-1.2572],[45,-1.0716],[46,-0.9824],[48,-0.7952],[49,-0.7056],[50,-0.6259],[51,-0.5571],[52,-0.5087],[53,-0.4516],[55,-0.3947],[56,-0.3785],[57,-0.3673],[58,-0.3514],[59,-0.3274],[60,-0.2988],[61,-0.268],[62,-0.2334],[63,-0.216],[64,-0.2213],[65,-0.2257],[66,-0.2013],[67,-0.127],[68,-0.0314],[69,0.0272],[70,0.1214],[71,0.284],[72,0.47],[73,0.7133],[74,1.0384],[75,1.434],[76,1.9095],[78,3.0883],[79,3.4994],[80,3.7368],[81,3.9114],[82,4.0409],[83,4.1291],[84,4.1493],[85,4.1244],[86,4.0635],[87,3.9774],[88,3.8775],[90,3.6513],[92,3.5164],[93,3.4801],[94,3.4559],[97,3.4351],[98,3.4041],[99,3.3475],[100,3.3015],[101,3.2862],[102,3.2872],[103,3.2991],[105,3.3455],[107,3.4462],[108,3.5048],[110,3.6315],[111,3.6773],[112,3.71],[114,3.7205],[115,3.7333],[116,3.7354],[118,3.7164],[123,3.6051],[124,3.6011],[125,3.5818],[128,3.5762],[130,3.6036],[132,3.6605],[133,3.6776],[134,3.6777],[135,3.691],[142,3.6845],[143,3.6739],[144,3.6508],[145,3.6452],[146,3.6283],[149,3.6626],[150,3.6502],[151,3.6095],[152,3.5408],[153,3.4489],[154,3.3293],[155,3.17],[156,2.9194],[157,2.6017],[158,2.2672],[159,1.9426],[160,1.6135],[161,1.2692],[162,0.9491],[163,0.6881],[164,0.4787],[165,0.2956],[166,0.1425],[167,0.0246],[168,-0.0403],[169,-0.0491],[170,-0.0161],[171,0.0394],[172,0.0979],[173,0.1471],[174,0.1924],[175,0.2332],[177,0.3126],[178,0.3485],[179,0.3786],[180,0.4008],[181,0.4172],[182,0.428],[183,0.4337],[185,0.4131],[186,0.3985],[187,0.3722],[188,0.3444],[189,0.3176],[190,0.2946],[191,0.2728],[192,0.253],[193,0.2355],[194,0.2234],[195,0.2313],[196,0.2298],[197,0.2564],[198,0.3035],[199,0.3854],[200,0.5213],[201,0.7023],[202,0.9205],[203,1.163],[204,1.4237],[205,1.7194],[206,2.0705],[207,2.4965],[208,2.9096],[209,3.1931],[210,3.3769],[212,3.724],[213,3.8685],[214,3.9917],[215,4.0719],[216,4.1238],[217,4.1523],[221,4.1567],[224,4.1361],[225,4.1099],[226,4.0731],[228,3.9724],[229,3.8966],[231,3.7322],[235,3.3786],[237,3.2226],[240,2.9992],[241,2.9128],[242,2.8156],[243,2.7069],[244,2.5801],[245,2.4414],[246,2.2924],[247,2.1297],[248,1.9856],[249,1.8188],[252,1.2994],[253,1.1135],[254,0.935],[255,0.7589],[256,0.5937],[257,0.4425],[258,0.314],[259,0.2101],[260,0.1345],[261,0.0855],[262,0.0466],[263,0.0139],[264,-0.009],[265,-0.0249],[266,-0.033],[267,-0.0353],[268,-0.0299],[269,-0.0165],[270,0.0034],[271,0.0318],[272,0.0681],[273,0.1124],[274,0.1618],[275,0.2141],[276,0.2698],[277,0.3262],[278,0.3865],[279,0.4494],[280,0.5269],[281,0.5877],[282,0.668],[284,0.8313],[285,0.9187],[286,1.0202],[287,1.1077],[288,1.2117],[289,1.2929],[290,1.3906],[291,1.4711],[292,1.5455],[293,1.6134],[294,1.6719],[296,1.7692],[297,1.8111],[298,1.847],[299,1.8754],[300,1.8933],[301,1.9018],[302,1.8949],[303,1.904],[304,1.8956],[305,1.8956],[306,0]]
       }
       body_ctrl.Rotation.Z {
        type=MovieSceneScriptingFloatChannel
        default=0
        thinned_from=307
        keys [[0,-16.3004],[4,-16.3711],[7,-16.3125],[8,-16.2426],[10,-16.0047],[12,-15.6458],[13,-15.413],[14,-15.0568],[15,-14.5263],[16,-13.8341],[17,-12.9573],[18,-11.8339],[20,-9.1692],[21,-7.9406],[22,-6.7638],[23,-5.7695],[24,-5.032],[25,-4.4246],[26,-3.8857],[27,-3.4092],[28,-2.9859],[29,-2.604],[30,-2.2517],[31,-1.932],[32,-1.6524],[33,-1.4188],[34,-1.2397],[35,-1.1107],[36,-1.0075],[37,-0.9227],[38,-0.8534],[39,-0.8013],[40,-0.7604],[41,-0.7307],[42,-0.7108],[43,-0.7],[44,-0.694],[45,-0.694],[46,-0.6993],[47,-0.7091],[48,-0.725],[49,-0.7483],[50,-0.7796],[51,-0.8203],[52,-0.8673],[53,-0.9238],[54,-0.9853],[55,-1.0521],[56,-1.1272],[57,-1.2056],[58,-1.2925],[59,-1.3897],[60,-1.4979],[61,-1.6179],[62,-1.7512],[63,-1.9026],[64,-2.0997],[65,-2.3605],[66,-2.7046],[67,-3.1465],[68,-3.8108],[69,-4.764],[70,-5.906],[71,-7.2793],[72,-8.7438],[73,-10.2295],[74,-11.7773],[75,-13.1612],[76,-14.1827],[77,-14.9533],[78,-15.5516],[79,-15.9506],[80,-16.1447],[106,-17.954],[112,-18.1571],[144,-17.7704],[146,-17.6753],[148,-17.4922],[151,-17.0774],[156,-16.1561],[158,-15.8562],[160,-15.6481],[162,-15.5175],[164,-15.4604],[166,-15.4788],[167,-15.5352],[168,-15.6271],[169,-15.7556],[170,-15.9207],[171,-16.1207],[172,-16.3546],[173,-16.6219],[174,-16.9407],[175,-17.3194],[177,-18.1952],[180,-19.5743],[181,-19.9926],[182,-20.362],[183,-20.6694],[184,-20.9124],[186,-21.2633],[188,-21.4752],[195,-21.868],[197,-22.11],[198,-22.3008],[199,-22.5497],[200,-22.876],[201,-23.3126],[202,-23.8859],[203,-24.6251],[204,-25.5302],[205,-26.5527],[209,-31.0985],[210,-32.163],[211,-33.144],[212,-33.9775],[213,-34.6626],[214,-35.2359],[216,-36.1549],[217,-36.5089],[219,-36.995],[221,-37.2343],[224,-37.3064],[226,-37.2196],[228,-36.939],[230,-36.4322],[231,-36.1066],[232,-35.704],[233,-35.2019],[234,-34.5932],[236,-33.1586],[237,-32.3375],[238,-31.4433],[239,-30.4764],[240,-29.3944],[241,-28.1316],[242,-26.7337],[244,-23.7755],[245,-22.1415],[246,-20.4524],[247,-18.881],[248,-17.5833],[249,-16.5613],[250,-15.7225],[251,-15.0427],[252,-14.4902],[253,-14.0102],[254,-13.576],[255,-13.1834],[256,-12.8345],[257,-12.526],[258,-12.2593],[259,-12.0337],[260,-11.8469],[262,-11.5548],[264,-11.35],[266,-11.2195],[268,-11.1487],[271,-11.1222],[272,-11.1399],[273,-11.1801],[274,-11.2451],[275,-11.3337],[276,-11.4509],[277,-11.6014],[278,-11.7895],[284,-13.2353],[290,-14.7789],[292,-15.2191],[294,-15.571],[296,-15.8373],[299,-16.1153],[302,-16.2559],[305,-16.2923],[306,0]]
       }
       body_ctrl.Scale.X {type=MovieSceneScriptingFloatChannel, default=1, const [1,307]}
       body_ctrl.Scale.Y {type=MovieSceneScriptingFloatChannel, default=1, const [1,307]}
       body_ctrl.Scale.Z {type=MovieSceneScriptingFloatChannel, default=1, const [1,307]}
       hips_ctrl.Location.X {type=MovieSceneScriptingFloatChannel, default=0, thinned_from=307, keys [[0,0],[306,0]]}
       hips_ctrl.Location.Y {type=MovieSceneScriptingFloatChannel, default=0, thinned_from=307, keys [[0,0],[306,0]]}
       hips_ctrl.Location.Z {type=MovieSceneScriptingFloatChannel, default=0, thinned_from=307, keys [[0,0],[306,0]]}
       hips_ctrl.Rotation.X {
        type=MovieSceneScriptingFloatChannel
        default=0
        thinned_from=307
        keys [[0,12.1945],[69,12.0859],[70,12.0587],[75,11.7265],[79,11.5578],[97,11.4734],[114,11.5294],[145,11.6923],[149,11.7874],[151,11.9],[154,12.1585],[156,12.2674],[158,12.298],[167,12.1012],[185,12.08],[195,12.0128],[198,12.0651],[203,12.2832],[206,12.3302],[212,12.2799],[238,12.3219],[246,12.2106],[283,12.1846],[305,12.1982],[306,0]]
       }
       hips_ctrl.Rotation.Y {
        type=MovieSceneScriptingFloatChannel
        default=0
        thinned_from=307
        keys [[0,-2.9061],[5,-2.9061],[6,-2.8922],[7,-2.8922],[8,-2.8645],[10,-2.8198],[11,-2.7894],[13,-2.7175],[14,-2.6872],[15,-2.6713],[16,-2.6713],[17,-2.6871],[18,-2.7258],[20,-2.8334],[25,-3.0436],[31,-3.3362],[33,-3.4204],[34,-3.4518],[36,-3.4924],[37,-3.5057],[38,-3.5057],[39,-3.5222],[42,-3.5222],[43,-3.5337],[51,-3.5337],[52,-3.5205],[53,-3.5205],[57,-3.4445],[59,-3.3912],[62,-3.2859],[63,-3.2246],[64,-3.1246],[65,-2.9978],[66,-2.8591],[67,-2.7301],[68,-2.6116],[69,-2.4491],[70,-2.2986],[71,-2.1974],[72,-2.1064],[73,-2.0272],[74,-1.9597],[75,-1.9329],[76,-1.9756],[77,-2.0792],[78,-2.2219],[79,-2.3781],[80,-2.5274],[81,-2.6615],[82,-2.7803],[83,-2.8773],[84,-2.9442],[85,-2.9901],[86,-3.0257],[87,-3.0529],[88,-3.0738],[89,-3.0884],[90,-3.0884],[91,-3.1051],[94,-3.1051],[95,-3.1176],[97,-3.1176],[98,-3.0904],[99,-3.0363],[100,-2.9919],[101,-2.9771],[105,-2.9806],[107,-3.0231],[108,-3.0526],[110,-3.1283],[111,-3.1579],[112,-3.18],[114,-3.18],[115,-3.1892],[116,-3.1892],[118,-3.1691],[122,-3.0902],[123,-3.0745],[124,-3.0745],[125,-3.0588],[128,-3.0588],[130,-3.0869],[132,-3.1433],[133,-3.1602],[134,-3.1602],[135,-3.1731],[141,-3.1731],[143,-3.1769],[144,-3.1637],[145,-3.1637],[147,-3.1198],[148,-3.0896],[149,-3.05],[150,-2.9991],[151,-2.9349],[152,-2.8538],[153,-2.7557],[154,-2.6354],[155,-2.4853],[156,-2.298],[157,-2.0866],[158,-1.8828],[159,-1.6959],[160,-1.526],[161,-1.3616],[162,-1.2196],[163,-1.1145],[164,-1.0413],[165,-0.9878],[166,-0.9537],[167,-0.9404],[168,-0.9615],[169,-1.0201],[170,-1.1062],[171,-1.2036],[172,-1.2954],[173,-1.3719],[174,-1.4405],[176,-1.5624],[177,-1.6199],[178,-1.6724],[179,-1.718],[180,-1.7542],[181,-1.7837],[182,-1.8063],[183,-1.8235],[185,-1.8235],[186,-1.8182],[188,-1.7792],[192,-1.6965],[194,-1.6647],[195,-1.6647],[196,-1.6459],[197,-1.6424],[198,-1.644],[199,-1.6491],[200,-1.6618],[201,-1.6832],[202,-1.7185],[203,-1.7676],[204,-1.8375],[205,-1.9378],[206,-2.0735],[207,-2.27],[208,-2.4483],[209,-2.4995],[211,-2.4243],[213,-2.3967],[217,-2.3928],[218,-2.3851],[223,-2.3851],[224,-2.3969],[227,-2.3969],[228,-2.4103],[230,-2.4103],[232,-2.438],[233,-2.4588],[234,-2.4888],[235,-2.528],[236,-2.5789],[237,-2.6389],[240,-2.8529],[241,-2.92],[242,-2.9791],[243,-3.0293],[244,-3.0629],[245,-3.0863],[246,-3.0993],[247,-3.0993],[248,-3.1173],[252,-3.1198],[254,-3.0887],[260,-2.9379],[275,-2.6933],[279,-2.6438],[280,-2.6438],[281,-2.6263],[283,-2.6263],[285,-2.6219],[286,-2.6349],[287,-2.6349],[288,-2.6537],[289,-2.6537],[290,-2.6754],[292,-2.7016],[295,-2.7604],[299,-2.8523],[301,-2.8796],[302,-2.8796],[303,-2.8969],[305,-2.8969],[306,0]]
       }
       hips_ctrl.Rotation.Z {
        type=MovieSceneScriptingFloatChannel
        default=0
        thinned_from=307
        keys [[0,-0.6277],[5,-0.6277],[6,-0.625],[7,-0.625],[8,-0.6192],[10,-0.6096],[11,-0.6031],[13,-0.5877],[14,-0.5812],[15,-0.5777],[16,-0.5777],[17,-0.5808],[18,-0.5889],[20,-0.6116],[25,-0.6562],[31,-0.7192],[33,-0.7374],[34,-0.7441],[36,-0.753],[37,-0.7558],[38,-0.7558],[39,-0.7594],[42,-0.7594],[43,-0.762],[51,-0.762],[52,-0.7591],[53,-0.7591],[57,-0.742],[59,-0.7301],[62,-0.7062],[63,-0.6927],[64,-0.6707],[65,-0.6431],[66,-0.613],[67,-0.5851],[68,-0.5595],[69,-0.5242],[70,-0.4908],[71,-0.4665],[72,-0.4442],[73,-0.4249],[74,-0.4086],[75,-0.4012],[76,-0.4081],[77,-0.4279],[78,-0.4555],[79,-0.4862],[80,-0.5159],[81,-0.5424],[82,-0.5662],[83,-0.5853],[84,-0.5985],[85,-0.6074],[86,-0.6142],[88,-0.6235],[89,-0.6266],[90,-0.6266],[91,-0.6298],[94,-0.6298],[95,-0.6323],[97,-0.6323],[98,-0.6275],[99,-0.6171],[100,-0.6086],[101,-0.6057],[103,-0.6057],[104,-0.6072],[105,-0.6072],[107,-0.6163],[108,-0.6224],[110,-0.6377],[111,-0.6438],[112,-0.6483],[114,-0.6483],[115,-0.6513],[117,-0.6507],[118,-0.6491],[123,-0.633],[124,-0.633],[125,-0.6305],[128,-0.6305],[130,-0.6368],[132,-0.6484],[133,-0.6518],[134,-0.6518],[135,-0.6544],[141,-0.6544],[142,-0.656],[143,-0.656],[144,-0.6545],[145,-0.6545],[147,-0.6477],[148,-0.6427],[149,-0.6362],[150,-0.6282],[151,-0.6182],[152,-0.6054],[153,-0.5893],[154,-0.5675],[155,-0.5382],[156,-0.4996],[158,-0.4104],[159,-0.3694],[160,-0.3317],[161,-0.2953],[162,-0.2637],[163,-0.2404],[164,-0.2242],[165,-0.2124],[166,-0.2047],[167,-0.2016],[168,-0.2061],[169,-0.2188],[170,-0.2375],[171,-0.2587],[172,-0.2785],[173,-0.295],[174,-0.3098],[176,-0.3357],[177,-0.3479],[178,-0.3589],[179,-0.3685],[180,-0.376],[181,-0.382],[182,-0.3867],[183,-0.3902],[185,-0.3902],[186,-0.3885],[192,-0.3612],[194,-0.3542],[195,-0.3542],[196,-0.3506],[197,-0.3504],[198,-0.3513],[199,-0.3536],[200,-0.3577],[201,-0.3639],[202,-0.3728],[203,-0.3848],[204,-0.4009],[205,-0.4233],[206,-0.4531],[207,-0.4957],[208,-0.5336],[209,-0.5439],[210,-0.535],[211,-0.5274],[213,-0.5218],[217,-0.5216],[223,-0.5205],[224,-0.5237],[227,-0.5237],[228,-0.5267],[230,-0.5267],[232,-0.5328],[233,-0.5374],[234,-0.5441],[235,-0.5528],[236,-0.5637],[237,-0.5766],[240,-0.6212],[241,-0.6349],[242,-0.6467],[243,-0.6566],[244,-0.6633],[245,-0.6677],[246,-0.6703],[247,-0.6703],[248,-0.6744],[250,-0.6744],[252,-0.6756],[254,-0.6696],[260,-0.6357],[270,-0.5965],[278,-0.5725],[279,-0.5702],[280,-0.5702],[281,-0.5668],[285,-0.5666],[286,-0.5698],[287,-0.5698],[288,-0.5743],[289,-0.5743],[290,-0.5795],[292,-0.5852],[299,-0.617],[301,-0.6226],[302,-0.6226],[303,-0.626],[305,-0.626],[306,0]]
       }
       hips_ctrl.Scale.X {type=MovieSceneScriptingFloatChannel, default=1, const [1,307]}
       hips_ctrl.Scale.Y {type=MovieSceneScriptingFloatChannel, default=1, const [1,307]}
       hips_ctrl.Scale.Z {type=MovieSceneScriptingFloatChannel, default=1, const [1,307]}
       hips_tan_ctrl.Location.X {
        type=MovieSceneScriptingFloatChannel
        default=0
        thinned_from=307
        keys [[0,-0.0166],[6,-0.0166],[7,-0.0161],[8,-0.0165],[9,-0.0158],[10,-0.0158],[11,-0.0147],[12,-0.0148],[13,-0.0217],[15,-0.0215],[16,-0.0204],[17,-0.0205],[18,-0.0217],[19,-0.0207],[20,-0.0213],[21,-0.0181],[23,-0.0179],[24,-0.0182],[25,-0.0168],[26,-0.0173],[28,-0.0152],[29,-0.0143],[30,-0.017],[31,-0.0167],[32,-0.0155],[33,-0.0138],[34,-0.0144],[36,-0.0135],[37,-0.0124],[38,-0.0121],[50,-0.0121],[51,-0.0132],[52,-0.0132],[54,-0.0138],[55,-0.0122],[56,-0.0115],[57,-0.0127],[58,-0.0127],[59,-0.0114],[61,-0.0132],[62,-0.0119],[63,-0.0127],[64,-0.0284],[65,-0.0267],[66,-0.0272],[67,-0.0292],[68,-0.0289],[69,-0.0387],[70,-0.0383],[71,-0.0404],[72,-0.0422],[73,-0.0375],[74,-0.0379],[75,-0.0358],[76,-0.0331],[77,-0.0351],[79,-0.0271],[80,-0.0235],[81,-0.0219],[82,-0.0187],[83,-0.0175],[84,-0.0192],[85,-0.017],[88,-0.0197],[90,-0.0198],[91,-0.0207],[97,-0.0209],[98,-0.02],[99,-0.0168],[100,-0.0185],[101,-0.0193],[102,-0.0198],[104,-0.0199],[105,-0.0187],[106,-0.0188],[107,-0.0181],[111,-0.0133],[112,-0.0124],[113,-0.0113],[114,-0.0116],[115,-0.0149],[119,-0.0142],[120,-0.0134],[121,-0.0132],[122,-0.0155],[123,-0.0143],[125,-0.0143],[126,-0.0139],[128,-0.014],[130,-0.0147],[131,-0.0166],[132,-0.0167],[133,-0.0171],[136,-0.0173],[137,-0.0167],[139,-0.0167],[140,-0.0166],[141,-0.0169],[143,-0.0169],[144,-0.0162],[146,-0.016],[147,-0.0155],[148,-0.0154],[149,-0.0137],[150,-0.0135],[151,-0.0186],[152,-0.0195],[153,-0.0163],[154,-0.0211],[155,-0.0189],[156,-0.0232],[157,-0.0247],[158,-0.0275],[159,-0.0251],[160,-0.0311],[161,-0.0263],[162,-0.0324],[164,-0.0345],[165,-0.0353],[169,-0.037],[171,-0.0375],[178,-0.0414],[179,-0.0436],[180,-0.0393],[181,-0.0396],[182,-0.0406],[185,-0.0409],[186,-0.0399],[189,-0.0391],[190,-0.0427],[192,-0.0422],[193,-0.0412],[194,-0.041],[195,-0.0402],[196,-0.0397],[198,-0.0396],[199,-0.0392],[201,-0.0393],[202,-0.036],[203,-0.0363],[204,-0.0392],[205,-0.0394],[206,-0.0386],[207,-0.0347],[208,-0.0342],[209,-0.0357],[211,-0.0417],[212,-0.044],[213,-0.0475],[214,-0.0468],[216,-0.0508],[217,-0.051],[218,-0.0528],[221,-0.0533],[222,-0.055],[224,-0.0552],[225,-0.0539],[226,-0.0539],[227,-0.0552],[228,-0.0552],[229,-0.0543],[231,-0.0543],[232,-0.0536],[233,-0.0546],[234,-0.0533],[235,-0.053],[236,-0.0515],[237,-0.0518],[238,-0.0503],[239,-0.0476],[241,-0.0447],[242,-0.043],[243,-0.0409],[245,-0.0357],[246,-0.0376],[248,-0.0319],[249,-0.0298],[251,-0.0241],[253,-0.0202],[256,-0.0167],[257,-0.0163],[259,-0.0142],[260,-0.0139],[261,-0.0113],[266,-0.0105],[267,-0.0085],[268,-0.0084],[269,-0.0089],[271,-0.0087],[272,-0.009],[273,-0.0075],[274,-0.007],[275,-0.0138],[276,-0.015],[277,-0.015],[278,-0.0155],[279,-0.0152],[280,-0.0156],[281,-0.0156],[282,-0.0162],[283,-0.0162],[284,-0.0169],[285,-0.0169],[286,-0.0171],[287,-0.0177],[289,-0.0179],[290,-0.0186],[292,-0.0187],[293,-0.0204],[296,-0.0209],[297,-0.0166],[298,-0.0167],[299,-0.0147],[300,-0.0148],[301,-0.0165],[305,-0.0165],[306,0]]
       }
       hips_tan_ctrl.Location.Y {
        type=MovieSceneScriptingFloatChannel
        default=0.008
        thinned_from=307
        keys [[0,0.3096],[8,0.3096],[17,0.3101],[21,0.3098],[25,0.3106],[26,0.3095],[39,0.3106],[50,0.3114],[61,0.3112],[62,0.3093],[66,0.3096],[70,0.3082],[75,0.3025],[81,0.2988],[84,0.2989],[97,0.2986],[98,0.2991],[99,0.2983],[114,0.2999],[120,0.2978],[121,0.2997],[122,0.2997],[123,0.2989],[130,0.2984],[131,0.3011],[141,0.3014],[146,0.3006],[147,0.3022],[148,0.3019],[149,0.3038],[150,0.3034],[155,0.3101],[156,0.3119],[159,0.3119],[160,0.31],[162,0.3106],[163,0.3096],[179,0.3085],[186,0.3076],[187,0.3052],[194,0.3048],[195,0.3067],[196,0.307],[197,0.3085],[198,0.3064],[200,0.3099],[201,0.3093],[202,0.3114],[203,0.3104],[204,0.3107],[205,0.3121],[207,0.3117],[208,0.3104],[209,0.3121],[215,0.3116],[216,0.3132],[240,0.3127],[245,0.3112],[250,0.3113],[251,0.31],[274,0.3106],[275,0.3094],[290,0.31],[298,0.3107],[299,0.3093],[305,0.3097],[306,0.008]]
       }
       hips_tan_ctrl.Location.Z {
        type=MovieSceneScriptingFloatChannel
        default=-0.0073
        thinned_from=307
        keys [[0,0.1134],[8,0.1133],[9,0.1139],[19,0.1138],[20,0.1142],[21,0.1143],[22,0.1147],[23,0.1148],[24,0.1152],[25,0.1139],[26,0.1144],[27,0.1144],[28,0.1148],[29,0.1149],[30,0.1152],[32,0.1153],[33,0.1157],[38,0.1157],[39,0.1163],[50,0.1163],[60,0.1165],[61,0.1146],[69,0.1148],[70,0.1131],[71,0.113],[74,0.1082],[75,0.1081],[77,0.1049],[78,0.105],[79,0.1034],[84,0.1033],[96,0.1035],[98,0.1034],[99,0.1031],[103,0.1038],[105,0.1037],[106,0.104],[110,0.104],[111,0.1042],[113,0.1042],[114,0.1044],[115,0.104],[120,0.1033],[130,0.1027],[131,0.1056],[140,0.1057],[141,0.1054],[143,0.1054],[144,0.1051],[145,0.1051],[146,0.1049],[147,0.1066],[148,0.1062],[149,0.1085],[150,0.108],[155,0.1149],[156,0.1144],[157,0.1163],[158,0.1158],[162,0.1159],[163,0.1152],[178,0.113],[183,0.113],[184,0.1105],[194,0.1104],[195,0.1115],[196,0.1115],[197,0.1126],[198,0.1115],[200,0.1137],[201,0.1137],[202,0.1148],[203,0.1148],[204,0.1158],[206,0.1158],[207,0.1148],[215,0.1144],[216,0.1154],[232,0.1153],[233,0.1154],[235,0.1153],[236,0.1154],[241,0.1145],[243,0.1138],[244,0.1142],[250,0.1143],[251,0.114],[258,0.114],[259,0.1139],[266,0.1144],[267,0.1147],[274,0.1147],[275,0.1131],[290,0.113],[291,0.1141],[294,0.1141],[298,0.1142],[299,0.1131],[305,0.1134],[306,-0.0073]]
       }
       hips_tan_ctrl.Rotation.X {type=MovieSceneScriptingFloatChannel, default=0, const [0,307]}
       hips_tan_ctrl.Rotation.Y {type=MovieSceneScriptingFloatChannel, default=0, const [0,307]}
       hips_tan_ctrl.Rotation.Z {type=MovieSceneScriptingFloatChannel, default=0, const [0,307]}
       hips_tan_ctrl.Scale.X {type=MovieSceneScriptingFloatChannel, default=1, const [1,307]}
       hips_tan_ctrl.Scale.Y {type=MovieSceneScriptingFloatChannel, default=1, const [1,307]}
       hips_tan_ctrl.Scale.Z {type=MovieSceneScriptingFloatChannel, default=1, const [1,307]}
       spine_01_ctrl.Location.X {
        type=MovieSceneScriptingFloatChannel
        default=0
        thinned_from=307
        keys [[0,-0.0321],[12,-0.0321],[13,-0.0364],[16,-0.0364],[17,-0.0366],[18,-0.0376],[19,-0.0376],[24,-0.0425],[25,-0.0425],[26,-0.0434],[32,-0.0434],[33,-0.0425],[34,-0.043],[35,-0.0426],[50,-0.0426],[63,-0.0427],[64,-0.0514],[66,-0.0514],[67,-0.0521],[68,-0.0521],[69,-0.0551],[72,-0.0551],[73,-0.0521],[74,-0.0521],[76,-0.046],[77,-0.046],[80,-0.0368],[81,-0.0368],[82,-0.0338],[83,-0.0325],[98,-0.0325],[99,-0.0304],[101,-0.0313],[104,-0.0313],[105,-0.0308],[106,-0.0308],[112,-0.0283],[130,-0.0283],[131,-0.0294],[146,-0.0294],[150,-0.0292],[151,-0.0329],[153,-0.0329],[154,-0.0366],[155,-0.0366],[158,-0.0477],[159,-0.0477],[160,-0.0514],[161,-0.0514],[162,-0.0551],[163,-0.0564],[164,-0.0574],[165,-0.0581],[166,-0.0585],[167,-0.0587],[168,-0.0584],[169,-0.0577],[172,-0.054],[177,-0.0496],[178,-0.0489],[179,-0.0492],[180,-0.0467],[189,-0.0467],[190,-0.0492],[194,-0.0492],[195,-0.0495],[201,-0.0495],[202,-0.0478],[203,-0.0478],[204,-0.0461],[205,-0.0461],[206,-0.0427],[207,-0.041],[208,-0.0376],[210,-0.0376],[211,-0.0387],[212,-0.0387],[213,-0.0394],[224,-0.0394],[225,-0.0387],[226,-0.0387],[227,-0.0391],[228,-0.0391],[229,-0.0386],[232,-0.0386],[233,-0.0381],[234,-0.0381],[235,-0.0375],[236,-0.0365],[237,-0.0359],[240,-0.0327],[241,-0.0322],[242,-0.0311],[243,-0.0305],[258,-0.0305],[259,-0.0314],[274,-0.0314],[275,-0.035],[290,-0.035],[291,-0.0353],[296,-0.0353],[297,-0.0321],[305,-0.0321],[306,0]]
       }
       spine_01_ctrl.Location.Y {
        type=MovieSceneScriptingFloatChannel
        default=0
        thinned_from=307
        keys [[0,0.3493],[61,0.3507],[62,0.3496],[69,0.3495],[72,0.3457],[73,0.3458],[75,0.3432],[76,0.3432],[77,0.342],[79,0.342],[80,0.3407],[130,0.3411],[131,0.3426],[146,0.3426],[147,0.3435],[148,0.3435],[149,0.3445],[150,0.3445],[156,0.3508],[159,0.3507],[160,0.3496],[162,0.3496],[168,0.3477],[186,0.3483],[187,0.347],[196,0.347],[203,0.3506],[207,0.3512],[215,0.3504],[216,0.3512],[274,0.3497],[290,0.3491],[291,0.3499],[298,0.3499],[305,0.3493],[306,0]]
       }
       spine_01_ctrl.Location.Z {
        type=MovieSceneScriptingFloatChannel
        default=0
        thinned_from=307
        keys [[0,0.0954],[8,0.0954],[9,0.0957],[19,0.0956],[20,0.0958],[21,0.0958],[22,0.096],[23,0.096],[24,0.0963],[25,0.0963],[26,0.0965],[27,0.0965],[28,0.0967],[29,0.0967],[30,0.0969],[32,0.0969],[33,0.0971],[38,0.0971],[39,0.0974],[60,0.0974],[61,0.0965],[69,0.0965],[70,0.0949],[71,0.0949],[74,0.0902],[75,0.0902],[77,0.0871],[78,0.0871],[79,0.0856],[99,0.0853],[130,0.0862],[131,0.0877],[146,0.0877],[147,0.0887],[148,0.0887],[149,0.09],[150,0.09],[155,0.0965],[156,0.0965],[157,0.0978],[162,0.0977],[163,0.0972],[179,0.095],[183,0.095],[184,0.0937],[196,0.0937],[197,0.0943],[198,0.0943],[204,0.098],[206,0.098],[208,0.0968],[215,0.0967],[216,0.0976],[237,0.0976],[243,0.0961],[258,0.0961],[274,0.0959],[275,0.0951],[290,0.0951],[291,0.096],[298,0.096],[299,0.0954],[305,0.0954],[306,0]]
       }
       spine_01_ctrl.Rotation.X {
        type=MovieSceneScriptingFloatChannel
        default=0
        thinned_from=307
        keys [[0,5.7101],[67,5.7274],[70,5.6988],[75,5.6013],[119,5.5569],[148,5.6003],[151,5.6372],[156,5.7382],[159,5.758],[196,5.6676],[204,5.7316],[305,5.7111],[306,0]]
       }
       spine_01_ctrl.Rotation.Y {
        type=MovieSceneScriptingFloatChannel
        default=0
        thinned_from=307
        keys [[0,0.4526],[7,0.4526],[8,0.4626],[10,0.473],[11,0.4801],[15,0.5113],[18,0.5328],[22,0.5726],[23,0.5799],[24,0.5836],[25,0.585],[26,0.585],[28,0.5814],[31,0.5728],[32,0.5728],[33,0.5672],[35,0.5624],[36,0.5624],[37,0.5597],[50,0.5578],[51,0.5599],[53,0.5599],[54,0.5674],[55,0.5674],[56,0.5764],[58,0.5891],[60,0.6056],[62,0.6264],[63,0.6425],[64,0.6678],[65,0.6988],[66,0.7316],[67,0.7627],[68,0.7877],[69,0.8157],[70,0.8339],[71,0.8346],[72,0.8275],[73,0.8133],[74,0.7913],[75,0.7623],[76,0.7261],[77,0.681],[78,0.6311],[79,0.5837],[80,0.5455],[81,0.5137],[82,0.485],[83,0.4615],[84,0.4458],[85,0.4355],[86,0.4274],[87,0.4211],[88,0.4167],[89,0.4133],[91,0.4098],[96,0.4086],[97,0.41],[98,0.4164],[99,0.4299],[100,0.4414],[101,0.4459],[102,0.4472],[104,0.4463],[105,0.4445],[106,0.4415],[107,0.4373],[108,0.4304],[110,0.412],[111,0.4046],[112,0.3987],[113,0.3987],[114,0.3957],[115,0.3955],[117,0.3974],[118,0.3997],[119,0.3997],[120,0.409],[122,0.409],[123,0.4232],[125,0.4232],[126,0.428],[128,0.4278],[129,0.4254],[130,0.422],[131,0.4162],[132,0.4162],[133,0.4045],[140,0.4045],[141,0.3992],[143,0.4003],[145,0.4035],[146,0.4062],[147,0.4105],[148,0.4168],[149,0.4258],[150,0.4374],[151,0.4517],[152,0.4693],[153,0.4901],[154,0.5152],[155,0.545],[156,0.5817],[158,0.6621],[161,0.7745],[162,0.8071],[163,0.8316],[164,0.8488],[165,0.8617],[166,0.8703],[167,0.8747],[168,0.8732],[169,0.8652],[170,0.852],[171,0.837],[172,0.8237],[174,0.8048],[177,0.7827],[179,0.7711],[182,0.7605],[184,0.7571],[186,0.7579],[195,0.7725],[198,0.7716],[200,0.7671],[201,0.7625],[202,0.7561],[203,0.7473],[204,0.7347],[205,0.7161],[206,0.69],[207,0.6496],[208,0.6146],[209,0.6122],[211,0.6521],[213,0.6787],[214,0.6886],[216,0.7011],[218,0.7103],[220,0.7152],[227,0.7177],[230,0.7169],[231,0.7144],[232,0.7144],[233,0.7064],[234,0.7064],[235,0.6892],[236,0.6769],[237,0.661],[239,0.6244],[242,0.5647],[243,0.5466],[244,0.5299],[245,0.5147],[248,0.4744],[249,0.462],[250,0.4509],[251,0.4412],[252,0.4334],[253,0.428],[254,0.4251],[255,0.4241],[256,0.4248],[267,0.451],[272,0.4669],[273,0.4669],[274,0.4737],[275,0.4772],[278,0.4772],[279,0.4902],[282,0.4979],[284,0.5003],[287,0.5002],[291,0.4961],[293,0.4916],[295,0.4845],[299,0.4649],[301,0.4586],[302,0.4586],[303,0.4546],[305,0.4546],[306,0]]
       }
       spine_01_ctrl.Rotation.Z {
        type=MovieSceneScriptingFloatChannel
        default=0
        thinned_from=307
        keys [[0,0.0979],[7,0.0979],[8,0.1153],[9,0.1274],[10,0.1436],[11,0.1651],[12,0.1974],[13,0.2418],[14,0.2889],[15,0.3253],[16,0.3549],[17,0.3867],[18,0.4193],[19,0.4847],[20,0.6061],[21,0.8039],[22,1.0515],[23,1.2898],[24,1.469],[25,1.6108],[26,1.7276],[27,1.8149],[28,1.8725],[29,1.9099],[30,1.9287],[31,1.9349],[32,1.9349],[33,1.9303],[34,1.9295],[35,1.9369],[36,1.9369],[37,1.9579],[38,1.9579],[39,1.9788],[40,1.9899],[41,1.9899],[42,2.0071],[44,2.0071],[45,2.0206],[50,2.0206],[51,2.0131],[53,2.0131],[54,1.9933],[55,1.9933],[56,1.9714],[57,1.9579],[58,1.9402],[59,1.9173],[60,1.8881],[61,1.8527],[64,1.7214],[65,1.6557],[66,1.5442],[67,1.3373],[68,1.0449],[69,0.763],[70,0.5412],[71,0.3665],[72,0.2493],[73,0.2186],[74,0.2449],[75,0.2901],[76,0.3209],[77,0.3337],[78,0.3271],[79,0.2943],[80,0.2301],[81,0.1621],[82,0.1047],[83,0.0562],[84,0.0151],[85,-0.023],[86,-0.0553],[87,-0.0847],[88,-0.1096],[89,-0.1331],[90,-0.1537],[92,-0.189],[94,-0.2184],[95,-0.2316],[96,-0.2463],[97,-0.2625],[98,-0.2773],[99,-0.2913],[100,-0.3068],[101,-0.3245],[104,-0.3841],[105,-0.405],[106,-0.4281],[107,-0.4524],[108,-0.4755],[109,-0.493],[110,-0.5007],[111,-0.4995],[112,-0.494],[113,-0.494],[114,-0.4741],[115,-0.461],[116,-0.4437],[118,-0.4189],[119,-0.4189],[120,-0.4041],[122,-0.4041],[123,-0.3942],[125,-0.3942],[126,-0.4141],[127,-0.4265],[128,-0.4438],[129,-0.4686],[130,-0.4884],[131,-0.5005],[132,-0.5005],[133,-0.5181],[140,-0.5181],[141,-0.4974],[142,-0.4812],[143,-0.4605],[144,-0.4369],[145,-0.4074],[146,-0.3764],[147,-0.3411],[148,-0.3026],[149,-0.2582],[150,-0.2108],[151,-0.1398],[152,-0.0302],[153,0.1061],[154,0.2986],[155,0.5769],[156,0.9679],[157,1.4151],[158,1.8119],[159,2.1288],[160,2.333],[161,2.4868],[162,2.6022],[163,2.678],[164,2.7241],[165,2.7488],[166,2.7517],[167,2.7223],[168,2.5822],[169,2.337],[170,2.0268],[171,1.6771],[172,1.3158],[173,0.9804],[174,0.668],[175,0.379],[176,0.0934],[177,-0.1904],[178,-0.4538],[179,-0.6848],[180,-0.8667],[181,-1.0122],[182,-1.1257],[183,-1.2091],[184,-1.2434],[185,-1.2177],[186,-1.1428],[187,-1.0294],[188,-0.8988],[189,-0.7598],[190,-0.6249],[191,-0.4987],[192,-0.381],[193,-0.2761],[194,-0.1863],[195,-0.1111],[196,-0.0566],[197,-0.0146],[198,0.0102],[199,0.0202],[200,0.0104],[201,-0.0365],[202,-0.123],[203,-0.2565],[204,-0.4419],[205,-0.6793],[206,-0.9661],[207,-1.2999],[208,-1.6684],[209,-2.0569],[211,-2.8505],[212,-3.2365],[213,-3.5951],[214,-3.917],[215,-4.1817],[216,-4.3914],[217,-4.5484],[218,-4.671],[219,-4.7684],[220,-4.8532],[221,-4.9254],[222,-4.985],[224,-5.0744],[227,-5.1566],[229,-5.1773],[230,-5.1773],[231,-5.1932],[232,-5.1932],[233,-5.2059],[237,-5.2136],[238,-5.1959],[239,-5.1591],[240,-5.0936],[241,-4.9915],[242,-4.8367],[243,-4.6194],[244,-4.3256],[245,-3.9623],[246,-3.5547],[247,-3.134],[248,-2.7252],[249,-2.3082],[250,-1.8816],[251,-1.4801],[252,-1.1387],[253,-0.8475],[254,-0.5791],[255,-0.3348],[256,-0.1131],[257,0.0809],[258,0.2486],[259,0.3927],[260,0.5198],[261,0.6304],[262,0.7261],[263,0.8086],[264,0.878],[265,0.9374],[266,0.9885],[267,1.0314],[268,1.0677],[269,1.0974],[270,1.1205],[271,1.1387],[272,1.1535],[273,1.1535],[274,1.17],[275,1.174],[278,1.174],[279,1.1585],[280,1.1405],[281,1.1123],[282,1.0763],[283,1.0288],[284,0.9748],[285,0.9145],[286,0.849],[289,0.6449],[290,0.5794],[291,0.5175],[292,0.4529],[293,0.3862],[294,0.3215],[295,0.2647],[296,0.2196],[297,0.1883],[298,0.1628],[299,0.1433],[300,0.1276],[301,0.1158],[302,0.1158],[303,0.1021],[305,0.1021],[306,0]]
       }
       spine_01_ctrl.Scale.X {type=MovieSceneScriptingFloatChannel, default=1, const [1,307]}
       spine_01_ctrl.Scale.Y {type=MovieSceneScriptingFloatChannel, default=1, const [1,307]}
       spine_01_ctrl.Scale.Z {type=MovieSceneScriptingFloatChannel, default=1, const [1,307]}
       spine_02_ctrl.Location.X {
        type=MovieSceneScriptingFloatChannel
        default=0
        thinned_from=307
        keys [[0,-0.0098],[6,-0.0098],[7,-0.0106],[8,-0.0106],[9,-0.0114],[10,-0.0114],[13,-0.0138],[14,-0.0138],[16,-0.0154],[20,-0.0152],[21,-0.0181],[29,-0.0181],[30,-0.0152],[33,-0.0152],[34,-0.0149],[50,-0.0149],[51,-0.0142],[54,-0.0142],[55,-0.0161],[58,-0.0161],[59,-0.0179],[61,-0.0179],[62,-0.0198],[63,-0.0198],[64,-0.0217],[65,-0.0254],[67,-0.0289],[68,-0.0322],[69,-0.0322],[70,-0.0355],[74,-0.0355],[80,-0.0157],[81,-0.0157],[82,-0.0124],[83,-0.0108],[84,-0.0095],[88,-0.0068],[90,-0.0068],[91,-0.0061],[97,-0.0061],[98,-0.0068],[99,-0.0077],[114,-0.0077],[115,-0.0048],[118,-0.0052],[123,-0.007],[125,-0.0073],[128,-0.0073],[130,-0.0068],[131,-0.0061],[132,-0.0055],[134,-0.0055],[135,-0.005],[143,-0.005],[144,-0.0055],[146,-0.0055],[147,-0.0059],[148,-0.0059],[149,-0.0083],[150,-0.0083],[151,-0.0106],[152,-0.0106],[156,-0.02],[158,-0.0293],[160,-0.034],[161,-0.0387],[162,-0.041],[163,-0.0428],[164,-0.0442],[165,-0.0453],[166,-0.0459],[167,-0.0462],[168,-0.0459],[169,-0.0449],[172,-0.0402],[174,-0.0378],[178,-0.034],[179,-0.0333],[194,-0.0333],[195,-0.0345],[203,-0.0345],[204,-0.0293],[205,-0.0293],[206,-0.0241],[207,-0.0241],[208,-0.0189],[210,-0.0189],[211,-0.0209],[213,-0.0209],[214,-0.0233],[226,-0.0233],[227,-0.023],[232,-0.023],[233,-0.0214],[235,-0.0214],[238,-0.0164],[239,-0.0164],[242,-0.0115],[243,-0.0107],[245,-0.0107],[246,-0.0066],[258,-0.0066],[259,-0.0085],[260,-0.0085],[261,-0.0102],[266,-0.0102],[267,-0.0118],[272,-0.0118],[273,-0.0134],[290,-0.0134],[291,-0.0141],[292,-0.0141],[293,-0.0127],[296,-0.0127],[297,-0.0113],[300,-0.0113],[301,-0.0098],[305,-0.0098],[306,0]]
       }
       spine_02_ctrl.Location.Y {
        type=MovieSceneScriptingFloatChannel
        default=0
        thinned_from=307
        keys [[0,0.2798],[25,0.2798],[26,0.2811],[50,0.2811],[69,0.28],[72,0.2744],[73,0.2744],[75,0.2706],[76,0.2707],[77,0.2688],[79,0.2689],[80,0.267],[114,0.2665],[115,0.2675],[141,0.2696],[147,0.271],[148,0.2718],[149,0.2718],[150,0.2726],[155,0.2805],[158,0.2829],[160,0.2829],[161,0.282],[162,0.282],[194,0.2778],[195,0.2764],[197,0.2764],[198,0.2783],[200,0.2783],[201,0.2802],[202,0.2802],[203,0.2821],[208,0.2822],[209,0.2802],[218,0.2802],[239,0.2807],[250,0.2795],[251,0.2806],[290,0.2796],[291,0.2806],[305,0.2797],[306,0]]
       }
       spine_02_ctrl.Location.Z {
        type=MovieSceneScriptingFloatChannel
        default=0
        thinned_from=307
        keys [[0,0.0579],[24,0.058],[25,0.0592],[50,0.0593],[56,0.0594],[69,0.0588],[70,0.0576],[71,0.0574],[74,0.0538],[75,0.0537],[77,0.0514],[78,0.0514],[79,0.0503],[114,0.05],[115,0.0506],[125,0.0517],[140,0.0518],[143,0.0521],[144,0.0523],[145,0.0523],[147,0.0526],[148,0.0531],[149,0.0531],[150,0.0536],[155,0.0584],[158,0.0598],[162,0.0597],[179,0.0575],[194,0.0575],[195,0.0565],[197,0.0565],[198,0.0576],[200,0.0576],[201,0.0587],[202,0.0587],[203,0.0599],[207,0.0599],[208,0.0589],[215,0.0588],[216,0.0594],[233,0.0594],[235,0.0595],[240,0.0589],[242,0.0585],[243,0.0584],[244,0.0581],[250,0.0581],[251,0.0585],[258,0.0585],[260,0.0581],[262,0.0581],[263,0.0578],[266,0.0578],[267,0.0576],[290,0.0576],[291,0.0583],[299,0.0581],[305,0.0578],[306,0]]
       }
       spine_02_ctrl.Rotation.X {
        type=MovieSceneScriptingFloatChannel
        default=0
        thinned_from=307
        keys [[0,5.4346],[68,5.4369],[70,5.4098],[73,5.3099],[78,5.2058],[114,5.1755],[147,5.2536],[150,5.293],[155,5.4515],[157,5.4917],[159,5.5048],[193,5.361],[198,5.3868],[203,5.4661],[206,5.4724],[232,5.3728],[305,5.4353],[306,0]]
       }
       spine_02_ctrl.Rotation.Y {
        type=MovieSceneScriptingFloatChannel
        default=0
        thinned_from=307
        keys [[0,0.141],[6,0.141],[7,0.1499],[8,0.1499],[9,0.1622],[10,0.1699],[11,0.1795],[12,0.1904],[13,0.1997],[14,0.2082],[15,0.2155],[16,0.221],[17,0.2243],[18,0.2255],[19,0.2244],[20,0.2205],[21,0.2123],[22,0.199],[23,0.1845],[27,0.1282],[28,0.1155],[30,0.0939],[31,0.0939],[32,0.0763],[33,0.0763],[34,0.0621],[35,0.0621],[36,0.0527],[37,0.0527],[38,0.0455],[39,0.0455],[40,0.0429],[41,0.0429],[42,0.0401],[50,0.0401],[51,0.042],[52,0.042],[53,0.0499],[54,0.0499],[55,0.0621],[56,0.0621],[57,0.0805],[58,0.0921],[59,0.1055],[60,0.1212],[61,0.1385],[62,0.1587],[63,0.1861],[64,0.2284],[65,0.2821],[66,0.3424],[67,0.4067],[68,0.4693],[69,0.5402],[70,0.5942],[71,0.6197],[72,0.6309],[73,0.6271],[74,0.6091],[75,0.5772],[76,0.5316],[77,0.4695],[78,0.3998],[79,0.3324],[80,0.2784],[81,0.233],[82,0.1915],[83,0.1587],[84,0.1368],[85,0.1231],[86,0.1125],[87,0.1045],[88,0.099],[89,0.0952],[90,0.0931],[91,0.0924],[92,0.0922],[95,0.0935],[96,0.0947],[97,0.0977],[98,0.1082],[99,0.1302],[100,0.1488],[101,0.1568],[102,0.1598],[103,0.1609],[104,0.1606],[105,0.1591],[106,0.1558],[107,0.151],[108,0.1413],[110,0.1147],[111,0.1029],[112,0.1029],[113,0.0891],[114,0.0891],[115,0.087],[116,0.0867],[117,0.0877],[118,0.0911],[119,0.0972],[120,0.1041],[121,0.1116],[122,0.1195],[123,0.1256],[124,0.1296],[125,0.1324],[126,0.1324],[127,0.1349],[128,0.1355],[129,0.1338],[130,0.1297],[131,0.1208],[132,0.1097],[134,0.1097],[135,0.0997],[139,0.0997],[140,0.0952],[141,0.0945],[142,0.0946],[143,0.0942],[144,0.0945],[145,0.0958],[146,0.0983],[147,0.1029],[148,0.1104],[149,0.1216],[150,0.1364],[151,0.1546],[152,0.1752],[153,0.1994],[154,0.226],[155,0.2552],[156,0.2874],[157,0.3223],[158,0.3602],[159,0.3996],[160,0.445],[161,0.4931],[162,0.5374],[163,0.5715],[164,0.5958],[165,0.6146],[166,0.6281],[167,0.6368],[168,0.6432],[169,0.6457],[171,0.6422],[172,0.6437],[176,0.6656],[179,0.6843],[181,0.6918],[183,0.6959],[184,0.6961],[185,0.6944],[187,0.6866],[191,0.6656],[195,0.6509],[200,0.6356],[202,0.6271],[203,0.6217],[204,0.6137],[205,0.5997],[206,0.5768],[207,0.5345],[208,0.503],[209,0.523],[210,0.5793],[211,0.6342],[212,0.6785],[213,0.7212],[214,0.7561],[215,0.7828],[216,0.8048],[217,0.8224],[218,0.8362],[219,0.8467],[220,0.8554],[222,0.8668],[225,0.8751],[230,0.8775],[231,0.8748],[232,0.8748],[234,0.8529],[235,0.8378],[236,0.8186],[237,0.7942],[238,0.7649],[239,0.7331],[240,0.6971],[242,0.6205],[243,0.5783],[244,0.5341],[247,0.3948],[249,0.3046],[250,0.2608],[251,0.2209],[252,0.1878],[253,0.1613],[254,0.1402],[255,0.1235],[256,0.111],[257,0.1016],[258,0.0947],[259,0.0894],[260,0.0851],[261,0.0817],[262,0.0794],[263,0.0781],[264,0.0779],[265,0.0782],[266,0.0794],[267,0.0812],[268,0.0836],[269,0.0866],[270,0.0902],[271,0.094],[272,0.094],[273,0.103],[275,0.103],[276,0.1181],[278,0.1181],[279,0.1342],[281,0.1456],[283,0.1564],[284,0.1611],[287,0.1729],[289,0.1785],[290,0.1809],[292,0.1838],[293,0.1837],[294,0.1827],[295,0.1802],[296,0.176],[298,0.1634],[299,0.1574],[300,0.1526],[301,0.1494],[302,0.1494],[303,0.1439],[304,0.1421],[305,0.1421],[306,0]]
       }
       spine_02_ctrl.Rotation.Z {
        type=MovieSceneScriptingFloatChannel
        default=0
        thinned_from=307
        keys [[0,0.0161],[6,0.0161],[7,0.0233],[8,0.0233],[9,0.0405],[10,0.0533],[11,0.0721],[12,0.0985],[13,0.1351],[14,0.1749],[15,0.2045],[17,0.2529],[18,0.2788],[19,0.3308],[20,0.4347],[21,0.5959],[22,0.7989],[23,0.9968],[24,1.1479],[25,1.2627],[26,1.3567],[27,1.4301],[28,1.4774],[29,1.509],[30,1.5251],[33,1.5259],[34,1.5214],[35,1.5214],[36,1.5386],[37,1.5386],[38,1.5556],[39,1.5556],[40,1.5725],[41,1.5725],[42,1.586],[50,1.586],[51,1.5912],[52,1.5912],[53,1.5819],[54,1.5819],[55,1.5667],[56,1.5667],[57,1.5447],[58,1.5303],[59,1.511],[60,1.4875],[61,1.4575],[64,1.3484],[65,1.2939],[66,1.2009],[67,1.0287],[68,0.7864],[69,0.554],[70,0.3703],[71,0.2274],[72,0.133],[73,0.108],[74,0.1316],[75,0.1726],[76,0.1997],[77,0.211],[78,0.2082],[79,0.1823],[80,0.1287],[81,0.0727],[82,0.0248],[83,-0.015],[84,-0.0494],[85,-0.0804],[86,-0.1068],[87,-0.1308],[88,-0.1525],[89,-0.1718],[90,-0.1888],[92,-0.2181],[94,-0.2425],[95,-0.2536],[99,-0.3032],[100,-0.3172],[102,-0.3472],[103,-0.3635],[106,-0.4172],[107,-0.4382],[108,-0.4574],[109,-0.4714],[110,-0.4773],[112,-0.4767],[113,-0.4646],[114,-0.4646],[115,-0.4435],[116,-0.4308],[117,-0.4192],[118,-0.4107],[119,-0.4039],[120,-0.3984],[121,-0.3939],[122,-0.3913],[123,-0.3903],[124,-0.3923],[125,-0.3972],[126,-0.3972],[127,-0.4169],[128,-0.4314],[129,-0.4509],[130,-0.4677],[132,-0.4873],[134,-0.4873],[135,-0.4991],[139,-0.4991],[140,-0.4866],[141,-0.4766],[142,-0.4618],[143,-0.4445],[144,-0.4247],[146,-0.3755],[147,-0.3461],[148,-0.3145],[149,-0.2781],[150,-0.2383],[151,-0.1801],[152,-0.0888],[153,0.0219],[154,0.183],[155,0.414],[156,0.7371],[157,1.1092],[158,1.4393],[159,1.703],[160,1.8743],[161,2.0011],[162,2.0961],[163,2.1597],[164,2.198],[165,2.2189],[166,2.2215],[167,2.1967],[168,2.0804],[169,1.8767],[170,1.6184],[171,1.3279],[172,1.0274],[173,0.7484],[174,0.4886],[175,0.2484],[176,0.0109],[177,-0.2246],[178,-0.4437],[179,-0.6359],[180,-0.7869],[181,-0.9082],[182,-1.0019],[183,-1.0708],[184,-1.1001],[185,-1.0783],[186,-1.0169],[187,-0.9222],[188,-0.8129],[189,-0.6973],[190,-0.5866],[191,-0.4808],[192,-0.3821],[193,-0.295],[194,-0.2203],[195,-0.1594],[196,-0.112],[197,-0.0789],[198,-0.0582],[199,-0.0497],[200,-0.0577],[201,-0.0964],[202,-0.168],[203,-0.2806],[204,-0.4343],[205,-0.6308],[206,-0.8701],[207,-1.1475],[208,-1.4541],[209,-1.7777],[211,-2.4352],[212,-2.7573],[213,-3.0565],[214,-3.3248],[215,-3.5433],[216,-3.7196],[217,-3.85],[218,-3.9497],[219,-4.0341],[220,-4.1031],[222,-4.2103],[224,-4.2868],[227,-4.355],[228,-4.355],[229,-4.3721],[230,-4.3721],[231,-4.3854],[232,-4.3854],[236,-4.4046],[237,-4.401],[238,-4.3874],[239,-4.3565],[240,-4.3023],[241,-4.2162],[242,-4.0873],[243,-3.9059],[244,-3.663],[245,-3.3602],[246,-3.0207],[247,-2.6712],[248,-2.3308],[249,-1.9845],[250,-1.6293],[251,-1.2961],[252,-1.0111],[253,-0.7701],[254,-0.5474],[255,-0.3437],[256,-0.1601],[257,0.0025],[258,0.141],[259,0.2627],[260,0.3672],[261,0.4581],[262,0.5378],[263,0.6065],[264,0.6642],[265,0.7137],[266,0.7576],[267,0.7932],[268,0.8233],[269,0.8479],[270,0.8669],[271,0.8832],[272,0.8832],[273,0.902],[275,0.902],[276,0.9117],[278,0.9117],[280,0.8829],[281,0.8591],[282,0.8289],[283,0.7902],[284,0.7452],[285,0.6937],[286,0.6401],[287,0.5844],[288,0.5266],[289,0.4709],[291,0.3634],[292,0.3113],[294,0.201],[295,0.1552],[296,0.1186],[297,0.0913],[298,0.0702],[300,0.0401],[301,0.031],[302,0.031],[303,0.019],[304,0.016],[305,0.016],[306,0]]
       }
       spine_02_ctrl.Scale.X {type=MovieSceneScriptingFloatChannel, default=1, const [1,307]}
       spine_02_ctrl.Scale.Y {type=MovieSceneScriptingFloatChannel, default=1, const [1,307]}
       spine_02_ctrl.Scale.Z {type=MovieSceneScriptingFloatChannel, default=1, const [1,307]}
       spine_03_ctrl.Location.X {
        type=MovieSceneScriptingFloatChannel
        default=0
        thinned_from=307
        keys [[0,0.007],[7,0.007],[8,0.0032],[11,0.0032],[12,-0.0006],[14,-0.0006],[15,-0.0044],[16,-0.0044],[17,-0.0039],[18,-0.0036],[19,-0.0029],[20,-0.0017],[21,0.0002],[24,0.0077],[26,0.0124],[28,0.0166],[30,0.0201],[32,0.0229],[33,0.024],[34,0.0247],[36,0.0247],[37,0.0287],[50,0.0287],[51,0.0282],[53,0.0282],[54,0.0254],[56,0.0254],[57,0.0226],[58,0.0226],[63,0.0086],[64,0.003],[66,-0.0138],[67,-0.0216],[68,-0.0306],[69,-0.0414],[70,-0.0495],[71,-0.0532],[72,-0.0559],[73,-0.0559],[74,-0.0541],[75,-0.0504],[76,-0.045],[77,-0.0369],[79,-0.0189],[80,-0.0117],[81,-0.0054],[82,0],[83,0.0035],[84,0.0035],[85,0.0134],[98,0.0134],[99,0.0077],[100,0.0046],[101,0.0046],[102,0.003],[104,0.003],[105,0.0046],[107,0.0046],[113,0.0139],[121,0.0139],[122,0.0065],[130,0.0065],[131,0.0096],[132,0.0096],[133,0.0112],[136,0.0112],[137,0.0129],[146,0.0129],[147,0.0119],[148,0.0119],[149,0.008],[150,0.008],[152,0.0002],[153,0.0002],[157,-0.0153],[158,-0.0231],[159,-0.027],[160,-0.0347],[161,-0.0386],[162,-0.0464],[163,-0.05],[164,-0.0533],[165,-0.0558],[166,-0.0577],[167,-0.0589],[168,-0.0598],[169,-0.0603],[171,-0.0602],[172,-0.0607],[176,-0.0644],[179,-0.0676],[181,-0.0676],[182,-0.0701],[185,-0.0701],[186,-0.0676],[189,-0.0676],[190,-0.0652],[192,-0.0652],[193,-0.0627],[194,-0.0627],[195,-0.0632],[196,-0.0619],[198,-0.0619],[199,-0.0605],[201,-0.0605],[202,-0.0592],[203,-0.0592],[205,-0.0566],[206,-0.0526],[207,-0.0474],[208,-0.0434],[209,-0.0461],[210,-0.054],[211,-0.0622],[216,-0.0876],[217,-0.0876],[218,-0.0927],[221,-0.0927],[222,-0.0977],[231,-0.0977],[232,-0.0953],[233,-0.0953],[234,-0.0929],[235,-0.0929],[236,-0.0881],[237,-0.0857],[238,-0.0809],[239,-0.0785],[240,-0.0737],[241,-0.0665],[243,-0.0568],[248,-0.0228],[249,-0.0183],[251,-0.0047],[253,0.0043],[256,0.0111],[257,0.0111],[258,0.0134],[259,0.0142],[260,0.0142],[261,0.0158],[268,0.0158],[269,0.0142],[271,0.0142],[272,0.0126],[274,0.0126],[275,0.0114],[276,0.0099],[277,0.0099],[278,0.0084],[279,0.0084],[280,0.007],[281,0.007],[282,0.0055],[283,0.0055],[284,0.004],[286,0.004],[287,0.0025],[289,0.0025],[290,0.001],[291,0.0005],[298,0.0005],[299,0.007],[305,0.007],[306,0]]
       }
       spine_03_ctrl.Location.Y {
        type=MovieSceneScriptingFloatChannel
        default=0
        thinned_from=307
        keys [[0,0.5708],[44,0.5715],[68,0.5697],[70,0.5663],[73,0.5534],[77,0.5417],[79,0.5378],[83,0.5346],[84,0.5346],[85,0.5328],[97,0.5328],[98,0.5346],[120,0.5369],[121,0.5431],[140,0.5427],[141,0.5438],[143,0.5438],[147,0.5465],[148,0.5487],[149,0.5487],[150,0.5509],[155,0.5726],[158,0.5791],[160,0.5791],[161,0.5769],[179,0.5655],[180,0.564],[181,0.564],[182,0.5625],[187,0.5626],[188,0.5611],[195,0.5615],[198,0.5646],[201,0.5709],[202,0.5741],[205,0.5773],[208,0.5752],[210,0.572],[245,0.5695],[259,0.5726],[266,0.5704],[290,0.5723],[298,0.573],[299,0.5708],[305,0.5708],[306,0]]
       }
       spine_03_ctrl.Location.Z {
        type=MovieSceneScriptingFloatChannel
        default=0
        thinned_from=307
        keys [[0,0.092],[21,0.0925],[37,0.0938],[38,0.094],[66,0.0933],[68,0.0931],[69,0.0929],[70,0.0923],[71,0.0912],[72,0.0897],[74,0.0874],[77,0.0845],[79,0.0831],[82,0.0823],[84,0.0822],[85,0.0818],[97,0.0818],[98,0.0822],[99,0.0821],[120,0.0829],[121,0.0845],[139,0.0844],[141,0.0847],[143,0.0847],[144,0.085],[145,0.085],[147,0.0854],[148,0.086],[149,0.086],[150,0.0866],[155,0.0926],[158,0.0944],[160,0.0944],[161,0.0941],[162,0.0941],[179,0.0915],[180,0.0911],[184,0.0909],[185,0.0906],[187,0.0906],[189,0.0904],[190,0.0902],[195,0.0902],[198,0.0911],[201,0.0928],[202,0.0936],[205,0.0944],[207,0.094],[209,0.0932],[215,0.0932],[217,0.0934],[223,0.0934],[237,0.0935],[245,0.0922],[257,0.0926],[259,0.0925],[260,0.0922],[262,0.0922],[263,0.0919],[267,0.0918],[285,0.0921],[298,0.0926],[299,0.092],[305,0.092],[306,0]]
       }
       spine_03_ctrl.Rotation.X {
        type=MovieSceneScriptingFloatChannel
        default=0
        thinned_from=307
        keys [[0,8.2134],[68,8.1965],[69,8.1839],[70,8.1534],[72,8.0152],[75,7.8626],[79,7.7212],[115,7.7147],[118,7.7316],[123,7.7843],[145,7.8343],[148,7.8734],[150,7.9328],[152,8.0498],[154,8.1869],[156,8.2815],[158,8.3304],[175,8.1887],[183,8.0681],[195,8.073],[198,8.1169],[203,8.2764],[205,8.2927],[207,8.2725],[215,8.0692],[232,7.9962],[243,8.0938],[251,8.2415],[305,8.2162],[306,0]]
       }
       spine_03_ctrl.Rotation.Y {
        type=MovieSceneScriptingFloatChannel
        default=0
        thinned_from=307
        keys [[0,-0.1622],[4,-0.1622],[5,-0.1563],[7,-0.1563],[8,-0.1351],[9,-0.1227],[10,-0.1089],[11,-0.0917],[12,-0.0735],[13,-0.0594],[14,-0.0481],[15,-0.039],[16,-0.034],[17,-0.0345],[18,-0.0432],[19,-0.0611],[20,-0.0928],[21,-0.1415],[22,-0.204],[23,-0.2687],[24,-0.3242],[25,-0.3743],[26,-0.4207],[27,-0.4628],[28,-0.4993],[29,-0.5307],[30,-0.5582],[31,-0.5809],[32,-0.6007],[33,-0.6179],[35,-0.6464],[36,-0.6464],[37,-0.6668],[38,-0.6749],[39,-0.6749],[40,-0.6856],[41,-0.6856],[42,-0.6921],[43,-0.6921],[44,-0.696],[50,-0.6973],[51,-0.6888],[52,-0.6888],[53,-0.6725],[54,-0.6603],[55,-0.6454],[56,-0.6272],[57,-0.6055],[58,-0.5798],[59,-0.5501],[60,-0.515],[61,-0.4753],[62,-0.4304],[63,-0.3689],[64,-0.2785],[65,-0.1625],[66,-0.0264],[67,0.1291],[68,0.2929],[69,0.4735],[70,0.6164],[71,0.6985],[72,0.7469],[73,0.7552],[74,0.7299],[75,0.6743],[76,0.587],[77,0.4674],[78,0.3323],[79,0.2038],[80,0.102],[81,0.0173],[82,-0.0598],[83,-0.1219],[84,-0.1616],[85,-0.1856],[86,-0.2038],[87,-0.2172],[88,-0.2258],[89,-0.2307],[90,-0.2329],[91,-0.233],[92,-0.2314],[94,-0.2272],[95,-0.2246],[96,-0.2209],[97,-0.2131],[98,-0.1903],[99,-0.1446],[100,-0.1052],[101,-0.087],[102,-0.0791],[103,-0.0749],[104,-0.0731],[105,-0.0744],[106,-0.0782],[107,-0.0862],[108,-0.1026],[109,-0.1286],[110,-0.1548],[111,-0.1783],[112,-0.1975],[113,-0.2069],[114,-0.2113],[115,-0.2143],[116,-0.2166],[117,-0.2152],[118,-0.2152],[119,-0.1983],[120,-0.1847],[121,-0.1702],[122,-0.1548],[123,-0.1423],[124,-0.1423],[125,-0.1273],[126,-0.1273],[127,-0.1204],[128,-0.1171],[129,-0.1178],[130,-0.1242],[131,-0.1411],[132,-0.1622],[133,-0.1734],[136,-0.1734],[137,-0.1871],[139,-0.1871],[140,-0.1916],[141,-0.1936],[142,-0.1968],[144,-0.2002],[145,-0.2005],[146,-0.1995],[147,-0.1934],[148,-0.1818],[149,-0.1636],[150,-0.1386],[151,-0.1093],[152,-0.0794],[153,-0.0451],[154,-0.0124],[155,0.0155],[156,0.0367],[157,0.0569],[158,0.0887],[159,0.1327],[160,0.2018],[161,0.2829],[162,0.3604],[163,0.4213],[164,0.4658],[165,0.5017],[166,0.5292],[167,0.5507],[168,0.5795],[170,0.6429],[171,0.6778],[172,0.7209],[173,0.768],[174,0.8135],[177,0.9444],[178,0.986],[179,1.0224],[180,1.0506],[181,1.0729],[182,1.09],[183,1.1021],[184,1.1063],[185,1.0998],[186,1.0852],[187,1.0637],[190,0.9873],[191,0.9634],[192,0.9417],[193,0.9226],[194,0.9061],[195,0.8921],[196,0.8804],[198,0.862],[200,0.8484],[201,0.8457],[202,0.8462],[204,0.8556],[205,0.8538],[206,0.8398],[207,0.7914],[208,0.7683],[209,0.8517],[210,1.009],[211,1.1628],[212,1.2951],[213,1.4201],[214,1.5259],[215,1.6086],[216,1.6754],[217,1.7279],[218,1.7693],[219,1.8015],[220,1.827],[221,1.8478],[222,1.8644],[224,1.8872],[226,1.8998],[228,1.9069],[230,1.9065],[232,1.895],[233,1.8817],[234,1.8617],[235,1.8328],[236,1.795],[237,1.7461],[238,1.6864],[239,1.6159],[240,1.5372],[241,1.4505],[242,1.3542],[243,1.2446],[244,1.1234],[245,0.9906],[247,0.7107],[249,0.437],[250,0.3011],[251,0.1759],[252,0.0704],[253,-0.0156],[254,-0.0886],[255,-0.1495],[256,-0.1996],[257,-0.2404],[258,-0.273],[259,-0.2997],[260,-0.3224],[261,-0.3414],[262,-0.3566],[263,-0.368],[264,-0.3762],[265,-0.3819],[266,-0.3852],[267,-0.3863],[268,-0.3853],[269,-0.3824],[270,-0.3779],[271,-0.3719],[272,-0.3648],[273,-0.3648],[274,-0.3471],[276,-0.3277],[277,-0.3277],[278,-0.306],[279,-0.2931],[280,-0.2793],[281,-0.265],[282,-0.2498],[284,-0.2185],[285,-0.204],[287,-0.1734],[289,-0.1468],[290,-0.1353],[291,-0.1244],[292,-0.1153],[293,-0.1077],[294,-0.1028],[295,-0.1015],[296,-0.1052],[297,-0.1139],[298,-0.1245],[299,-0.1345],[300,-0.1421],[301,-0.1476],[302,-0.1476],[303,-0.1567],[305,-0.1567],[306,0]]
       }
       spine_03_ctrl.Rotation.Z {
        type=MovieSceneScriptingFloatChannel
        default=0
        thinned_from=307
        keys [[0,-0.0032],[4,-0.0032],[5,-0.002],[7,-0.002],[8,0.0144],[9,0.0269],[10,0.0434],[11,0.0673],[12,0.1007],[13,0.1489],[14,0.2],[15,0.239],[16,0.2706],[17,0.3037],[18,0.3383],[19,0.4093],[20,0.5426],[21,0.7589],[22,1.0341],[23,1.2955],[24,1.4948],[25,1.6527],[26,1.7812],[27,1.8786],[28,1.9449],[29,1.9853],[30,2.0066],[31,2.0141],[33,2.0081],[34,2.0087],[35,2.0178],[36,2.0178],[37,2.0403],[38,2.0536],[39,2.0536],[40,2.0758],[41,2.0758],[42,2.0957],[43,2.0957],[44,2.1089],[50,2.1112],[51,2.1018],[52,2.1018],[55,2.0675],[57,2.0361],[58,2.0154],[59,1.9888],[60,1.9547],[61,1.913],[64,1.7578],[65,1.6797],[66,1.549],[67,1.3136],[68,0.9843],[69,0.6652],[70,0.4159],[71,0.2231],[72,0.0959],[73,0.0656],[74,0.0997],[75,0.1576],[76,0.2004],[77,0.2238],[78,0.2262],[79,0.1988],[80,0.1342],[81,0.0652],[82,0.0067],[83,-0.0434],[84,-0.0877],[85,-0.126],[86,-0.1612],[87,-0.1933],[88,-0.2206],[89,-0.2449],[90,-0.2675],[91,-0.2886],[92,-0.3066],[97,-0.3883],[99,-0.4257],[100,-0.4453],[102,-0.4872],[103,-0.5101],[104,-0.5318],[106,-0.5804],[107,-0.6072],[108,-0.6306],[109,-0.6482],[110,-0.6561],[111,-0.6533],[112,-0.6462],[113,-0.6362],[114,-0.6231],[115,-0.6079],[116,-0.5902],[117,-0.5753],[118,-0.5753],[119,-0.5554],[120,-0.549],[121,-0.544],[122,-0.5418],[124,-0.5408],[125,-0.5507],[126,-0.5507],[127,-0.5767],[128,-0.5972],[129,-0.623],[130,-0.6446],[132,-0.6664],[133,-0.6737],[136,-0.6737],[137,-0.681],[139,-0.681],[142,-0.6347],[143,-0.6117],[144,-0.5848],[145,-0.5541],[146,-0.5196],[147,-0.4795],[148,-0.4382],[149,-0.3907],[150,-0.3386],[151,-0.2641],[152,-0.1448],[153,-0.0016],[154,0.2091],[155,0.5127],[156,0.9399],[157,1.4296],[158,1.863],[159,2.2112],[160,2.4323],[161,2.597],[162,2.7186],[163,2.7994],[164,2.8481],[165,2.8735],[166,2.8757],[167,2.8424],[168,2.6868],[169,2.4152],[170,2.072],[171,1.6854],[172,1.2854],[173,0.9138],[174,0.5677],[175,0.2474],[176,-0.0692],[177,-0.3838],[178,-0.6761],[179,-0.9321],[180,-1.1339],[181,-1.2955],[182,-1.4205],[183,-1.5126],[184,-1.5522],[185,-1.5223],[186,-1.4406],[187,-1.3144],[188,-1.1674],[189,-1.0131],[190,-0.8649],[191,-0.7235],[192,-0.5918],[193,-0.4753],[194,-0.3753],[195,-0.2946],[196,-0.2314],[197,-0.186],[198,-0.1583],[199,-0.147],[200,-0.1588],[201,-0.2077],[202,-0.3044],[203,-0.4519],[204,-0.6554],[205,-0.9176],[206,-1.2314],[207,-1.5951],[208,-1.999],[209,-2.4316],[211,-3.323],[212,-3.7549],[213,-4.1576],[214,-4.5181],[215,-4.8142],[216,-5.0485],[217,-5.2254],[218,-5.3621],[219,-5.4726],[220,-5.5671],[221,-5.6476],[222,-5.7146],[224,-5.8141],[227,-5.9045],[228,-5.9205],[229,-5.9205],[235,-5.9656],[237,-5.9576],[238,-5.9347],[239,-5.8886],[240,-5.8124],[241,-5.6955],[242,-5.5183],[243,-5.2719],[244,-4.9426],[245,-4.5351],[246,-4.078],[247,-3.6063],[248,-3.1496],[249,-2.6822],[250,-2.2056],[251,-1.7575],[252,-1.3756],[253,-1.0505],[254,-0.7525],[255,-0.479],[256,-0.2338],[257,-0.0171],[258,0.1687],[259,0.3286],[260,0.4696],[261,0.5923],[262,0.6985],[263,0.79],[264,0.8668],[265,0.9345],[266,0.9893],[267,1.0367],[268,1.0768],[269,1.1096],[270,1.135],[271,1.1549],[272,1.1694],[273,1.1694],[274,1.1871],[275,1.1915],[277,1.1912],[278,1.185],[279,1.1704],[280,1.1502],[281,1.1185],[282,1.0784],[283,1.0268],[284,0.9639],[285,0.8981],[287,0.7495],[288,0.6725],[290,0.524],[291,0.4567],[293,0.3102],[294,0.239],[295,0.178],[296,0.1274],[297,0.0931],[298,0.0669],[299,0.0448],[300,0.0288],[301,0.0167],[302,0.0167],[303,0.0028],[305,0.0028],[306,0]]
       }
       spine_03_ctrl.Scale.X {type=MovieSceneScriptingFloatChannel, default=1, const [1,307]}
       spine_03_ctrl.Scale.Y {type=MovieSceneScriptingFloatChannel, default=1, const [1,307]}
       spine_03_ctrl.Scale.Z {type=MovieSceneScriptingFloatChannel, default=1, const [1,307]}
       spine_04_ctrl.Location.X {
        type=MovieSceneScriptingFloatChannel
        default=0
        thinned_from=307
        keys [[0,0.0568],[5,0.0568],[6,0.0539],[8,0.0539],[11,0.045],[12,0.045],[13,0.0421],[14,0.0421],[15,0.0392],[16,0.0392],[17,0.0397],[18,0.0415],[19,0.045],[20,0.0506],[21,0.0589],[23,0.0798],[24,0.0889],[25,0.0969],[26,0.1043],[27,0.1108],[28,0.1165],[29,0.1214],[30,0.1256],[31,0.1291],[32,0.1322],[33,0.1348],[35,0.1391],[37,0.1423],[39,0.1443],[42,0.1462],[45,0.1469],[49,0.1468],[50,0.1464],[52,0.1446],[54,0.1416],[56,0.1369],[57,0.1338],[58,0.13],[60,0.1207],[62,0.1084],[63,0.0999],[64,0.0867],[65,0.0697],[66,0.0497],[67,0.0279],[68,0.0022],[69,-0.0253],[70,-0.0473],[71,-0.0601],[72,-0.0674],[73,-0.0692],[74,-0.0674],[75,-0.0601],[76,-0.0491],[77,-0.0326],[78,-0.0124],[79,0.0059],[81,0.0316],[82,0.0426],[83,0.0514],[84,0.0563],[87,0.0638],[88,0.0638],[89,0.0663],[93,0.0663],[94,0.0638],[96,0.0638],[98,0.0588],[99,0.0521],[100,0.0457],[101,0.0457],[102,0.0426],[107,0.0426],[113,0.0615],[114,0.0615],[115,0.0631],[117,0.0631],[118,0.0606],[119,0.0606],[123,0.0507],[125,0.0507],[126,0.0482],[130,0.0482],[131,0.0531],[132,0.0531],[133,0.0573],[142,0.0573],[143,0.0615],[146,0.0615],[147,0.06],[148,0.06],[150,0.0503],[151,0.0503],[155,0.0309],[157,0.0309],[159,0.0212],[162,-0.008],[163,-0.0186],[164,-0.0247],[165,-0.0298],[166,-0.0337],[167,-0.0368],[168,-0.0413],[170,-0.0521],[171,-0.0581],[172,-0.0654],[173,-0.0731],[174,-0.0806],[177,-0.102],[178,-0.1088],[179,-0.1145],[181,-0.1242],[183,-0.129],[184,-0.129],[187,-0.1217],[188,-0.1169],[189,-0.1145],[191,-0.1048],[194,-0.0975],[195,-0.094],[196,-0.0914],[197,-0.0914],[198,-0.0887],[200,-0.0887],[201,-0.0861],[202,-0.0887],[203,-0.0887],[204,-0.0914],[205,-0.0914],[206,-0.0887],[207,-0.0835],[208,-0.0808],[209,-0.094],[210,-0.1177],[211,-0.1428],[212,-0.1614],[213,-0.1818],[214,-0.1967],[215,-0.2097],[216,-0.2208],[218,-0.2357],[222,-0.2506],[223,-0.2524],[224,-0.2524],[226,-0.2561],[227,-0.2559],[230,-0.2564],[232,-0.2548],[233,-0.253],[234,-0.2501],[235,-0.2459],[236,-0.2406],[237,-0.2337],[238,-0.2252],[239,-0.2152],[240,-0.2038],[241,-0.191],[242,-0.1766],[243,-0.1593],[245,-0.1219],[246,-0.0995],[247,-0.0808],[249,-0.0359],[252,0.0202],[255,0.0539],[257,0.0689],[258,0.0726],[259,0.0767],[260,0.0792],[261,0.0841],[262,0.0866],[263,0.0866],[264,0.0891],[265,0.0891],[266,0.0916],[269,0.0916],[270,0.0891],[272,0.0891],[273,0.0866],[274,0.0866],[275,0.0853],[277,0.0808],[278,0.0808],[290,0.0543],[291,0.0522],[305,0.0522],[306,0]]
       }
       spine_04_ctrl.Location.Y {
        type=MovieSceneScriptingFloatChannel
        default=0
        thinned_from=307
        keys [[0,0.9479],[67,0.9425],[68,0.9435],[69,0.9425],[70,0.9383],[71,0.9309],[72,0.9214],[74,0.9066],[76,0.895],[78,0.8855],[81,0.8781],[96,0.8744],[115,0.8826],[119,0.8878],[127,0.893],[130,0.8947],[139,0.894],[140,0.8962],[143,0.8962],[144,0.8984],[145,0.8984],[147,0.9013],[148,0.9053],[149,0.9053],[150,0.9094],[153,0.9337],[154,0.9458],[158,0.962],[161,0.962],[162,0.9579],[176,0.9422],[183,0.9303],[195,0.9292],[198,0.9353],[200,0.9435],[201,0.9497],[202,0.9538],[205,0.9599],[206,0.9599],[208,0.9558],[209,0.9517],[216,0.9342],[232,0.9283],[244,0.9431],[251,0.9549],[276,0.942],[305,0.9504],[306,0]]
       }
       spine_04_ctrl.Location.Z {
        type=MovieSceneScriptingFloatChannel
        default=0
        thinned_from=307
        keys [[0,0.1376],[68,0.1394],[69,0.1392],[70,0.1385],[74,0.1325],[78,0.1282],[79,0.1275],[84,0.1261],[96,0.1258],[115,0.1269],[116,0.1272],[117,0.1272],[124,0.1286],[130,0.1289],[139,0.1287],[143,0.1291],[144,0.1294],[145,0.1294],[147,0.1299],[148,0.1306],[149,0.1306],[150,0.1313],[153,0.1356],[154,0.1374],[155,0.1384],[158,0.1406],[161,0.1406],[162,0.1399],[163,0.14],[184,0.136],[195,0.1356],[198,0.1366],[200,0.1379],[201,0.1389],[203,0.14],[205,0.1405],[206,0.1403],[209,0.1389],[232,0.1373],[245,0.1374],[256,0.1383],[276,0.137],[305,0.1382],[306,0]]
       }
       spine_04_ctrl.Rotation.X {
        type=MovieSceneScriptingFloatChannel
        default=0
        thinned_from=307
        keys [[0,9.3065],[14,9.3391],[18,9.2782],[26,9.0001],[47,8.8743],[62,8.9282],[64,8.9741],[68,9.1066],[69,9.0865],[70,8.9683],[71,8.6752],[72,8.3346],[73,8.0654],[74,7.8223],[75,7.6011],[77,7.212],[78,7.038],[79,6.9019],[80,6.821],[82,6.7302],[84,6.6716],[87,6.6331],[93,6.6333],[97,6.6929],[105,6.8764],[113,6.953],[116,7.0355],[123,7.3169],[126,7.3837],[141,7.468],[144,7.5322],[146,7.6098],[147,7.6611],[148,7.7309],[149,7.8376],[150,7.9979],[151,8.2307],[152,8.5319],[153,8.8564],[154,9.1519],[155,9.3798],[156,9.5598],[157,9.7048],[158,9.8126],[159,9.8787],[161,9.9297],[164,9.9166],[166,9.8502],[169,9.6761],[177,9.0404],[180,8.8542],[182,8.7747],[192,8.6211],[194,8.6249],[195,8.6504],[196,8.6956],[197,8.7638],[198,8.8558],[199,8.9972],[201,9.3623],[202,9.5211],[203,9.6568],[204,9.7426],[205,9.7875],[206,9.8103],[207,9.809],[208,9.7877],[211,9.6378],[214,9.5462],[230,9.4742],[232,9.5142],[235,9.6683],[239,9.9377],[241,10.0178],[243,10.0485],[246,10.0012],[249,9.9034],[261,9.2463],[265,9.0982],[269,9.0243],[277,9.0354],[291,9.3687],[296,9.4015],[304,9.3122],[305,9.3122],[306,0]]
       }
       spine_04_ctrl.Rotation.Y {
        type=MovieSceneScriptingFloatChannel
        default=0
        thinned_from=307
        keys [[0,-2.8142],[2,-2.8142],[3,-2.8036],[4,-2.8036],[6,-2.7647],[7,-2.7258],[8,-2.6692],[9,-2.5985],[10,-2.5171],[11,-2.4145],[12,-2.2977],[13,-2.188],[14,-2.0925],[15,-2.0217],[16,-1.9862],[17,-1.9799],[18,-2.0204],[19,-2.0912],[20,-2.1756],[21,-2.2681],[22,-2.3815],[24,-2.6365],[29,-3.3236],[30,-3.4534],[31,-3.5749],[32,-3.6846],[33,-3.7822],[34,-3.8645],[35,-3.9299],[36,-3.9801],[37,-4.0182],[39,-4.0671],[42,-4.1038],[48,-4.1179],[50,-4.1023],[51,-4.0831],[52,-4.0539],[53,-4.0151],[54,-3.9647],[55,-3.9014],[56,-3.8259],[57,-3.7353],[58,-3.6298],[59,-3.508],[60,-3.3685],[61,-3.212],[62,-3.0399],[63,-2.7876],[64,-2.3924],[65,-1.8939],[66,-1.3417],[67,-0.7843],[68,-0.2553],[69,0.3831],[70,0.903],[71,1.1841],[72,1.3605],[73,1.434],[74,1.4098],[75,1.2508],[76,0.9122],[77,0.391],[78,-0.2319],[79,-0.8573],[80,-1.3984],[81,-1.8661],[82,-2.2875],[83,-2.6288],[84,-2.8609],[85,-3.0111],[86,-3.1294],[87,-3.22],[88,-3.2874],[89,-3.3344],[90,-3.3654],[91,-3.3835],[94,-3.401],[96,-3.3999],[97,-3.3788],[98,-3.2836],[99,-3.0807],[100,-2.9115],[101,-2.8416],[102,-2.8229],[103,-2.8235],[104,-2.8336],[105,-2.8598],[106,-2.902],[107,-2.9635],[108,-3.0634],[109,-3.2046],[110,-3.3361],[111,-3.448],[112,-3.5343],[113,-3.5692],[114,-3.5783],[116,-3.5741],[117,-3.5579],[118,-3.5187],[119,-3.4566],[122,-3.2329],[123,-3.1742],[124,-3.1354],[125,-3.1114],[126,-3.1012],[128,-3.1018],[129,-3.1296],[130,-3.1792],[131,-3.2686],[132,-3.3781],[133,-3.4402],[135,-3.4853],[136,-3.5011],[137,-3.5011],[138,-3.5168],[141,-3.5198],[143,-3.5067],[144,-3.4903],[145,-3.464],[146,-3.4256],[147,-3.366],[148,-3.2735],[149,-3.1464],[150,-2.9808],[151,-2.7739],[152,-2.5285],[153,-2.2413],[154,-1.9068],[155,-1.516],[156,-1.0545],[157,-0.5451],[158,-0.0279],[159,0.474],[160,0.9957],[161,1.5263],[162,2.0043],[163,2.3689],[164,2.6268],[165,2.8238],[166,2.9615],[167,3.0398],[168,3.0507],[169,2.9795],[170,2.8418],[171,2.6845],[172,2.5544],[173,2.4664],[174,2.3917],[175,2.3261],[176,2.2664],[177,2.2113],[178,2.1618],[179,2.1182],[180,2.0808],[181,2.0495],[182,2.0238],[183,2.0036],[184,1.9894],[185,1.9826],[186,1.9813],[191,2.0049],[194,2.0268],[195,2.0297],[196,2.0278],[197,2.0194],[198,2.0033],[199,1.9782],[200,1.9412],[201,1.8869],[202,1.8109],[203,1.7103],[204,1.5636],[205,1.3371],[206,1.0072],[207,0.4698],[208,0.0216],[209,0.0666],[210,0.4594],[211,0.8376],[212,1.126],[213,1.4056],[214,1.6295],[215,1.7938],[216,1.9313],[217,2.0466],[218,2.1383],[219,2.2074],[220,2.2577],[221,2.2962],[222,2.3246],[223,2.3443],[224,2.3578],[226,2.3683],[230,2.3652],[231,2.3337],[232,2.2946],[233,2.2319],[234,2.1349],[235,1.9985],[236,1.8211],[237,1.5994],[238,1.3362],[239,1.0392],[240,0.72],[241,0.3892],[242,0.0534],[243,-0.2834],[244,-0.6159],[245,-0.941],[246,-1.2571],[247,-1.5625],[248,-1.8571],[249,-2.1429],[250,-2.4115],[251,-2.6499],[252,-2.8446],[253,-2.9902],[254,-3.093],[255,-3.1594],[256,-3.1942],[257,-3.2067],[258,-3.2066],[261,-3.1742],[263,-3.1314],[266,-3.0412],[269,-2.9232],[272,-2.7857],[279,-2.439],[281,-2.3499],[282,-2.3124],[283,-2.2824],[284,-2.2609],[285,-2.2454],[287,-2.2288],[289,-2.2299],[291,-2.2426],[292,-2.2584],[293,-2.2852],[294,-2.3216],[295,-2.3677],[296,-2.4284],[298,-2.574],[299,-2.6395],[300,-2.6892],[301,-2.7268],[302,-2.7583],[303,-2.7838],[304,-2.8019],[305,-2.8019],[306,0]]
       }
       spine_04_ctrl.Rotation.Z {
        type=MovieSceneScriptingFloatChannel
        default=0
        thinned_from=307
        keys [[0,0.3714],[2,0.3714],[3,0.3697],[4,0.3697],[5,0.368],[6,0.3694],[7,0.3723],[8,0.3782],[9,0.387],[10,0.3989],[11,0.4184],[12,0.4483],[13,0.4964],[14,0.5505],[15,0.5926],[16,0.6288],[17,0.6703],[18,0.7196],[19,0.8156],[20,0.9919],[21,1.2732],[22,1.628],[23,1.9701],[24,2.2331],[25,2.4449],[26,2.6217],[27,2.759],[28,2.8555],[29,2.9217],[30,2.9634],[31,2.9864],[34,3.0144],[41,3.1378],[44,3.1665],[49,3.1672],[52,3.1458],[54,3.1151],[56,3.0677],[57,3.037],[58,2.9995],[59,2.9533],[60,2.8967],[61,2.828],[62,2.7454],[63,2.6541],[64,2.5479],[65,2.3979],[66,2.1773],[67,1.8253],[68,1.3602],[69,0.8986],[70,0.536],[71,0.2671],[72,0.0916],[73,0.0455],[74,0.089],[76,0.2572],[77,0.3345],[78,0.3964],[79,0.4211],[80,0.3926],[81,0.3503],[82,0.3173],[83,0.2879],[84,0.2551],[85,0.2203],[86,0.188],[87,0.1575],[88,0.1292],[89,0.1025],[90,0.0776],[91,0.0542],[92,0.0323],[93,0.0117],[94,-0.0081],[95,-0.0272],[96,-0.0472],[97,-0.0705],[98,-0.1017],[99,-0.1441],[100,-0.1847],[101,-0.2163],[102,-0.2453],[104,-0.2995],[106,-0.3522],[107,-0.3792],[108,-0.4],[109,-0.4079],[110,-0.4052],[111,-0.391],[113,-0.3584],[114,-0.3416],[115,-0.3215],[116,-0.3002],[117,-0.283],[118,-0.2725],[119,-0.268],[120,-0.2656],[121,-0.2665],[122,-0.2703],[123,-0.2749],[124,-0.2822],[125,-0.2928],[126,-0.3079],[127,-0.327],[128,-0.351],[129,-0.3802],[130,-0.4025],[131,-0.4099],[132,-0.4112],[134,-0.4189],[135,-0.4196],[136,-0.4184],[137,-0.4184],[138,-0.4123],[139,-0.4123],[140,-0.3946],[141,-0.3789],[142,-0.3564],[143,-0.3282],[144,-0.2963],[145,-0.2597],[146,-0.2206],[147,-0.1791],[148,-0.1353],[149,-0.0886],[150,-0.0391],[151,0.0357],[152,0.1609],[153,0.3138],[154,0.5429],[155,0.881],[156,1.3652],[157,1.9218],[158,2.4074],[159,2.7856],[160,3.0065],[161,3.1542],[162,3.2545],[163,3.3167],[164,3.35],[165,3.3608],[166,3.3493],[167,3.3002],[168,3.1067],[169,2.7783],[170,2.3684],[171,1.9066],[172,1.4253],[173,0.9751],[174,0.5548],[175,0.1654],[176,-0.2199],[177,-0.6033],[178,-0.9595],[179,-1.2705],[180,-1.5176],[181,-1.7153],[182,-1.8667],[183,-1.9779],[184,-2.0244],[185,-1.9877],[186,-1.8864],[187,-1.7297],[188,-1.5483],[189,-1.3576],[190,-1.1762],[191,-1.0009],[192,-0.8379],[193,-0.6964],[194,-0.5731],[195,-0.472],[196,-0.3938],[197,-0.3383],[198,-0.3026],[199,-0.2875],[200,-0.2983],[201,-0.3557],[202,-0.4673],[203,-0.6416],[204,-0.8793],[205,-1.1803],[206,-1.5368],[207,-1.934],[208,-2.3909],[209,-2.9309],[211,-4.1091],[212,-4.6705],[213,-5.196],[214,-5.6626],[215,-6.0439],[216,-6.3464],[217,-6.5751],[218,-6.7528],[219,-6.8944],[220,-7.0161],[221,-7.1198],[223,-7.2745],[225,-7.3764],[227,-7.4439],[229,-7.4723],[230,-7.4723],[233,-7.4978],[236,-7.4684],[237,-7.4392],[238,-7.3874],[239,-7.3038],[240,-7.1799],[241,-7.003],[242,-6.7537],[243,-6.417],[244,-5.9771],[245,-5.4403],[246,-4.8432],[247,-4.2285],[248,-3.6329],[249,-3.0256],[250,-2.4057],[251,-1.8255],[252,-1.3312],[253,-0.9134],[254,-0.5319],[255,-0.185],[256,0.1236],[257,0.3939],[258,0.625],[259,0.8233],[260,0.9967],[261,1.1482],[262,1.278],[263,1.3882],[264,1.4821],[265,1.5607],[266,1.6272],[267,1.6828],[268,1.7274],[269,1.7632],[270,1.7902],[271,1.8107],[272,1.8246],[273,1.833],[275,1.8362],[277,1.8273],[278,1.814],[279,1.7924],[280,1.7607],[281,1.718],[282,1.6626],[283,1.5952],[284,1.5178],[285,1.4322],[286,1.3411],[288,1.1516],[289,1.0588],[290,0.9687],[291,0.8833],[293,0.7068],[294,0.623],[295,0.5505],[296,0.4944],[297,0.4596],[298,0.4336],[299,0.414],[300,0.3981],[301,0.3871],[302,0.3787],[303,0.3739],[304,0.3716],[305,0.3716],[306,0]]
       }
       spine_04_ctrl.Scale.X {type=MovieSceneScriptingFloatChannel, default=1, const [1,307]}
       spine_04_ctrl.Scale.Y {type=MovieSceneScriptingFloatChannel, default=1, const [1,307]}
       spine_04_ctrl.Scale.Z {type=MovieSceneScriptingFloatChannel, default=1, const [1,307]}
       head_ctrl.Location.X {type=MovieSceneScriptingFloatChannel, default=0, thinned_from=307, keys [[0,0],[306,0]]}
       head_ctrl.Location.Y {type=MovieSceneScriptingFloatChannel, default=0, thinned_from=307, keys [[0,0],[306,0]]}
       head_ctrl.Location.Z {type=MovieSceneScriptingFloatChannel, default=0, thinned_from=307, keys [[0,0],[306,0]]}
       head_ctrl.Rotation.X {
        type=MovieSceneScriptingFloatChannel
        default=0
        thinned_from=307
        keys [[0,24.708],[2,24.772],[4,24.9544],[9,25.6811],[10,25.7341],[11,25.7015],[12,25.596],[13,25.3958],[16,24.6224],[17,24.4381],[18,24.3198],[19,24.2701],[20,24.2905],[29,25.3588],[39,25.6463],[50,25.6639],[58,25.9174],[62,25.8166],[64,25.6426],[65,25.4927],[66,25.2743],[67,24.9765],[69,24.1757],[70,23.6981],[71,23.0911],[73,21.722],[75,20.2539],[76,19.5983],[77,19.0125],[78,18.4861],[79,18.0859],[80,17.8173],[82,17.4227],[83,17.2862],[84,17.2027],[86,17.1664],[87,17.11],[89,16.906],[91,16.8029],[94,16.7718],[95,16.9611],[96,17.3338],[97,17.7541],[98,18.0913],[99,18.3084],[100,18.4372],[101,18.4478],[102,18.3439],[104,17.9519],[105,17.8072],[107,17.6432],[109,17.5861],[110,17.6658],[111,17.8558],[114,18.7029],[115,18.9588],[116,19.1645],[117,19.3139],[118,19.3392],[119,19.2697],[120,19.1485],[123,18.594],[125,18.1992],[126,18.0364],[127,17.9255],[128,17.8526],[129,17.8271],[130,17.9412],[131,18.2266],[135,19.6551],[136,19.9441],[137,20.1227],[138,20.2343],[139,20.2954],[140,20.3099],[141,20.2606],[144,19.9217],[145,19.8433],[146,19.8258],[147,19.8832],[148,20.021],[149,20.2713],[150,20.6545],[151,21.2089],[152,22.0382],[156,25.8303],[157,26.7189],[158,27.4577],[159,27.9146],[160,28.1561],[161,28.2304],[163,28.111],[164,27.9407],[165,27.6722],[168,26.5421],[170,25.6841],[173,24.2577],[174,23.8225],[175,23.4449],[176,23.1189],[177,22.8677],[178,22.681],[180,22.4406],[185,22.2202],[189,22.1576],[191,22.2087],[193,22.3606],[195,22.6096],[198,23.1586],[199,23.4579],[201,24.2413],[202,24.5627],[203,24.7917],[204,24.8169],[205,24.4641],[206,23.859],[207,23.0453],[208,22.0633],[210,19.9857],[211,19.1588],[212,18.474],[213,17.9028],[214,17.4905],[215,17.2156],[216,17.0218],[217,16.8982],[219,16.7657],[224,16.7578],[227,16.877],[228,16.9449],[229,17.0566],[230,17.2263],[231,17.4804],[232,17.7898],[233,18.2406],[234,18.9039],[235,19.8487],[236,21.0082],[237,22.2573],[238,23.4595],[239,24.5173],[240,25.3692],[241,25.9767],[242,26.3447],[243,26.4934],[244,26.4636],[246,26.2348],[248,25.8639],[251,24.9396],[252,24.6729],[254,24.2921],[256,24.0395],[258,23.9051],[261,23.9029],[266,24.2396],[275,25.1945],[282,25.3987],[297,24.7426],[305,24.7152],[306,0]]
       }
       head_ctrl.Rotation.Y {
        type=MovieSceneScriptingFloatChannel
        default=0
        thinned_from=307
        keys [[0,8.3343],[2,8.2367],[3,8.1678],[5,7.9778],[6,7.811],[7,7.5427],[8,7.1507],[9,6.5114],[10,5.3487],[11,3.6704],[12,1.7282],[13,-0.4184],[14,-2.6696],[15,-4.9125],[16,-7.0864],[17,-9.1156],[18,-10.9348],[19,-12.5085],[20,-13.7815],[21,-14.715],[22,-15.368],[23,-15.7902],[24,-16.0472],[25,-16.1842],[26,-16.2485],[31,-16.3287],[36,-16.5876],[39,-16.8915],[50,-17.8569],[52,-17.9935],[54,-18.0381],[55,-17.9889],[56,-17.8897],[57,-17.6771],[58,-17.2611],[59,-16.7013],[62,-14.6552],[66,-11.8017],[67,-11.1393],[68,-10.5327],[69,-9.9816],[70,-9.5342],[72,-8.8253],[73,-8.553],[74,-8.3388],[75,-8.0571],[76,-7.6669],[77,-7.1876],[78,-6.3587],[79,-4.9989],[80,-3.0635],[81,-0.8445],[82,1.5172],[83,3.9602],[84,6.4405],[85,8.9382],[86,11.0613],[87,12.4715],[88,13.3318],[89,13.9996],[90,14.544],[91,15.0106],[92,15.3845],[94,15.9304],[97,16.5448],[99,16.887],[100,17.0045],[102,17.0849],[107,17.0414],[113,16.7722],[114,16.6888],[116,16.4272],[117,16.3311],[118,16.2704],[119,16.2563],[121,16.3515],[124,16.6131],[129,17.2133],[130,17.3564],[131,17.5986],[134,18.6162],[135,18.9259],[136,19.1954],[137,19.3783],[139,19.609],[146,19.8038],[149,19.7588],[150,19.6942],[151,19.5525],[152,19.2956],[153,18.8071],[155,17.0053],[156,16.3073],[157,15.8308],[158,15.5978],[159,15.5946],[160,15.7867],[161,16.1946],[162,16.7453],[163,17.4126],[164,18.1952],[165,19.0486],[167,20.8254],[168,21.6107],[169,22.2479],[170,22.757],[171,23.1454],[172,23.3728],[173,23.3195],[174,22.9181],[175,22.1611],[177,20.0299],[178,19.0362],[179,18.1137],[180,17.2973],[181,16.5974],[182,15.9882],[183,15.4759],[184,15.0757],[185,14.7782],[186,14.5807],[187,14.4459],[188,14.3817],[189,14.4712],[190,14.7271],[191,15.1478],[192,15.7319],[193,16.4509],[194,17.2458],[196,18.9687],[197,19.7122],[198,20.262],[199,20.6715],[200,20.9632],[203,21.6056],[204,21.7485],[205,21.7594],[206,21.5678],[207,21.0631],[209,19.5877],[210,18.9068],[212,17.4448],[213,16.7493],[214,16.1394],[215,15.6555],[216,15.2804],[217,15.0118],[218,14.8376],[219,14.6975],[221,14.5433],[223,14.4949],[224,14.5221],[225,14.601],[226,14.7348],[227,14.9128],[228,15.1292],[230,15.7311],[231,15.9826],[232,16.134],[233,16.1224],[234,15.7913],[235,14.8807],[236,13.364],[237,11.4317],[239,7.1387],[240,5.0856],[241,3.0139],[242,1.0274],[243,-0.7273],[244,-2.1277],[245,-3.2445],[246,-4.1098],[247,-4.6922],[248,-4.9671],[249,-4.948],[250,-4.6699],[251,-4.181],[252,-3.5136],[253,-2.6601],[254,-1.6361],[256,0.5801],[257,1.6832],[258,2.8238],[259,3.944],[260,4.9751],[261,5.9062],[262,6.7962],[263,7.6519],[264,8.4481],[265,9.1695],[266,9.8006],[267,10.3584],[268,10.8441],[269,11.2647],[270,11.6267],[271,11.9255],[272,12.1555],[273,12.3276],[274,12.4639],[276,12.6472],[279,12.7629],[281,12.7051],[283,12.5615],[284,12.4527],[286,12.1639],[289,11.6246],[290,11.4254],[291,11.1833],[294,10.3375],[295,10.0803],[298,9.3804],[299,9.1762],[300,9.0035],[301,8.8586],[304,8.5313],[305,8.4382],[306,0]]
       }
       head_ctrl.Rotation.Z {
        type=MovieSceneScriptingFloatChannel
        default=0
        thinned_from=307
        keys [[0,0.0814],[1,0.0866],[2,0.1215],[3,0.1697],[4,0.2366],[5,0.413],[6,0.7846],[7,1.4406],[8,2.4651],[9,3.9613],[10,6.008],[11,8.5425],[12,11.3055],[13,14.133],[14,16.8108],[15,19.1511],[16,21.0921],[17,22.5838],[18,23.5137],[19,23.9586],[20,24.0193],[21,23.7905],[22,23.3799],[27,20.9092],[29,20.0263],[31,19.2404],[32,18.9029],[33,18.6139],[34,18.3713],[36,17.9977],[39,17.5608],[41,17.3758],[44,17.2155],[48,17.1611],[51,17.2172],[54,17.38],[56,17.5994],[57,17.7473],[58,17.9918],[59,18.2901],[62,19.2885],[64,20.0416],[66,20.9164],[67,21.3921],[68,21.9951],[69,22.775],[71,24.4688],[72,25.1373],[73,25.5015],[74,25.6446],[75,25.4093],[76,24.6236],[77,23.4467],[78,22.0558],[79,20.5071],[81,17.1809],[82,15.566],[83,14.0426],[84,12.6467],[85,11.3878],[86,10.275],[87,9.6583],[88,9.6489],[89,9.8125],[90,10],[92,10.2889],[95,10.6693],[97,10.8981],[98,11.0418],[99,11.2523],[100,11.5652],[102,12.4008],[103,12.7873],[104,13.1223],[105,13.4161],[106,13.683],[107,13.9152],[109,14.2942],[111,14.5698],[113,14.7541],[116,14.8843],[117,14.8876],[118,14.8472],[119,14.7565],[120,14.5677],[121,14.3187],[123,13.7247],[124,13.3807],[125,13.0057],[126,12.5652],[128,11.6053],[129,11.147],[130,10.7243],[131,10.3708],[133,9.7832],[135,9.2906],[136,9.1315],[138,8.8756],[139,8.7611],[140,8.7178],[145,8.5508],[147,8.4407],[149,8.2642],[150,8.1531],[151,8.0039],[152,7.7478],[153,7.3052],[154,6.6194],[155,5.7312],[156,4.6549],[157,3.4138],[158,1.8817],[159,0.1298],[160,-1.5905],[161,-3.2636],[162,-4.905],[163,-6.5137],[164,-8.0746],[165,-9.5988],[166,-11.067],[167,-12.4886],[168,-13.8635],[169,-15.2054],[170,-16.4915],[171,-17.7058],[173,-19.991],[174,-20.9673],[175,-21.6811],[176,-22.1817],[177,-22.446],[178,-22.5297],[179,-22.5007],[181,-22.2959],[184,-21.8959],[186,-21.7241],[188,-21.6979],[190,-21.78],[191,-21.8857],[192,-22.0594],[193,-22.3283],[194,-22.7204],[195,-23.2248],[196,-23.8206],[197,-24.5389],[198,-25.3788],[202,-29.0643],[209,-35.9535],[210,-36.8114],[213,-38.75],[214,-39.2485],[216,-39.8986],[219,-40.5026],[221,-40.6837],[231,-42.1433],[232,-42.2145],[233,-42.1948],[234,-41.4643],[235,-39.392],[236,-35.9586],[237,-31.7943],[238,-27.3815],[239,-23.1943],[240,-19.4248],[241,-15.8493],[242,-12.5418],[243,-9.6737],[244,-7.4307],[245,-5.8833],[246,-4.8067],[247,-3.9222],[248,-2.9548],[249,-1.9185],[250,-0.9673],[251,-0.1552],[252,0.457],[253,0.9505],[254,1.4333],[255,1.8927],[256,2.3039],[257,2.6614],[258,2.9869],[259,3.2649],[260,3.4757],[261,3.5992],[262,3.6244],[263,3.5892],[264,3.5266],[265,3.4771],[267,3.4304],[270,3.3419],[271,3.333],[272,3.3422],[273,3.3744],[274,3.4488],[276,3.5472],[277,3.5719],[278,3.6049],[280,3.6292],[281,3.6256],[282,3.6104],[283,3.571],[284,3.5067],[285,3.4195],[286,3.3136],[287,3.1899],[288,3.0367],[289,2.866],[290,2.6764],[291,2.4659],[292,2.2117],[293,1.9281],[294,1.6496],[295,1.4063],[296,1.1807],[297,0.9582],[298,0.7474],[299,0.5653],[300,0.4252],[301,0.3218],[302,0.2712],[303,0.1911],[304,0.1554],[305,0.1292],[306,0]]
       }
       head_ctrl.Scale.X {type=MovieSceneScriptingFloatChannel, default=1, const [1,307]}
       head_ctrl.Scale.Y {type=MovieSceneScriptingFloatChannel, default=1, const [1,307]}
       head_ctrl.Scale.Z {type=MovieSceneScriptingFloatChannel, default=1, const [1,307]}
       clavicle_l_ctrl.Location.X {type=MovieSceneScriptingFloatChannel, default=0, thinned_from=307, keys [[0,0],[306,0]]}
       clavicle_l_ctrl.Location.Y {type=MovieSceneScriptingFloatChannel, default=0, thinned_from=307, keys [[0,0],[306,0]]}
       clavicle_l_ctrl.Location.Z {type=MovieSceneScriptingFloatChannel, default=0, thinned_from=307, keys [[0,0],[306,0]]}
       clavicle_l_ctrl.Rotation.X {
        type=MovieSceneScriptingFloatChannel
        default=0
        thinned_from=307
        keys [[0,0],[61,0],[181,0],[182,0.0001],[205,0],[244,0.0001],[245,-0.0001],[247,0.0001],[252,0],[253,-0.0002],[255,-0.0001],[256,0.0001],[306,0]]
       }
       clavicle_l_ctrl.Rotation.Y {
        type=MovieSceneScriptingFloatChannel
        default=0
        thinned_from=307
        keys [[0,-8.4244],[5,-8.4244],[9,-8.5282],[10,-8.5792],[11,-8.6492],[12,-8.7438],[13,-8.8612],[17,-9.4677],[18,-9.5912],[19,-9.6816],[20,-9.7265],[21,-9.7108],[22,-9.6553],[25,-9.4078],[32,-8.7441],[34,-8.5865],[36,-8.4797],[39,-8.3894],[40,-8.3712],[41,-8.3712],[42,-8.345],[55,-8.3287],[59,-8.4031],[60,-8.4497],[61,-8.5428],[62,-8.706],[63,-8.9436],[64,-9.2561],[65,-9.6757],[66,-10.2079],[67,-11.2416],[69,-14.4285],[70,-15.8456],[71,-16.9857],[72,-17.9147],[73,-18.5696],[74,-19.0411],[75,-19.3938],[76,-19.6682],[77,-19.8705],[79,-20.1112],[81,-20.2579],[83,-20.3199],[151,-20.3199],[153,-20.2471],[155,-20.082],[160,-19.3554],[173,-16.9852],[177,-16.4211],[180,-16.1098],[187,-15.6995],[195,-15.3548],[197,-15.2094],[199,-14.9818],[201,-14.6749],[202,-14.4709],[203,-14.2255],[204,-13.8488],[207,-12.1622],[208,-11.638],[209,-11.2512],[210,-10.9808],[211,-10.8344],[212,-10.7705],[213,-10.7444],[214,-10.7444],[215,-10.7681],[217,-10.9694],[221,-11.4808],[223,-11.6861],[225,-11.8201],[227,-11.8917],[229,-11.8917],[231,-11.8129],[232,-11.7286],[233,-11.6011],[234,-11.4305],[235,-11.2088],[238,-10.4054],[247,-7.8374],[249,-7.2988],[250,-7.0513],[251,-6.822],[252,-6.6284],[253,-6.4585],[254,-6.3182],[255,-6.2073],[256,-6.1315],[258,-6.0382],[260,-5.9764],[261,-5.9551],[262,-5.9551],[263,-5.9304],[268,-5.9304],[274,-6.1047],[276,-6.1981],[278,-6.334],[280,-6.5046],[282,-6.6984],[284,-6.9273],[290,-7.727],[292,-7.9431],[294,-8.114],[296,-8.2322],[299,-8.3397],[302,-8.401],[303,-8.401],[304,-8.42],[305,-8.42],[306,0]]
       }
       clavicle_l_ctrl.Rotation.Z {
        type=MovieSceneScriptingFloatChannel
        default=0
        thinned_from=307
        keys [[0,0],[61,0],[181,-0.0001],[232,0.0001],[244,0.0001],[252,0],[255,-0.0001],[256,0.0002],[306,0]]
       }
       clavicle_l_ctrl.Scale.X {type=MovieSceneScriptingFloatChannel, default=1, const [1,307]}
       clavicle_l_ctrl.Scale.Y {type=MovieSceneScriptingFloatChannel, default=1, const [1,307]}
       clavicle_l_ctrl.Scale.Z {type=MovieSceneScriptingFloatChannel, default=1, const [1,307]}
       clavicle_r_ctrl.Location.X {type=MovieSceneScriptingFloatChannel, default=0, thinned_from=307, keys [[0,0],[306,0]]}
       clavicle_r_ctrl.Location.Y {type=MovieSceneScriptingFloatChannel, default=0, thinned_from=307, keys [[0,0],[306,0]]}
       clavicle_r_ctrl.Location.Z {type=MovieSceneScriptingFloatChannel, default=0, thinned_from=307, keys [[0,0],[306,0]]}
       clavicle_r_ctrl.Rotation.X {
        type=MovieSceneScriptingFloatChannel
        default=0
        thinned_from=307
        keys [[0,-0.0009],[5,-0.0009],[6,-0.0006],[7,-0.0007],[8,-0.0002],[9,-0.0003],[10,0.0008],[11,-0.0002],[13,-0.0007],[14,-0.0002],[15,0],[16,0.0005],[17,-0.0004],[19,0],[20,0.0008],[21,-0.0005],[22,-0.0005],[23,0.0008],[24,0.0008],[25,-0.0008],[26,0.0003],[27,-0.0009],[29,0.0008],[30,0.0007],[31,-0.0001],[32,-0.0006],[33,-0.0005],[34,0.0001],[35,0],[36,-0.0012],[37,-0.0011],[38,0.0003],[42,0.0003],[43,0.0005],[44,0.0004],[45,-0.0002],[46,0.0002],[47,-0.0002],[48,-0.0004],[49,0.0022],[50,0.0006],[51,-0.0003],[52,-0.0002],[54,0.0002],[60,0],[61,-0.0003],[63,0.0006],[64,0],[65,0.0015],[66,0.0008],[67,-0.0019],[68,0.0007],[69,0.0003],[70,-0.0003],[71,0.0004],[72,-0.0001],[73,0],[75,-0.0002],[76,0.0002],[77,-0.0002],[78,-0.0002],[80,0.0002],[81,-0.0001],[82,0.0003],[83,-0.0017],[84,-0.0017],[85,0.0008],[86,0.0001],[88,0.0001],[89,-0.0006],[90,0.0007],[91,-0.0016],[92,-0.0016],[93,-0.0008],[94,-0.0005],[96,0.0013],[98,0.0013],[99,0.0006],[100,-0.0006],[101,-0.0012],[103,-0.001],[104,0.0009],[107,0.0009],[108,0.0013],[109,0.0007],[110,0.0009],[111,-0.0011],[112,0.0002],[115,0],[116,-0.0005],[117,0.0004],[119,0.0003],[120,0.0009],[121,0.0007],[123,-0.0003],[124,0.0004],[125,0.0004],[126,0.0014],[127,0.0008],[128,0.0013],[129,0.0001],[130,0.0006],[131,0.0025],[132,0.0025],[133,-0.0003],[137,-0.0003],[138,0.0008],[140,0.0008],[141,-0.0009],[142,0.0002],[143,-0.0019],[144,-0.0018],[145,0.0012],[146,-0.0018],[147,0],[148,0.0006],[149,0.0002],[150,0.0002],[151,0.0006],[152,0.0014],[153,-0.0015],[154,-0.0004],[155,-0.0013],[156,0.0015],[157,0.0003],[158,0.0003],[159,-0.0002],[160,0.0001],[167,-0.0002],[168,0],[170,-0.0001],[171,0.0091],[172,-0.0002],[174,0.0001],[175,-0.0003],[176,-0.0001],[177,-0.0002],[178,0.0001],[179,-0.0002],[180,-0.0008],[181,0.0006],[182,-0.0004],[183,0.0001],[184,-0.0007],[185,-0.0001],[186,-0.0007],[188,0.0001],[190,-0.0007],[192,-0.0008],[193,-0.0002],[194,0.0012],[195,-0.0001],[196,0.0019],[197,0],[198,0.0001],[200,0.0008],[201,-0.0004],[202,0.0001],[204,0.0018],[205,-0.0007],[206,0.0002],[207,-0.0002],[208,0.0002],[209,0.0004],[210,-0.0006],[211,0.0007],[212,-0.0005],[213,-0.0002],[214,-0.0003],[215,0.0002],[216,-0.0003],[217,0.0006],[218,0.0004],[219,-0.0016],[220,-0.0001],[221,-0.0003],[222,0.0004],[223,0.0008],[224,0],[226,0],[227,-0.0005],[228,-0.0005],[229,0.0005],[230,0],[237,0],[238,0.0012],[239,-0.0009],[241,0.0005],[242,0.0003],[243,-0.0003],[245,0.0008],[246,0.0001],[247,-0.0002],[248,-0.0008],[249,-0.0008],[250,-0.0001],[251,-0.001],[252,0.001],[254,-0.0005],[255,0.0011],[256,0],[257,-0.0007],[258,-0.0004],[259,0.0004],[260,-0.0007],[261,0.0002],[262,0],[263,-0.0011],[264,-0.0011],[265,0.0005],[279,0.0005],[280,-0.0005],[281,-0.0008],[282,-0.0007],[283,0.0002],[285,0],[286,0.0004],[287,-0.0011],[288,-0.0002],[289,-0.0008],[290,-0.0005],[291,-0.0006],[292,0.0008],[293,-0.0005],[294,-0.0007],[295,0.0008],[296,-0.0005],[297,-0.0001],[298,-0.0003],[299,0.0003],[300,-0.0006],[301,-0.0002],[302,-0.0002],[303,0.0001],[306,0]]
       }
       clavicle_r_ctrl.Rotation.Y {
        type=MovieSceneScriptingFloatChannel
        default=0
        thinned_from=307
        keys [[0,-6.7731],[5,-6.7731],[8,-6.8554],[9,-6.8995],[10,-6.9607],[11,-7.0498],[12,-7.1681],[13,-7.3086],[15,-7.6497],[19,-8.4253],[22,-8.9637],[23,-9.0869],[24,-9.0869],[25,-8.9904],[26,-8.8191],[28,-8.4232],[32,-7.7945],[33,-7.6582],[34,-7.5482],[35,-7.4702],[36,-7.4176],[37,-7.3844],[38,-7.3662],[42,-7.3662],[46,-7.546],[49,-7.7369],[50,-7.7878],[53,-7.87],[56,-7.9028],[58,-7.9028],[59,-7.8272],[60,-7.7167],[61,-7.554],[62,-7.3354],[63,-7.0566],[64,-6.6961],[65,-6.0017],[66,-4.9555],[67,-3.7181],[68,-1.8643],[69,1.3009],[70,4.5945],[71,7.2261],[72,9.5008],[73,11.3734],[74,12.7745],[75,13.7757],[76,14.4367],[77,14.9158],[78,15.2758],[79,15.5563],[80,15.7556],[81,15.8771],[82,15.9437],[84,15.9657],[85,15.8804],[86,15.7145],[87,15.6325],[88,15.6325],[89,15.692],[90,15.8132],[91,15.8815],[92,15.8815],[95,15.689],[96,15.6514],[98,15.6514],[102,15.9051],[104,15.9554],[107,15.9554],[111,15.7516],[114,15.7318],[118,15.907],[119,15.907],[123,15.6753],[125,15.6501],[126,15.6708],[130,15.9176],[137,15.962],[141,15.8838],[142,15.8085],[143,15.6994],[144,15.3875],[145,14.7368],[146,13.3769],[147,11.5752],[148,9.822],[149,7.9646],[150,5.9125],[151,3.3335],[152,0.4328],[153,-2.5065],[154,-5.2403],[155,-7.594],[156,-9.6698],[157,-11.5119],[158,-13.0811],[159,-14.3493],[160,-15.3223],[161,-16.0167],[162,-16.4308],[163,-16.5617],[164,-16.4685],[165,-16.2356],[166,-15.9157],[168,-15.2119],[171,-14.2705],[172,-14.0022],[173,-13.7745],[174,-13.5989],[176,-13.3389],[178,-13.1477],[180,-13.0177],[183,-12.9164],[193,-12.8685],[195,-12.8038],[196,-12.7408],[197,-12.6517],[198,-12.5345],[199,-12.3441],[200,-12.0098],[201,-11.5134],[202,-10.8244],[204,-9.2705],[206,-7.6749],[207,-6.8591],[208,-6.0591],[209,-5.3125],[210,-4.6121],[211,-3.9618],[212,-3.344],[213,-2.7548],[214,-2.2149],[215,-1.7586],[216,-1.3938],[217,-1.0885],[218,-0.8368],[219,-0.6446],[220,-0.4931],[221,-0.3622],[222,-0.2528],[223,-0.1694],[224,-0.1091],[225,-0.0578],[226,0],[227,0.027],[228,0.0596],[229,0.0853],[230,0.1039],[236,0.1039],[237,0],[238,-0.1184],[239,-0.2803],[240,-0.4693],[241,-0.6979],[242,-0.9559],[244,-1.4981],[245,-1.7828],[247,-2.3681],[248,-2.643],[249,-2.9082],[250,-3.1666],[251,-3.4088],[252,-3.6258],[253,-3.8165],[254,-3.9852],[255,-4.1326],[256,-4.2596],[257,-4.3637],[258,-4.4437],[259,-4.5047],[260,-4.5527],[261,-4.5915],[262,-4.6204],[263,-4.6399],[264,-4.6399],[265,-4.6562],[279,-4.6562],[281,-4.7295],[282,-4.7857],[283,-4.8859],[284,-5.048],[286,-5.4352],[289,-5.9391],[290,-6.0888],[291,-6.2148],[292,-6.3184],[293,-6.4054],[295,-6.5406],[297,-6.6364],[300,-6.7244],[301,-6.7428],[302,-6.7428],[303,-6.7649],[305,-6.7649],[306,0]]
       }
       clavicle_r_ctrl.Rotation.Z {
        type=MovieSceneScriptingFloatChannel
        default=0
        thinned_from=307
        keys [[0,-0.0008],[5,-0.0008],[6,-0.0005],[7,-0.0006],[8,-0.0001],[9,-0.0002],[10,0.0008],[11,-0.0001],[13,-0.0007],[14,-0.0001],[15,0],[16,0.0005],[17,-0.0004],[19,0],[20,0.0008],[21,-0.0005],[22,-0.0005],[23,0.0007],[24,0.0007],[25,-0.0007],[26,0.0003],[27,-0.0008],[29,0.0008],[30,0.0007],[31,-0.0001],[32,-0.0005],[33,-0.0005],[34,0.0001],[35,0],[36,-0.001],[37,-0.0009],[38,0.0003],[42,0.0003],[43,0.0005],[44,0.0004],[45,-0.0001],[46,0.0002],[48,-0.0004],[49,0.0021],[50,0.0006],[51,-0.0002],[52,-0.0002],[54,0.0002],[60,0],[61,-0.0002],[62,0.0001],[63,-0.0227],[64,-0.1477],[65,-0.436],[66,-0.9585],[67,-1.7107],[68,-2.5696],[69,-3.5036],[70,-4.4928],[71,-5.5169],[73,-7.6029],[74,-8.6284],[75,-9.6207],[76,-10.5175],[77,-11.2012],[78,-11.7475],[79,-12.1785],[80,-12.5143],[81,-12.7388],[82,-12.8519],[83,-12.8962],[140,-12.8943],[141,-12.8635],[142,-12.7912],[143,-12.6668],[144,-12.4688],[145,-12.1636],[146,-11.6858],[147,-10.8244],[148,-9.7259],[149,-8.5111],[150,-7.3358],[151,-6.1351],[152,-4.734],[153,-3.424],[154,-2.5663],[155,-2.4874],[156,-3.178],[157,-4.424],[158,-6.0591],[159,-7.8716],[160,-9.3821],[161,-10.4013],[162,-10.9831],[163,-11.1587],[164,-10.4212],[165,-8.6717],[166,-6.4656],[167,-4.3277],[168,-2.79],[169,-1.8326],[170,-1.0742],[171,-0.4791],[172,-0.0488],[173,0.2653],[174,0.4775],[175,0.619],[176,0.7127],[177,0.768],[178,0.7953],[179,0.8037],[180,0.8021],[181,0.7616],[182,0.6767],[183,0.5662],[184,0.4431],[185,0.3255],[186,0.2276],[187,0.1596],[188,0.1068],[189,0.0636],[190,0.0311],[191,0.01],[192,0.01],[193,-0.0002],[194,0.0011],[195,-0.0001],[196,0.0017],[197,0],[198,0.0001],[200,0.0007],[201,-0.0004],[202,0.0001],[204,0.0017],[205,-0.0006],[206,0.0002],[207,-0.0002],[209,0.0004],[210,-0.0005],[211,0.0007],[212,-0.0005],[213,-0.0002],[214,-0.0003],[215,0.0001],[216,-0.0003],[217,0.0006],[218,0.0005],[219,-0.0015],[220,0],[221,-0.0002],[222,0.0004],[223,0.0007],[224,0],[226,0],[227,-0.0006],[228,-0.0006],[229,0.0004],[230,0],[237,0],[238,0.0011],[239,-0.0008],[241,0.0005],[242,0.0003],[243,-0.0003],[245,0.0006],[246,0.0001],[247,-0.0002],[248,-0.0008],[249,-0.0008],[250,-0.0001],[251,-0.0009],[252,0.0009],[254,-0.0004],[255,0.0011],[256,0.0001],[257,-0.0006],[258,-0.0004],[259,0.0004],[260,-0.0006],[261,0.0003],[262,0.0001],[263,-0.0009],[264,-0.0009],[265,0.0005],[279,0.0005],[280,-0.0004],[281,-0.0007],[282,-0.0006],[283,0.0003],[285,0.0001],[286,0.0004],[287,-0.001],[288,-0.0002],[289,-0.0007],[290,-0.0004],[291,-0.0005],[292,0.0008],[293,-0.0004],[294,-0.0006],[295,0.0007],[296,-0.0005],[297,0],[298,-0.0002],[299,0.0003],[300,-0.0005],[301,-0.0002],[302,-0.0002],[303,0.0002],[305,0.0002],[306,0]]
       }
       clavicle_r_ctrl.Scale.X {type=MovieSceneScriptingFloatChannel, default=1, const [1,307]}
       clavicle_r_ctrl.Scale.Y {type=MovieSceneScriptingFloatChannel, default=1, const [1,307]}
       clavicle_r_ctrl.Scale.Z {type=MovieSceneScriptingFloatChannel, default=1, const [1,307]}
       neck_01_ctrl.Location.X {type=MovieSceneScriptingFloatChannel, default=0, thinned_from=307, keys [[0,0],[306,0]]}
       neck_01_ctrl.Location.Y {type=MovieSceneScriptingFloatChannel, default=0, thinned_from=307, keys [[0,0],[306,0]]}
       neck_01_ctrl.Location.Z {type=MovieSceneScriptingFloatChannel, default=0, thinned_from=307, keys [[0,0],[306,0]]}
       neck_01_ctrl.Rotation.X {
        type=MovieSceneScriptingFloatChannel
        default=0
        thinned_from=307
        keys [[0,0],[77,0],[78,0.0001],[79,0.0017],[80,0.0075],[81,0.0209],[82,0.04],[83,0.0623],[84,0.0898],[85,0.1199],[86,0.1494],[87,0.1765],[88,0.1978],[89,0.2112],[90,0.2147],[150,0.2147],[151,0.2105],[152,0.1911],[153,0.1631],[154,0.1339],[156,0.0822],[157,0.0585],[158,0.0391],[159,0.0252],[160,0.0149],[161,0.0078],[162,0.0034],[163,0.0013],[164,0.0004],[165,0.0001],[306,0]]
       }
       neck_01_ctrl.Rotation.Y {
        type=MovieSceneScriptingFloatChannel
        default=0
        thinned_from=307
        keys [[0,0],[76,0],[77,0.0178],[78,0.2374],[79,0.7792],[80,1.6028],[81,2.6755],[82,3.6969],[84,5.5339],[85,6.3921],[86,7.1278],[87,7.7412],[88,8.1934],[89,8.462],[90,8.5308],[150,8.5308],[151,8.4469],[152,8.0529],[153,7.4437],[155,6.0653],[156,5.2962],[157,4.469],[158,3.6563],[159,2.9349],[160,2.2646],[161,1.6345],[162,1.0881],[163,0.6731],[164,0.3952],[165,0.1982],[166,0.0825],[167,0.0187],[168,0],[306,0]]
       }
       neck_01_ctrl.Rotation.Z {
        type=MovieSceneScriptingFloatChannel
        default=0
        thinned_from=307
        keys [[0,0],[76,0],[77,0.006],[78,0.0799],[79,0.261],[80,0.5369],[81,0.896],[82,1.2395],[84,1.8588],[85,2.1493],[86,2.399],[87,2.6075],[88,2.7629],[89,2.8537],[90,2.8791],[150,2.8791],[151,2.8498],[152,2.7146],[153,2.5068],[155,2.0387],[156,1.7786],[157,1.4987],[158,1.226],[159,0.9837],[160,0.7589],[161,0.5471],[162,0.3643],[163,0.2256],[164,0.1317],[165,0.0665],[166,0.0265],[167,0.0058],[168,0],[306,0]]
       }
       neck_01_ctrl.Scale.X {type=MovieSceneScriptingFloatChannel, default=1, const [1,307]}
       neck_01_ctrl.Scale.Y {type=MovieSceneScriptingFloatChannel, default=1, const [1,307]}
       neck_01_ctrl.Scale.Z {type=MovieSceneScriptingFloatChannel, default=1, const [1,307]}
       neck_02_ctrl.Location.X {type=MovieSceneScriptingFloatChannel, default=0, const [0,307]}
       neck_02_ctrl.Location.Y {type=MovieSceneScriptingFloatChannel, default=0, const [0,307]}
       neck_02_ctrl.Location.Z {type=MovieSceneScriptingFloatChannel, default=0, const [0,307]}
       neck_02_ctrl.Rotation.X {type=MovieSceneScriptingFloatChannel, default=0, const [0,307]}
       neck_02_ctrl.Rotation.Y {type=MovieSceneScriptingFloatChannel, default=0, const [0,307]}
       neck_02_ctrl.Rotation.Z {type=MovieSceneScriptingFloatChannel, default=0, const [0,307]}
       neck_02_ctrl.Scale.X {type=MovieSceneScriptingFloatChannel, default=1, const [1,307]}
       neck_02_ctrl.Scale.Y {type=MovieSceneScriptingFloatChannel, default=1, const [1,307]}
       neck_02_ctrl.Scale.Z {type=MovieSceneScriptingFloatChannel, default=1, const [1,307]}
       head_ik_ctrl.Location.X {
        type=MovieSceneScriptingFloatChannel
        default=0
        thinned_from=307
        keys [[0,0],[76,0],[77,-0.0033],[78,-0.0434],[79,-0.1423],[80,-0.2928],[81,-0.4887],[82,-0.6752],[84,-1.0106],[85,-1.1671],[86,-1.3013],[87,-1.4131],[88,-1.4956],[89,-1.5445],[90,-1.5571],[150,-1.557],[151,-1.5414],[152,-1.4692],[153,-1.3578],[155,-1.1046],[156,-0.9625],[157,-0.8095],[158,-0.66],[159,-0.5279],[160,-0.4064],[161,-0.2927],[162,-0.1946],[163,-0.1202],[164,-0.0705],[165,-0.0354],[166,-0.0147],[167,-0.0033],[168,0],[306,0]]
       }
       head_ik_ctrl.Location.Y {
        type=MovieSceneScriptingFloatChannel
        default=0
        thinned_from=307
        keys [[0,0],[77,-0.0001],[78,-0.001],[79,-0.0029],[80,-0.0054],[82,-0.0126],[83,-0.0164],[84,-0.0214],[85,-0.0272],[86,-0.0326],[87,-0.0372],[88,-0.0405],[89,-0.0415],[90,-0.04],[91,-0.0378],[92,-0.0358],[97,-0.0277],[98,-0.0266],[99,-0.0264],[100,-0.0258],[101,-0.0242],[102,-0.0222],[104,-0.0177],[105,-0.0151],[106,-0.0123],[108,-0.0058],[109,-0.0029],[110,-0.0011],[111,-0.0005],[112,-0.0004],[113,-0.0011],[114,-0.0023],[116,-0.0061],[117,-0.0078],[118,-0.0091],[120,-0.0121],[123,-0.0149],[124,-0.015],[125,-0.0147],[126,-0.0139],[128,-0.0099],[129,-0.007],[130,-0.0044],[131,-0.0024],[132,-0.0008],[133,0.0006],[134,0.0009],[135,0.0014],[137,0.0018],[139,0.0018],[140,0.0006],[141,-0.001],[142,-0.0029],[143,-0.0054],[144,-0.0083],[145,-0.0118],[146,-0.0157],[147,-0.0202],[148,-0.0253],[149,-0.0314],[150,-0.0383],[151,-0.0467],[152,-0.0546],[153,-0.0604],[154,-0.0686],[155,-0.0811],[156,-0.0953],[157,-0.1042],[158,-0.1023],[159,-0.093],[160,-0.077],[161,-0.0583],[162,-0.0401],[163,-0.0253],[164,-0.0151],[165,-0.0076],[166,-0.0032],[167,-0.0007],[168,0],[306,0]]
       }
       head_ik_ctrl.Location.Z {
        type=MovieSceneScriptingFloatChannel
        default=0
        thinned_from=307
        keys [[0,0],[77,-0.0001],[78,-0.0008],[79,-0.001],[80,-0.0002],[81,0.001],[83,0.0046],[84,0.0042],[85,0.0015],[86,-0.0018],[87,-0.0054],[88,-0.0085],[89,-0.0104],[90,-0.0106],[91,-0.0104],[95,-0.0113],[96,-0.0118],[97,-0.0129],[98,-0.0168],[99,-0.0246],[100,-0.0313],[101,-0.0343],[102,-0.0355],[103,-0.0359],[104,-0.036],[105,-0.0356],[106,-0.0346],[107,-0.033],[108,-0.0299],[110,-0.0204],[111,-0.0162],[112,-0.0132],[113,-0.0114],[115,-0.0102],[116,-0.0099],[117,-0.0101],[118,-0.0111],[119,-0.0131],[123,-0.0229],[125,-0.0252],[127,-0.0264],[128,-0.0267],[129,-0.0263],[130,-0.025],[131,-0.022],[132,-0.0184],[133,-0.0162],[135,-0.0149],[138,-0.0137],[139,-0.0137],[141,-0.0126],[143,-0.012],[144,-0.0119],[145,-0.0121],[146,-0.0126],[147,-0.0138],[148,-0.016],[149,-0.0194],[150,-0.0239],[151,-0.0281],[152,-0.0284],[153,-0.0266],[155,-0.0216],[157,-0.0147],[158,-0.0123],[159,-0.0109],[160,-0.0104],[161,-0.0093],[162,-0.0074],[163,-0.0052],[164,-0.0033],[165,-0.0018],[166,-0.0008],[167,-0.0002],[306,0]]
       }
       head_ik_ctrl.Rotation.X {
        type=MovieSceneScriptingFloatChannel
        default=0
        thinned_from=307
        keys [[0,24.708],[2,24.772],[4,24.9544],[9,25.6811],[10,25.7341],[11,25.7015],[12,25.596],[13,25.3958],[16,24.6224],[17,24.4381],[18,24.3198],[19,24.2701],[20,24.2905],[29,25.3588],[39,25.6463],[50,25.6639],[58,25.9174],[62,25.8166],[64,25.6426],[65,25.4927],[66,25.2743],[67,24.9765],[69,24.1757],[70,23.6981],[71,23.0911],[73,21.722],[75,20.2539],[76,19.5983],[77,19.0125],[78,18.4861],[79,18.0859],[80,17.8173],[82,17.4227],[83,17.2862],[84,17.2027],[86,17.1664],[87,17.11],[89,16.906],[91,16.8029],[94,16.7718],[95,16.9611],[96,17.3338],[97,17.7541],[98,18.0913],[99,18.3084],[100,18.4372],[101,18.4478],[102,18.3439],[104,17.9519],[105,17.8072],[107,17.6432],[109,17.5861],[110,17.6658],[111,17.8558],[114,18.7029],[115,18.9588],[116,19.1645],[117,19.3139],[118,19.3392],[119,19.2697],[120,19.1485],[123,18.594],[125,18.1992],[126,18.0364],[127,17.9255],[128,17.8526],[129,17.8271],[130,17.9412],[131,18.2266],[135,19.6551],[136,19.9441],[137,20.1227],[138,20.2343],[139,20.2954],[140,20.3099],[141,20.2606],[144,19.9217],[145,19.8433],[146,19.8258],[147,19.8832],[148,20.021],[149,20.2713],[150,20.6545],[151,21.2089],[152,22.0382],[156,25.8303],[157,26.7189],[158,27.4577],[159,27.9146],[160,28.1561],[161,28.2304],[163,28.111],[164,27.9407],[165,27.6722],[168,26.5421],[170,25.6841],[173,24.2577],[174,23.8225],[175,23.4449],[176,23.1189],[177,22.8677],[178,22.681],[180,22.4406],[185,22.2202],[189,22.1576],[191,22.2087],[193,22.3606],[195,22.6096],[198,23.1586],[199,23.4579],[201,24.2413],[202,24.5627],[203,24.7917],[204,24.8169],[205,24.4641],[206,23.859],[207,23.0453],[208,22.0633],[210,19.9857],[211,19.1588],[212,18.474],[213,17.9028],[214,17.4905],[215,17.2156],[216,17.0218],[217,16.8982],[219,16.7657],[224,16.7578],[227,16.877],[228,16.9449],[229,17.0566],[230,17.2263],[231,17.4804],[232,17.7898],[233,18.2406],[234,18.9039],[235,19.8487],[236,21.0082],[237,22.2573],[238,23.4595],[239,24.5173],[240,25.3692],[241,25.9767],[242,26.3447],[243,26.4934],[244,26.4636],[246,26.2348],[248,25.8639],[251,24.9396],[252,24.6729],[254,24.2921],[256,24.0395],[258,23.9051],[261,23.9029],[266,24.2396],[275,25.1945],[282,25.3987],[297,24.7426],[305,24.7152],[306,0]]
       }
       head_ik_ctrl.Rotation.Y {
        type=MovieSceneScriptingFloatChannel
        default=0
        thinned_from=307
        keys [[0,8.3343],[2,8.2367],[3,8.1678],[5,7.9778],[6,7.811],[7,7.5427],[8,7.1507],[9,6.5114],[10,5.3487],[11,3.6704],[12,1.7282],[13,-0.4184],[14,-2.6696],[15,-4.9125],[16,-7.0864],[17,-9.1156],[18,-10.9348],[19,-12.5085],[20,-13.7815],[21,-14.715],[22,-15.368],[23,-15.7902],[24,-16.0472],[25,-16.1842],[26,-16.2485],[31,-16.3287],[36,-16.5876],[39,-16.8915],[50,-17.8569],[52,-17.9935],[54,-18.0381],[55,-17.9889],[56,-17.8897],[57,-17.6771],[58,-17.2611],[59,-16.7013],[62,-14.6552],[66,-11.8017],[67,-11.1393],[68,-10.5327],[69,-9.9816],[70,-9.5342],[72,-8.8253],[73,-8.553],[74,-8.3388],[75,-8.0571],[76,-7.6669],[77,-7.1876],[78,-6.3587],[79,-4.9989],[80,-3.0635],[81,-0.8445],[82,1.5172],[83,3.9602],[84,6.4405],[85,8.9382],[86,11.0613],[87,12.4715],[88,13.3318],[89,13.9996],[90,14.544],[91,15.0106],[92,15.3845],[94,15.9304],[97,16.5448],[99,16.887],[100,17.0045],[102,17.0849],[107,17.0414],[113,16.7722],[114,16.6888],[116,16.4272],[117,16.3311],[118,16.2704],[119,16.2563],[121,16.3515],[124,16.6131],[129,17.2133],[130,17.3564],[131,17.5986],[134,18.6162],[135,18.9259],[136,19.1954],[137,19.3783],[139,19.609],[146,19.8038],[149,19.7588],[150,19.6942],[151,19.5525],[152,19.2956],[153,18.8071],[155,17.0053],[156,16.3073],[157,15.8308],[158,15.5978],[159,15.5946],[160,15.7867],[161,16.1946],[162,16.7453],[163,17.4126],[164,18.1952],[165,19.0486],[167,20.8254],[168,21.6107],[169,22.2479],[170,22.757],[171,23.1454],[172,23.3728],[173,23.3195],[174,22.9181],[175,22.1611],[177,20.0299],[178,19.0362],[179,18.1137],[180,17.2973],[181,16.5974],[182,15.9882],[183,15.4759],[184,15.0757],[185,14.7782],[186,14.5807],[187,14.4459],[188,14.3817],[189,14.4712],[190,14.7271],[191,15.1478],[192,15.7319],[193,16.4509],[194,17.2458],[196,18.9687],[197,19.7122],[198,20.262],[199,20.6715],[200,20.9632],[203,21.6056],[204,21.7485],[205,21.7594],[206,21.5678],[207,21.0631],[209,19.5877],[210,18.9068],[212,17.4448],[213,16.7493],[214,16.1394],[215,15.6555],[216,15.2804],[217,15.0118],[218,14.8376],[219,14.6975],[221,14.5433],[223,14.4949],[224,14.5221],[225,14.601],[226,14.7348],[227,14.9128],[228,15.1292],[230,15.7311],[231,15.9826],[232,16.134],[233,16.1224],[234,15.7913],[235,14.8807],[236,13.364],[237,11.4317],[239,7.1387],[240,5.0856],[241,3.0139],[242,1.0274],[243,-0.7273],[244,-2.1277],[245,-3.2445],[246,-4.1098],[247,-4.6922],[248,-4.9671],[249,-4.948],[250,-4.6699],[251,-4.181],[252,-3.5136],[253,-2.6601],[254,-1.6361],[256,0.5801],[257,1.6832],[258,2.8238],[259,3.944],[260,4.9751],[261,5.9062],[262,6.7962],[263,7.6519],[264,8.4481],[265,9.1695],[266,9.8006],[267,10.3584],[268,10.8441],[269,11.2647],[270,11.6267],[271,11.9255],[272,12.1555],[273,12.3276],[274,12.4639],[276,12.6472],[279,12.7629],[281,12.7051],[283,12.5615],[284,12.4527],[286,12.1639],[289,11.6246],[290,11.4254],[291,11.1833],[294,10.3375],[295,10.0803],[298,9.3804],[299,9.1762],[300,9.0035],[301,8.8586],[304,8.5313],[305,8.4382],[306,0]]
       }
       head_ik_ctrl.Rotation.Z {
        type=MovieSceneScriptingFloatChannel
        default=0
        thinned_from=307
        keys [[0,0.0814],[1,0.0866],[2,0.1215],[3,0.1697],[4,0.2366],[5,0.413],[6,0.7846],[7,1.4406],[8,2.4651],[9,3.9613],[10,6.008],[11,8.5425],[12,11.3055],[13,14.133],[14,16.8108],[15,19.1511],[16,21.0921],[17,22.5838],[18,23.5137],[19,23.9586],[20,24.0193],[21,23.7905],[22,23.3799],[27,20.9092],[29,20.0263],[31,19.2404],[32,18.9029],[33,18.6139],[34,18.3713],[36,17.9977],[39,17.5608],[41,17.3758],[44,17.2155],[48,17.1611],[51,17.2172],[54,17.38],[56,17.5994],[57,17.7473],[58,17.9918],[59,18.2901],[62,19.2885],[64,20.0416],[66,20.9164],[67,21.3921],[68,21.9951],[69,22.775],[71,24.4688],[72,25.1373],[73,25.5015],[74,25.6446],[75,25.4093],[76,24.6236],[77,23.4467],[78,22.0558],[79,20.5071],[81,17.1809],[82,15.566],[83,14.0426],[84,12.6467],[85,11.3878],[86,10.275],[87,9.6583],[88,9.6489],[89,9.8125],[90,10],[92,10.2889],[95,10.6693],[97,10.8981],[98,11.0418],[99,11.2523],[100,11.5652],[102,12.4008],[103,12.7873],[104,13.1223],[105,13.4161],[106,13.683],[107,13.9152],[109,14.2942],[111,14.5698],[113,14.7541],[116,14.8843],[117,14.8876],[118,14.8472],[119,14.7565],[120,14.5677],[121,14.3187],[123,13.7247],[124,13.3807],[125,13.0057],[126,12.5652],[128,11.6053],[129,11.147],[130,10.7243],[131,10.3708],[133,9.7832],[135,9.2906],[136,9.1315],[138,8.8756],[139,8.7611],[140,8.7178],[145,8.5508],[147,8.4407],[149,8.2642],[150,8.1531],[151,8.0039],[152,7.7478],[153,7.3052],[154,6.6194],[155,5.7312],[156,4.6549],[157,3.4138],[158,1.8817],[159,0.1298],[160,-1.5905],[161,-3.2636],[162,-4.905],[163,-6.5137],[164,-8.0746],[165,-9.5988],[166,-11.067],[167,-12.4886],[168,-13.8635],[169,-15.2054],[170,-16.4915],[171,-17.7058],[173,-19.991],[174,-20.9673],[175,-21.6811],[176,-22.1817],[177,-22.446],[178,-22.5297],[179,-22.5007],[181,-22.2959],[184,-21.8959],[186,-21.7241],[188,-21.6979],[190,-21.78],[191,-21.8857],[192,-22.0594],[193,-22.3283],[194,-22.7204],[195,-23.2248],[196,-23.8206],[197,-24.5389],[198,-25.3788],[202,-29.0643],[209,-35.9535],[210,-36.8114],[213,-38.75],[214,-39.2485],[216,-39.8986],[219,-40.5026],[221,-40.6837],[231,-42.1433],[232,-42.2145],[233,-42.1948],[234,-41.4643],[235,-39.392],[236,-35.9586],[237,-31.7943],[238,-27.3815],[239,-23.1943],[240,-19.4248],[241,-15.8493],[242,-12.5418],[243,-9.6737],[244,-7.4307],[245,-5.8833],[246,-4.8067],[247,-3.9222],[248,-2.9548],[249,-1.9185],[250,-0.9673],[251,-0.1552],[252,0.457],[253,0.9505],[254,1.4333],[255,1.8927],[256,2.3039],[257,2.6614],[258,2.9869],[259,3.2649],[260,3.4757],[261,3.5992],[262,3.6244],[263,3.5892],[264,3.5266],[265,3.4771],[267,3.4304],[270,3.3419],[271,3.333],[272,3.3422],[273,3.3744],[274,3.4488],[276,3.5472],[277,3.5719],[278,3.6049],[280,3.6292],[281,3.6256],[282,3.6104],[283,3.571],[284,3.5067],[285,3.4195],[286,3.3136],[287,3.1899],[288,3.0367],[289,2.866],[290,2.6764],[291,2.4659],[292,2.2117],[293,1.9281],[294,1.6496],[295,1.4063],[296,1.1807],[297,0.9582],[298,0.7474],[299,0.5653],[300,0.4252],[301,0.3218],[302,0.2712],[303,0.1911],[304,0.1554],[305,0.1292],[306,0]]
       }
       head_ik_ctrl.Scale.X {type=MovieSceneScriptingFloatChannel, default=1, const [1,307]}
       head_ik_ctrl.Scale.Y {type=MovieSceneScriptingFloatChannel, default=1, const [1,307]}
       head_ik_ctrl.Scale.Z {type=MovieSceneScriptingFloatChannel, default=1, const [1,307]}
       chest_ctrl.Location.X {
        type=MovieSceneScriptingFloatChannel
        default=0
        thinned_from=307
        keys [[0,-0.1693],[6,-0.1693],[7,-0.1714],[8,-0.1782],[9,-0.1824],[10,-0.1854],[11,-0.1901],[12,-0.1994],[13,-0.2093],[14,-0.2141],[15,-0.2229],[16,-0.2273],[17,-0.2301],[19,-0.2377],[20,-0.2438],[21,-0.2543],[22,-0.2622],[23,-0.2688],[24,-0.2727],[26,-0.2747],[27,-0.2734],[29,-0.2684],[30,-0.2623],[31,-0.2603],[33,-0.253],[34,-0.2498],[35,-0.2491],[36,-0.2479],[37,-0.2439],[38,-0.243],[44,-0.2436],[50,-0.244],[51,-0.2445],[52,-0.2445],[53,-0.2455],[54,-0.2496],[55,-0.253],[56,-0.2546],[57,-0.2608],[58,-0.2635],[59,-0.2711],[61,-0.2835],[63,-0.3011],[64,-0.3279],[65,-0.3529],[66,-0.3757],[67,-0.3951],[68,-0.412],[69,-0.4327],[70,-0.4482],[71,-0.4493],[72,-0.4477],[74,-0.4305],[75,-0.4104],[76,-0.3842],[77,-0.3533],[79,-0.2778],[80,-0.245],[81,-0.2222],[82,-0.1956],[83,-0.1773],[84,-0.1676],[85,-0.1514],[86,-0.1462],[87,-0.1419],[88,-0.1385],[89,-0.1364],[91,-0.1333],[93,-0.132],[96,-0.1309],[97,-0.1311],[98,-0.1344],[99,-0.1448],[100,-0.1533],[101,-0.1554],[102,-0.1571],[103,-0.1565],[104,-0.1556],[105,-0.1523],[106,-0.1504],[107,-0.1474],[108,-0.1416],[109,-0.1349],[110,-0.1286],[111,-0.1231],[113,-0.1164],[114,-0.1162],[115,-0.1135],[117,-0.1151],[118,-0.1167],[119,-0.118],[120,-0.122],[121,-0.1234],[122,-0.1324],[123,-0.1373],[124,-0.138],[126,-0.1391],[127,-0.139],[128,-0.1384],[129,-0.1366],[130,-0.1342],[131,-0.1283],[132,-0.1261],[133,-0.1211],[134,-0.1211],[135,-0.1191],[136,-0.1191],[137,-0.1174],[139,-0.1174],[140,-0.1169],[141,-0.1161],[143,-0.1175],[145,-0.1208],[146,-0.1228],[147,-0.127],[148,-0.1309],[149,-0.1423],[150,-0.1488],[151,-0.1672],[152,-0.1821],[153,-0.1977],[154,-0.2245],[155,-0.2523],[156,-0.2901],[157,-0.3337],[158,-0.3797],[159,-0.4124],[160,-0.4492],[161,-0.4788],[162,-0.5109],[163,-0.531],[164,-0.5457],[165,-0.5565],[166,-0.5634],[167,-0.5661],[168,-0.5619],[169,-0.5503],[170,-0.5331],[171,-0.5136],[172,-0.4953],[173,-0.4801],[174,-0.4664],[176,-0.4421],[177,-0.4307],[178,-0.4203],[179,-0.4122],[180,-0.4029],[181,-0.3974],[182,-0.3956],[183,-0.3924],[184,-0.3909],[186,-0.3915],[192,-0.4167],[193,-0.4178],[194,-0.4208],[195,-0.4253],[198,-0.4268],[200,-0.4234],[201,-0.42],[203,-0.4038],[204,-0.3846],[205,-0.3681],[206,-0.3354],[207,-0.3005],[208,-0.2616],[209,-0.2527],[210,-0.2595],[211,-0.2691],[212,-0.2701],[213,-0.2723],[214,-0.2756],[215,-0.2767],[216,-0.2787],[217,-0.2768],[218,-0.2805],[221,-0.2764],[222,-0.2803],[224,-0.2781],[225,-0.2765],[229,-0.2737],[230,-0.2737],[231,-0.272],[232,-0.2696],[233,-0.2633],[234,-0.2592],[235,-0.2519],[236,-0.2386],[237,-0.2268],[238,-0.2114],[239,-0.2003],[240,-0.1853],[241,-0.1697],[242,-0.1577],[243,-0.1489],[244,-0.1424],[245,-0.1383],[246,-0.1319],[248,-0.1283],[249,-0.1291],[251,-0.1281],[253,-0.13],[254,-0.1336],[256,-0.1416],[257,-0.1477],[258,-0.1511],[259,-0.1581],[260,-0.1626],[262,-0.1702],[264,-0.1767],[266,-0.1821],[267,-0.1862],[268,-0.1885],[269,-0.1924],[271,-0.1966],[273,-0.2025],[274,-0.2046],[275,-0.2104],[276,-0.214],[277,-0.214],[278,-0.2155],[279,-0.2205],[280,-0.2229],[281,-0.2233],[282,-0.2248],[283,-0.2243],[284,-0.2248],[286,-0.2215],[287,-0.221],[289,-0.2165],[291,-0.2149],[292,-0.2123],[293,-0.2078],[296,-0.1978],[297,-0.1899],[298,-0.1868],[299,-0.1777],[300,-0.1756],[301,-0.1726],[302,-0.1726],[303,-0.1704],[305,-0.1701],[306,0]]
       }
       chest_ctrl.Location.Y {
        type=MovieSceneScriptingFloatChannel
        default=0
        thinned_from=307
        keys [[0,3.383],[69,3.3751],[71,3.3491],[75,3.29],[80,3.2509],[120,3.2603],[147,3.2897],[150,3.3087],[155,3.3892],[158,3.4055],[167,3.3639],[196,3.35],[200,3.3744],[203,3.4024],[206,3.4119],[239,3.4068],[250,3.3877],[298,3.389],[305,3.3832],[306,0]]
       }
       chest_ctrl.Location.Z {
        type=MovieSceneScriptingFloatChannel
        default=0
        thinned_from=307
        keys [[0,0.1417],[8,0.1416],[19,0.1415],[20,0.1418],[24,0.1421],[25,0.1433],[33,0.1445],[37,0.1444],[39,0.145],[60,0.1445],[61,0.1434],[63,0.1432],[69,0.1414],[70,0.1388],[71,0.1397],[74,0.1344],[75,0.1353],[77,0.132],[78,0.1329],[79,0.1311],[82,0.1326],[98,0.1333],[102,0.133],[114,0.1336],[115,0.1342],[120,0.1338],[121,0.1342],[130,0.1338],[131,0.1354],[144,0.1356],[146,0.1355],[147,0.1363],[148,0.1365],[149,0.1372],[150,0.137],[151,0.1383],[155,0.142],[156,0.1413],[157,0.1421],[160,0.1409],[162,0.1406],[163,0.1397],[168,0.1387],[180,0.1389],[183,0.1395],[184,0.1382],[185,0.138],[189,0.1383],[194,0.1382],[195,0.1374],[196,0.1373],[198,0.1384],[200,0.1387],[201,0.1397],[202,0.14],[203,0.1411],[204,0.1417],[207,0.1423],[208,0.1415],[209,0.1419],[215,0.1413],[216,0.1427],[223,0.1424],[226,0.1425],[233,0.1425],[240,0.1429],[244,0.1426],[250,0.1427],[251,0.143],[258,0.1427],[260,0.1421],[262,0.1422],[263,0.142],[274,0.1418],[275,0.1409],[290,0.1403],[291,0.1418],[298,0.1422],[299,0.1416],[305,0.1416],[306,0]]
       }
       chest_ctrl.Rotation.X {
        type=MovieSceneScriptingFloatChannel
        default=0
        thinned_from=307
        keys [[0,19.358],[68,19.3849],[70,19.2813],[73,18.9098],[78,18.5301],[118,18.4739],[147,18.6979],[150,18.8405],[155,19.424],[157,19.5822],[159,19.6456],[175,19.3243],[183,19.0526],[196,19.1205],[199,19.2363],[203,19.4615],[206,19.4561],[216,18.8901],[232,18.7464],[243,18.9429],[251,19.3456],[305,19.3627],[306,0]]
       }
       chest_ctrl.Rotation.Y {
        type=MovieSceneScriptingFloatChannel
        default=0
        thinned_from=307
        keys [[0,0.4328],[4,0.4328],[5,0.4384],[6,0.4384],[7,0.4465],[8,0.4741],[9,0.4992],[10,0.5215],[12,0.5763],[13,0.5942],[15,0.6209],[16,0.6299],[17,0.6308],[18,0.6216],[19,0.5935],[20,0.5326],[21,0.4289],[22,0.2903],[23,0.149],[24,0.0304],[25,-0.0737],[26,-0.1678],[27,-0.251],[28,-0.3196],[29,-0.3752],[30,-0.4216],[31,-0.4482],[32,-0.485],[33,-0.5067],[34,-0.5375],[35,-0.5551],[36,-0.5662],[37,-0.5933],[38,-0.6127],[39,-0.6138],[40,-0.6331],[41,-0.6331],[42,-0.648],[43,-0.648],[44,-0.6544],[45,-0.6564],[50,-0.6564],[51,-0.6429],[52,-0.6429],[53,-0.6156],[54,-0.5942],[56,-0.5339],[57,-0.4829],[58,-0.4337],[59,-0.3762],[60,-0.3085],[61,-0.2312],[62,-0.1421],[63,-0.0246],[64,0.1451],[65,0.364],[66,0.6253],[67,0.9362],[68,1.2727],[69,1.6338],[70,1.913],[71,2.0715],[72,2.1572],[73,2.1558],[74,2.0825],[75,1.9523],[76,1.7746],[77,1.5449],[78,1.2929],[79,1.06],[80,0.8854],[81,0.7434],[82,0.6131],[83,0.5092],[84,0.4442],[85,0.4068],[86,0.3793],[87,0.3601],[88,0.349],[89,0.3434],[90,0.3428],[91,0.3459],[93,0.3575],[95,0.3712],[96,0.3803],[97,0.3967],[98,0.4407],[99,0.5256],[100,0.5994],[101,0.6352],[102,0.6526],[103,0.6635],[104,0.67],[105,0.6715],[106,0.6679],[107,0.6581],[108,0.6315],[109,0.5881],[110,0.542],[111,0.4992],[113,0.4473],[115,0.4275],[116,0.4212],[117,0.4209],[118,0.4259],[119,0.4443],[120,0.4723],[121,0.4927],[122,0.515],[123,0.5475],[124,0.5517],[125,0.5716],[126,0.5765],[127,0.5925],[128,0.6013],[129,0.6033],[130,0.5952],[132,0.5381],[133,0.5168],[134,0.5168],[135,0.5079],[136,0.5079],[137,0.4959],[139,0.4959],[140,0.483],[141,0.4712],[144,0.45],[145,0.4448],[146,0.4423],[147,0.4469],[148,0.4614],[149,0.487],[150,0.5245],[151,0.5662],[152,0.6022],[153,0.6423],[154,0.6693],[155,0.6734],[156,0.647],[157,0.6081],[158,0.5984],[159,0.6237],[160,0.7133],[162,0.9529],[163,1.0493],[164,1.1214],[165,1.1817],[166,1.2307],[167,1.2744],[168,1.3502],[169,1.45],[170,1.5588],[171,1.681],[173,1.961],[174,2.0953],[177,2.4745],[178,2.5932],[179,2.697],[180,2.7779],[181,2.8421],[182,2.8914],[183,2.9269],[184,2.9404],[185,2.9243],[186,2.886],[187,2.8286],[190,2.6245],[191,2.5604],[192,2.5017],[193,2.4499],[194,2.4052],[195,2.3679],[196,2.3372],[197,2.3127],[198,2.293],[199,2.2779],[200,2.2684],[201,2.2704],[202,2.2859],[203,2.3163],[205,2.3915],[206,2.4134],[207,2.381],[208,2.3996],[209,2.6131],[210,2.9607],[211,3.302],[212,3.6021],[213,3.885],[214,4.1269],[215,4.3176],[216,4.4713],[217,4.5909],[218,4.6842],[219,4.7578],[220,4.8175],[222,4.9052],[224,4.9601],[227,5.0024],[230,5.0133],[232,5.001],[234,4.9428],[236,4.8159],[237,4.7266],[238,4.6158],[239,4.4847],[240,4.3315],[241,4.1588],[242,3.9607],[243,3.7295],[244,3.4649],[245,3.1682],[247,2.5319],[250,1.5921],[251,1.3013],[252,1.0554],[253,0.853],[254,0.6777],[255,0.5272],[256,0.4006],[257,0.2949],[258,0.2084],[259,0.1361],[260,0.0743],[261,0.0215],[262,-0.0218],[263,-0.0563],[264,-0.0826],[265,-0.1033],[266,-0.1176],[267,-0.1268],[268,-0.1311],[269,-0.1309],[270,-0.1267],[272,-0.1117],[273,-0.1047],[274,-0.0838],[275,-0.0715],[276,-0.0481],[277,-0.0481],[278,-0.0256],[279,0.0201],[280,0.0476],[281,0.0784],[282,0.1115],[283,0.1477],[284,0.1852],[285,0.2216],[287,0.2985],[288,0.3345],[289,0.3692],[290,0.4015],[291,0.4309],[292,0.4587],[293,0.4831],[294,0.503],[295,0.5142],[296,0.5152],[297,0.5047],[299,0.4754],[300,0.4642],[301,0.4561],[302,0.4561],[303,0.4416],[304,0.4402],[305,0.4402],[306,0]]
       }
       chest_ctrl.Rotation.Z {
        type=MovieSceneScriptingFloatChannel
        default=0
        thinned_from=307
        keys [[0,0.0934],[4,0.0934],[5,0.0957],[6,0.0957],[7,0.1038],[8,0.1414],[9,0.1865],[10,0.2351],[11,0.3031],[12,0.3989],[13,0.5308],[14,0.6706],[15,0.7771],[16,0.8633],[17,0.952],[18,1.0428],[19,1.226],[20,1.5751],[21,2.1352],[22,2.8413],[23,3.5189],[24,4.0317],[25,4.4313],[26,4.7573],[27,5.0034],[28,5.1648],[29,5.266],[30,5.3152],[31,5.3243],[34,5.2967],[36,5.3265],[37,5.3654],[39,5.4141],[40,5.4614],[41,5.4614],[42,5.51],[43,5.51],[45,5.5377],[50,5.5377],[52,5.528],[55,5.4606],[56,5.4283],[58,5.336],[59,5.275],[60,5.1974],[61,5.1007],[63,4.8633],[64,4.7561],[65,4.5876],[66,4.288],[67,3.7155],[68,2.8968],[69,2.1127],[70,1.496],[71,1.0073],[72,0.6805],[73,0.5956],[74,0.6717],[75,0.8004],[76,0.8788],[77,0.897],[78,0.8576],[79,0.7413],[80,0.5361],[81,0.3242],[82,0.1433],[83,-0.0088],[84,-0.137],[85,-0.2495],[86,-0.3469],[87,-0.435],[88,-0.5106],[89,-0.5784],[90,-0.6388],[91,-0.6918],[92,-0.7413],[94,-0.8263],[95,-0.866],[97,-0.9511],[98,-0.9915],[99,-1.0257],[100,-1.0653],[101,-1.1146],[102,-1.168],[104,-1.2844],[105,-1.347],[106,-1.4132],[107,-1.4867],[108,-1.5558],[109,-1.6108],[110,-1.6382],[111,-1.6391],[112,-1.6302],[113,-1.6114],[114,-1.5794],[116,-1.4843],[117,-1.4454],[118,-1.4242],[120,-1.3644],[121,-1.3516],[122,-1.3431],[123,-1.3284],[124,-1.3299],[125,-1.3416],[126,-1.3614],[127,-1.4174],[128,-1.4687],[129,-1.5385],[130,-1.5978],[131,-1.6348],[133,-1.6868],[134,-1.6868],[135,-1.6995],[136,-1.6995],[137,-1.7093],[139,-1.7093],[140,-1.6828],[142,-1.592],[143,-1.5318],[144,-1.4624],[145,-1.3782],[146,-1.2884],[147,-1.1829],[148,-1.0695],[149,-0.9377],[150,-0.7933],[151,-0.5839],[152,-0.2586],[153,0.1374],[154,0.7059],[155,1.5204],[156,2.6592],[157,3.9644],[158,5.1248],[159,6.0581],[160,6.6677],[161,7.1301],[162,7.4789],[163,7.7126],[164,7.8558],[165,7.9351],[166,7.9496],[167,7.8679],[168,7.4657],[169,6.7578],[170,5.8598],[171,4.8483],[172,3.8037],[173,2.8354],[174,1.9339],[175,1.1001],[176,0.2759],[177,-0.5425],[178,-1.3028],[179,-1.9693],[180,-2.4944],[181,-2.9151],[182,-3.2414],[183,-3.4817],[184,-3.5833],[185,-3.508],[186,-3.2949],[187,-2.9677],[188,-2.5891],[189,-2.1888],[190,-1.8035],[191,-1.438],[192,-1.0972],[193,-0.7953],[194,-0.5362],[195,-0.3241],[196,-0.1627],[197,-0.045],[198,0.026],[199,0.0542],[200,0.0239],[201,-0.1099],[202,-0.3625],[203,-0.752],[204,-1.2898],[205,-1.9819],[206,-2.8201],[207,-3.8009],[208,-4.8794],[211,-8.2526],[212,-9.3553],[213,-10.3809],[214,-11.3022],[215,-12.0583],[216,-12.6601],[217,-13.1101],[218,-13.458],[219,-13.7415],[220,-13.9828],[221,-14.1872],[222,-14.359],[224,-14.6181],[227,-14.8542],[236,-15.047],[237,-15.0445],[238,-15.0043],[239,-14.9073],[240,-14.7306],[241,-14.4471],[242,-14.0111],[243,-13.3949],[244,-12.5617],[245,-11.5251],[246,-10.3604],[247,-9.1588],[248,-7.9924],[249,-6.8019],[250,-5.5839],[251,-4.4389],[252,-3.4624],[253,-2.6316],[254,-1.8654],[255,-1.1633],[256,-0.5294],[257,0.0302],[258,0.511],[259,0.9273],[260,1.292],[261,1.6093],[262,1.8854],[263,2.1238],[264,2.3243],[265,2.4982],[266,2.6463],[267,2.7712],[268,2.8772],[269,2.9644],[270,3.0326],[271,3.088],[272,3.1184],[273,3.138],[274,3.1754],[276,3.1983],[278,3.1963],[279,3.1571],[280,3.1073],[281,3.0277],[282,2.9258],[283,2.7927],[284,2.6357],[285,2.4628],[286,2.2745],[287,2.0814],[288,1.8831],[289,1.6897],[290,1.5006],[291,1.3214],[292,1.1348],[293,0.9428],[294,0.7543],[295,0.5921],[296,0.4598],[297,0.3654],[298,0.2905],[299,0.2318],[300,0.1834],[301,0.1495],[302,0.1495],[303,0.1078],[304,0.1047],[305,0.1047],[306,0]]
       }
       chest_ctrl.Scale.X {type=MovieSceneScriptingFloatChannel, default=1, const [1,307]}
       chest_ctrl.Scale.Y {type=MovieSceneScriptingFloatChannel, default=1, const [1,307]}
       chest_ctrl.Scale.Z {type=MovieSceneScriptingFloatChannel, default=1, const [1,307]}
       chest_tan_ctrl.Location.X {
        type=MovieSceneScriptingFloatChannel
        default=0
        thinned_from=307
        keys [[0,0.0065],[4,0.0065],[5,0.0055],[6,0.0055],[7,0.0053],[8,0.008],[9,0.0057],[10,0.0037],[11,0.0002],[12,0.0035],[14,0.0005],[15,0.0052],[16,0.0035],[17,0.0029],[18,0.0041],[19,0.0055],[20,0.0086],[21,0.0097],[23,0.0211],[24,0.0256],[26,0.0325],[27,0.035],[28,0.0369],[29,0.0383],[30,0.0429],[31,0.0443],[32,0.044],[34,0.0456],[35,0.048],[36,0.0472],[37,0.0438],[38,0.0446],[39,0.0447],[40,0.0463],[41,0.0463],[42,0.0473],[43,0.0473],[44,0.048],[50,0.0482],[51,0.0486],[52,0.0486],[53,0.0463],[54,0.0489],[55,0.0451],[56,0.0415],[57,0.044],[58,0.0401],[59,0.0384],[60,0.0382],[61,0.0373],[62,0.0334],[63,0.0292],[64,0.0265],[65,0.0199],[66,0.0123],[67,0.0011],[68,-0.0117],[69,-0.0185],[70,-0.0295],[71,-0.0355],[72,-0.038],[73,-0.0402],[74,-0.0394],[75,-0.0345],[76,-0.0282],[77,-0.021],[78,-0.0143],[79,-0.0087],[80,-0.0039],[81,-0.003],[83,0.0057],[84,0.013],[85,-0.0006],[86,0.0028],[87,0.0055],[88,0.0075],[89,0.0081],[90,0.0084],[91,0.0092],[95,0.0077],[96,0.0071],[97,0.0059],[98,0.0017],[99,0.0033],[100,0.0032],[101,0.0006],[102,0.0022],[103,0.0015],[104,0.0011],[105,-0.0017],[106,-0.0012],[107,-0.0002],[108,-0.0009],[110,0.0009],[112,0.0021],[113,-0.0002],[114,0.0008],[115,0.0046],[116,0.0049],[117,0.0046],[118,0.0045],[120,-0.0009],[121,-0.0032],[122,0.0077],[123,0.005],[124,0.0051],[125,0.0025],[126,0.0023],[127,0.0012],[128,0.0007],[129,0.001],[130,0.0022],[131,0.0003],[132,0.0037],[133,0.0032],[136,0.003],[137,0.0024],[139,0.0024],[142,0.004],[143,0.0043],[144,0.004],[146,0.0041],[147,0.0046],[148,0.003],[149,0.0046],[150,0.0011],[151,0.0025],[152,0.0054],[153,-0.0021],[154,-0.001],[155,-0.0002],[156,0.003],[157,0.0038],[158,0.0097],[159,0.0081],[160,0.0109],[161,0.0007],[162,0.002],[163,-0.002],[164,-0.0038],[165,-0.0055],[167,-0.0082],[168,-0.0108],[169,-0.0144],[170,-0.0185],[171,-0.023],[173,-0.0332],[174,-0.038],[177,-0.0515],[178,-0.0558],[179,-0.0589],[180,-0.0645],[181,-0.0682],[182,-0.0666],[183,-0.0687],[184,-0.0694],[185,-0.0683],[186,-0.0704],[187,-0.0669],[188,-0.0629],[189,-0.0586],[190,-0.058],[191,-0.0541],[192,-0.0505],[193,-0.0519],[194,-0.0492],[195,-0.0474],[196,-0.0479],[198,-0.045],[199,-0.0462],[200,-0.0454],[201,-0.045],[202,-0.0483],[203,-0.0494],[204,-0.0472],[205,-0.0499],[206,-0.0506],[207,-0.0544],[208,-0.0541],[209,-0.0627],[210,-0.0729],[211,-0.0839],[212,-0.0956],[213,-0.1059],[214,-0.1164],[215,-0.1204],[216,-0.1217],[217,-0.13],[218,-0.1274],[219,-0.1326],[220,-0.1366],[221,-0.1399],[222,-0.1336],[223,-0.1356],[225,-0.1387],[226,-0.1395],[227,-0.1395],[228,-0.1402],[230,-0.1403],[231,-0.1396],[232,-0.1427],[233,-0.1391],[234,-0.1408],[235,-0.1363],[236,-0.1377],[237,-0.1329],[238,-0.1311],[239,-0.1251],[240,-0.1203],[241,-0.1184],[242,-0.1106],[243,-0.1017],[244,-0.0951],[245,-0.0865],[246,-0.072],[247,-0.0621],[248,-0.053],[249,-0.0396],[250,-0.0305],[251,-0.0232],[252,-0.0148],[253,-0.0095],[254,-0.0021],[255,0.0035],[256,0.0074],[257,0.014],[258,0.0151],[259,0.016],[260,0.0197],[261,0.0178],[262,0.0203],[263,0.0223],[264,0.0237],[265,0.0247],[266,0.0253],[267,0.0236],[268,0.0236],[269,0.0262],[270,0.0256],[271,0.0248],[272,0.0264],[273,0.0252],[274,0.0218],[275,0.0232],[276,0.0256],[277,0.0256],[278,0.0244],[279,0.0229],[280,0.0234],[281,0.0212],[282,0.0215],[283,0.0191],[284,0.0193],[286,0.0145],[287,0.0148],[289,0.0104],[290,0.0113],[292,0.0081],[293,0.0085],[294,0.0076],[295,0.0073],[296,0.0077],[297,0.0095],[298,0.0111],[299,0.0009],[300,0.0019],[301,0.0044],[302,0.0044],[303,0.0057],[305,0.0056],[306,0]]
       }
       chest_tan_ctrl.Location.Y {
        type=MovieSceneScriptingFloatChannel
        default=-0.0303
        thinned_from=307
        keys [[0,0.3904],[7,0.3901],[9,0.3909],[24,0.3897],[25,0.3892],[26,0.3905],[44,0.3907],[45,0.3895],[50,0.3895],[51,0.3905],[61,0.3897],[67,0.3904],[69,0.3898],[70,0.3877],[71,0.3834],[72,0.3798],[73,0.3785],[75,0.3715],[76,0.3706],[77,0.3681],[79,0.366],[80,0.3636],[81,0.3639],[83,0.3626],[84,0.3614],[85,0.3637],[93,0.3615],[97,0.3632],[98,0.3607],[99,0.3626],[101,0.3626],[102,0.3632],[104,0.3635],[105,0.3631],[110,0.3638],[111,0.3632],[114,0.364],[117,0.3679],[118,0.3679],[119,0.3709],[120,0.3726],[121,0.3634],[123,0.3664],[124,0.3663],[125,0.368],[139,0.3697],[140,0.3703],[141,0.3688],[143,0.3706],[144,0.3701],[145,0.3714],[146,0.3715],[147,0.3724],[148,0.3718],[149,0.3751],[150,0.3773],[151,0.3787],[154,0.3902],[155,0.3918],[156,0.3946],[157,0.3963],[158,0.3957],[159,0.3966],[160,0.3962],[161,0.3984],[162,0.3966],[178,0.3863],[179,0.385],[180,0.3855],[181,0.3838],[182,0.3852],[186,0.3822],[187,0.3824],[188,0.3849],[194,0.3846],[196,0.383],[197,0.3832],[198,0.386],[199,0.3865],[200,0.3884],[201,0.3918],[202,0.3909],[203,0.3947],[204,0.3953],[205,0.394],[208,0.3926],[211,0.3877],[214,0.3847],[215,0.3843],[217,0.3818],[218,0.3826],[221,0.3808],[222,0.3823],[232,0.3806],[235,0.3822],[244,0.3859],[249,0.3904],[250,0.3908],[251,0.3923],[258,0.3918],[259,0.3908],[260,0.3915],[262,0.3898],[263,0.3906],[266,0.3886],[267,0.3899],[278,0.3889],[279,0.3893],[280,0.3889],[281,0.3896],[282,0.3892],[283,0.3898],[284,0.3895],[285,0.3902],[286,0.39],[287,0.3905],[288,0.3902],[293,0.3912],[298,0.3896],[299,0.3926],[302,0.3918],[303,0.3907],[305,0.3908],[306,-0.0303]]
       }
       chest_tan_ctrl.Location.Z {
        type=MovieSceneScriptingFloatChannel
        default=0.265
        thinned_from=307
        keys [[0,0.2406],[21,0.2398],[24,0.2388],[25,0.2399],[37,0.2385],[50,0.2376],[51,0.2381],[60,0.2383],[69,0.2398],[70,0.2395],[73,0.2427],[76,0.245],[77,0.2451],[78,0.2464],[83,0.2472],[84,0.247],[85,0.2479],[97,0.2477],[98,0.2467],[99,0.2474],[114,0.246],[117,0.2475],[120,0.2488],[121,0.2447],[130,0.2463],[141,0.2454],[143,0.2458],[147,0.2451],[150,0.2445],[155,0.2401],[156,0.2401],[158,0.2391],[160,0.2393],[161,0.24],[162,0.2396],[179,0.24],[180,0.2409],[181,0.2406],[182,0.2411],[183,0.2409],[185,0.2417],[187,0.2418],[188,0.2425],[189,0.2424],[190,0.2429],[194,0.243],[195,0.2414],[197,0.2401],[198,0.2416],[200,0.2394],[201,0.2402],[202,0.2383],[203,0.2397],[205,0.2382],[206,0.2384],[207,0.2392],[210,0.239],[233,0.2365],[248,0.2398],[250,0.2397],[251,0.2404],[258,0.2403],[259,0.2401],[262,0.2401],[274,0.2405],[290,0.2397],[298,0.2397],[299,0.2413],[305,0.2406],[306,0.265]]
       }
       chest_tan_ctrl.Rotation.X {type=MovieSceneScriptingFloatChannel, default=0, thinned_from=307, keys [[0,11.145],[1,0],[306,0]]}
       chest_tan_ctrl.Rotation.Y {
        type=MovieSceneScriptingFloatChannel
        default=0
        thinned_from=307
        keys [[0,0.5913],[1,0.161],[6,0.161],[7,0.1653],[8,0.1702],[9,0.1563],[10,0.1387],[11,0.1291],[12,0.1164],[13,0.0982],[14,0.0782],[15,0.0522],[16,0.032],[17,0.021],[18,0.0158],[19,0.0363],[20,0.0727],[21,0.1197],[22,0.1579],[23,0.1728],[24,0.166],[25,0.1693],[26,0.1819],[27,0.1934],[28,0.2057],[29,0.2233],[30,0.2388],[31,0.2654],[32,0.2706],[33,0.3003],[34,0.3074],[35,0.3397],[36,0.3466],[38,0.3658],[39,0.3807],[40,0.3797],[41,0.3821],[42,0.388],[43,0.3857],[44,0.3857],[45,0.3919],[46,0.3887],[50,0.3887],[51,0.3911],[52,0.3815],[53,0.387],[54,0.3679],[55,0.3649],[56,0.3448],[57,0.3412],[58,0.3171],[59,0.2944],[60,0.2687],[61,0.2382],[62,0.2038],[63,0.1793],[64,0.1493],[65,0.0716],[66,0],[83,0],[84,0.0697],[85,0.1299],[86,0.1667],[87,0.1951],[88,0.2188],[89,0.2346],[90,0.2469],[91,0.2556],[92,0.2598],[93,0.2628],[95,0.2641],[96,0.2639],[97,0.2649],[98,0.2719],[99,0.2698],[100,0.2212],[101,0.1669],[102,0.1432],[103,0.1346],[104,0.1312],[105,0.1298],[106,0.1308],[107,0.1342],[108,0.1392],[109,0.1557],[110,0.1889],[111,0.2233],[112,0.2604],[113,0.2716],[114,0.2937],[115,0.2995],[116,0.3027],[118,0.302],[119,0.2981],[120,0.293],[121,0.2655],[122,0.2504],[123,0.2495],[124,0.2178],[125,0.2159],[126,0.199],[127,0.1948],[128,0.1891],[129,0.181],[130,0.1839],[131,0.1902],[132,0.2158],[133,0.2345],[134,0.2345],[135,0.2509],[136,0.2509],[137,0.2633],[139,0.2633],[140,0.2764],[141,0.2833],[142,0.2892],[143,0.292],[144,0.2935],[145,0.296],[146,0.2947],[147,0.2951],[148,0.29],[149,0.282],[150,0.2656],[151,0.2516],[152,0.2382],[153,0.2122],[154,0.1954],[155,0.1854],[156,0.1834],[157,0.1378],[158,0.0232],[159,0],[248,0],[249,0.0457],[250,0.1232],[251,0.1819],[252,0.2191],[253,0.2507],[254,0.2848],[255,0.3081],[256,0.3224],[257,0.3267],[258,0.3242],[260,0.3146],[261,0.3087],[262,0.3038],[263,0.2978],[265,0.2828],[266,0.2757],[268,0.259],[269,0.2495],[270,0.239],[271,0.2291],[272,0.2118],[273,0.2078],[274,0.2065],[275,0.1791],[276,0.1811],[277,0.1811],[278,0.1547],[279,0.1535],[280,0.1221],[281,0.1055],[282,0.0909],[283,0.0751],[284,0.0618],[285,0.0506],[286,0.0428],[288,0.0303],[289,0.0281],[290,0.0264],[291,0.0259],[292,0.0234],[293,0.0213],[294,0.0236],[295,0.0317],[296,0.0428],[297,0.0581],[298,0.0749],[299,0.0935],[300,0.1098],[301,0.1248],[302,0.1436],[303,0.1271],[304,0.1523],[305,0.1523],[306,0]]
       }
       chest_tan_ctrl.Rotation.Z {
        type=MovieSceneScriptingFloatChannel
        default=0
        thinned_from=307
        keys [[0,0.1279],[1,0],[73,0],[74,0.0986],[75,0.1008],[76,0.0136],[77,0],[84,0],[85,0.0051],[86,0.0487],[87,0.0851],[88,0.1204],[89,0.1493],[90,0.177],[91,0.2038],[92,0.2243],[93,0.2472],[94,0.2635],[95,0.2825],[96,0.2962],[97,0.3106],[98,0.3257],[99,0.3435],[100,0.3665],[101,0.3926],[102,0.4162],[103,0.4373],[106,0.503],[107,0.5236],[108,0.554],[109,0.5867],[110,0.618],[111,0.6361],[112,0.6306],[113,0.6288],[114,0.6226],[115,0.6225],[116,0.6023],[117,0.5783],[118,0.5596],[119,0.5464],[120,0.5385],[121,0.5217],[122,0.5169],[123,0.5202],[124,0.5137],[126,0.508],[127,0.4969],[128,0.5244],[129,0.5345],[130,0.5681],[131,0.6042],[132,0.6237],[133,0.6252],[134,0.6252],[135,0.6347],[136,0.6347],[137,0.644],[139,0.644],[140,0.6626],[141,0.6645],[142,0.6475],[143,0.6384],[144,0.6201],[145,0.6032],[146,0.5735],[147,0.5472],[148,0.5121],[149,0.4812],[150,0.4411],[151,0.4313],[152,0.4284],[153,0.3565],[154,0.3192],[155,0.2619],[156,0.1576],[157,0],[177,0],[178,0.0528],[179,0.4045],[180,0.749],[181,1.0158],[182,1.2358],[184,1.592],[185,1.7381],[186,1.7926],[187,1.7785],[188,1.681],[189,1.5462],[191,1.2177],[193,0.9022],[194,0.758],[195,0.6295],[196,0.5151],[197,0.4256],[198,0.3514],[199,0.2975],[200,0.2516],[201,0.1987],[202,0.1786],[203,0.1917],[204,0.2518],[205,0.3669],[206,0.5426],[207,0.7792],[208,1.0663],[209,1.404],[210,1.8148],[211,2.2782],[212,2.7624],[213,3.2571],[215,4.2177],[216,4.6164],[217,4.9532],[218,5.199],[219,5.38],[220,5.5214],[221,5.6422],[222,5.7446],[223,5.8308],[224,5.9034],[225,5.9641],[227,6.0471],[228,6.1008],[230,6.1047],[231,6.1198],[232,6.1525],[233,6.1432],[234,6.1614],[239,6.1912],[240,6.1871],[241,6.1663],[242,6.1298],[243,6.0513],[244,5.9241],[245,5.7047],[246,5.3601],[247,4.9096],[248,4.4022],[250,3.4788],[251,2.941],[252,2.374],[253,1.8882],[254,1.5129],[255,1.1647],[256,0.8427],[257,0.5453],[258,0.2761],[259,0.0485],[260,0],[305,0],[306,0.0472]]
       }
       chest_tan_ctrl.Scale.X {type=MovieSceneScriptingFloatChannel, default=1, const [1,307]}
       chest_tan_ctrl.Scale.Y {type=MovieSceneScriptingFloatChannel, default=1, const [1,307]}
       chest_tan_ctrl.Scale.Z {type=MovieSceneScriptingFloatChannel, default=1, const [1,307]}
      ]
     }]
   }
  ]
 }]
display_rate {denominator=1, numerator=30}
tick_resolution {denominator=1, numerator=24000}
tracks []
